using { /Fortnite.com/Devices }
using { /Verse.org/Simulation }

accolades_manager := class(creative_device):
    @editable 
    PeelyFoundAccolade : accolades_device = accolades_device{}
    
    @editable 
    GunLevelUp : accolades_device = accolades_device{}
    
    @editable 
    PlayerLevelUp : accolades_device = accolades_device{}
    
    @editable 
    QuestCompleted : accolades_device = accolades_device{}

    @editable 
    GuardKOAccolade : accolades_device = accolades_device{}

    @editable 
    CollectStashAccolade : accolades_device = accolades_device{}

    @editable 
    BossSlainAccolade : accolades_device = accolades_device{}

    @editable 
    ThankYou : accolades_device = accolades_device{}

    @editable 
    PleaseLikeAndRecommend : accolades_device = accolades_device{}

    @editable 
    XPCoin : accolades_device = accolades_device{}

    FoundPeelyAward(Agent:agent):void=
        PeelyFoundAccolade.Award(Agent)

    GuardKOAward(Agent:agent):void=
        GuardKOAccolade.Award(Agent)

    CollectStashAward(Agent:agent):void=
        CollectStashAccolade.Award(Agent)

    BossSlainAward(Agent:agent):void=
        BossSlainAccolade.Award(Agent)