using { /Fortnite.com/Devices }
using { /Verse.org/Simulation }

analytics_manager := class(creative_device):
    @editable
    big_rock_1 : analytics_device = analytics_device{}
    
    @editable
    big_rock_2 : analytics_device = analytics_device{}
    
    @editable
    big_rock_3 : analytics_device = analytics_device{}

    @editable
    candy_vendor : analytics_device = analytics_device{}

    @editable
    upgrade_hammer : analytics_device = analytics_device{}

    @editable
    peely_found_first : analytics_device = analytics_device{}

    @editable
    peely_found_ten : analytics_device = analytics_device{}
    
    @editable
    quest_complete_0_0 : analytics_device = analytics_device{}
    
    @editable
    quest_complete_0_1 : analytics_device = analytics_device{}
    
    @editable
    quest_complete_0_2 : analytics_device = analytics_device{}
    
    @editable
    quest_complete_0_3 : analytics_device = analytics_device{}
    
    @editable
    quest_complete_0_4 : analytics_device = analytics_device{}
    
    @editable
    quest_complete_0_5 : analytics_device = analytics_device{}
    
    @editable
    quest_complete_0_6 : analytics_device = analytics_device{}
    
    @editable
    quest_complete_0_7 : analytics_device = analytics_device{}
    
    @editable
    quest_complete_0_8 : analytics_device = analytics_device{}
    
    @editable
    quest_complete_0_9 : analytics_device = analytics_device{}
    
    @editable
    quest_complete_1_0 : analytics_device = analytics_device{}
    
    @editable
    quest_complete_1_1 : analytics_device = analytics_device{}
    
    @editable
    quest_complete_1_2 : analytics_device = analytics_device{}
    
    @editable
    quest_complete_1_3 : analytics_device = analytics_device{}
    
    @editable
    quest_complete_1_4 : analytics_device = analytics_device{}
    
    @editable
    quest_complete_1_5 : analytics_device = analytics_device{}
    
    @editable
    quest_complete_1_6 : analytics_device = analytics_device{}

    @editable
    quest_complete_1_7 : analytics_device = analytics_device{}
    
    @editable
    quest_complete_1_8 : analytics_device = analytics_device{}
    
    @editable
    quest_complete_1_9 : analytics_device = analytics_device{}

    @editable
    quest_complete_1_10 : analytics_device = analytics_device{}
    
    @editable
    quest_complete_1_11 : analytics_device = analytics_device{}
    
    @editable
    quest_complete_1_12 : analytics_device = analytics_device{}

    @editable
    quest_complete_1_13 : analytics_device = analytics_device{}
    
    @editable
    quest_complete_1_14 : analytics_device = analytics_device{}
    
    @editable
    quest_complete_1_15 : analytics_device = analytics_device{}
    
    @editable
    quest_complete_1_16 : analytics_device = analytics_device{}
    
    @editable
    quest_complete_1_17 : analytics_device = analytics_device{}

    @editable
    quest_complete_1_18 : analytics_device = analytics_device{}
    
    @editable
    quest_complete_1_19 : analytics_device = analytics_device{}

    @editable
    quest_complete_1_20 : analytics_device = analytics_device{}

    @editable
    quest_complete_1_21 : analytics_device = analytics_device{}

    @editable
    quest_complete_1_22 : analytics_device = analytics_device{}

    @editable
    quest_complete_1_23 : analytics_device = analytics_device{}

    @editable
    quest_complete_1_24 : analytics_device = analytics_device{}

    @editable
    quest_complete_1_25 : analytics_device = analytics_device{}

    @editable
    quest_complete_1_26 : analytics_device = analytics_device{}

    @editable
    quest_complete_2_0 : analytics_device = analytics_device{}

    @editable
    quest_complete_2_1 : analytics_device = analytics_device{}

    @editable
    quest_complete_2_2 : analytics_device = analytics_device{}

    @editable
    quest_complete_2_3 : analytics_device = analytics_device{}

    @editable
    quest_complete_2_4 : analytics_device = analytics_device{}

    @editable
    quest_complete_2_5 : analytics_device = analytics_device{}

    @editable
    quest_complete_2_6 : analytics_device = analytics_device{}

    @editable
    quest_complete_2_7 : analytics_device = analytics_device{}
    
    @editable
    quest_complete_2_8 : analytics_device = analytics_device{}
    
    @editable
    quest_complete_2_9 : analytics_device = analytics_device{}

    @editable
    quest_complete_2_10 : analytics_device = analytics_device{}

    @editable
    quest_complete_2_11 : analytics_device = analytics_device{}

    @editable
    quest_complete_2_12 : analytics_device = analytics_device{}

    @editable
    quest_complete_2_13 : analytics_device = analytics_device{}

    @editable
    quest_complete_2_14 : analytics_device = analytics_device{}

    @editable
    quest_complete_2_15 : analytics_device = analytics_device{}

    @editable
    quest_complete_2_16 : analytics_device = analytics_device{}

    @editable
    quest_complete_2_17 : analytics_device = analytics_device{}

    @editable
    quest_complete_2_18 : analytics_device = analytics_device{}

    @editable
    quest_complete_2_19 : analytics_device = analytics_device{}

    @editable
    quest_complete_2_20 : analytics_device = analytics_device{}

    @editable
    quest_complete_3_0 : analytics_device = analytics_device{}

    @editable
    quest_complete_3_1 : analytics_device = analytics_device{}

    @editable
    quest_complete_3_2 : analytics_device = analytics_device{}

    @editable
    quest_complete_3_3 : analytics_device = analytics_device{}

    @editable
    quest_complete_3_4 : analytics_device = analytics_device{}

    @editable
    quest_complete_3_5 : analytics_device = analytics_device{}

    @editable
    quest_complete_3_6 : analytics_device = analytics_device{}

    @editable
    quest_complete_3_7 : analytics_device = analytics_device{}
    
    @editable
    quest_complete_3_8 : analytics_device = analytics_device{}

    @editable
    quest_complete_3_9 : analytics_device = analytics_device{}

    @editable
    quest_complete_3_10 : analytics_device = analytics_device{}

    @editable
    quest_complete_3_11 : analytics_device = analytics_device{}

    @editable
    quest_complete_3_12 : analytics_device = analytics_device{}

    @editable
    quest_complete_3_13 : analytics_device = analytics_device{}

    @editable
    quest_complete_3_14 : analytics_device = analytics_device{}

    @editable
    quest_complete_3_15 : analytics_device = analytics_device{}

    @editable
    quest_complete_3_16 : analytics_device = analytics_device{}

    @editable
    quest_complete_3_20 : analytics_device = analytics_device{}

    @editable
    quest_complete_3_25 : analytics_device = analytics_device{}

    @editable
    quest_complete_4_0 : analytics_device = analytics_device{}

    @editable
    quest_complete_4_3 : analytics_device = analytics_device{}

    @editable
    quest_complete_4_6 : analytics_device = analytics_device{}

    @editable
    quest_complete_4_9 : analytics_device = analytics_device{}

    @editable
    quest_complete_4_12 : analytics_device = analytics_device{}

    @editable
    quest_complete_4_22 : analytics_device = analytics_device{}

    @editable
    quest_complete_4_29 : analytics_device = analytics_device{}

    @editable
    session_start_bad_spawn : analytics_device = analytics_device{}

    @editable
    session_start_d0_1 : analytics_device = analytics_device{}

    @editable
    session_start_d2_7 : analytics_device = analytics_device{}

    @editable
    session_start_d8_14 : analytics_device = analytics_device{}

    @editable
    session_start_d15_30 : analytics_device = analytics_device{}

    @editable
    session_start_d_gte_31 : analytics_device = analytics_device{}

    @editable
    session_start_new_player : analytics_device = analytics_device{}

    @editable
    session_start_old_player : analytics_device = analytics_device{}