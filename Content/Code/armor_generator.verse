
using { /Fortnite.com/Devices }
using { /Verse.org/Simulation }
using { /UnrealEngine.com/Temporary/Diagnostics }
using { /Verse.org/Random }
using { /Verse.org/Assets }

armor_data := class<persistable><concrete><final>:
    Slot : string = ""
    Rarity : string = ""

    Base : string = ""
    Prefix : string = ""
    Suffix : string = ""

    Potential : int = 0
    Stats : []string = array{}
    StatLevels : []int = array{}

    UID : int = 0

    ToArmor()<transacts>:armor=
        var StatValues : []player_stat = array{}
        for (Stat : Stats):
            set StatValues = StatValues + array{PlayerStatFromString(Stat)}
        Armor := armor{
            Slot := ArmorSlotFromString(Slot)
            Rarity := ArmorRarityFromString(Rarity)
            Base := ArmorBaseFromString(Base)
            Prefix := ArmorPrefixFromString(Prefix)
            Suffix := ArmorSuffixFromString(Suffix)
            Potential := Potential
            Stats := StatValues
            StatLevels := StatLevels
            UID := UID
        }

        return Armor

    WithUID(ID:int)<transacts>:armor_data=
        armor_data{
            Slot := Slot
            Rarity := Rarity
            Base := Base
            Prefix := Prefix
            Suffix := Suffix
            Potential := Potential
            Stats := Stats
            StatLevels := StatLevels
            UID := ID
        }

    WithUpgrade()<transacts> : armor_data =
        UpgradeID := GetRandomInt(0, Stats.Length - 1)

        var NewStatLevels : []int = array{}
        for (i := 0..StatLevels.Length - 1; StatLevel := StatLevels[i]):
            var Improvement : int = 1
            if:
                Temp:= Int[StatLevel * 0.1] 
                Temp > Improvement
            then:
                set Improvement = Temp 

            if (i = UpgradeID):
                set NewStatLevels = NewStatLevels + array{StatLevel + Improvement}
            else:
                set NewStatLevels = NewStatLevels + array{StatLevel}

        return armor_data{
            Slot := Slot
            Rarity := Rarity
            Base := Base
            Prefix := Prefix
            Suffix := Suffix
            Potential := Potential - 1
            Stats := Stats
            StatLevels := NewStatLevels
            UID := UID
        }

GetStatDiff(Armor : armor, Armor2 : armor) : string = 
    var Diff : string = ""
    for (i := 0..Armor.Stats.Length - 1; Stat := Armor.Stats[i]; Level := Armor.StatLevels[i]):
        for (j := 0..Armor2.Stats.Length - 1; Stat2 := Armor2.Stats[j]; Level2 := Armor2.StatLevels[j]):
            if:
                Stat = Stat2
                StatVal1 := Stat.ToArmorStatValue(Level)
                StatVal2 := Stat2.ToArmorStatValue(Level2)
                DiffVal := StatVal1 - StatVal2
                not BigNumber_GreaterThanOrEquals(big_number{}, DiffVal)?
            then:
                set Diff = Diff + "{Stat.ValToStatline(DiffVal)}"
    return Diff


armor := class<final><concrete>:
    @editable
    Slot: armor_slot = armor_slot.None
    @editable
    Rarity: armor_rarity = armor_rarity.None

    @editable
    Base: armor_base = armor_base.None
    @editable
    Prefix: armor_prefix = armor_prefix.None
    @editable
    Suffix: armor_suffix = armor_suffix.None

    @editable
    Potential: int = 0
    @editable
    Stats: []player_stat = array{}
    @editable
    StatLevels: []int = array{} 

    UID : int = 0

    GetName()<transacts>:string=
        return "{Prefix} {Base}" # {Suffix}

    GetFullName()<transacts>:string=
        return "{Rarity} {GetName()} (Lv{GetTotalLevel()})"
        
    GetTotalLevel()<transacts>:int=
        var Lvl : int = 0
        for (StatLevel : StatLevels):
            set Lvl += StatLevel
        return Lvl

    GetStatblock()<transacts>:string=
        var StatBlock : string = "{GetFullName()} | ID : {UID}"# | Potential: {Potential}"
        for (i := 0..Stats.Length - 1; Stat := Stats[i]; StatLevel := StatLevels[i]):
            set StatBlock = StatBlock + " | {Stat}: {StatLevel} [{Stat.ToArmorStatValue(StatLevel)}]"

        return StatBlock

    GetDescription()<transacts>:string=
        var Desc : string = ""
        for (i := 0..Stats.Length - 1; Stat := Stats[i]; StatLevel := StatLevels[i]):
            set Desc = Desc + "\<br\>{Stat.ToStatline(StatLevel)}"

        return Desc

    GetImage():texture=
        case (Slot):
            armor_slot.Head => Images.Helm
            armor_slot.Torso => Images.Body
            armor_slot.Legs => Images.Legs
            armor_slot.Boots => Images.Boots
            armor_slot.Gloves => Images.Glove
            armor_slot.Accessory => Images.Ring
            armor_slot.None => Images.QuestionMark

    GetImageKey():string= 
        case (Slot):
            armor_slot.Head => "Helmet"
            armor_slot.Torso => "Body"
            armor_slot.Legs => "Legs"
            armor_slot.Boots => "Boots"
            armor_slot.Gloves => "Glove"
            armor_slot.Accessory => "Ring"
            armor_slot.None => ""

    ToData()<transacts>:armor_data=
        var StatStrings : []string = array{}
        for (Stat : Stats):
            set StatStrings = StatStrings + array{ToString(Stat)}
        Data := armor_data{
            Slot := ToString(Slot)
            Rarity := ToString(Rarity)
            Base := ToString(Base)
            Prefix := ToString(Prefix)
            Suffix := ToString(Suffix)
            Potential := Potential
            Stats := StatStrings
            StatLevels := StatLevels
            UID := UID
        }

        return Data

    Equals(Other:armor)<decides><transacts>:void=
        Other.UID = UID

armor_gen_properties := class<concrete>:
    @editable
    MinLevel : int = 1
    @editable
    MaxLevel : int = 1
    @editable
    rarity_options : []armor_rarity = ArmorRaritiesList
    @editable
    slot_options : []armor_slot = ArmorSlotsList

armor_generator := class():

    RandomSlot():armor_slot=
        if (S := ArmorSlotsList[GetRandomInt(0, ArmorSlotsList.Length - 1)]):
            return S
        
        return armor_slot.None

    RandomRarity():armor_rarity=
        Selector := GetRandomFloat(0.0, 1.0)
        if (Selector < 0.20):               #20%
            return armor_rarity.Common
        else if (Selector < 0.75):          #55%
            return armor_rarity.Uncommon
        else if (Selector < 0.95):         #20%
            return armor_rarity.Rare
        else if (Selector < 0.99):          #4%
            return armor_rarity.Epic
        else:                               #1%
            return armor_rarity.Legendary

    GenArmor(GenProps:armor_gen_properties):armor=
        if:
            Opt1 := GenProps.slot_options[GetRandomInt(0, GenProps.slot_options.Length - 1)]
            Opt2 := GenProps.rarity_options[GetRandomInt(0, GenProps.rarity_options.Length - 1)]
        then:
            return GenArmor(Opt1, Opt2, GetRandomInt(GenProps.MinLevel, GenProps.MaxLevel))

        return armor{}

    GenArmor(MinLevel:int, MaxLevel:int):armor=
        return GenArmor(GetRandomInt(MinLevel, MaxLevel))

    GenArmor(MinLevel:int, MaxLevel:int, Rarity:armor_rarity):armor=
        return GenArmor(RandomSlot(), Rarity, GetRandomInt(MinLevel, MaxLevel))

    GenArmor(Level:int):armor=
        return GenArmor(RandomSlot(), RandomRarity(), Level)

    GenArmor(InSlot:armor_slot, Rarity:armor_rarity, Level:int):armor=
        return GenArmor(InSlot, Rarity, Level, ArmorPrefixesList, ArmorSuffixesList, ArmorStatsList)

    GenArmor(InSlot:armor_slot, Rarity:armor_rarity, Level:int, Prefixes:[]armor_prefix, Suffixes:[]armor_suffix, SecondaryStatOptions:[]player_stat):armor=
        SlotOptions := InSlot.Options()
        if:
            Base := SlotOptions[GetRandomInt(0, SlotOptions.Length - 1)]
            Prefix := Prefixes[GetRandomInt(0, Prefixes.Length - 1)]
            Suffix := Suffixes[GetRandomInt(0, Suffixes.Length - 1)]
            RarityAdjustedLevel := Ceil[Level * Rarity.StatMult()]
            PotentialValue := GetRandomInt(Floor(RarityAdjustedLevel / 2), RarityAdjustedLevel)
            SecondaryMin := Floor[RarityAdjustedLevel * 0.1] + 1
            SecondaryMax := Floor[RarityAdjustedLevel * 0.2] + 1
        then:

            var Stats : []player_stat = array{Suffix.ToStat()}
            var StatLevels : []int = array{RarityAdjustedLevel}

            var Breakout : int = 0
            loop:
                if (Stats.Length >= Rarity.SecondaryCount() + 1):
                    break
                if:
                    Stat := SecondaryStatOptions[GetRandomInt(0, SecondaryStatOptions.Length - 1)]
                    not Stats.Find[Stat]
                    Val := GetRandomInt(SecondaryMin, SecondaryMax)
                    set Stats = Stats + array{Stat}
                    set StatLevels = StatLevels + array{Val}
                    NewBaseStatLevel := StatLevels[0] - Val
                    set StatLevels[0] = NewBaseStatLevel

                set Breakout = Breakout + 1
                if (Breakout > 50):
                    break
                

            Armor := armor{
                Slot := InSlot
                Rarity := Rarity

                Base := Base
                Prefix := Prefix
                Suffix := Suffix

                Potential := PotentialValue
                Stats := Stats
                StatLevels := StatLevels
            }

            return Armor
        
        #Should never happen
        Log("Armor Generator Failed -- this should never happen")
        return armor{}

    MidasSecondaryStats : []player_stat = array{player_stat.MaxHealth, player_stat.Regen, player_stat.Damage_Beast, 
    player_stat.Damage_Guard, player_stat.Damage_Mineral, player_stat.Damage_Boss, player_stat.Damage_With_Pickaxe,
    player_stat.Damage_With_Pistol, player_stat.Damage_With_AR, player_stat.Damage_With_Shotgun, player_stat.Damage_With_Railgun,
    player_stat.Damage_With_Bow, player_stat.Damage_With_Minigun, player_stat.Passive_Income_Flat, player_stat.Passive_Income_Percent, player_stat.Crit_Mult}

    GenArmorForBoss(Boss:boss_enum, MinLevel:int, MaxLevel:int, Rarity:armor_rarity):armor=
        return case (Boss):
            boss_enum.LyraTutorial =>
                GenArmor(RandomSlot(), Rarity, GetRandomInt(MinLevel, MaxLevel), ArmorPrefixesList, array{armor_suffix.OfMidas}, MidasSecondaryStats)
            _ => GenArmor(MinLevel, MaxLevel, Rarity)
