using { /Verse.org/Colors }

player_stat := enum:
    None
    MaxHealth
    Regen
    DamageMult
    Damage_Beast
    Damage_Guard
    Damage_Mineral
    Damage_Boss
    Damage_With_Pickaxe
    Damage_With_Pistol
    Damage_With_AR
    Damage_With_Shotgun
    Damage_With_Railgun
    Damage_With_Bow
    Empowered_Mining
    Passive_Income_Flat
    Passive_Income_Percent
    MoveSpeed #doesn't work
    Damage_With_Hand_Cannon
    Damage_With_Minigun
    Crit_Mult
    Damage_With_Stake_Gun
    Damage_With_Scythe
    Damage_Melee
    Damage_Rapid
    Damage_Precision


armor_slot := enum:
    None
    Head
    Torso
    Legs
    Boots
    Gloves
    Accessory

armor_rarity := enum:
    None
    Common
    Uncommon
    Rare
    Legendary
    Epic
    Mythic
    Godly
    Exalted
    Ultimate
    Supreme
    Transcendent
    Infinite

armor_base := enum:
    None
    #Helm
    Helmet
    Hood
    Mask
    Cap
    Crown
    Circlet
    Headdress
    Visor
    Tiara
    Cowl
    WizardHat
    Bandana
    Headband
    Beret
    Beanie
    #Torso
    Chestplate
    Robe
    Jacket
    Tunic
    Coat
    Vest
    Chainmail
    Jerkin
    Hauberk
    Tabard
    Kimono
    Hoodie
    #Legs
    Greaves
    Pants
    Shorts
    Kilt
    Chaps
    Jeans
    Trousers
    Pantaloons
    #Boots
    Boots
    Sabatons
    Shoes
    Slippers
    Sandals
    Footwraps
    #Gloves
    Gauntlets
    Bracers
    Handguards
    Handwraps
    Mittens
    Armguards
    #Accessory
    Belt
    Ring
    Necklace
    Bracelet
    Amulet
    Talisman
    Brooch
    Earrings
    Pendant
    Medal
    Pin
    Charm
    Medallion
    Wristband
    Locket
    Bangle

armor_prefix := enum:
    None
    Savage
    Valiant
    Arcane
    Mystic
    Infernal
    Stalwart
    Celestial
    Shadow
    Sacred
    Radiant
    Warlords
    Guardians
    Ember
    Frost
    Thunder
    Titan
    Ethereal
    Phantom
    Dawn
    Lyras
    BeachInvader
    Mining

armor_suffix := enum:
    None
    OfThePhoenix
    OfTheTitan
    OfTheWolf
    OfTheDragon
    OfTheSerpent
    OfTheGriffin
    OfTheLion
    OfTheEagle
    OfTheBear
    OfTheStorm
    OfTheTempest
    OfNight
    OfEarth
    OfTheStars
    OfVoid
    OfTheInferno
    OfTheGlacier
    OfThunder
    OfTheBlaze
    OfWind
    OfMidas

(Slot:armor_slot).Options():[]armor_base =
    case (Slot):
        armor_slot.Head => array{
            armor_base.Helmet, armor_base.Hood, armor_base.Mask, armor_base.Cap, armor_base.Crown, armor_base.Circlet, armor_base.Headdress, 
            armor_base.Visor, armor_base.Tiara, armor_base.Cowl, armor_base.WizardHat, armor_base.Bandana, armor_base.Headband, 
            armor_base.Beret, armor_base.Beanie
        }
        armor_slot.Torso => array{
            armor_base.Chestplate, armor_base.Robe, armor_base.Jacket, armor_base.Tunic, armor_base.Coat, armor_base.Vest, armor_base.Chainmail, 
            armor_base.Jerkin, armor_base.Hauberk, armor_base.Tabard, armor_base.Kimono, armor_base.Hoodie
        }
        armor_slot.Legs => array{
            armor_base.Greaves, armor_base.Pants, armor_base.Shorts, armor_base.Kilt, armor_base.Chaps, armor_base.Jeans, armor_base.Trousers, 
            armor_base.Pantaloons
        }
        armor_slot.Boots => array{
            armor_base.Boots, armor_base.Sabatons, armor_base.Shoes, armor_base.Slippers, armor_base.Sandals, armor_base.Footwraps
        }
        armor_slot.Gloves => array{
            armor_base.Gauntlets, armor_base.Bracers, armor_base.Handguards, armor_base.Handwraps, armor_base.Mittens, 
            armor_base.Armguards
        }
        armor_slot.Accessory => array{
            armor_base.Belt, armor_base.Ring, armor_base.Necklace, armor_base.Bracelet, armor_base.Amulet, armor_base.Talisman, 
            armor_base.Brooch, armor_base.Earrings, armor_base.Pendant, armor_base.Medal, armor_base.Pin, armor_base.Charm, 
            armor_base.Medallion, armor_base.Wristband, armor_base.Locket, armor_base.Bangle
        }
        _ => array{}

(Base:armor_base).GetSlot():armor_slot =
    case (Base):
        armor_base.Helmet => armor_slot.Head
        armor_base.Hood => armor_slot.Head
        armor_base.Mask => armor_slot.Head
        armor_base.Cap => armor_slot.Head
        armor_base.Crown => armor_slot.Head
        armor_base.Circlet => armor_slot.Head
        armor_base.Headdress => armor_slot.Head
        armor_base.Visor => armor_slot.Head
        armor_base.Tiara => armor_slot.Head
        armor_base.Cowl => armor_slot.Head
        armor_base.WizardHat => armor_slot.Head
        armor_base.Bandana => armor_slot.Head
        armor_base.Headband => armor_slot.Head
        armor_base.Beret => armor_slot.Head
        armor_base.Beanie => armor_slot.Head
        armor_base.Chestplate => armor_slot.Torso
        armor_base.Robe => armor_slot.Torso
        armor_base.Jacket => armor_slot.Torso
        armor_base.Tunic => armor_slot.Torso
        armor_base.Coat => armor_slot.Torso
        armor_base.Vest => armor_slot.Torso
        armor_base.Chainmail => armor_slot.Torso
        armor_base.Jerkin => armor_slot.Torso
        armor_base.Hauberk => armor_slot.Torso
        armor_base.Tabard => armor_slot.Torso
        armor_base.Kimono => armor_slot.Torso
        armor_base.Hoodie => armor_slot.Torso
        armor_base.Greaves => armor_slot.Legs
        armor_base.Pants => armor_slot.Legs
        armor_base.Shorts => armor_slot.Legs
        armor_base.Kilt => armor_slot.Legs
        armor_base.Chaps => armor_slot.Legs
        armor_base.Jeans => armor_slot.Legs
        armor_base.Trousers => armor_slot.Legs
        armor_base.Pantaloons => armor_slot.Legs
        armor_base.Boots => armor_slot.Boots
        armor_base.Sabatons => armor_slot.Boots
        armor_base.Shoes => armor_slot.Boots
        armor_base.Slippers => armor_slot.Boots
        armor_base.Sandals => armor_slot.Boots
        armor_base.Footwraps => armor_slot.Boots
        armor_base.Gauntlets => armor_slot.Gloves
        armor_base.Bracers => armor_slot.Gloves
        armor_base.Handguards => armor_slot.Gloves
        armor_base.Handwraps => armor_slot.Gloves
        armor_base.Mittens => armor_slot.Gloves
        armor_base.Armguards => armor_slot.Gloves
        armor_base.Belt => armor_slot.Accessory
        armor_base.Ring => armor_slot.Accessory
        armor_base.Necklace => armor_slot.Accessory
        armor_base.Bracelet => armor_slot.Accessory
        armor_base.Amulet => armor_slot.Accessory
        armor_base.Talisman => armor_slot.Accessory
        armor_base.Brooch => armor_slot.Accessory
        armor_base.Earrings => armor_slot.Accessory
        armor_base.Pendant => armor_slot.Accessory
        armor_base.Medal => armor_slot.Accessory
        armor_base.Pin => armor_slot.Accessory
        armor_base.Charm => armor_slot.Accessory
        armor_base.Medallion => armor_slot.Accessory
        armor_base.Wristband => armor_slot.Accessory
        armor_base.Locket => armor_slot.Accessory
        armor_base.Bangle => armor_slot.Accessory
        _ => armor_slot.None
 
(Suffix:armor_suffix).ToStat():player_stat = 
    case (Suffix):    
        armor_suffix.OfThePhoenix => player_stat.DamageMult
        armor_suffix.OfTheTitan => player_stat.MaxHealth
        armor_suffix.OfTheWolf => player_stat.DamageMult #player_stat.MoveSpeed
        armor_suffix.OfTheDragon => player_stat.Regen
        armor_suffix.OfTheSerpent => player_stat.Passive_Income_Flat
        armor_suffix.OfTheGriffin => player_stat.Damage_Beast
        armor_suffix.OfTheLion => player_stat.Damage_Guard
        armor_suffix.OfTheEagle => player_stat.Damage_Mineral
        armor_suffix.OfTheBear => player_stat.MaxHealth
        armor_suffix.OfTheStorm => player_stat.Empowered_Mining
        armor_suffix.OfTheTempest => player_stat.Damage_Boss
        armor_suffix.OfNight => player_stat.Damage_Guard
        armor_suffix.OfEarth => player_stat.MaxHealth
        armor_suffix.OfTheStars => player_stat.Passive_Income_Flat
        armor_suffix.OfVoid => player_stat.DamageMult
        armor_suffix.OfTheInferno => player_stat.Damage_Beast
        armor_suffix.OfTheGlacier => player_stat.Regen
        armor_suffix.OfThunder => player_stat.DamageMult
        armor_suffix.OfTheBlaze => player_stat.Damage_Guard
        armor_suffix.OfWind => player_stat.DamageMult 
        armor_suffix.OfMidas => player_stat.DamageMult
        _ => player_stat.None

ToString(Stat:player_stat)<transacts>:string =
    case (Stat):
        player_stat.MaxHealth => "MaxHealth"
        player_stat.Regen => "Regen"
        player_stat.DamageMult => "DamageMult"
        player_stat.Damage_Beast => "Damage_Beast"
        player_stat.Damage_Guard => "Damage_Guard"
        player_stat.Damage_Mineral => "Damage_Mineral"
        player_stat.Damage_Boss => "Damage_Boss"
        player_stat.Damage_With_Pickaxe => "Damage_With_Pickaxe"
        player_stat.Damage_With_Pistol => "Damage_With_Pistol"
        player_stat.Damage_With_AR => "Damage_With_AR"
        player_stat.Damage_With_Shotgun => "Damage_With_Shotgun"
        player_stat.Damage_With_Railgun => "Damage_With_Railgun"
        player_stat.Damage_With_Bow => "Damage_With_Bow"
        player_stat.Damage_With_Minigun => "Damage_With_Minigun"
        player_stat.Damage_With_Hand_Cannon => "Damage_With_Hand_Cannon"
        player_stat.Damage_With_Scythe => "Damage_With_Scythe"
        player_stat.Damage_With_Stake_Gun => "Damage_With_Stake_Gun"
        player_stat.Empowered_Mining => "Empowered_Mining"
        player_stat.Passive_Income_Flat => "Passive_Income_Flat"
        player_stat.Passive_Income_Percent => "Passive_Income_Percent"
        player_stat.MoveSpeed => "MoveSpeed"
        player_stat.Crit_Mult => "CritMult"
        player_stat.None => "None"
        player_stat.Damage_Melee => "Damage_Melee"
        player_stat.Damage_Rapid => "Damage_Rapid"
        player_stat.Damage_Precision => "Damage_Precision"

ToString(Slot:armor_slot)<transacts>:string =
    case (Slot):
        armor_slot.Head => "Head"
        armor_slot.Torso => "Torso"
        armor_slot.Legs => "Legs"
        armor_slot.Boots => "Boots"
        armor_slot.Gloves => "Gloves"
        armor_slot.Accessory => "Accessory"
        armor_slot.None => "None"

ToString(Rarity:armor_rarity)<transacts>:string =
    case (Rarity):
        armor_rarity.Common => "Common"
        armor_rarity.Uncommon => "Uncommon"
        armor_rarity.Rare => "Rare"
        armor_rarity.Legendary => "Legendary"
        armor_rarity.Epic => "Epic"
        armor_rarity.Mythic => "Mythic"
        armor_rarity.Godly => "Godly"
        armor_rarity.Exalted => "Exalted"
        armor_rarity.Ultimate => "Ultimate"
        armor_rarity.Supreme => "Supreme"
        armor_rarity.Transcendent => "Transcendent"
        armor_rarity.Infinite => "Infinite"
        armor_rarity.None => "None"

ToString(Base:armor_base)<transacts>:string =
    case (Base):
        armor_base.Helmet => "Helmet"
        armor_base.Hood => "Hood"
        armor_base.Mask => "Mask"
        armor_base.Cap => "Cap"
        armor_base.Crown => "Crown"
        armor_base.Circlet => "Circlet"
        armor_base.Headdress => "Headdress"
        armor_base.Visor => "Visor"
        armor_base.Tiara => "Tiara"
        armor_base.Cowl => "Cowl"
        armor_base.WizardHat => "WizardHat"
        armor_base.Bandana => "Bandana"
        armor_base.Headband => "Headband"
        armor_base.Beret => "Beret"
        armor_base.Beanie => "Beanie"
        armor_base.Chestplate => "Chestplate"
        armor_base.Robe => "Robe"
        armor_base.Jacket => "Jacket"
        armor_base.Tunic => "Tunic"
        armor_base.Coat => "Coat"
        armor_base.Vest => "Vest"
        armor_base.Chainmail => "Chainmail"
        armor_base.Jerkin => "Jerkin"
        armor_base.Hauberk => "Hauberk"
        armor_base.Tabard => "Tabard"
        armor_base.Kimono => "Kimono"
        armor_base.Hoodie => "Hoodie"
        armor_base.Greaves => "Greaves"
        armor_base.Pants => "Pants"
        armor_base.Shorts => "Shorts"
        armor_base.Kilt => "Kilt"
        armor_base.Chaps => "Chaps"
        armor_base.Jeans => "Jeans"
        armor_base.Trousers => "Trousers"
        armor_base.Pantaloons => "Pantaloons"
        armor_base.Boots => "Boots"
        armor_base.Sabatons => "Sabatons"
        armor_base.Shoes => "Shoes"
        armor_base.Slippers => "Slippers"
        armor_base.Sandals => "Sandals"
        armor_base.Footwraps => "Footwraps"
        armor_base.Gauntlets => "Gauntlets"
        armor_base.Bracers => "Bracers"
        armor_base.Handguards => "Handguards"
        armor_base.Handwraps => "Handwraps"
        armor_base.Mittens => "Mittens"
        armor_base.Armguards => "Armguards"
        armor_base.Belt => "Belt"
        armor_base.Ring => "Ring"
        armor_base.Necklace => "Necklace"
        armor_base.Bracelet => "Bracelet"
        armor_base.Amulet => "Amulet"
        armor_base.Talisman => "Talisman"
        armor_base.Brooch => "Brooch"
        armor_base.Earrings => "Earrings"
        armor_base.Pendant => "Pendant"
        armor_base.Medal => "Medal"
        armor_base.Pin => "Pin"
        armor_base.Charm => "Charm"
        armor_base.Medallion => "Medallion"
        armor_base.Wristband => "Wristband"
        armor_base.Locket => "Locket"
        armor_base.Bangle => "Bangle"
        _ => "???"

ToString(Prefix:armor_prefix)<transacts>:string =
    case (Prefix):
        armor_prefix.Savage => "Savage"
        armor_prefix.Valiant => "Valiant"
        armor_prefix.Arcane => "Arcane"
        armor_prefix.Mystic => "Mystic"
        armor_prefix.Infernal => "Infernal"
        armor_prefix.Stalwart => "Stalwart"
        armor_prefix.Celestial => "Celestial"
        armor_prefix.Shadow => "Shadow"
        armor_prefix.Sacred => "Sacred"
        armor_prefix.Radiant => "Radiant"
        armor_prefix.Warlords => "Warlords"
        armor_prefix.Guardians => "Guardians"
        armor_prefix.Ember => "Ember"
        armor_prefix.Frost => "Frost"
        armor_prefix.Thunder => "Thunder"
        armor_prefix.Titan => "Titan"
        armor_prefix.Ethereal => "Ethereal"
        armor_prefix.Phantom => "Phantom"
        armor_prefix.Dawn => "Dawn"
        armor_prefix.Lyras => "Lyra's"
        armor_prefix.BeachInvader => "Beach Invader"
        armor_prefix.Mining => "Mining"
        _ => ""

ToString(Suffix:armor_suffix)<transacts>:string =
    case (Suffix):
        # armor_suffix.OfThePhoenix => "Of The Phoenix"
        # armor_suffix.OfTheTitan => "Of The Titan"
        # armor_suffix.OfTheWolf => "Of The Wolf"
        # armor_suffix.OfTheDragon => "Of The Dragon"
        # armor_suffix.OfTheSerpent => "Of The Serpent"
        # armor_suffix.OfTheGriffin => "Of The Griffin"
        # armor_suffix.OfTheLion => "Of The Lion"
        # armor_suffix.OfTheEagle => "Of The Eagle"
        # armor_suffix.OfTheBear => "Of The Bear"
        # armor_suffix.OfTheStorm => "Of The Storm"
        # armor_suffix.OfTheTempest => "Of The Tempest"
        # armor_suffix.OfNight => "Of Night"
        # armor_suffix.OfEarth => "Of Earth"
        # armor_suffix.OfTheStars => "Of The Stars"
        # armor_suffix.OfVoid => "Of Void"
        # armor_suffix.OfTheInferno => "Of The Inferno"
        # armor_suffix.OfTheGlacier => "Of The Glacier"
        # armor_suffix.OfThunder => "Of Thunder"
        # armor_suffix.OfTheBlaze => "Of The Blaze"
        # armor_suffix.OfWind => "Of Wind"
        # armor_suffix.OfMidas => "Of Midas"
        _ => ""

PlayerStatFromString(Stat:string)<transacts>:player_stat=
    case (Stat):
        "MaxHealth" => player_stat.MaxHealth
        "Regen" => player_stat.Regen
        "DamageMult" => player_stat.DamageMult
        "Damage_Beast" => player_stat.Damage_Beast
        "Damage_Guard" => player_stat.Damage_Guard
        "Damage_Mineral" => player_stat.Damage_Mineral
        "Damage_Boss" => player_stat.Damage_Boss
        "Damage_With_Pickaxe" => player_stat.Damage_With_Pickaxe
        "Damage_With_Pistol" => player_stat.Damage_With_Pistol
        "Damage_With_AR" => player_stat.Damage_With_AR
        "Damage_With_Shotgun" => player_stat.Damage_With_Shotgun
        "Damage_With_Railgun" => player_stat.Damage_With_Railgun
        "Damage_With_Minigun" => player_stat.Damage_With_Minigun
        "Damage_With_Hand_Cannon" => player_stat.Damage_With_Hand_Cannon
        "Damage_With_Scythe" => player_stat.Damage_With_Scythe 
        "Damage_With_Stake_Gun" => player_stat.Damage_With_Stake_Gun
        "Damage_With_Bow" => player_stat.Damage_With_Bow
        "Empowered_Mining" => player_stat.Empowered_Mining
        "Passive_Income_Flat" => player_stat.Passive_Income_Flat
        "Passive_Income_Percent" => player_stat.Passive_Income_Percent
        "MoveSpeed" => player_stat.MoveSpeed
        "CritMult" => player_stat.Crit_Mult
        _ => player_stat.None    

ArmorSlotFromString(Slot:string)<transacts>:armor_slot=
    case (Slot):
        "Head" => armor_slot.Head
        "Torso" => armor_slot.Torso
        "Legs" => armor_slot.Legs
        "Boots" => armor_slot.Boots
        "Gloves" => armor_slot.Gloves
        "Accessory" => armor_slot.Accessory
        _ => armor_slot.None

ArmorRarityFromString(Rarity:string)<transacts>:armor_rarity=
    case (Rarity):
        "Common" => armor_rarity.Common
        "Uncommon" => armor_rarity.Uncommon
        "Rare" => armor_rarity.Rare
        "Legendary" => armor_rarity.Legendary
        "Epic" => armor_rarity.Epic
        "Mythic" => armor_rarity.Mythic
        "Godly" => armor_rarity.Godly
        "Exalted" => armor_rarity.Exalted
        "Ultimate" => armor_rarity.Ultimate
        "Supreme" => armor_rarity.Supreme
        "Transcendent" => armor_rarity.Transcendent
        "Infinite" => armor_rarity.Infinite
        _ => armor_rarity.None

ArmorBaseFromString(Base:string)<transacts>:armor_base=
    case (Base):
        #Head
        "Helmet" => armor_base.Helmet
        "Hood" => armor_base.Hood
        "Mask" => armor_base.Mask
        "Cap" => armor_base.Cap
        "Crown" => armor_base.Crown
        "Circlet" => armor_base.Circlet
        "Headdress" => armor_base.Headdress
        "Visor" => armor_base.Visor
        "Tiara" => armor_base.Tiara
        "Cowl" => armor_base.Cowl
        "WizardHat" => armor_base.WizardHat
        "Bandana" => armor_base.Bandana
        "Headband" => armor_base.Headband
        "Beret" => armor_base.Beret
        "Beanie" => armor_base.Beanie
        #Torso
        "Chestplate" => armor_base.Chestplate
        "Robe" => armor_base.Robe
        "Jacket" => armor_base.Jacket
        "Tunic" => armor_base.Tunic
        "Coat" => armor_base.Coat
        "Vest" => armor_base.Vest
        "Chainmail" => armor_base.Chainmail
        "Jerkin" => armor_base.Jerkin
        "Hauberk" => armor_base.Hauberk
        "Tabard" => armor_base.Tabard
        "Kimono" => armor_base.Kimono
        "Hoodie" => armor_base.Hoodie
        #Legs
        "Greaves" => armor_base.Greaves
        "Pants" => armor_base.Pants
        "Shorts" => armor_base.Shorts
        "Kilt" => armor_base.Kilt
        "Chaps" => armor_base.Chaps
        "Jeans" => armor_base.Jeans
        "Trousers" => armor_base.Trousers
        "Pantaloons" => armor_base.Pantaloons
        #Boots
        "Boots" => armor_base.Boots
        "Sabatons" => armor_base.Sabatons
        "Shoes" => armor_base.Shoes
        "Slippers" => armor_base.Slippers        
        "Sandals" => armor_base.Sandals
        "Footwraps" => armor_base.Footwraps
        #Gloves
        "Gauntlets" => armor_base.Gauntlets
        "Bracers" => armor_base.Bracers
        "Handguards" => armor_base.Handguards
        "Handwraps" => armor_base.Handwraps
        "Mittens" => armor_base.Mittens
        "Armguards" => armor_base.Armguards
        #Accessory
        "Belt" => armor_base.Belt        
        "Ring" => armor_base.Ring
        "Necklace" => armor_base.Necklace
        "Bracelet" => armor_base.Bracelet
        "Amulet" => armor_base.Amulet
        "Talisman" => armor_base.Talisman
        "Brooch" => armor_base.Brooch
        "Earrings" => armor_base.Earrings
        "Pendant" => armor_base.Pendant
        "Medal" => armor_base.Medal
        "Pin" => armor_base.Pin
        "Charm" => armor_base.Charm
        "Medallion" => armor_base.Medallion
        "Wristband" => armor_base.Wristband
        "Locket" => armor_base.Locket
        "Bangle" => armor_base.Bangle
        _ => armor_base.None

ArmorPrefixFromString(Prefix:string)<transacts>:armor_prefix=
    case (Prefix):
        "Savage" => armor_prefix.Savage
        "Valiant" => armor_prefix.Valiant
        "Arcane" => armor_prefix.Arcane
        "Mystic" => armor_prefix.Mystic
        "Infernal" => armor_prefix.Infernal
        "Stalwart" => armor_prefix.Stalwart
        "Celestial" => armor_prefix.Celestial
        "Shadow" => armor_prefix.Shadow
        "Sacred" => armor_prefix.Sacred
        "Radiant" => armor_prefix.Radiant
        "Warlords" => armor_prefix.Warlords
        "Guardians" => armor_prefix.Guardians
        "Ember" => armor_prefix.Ember
        "Frost" => armor_prefix.Frost
        "Thunder" => armor_prefix.Thunder
        "Titan" => armor_prefix.Titan
        "Ethereal" => armor_prefix.Ethereal
        "Phantom" => armor_prefix.Phantom
        "Dawn" => armor_prefix.Dawn
        "Lyra's" => armor_prefix.Lyras
        "Beach Invader" => armor_prefix.BeachInvader
        "Mining" => armor_prefix.Mining
        _ => armor_prefix.None

ArmorSuffixFromString(Suffix:string)<transacts>:armor_suffix=
    case (Suffix):
        "Of The Phoenix" => armor_suffix.OfThePhoenix
        "Of The Titan" => armor_suffix.OfTheTitan
        "Of The Wolf" => armor_suffix.OfTheWolf
        "Of The Dragon" => armor_suffix.OfTheDragon
        "Of The Serpent" => armor_suffix.OfTheSerpent
        "Of The Griffin" => armor_suffix.OfTheGriffin
        "Of The Lion" => armor_suffix.OfTheLion
        "Of The Eagle" => armor_suffix.OfTheEagle
        "Of The Bear" => armor_suffix.OfTheBear
        "Of The Storm" => armor_suffix.OfTheStorm
        "Of The Tempest" => armor_suffix.OfTheTempest
        "Of Night" => armor_suffix.OfNight
        "Of Earth" => armor_suffix.OfEarth
        "Of The Stars" => armor_suffix.OfTheStars
        "Of Void" => armor_suffix.OfVoid
        "Of The Inferno" => armor_suffix.OfTheInferno        
        "Of The Glacier" => armor_suffix.OfTheGlacier
        "Of Thunder" => armor_suffix.OfThunder
        "Of The Blaze" => armor_suffix.OfTheBlaze
        "Of Wind" => armor_suffix.OfWind
        "Of Midas" => armor_suffix.OfMidas
        _ => armor_suffix.None

(Rarity:armor_rarity).ToRarityID()<transacts>:int =
    case (Rarity):
        armor_rarity.Common => 0
        armor_rarity.Uncommon => 1
        armor_rarity.Rare => 2
        armor_rarity.Epic => 3
        armor_rarity.Legendary => 4
        armor_rarity.Mythic => 5
        armor_rarity.Godly => 6
        armor_rarity.Exalted => 7
        armor_rarity.Ultimate => 8
        armor_rarity.Supreme => 9
        armor_rarity.Transcendent => 10
        armor_rarity.Infinite => 11
        armor_rarity.None => -1

ArmorRarityFromID(ID:int)<transacts>:armor_rarity=
    case (ID):
        0 => armor_rarity.Common
        1 => armor_rarity.Uncommon
        2 => armor_rarity.Rare
        3 => armor_rarity.Epic
        4 => armor_rarity.Legendary
        5 => armor_rarity.Mythic
        6 => armor_rarity.Godly
        7 => armor_rarity.Exalted
        8 => armor_rarity.Ultimate
        9 => armor_rarity.Supreme
        10 => armor_rarity.Transcendent
        11 => armor_rarity.Infinite
        _ => armor_rarity.None
        
(Rarity:armor_rarity).ToColor():color =
    case (Rarity):
        armor_rarity.Common => NamedColors.White
        armor_rarity.Uncommon => MakeColorFromHex("1EFF00")
        armor_rarity.Rare => MakeColorFromHex("0070DD")
        armor_rarity.Epic => MakeColorFromHex("A335EE")
        armor_rarity.Legendary => MakeColorFromHex("FFDD00")
        armor_rarity.Mythic => MakeColorFromHex("ff65e0")
        armor_rarity.Godly => MakeColorFromHex("FF8000")
        armor_rarity.Exalted => MakeColorFromHex("ff3838")
        armor_rarity.Ultimate => MakeColorFromHex("60fbff")
        armor_rarity.Supreme => MakeColorFromHex("333333")
        armor_rarity.Transcendent => MakeColorFromHex("60ffb6")
        armor_rarity.Infinite => MakeColorFromHex("eeeeff")
        armor_rarity.None => NamedColors.Grey

(Rarity:armor_rarity).StatMult()<transacts>:float =
    case (Rarity):
        armor_rarity.Common => 1.0
        armor_rarity.Uncommon => 1.2
        armor_rarity.Rare => 1.5
        armor_rarity.Epic => 2.0
        armor_rarity.Legendary => 3.0
        armor_rarity.Mythic => 4.0
        armor_rarity.Godly => 5.0
        armor_rarity.Exalted => 7.5
        armor_rarity.Ultimate => 10.0
        armor_rarity.Supreme => 15.0
        armor_rarity.Transcendent => 20.0
        armor_rarity.Infinite => 30.0
        armor_rarity.None => 1.0

(Rarity:armor_rarity).SecondaryCount()<transacts>:int =
    case (Rarity):
        armor_rarity.Common => 0
        armor_rarity.Uncommon => 1
        armor_rarity.Rare => 2
        armor_rarity.Epic => 3
        armor_rarity.Legendary => 3
        armor_rarity.Mythic => 3
        armor_rarity.Godly => 4
        armor_rarity.Exalted => 4
        armor_rarity.Ultimate => 4
        armor_rarity.Supreme => 4
        armor_rarity.Transcendent => 4
        armor_rarity.Infinite => 5
        armor_rarity.None => 0

ArmorRaritiesList : []armor_rarity = array{armor_rarity.Common, armor_rarity.Uncommon, armor_rarity.Rare, armor_rarity.Legendary}

ArmorStatsList : []player_stat = array{player_stat.MaxHealth, player_stat.Regen, player_stat.DamageMult, player_stat.Damage_Beast, 
                                        player_stat.Damage_Guard, player_stat.Damage_Mineral, player_stat.Damage_Boss, player_stat.Damage_With_Pickaxe,
                                        player_stat.Damage_With_Pistol, player_stat.Damage_With_AR, player_stat.Damage_With_Shotgun, player_stat.Damage_With_Railgun,
                                        player_stat.Damage_With_Minigun, player_stat.Damage_With_Stake_Gun, player_stat.Damage_With_Scythe, player_stat.Damage_With_Bow,
                                        player_stat.Crit_Mult}#, player_stat.MoveSpeed}

ArmorSlotsList : []armor_slot = array{armor_slot.Head, armor_slot.Torso, armor_slot.Legs, armor_slot.Boots, armor_slot.Gloves, 
                                        armor_slot.Accessory}

ArmorPrefixesList : []armor_prefix = array{armor_prefix.Savage, armor_prefix.Valiant, armor_prefix.Arcane, armor_prefix.Mystic, 
                                        armor_prefix.Infernal, armor_prefix.Stalwart, armor_prefix.Celestial, armor_prefix.Shadow, 
                                        armor_prefix.Sacred, armor_prefix.Radiant, armor_prefix.Warlords, armor_prefix.Guardians, 
                                        armor_prefix.Ember, armor_prefix.Frost, armor_prefix.Thunder, armor_prefix.Titan, 
                                        armor_prefix.Ethereal, armor_prefix.Phantom, armor_prefix.Dawn, armor_prefix.BeachInvader}

ArmorSuffixesList : []armor_suffix = array{armor_suffix.OfThePhoenix, armor_suffix.OfTheTitan, armor_suffix.OfTheWolf, armor_suffix.OfTheDragon, 
                                        armor_suffix.OfTheSerpent, armor_suffix.OfTheGriffin, armor_suffix.OfTheLion, armor_suffix.OfTheEagle, 
                                        armor_suffix.OfTheBear, armor_suffix.OfTheStorm, armor_suffix.OfTheTempest, armor_suffix.OfNight, 
                                        armor_suffix.OfEarth, armor_suffix.OfTheStars, armor_suffix.OfVoid, armor_suffix.OfTheInferno, 
                                        armor_suffix.OfTheGlacier, armor_suffix.OfThunder, armor_suffix.OfTheBlaze, armor_suffix.OfWind}

(Stat:player_stat).ToStatKey():string =
    case (Stat):
        player_stat.MaxHealth => "Gear_MaxHealth"
        player_stat.Regen => "Gear_Regen"
        player_stat.DamageMult => "Gear_DamageMult"
        player_stat.Damage_Beast => "Gear_MasteryBeast"
        player_stat.Damage_Guard => "Gear_MasteryGuard"
        player_stat.Damage_Mineral => "Gear_MasteryMineral"
        player_stat.Damage_Boss => "Gear_MasteryBoss"
        player_stat.Damage_With_Pickaxe => "Gear_WeaponDamagePickaxe"
        player_stat.Damage_With_Pistol => "Gear_WeaponDamagePistol"
        player_stat.Damage_With_AR => "Gear_WeaponDamageAR"
        player_stat.Damage_With_Shotgun => "Gear_WeaponDamageShotgun"
        player_stat.Damage_With_Railgun => "Gear_WeaponDamageRailgun"
        player_stat.Damage_With_Bow => "Gear_WeaponDamageBow"
        player_stat.Damage_With_Minigun => "Gear_WeaponDamageMinigun"
        player_stat.Damage_With_Hand_Cannon => "Gear_WeaponDamageHand_Cannon"
        player_stat.Damage_With_Stake_Gun => "Gear_WeaponDamageStake_Gun"
        player_stat.Damage_With_Scythe => "Gear_WeaponDamageScythe"
        player_stat.Empowered_Mining => "Gear_EmpoweredMineral"
        player_stat.Passive_Income_Flat => "Gear_PassiveIncomeFlat"
        player_stat.Passive_Income_Percent => "Gear_PassiveIncomePercent"
        player_stat.MoveSpeed => "Gear_MoveSpeed"
        player_stat.Crit_Mult => "Gear_CritMult"
        player_stat.Damage_Melee => "Gear_DamageMelee"
        player_stat.Damage_Rapid => "Gear_DamageRapid"
        player_stat.Damage_Precision => "Gear_DamagePrecision"
        player_stat.None => "None"

(Stat:player_stat).ToArmorStatValue(Level : int)<transacts>:big_number =
    case (Stat):
        player_stat.MaxHealth => big_number{Value:=Level * 10.0} #flat
        player_stat.Regen => big_number{Value:=Level * 0.2} #flat
        player_stat.DamageMult => big_number{Value:= Level * 4.0} #percent
        player_stat.Damage_Beast => big_number{Value:= Level * 8.0} #percent
        player_stat.Damage_Guard => big_number{Value:= Level * 8.0} #percent
        player_stat.Damage_Mineral => big_number{Value:= Level * 8.0} #percent
        player_stat.Damage_Boss => big_number{Value:= Level * 8.0} #percent
        player_stat.Damage_With_Pickaxe => big_number{Value:= Level * 150.0} #percent
        player_stat.Damage_With_Pistol => big_number{Value:= Level * 15.0} #percent
        player_stat.Damage_With_AR => big_number{Value:= Level * 15.0} #percent
        player_stat.Damage_With_Shotgun => big_number{Value:= Level * 15.0} #percent
        player_stat.Damage_With_Railgun => big_number{Value:= Level * 15.0} #percent
        player_stat.Damage_With_Bow => big_number{Value:= Level * 15.0} #percent
        player_stat.Damage_With_Minigun => big_number{Value:= Level * 15.0} #percent
        player_stat.Damage_With_Hand_Cannon => big_number{Value:= Level * 15.0} #percent
        player_stat.Damage_With_Stake_Gun => big_number{Value:= Level * 15.0} #percent
        player_stat.Damage_With_Scythe => big_number{Value:= Level * 15.0} #percent
        player_stat.Empowered_Mining => big_number{Value:= Level * 20.0} #percent
        player_stat.Passive_Income_Flat => MakeBigNumber(Level*1.0, 1.5) * 1000.0 #flat
        player_stat.Passive_Income_Percent => big_number{Value:=Level * 10.0} #percent
        player_stat.MoveSpeed => big_number{Value:= Level * 2.0} #doesn't work
        player_stat.Crit_Mult => big_number{Value:= Level * 2.0} #doesn't work
        player_stat.Damage_Melee => big_number{Value:= Level * 1.0} #percent
        player_stat.Damage_Rapid => big_number{Value:= Level * 1.0} #percent
        player_stat.Damage_Precision => big_number{Value:= Level * 1.0} #percent
        player_stat.None => big_number{}

(Stat:player_stat).ValToStatline(Val : big_number)<transacts>:string =
    case (Stat):
        player_stat.MaxHealth => "+{Val} Max Health"
        player_stat.Regen => "+{Val}HP/s"
        player_stat.DamageMult => "+{Val}% Total Damage"
        player_stat.Damage_Beast => "+{Val}% Beasts Damage"
        player_stat.Damage_Guard => "+{Val}% Humanoids Damage"
        player_stat.Damage_Mineral => "+{Val}% Minerals Damage"
        player_stat.Damage_Boss => "+{Val}% Bosses Damage"
        player_stat.Damage_With_Pickaxe => "+{Val}% Pickaxes Damage"
        player_stat.Damage_With_Pistol => "+{Val}% Pistols Damage"
        player_stat.Damage_With_AR => "+{Val}% ARs Damage"
        player_stat.Damage_With_Shotgun => "+{Val}% Shotguns Damage"
        player_stat.Damage_With_Railgun => "+{Val}% Railguns Damage"
        player_stat.Damage_With_Bow => "+{Val}% Bows Damage"
        player_stat.Damage_With_Minigun => "+{Val}% Minigun Damage"
        player_stat.Damage_With_Hand_Cannon => "+{Val}% Hand Cannons Damage"
        player_stat.Damage_With_Stake_Gun => "+{Val}% Stake Gun Damage"
        player_stat.Damage_With_Scythe => "+{Val}% Scythe Damage"
        player_stat.Empowered_Mining => "+{Val}% Mining Power"
        player_stat.Passive_Income_Flat => "+{Val} Essence/min"
        player_stat.Passive_Income_Percent => "+{Val}% Passive Income"
        player_stat.MoveSpeed => "{Stat}: +{Val}???"
        player_stat.Crit_Mult => "+{Val}% Weakspot Damage"
        player_stat.Damage_Melee => "+{Val}% Melee Damage"
        player_stat.Damage_Rapid => "+{Val}% Rapid Damage"
        player_stat.Damage_Precision => "+{Val}% Precision Damage"
        player_stat.None => ""

(Stat:player_stat).ToStatline(Level : int)<transacts>:string =
    Stat.ValToStatline(Stat.ToArmorStatValue(Level))
