using { /Fortnite.com/Devices }
using { /Verse.org/Simulation }

audio_player_wrapper := class<concrete>:
    @editable
    APD : audio_player_device = audio_player_device{}

    @editable
    ID : int = 0

audio_manager := class(creative_device):

    @editable
    MainGameMusic : audio_player_wrapper = audio_player_wrapper{}

    @editable
    BossMusic : audio_player_wrapper = audio_player_wrapper{}

    @editable
    GoodtownMusic : audio_player_wrapper = audio_player_wrapper{}

    @editable
    GiantlandsMusic : audio_player_wrapper = audio_player_wrapper{}

    @editable
    CanyonMusic : audio_player_wrapper = audio_player_wrapper{}

    @editable
    BeachMusic : audio_player_wrapper = audio_player_wrapper{}

    @editable
    UmbralForestMusic : audio_player_wrapper = audio_player_wrapper{}
    
    @editable
    ActiveEngagementMusic : audio_player_wrapper = audio_player_wrapper{}

    @editable
    ForestBanditMusic : audio_player_wrapper = audio_player_wrapper{}

    @editable 
    LevelUpAudioPlayer : audio_player_device = audio_player_device{}
    
    @editable 
    HighSpeedLevelUpAudioPlayer : audio_player_device = audio_player_device{}

    @editable 
    VeryHighSpeedLevelUpAudioPlayer : audio_player_device = audio_player_device{}

    @editable 
    GuardDestroyedPlayerA : audio_player_device = audio_player_device{}
    
    @editable 
    GuardDestroyedPlayerC : audio_player_device = audio_player_device{}
    
    @editable
    GuardDestroyedPlayerEm : audio_player_device = audio_player_device{}
    
    @editable
    GuardDestroyedPlayerGm : audio_player_device = audio_player_device{}

    @editable
    EnterZonePlayer : audio_player_device = audio_player_device{}

    @editable
    ZoneLevelUpPlayer : audio_player_device = audio_player_device{}

    @editable
    UnlockZonePlayer : audio_player_device = audio_player_device{}

    @editable
    WeakspotHitPlayer : audio_player_device = audio_player_device{}
    
    @editable
    WeakspotHitPlayer1 : audio_player_device = audio_player_device{}

    @editable
    WeakspotHitPlayer2 : audio_player_device = audio_player_device{}

    @editable
    WeakspotHitPlayer3 : audio_player_device = audio_player_device{}

    @editable
    WeakspotHitPlayer4 : audio_player_device = audio_player_device{}

    @editable
    PassiveIncomeCollectedPlayer : audio_player_device = audio_player_device{}

    @editable
    CollectGemPlayer  : audio_player_device = audio_player_device{}

    @editable
    NewTaskPlayer : audio_player_device = audio_player_device{}

    @editable
    TaskCompletePlayer : audio_player_device = audio_player_device{}

    @editable
    NarratorWelcomePlayer : audio_player_device = audio_player_device{}

    @editable
    NarratorPrepareToBeginPlayer : audio_player_device = audio_player_device{}

    @editable
    Boss1WarpSound : audio_player_device = audio_player_device{}

    @editable
    UpgradeSuccessSound : audio_player_device = audio_player_device{}

    @editable
    UpgradeFailSound : audio_player_device = audio_player_device{}

    @editable
    BossSlainAudio : audio_player_device = audio_player_device{}

    @editable
    MidasBossVOEntrance : []audio_player_device = array{}

    @editable
    MidasBossVOHurt : []audio_player_device = array{}

    @editable
    MidasBossVOExit : []audio_player_device = array{}

    @editable
    GengaBossVOEntrance : []audio_player_device = array{}

    @editable
    GengaBossVOHurt : []audio_player_device = array{}

    @editable
    GengaBossVOExit : []audio_player_device = array{}

    @editable
    DamageMasteryTimerStart : audio_player_device = audio_player_device{}
    
    @editable
    DamageMasteryTimerTick : audio_player_device = audio_player_device{}

    @editable
    DamageMasteryTimerEnd : audio_player_device = audio_player_device{}

    @editable
    XPCoin : audio_player_device = audio_player_device{}

    @editable
    DropCandySound : audio_player_device = audio_player_device{}

    @editable
    PickCandySound : audio_player_device = audio_player_device{}

    @editable
    PickCandyCheerSound : audio_player_device = audio_player_device{}

    @editable
    EatCandySound : audio_player_device = audio_player_device{}

    @editable
    CollectCoinsSound : audio_player_device = audio_player_device{}

    @editable
    SellCoinsSound : audio_player_device = audio_player_device{}

    @editable
    DropCoinsSound : audio_player_device = audio_player_device{}

    @editable
    PeelyLaugh : audio_player_device = audio_player_device{}

    @editable
    Handcannon : audio_player_device = audio_player_device{}
    
    @editable
    Ambush : audio_player_device = audio_player_device{}

    @editable
    BlossomBuff : audio_player_device = audio_player_device{}

    @editable
    NarratorCountdown : []audio_player_device = array{}

    @editable
    LyraSwim : []audio_player_device = array{}

    @editable
    ArmorDeflectionSound : audio_player_device = audio_player_device{}

    @editable
    GunLevelUp : audio_player_device = audio_player_device{}
    
    @editable
    SharkeyVO2a : audio_player_device = audio_player_device{}

    @editable
    SharkeyVO2b : audio_player_device = audio_player_device{}

    @editable
    SargeVO2a : audio_player_device = audio_player_device{}

    @editable
    SargeVO2b : audio_player_device = audio_player_device{}

    @editable
    SargeVO2c : audio_player_device = audio_player_device{}

    @editable
    SargeVO2d : audio_player_device = audio_player_device{}

    @editable
    SargeVO2e : audio_player_device = audio_player_device{}

    @editable
    SargeVO2f : audio_player_device = audio_player_device{}

    @editable
    SargeVO2g : audio_player_device = audio_player_device{}

    @editable
    SargeVO2h : audio_player_device = audio_player_device{}

    @editable
    SargeVO2i : audio_player_device = audio_player_device{}

    @editable
    SargeVO2j : audio_player_device = audio_player_device{}

    @editable
    SargeVO2k : audio_player_device = audio_player_device{}

    @editable
    SargeVO2l : audio_player_device = audio_player_device{}

    @editable
    UrjentVO2a : audio_player_device = audio_player_device{}

    @editable
    UrjentVO2b : audio_player_device = audio_player_device{}

    @editable
    UrjentVO2c : audio_player_device = audio_player_device{}

    @editable
    UrjentVO2d : audio_player_device = audio_player_device{}

    @editable
    UrjentVO2e : audio_player_device = audio_player_device{}

    @editable
    UrjentVO2f : audio_player_device = audio_player_device{}

    @editable
    UrjentVO2g : audio_player_device = audio_player_device{}

    @editable
    GengaVO2a : audio_player_device = audio_player_device{}

    @editable
    GengaVO2b : audio_player_device = audio_player_device{}

    @editable
    GengaVO2c : audio_player_device = audio_player_device{}

    @editable
    CharonVO2a : audio_player_device = audio_player_device{}

    @editable
    CharonVO2b : audio_player_device = audio_player_device{}

    @editable
    CharonVO2c : audio_player_device = audio_player_device{}

    @editable
    CharonVO2d : audio_player_device = audio_player_device{}

    @editable
    CharonVO2e : audio_player_device = audio_player_device{}

    @editable
    CharonVO2f : audio_player_device = audio_player_device{}

    @editable
    CharonVO2g : audio_player_device = audio_player_device{}

    @editable
    CharonVO2h : audio_player_device = audio_player_device{}

    Countdown(Agent:agent, TimeLeft:int):void=
        if:
            Audio:= NarratorCountdown[3-TimeLeft]
        then:
            Audio.Play(Agent)

    PlayHandcannon(Agent:agent):void=
        Handcannon.Play(Agent)

    PlayDamageMasteryTimerEnd(Agent:agent):void=
        DamageMasteryTimerEnd.Play(Agent)

    PlayDamageMasteryTimerTick(Agent:agent):void=
        DamageMasteryTimerTick.Play(Agent)

    PlayBossSlainAudio(Agent:agent):void=
        BossSlainAudio.Play(Agent)

    PlayDamageMasteryTimerStart(Agent:agent):void=
        DamageMasteryTimerStart.Play(Agent)

    PlayLevelUpSFX(Agent:agent):void=
        LevelUpAudioPlayer.Play(Agent)

    PlayHighSpeedLevelUpSFX(Agent:agent):void=
        HighSpeedLevelUpAudioPlayer.Play(Agent)
    
    PlayVeryHighSpeedLevelUpSFX(Agent:agent):void=
        VeryHighSpeedLevelUpAudioPlayer.Play(Agent)
      
    PlayWeakSpotHitSFX(Agent:agent, WeakSpotSFX:int):void=
        case(WeakSpotSFX):
            1 => WeakspotHitPlayer.Play(Agent)
            2 => WeakspotHitPlayer1.Play(Agent)
            3 => WeakspotHitPlayer2.Play(Agent)
            4 => WeakspotHitPlayer3.Play(Agent)
            _ => WeakspotHitPlayer4.Play(Agent)
        
    PlayBossEntranceVO(Agent:agent, VOIndex:int, BossType:boss_enum):void=
        if:
            VO:=case(BossType):
                boss_enum.Sharky => MidasBossVOEntrance[VOIndex]
                boss_enum.Genga => GengaBossVOEntrance[VOIndex]
        then:
            VO.Play(Agent)

    PlayBossExitVO(Agent:agent, VOIndex:int, BossType:boss_enum):void=
        if:
            VO:=case(BossType):
                boss_enum.Sharky => MidasBossVOExit[VOIndex]
                boss_enum.Genga => GengaBossVOExit[VOIndex]
        then:
            VO.Play(Agent)

    PlayBossEffort(Agent:agent, EffortIndex:int, BossType:boss_enum):void=
        if:
            Effort:=case(BossType):
                boss_enum.Sharky => MidasBossVOHurt[EffortIndex]
                boss_enum.Genga => GengaBossVOHurt[EffortIndex]
                boss_enum.LyraTutorial => case(EffortIndex):
                    0 => SargeVO2h
                    1 => SargeVO2i
                    2 => SargeVO2j
                    _ => SargeVO2k
                boss_enum.Urjent => case(EffortIndex):
                    0 => UrjentVO2d
                    1 => UrjentVO2e
                    _ => UrjentVO2f
                boss_enum.PirateKing => case(EffortIndex):
                    0 => CharonVO2f
                    1 => CharonVO2g
                    _ => CharonVO2h
        then:
            Effort.Play(Agent)

    PlayZoneLevelUpSFX(Agent:agent):void=
        ZoneLevelUpPlayer.Play(Agent)

    PlayEnterZoneSFX(Agent:agent):void=
        EnterZonePlayer.Play(Agent)

    PlayGuardDestroyedSFX(Agent:agent, GuardComboSFX:int):void=
        case(GuardComboSFX):
            1 => GuardDestroyedPlayerA.Play(Agent)
            2 => GuardDestroyedPlayerC.Play(Agent)
            3 => GuardDestroyedPlayerEm.Play(Agent)
            4 => GuardDestroyedPlayerGm.Play(Agent)
            _ => GuardDestroyedPlayerA.Play(Agent)

    PlayUnlockZoneSFX(Agent:agent):void=
        UnlockZonePlayer.Play(Agent)
        
    PlayPassiveIncomeCollectedSFX(Agent:agent):void=
        PassiveIncomeCollectedPlayer.Play(Agent)

    PlayCollectGemSFX(Agent:agent):void=
        CollectGemPlayer.Play(Agent)

    PlayNewTaskSFX(Agent:agent):void=
        NewTaskPlayer.Play(Agent)
            
    PlayNarratorWelcomeSFX(Agent:agent):void=
        NarratorWelcomePlayer.Play(Agent)

    PlayNarratorPrepareToBeginSFX(Agent:agent):void=
        NarratorPrepareToBeginPlayer.Play(Agent)
        
    PlayBoss1WarpSFX(Agent:agent):void=
        Boss1WarpSound.Play(Agent)
    
    PlayUpgradeSuccessSFX(Agent:agent):void=
        UpgradeSuccessSound.Play(Agent)

    PlayUpgradeFailSFX(Agent:agent):void=
        UpgradeFailSound.Play(Agent)

    PlayDropCandySFX(Agent:agent):void=
        DropCandySound.Play(Agent)
    
    PlayPickCandySFX(Agent:agent):void=
        PickCandySound.Play(Agent)
    
    PlayEatCandySFX(Agent:agent):void=
        EatCandySound.Play(Agent)

    PlayCollectCoinsSFX(Agent:agent):void=
        CollectCoinsSound.Play(Agent)

    PlaySellCoinsSFX(Agent:agent):void=
        SellCoinsSound.Play(Agent)

    PlayDropCoinsSFX(Agent:agent):void=
        DropCoinsSound.Play(Agent)

    GetMusicForZone(Agent:agent, ZoneTag : fetchable_tag):audio_player_wrapper=
        case(ZoneTag):
            fetchable_tag.towntag => GoodtownMusic
            fetchable_tag.range_zonetag => GoodtownMusic
            fetchable_tag.afk_zonetag => GoodtownMusic

            fetchable_tag.forest_zone1tag => UmbralForestMusic
            fetchable_tag.forest_zone2tag => UmbralForestMusic
            fetchable_tag.forest_zone3tag => UmbralForestMusic
            fetchable_tag.forest_zone4tag => UmbralForestMusic

            fetchable_tag.forest_bandit_zone1tag => ForestBanditMusic
            fetchable_tag.forest_bandit_zone2tag => ForestBanditMusic
            fetchable_tag.forest_bandit_zone3tag => ForestBanditMusic

            fetchable_tag.giantlands_zone2tag => GiantlandsMusic
            fetchable_tag.giantlands_zone3tag => GiantlandsMusic
            fetchable_tag.giantlands_zone4tag => GiantlandsMusic
            fetchable_tag.giantlands_zone5tag => GiantlandsMusic
            fetchable_tag.giantlands_zone6tag => GiantlandsMusic

            fetchable_tag.beach_zone1tag => BeachMusic
            fetchable_tag.beach_zone2tag => BeachMusic
            fetchable_tag.beach_zone3tag => BeachMusic

            fetchable_tag.canyon_zone1tag => CanyonMusic
            fetchable_tag.canyon_zone2tag => CanyonMusic
            fetchable_tag.canyon_zone3tag => CanyonMusic
            fetchable_tag.canyon_zone4tag => CanyonMusic
            fetchable_tag.canyon_zone5tag => CanyonMusic
            fetchable_tag.canyon_zone6tag => CanyonMusic
    
            fetchable_tag.zone3tag => BossMusic

            _ => MainGameMusic

        