using { /Fortnite.com/Devices }
using { /UnrealEngine.com/Temporary/SpatialMath }
using { /Verse.org/Simulation }

automatic_door_device := class(creative_device):
    @editable 
    var left_door : creative_prop = creative_prop{}
    
    @editable 
    var right_door : creative_prop = creative_prop{}
    
    @editable 
    var door_trigger :mutator_zone_device = mutator_zone_device{}

    var left_original_scale : transform = transform{}
    var right_original_scale : transform = transform{}
    var players_present : int = 0

    var door_speed : float = 1.0
    
    OnEnteredTriggered(Agent : agent):void=
        set players_present = players_present + 1
        spawn:
            OpenDoors()
    
    OnExitTriggered(Agent : agent):void=
        set players_present = players_present - 1
        if(players_present = 0):
            spawn:
                CloseDoors()
    
    OnBegin<override>()<suspends>:void=
        # Log("Door Begin")
        door_trigger.AgentEntersEvent.Subscribe(OnEnteredTriggered)
        door_trigger.AgentExitsEvent.Subscribe(OnExitTriggered)

        if(left_door.IsValid[]):
            set left_original_scale = left_door.GetTransform() 
        if(right_door.IsValid[]):
            set right_original_scale = right_door.GetTransform()
        # Log("Door End") 
    
    OpenDoors()<suspends>:void=
        var left_new_scale : transform = left_original_scale
        set left_new_scale.Scale.X = left_new_scale.Scale.X * (0.5) 

        var right_new_scale : transform = right_original_scale
        set right_new_scale.Scale.X = right_new_scale.Scale.X * (0.5) 

        spawn:
            left_door.MoveTo(left_new_scale, door_speed)

        spawn:
            right_door.MoveTo(right_new_scale, door_speed)
    
    CloseDoors()<suspends>:void=
        spawn:
            left_door.MoveTo(left_original_scale, door_speed)

        spawn:
            right_door.MoveTo(right_original_scale, door_speed)
