using { /Verse.org/Simulation }
using { /Verse.org/Random }

#################################
# We can represent very big numbers as a struct containing both a value and an exponent
#################################
big_number := struct <persistable><concrete><public>:
    @editable
    Value <public>: float = 0.0
    @editable
    Exponent <public>: int = 0

#################################
# Below are various functions to operate on big numbers
#################################

MakeBigNumber(Float:float,Power:float)<transacts>:big_number=
    if (Float < 0.0):
        PositivePart := MakeBigNumber(-1.0 * Float, Power)
        return big_number{Value:=-1.0 * PositivePart.Value, Exponent:=PositivePart.Exponent}

    if (Float = 0.0):
        return big_number{}

    if (Float = Inf):
        Log("BigNumber value was INFINITY! This shouldn't happen!!")
        return big_number{}

    PowerTen : float = Log(10.0, Float) * Power 
    var PowerTenFloor : int = 1
    if:
        Temp:=Floor[PowerTen]
        set PowerTenFloor = Temp
    FloatValue : float = Pow(10.0, PowerTen - PowerTenFloor * 1.0 )
    return big_number{Value:=FloatValue, Exponent:=PowerTenFloor}

MakeBigNumber(Float:float,Power:int)<transacts>:big_number=
    if (Float < 0.0):
        PositivePart := MakeBigNumber(-1.0 * Float, Power)
        return big_number{Value:=-1.0 * PositivePart.Value, Exponent:=PositivePart.Exponent}

    if (Float = 0.0):
        return big_number{}

    if (Float = Inf):
        Log("BigNumber value was INFINITY! This shouldn't happen!!")
        return big_number{}

    PowerTen : float = Log(10.0, Float) * Power 
    var PowerTenFloor : int = 1
    if:
        Temp:=Floor[PowerTen]
        set PowerTenFloor = Temp
    FloatValue : float = Pow(10.0, PowerTen - PowerTenFloor * 1.0 )
    return big_number{Value:=FloatValue, Exponent:=PowerTenFloor}

# Operator overloads
operator'+'(Left:big_number, Right:big_number)<transacts>:big_number = BigNumber_Add(Left, Right)
operator'-'(Left:big_number, Right:big_number)<transacts>:big_number = BigNumber_Subtract(Left, Right)
operator'*'(Left:big_number, Right:big_number)<transacts>:big_number = BigNumber_Multiply(Left, Right)
operator'/'(Left:big_number, Right:big_number)<transacts>:big_number = BigNumber_Divide(Left, Right)

operator'*'(Left:float, Right:big_number)<transacts>:big_number = BigNumber_Multiply(Right, Left)
operator'*'(Left:big_number, Right:float)<transacts>:big_number = BigNumber_Multiply(Left, Right)
operator'/'(Left:float, Right:big_number)<transacts>:big_number = BigNumber_Divide(MakeBigNumber(Left,1), Right)
operator'/'(Left:big_number, Right:float)<transacts>:big_number = BigNumber_Divide(Left, MakeBigNumber(Right,1))
operator'+'(Left:float, Right:big_number)<transacts>:big_number = BigNumber_Add(MakeBigNumber(Left,1), Right)
operator'+'(Left:big_number, Right:float)<transacts>:big_number = BigNumber_Add(Left, MakeBigNumber(Right,1))
operator'-'(Left:float, Right:big_number)<transacts>:big_number = BigNumber_Subtract(MakeBigNumber(Left,1), Right)
operator'-'(Left:big_number, Right:float)<transacts>:big_number = BigNumber_Subtract(Left, MakeBigNumber(Right,1))

ToString(bn : big_number)<transacts> : []char = FormatBigNumber(bn)

# Adds two big_numbers together
BigNumber_Add(NumberA : big_number, NumberB : big_number)<transacts>:big_number=
    CommonExponent : int = Max(NumberA.Exponent, NumberB.Exponent)

    NewValue : float =  (NumberA.Value * Pow(10.0, (NumberA.Exponent - CommonExponent) * 1.0)) + 
                        (NumberB.Value * Pow(10.0, (NumberB.Exponent - CommonExponent) * 1.0))

    BigNumber_Normalize(big_number{Value := NewValue, Exponent := CommonExponent})

# Subtracts a big_number from another
BigNumber_Subtract(NumberA : big_number, NumberB : big_number)<transacts>:big_number=
    CommonExponent : int = Max(NumberA.Exponent, NumberB.Exponent)

    NewValue : float =  (NumberA.Value * Pow(10.0, (NumberA.Exponent - CommonExponent) * 1.0)) - 
                        (NumberB.Value * Pow(10.0, (NumberB.Exponent - CommonExponent) * 1.0))

    BigNumber_Normalize(big_number{Value := NewValue, Exponent := CommonExponent})

# Multiply two big_numbers together
BigNumber_Multiply(NumberA : big_number, NumberB : big_number)<transacts>:big_number=
    NewValue := NumberA.Value * NumberB.Value
    NewExponent := NumberA.Exponent + NumberB.Exponent

    BigNumber_Normalize(big_number{Value := NewValue, Exponent := NewExponent})

# Divide two big_numbers together
BigNumber_Divide(NumberA : big_number, NumberB : big_number)<transacts>:big_number=
    NewValue := NumberA.Value / NumberB.Value
    NewExponent := NumberA.Exponent - NumberB.Exponent

    BigNumber_Normalize(big_number{Value := NewValue, Exponent := NewExponent})

BigNumber_Percentage(InNumerator:big_number, InDivisor:big_number)<transacts>: float =
    Numerator := BigNumber_Normalize(InNumerator)
    Divisor := BigNumber_Normalize(InDivisor)
    
    var Percent : float = 0.001
    if(Numerator.Value <= 0.0):
        set Percent = 0.0
        return Percent
    if(Numerator.Exponent = Divisor.Exponent):
        set Percent = Numerator.Value / Divisor.Value
        return Percent
    if(Numerator.Exponent = Divisor.Exponent - 1):
        set Percent = Numerator.Value / (Divisor.Value * 10.0)
        return Percent
    if(Numerator.Exponent = Divisor.Exponent - 2):
        set Percent = Numerator.Value / (Divisor.Value * 100.0)
        return Percent
    return Percent

BigNumber_ToFloat(Number : big_number)<transacts>:float=
    return Number.Value * Pow(10.0, Number.Exponent * 1.0)

# Multiply a big_number with a float
BigNumber_Multiply(NumberA : big_number, NumberB : float)<transacts>:big_number=
    NewValue := NumberA.Value * NumberB
    NewExponent := NumberA.Exponent

    BigNumber_Normalize(big_number{Value := NewValue, Exponent := NewExponent})

# Tells you if one big_number is greater than another
BigNumber_Equals(TempA : big_number, TempB : big_number)<transacts>:logic=
    NumberA := BigNumber_Normalize(TempA)
    NumberB := BigNumber_Normalize(TempB)

    if:
        (NumberA.Exponent = NumberB.Exponent) and (NumberA.Value = NumberB.Value)
    then:
        return true

    return false


# Tells you if one big_number is greater than another
BigNumber_GreaterThanOrEquals(TempA : big_number, TempB : big_number)<transacts>:logic=
    NumberA := BigNumber_Normalize(TempA)
    NumberB := BigNumber_Normalize(TempB)

    if:
        (NumberA.Exponent > NumberB.Exponent) or (NumberA.Exponent = NumberB.Exponent and NumberA.Value >= NumberB.Value) or (NumberA.Value > 0.0 and NumberB.Value <= 0.0)
    then:
        return true

    return false

# Tells you how much it would cost to buy X items
#   Start is how many items you currently own, and count is how many you would like to purchase (used for price scaling)
#   CommonRatio refers to the scaling in price for subsequent purchases.  If set to 1.0, no price scaling occurs.  
BigNumber_CalculateCostRange(BaseCost : big_number, Start : int, Count : int, CommonRatio : float):big_number=
    term1 := Pow(CommonRatio, (Start + Count)*1.0);
    term2 := Pow(CommonRatio, Start * 1.0);

    TotalCost_Value := BaseCost.Value * (term1 - term2) / (CommonRatio - 1.0);
    TotalCost_Exponent := BaseCost.Exponent;

    BigNumber_Normalize(big_number{Value:=TotalCost_Value, Exponent:=TotalCost_Exponent});


# Tells you how many items you could afford to purchase
#   CommonRatio refers to the scaling in price for subsequent purchases.  If set to 1.0, no price scaling occurs.  
BigNumber_CalculateAffordableItems(CurrentMoney : big_number, BaseCost : big_number, AmountOwned : int, CommonRatio : float):int=
    var AffordableItems : int = 1
    loop:
        Affordable := BigNumber_GreaterThanOrEquals(CurrentMoney, BigNumber_CalculateCostRange(BaseCost, AmountOwned, AffordableItems, CommonRatio))
        if:
            not Affordable?
        then:
            break
        
        set AffordableItems += 1

    return AffordableItems - 1


# Formats a big_number into a string
BigNumber_FormatDisplay(ValueIN : big_number)<transacts>:string=
    if (ValueIN.Value = 0):
        return "0"

    Normalized := BigNumber_Normalize(ValueIN)
    #This isn't a game where we care about decimal points.
    if(Normalized.Exponent < -3):
        return "0"
    
    #Actually it is sometimes
    if(Normalized.Exponent < 0):
        var Result : string = "0."
        for (i := 2..-Normalized.Exponent):
            set Result += "0"
        if:
            Dec:="{Normalized.Value}"
            NoDec := Dec.Slice[0, Dec.Find["."[0]]]
        then:
            return Result + NoDec

    CorrectedNumber := BigNumber_NormalizeToNearestNumberSet(Normalized)

    if:
        CorrectedNumber.Exponent < 3
        ValueAsInt := Floor[CorrectedNumber.Value]
    then:
        return "{ValueAsInt}"
    
    var ValueAsString : string = "{CorrectedNumber.Value}"

    var TargetLength : int = 4
    if(CorrectedNumber.Value >= 100.0):
        set TargetLength = 3

    if:
        Value:= ValueAsString.Remove[TargetLength, ValueAsString.Length]
        set ValueAsString = Value
    
    if:
        SuffixIndex : int = Quotient[CorrectedNumber.Exponent, 3]
        set ValueAsString += NumberSuffixes[SuffixIndex]      

    ValueAsString


NumberSuffixes : []string = array{""," K"," M"," B"," T"," Qd"," Qin"," Sxt"," Spt"," Oct"," Nov"," Dec",
                                    " UD" , " DD" , " TD" , " QD" , " QiD" , " SD" , " SpD" , " OD" , " ND" , " Vig" ,
                                    " UV" , " DV" , " TV" , " QV" , " QiV" , " SV" , " SpV" , " OV" , " NV" , " Ti" ,
                                    " UT" , " DT" , " TT" , " QT" , " QiT" , " ST" , " SpT" , " OT" , " NT" , " Qua" ,
                                    " UQ" , " DQ" , " TQ" , " QQ" , " QiQ" , " SQ" , " SpQ" , " OQ" , " NQ" , " Qui" ,
                                    " UQi" , " DQi" , " TQi" , " QQi" , " QiQi" , " SQi" , " SpQi" , " OQi" , " NQi" , " Sx" ,
                                    " US" , " DS" , " TS" , " QS" , " QiS" , " SS" , " SpS" , " OS" , " NS" , " Spg" ,
                                    " USp" , " DSp" , " TSp" , " QSp" , " QiSp" , " SSp" , " SpSp" , " OSp" , " NSp" , " Ocg" ,
                                    " UO" , " DO" , " TO" , " QO" , " QiO" , " SO" , " SpO" , " OO" , " NO" , " Non" ,
                                    " UN" , " D" , " TN" , " QN" , " QiN" , " SN" , " SpN" , " ON" , " NN" , " Cen" , 
                                    #Centillion
                                    " UC" , " DC" , " TC" , " QC" , " QiC" , " SC" , " SpC" , " OC" , " NC" , " DecC" ,
                                    " UDC" , " DDC" , " TDC" , " QDC" , " QiDC" , " SDC" , " SpDC" , " ODC" , " NDC" , " VigC",
                                    " UVC" , " DVC" , " TVC" , " QVC" , " QiVC" , " SVC" , " SpVC" , " OVC" , " NVC" , " TriC",  
                                    " UTC" , " DTC" , " TTC" , " QTC" , " QiTC" , " STC" , " SpTC" , " OTC" , " NTC" , " QC",
                                    " UQC" , " DQC" , " TQC" , " QQC" , " QiQC" , " STQ" , " SpQC" , " OQC" , " NQC" , " QiC",
                                    " UQiC" , " DQiC" , " TQiC" , " QQiC" , " QiQiC" , " STQ" , " SpQiC" , " OQiC" , " NQiC" , " SC", 
                                    " USC" , " DSC" , " TSC" , " QSC" , " QiSC" , " STQ" , " SpSC" , " OSC" , " NSC" , " SepC",
                                    " USpC" , " DSpC" , " TSpC" , " QSpC" , " QiSpC" , " STQ" , " SpSpC" , " OSpC" , " NSpC" , " OctC",
                                    " UOC" , " DOC" , " TOC" , " QOC" , " QiOC" , " STQ" , " SpOC" , " OOC" , " NOC" , " NonC",
                                    " UNC" , " DNC" , " TNC" , " QNC" , " QiNC" , " STQ" , " SpNC" , " ONC" , " NNC" , " Duc" ,
                                    #Ducentillion
                                    " UD" , " DD" , " TD" , " QD" , " QiD" , " SD" , " SpD" , " OD" , " ND" , " DecD" ,
                                    " UDD" , " DDD" , " TDD" , " QDD" , " QiDD" , " SDD" , " SpDD" , " ODD" , " NDD" , " VigD",
                                    " UVD" , " DVD" , " TVD" , " QVD" , " QiVD" , " SVD" , " SpVD" , " OVD" , " NVD" , " TriD",  
                                    " UTD" , " DTD" , " TTD" , " QTD" , " QiTD" , " STD" , " SpTD" , " OTD" , " NTD" , " QD",
                                    " UQD" , " DQD" , " TQD" , " QQD" , " QiQD" , " SQD" , " SpQD" , " OQD" , " NQD" , " QiD",
                                    " UQiD" , " DQiD" , " TQiD" , " QQiD" , " QiQiD" , " SQiD" , " SpQiD" , " OQiD" , " NQiD" , " SD", 
                                    " USD" , " DSD" , " TSD" , " QSD" , " QiSD" , " SSD" , " SpSD" , " OSD" , " NSD" , " SepD",
                                    " USpD" , " DSpD" , " TSpD" , " QSpD" , " QiSpD" , " SSpD" , " SpSpD" , " OSpD" , " NSpD" , " OctD",
                                    " UOD" , " DOD" , " TOD" , " QOD" , " QiOD" , " SOD" , " SpOD" , " OOD" , " NOD" , " NonD",
                                    " UND" , " DND" , " TND" , " QND" , " QiND" , " SND" , " SpND" , " OND" , " NND" , " Trc",
                                    #Trecentillion
                                    " UT" , " DT" , " TT" , " QT" , " QiT" , " ST" , " SpT" , " OT" , " NT" , " DecT" ,
                                    " UDT" , " DDT" , " TDT" , " QDT" , " QiDT" , " SDT" , " SpDT" , " ODT" , " NDT" , " VigT",
                                    " UVT" , " DVT" , " TVT" , " QVT" , " QiVT" , " SVT" , " SpVT" , " OVT" , " NVT" , " TriT",  
                                    " UTT" , " DTT" , " TTT" , " QTT" , " QiTT" , " STT" , " SpTT" , " OTT" , " NTT" , " QT",
                                    " UQT" , " DQT" , " TQT" , " QQT" , " QiQT" , " SQT" , " SpQT" , " OQT" , " NQT" , " QiT",
                                    " UQiT" , " DQiT" , " TQiT" , " QQiT" , " QiQiT" , " SQiT" , " SpQiT" , " OQiT" , " NQiT" , " ST", 
                                    " UST" , " DST" , " TST" , " QST" , " QiST" , " SST" , " SpST" , " OST" , " NST" , " SepT",
                                    " USpT" , " DSpT" , " TSpT" , " QSpT" , " QiSpT" , " SSpT" , " SpSpT" , " OSpT" , " NSpT" , " OctT",
                                    " UOT" , " DOT" , " TOT" , " QOT" , " QiOT" , " SOT" , " SpOT" , " OOT" , " NOT" , " NonT",
                                    " UNT" , " DNT" , " TNT" , " QNT" , " QiNT" , " STT" , " SpNT" , " ONT" , " NNT" , " Qdr",
                                    #Quadringentillion
                                    " UQ" , " DQ" , " TQ" , " QQ" , " QiQ" , " SQ" , " SpQ" , " OQ" , " NQ" , " DecQ" ,
                                    " UDQ" , " DDQ" , " TDQ" , " QDQ" , " QiDQ" , " SDQ" , " SpDQ" , " ODQ" , " NDQ" , " VigQ",
                                    " UVQ" , " DVQ" , " TVQ" , " QVQ" , " QiVQ" , " SVQ" , " SpVQ" , " OVQ" , " NVQ" , " TriQ",  
                                    " UTQ" , " DTQ" , " TTQ" , " QTQ" , " QiTQ" , " STQ" , " SpTQ" , " OTQ" , " NTQ" , " QQ",
                                    " UQQ" , " DQQ" , " TQQ" , " QQQ" , " QiQQ" , " SQQ" , " SpQQ" , " OQQ" , " NQQ" , " QiQ",
                                    " UQiQ" , " DQiQ" , " TQiQ" , " QQiQ" , " QiQiQ" , " SQiQ" , " SpQiQ" , " OQiQ" , " NQiQ" , " SQ", 
                                    " USQ" , " DSQ" , " TSQ" , " QSQ" , " QiSQ" , " SSQ" , " SpSQ" , " OSQ" , " NSQ" , " SepQ",
                                    " USpQ" , " DSpQ" , " TSpQ" , " QSpQ" , " QiSpQ" , " SSpQ" , " SpSpQ" , " OSpQ" , " NSpQ" , " OctQ",
                                    " UOQ" , " DOQ" , " TOQ" , " QOQ" , " QiOQ" , " SOQ" , " SpOQ" , " OOQ" , " NOQ" , " NonQ",
                                    " UNQ" , " DNQ" , " TNQ" , " QNQ" , " QiNQ" , " STQ" , " SpNQ" , " ONQ" , " NNQ" , " Qig",
                                    #Quingentillion
                                    " UQi" , " DQi" , " TQi" , " QQi" , " QiQi" , " SQi" , " SpQi" , " OQi" , " NQi" , " DecQi" ,
                                    " UDQi" , " DDQi" , " TDQi" , " QDQi" , " QiDQi" , " SDQi" , " SpDQi" , " ODQi" , " NDQi" , " VigQi",
                                    " UVQi" , " DVQi" , " TVQi" , " QVQi" , " QiVQi" , " SVQi" , " SpVQi" , " OVQi" , " NVQi" , " TriQi",  
                                    " UTQi" , " DTQi" , " TTQi" , " QTQi" , " QiTQi" , " STQi" , " SpTQi" , " OTQi" , " NTQi" , " QQi",
                                    " UQQi" , " DQQi" , " TQQi" , " QQQi" , " QiQQi" , " SQQi" , " SpQQi" , " OQQi" , " NQQi" , " QiQi",
                                    " UQiQi" , " DQiQi" , " TQiQi" , " QQiQi" , " QiQiQi" , " SQiQi" , " SpQiQi" , " OQiQi" , " NQiQi" , " SQi", 
                                    " USQi" , " DSQi" , " TSQi" , " QSQi" , " QiSQi" , " SSQi" , " SpSQi" , " OSQi" , " NSQi" , " SepQi",
                                    " USpQi" , " DSpQi" , " TSpQi" , " QSpQi" , " QiSpQi" , " SSpQi" , " SpSpQi" , " OSpQi" , " NSpQi" , " OctQi",
                                    " UOQi" , " DOQi" , " TOQi" , " QOQi" , " QiOQi" , " SOQi" , " SpOQi" , " OOQi" , " NOQi" , " NonQi",
                                    " UNQi" , " DNQi" , " TNQi" , " QNQi" , " QiNQi" , " STQi" , " SpNQi" , " ONQi" , " NNQi" , " Qig"}
                                    


# Takes a big_number, and converts it to ensure a single-digit whole-number component for the Value field
#   Ex)  14.2 * 10^2  will become 1.42 * 10^3
BigNumber_Normalize(Number : big_number)<transacts>:big_number=
    var NewExponent : int = Number.Exponent
    var NewValue : float = Number.Value

    if (NewValue < 0.0):
        PositivePart := BigNumber_Normalize_Internal(big_number{Value:=-1.0 * NewValue, Exponent:=NewExponent})
        return big_number{Value:=-1.0 * PositivePart.Value, Exponent:=PositivePart.Exponent}

    if (NewValue = 0.0):
        return big_number{}

    if (NewValue = Inf):
        Log("BigNumber value was INFINITY! This shouldn't happen!!")
        return big_number{}
        
    return BigNumber_Normalize_Internal(Number)

#To avoid recursion in BigNumber_Normalize, this is a separate function
BigNumber_Normalize_Internal(Number : big_number)<transacts>:big_number=
    var NewExponent : int = Number.Exponent
    var NewValue : float = Number.Value
    loop:
        if (NewValue < 10.0):
            break
        set NewValue *= 0.1
        set NewExponent += 1
    
    loop:
        if (NewValue >= 1.0):
            break
        set NewValue *= 10.0
        set NewExponent -= 1  

    return big_number:
        Value := NewValue
        Exponent := NewExponent


# Takes a big_number, and converts it to ensure a 1,2 or 3 digit whole-number component for the Value field.  This is used for string formatting.
#   Ex)  1.42K or 14.2K or 142K are all valid
#        1420k is not valid, and should instead be 1.42M
BigNumber_NormalizeToNearestNumberSet(Number : big_number)<transacts>:big_number=
    var NewExponent : int = Number.Exponent
    var NewValue : float = Number.Value

    loop:
        if (Mod[NewExponent, 3] = 0):
            break

        set NewValue *= 10.0
        set NewExponent -= 1

    Result := big_number:
        Value := NewValue
        Exponent := NewExponent

FormatBigNumber(BigNumber:big_number)<transacts>:string=
    return BigNumber_FormatDisplay(BigNumber)
    
GetRandomBigNumber(MinVal: big_number, MaxVal : big_number)<transacts>:big_number= 
    if (MinVal.Exponent = MaxVal.Exponent):
        return big_number{Value := GetRandomFloat(MinVal.Value, MaxVal.Value), Exponent := MinVal.Exponent}

    TargetExp := GetRandomInt(MinVal.Exponent, MaxVal.Exponent)
    if (TargetExp = MinVal.Exponent):
        return big_number{Value := GetRandomFloat(MinVal.Value, 1.0), Exponent := TargetExp}
    else if (TargetExp = MaxVal.Exponent):
        return big_number{Value := GetRandomFloat(0.1, MaxVal.Value), Exponent := TargetExp}
    else:
        return big_number{Value := GetRandomFloat(0.1, 1.0), Exponent := TargetExp}
    