using { /Verse.org/Simulation }

big_rock_data := struct <persistable><concrete>:
    @editable
    ID: int = 0