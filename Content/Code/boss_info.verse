
using { /Fortnite.com/Devices }
using { /Verse.org/Simulation }
using { /UnrealEngine.com/Temporary/Diagnostics }
using { /Verse.org/Assets }
using { /UnrealEngine.com/Temporary/SpatialMath }
using { /Verse.org/Random }
using { /Fortnite.com/AI }
using { /Fortnite.com/Characters }
using { /UnrealEngine.com/Assets }

boss_enum := enum:
    Sharky
    Genga
    Urjent
    Giant
    Jenkins
    PirateKing
    LyraTutorial
    None
    
BossList : []boss_enum := array{boss_enum.LyraTutorial, boss_enum.Sharky, boss_enum.Genga, boss_enum.Urjent, boss_enum.Giant, boss_enum.PirateKing}

BossTypeFromID(ID : int)<transacts> : boss_enum = case(ID):
    0 => boss_enum.LyraTutorial
    1 => boss_enum.Sharky
    2 => boss_enum.Genga
    3 => boss_enum.Urjent
    5 => boss_enum.PirateKing
    _ => boss_enum.None

(Boss : boss_enum).GetBossID()<transacts> : int = case(Boss):
    boss_enum.LyraTutorial => 0
    boss_enum.Sharky => 1
    boss_enum.Genga => 2
    boss_enum.Urjent => 3
    boss_enum.PirateKing => 5
    _ => 0

(Boss : boss_enum).GetBossImage()<transacts> : texture = case(Boss):
    boss_enum.LyraTutorial => Images.SargeIcon
    boss_enum.Sharky => Images.Sharkey
    boss_enum.Genga => Images.GengaIcon
    boss_enum.Urjent => Images.Urjent
    boss_enum.PirateKing => Images.Charon
    _ => Images.LyraIcon

(Boss : boss_enum).GetBossName()<transacts> : string = case(Boss):
    boss_enum.LyraTutorial => "Sarge"
    boss_enum.Sharky => "Sharkey"
    boss_enum.Genga => "Genga"
    boss_enum.Urjent => "Urjent"
    boss_enum.PirateKing => "PirateKing"
    _ => "..."

(Boss : boss_enum).KeyName()<transacts> : string = case(Boss):
    boss_enum.LyraTutorial => "KeySarge"
    boss_enum.Sharky => "KeySharkey"
    boss_enum.Genga => "KeyGenga"
    boss_enum.Urjent => "KeyUrjent"
    boss_enum.PirateKing => "KeyPirateKing"
    _ => "..."

(Boss : boss_enum).AchievementName()<transacts> : string = case(Boss):
    boss_enum.LyraTutorial => "Sarge_NextTier"
    boss_enum.Sharky => "Sharkey_NextTier"
    boss_enum.Genga => "Genga_NextTier"
    boss_enum.Urjent => "Urjent_NextTier"
    boss_enum.PirateKing => "PirateKing_NextTier"
    _ => "..."

(Boss : boss_enum).RewardStatName()<transacts> : string = case(Boss):
    boss_enum.Sharky => "Passive Income I"
    boss_enum.Genga => "Heartofthemountain"
    boss_enum.Urjent => "Urjent_CritMult"
    boss_enum.PirateKing => "Charon_DamageMult"
    _ => "---"

(Boss : boss_enum).RewardStatBaseValue()<transacts> : float = case(Boss):
    boss_enum.Sharky => 0.0
    boss_enum.Genga => 2000.0
    boss_enum.Urjent => 150.0
    boss_enum.PirateKing => 100.0
    _ => 0.0

(Boss: boss_enum).BossMaxHealth(Tier : int)<transacts> : big_number = case(Boss):
    boss_enum.LyraTutorial => big_number{Value := 1.0, Exponent := 5} * MakeBigNumber(1000.0, (Tier-1))
    boss_enum.Sharky => big_number{Value := 4.0, Exponent := 6} * MakeBigNumber(1000.0, (Tier-1))
    boss_enum.Genga => big_number{Value := 1.0, Exponent := 9} * MakeBigNumber(1000.0, (Tier-1))
    boss_enum.Urjent => big_number{Value := 8.0, Exponent := 11} * MakeBigNumber(1000.0, (Tier-1))
    boss_enum.PirateKing => big_number{Value := 2.0, Exponent := 15} * MakeBigNumber(1000.0, (Tier-1))
    _ => big_number{Value := 1.0, Exponent := 999}

(Boss: boss_enum).MinGearLevel(Tier : int)<transacts> : int = 
    (case(Boss):
        boss_enum.LyraTutorial => Tier * 12
        boss_enum.Sharky => Tier * 13 + 5
        boss_enum.Genga => Tier * 14 + 15
        boss_enum.Urjent => Tier * 15 + 30
        boss_enum.PirateKing => Tier * 16 + 50
        _ => 1
    ) + (Floor((Tier*Tier-1)/2) or 0)
    # case(Boss):
    #     boss_enum.LyraTutorial => Tier * 7
    #     boss_enum.Sharky => Tier * 8
    #     boss_enum.Genga => Tier * 9 + 4
    #     boss_enum.Urjent => Tier * 10 + 7
    #     boss_enum.PirateKing => Tier * 11 + 12
    #     _ => 1

(Boss: boss_enum).MaxGearLevel(TierIn : int)<transacts> : int = 
    Tier := TierIn + 1
    (case(Boss):
        boss_enum.LyraTutorial => Tier * 12
        boss_enum.Sharky => Tier * 13 + 5
        boss_enum.Genga => Tier * 14 + 15
        boss_enum.Urjent => Tier * 15 + 30
        boss_enum.PirateKing => Tier * 16 + 50
        _ => 1
    ) + (Floor((Tier*Tier-1)/2) or 0)

(Boss: boss_enum).DropLoot(GM : game_manager, Player : player, ActiveTier : int, ?SkipUpgrades : logic = false):void=

    GM.UpgradePopupDevice.OpenMenu(Player, ?Force := true)
    if:
        Menu := upgrade_level_popup[GM.UpgradePopupDevice.GetOpenMenu[Player]]
        PlayerManager := GM.PlayerManagers[Player]
    then:
        var ArmorDrops : []collectible_data = array{}
        for (i := 0..3):            
            MinLevel := Boss.MinGearLevel(ActiveTier)
            MaxLevel := Boss.MaxGearLevel(ActiveTier) + (PlayerManager.GetUpgradeLevel["P_BonusItemLevelDrops"] or 0)
            RarityUpgradeLevel := PlayerManager.GetUpgradeLevel["P_ItemRarityImprovement"] or 0
            Rarity := ArmorRarityFromID(SampleWeights(GetRarityWeights(RarityUpgradeLevel)))
            GeneratedArmor := GM.DropManager.ArmorGenerator.GenArmorForBoss(Boss, MinLevel, MaxLevel, Rarity) 
            Item := collectible_data{
                Name := "Gear"
                Key := ""
                Count := big_number{Value:=1.0, Exponent:=0}
                ArmorData := option{GeneratedArmor.ToData()}
            }
            set ArmorDrops = ArmorDrops + array{Item}
        set Menu.GearChoices = ArmorDrops
        set Menu.StartLevel = ActiveTier-1
        set Menu.BossType = Boss
        set Menu.EndLevel = ActiveTier
        Menu.Update()
        if (SkipUpgrades?):
            Menu.NextPopup()

(Boss : boss_enum).CheckHasKey(PM: player_manager, ?Spend : logic = false) : logic = 
    if (BigNumber_GreaterThanOrEquals(PM.GetCollectibleCount("BossKey"), big_number{Value:=1.0})?):
        if (Spend?):
            PM.GainCollectible(collectible_data{Key := "BossKey", Count := big_number{Value:=-1.0}})
        return true
    else if (BigNumber_GreaterThanOrEquals(PM.GetCollectibleCount(Boss.KeyName()), big_number{Value:=10.0})?):
        if (Spend?):
            PM.GainCollectible(collectible_data{Key := Boss.KeyName(), Count := big_number{Value:=-10.0}})
        return true
    return false

boss_info_device := class(creative_device):
    @editable:
    SargeSpawner : npc_spawner_device = npc_spawner_device{}
    @editable:
    SharkeySpawner : npc_spawner_device = npc_spawner_device{}
    @editable:
    GengaSpawner : npc_spawner_device = npc_spawner_device{}
    @editable:
    UrjentSpawner : npc_spawner_device = npc_spawner_device{}
    @editable:
    PirateKingSpawner : npc_spawner_device = npc_spawner_device{}

    @editable:
    SargeBehavior : sarge_behavior = sarge_behavior{}
    @editable:
    SharkyBehavior : sharky_behavior = sharky_behavior{}
    @editable:
    GengaBehavior : genga_behavior = genga_behavior{}
    @editable:
    UrjentBehavior : urjent_behavior = urjent_behavior{}
    @editable:
    PirateKingBehavior : pirateking_behavior = pirateking_behavior{}

    @editable
    SargeWaitingRoom : creative_device = creative_device{}
    @editable
    SharkyWaitingRoom : creative_device = creative_device{}
    @editable
    GengaWaitingRoom : creative_device = creative_device{}
    @editable
    UrjentWaitingRoom : creative_device = creative_device{}
    @editable
    PirateKingWaitingRoom : creative_device = creative_device{}


boss_behavior := class<abstract>:
    DoBehavior(GameManager:game_manager, NPC : boss_midas_behavior)<suspends>:void=
        block:

    CleanUp(GameManager:game_manager, NPC : boss_midas_behavior)<suspends>:void=
        block:
    
    ShowDialog(GameManager:game_manager, NPC: boss_midas_behavior, Dialog : dialog_message, Duration : float, ?VODelay : float = 0.0)<suspends>:void=

        PlayersToShow := NPC.GetPlayersInZone()

        var Count : int = 0
        Leader := PlayersToShow.Length

        for (Player : PlayersToShow; PlayerManager := GameManager.PlayerManagers[Player]; UIManager := PlayerManager.UIManager?):
            set Count += 1
            if (Count = Leader):
                UIManager.ShowNonBlockingDialog(Dialog, Duration, ?VODelay := VODelay)
            else:
                spawn{ UIManager.ShowNonBlockingDialog(Dialog, Duration, ?VODelay := VODelay) }

    WaitForThreshold(NPC : boss_midas_behavior, ?Period : float = 0.1)<suspends>:void=
        loop:
            if (NPC.IsAtPhaseThreshold()?):
                break
            Sleep(Period)

    InitGuards(Guards : []guard_manager, GameManager : game_manager, ZoneManager: zone_manager_device):void=
        var GuardCount : int = 0
        for (GM : Guards):
            GM.Init(GameManager, ZoneManager, ZoneManager.ZoneTag, ZoneManager.ZoneTag.ToZoneName()+"{GuardCount}")
            set GuardCount += 1

    SpawnGuards(Guards : []guard_manager, ?OverrideMaxHealth : ?big_number = false):void=
        for (GM : Guards):
            if (_OMH := OverrideMaxHealth?):
                set GM.MaxHealth = _OMH
            set GM.CurrentHealth = GM.MaxHealth

    DespawnGuards(Guards : []guard_manager)<suspends>:void=
        for (GM : Guards):
            GM.DespawnGuard()

    WaitForGuardsKilled(Guards : []guard_manager,  ?Period : float = 0.1)<suspends>:void=
        loop:
            
            var AnyAlive : logic = false
            for (GM : Guards):
                if (GM.IsAlive()?):
                    set AnyAlive = true
            
            if (not AnyAlive?):
                break
            
            Sleep(Period)

    CheckBossLife(NPC : boss_midas_behavior, Amount : big_number)<suspends>:void=
        loop:
            if (not NPC.IsAlive()?):
                break
            Sleep(1.0)


sarge_behavior := class<concrete>(boss_behavior):
    @editable:
    Pillar : creative_prop = creative_prop{}

    @editable:
    ZoneCenter : creative_prop = creative_prop{}

    @editable:
    ZoneWidth : float = 2000.0

    @editable:
    CanonballAsset : creative_prop_asset = DefaultCreativePropAsset

    @editable:
    BigCanonballAsset : creative_prop_asset = DefaultCreativePropAsset

    var OriginalPillarLocation : vector3 = vector3{}

    CleanUp<override>(GameManager:game_manager, NPC : boss_midas_behavior)<suspends>:void=
        if:
            Pillar.TeleportTo[OriginalPillarLocation, Pillar.GetTransform().Rotation]

    DoBehavior<override>(GameManager:game_manager, NPC : boss_midas_behavior)<suspends>:void=
        #Shield up
        NPC.StartInvulnerability()
        set NPC.PhaseThreshold = NPC.MaxHealth * 0.75
        
        #Start buffer
        Sleep(3.0)

        
        if (StepTarget := NPC.GetAgent[].GetFortCharacter[].GetTransform().Translation + vector3{X:=1.0, Y:=1.0, Z:=0.0}): 
            NPC.NavigateTo(StepTarget, ?MovementType := movement_types.Walking)

        
        #Pillar raises
        set OriginalPillarLocation = Pillar.GetTransform().Translation
        spawn:
            Pillar.MoveTo(OriginalPillarLocation + vector3{X:=0.0, Y:=0.0, Z:=400.0}, Pillar.GetTransform().Rotation, 4.0)
        
        NPC.EndInvulnerability()
        
        #DIALOG
        spawn:
            ShowDialog(GameManager, NPC, dialog_message{
                Title := option{"Sarge"}
                Message := "Welcome to Sarge's Bossfight Bootcamp! Yes, I talk about myself in third person. Now hit me with all you got!"
                VO :=  option{GameManager.AudioManager.SargeVO2a}
            }, 9.5)
                
        #Damage phase 
        loop:
            Sleep(0.1)
            if (NPC.IsAtPhaseThreshold()?):
                break
        

        #Shield up
        NPC.StartInvulnerability()
        set NPC.PhaseThreshold = NPC.MaxHealth * 0.25
        Sleep(0.5)

        #DIALOG
        spawn:
            ShowDialog(GameManager, NPC, dialog_message{
                    Title := option{"Sarge"}
                    Message := "Ahah! See this Glowing shield around me? Your guns can't touch me now. You'll have to survive this barrage!"
                    VO :=  option{GameManager.AudioManager.SargeVO2b}
                }, 6.9)
        
        #Dodge phase
        for (Count := 0..10):

            Slot := GetRandomInt(-2,2)
            var Side : int = GetRandomInt(2,3)
            if (Count < 5):
                set Side = 3

            OrthoOffset := Slot * 500.0
            SideOffset := ZoneWidth * 0.6

            StartOffset := case(Side):
                0 => vector3{X:= OrthoOffset, Y:= SideOffset,  Z:=0.0}  #U
                1 => vector3{X:= OrthoOffset, Y:=-SideOffset,  Z:=0.0}  #D
                2 => vector3{X:= SideOffset,  Y:= OrthoOffset+250.0, Z:=0.0}  #R
                3 => vector3{X:=-SideOffset,  Y:= OrthoOffset+250.0, Z:=0.0}  #L
                _ => vector3{X:=0.0, Y:=0.0, Z:=0.0}

            EndOffset := case(Side):
                0 => vector3{X:= OrthoOffset, Y:=-SideOffset,  Z:=0.0}  #U
                1 => vector3{X:= OrthoOffset, Y:= SideOffset,  Z:=0.0}  #D
                2 => vector3{X:=-SideOffset,  Y:= OrthoOffset+250.0, Z:=0.0}  #R
                3 => vector3{X:= SideOffset,  Y:= OrthoOffset+250.0, Z:=0.0}  #L
                _ => vector3{X:=0.0, Y:=0.0, Z:=0.0}

            StartPos := ZoneCenter.GetTransform().Translation + StartOffset + vector3{X:=0.0, Y:=0.0, Z:=230.0}

            SpawnProjectile(GameManager, BigCanonballAsset, StartPos, 265.0, (EndOffset-StartOffset) * 2.0/8.0, 8.0, 200.0, ?Explosion := option{VFX_PlayerHit})
            Sleep(0.75)

            if (Count = 4):
                spawn:
                    ShowDialog(GameManager, NPC, dialog_message{
                        Title := option{"Sarge"}
                        Message := "And do watch out for the attacks that come from behind - HAHA!"
                        VO :=  option{GameManager.AudioManager.SargeVO2c}
                    }, 5.5)

        #DIALOG
        Sleep(0.5)

        #Shield down
        NPC.EndInvulnerability()

        spawn:
            ShowDialog(GameManager, NPC, dialog_message{
                Title := option{"Sarge"}
                Message := "I've taken my shield down -- shoot me with everything you've got! Let's see those guns blazing!"
                VO :=  option{GameManager.AudioManager.SargeVO2d}
            }, 6.8)


        #Damage phase 
        loop:
            Sleep(0.1)
            if (NPC.IsAtPhaseThreshold()?):
                break
        

        #Shield up
        NPC.StartInvulnerability()

        #Dialog
        spawn:
            ShowDialog(GameManager, NPC, dialog_message{
                Title := option{"Sarge"}
                Message := "Not too shabby! Now let's see you dodge these!"
                VO :=  option{GameManager.AudioManager.SargeVO2e}
            }, 3.7)

        #Damage + dodge phase
        var TempCount : int = 0
        loop:
            Slot := GetRandomInt(-2,2)
            Side := GetRandomInt(2,3)

            OrthoOffset := Slot * 500.0
            SideOffset := ZoneWidth * 0.6

            StartOffset := case(Side):
                0 => vector3{X:= OrthoOffset, Y:= SideOffset,  Z:=0.0}  #U
                1 => vector3{X:= OrthoOffset, Y:=-SideOffset,  Z:=0.0}  #D
                2 => vector3{X:= SideOffset,  Y:= OrthoOffset+250.0, Z:=0.0}  #R
                3 => vector3{X:=-SideOffset,  Y:= OrthoOffset+250.0, Z:=0.0}  #L
                _ => vector3{X:=0.0, Y:=0.0, Z:=0.0}

            EndOffset := case(Side):
                0 => vector3{X:= OrthoOffset, Y:=-SideOffset,  Z:=0.0}  #U
                1 => vector3{X:= OrthoOffset, Y:= SideOffset,  Z:=0.0}  #D
                2 => vector3{X:=-SideOffset,  Y:= OrthoOffset+250.0, Z:=0.0}  #R
                3 => vector3{X:= SideOffset,  Y:= OrthoOffset+250.0, Z:=0.0}  #L
                _ => vector3{X:=0.0, Y:=0.0, Z:=0.0}

            StartPos := ZoneCenter.GetTransform().Translation + StartOffset + vector3{X:=0.0, Y:=0.0, Z:=230.0}

            SpawnProjectile(GameManager, BigCanonballAsset, StartPos, 250.0, (EndOffset-StartOffset) * 2.0/8.0, 8.0, 200.0, ?Explosion := option{VFX_PlayerHit}) 

            if (TempCount > 10 and NPC.IsAtPhaseThreshold()?):
                break
            Sleep(0.5)

            set TempCount = TempCount + 1

            if (TempCount = 10):
                #Shield down
                NPC.EndInvulnerability()
                set NPC.PhaseThreshold = big_number{}
                spawn:
                    ShowDialog(GameManager, NPC, dialog_message{
                        Title := option{"Sarge"}
                        Message := "Shield's are back down, but this time I'm keeping the attacks coming!"
                        VO :=  option{GameManager.AudioManager.SargeVO2l}
                    }, 4.5)   

        #Dialog
        Sleep(1.0)
        ShowDialog(GameManager, NPC, dialog_message{
            Title := option{"Sarge"}
            Message := "Well done! That concludes Sarge's Bossfight Bootcamp! Maybe we have a chance against Midas after all..."
            VO :=  option{GameManager.AudioManager.SargeVO2f}
        }, 9.0)

        #Lower pillar
        spawn:
            Pillar.MoveTo(OriginalPillarLocation, Pillar.GetTransform().Rotation, 4.0)

        #Dialog
        if(NPC.ActiveTier <= 1):
            ShowDialog(GameManager, NPC, dialog_message{
                Title := option{"Sarge"}
                Message := "Every time you beat a boss, you'll get to choose a piece of gear, so make sure to come back and fight bosses on higher difficulties whenever you can."
                VO :=  option{GameManager.AudioManager.SargeVO2g}
            }, 11.3)

        NPC.Die()

        #End buffer
        Sleep(1.0)
        NPC.BossDefeated()

sharky_behavior := class<concrete>(boss_behavior):
    @editable
    BossPlatform : creative_prop = creative_prop{}

    @editable:
    Platforms : []creative_prop = array{}

    @editable:
    ZoneWidth : float = 2000.0

    @editable:
    SharkProjectileAsset : creative_prop_asset = DefaultCreativePropAsset

    var OriginalBossPlatformPosition : vector3 = vector3{}
    var OriginalBossPlatformRotation : rotation = rotation{}
    var OriginalPlatformPositions : []vector3 = array{}

    CleanUp<override>(GameManager:game_manager, NPC : boss_midas_behavior)<suspends>:void=
        if:
            BossPlatform.TeleportTo[OriginalBossPlatformPosition, OriginalBossPlatformRotation]
        for (i := 0..(Platforms.Length-1)):
            if:
                Platform := Platforms[i]
                OriginalPosition := OriginalPlatformPositions[i]
                Platform.TeleportTo[OriginalPosition, Platform.GetTransform().Rotation]

    DoBehavior<override>(GameManager:game_manager, NPC : boss_midas_behavior)<suspends>:void=

        #Reset platform positions
        if (OriginalPlatformPositions.Length <= 0):
            set OriginalBossPlatformPosition = BossPlatform.GetTransform().Translation
            set OriginalBossPlatformRotation = BossPlatform.GetTransform().Rotation
            for (Platform : Platforms):
                set OriginalPlatformPositions += array{Platform.GetTransform().Translation}

            #Start buffer
        set NPC.PhaseThreshold = big_number{}
        Sleep(3.0)

        if (StepTarget := NPC.GetAgent[].GetFortCharacter[].GetTransform().Translation + vector3{X:=1.0, Y:=1.0, Z:=0.0}): 
            NPC.NavigateTo(StepTarget, ?MovementType := movement_types.Walking)

        #DIALOG
        ShowDialog(GameManager, NPC, dialog_message{
                Title := option{"Sharkey"}
                Message := "Hey! Nobody swims in my ocean but me! Get out!"
                VO :=  option{GameManager.AudioManager.SharkeyVO2a}
            }, 4.0)

        AllSlots := array{-3,-1,1,3}

        var RemainingRowsV : []int = AllSlots
        var RemainingRowsH : []int = AllSlots

        #Phase 1 - breaking the platforms
        var BreakCount : int = 0
        var HasBrokenAll : logic = false
        loop:
            if (NPC.IsAtPhaseThreshold()?):
                break

            Range := RemainingRowsV.Length + RemainingRowsH.Length
            if (Range = 0):
                set HasBrokenAll = true
                set RemainingRowsV = AllSlots
                set RemainingRowsH = AllSlots
                
                spawn{ SinkXPlatforms(NPC) } 

            # Choice := GetRandomInt(1, Range)
            # Horizontal := Choice > RemainingRowsV.Length

            var Horizontal : logic = false
            if (Mod[BreakCount, 2] = 0):
                set Horizontal = true

            var Side : int = GetRandomInt(0,1) * 2 - 1
            var Slot : int = 0

            var FlipX : float = 0.0
            var FlipY : float = 0.0

            var TX : float = 0.0
            var TY : float = 0.0

            if (Horizontal?): #X is side, Y is slot
                Choice := GetRandomInt(0, RemainingRowsH.Length-1)
                set Slot = RemainingRowsH[Choice] or 0
                if (not HasBrokenAll?):
                    set RemainingRowsH = RemainingRowsH.RemoveElement[Choice] or array{}
                # Slot = RemainingRowsH[Choice - RemainingRowsV.Length]
                set TX = Side * ZoneWidth * 0.7
                set TY = Slot * 500.0
                set FlipX = -1.0 * TX
            else:
                Choice := GetRandomInt(0, RemainingRowsV.Length-1)
                set Slot = RemainingRowsV[Choice] or 0
                if (not HasBrokenAll?):
                    set RemainingRowsV = RemainingRowsV.RemoveElement[Choice] or array{}
                # Slot = RemainingRowsV[Choice]
                set TY = Side * ZoneWidth * 0.7
                set TX = Slot * 500.0
                set FlipY = -1.0 * TY
            
            set BreakCount = BreakCount + 1

            set TX = TX + Platforms[24].GetTransform().Translation.X or 0.0
            set TY = TY + Platforms[24].GetTransform().Translation.Y or 0.0

            #Target selected, now move:
            BT := BossPlatform.GetTransform().Translation
            if (Horizontal?):
                D1 := Abs(TX - BT.X)
                D2 := Abs(TY - BT.Y)

                TargetRot := case (Side):
                    1 => MakeRotation(vector3{X:=0.0, Y:=0.0, Z:=1.0}, 180.0 / 57.2958)
                    -1 => MakeRotation(vector3{X:=0.0, Y:=0.0, Z:=1.0}, 0.0 / 57.2958)
                    _ => MakeRotation(vector3{X:=0.0, Y:=0.0, Z:=1.0}, 0.0 / 57.2958)

                BossPlatform.MoveTo(vector3{X:=TX, Y:=BT.Y, Z:=BT.Z}, BossPlatform.GetTransform().Rotation, D1/2000.0)
                BossPlatform.MoveTo(vector3{X:=TX, Y:=TY, Z:=BT.Z}, TargetRot, D2/2000.0)
            else:
                D1 := Abs(TY - BT.Y)
                D2 := Abs(TX - BT.X)

                TargetRot := case (Side):
                    1 => MakeRotation(vector3{X:=0.0, Y:=0.0, Z:=1.0}, 270.0 / 57.2958)
                    -1 => MakeRotation(vector3{X:=0.0, Y:=0.0, Z:=1.0}, 90.0 / 57.2958)
                    _ => MakeRotation(vector3{X:=0.0, Y:=0.0, Z:=1.0}, 0.0 / 57.2958)

                BossPlatform.MoveTo(vector3{X:=BT.X, Y:=TY, Z:=BT.Z}, BossPlatform.GetTransform().Rotation, D1/2000.0)
                BossPlatform.MoveTo(vector3{X:=TX, Y:=TY, Z:=BT.Z}, TargetRot, D2/2000.0)


            if (NPC.IsAtPhaseThreshold()?):
                break

            ProjectileMoveOffset := vector3{X:=FlipX*2.0, Y:=FlipY*2.0, Z:=0.0}
            StartPos := vector3{X:=TX, Y:=TY, Z:= BT.Z + 1100.0} + ProjectileMoveOffset*0.1

            #?Scale := vector3{X:=15.0, Y:=70.0, Z:=40.0}
            SpawnProjectile(GameManager, SharkProjectileAsset, StartPos, -750.0, ProjectileMoveOffset /3.5, 3.5, 200.0, ?Explosion := option{VFX_PlayerHit}, ?Scale := vector3{X:=1.0, Y:=1.0, Z:=1.0}, ?Rotation := BossPlatform.GetTransform().Rotation.ApplyYaw(270.0 /57.2958))

            #Break platforms:
            for (i := 0..6):
                var IX : int = 0
                var IY : int = 0

                #+slot -> -x index - horizontal
                #+slot -> +y index - vertical
                if (Horizontal?):
                    set IX = 3 + Slot
                    set IY = case (Side):
                        -1 => 6 - i
                        _ => i
                else:
                    set IX = case (Side):
                        -1 => i
                        _ => 6 - i
                    set IY = 3 - Slot

                if:
                    Platform : creative_prop = Platforms[IX + IY*7]
                then:
                    PT := Platform.GetTransform().Translation
                    spawn{ Platform.MoveTo(vector3{X:=PT.X, Y:=PT.Y, Z:=PT.Z - 1000.0}, Platform.GetTransform().Rotation, 1.0) } 
                    Sleep(0.5)

            if (NPC.IsAtPhaseThreshold()?):
                break
            Sleep(1.0)
            
        NPC.Die()
                
        ShowDialog(GameManager, NPC, dialog_message{
            Title := option{"Sharkey"}
            Message := "Nooo! You're not supposed to beat me! It's not fair! I'm a shaaaark!"
            VO :=  option{GameManager.AudioManager.SharkeyVO2b}
        }, 6.15)

        #End buffer
        NPC.BossDefeated()

    SinkXPlatforms(NPC : boss_midas_behavior)<suspends>:void=
        if:
            Platform : creative_prop = Platforms[24]
        then:
            PT := Platform.GetTransform().Translation
            SinkPlatform(NPC, Platform, 20.0)

        if:
            FourPlatforms := array{Platforms[8], Platforms[12], Platforms[36], Platforms[40]}
        then:
            for (Platform : FourPlatforms):
                spawn { SinkPlatform(NPC,Platform, 20.0) }

    SinkPlatform(NPC : boss_midas_behavior, Platform : creative_prop, Time: float)<suspends>:void=
        PT := Platform.GetTransform().Translation
        race:
            Platform.MoveTo(vector3{X:=PT.X, Y:=PT.Y, Z:=PT.Z - 1000.0}, Platform.GetTransform().Rotation, Time)
            CheckBossLife(NPC, big_number{})  



genga_behavior := class<concrete>(boss_behavior):
    @editable:
    RunTargets : []creative_device = array{}

    @editable:
    Attack1 : cinematic_sequence_device = cinematic_sequence_device{}
    @editable:
    Attack2 : cinematic_sequence_device = cinematic_sequence_device{}

    @editable
    DamageVolume1 : damage_volume_device = damage_volume_device{}
    @editable
    DamageVolume2 : damage_volume_device = damage_volume_device{}

    @editable:
    Guards : []guard_manager = array{}

    var RunTarget : int = 0

    CleanUp<override>(GameManager:game_manager, NPC : boss_midas_behavior)<suspends>:void=
        DespawnGuards(Guards)

    DoBehavior<override>(GameManager:game_manager, NPC : boss_midas_behavior)<suspends>:void=

        set NPC.PhaseThreshold = NPC.MaxHealth * 0.66
        Sleep(3.0)
        
        ShowDialog(GameManager, NPC, dialog_message{
            Title := option{"Genga"}
            Message := "I got those orbs from Lord Midas himself! You'll pay for breaking them!"
            VO :=  option{GameManager.AudioManager.GengaVO2a}
        }, 4.6)

        spawn {DoRunning(GameManager, NPC)}

        #Phase 1 - running
        WaitForThreshold(NPC)
        NPC.StartInvulnerability()
        set NPC.PhaseThreshold = NPC.MaxHealth * 0.33

        #Phase 1.5 - guards
        if (ZM := NPC.ZoneManager?):
            InitGuards(Guards, GameManager, ZM)
        SpawnGuards(Guards, ?OverrideMaxHealth := option{NPC.MaxHealth * 0.15})
        ShowDialog(GameManager, NPC, dialog_message{
            Title := option{"Genga"}
            Message := "Guards! To me! Wipe them out!"
            VO :=  option{GameManager.AudioManager.GengaVO2b}
        }, 2.75)
        WaitForGuardsKilled(Guards)
        NPC.EndInvulnerability()

        #Phase 2 - running 
        WaitForThreshold(NPC)
        NPC.StartInvulnerability()
        set NPC.PhaseThreshold = big_number{}

        #Phase 2.5 - guards
        SpawnGuards(Guards)
        ShowDialog(GameManager, NPC, dialog_message{
            Title := option{"Genga"}
            Message := "Guards! To me! Wipe them out!"
            VO :=  option{GameManager.AudioManager.GengaVO2b}
        }, 2.75)
        WaitForGuardsKilled(Guards)
        NPC.EndInvulnerability()
        
        #Phase 3 - running
        WaitForThreshold(NPC)

        NPC.Die()
        
        ShowDialog(GameManager, NPC, dialog_message{
            Title := option{"Genga"}
            Message := "You haven't seen the last of Genga, the bandit leader!"
            VO :=  option{GameManager.AudioManager.GengaVO2c}
        }, 3.5)
        NPC.BossDefeated()
    
    DoRunning(GameManager:game_manager, NPC : boss_midas_behavior)<suspends>:void=
        loop:
            if (RunTarget < 2):
                set RunTarget = GetRandomInt(2,3)
            else:
                set RunTarget = GetRandomInt(0,1)

            if:
                Character := NPC.GetAgent[].GetFortCharacter[]
                Navigatable := Character.GetNavigatable[]
                Focus := Character.GetFocusInterface[]
                RunTargetTranslation := RunTargets[RunTarget].GetTransform().Translation
            then:
                
                GoToPoint := vector3{X:=RunTargetTranslation.X, Y:=RunTargetTranslation.Y, Z:=Character.GetTransform().Translation.Z}

                NavTargetStart := MakeNavigationTarget(GoToPoint)
                var NavResultGoTo : navigation_result = navigation_result.Interrupted
                # Tell the navigation interface to walk to the location and store a navigation result for error checking.
                race:
                    set NavResultGoTo = Navigatable.NavigateTo(NavTargetStart, ?MovementType:=movement_types.Running)
                    Focus.MaintainFocus(GoToPoint)

            
                Sleep(1.0)
                if (RunTarget < 2):
                    Attack2.Play()
                    
                    Sleep(3.0)
                    DamageVolume1.Enable()
                    Sleep(0.3)
                    DamageVolume1.Disable()
                else:
                    Attack1.Play()
                    
                    Sleep(3.0)
                    DamageVolume2.Enable()
                    Sleep(0.3)
                    DamageVolume2.Disable()
            else:
                break





urjent_behavior := class<concrete>(boss_behavior):
    @editable:
    RunTargets : []creative_device = array{}

    @editable:
    ZoneCenter : creative_prop = creative_prop{}

    @editable:
    LavaWallCollision : creative_prop = creative_prop{}
    
    @editable:
    ProjectileAsset : creative_prop_asset = DefaultCreativePropAsset
    
    @editable:
    LavaAsset : creative_prop_asset = DefaultCreativePropAsset

    @editable:
    Guards : []guard_manager = array{}

    var TargetColumn : int = 0
    var Phase : int = 0

    var OriginalPosition : vector3 = vector3{}
    var OriginalLavaPosition : vector3 = vector3{}

    var ChargeVFX : ?cancelable = false

    CleanUp<override>(GameManager:game_manager, NPC : boss_midas_behavior)<suspends>:void=
        LavaWallCollision.MoveTo(OriginalLavaPosition, LavaWallCollision.GetTransform().Rotation, 0.1) 
        DespawnGuards(Guards)
        if (CancelCharge := ChargeVFX?):
            CancelCharge.Cancel()

    DoBehavior<override>(GameManager:game_manager, NPC : boss_midas_behavior)<suspends>:void=
        set Phase = 0
        set NPC.PhaseThreshold = NPC.MaxHealth * 0.66        
        
        spawn:
            ShowDialog(GameManager, NPC, dialog_message{
                Title := option{"Urjent"}
                Message := "Did you really think you can stop me? Genga and Sharkey have NO POWER compared to me!"
                VO :=  option{GameManager.AudioManager.UrjentVO2a}
            }, 9.0)
       
        if:
            Character := NPC.GetAgent[].GetFortCharacter[]
        then:
            set OriginalPosition = Character.GetTransform().Translation

        set OriginalLavaPosition = LavaWallCollision.GetTransform().Translation

        #Phase 1 - blasts + walking
        loop:
            var NewTargetColumn : int = GetRandomInt(-1, 1)
            loop:
                if (NewTargetColumn = TargetColumn):
                    set NewTargetColumn = GetRandomInt(-1, 1)
                else:
                    break
            
            set TargetColumn = NewTargetColumn
            

            TX := TargetColumn * 500.0 + 250.0

            if:
                Character := NPC.GetAgent[].GetFortCharacter[]
                Navigatable := Character.GetNavigatable[]
                Focus := Character.GetFocusInterface[]
                RunTargetTranslation := ZoneCenter.GetTransform().Translation + vector3{X:=0.0, Y:=TX, Z:=0.0}
            then:
                GoToPoint := vector3{X:=Character.GetTransform().Translation.X, Y:=RunTargetTranslation.Y, Z:=Character.GetTransform().Translation.Z}

                NavTargetStart := MakeNavigationTarget(GoToPoint)
                var NavResultGoTo : navigation_result = navigation_result.Interrupted
                
                race:
                    set NavResultGoTo = Navigatable.NavigateTo(NavTargetStart, ?MovementType:=movement_types.Walking)
                    Focus.MaintainFocus(GoToPoint)

                race:
                    Sleep(0.5)
                    Focus.MaintainFocus(Character.GetTransform().Translation + vector3{X:=1000.0, Y:=0.0, Z:=0.0})
                
                for (i := -1..1):

                    if (not i = TargetColumn):
                        StartPos := ZoneCenter.GetTransform().Translation + vector3{X:=-2000.0, Y:=i*500.0 + 250.0, Z:=250.0}
                        EndPos := ZoneCenter.GetTransform().Translation + vector3{X:=2000.0, Y:=i*500.0 + 250.0, Z:=250.0}
                        
                        SpawnProjectile(GameManager, ProjectileAsset, StartPos, 250.0, (EndPos-StartPos)/6.0, 6.0, 300.0, ?Explosion := option{VFX_PlayerHit})
            
                Sleep(1.0)
                
                if (NPC.IsAtPhaseThreshold()?):
                    #phase 2, add lava
                    if (Phase = 0):
                        set Phase = 1
                        LWT := LavaWallCollision.GetTransform().Translation
                        LavaWallCollision.MoveTo(OriginalLavaPosition + vector3{X:=-1200.0, Y:=0.0, Z:=0.0}, LavaWallCollision.GetTransform().Rotation, 2.0) 
                        
                        ShowDialog(GameManager, NPC, dialog_message{
                            Title := option{"Urjent"}
                            Message := "This should put a stop to all that annoying running around!"
                            VO :=  option{GameManager.AudioManager.UrjentVO2b}
                        }, 5.2)

                        set NPC.PhaseThreshold = NPC.MaxHealth * 0.33
                    else:
                        break
                
        #phase 3, dps check
        NPC.StartInvulnerability()
        set NPC.PhaseThreshold = big_number{}

        if (ZM := NPC.ZoneManager?):
            InitGuards(Guards, GameManager, ZM)
        SpawnGuards(Guards, ?OverrideMaxHealth := option{NPC.MaxHealth * 0.05})

        spawn:
            ShowDialog(GameManager, NPC, dialog_message{
                Title := option{"Urjent"}
                Message := "Arise my servants, destroy these intruders!"
                VO :=  option{GameManager.AudioManager.UrjentVO2g}
            }, 4.5)

        #Walk back to center
        if:
            Character := NPC.GetAgent[].GetFortCharacter[]
            Navigatable := Character.GetNavigatable[]
            Focus := Character.GetFocusInterface[]
        then:
            NavTargetStart := MakeNavigationTarget(OriginalPosition)
            var NavResultGoTo : navigation_result = navigation_result.Interrupted
            
            race:
                set NavResultGoTo = Navigatable.NavigateTo(NavTargetStart, ?MovementType:=movement_types.Walking)
                Focus.MaintainFocus(OriginalPosition)

            race:
                Sleep(0.5)
                Focus.MaintainFocus(Character.GetTransform().Translation + vector3{X:=1000.0, Y:=0.0, Z:=0.0})

        #Wipe in 60s if not killed
        set ChargeVFX = option{SpawnParticleSystem(VFX_UrjentCharge, NPC.GetAgent[].GetFortCharacter[].GetTransform().Translation)}
        race:
            ChargeWipe(GameManager, NPC)
            ResummonGuards(GameManager, NPC)
            WaitForThreshold(NPC)

        

        spawn:
            ShowDialog(GameManager, NPC, dialog_message{
                Title := option{"Urjent"}
                Message := "Aaaaaah, Lord Midas. Forgive me!"
                VO :=  option{GameManager.AudioManager.UrjentVO2c}
            }, 5.2)
        
        NPC.Die()
        
        Sleep(4.0)
        LavaWallCollision.MoveTo(OriginalLavaPosition, LavaWallCollision.GetTransform().Rotation, 0.1) 
        NPC.BossDefeated()

    ResummonGuards(GameManager:game_manager, NPC : boss_midas_behavior)<suspends>:void=
        loop:
            WaitForGuardsKilled(Guards)
            NPC.EndInvulnerability()
            Sleep(10.0)
            NPC.StartInvulnerability()
            SpawnGuards(Guards, ?OverrideMaxHealth := option{NPC.MaxHealth * 0.03})
    
    ChargeWipe(GameManager:game_manager, NPC : boss_midas_behavior)<suspends>:void=
        Sleep(60.0)

        if (CancelCharge := ChargeVFX?):
            CancelCharge.Cancel()
        
        if (Char := NPC.GetAgent[].GetFortCharacter[]):
            SpawnParticleSystem(VFX_UrjentCharge, Char.GetTransform().Translation)
        
        for (Player : NPC.GetPlayersInZone(); FortChar := Player.GetFortCharacter[]):
            FortChar.Damage(100000.0)

        Sleep(5.0)
        return

pirateking_behavior := class<concrete>(boss_behavior):
    @editable:
    Switches : []switch_device = array{}

    @editable:
    ProjectileAsset : creative_prop_asset = DefaultCreativePropAsset

    var OriginalPosition : vector3 = vector3{}

    CleanUp<override>(GameManager:game_manager, NPC : boss_midas_behavior)<suspends>:void=
        ResetSwitches(?State := false)

    DoBehavior<override>(GameManager:game_manager, NPC : boss_midas_behavior)<suspends>:void=
        
        if:
            Agent := NPC.GetAgent[]
            Character := Agent.GetFortCharacter[]
        then:
            set OriginalPosition = Character.GetTransform().Translation
        
        set NPC.PhaseThreshold = NPC.MaxHealth * 0.5
        Sleep(2.0)

        ShowDialog(GameManager, NPC, dialog_message{
            Title := option{"Charon"}
            Message := "Nyeheheheh, who are you! How'd you get here?"
            VO :=  option{GameManager.AudioManager.CharonVO2a}
        }, 7.0)

        ShowDialog(GameManager, NPC, dialog_message{
            Title := option{"Charon"}
            Message := "No matter! With my new powers, nobody can stop me!  Not even Lord Midas!"
            VO :=  option{GameManager.AudioManager.CharonVO2b}
        }, 6.0)

        #phase 1 - run around
        

        var FailCount : int= 0
        var CurrentAngle : float = GetRandomFloat(0.0, 2.0*PiFloat)

        if:
            Agent := NPC.GetAgent[]
            Character := Agent.GetFortCharacter[]
            Navigatable := Character.GetNavigatable[]
            Focus := Character.GetFocusInterface[]
        then:
            loop:

                PointOffset := vector3{X:=Cos(CurrentAngle), Y:=Sin(CurrentAngle), Z:=0.0} * 2000.0 * GetRandomFloat(0.6, 1.0)
                GoToPoint := OriginalPosition + PointOffset

                if (FailCount > 8):
                    set FailCount = 0
                    if (Character.TeleportTo[OriginalPosition, IdentityRotation()]){}
                else:
                
                    NavTargetStart := MakeNavigationTarget(GoToPoint)
                    var NavResultGoTo : navigation_result = navigation_result.Interrupted
                    race:
                        set NavResultGoTo = Navigatable.NavigateTo(NavTargetStart, ?MovementType:=movement_types.Running)
                        Focus.MaintainFocus(GoToPoint)
                        SpawnProjectiles(GameManager, NPC, 0.1, 3000.0, 0.0, 200.0)

                    set CurrentAngle += GetRandomFloat(PiFloat/2.0, 3.0*PiFloat/2.0)

                    NextPointOffset := vector3{X:=Cos(CurrentAngle), Y:=Sin(CurrentAngle), Z:=0.0} * 2000.0 * GetRandomFloat(0.6, 1.0)
                    LookAtPoint := OriginalPosition + NextPointOffset

                    if (NavResultGoTo = navigation_result.Reached):
                        set FailCount = 0
                        race:
                            Focus.MaintainFocus(LookAtPoint)
                            Sleep(3.0)
                    else:
                        set FailCount = FailCount + 1
                        Sleep(0.5)
                
                if (NPC.IsAtPhaseThreshold()?):
                    break


        #phase 2 - invincible, turn switches off
        NPC.StartInvulnerability()
        set NPC.PhaseThreshold = big_number{}
        ResetSwitches()

        #Move back to original position
        if:
            Agent := NPC.GetAgent[]
            Character := Agent.GetFortCharacter[]
            Navigatable := Character.GetNavigatable[]
            Focus := Character.GetFocusInterface[]
        then:
            NavTargetStart := MakeNavigationTarget(OriginalPosition)
            var NavResultGoTo : navigation_result = navigation_result.Interrupted
            race:
                set NavResultGoTo = Navigatable.NavigateTo(NavTargetStart, ?MovementType:=movement_types.Running)
                Focus.MaintainFocus(OriginalPosition)

        ShowDialog(GameManager, NPC, dialog_message{
            Title := option{"Charon"}
            Message := "My machine makes me invulnerable! There's nothing you can do now, nyahahahaha!"
            VO :=  option{GameManager.AudioManager.CharonVO2c}
        }, 4.3)

        race:
            CheckSwitches()
            SpawnProjectiles(GameManager, NPC, 0.05, 5000.0, 0.0, 500.0)
        NPC.EndInvulnerability()
        race:
            SpawnProjectiles(GameManager, NPC, 0.05, 5000.0, 0.0, 500.0)
            ShowDialog(GameManager, NPC, dialog_message{
                Title := option{"Charon"}
                Message := "Hey, stop messing with my machine!"
                VO :=  option{GameManager.AudioManager.CharonVO2d}
            }, 3.3)
        race:
            WaitForThreshold(NPC)
            SpawnProjectiles(GameManager, NPC, 0.05, 5000.0, 0.0, 500.0)

        #phase 3 - IDK it would be cool to have a phase 3 but none of the things I tried are possible in UEFN >:(


        #end
        spawn:
            ShowDialog(GameManager, NPC, dialog_message{
                Title := option{"Charon"}
                Message := "Wuaaaaaah, My Awesome Machine! How can I lose again?" 
                VO :=  option{GameManager.AudioManager.CharonVO2e}
            }, 5.0)
        
        NPC.Die()
        
        Sleep(4.0)
        NPC.BossDefeated()

    SpawnProjectiles(GameManager:game_manager, NPC : boss_midas_behavior, Period : float, Range : float, ZSpreadMin : float, ZSpreadMax : float)<suspends>:void=
        if (Character := NPC.GetAgent[].GetFortCharacter[]):
            loop:
                Sleep(Period)
                SpawnPoint := Character.GetTransform().Translation + vector3{X:=0.0, Y:=0.0, Z:=GetRandomFloat(0.0,150.0)}
                MoveTarget := Character.GetTransform().Translation + RandomUnitVecXY() * Range + vector3{X:=0.0, Y:=0.0, Z:=GetRandomFloat(ZSpreadMin, ZSpreadMax)}
                SpawnProjectile(GameManager, ProjectileAsset, SpawnPoint, 100.0, (MoveTarget-SpawnPoint)/12.0, 12.0, 300.0, ?Explosion := option{VFX_PlayerHit})

    CheckSwitches()<suspends>:void=
        loop:
            var AllOff : logic = true
            for (Switch : Switches):
                if (Switch.GetCurrentState[]):
                    set AllOff = false

            if (AllOff?):
                break
            
            Sleep(0.1)

    ResetSwitches(?State : logic = true):void=
        for (Switch : Switches):
            Switch.SetState(State)
                




