using { /Fortnite.com/AI }
using { /Fortnite.com/Characters }
using { /Fortnite.com/Devices }
using { /Fortnite.com/Game }
using { /Fortnite.com/Playspaces }
using { /Verse.org }
using { /Verse.org/Colors }
using { /Verse.org/Random }
using { /Verse.org/Simulation }
using { /Verse.org/Assets }
using { /UnrealEngine.com/Temporary/Diagnostics }
using { /UnrealEngine.com/Temporary/SpatialMath }
using { /UnrealEngine.com/Assets }
using { /Fortnite.com/Animation/PlayAnimation }
using { /UnrealEngine.com/Temporary/UI }
    

# A Verse-authored NPC Behavior that can be used within an NPC Character Definition or a NPC Spawner device's NPC Behavior Script Override.
boss_midas_behavior := class(npc_behavior):
    SettingsCategory<public><localizes>:message = "Settings"

    @editable:
        Categories:=array{SettingsCategory}
    BossType : boss_enum = boss_enum.Sharky

    @editable:
        Categories:=array{SettingsCategory}
    ZoneTag : fetchable_tag = fetchable_tag.none_tag
    
    # Liklihood of making an effort SFX when hit
    @editable_number(int):
        Categories:=array{SettingsCategory},
        MinValue:=option{0},
        MaxValue:=option{100}
    ProbabilityOfEffortSFX:int = 10

    # Minimum time between effort SFX when hit
    @editable_number(float):
        Categories:=array{SettingsCategory},
        MinValue:=option{0.0},
        MaxValue:=option{10.0}
    EffortSFXCooldown:float = 2.0

    # Time between teleport and attack.
    @editable_number(float):
        Categories:=array{SettingsCategory},
        MinValue:=option{0.5},
        MaxValue:=option{10.0}
    MoveToWaitDuration:float = 1.5

    # How long to wait after the player enters the zone before starting behavior.
    @editable_number(float):
        Categories:=array{SettingsCategory},
        MinValue:=option{0.0},
        MaxValue:=option{10.0}
    StartDelay:float = 4.0

    # How long to wait after the player defeats the boss before teleporting them out.
    @editable_number(float):
        Categories:=array{SettingsCategory},
        MinValue:=option{0.0},
        MaxValue:=option{10.0}
    EndDelay:float = 2.0

    # How far from the spawn point the boss teleports to.
    @editable_number(float):
        Categories:=array{SettingsCategory},
        MinValue:=option{0.0}
    DistanceFromSpawnPtToMove:float = 1050.0

    # DOESN'T WORK -- How high the boss floats
    @editable_number(float):
        Categories:=array{SettingsCategory},
        MinValue:=option{0.0}
    FloatingHeight:float = 0.0

    # Whether the boss should teleport or walk.
    @editable:
        Categories:=array{SettingsCategory}
    Teleport:logic = true

    #Offset for teleport VFX
    @editable:
        Categories:=array{SettingsCategory}
    VFXOffset:vector3 = vector3{X:=0.0, Y:=0.0, Z:=-50.0}

    #Offset for fire wave
    @editable:
        Categories:=array{SettingsCategory}
    FireOffset:vector3 = vector3{X:=0.0, Y:=0.0, Z:=-50.0}

    #Animation to play when the boss attacks
    @editable:
        Categories:=array{SettingsCategory}
    AttackAnim : cinematic_sequence_device = cinematic_sequence_device{}
    
    #Fire wave object
    @editable:
        Categories:=array{SettingsCategory}
    Firewall: damage_volume_device  = damage_volume_device {}
    @editable:
        Categories:=array{SettingsCategory}
    Firewall2: damage_volume_device  = damage_volume_device {}
    @editable:
        Categories:=array{SettingsCategory}
    FirewallDelay: float = 3.0
    @editable:
        Categories:=array{SettingsCategory}
    FirewallScalingMult: float = 1.5

    @editable:
        Categories:=array{SettingsCategory}
    GengaAttack1 : cinematic_sequence_device = cinematic_sequence_device{}
    @editable:
        Categories:=array{SettingsCategory}
    GengaAttack2 : cinematic_sequence_device = cinematic_sequence_device{}

    @editable:
        Categories:=array{SettingsCategory}
    PirateProjectile : creative_prop_asset := DefaultCreativePropAsset

    @editable:
        Categories:=array{SettingsCategory}
    DamageVolumes : []damage_volume_device = array{}
    
    @editable:
        Categories:=array{SettingsCategory}
    MaxTimer: float = 30.0

    # Players will be teleported here after the fight ends
    @editable: 
        Categories:=array{SettingsCategory}
    ExitTeleporter :teleporter_device = teleporter_device{}

    # Teleport vfx
    @editable: 
        Categories:=array{SettingsCategory}
    BlinkVFX :vfx_spawner_device = vfx_spawner_device{}

    @editable:
        Categories:=array{SettingsCategory}
    InvulnPropAsset : creative_prop_asset = DefaultCreativePropAsset

    #Used to record the boss's last teleport position so we don't teleport in place (1/2/3/4 = Top/Left/Bottom/Right) 
    var IDX :int = 2
    #Whether we're applying our "onStart" delay on the next iteration of the boss behavior. Happens once
    var DoStartDelay :logic = true

    var GameManager : ?game_manager = false
    var ZoneManager : ?zone_manager_device = false
    var Mutator : ?mutator_zone_device = false

    var BossLevel : int = 0
    var BossName : string = "Midas"
    var EffortCooldown : float = 0.0
    var FirstTeleport : logic = true

    var IsInvunlerable : logic = false
    var InvulnerabilityProp : ?creative_prop = false

    var GuardMultiplier : float = 0.0

    var PlayersHittingUsNow : []player = array{}
    var EliminatedEventsPerPlayer : [player]cancelable = map{}

    var ActiveTier : int = 0
    var MaxHealth : big_number = big_number{}
    var CurrentHealth : big_number = big_number{}
    var PhaseThreshold : big_number = big_number{}

    var HasSubscribed : logic = false
    var SubscribedEvents : []cancelable = array{}

    var BossRunning : logic = false

    StartInvulnerability():void=
        if (not IsInvunlerable?):

            if (Prop := SpawnProp(InvulnPropAsset, transform{Translation:=GetAgent[].GetFortCharacter[].GetTransform().Translation, Rotation:=IdentityRotation()})(0)?):
                
                set IsInvunlerable = true
                set InvulnerabilityProp = option{Prop}
                spawn { TickInvulnerability() }

    EndInvulnerability():void=
        if (IsInvunlerable?):
            set IsInvunlerable = false

            if (Prop := InvulnerabilityProp?):
                Prop.Dispose()
                set InvulnerabilityProp = false

    TickInvulnerability()<suspends>:void=
        loop:
            if (IsInvunlerable?):
                if:
                    Prop := InvulnerabilityProp?
                    Prop.TeleportTo[GetAgent[].GetFortCharacter[].GetTransform().Translation, GetAgent[].GetFortCharacter[].GetTransform().Rotation]
            else:
                return
            Sleep(0.01)

    OnBossDamaged(DamageResult:damage_result):void=
        if:
            GM := GameManager?
            Players := GM.GetPlayspace().GetPlayers()
            InstigatingAgent := DamageResult.Instigator?.GetInstigatorAgent[]
            Player:=player[InstigatingAgent]
            Player.IsActive[]
            PlayerManager := GameManager?.PlayerManagers[Player]
            GunStats := PlayerManager.GetCurrentGunStats[]
            BossChar := GetAgent[].GetFortCharacter[]
            PlayerChar := Player.GetFortCharacter[]
            not IsInvunlerable?
        then:
            Dist := Distance(PlayerChar.GetTransform().Translation, BossChar.GetTransform().Translation)
            FalloffMult := GunStats.GetFallofMult(Dist)
            

            if:
                GunStats.FortBaseDamage > 0.0
                DamageResult.Amount > GunStats.FortBaseDamage * 1.3
                PlayerWeakspotMult := PlayerManager.GetPlayerStat(player_stat.Crit_Mult)
            then:
                ProcessDamage(PlayerManager, BigNumber_Multiply(GunStats.WeakspotMult, PlayerWeakspotMult), true, FalloffMult)
                PlayerManager.OnWeakspotHit()
            else:
                ProcessDamage(PlayerManager, big_number{Value:=1.0}, false, FalloffMult)

    ProcessDamage(PlayerManager : player_manager, WeakspotDamageMult : big_number, IsWeakHit : logic, FalloffMult :big_number):void=
        if:
            PlayerManager.Player.IsActive[]
            BaseStrength := PlayerManager.GetAdjustedStrength[array{"Guard", "Boss"}]
            TeamworkDamage := (1.0 + (PlayerManager.GetUpgradeValue["Teamwork", false] / 100.0 * PlayersHittingUsNow.Length * 1.0))
            SetZoneManager := ZoneManager?
        then:

            #TODO: don't hardcode this
            var WeaknessMult : float = 1.0
            if:
                BossType = boss_enum.Urjent
                PlayerManager.GetHeldGun[] = 9 #Stake gun
            then:
                set WeaknessMult = 3.0

            var AmbushDmgMult : float = 1.0
            if:
                not(PlayersHittingUsNow.Find[PlayerManager.Player])
                Temp := PlayerManager.GetUpgradeValue["Ambush", false] / 100.0
                set AmbushDmgMult += Temp

            Strength := BaseStrength * WeakspotDamageMult * FalloffMult * TeamworkDamage * AmbushDmgMult * WeaknessMult

            if(not(PlayersHittingUsNow.Find[PlayerManager.Player])):
                set PlayersHittingUsNow = PlayersHittingUsNow + array{PlayerManager.Player}    

            var IsHandcannon : logic = PlayerManager.UpdateHandcannon()
            var IsSuperHandcannon : logic = PlayerManager.IsSuperHandcannon()
            var IsAmbush : logic = false

            if:
                IsHandcannon?
                FortChar := GetAgent[].GetFortCharacter[]
            then:
                AgentLoc := FortChar.GetTransform().Translation
                SpawnParticleSystem(VFX_HandCannon, AgentLoc + VFXOffset)

            if(AmbushDmgMult > 1.0):
                set IsAmbush = true
            else:
                set IsAmbush = false

            AdjustBossHealth(PlayerManager, Strength, IsWeakHit, IsHandcannon, IsAmbush)
            #GainBossPoints(PlayerManager, Strength, IsWeakHit, IsHandcannon, IsAmbush)
    
    GainBossPoints(PlayerManager : player_manager, Damage : big_number, IsWeakHit: logic, IsHandcannon : logic, IsAmbush:logic):void=
        AdjustBossHealth(PlayerManager, Damage, IsWeakHit, IsHandcannon, IsAmbush)

    AdjustBossHealth(PlayerManager : player_manager, Damage : big_number, IsWeakHit: logic, IsHandcannon : logic, IsAmbush:logic):void=
        
        var AdjustedDamage : big_number = Damage

        if:
            UIManager := PlayerManager.UIManager?
        then:
            if (BigNumber_GreaterThanOrEquals(PhaseThreshold+AdjustedDamage, CurrentHealth)?):
                set AdjustedDamage = CurrentHealth - PhaseThreshold
                set CurrentHealth = PhaseThreshold
            else:
                set CurrentHealth = CurrentHealth - AdjustedDamage
            
            PlayerManager.ProcessEssence(ZoneTag, AdjustedDamage, big_number{}, big_number{}, true, IsWeakHit, IsHandcannon, IsAmbush, false)   
            UIManager.QueueUpdateGuardUI(CurrentHealth, CurrentHealth + AdjustedDamage, MaxHealth , "Boss",  PlayersHittingUsNow.Length, big_number{}, big_number{}, "RepeatableMidas", true, true, false)
              
    IsAtPhaseThreshold()<transacts>:logic=
        return BigNumber_GreaterThanOrEquals(PhaseThreshold, CurrentHealth)

    BossDefeated():void=
        Log("Boss Defeated")

        spawn {PlayBossDeathVO()}

        for (Player : GetPlayersInZone(); PlayerManagerIn := GameManager?.PlayerManagers[Player]):
            if:
                GM := GameManager?
            then:
                #Achievements
                case (BossType):
                    boss_enum.LyraTutorial => PlayerManagerIn.UpdateQuestProgressForTask(task_event.DefeatBoss, big_number{Value := 1.0}, false, ?Target := "0", ?IsPersistent := true)
                    boss_enum.Sharky => PlayerManagerIn.UpdateQuestProgressForTask(task_event.DefeatBoss, big_number{Value := 1.0}, false, ?Target := "1", ?IsPersistent := true)
                    boss_enum.Genga => PlayerManagerIn.UpdateQuestProgressForTask(task_event.DefeatBoss, big_number{Value := 1.0}, false, ?Target := "2", ?IsPersistent := true)
                    boss_enum.Urjent => PlayerManagerIn.UpdateQuestProgressForTask(task_event.DefeatBoss, big_number{Value := 1.0}, false, ?Target := "3", ?IsPersistent := true)
                    boss_enum.PirateKing => PlayerManagerIn.UpdateQuestProgressForTask(task_event.DefeatBoss, big_number{Value := 1.0}, false, ?Target := "5", ?IsPersistent := true)
                    _ => PlayerManagerIn.UpdateQuestProgressForTask(task_event.DefeatBoss, big_number{Value := 1.0}, false, ?Target := "?", ?IsPersistent := true)

                #Accolades
                GM.AccoladesManager.BossSlainAward(Player)
                GM.AudioManager.PlayBossSlainAudio(Player)

                #Teleport Player Out
                TeleportPlayerOut(Player)

                #Rewards
                BossType.CheckHasKey(PlayerManagerIn, ?Spend := true) #spend keys
                PlayerManagerIn.GainUpgradeLevelTo(BossType.RewardStatName(), ActiveTier) 
                PlayerManagerIn.GainAchievementLevelTo(BossType.AchievementName(), ActiveTier+1)

                BossType.DropLoot(GM, Player, ActiveTier)
                
                set BossRunning = false

        # #Despawn
        # if (FortChar := GetAgent[].GetFortCharacter[]):
        #     FortChar.Damage(FortChar.GetMaxHealth()*2.0)

    Die():void=
        if (FortChar := GetAgent[].GetFortCharacter[]):
            FortChar.Damage(FortChar.GetMaxHealth()*2.0)

    NavigateTo(Target : vector3, ?MovementType : movement_type= movement_types.Walking)<suspends>:navigation_result=
        if:
            Character := GetAgent[].GetFortCharacter[]
            Navigatable := Character.GetNavigatable[]
            Focus := Character.GetFocusInterface[]
        then:

            NavTargetStart := MakeNavigationTarget(Target)
            var NavResultGoTo : navigation_result = navigation_result.Interrupted
            # Tell the navigation interface to walk to the location and store a navigation result for error checking.
            race:
                set NavResultGoTo = Navigatable.NavigateTo(NavTargetStart, ?MovementType:= MovementType)
                Focus.MaintainFocus(Target)

            return NavResultGoTo

        return navigation_result.Interrupted

    PlayersWiped():void=
        set BossRunning = false

    OnAgentEnterZone(Agent:agent):void=
        Log("OnAgentEnterZone")
        if:
            Player:=player[Agent]
            Player.IsActive[]
            Char := Player.GetFortCharacter[]
        then:
            Log("OnAgentEnterZone Success")
            ElimEvent := Char.EliminatedEvent()

            if (OldElimEvent := EliminatedEventsPerPlayer[Player]):
                OldElimEvent.Cancel()

            if (set EliminatedEventsPerPlayer[Player] = ElimEvent.Subscribe(OnPlayerEliminated)){}

            spawn:
                InitialInvincibility(Char)

    OnAgentLeaveZone(Agent:agent):void=
        if:
            Player:=player[Agent]
            Player.IsActive[]
            Char := Player.GetFortCharacter[]
            PlayerManager := GameManager?.PlayerManagers[Player]
            UIManager := PlayerManager.UIManager?
            CancelEvent := EliminatedEventsPerPlayer[Player]
        then:
            #UIManager.HideDamageMasteryTimer()
            spawn:
                DoCancel(Player)
                
    DoCancel(Player:player)<suspends>:void=
        Sleep(4.5)
        if (CancelEvent := EliminatedEventsPerPlayer[Player]):
            CancelEvent.Cancel()
                
    OnPlayerEliminated(Result:elimination_result):void=
        #Refund Keys
        Log("Player Eliminated")
        if (CancelEvent := EliminatedEventsPerPlayer[Result.EliminatedCharacter.GetAgent[]]):
            CancelEvent.Cancel()
        
    InitialInvincibility(Char:fort_character)<suspends>:void=
        Char.SetVulnerability(false)
        Sleep(5.0)
        Char.SetVulnerability(true)

    GetDamageTier(Damage : big_number):int=
        var Result : int = 0
        loop:
            NextTier := GetDamageForTier(Result+1)
            if (BigNumber_GreaterThanOrEquals(Damage, NextTier)?):
                set Result += 1
            else:
                break
        
        return Result

    GetDamageForTier(Tier : int):big_number=
        if (Tier = 0): 
            return big_number{}
        return BossType.BossMaxHealth(Tier)

    GetPercentToNextTier(Damage : big_number):float=
        CurrentTier := GetDamageTier(Damage)
        Start := GetDamageForTier(CurrentTier)
        End := GetDamageForTier(CurrentTier+1)
        return BigNumber_Percentage(Damage-Start, End-Start)
        
    CheckDoEfffort():void=
        if:
            EffortCooldown <= 0.0
            GetRandomInt(1, 100) < ProbabilityOfEffortSFX
            GM := GameManager?
        then:
            set EffortCooldown = EffortSFXCooldown
            VOIndex:= GetRandomInt(0, 2)
            for (Player: GetPlayersInZone()):
                GM.AudioManager.PlayBossEffort(Player, VOIndex, BossType)

    PlayEntranceVO():void=
        set EffortCooldown = EffortSFXCooldown
        VOIndex:= GetRandomInt(0, 2)
        for (Player: GetPlayersInZone()):
            if (GM := GameManager?):
                GM.AudioManager.PlayBossEntranceVO(Player, VOIndex, BossType)

    PlayBossDeathVO()<suspends>:void=
        VOIndex:= GetRandomInt(0, 2)

        for (Player: GetPlayersInZone()):
            if (GM := GameManager?):
                GM.AudioManager.PlayBossExitVO(Player, VOIndex, BossType)
                

    ActualOnBegin()<suspends>:void =
        set FirstTeleport = true
        
        Sleep(0.1)
        # Test
        if:
            TeleportInfo := GameManager?.BossTeleportManager.Teleporters[BossType.GetBossID()]
        then:
            set ActiveTier = TeleportInfo.BossQueueTier
            set TeleportInfo.BossQueueTier = 0

        Log("Starting Boss {BossType.GetBossName()} at {ActiveTier}, Max Health: {MaxHealth}")
        set MaxHealth = BossType.BossMaxHealth(ActiveTier)
        set CurrentHealth = MaxHealth


        Log("NPC Started!!!")

        if:
            Log("SUB1")
            not HasSubscribed?
            Log("SUB2")
            Mut := Mutator?
        then:
            set HasSubscribed = true
            Log("Subscribed")
            set SubscribedEvents += array{Mut.AgentEntersEvent.Subscribe(OnAgentEnterZone)}
            set SubscribedEvents += array{Mut.AgentExitsEvent.Subscribe(OnAgentLeaveZone)}

            for (Player : GetPlayersInZone()):
                OnAgentEnterZone(Player)

        if (Char := GetAgent[].GetFortCharacter[]):
            set SubscribedEvents += array{Char.DamagedEvent().Subscribe(OnBossDamaged)}

        if (GM :=  GameManager?):
            case (BossType):
                boss_enum.LyraTutorial => spawn {RunBehaviorUntilStopped(GM.BossInfo.SargeBehavior) }
                boss_enum.Sharky => spawn { RunBehaviorUntilStopped(GM.BossInfo.SharkyBehavior) }
                boss_enum.Genga => spawn { RunBehaviorUntilStopped(GM.BossInfo.GengaBehavior) }
                boss_enum.Urjent => spawn { RunBehaviorUntilStopped(GM.BossInfo.UrjentBehavior) }
                #boss_enum.Giant => spawn { GM.BossInfo.GiantBehavior.DoBehavior(GM, Self) }
                boss_enum.PirateKing => spawn { RunBehaviorUntilStopped(GM.BossInfo.PirateKingBehavior) }
                _ => block:

    RunBehaviorUntilStopped(Behavior:boss_behavior)<suspends>:void=
        set BossRunning = true
        if (GM :=  GameManager?):
            race:
                Behavior.DoBehavior(GM, Self)
                CheckBossStopped()

        #unsubscribe
        for (Event : SubscribedEvents):
            Event.Cancel()

        if (GM :=  GameManager?):
            Behavior.CleanUp(GM, Self)

        if (TeleportInfo := GameManager?.BossTeleportManager.Teleporters[BossType.GetBossID()]):
            TeleportInfo.ClearPlayersInFight()
            
        Print("Teleporting Players Out")
        WaitAndTeleportPlayersOut(?Delay := 4.0)
        
        EndInvulnerability()
        if (FortChar := GetAgent[].GetFortCharacter[]):
            FortChar.Damage(FortChar.GetMaxHealth()*2.0)

    CheckBossStopped()<suspends>:void=
        loop:
            if (GameManager?.BossTeleportManager.Teleporters[BossType.GetBossID()].PlayersInFight.Length <= 0):
                PlayersWiped()
            if (not BossRunning?):
                break
            Sleep(1.0)

    IsAlive()<transacts>:logic=
        if (BigNumber_GreaterThanOrEquals(big_number{}, CurrentHealth)?):
            return false
        return true

    #Seperate function so we can spawn{} it
    WaitAndTeleportPlayersOut(?Delay:float = 2.0)<suspends>:void =
        if (Delay > 0.0):
            Sleep(Delay)

        PlayersInZone := GetPlayersInZone()
        for (Player : PlayersInZone):
            TeleportPlayerOut(Player)

    TeleportPlayerOut(Player:player):void=
        if:
            Player.IsActive[]
            PlayerManager := GameManager?.PlayerManagers[Player]
        then:
            if (FortChar := Player.GetFortCharacter[]):
                FortChar.SetHealth(FortChar.GetMaxHealth())

            spawn:
                PlayerManager.GoToPersistentLocation()


    GetPlayersInZone()<transacts>:[]player =
        var Players : []player = array{}
        if:
            GM := GameManager?
        then:
            PlayspacePlayers := GM.PlayerManagers

            for (PlayerManager : PlayspacePlayers):
                if:
                    PlayerManager.Player.IsActive[]
                    Mutator?.IsInVolume[PlayerManager.Player] or PlayerManager.EliminatedZone = ZoneTag
                then:
                    set Players = Players + array{PlayerManager.Player}
        return Players

    GetBossSpawner()<transacts>:?npc_spawner_device =
        if:
            BossSpawners := GameManager?.FindCreativeObjectsWithTag(boss_spawner{})
        then:
            for (Obj : BossSpawners):
                if (Spawner := npc_spawner_device[Obj]):
                    return option{Spawner}
        return false