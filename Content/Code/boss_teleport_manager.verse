
using { /Fortnite.com/Devices }
using { /Verse.org/Simulation }
using { /UnrealEngine.com/Temporary/Diagnostics }
using { /Verse.org/Assets }
using { /UnrealEngine.com/Temporary/UI}
using { /Fortnite.com/UI }
using { /UnrealEngine.com/Temporary/SpatialMath }
using { /Verse.org/Colors }

# See https://dev.epicgames.com/documentation/en-us/uefn/create-your-own-device-in-verse for how to create a verse device.

boss_teleport_info :=class<concrete>:
    @editable
    TeleportInfo : teleport_info = teleport_info{}

    @editable
    ID :int = 0

    @editable
    QueueButton : button_device = button_device{}

    TimerStart : float = 10.0

    var GameManager: ?game_manager = false

    var BossQueue: []player_manager = array{}
    var PlayersInFight: []player_manager = array{}

    var TimerActive: logic = false
    var BossQueueTimer: float = 0.0
    var BossQueueTier: int = -1

    TickQueue(Delta:float):void=

        if (BossQueue.Length > 0 and TimerActive):
            set BossQueueTimer -= Delta
            Print("Boss Queue Timer: {BossQueueTimer}")

            if (BossQueueTimer <= 0.0):
                StartFight()
        else:
            set BossQueueTimer = TimerStart

        for (Key -> PlayerManager : GameManager?.PlayerManagers):
            if:
                Menu := GameManager?.BossTeleportManager.QueueMenus[PlayerManager.Player]
                Menu.TeleportInfo.ID = ID
            then:
                Menu.Update()

    StartCountdown():void=
        set TimerActive = true
        set BossQueueTimer = TimerStart

    StartFight():void=

        var DoActivate : logic = true
        for (Player : BossQueue):
            if:
                not Player.CurrentZone.IsBossTag()? and not Player.CurrentZone = fetchable_tag.prestigeareatag
            then:
                DidTeleport := teleport_wrapper{Teleporter := TeleportInfo.Teleporter, PlayerManager:= Player, BossType := BossTypeFromID(ID)}.Teleport_NoWData(Player.Player)
                if:
                    not DidTeleport?
                    Temp := BossQueue.RemoveFirstElement[Player]
                then:
                    set BossQueue = Temp
            
            if:
                DoActivate?
            then:
                set DoActivate = false
                Print("Activate Boss - {ID}")
                # Activate Boss
                case(ID):
                    0 => block:
                        Player.GameManager.BossInfo.SargeSpawner.Disable()
                        Player.GameManager.BossInfo.SargeSpawner.Enable()
                        # Player.GameManager.BossInfo.SargeSpawner.Spawn()
                    1 => block:
                        Player.GameManager.BossInfo.SharkeySpawner.Disable()
                        Player.GameManager.BossInfo.SharkeySpawner.Enable()
                        # Player.GameManager.BossInfo.SharkeySpawner.Spawn()
                    2 => block:
                        Player.GameManager.BossInfo.GengaSpawner.Disable()
                        Player.GameManager.BossInfo.GengaSpawner.Enable()
                        # Player.GameManager.BossInfo.GengaSpawner.Spawn()
                    3 => block:
                        Player.GameManager.BossInfo.UrjentSpawner.Disable()
                        Player.GameManager.BossInfo.UrjentSpawner.Enable()
                        # Player.GameManager.BossInfo.UrjentSpawner.Spawn()
                    5 => block:
                        Player.GameManager.BossInfo.PirateKingSpawner.Disable()
                        Player.GameManager.BossInfo.PirateKingSpawner.Enable()
                        # Player.GameManager.BossInfo.PirateKingSpawner.Spawn()
                    _ => block:

        set BossQueueTimer = TimerStart
        set PlayersInFight = BossQueue
        set BossQueue = array{}

    AddToQueue(PlayerManager : player_manager):void=
        set BossQueue = BossQueue + array{PlayerManager}

    RemoveFromQueue(PlayerManager : player_manager):void=
        set BossQueue = BossQueue.RemoveFirstElement[PlayerManager] or BossQueue

    RemovedFromFight(PlayerManager : player_manager):void=
        set PlayersInFight = PlayersInFight.RemoveFirstElement[PlayerManager] or PlayersInFight

    ClearPlayersInFight():void=
        set PlayersInFight = array{}




# A Verse-authored creative device that can be placed in a level
boss_teleport_manager_device := class(creative_device, menu_item):

    @editable
    GameManager : ?game_manager = false

    @editable
    Teleporters : []boss_teleport_info = array{}

    OnBegin<override>():void =        
        for (Teleporter : Teleporters):
            Teleporter.QueueButton.InteractedWithEvent.Subscribe(open_queue_menu_wrapper{TeleportManagerDevice := Self, Boss := BossList[Teleporter.ID] or boss_enum.LyraTutorial}.Open)
            set Teleporter.GameManager = GameManager

        spawn { TickQueues() }

    TickQueues()<suspends>:void=
        loop:
            Sleep(1.0)
            for (TeleportInfo : Teleporters):
                TeleportInfo.TickQueue(1.0)

    TeleportAgent(Agent : agent, TeleporterID : int) : void =
        if:
            Teleporter := Teleporters[TeleporterID].TeleportInfo.Teleporter
        then:
            Teleporter.Teleport(Agent)
    
    OpenMenu<override>(Agent:agent):void=
        if (PlayerManager := GameManager?.PlayerManagers[Agent]):
            
            if (not GameManager?.PlayerManagers[Agent].UIManager?.Popup?):
                if (GM := GameManager?):
                    MakeBossTeleportMenu(GM, Self, Agent).OpenMenu()
            else:
                block:#Log("DOUBLE MENU")

    GetTitle<override>():string=
        return "BOSSES"

    GetIcon<override>():texture=
        return Images.BossTeleport

    IsEnabled<override>(PlayerManager : player_manager)<transacts>:logic=
        if (PlayerManager.GetAchievementLevel["Sarge_NextTier"] >= 1):
            return true
        else:
            return false

    GetTPIcon(TeleporterID : int) : texture = 
        if (Temp := Teleporters[TeleporterID].TeleportInfo.Icon):
            return GetTextureFromString(Temp)
        else:
            return Images.Portal

    GetTPName(TeleporterID : int) : string = 
        if (Temp := Teleporters[TeleporterID].TeleportInfo.Name):
            return Temp
        else:
            return "TP #{TeleporterID}"

    GetTPInfoForBoss(Boss : boss_enum)<transacts> : ?boss_teleport_info = 
        for (Teleporter : Teleporters):
            if (Teleporter.ID = Boss.GetBossID()):
                return option{Teleporter}

        return false

    var QueueMenus : [player]boss_queue_menu = map{}
    OpenQueueMenu(Agent : agent, Boss: boss_enum, ?Force:logic = false):void=
        
        GM := GameManager? or game_manager{}
        
        if:
            PlayerManager := GM.PlayerManagers[Agent]
            UIManager := PlayerManager.UIManager?
        then:
            if (Force?):
                UIManager.ClosePopup()

            if:
                not PlayerManager.UIManager?.Popup?
            then:
                if (TeleportInfo := GetTPInfoForBoss(Boss)?):
                    if (Menu := QueueMenus[PlayerManager.Player]):
                        Menu.SetBoss(TeleportInfo)
                        Menu.OpenMenu()
                    else:
                        Menu := MakeBossQueueMenu(GM, TeleportInfo, PlayerManager)
                        if:
                            set QueueMenus[PlayerManager.Player] = Menu
                        Menu.OpenMenu()

open_queue_menu_wrapper := class:
    TeleportManagerDevice : boss_teleport_manager_device
    Boss : boss_enum

    Open(Agent : agent):void=
        TeleportManagerDevice.OpenQueueMenu(Agent, Boss, ?Force := true)

    OpenWData(Wdata:widget_message):void=
        TeleportManagerDevice.OpenQueueMenu(Wdata.Player, Boss, ?Force := true)

MakeBossQueueMenu<constructor>(GameManagerIn : game_manager, TeleportInfoIn : boss_teleport_info, PlayerManagerIn : player_manager) := boss_queue_menu: 
        GameManager :=  option{GameManagerIn}
        TeleportInfo := TeleportInfoIn
        Agent := option{PlayerManagerIn.Agent}
        PlayerManager := PlayerManagerIn
        Section1 := MakeBossQueueSection()
        Section2 := MakeBossQueueSection()
        Section3 := MakeBossQueueSection()

boss_queue_menu := class(menu_template):
    var TeleportInfo : boss_teleport_info
    PlayerManager : player_manager

    DialogBackground : texture_block = texture_block{ DefaultImage := Images.StoreBox, DefaultDesiredSize := vector2{X := 1920.0, Y := 1080.0}} 
    Section1 : boss_queue_section
    Section2 : boss_queue_section
    Section3 : boss_queue_section

    CloseButton : button_quiet = button_quiet:

    SetBoss(NewInfo : boss_teleport_info):void=
        set TeleportInfo = NewInfo
        
        if (Widget := _Widget?):
            UpdateMenu(Widget)

    UpdateMenu<override>(Widget:canvas):void=
        BossType := BossList[TeleportInfo.ID] or boss_enum.LyraTutorial

        CurrentTier := PlayerManager.GetAchievementLevel[BossType.AchievementName()] or 0
        InProgressTier := TeleportInfo.BossQueueTier
        
        HasKey := BossType.CheckHasKey(PlayerManager, ?Spend := false)

        Timer := Round[TeleportInfo.BossQueueTimer] or 0
        #CLEAR
        if (CurrentTier > 1):
            if (HasKey?):
                Section1.Update("Previous Tier", BossType.GetBossImage(), BossType.GetBossName(), CurrentTier-1, BossType.BossMaxHealth(CurrentTier-1), "Auto-Clear", "Instantly clears the previous-tier boss for rewards.")
            else:
                Section1.Update("Previous Tier", BossType.GetBossImage(), BossType.GetBossName(), CurrentTier-1, BossType.BossMaxHealth(CurrentTier-1), "Need Key", "You need a boss key to auto-clear.")
        else:
            Section1.Update("Previous Tier", BossType.GetBossImage(), BossType.GetBossName(), 0, big_number{}, "---", "Defeat this boss to unlock auto-clear.")

        #START
        if(TeleportInfo.BossQueue.Length = 0 and TeleportInfo.PlayersInFight.Length = 0):
            if (HasKey?):
                Section2.Update("Current Tier", BossType.GetBossImage(), BossType.GetBossName(), CurrentTier, BossType.BossMaxHealth(CurrentTier),  "Start", "")
            else:
                Section2.Update("Current Tier", BossType.GetBossImage(), BossType.GetBossName(), CurrentTier-1, BossType.BossMaxHealth(CurrentTier-1), "Need Key", "You need a boss key to start this fight.")
        else:
            if:
                TeleportInfo.BossQueue.Find[PlayerManager]
            then:
                Section2.Update("Current Tier", BossType.GetBossImage(), BossType.GetBossName(), CurrentTier, BossType.BossMaxHealth(CurrentTier),  "Queued", "you are in queue for this boss.")
            else:
                Section2.Update("Current Tier", BossType.GetBossImage(), BossType.GetBossName(), CurrentTier, BossType.BossMaxHealth(CurrentTier),  "Can't Start", "Boss is already in progress..")

        #JOIN
        if (TeleportInfo.BossQueue.Length > 0):
            if (CurrentTier < InProgressTier):
                Section3.Update("Queued", BossType.GetBossImage(), BossType.GetBossName(), InProgressTier, BossType.BossMaxHealth(InProgressTier), "Can't Join", "Your tier is too low to join this fight.")
            else:
                if (TeleportInfo.BossQueue.Find[PlayerManager]):
                    Section3.Update("Queued", BossType.GetBossImage(), BossType.GetBossName(), InProgressTier, BossType.BossMaxHealth(InProgressTier), "Starts In {Timer}s", "")
                else:
                    if (HasKey?):
                        Section3.Update("Queued", BossType.GetBossImage(), BossType.GetBossName(), InProgressTier, BossType.BossMaxHealth(InProgressTier), "Join ({Timer}s)", "")
                    else:
                        Section3.Update("Queued", BossType.GetBossImage(), BossType.GetBossName(), CurrentTier-1, BossType.BossMaxHealth(CurrentTier-1), "Need Key", "You need a boss key to join this fight.")
        else if (TeleportInfo.PlayersInFight.Length > 0):
            Section3.Update("In Progress", BossType.GetBossImage(), BossType.GetBossName(), InProgressTier, BossType.BossMaxHealth(InProgressTier), "Can't Join", "This fight has already begun.")
        else:
            Section3.Update("", BossType.GetBossImage(), BossType.GetBossName(), 0, big_number{}, "---", "No in-progress fights.")
        

    CreateMenu<override>():canvas=

        Section1.Create()
        Section2.Create()
        Section3.Create()

        Result := canvas:
            Slots := array:
                canvas_slot:
                    Widget:= DialogBackground
                    SizeToContent := true
                    Anchors := anchors{ Minimum := vector2{X := 0.1, Y := 0.1}, Maximum := vector2{X := 0.9, Y := 0.9}}  
                    ZOrder := 1
                canvas_slot:
                    Widget:= Section1.Widget
                    Anchors := anchors{ Minimum := vector2{X := 0.05, Y := 0.1}, Maximum := vector2{X := 0.35, Y := 0.9}}
                    Offsets := margin{Top := 0.0, Left := 0.0, Right := 0.0, Bottom := 0.0}
                    Alignment := vector2{X := 0.5, Y := 0.5}
                    ZOrder := 2
                canvas_slot:
                    Widget:= Section2.Widget
                    Anchors := anchors{ Minimum := vector2{X := 0.35, Y := 0.1}, Maximum := vector2{X := 0.65, Y := 0.9}}
                    Offsets := margin{Top := 0.0, Left := 0.0, Right := 0.0, Bottom := 0.0}
                    Alignment := vector2{X := 0.5, Y := 0.5}
                    ZOrder := 3
                canvas_slot:
                    Widget:= Section3.Widget
                    Anchors := anchors{ Minimum := vector2{X := 0.65, Y := 0.1}, Maximum := vector2{X := 0.95, Y := 0.9}}
                    Offsets := margin{Top := 0.0, Left := 0.0, Right := 0.0, Bottom := 0.0}
                    Alignment := vector2{X := 0.5, Y := 0.5}
                    ZOrder := 4
                canvas_slot:
                    Widget:= CloseButton
                    Anchors := anchors{ Minimum := vector2{X := 0.5, Y := 0.85}, Maximum := vector2{X := 0.5, Y := 0.85}}
                    Alignment := vector2{X := 0.5, Y := 1.0}
                    ZOrder := 5

        CloseButton.SetText(STM("                                    BACK                                    "))            
        CloseButton.OnClick().Subscribe(PlayerManager.ClosePopup)

        Section1.JoinButton.OnClick().Subscribe(OnPressSection1)
        Section2.JoinButton.OnClick().Subscribe(OnPressSection2)
        Section3.JoinButton.OnClick().Subscribe(OnPressSection3)

        return Result

    OnPressSection1(Wdata:widget_message):void=
        BossType := BossList[TeleportInfo.ID] or boss_enum.LyraTutorial

        PreviousTier := PlayerManager.GetAchievementLevel[BossType.AchievementName()] - 1 or -1
        if (PreviousTier < 1):
            return

        HasKey := BossType.CheckHasKey(PlayerManager, ?Spend := true)
        if (HasKey?):
            BossType.DropLoot(PlayerManager.GameManager, PlayerManager.Player, PreviousTier, ?SkipUpgrades := true)

    OnPressSection2(Wdata:widget_message):void=
        BossType := BossList[TeleportInfo.ID] or boss_enum.LyraTutorial

        CurrentTier := PlayerManager.GetAchievementLevel[BossType.AchievementName()] or -1
        if (CurrentTier < 1 or TeleportInfo.PlayersInFight.Length > 0 or TeleportInfo.BossQueue.Length > 0):
            return

        HasKey := BossType.CheckHasKey(PlayerManager, ?Spend := false)
        if (HasKey?):
            set TeleportInfo.BossQueueTier = CurrentTier
            JoinQueue()

    OnPressSection3(Wdata:widget_message):void=
        
        BossType := BossList[TeleportInfo.ID] or boss_enum.LyraTutorial
        CurrentTier := PlayerManager.GetAchievementLevel[BossType.AchievementName()] or -1
        if (TeleportInfo.PlayersInFight.Length > 0 or TeleportInfo.BossQueue.Length = 0):
            return
        if (CurrentTier < TeleportInfo.BossQueueTier or CurrentTier < 1 or TeleportInfo.BossQueueTier < 1):
            return

        HasKey := BossType.CheckHasKey(PlayerManager, ?Spend := false)
        if (HasKey?):
            JoinQueue()

    JoinQueue():void=
        TeleportInfo.AddToQueue(PlayerManager)

    LeaveQueue():void=
        TeleportInfo.RemoveFromQueue(PlayerManager)

    StartCountdown():void=
        TeleportInfo.StartCountdown()

MakeBossQueueSection<constructor>() := boss_queue_section:

boss_queue_section := class:
    var Widget : canvas = canvas{}
    BossIcon : texture_block = texture_block{DefaultImage:=Images.Empty, DefaultDesiredSize:= vector2{X := 150.0, Y := 150.0}}

    LabelText : text_block = text_block:
        DefaultTextColor:=NamedColors.White
        DefaultShadowColor:=NamedColors.Black   
        DefaultShadowOffset:=option{vector2{X:=2.0,Y:=2.0}}
        DefaultShadowOpacity:= 1.0

    NameText : text_block = text_block:
        DefaultTextColor:=NamedColors.White
        DefaultShadowColor:=NamedColors.Black   
        DefaultShadowOffset:=option{vector2{X:=2.0,Y:=2.0}}
        DefaultShadowOpacity:= 1.0

    TierText : text_block = text_block:
        DefaultTextColor:=NamedColors.White
        DefaultShadowColor:=NamedColors.Black   
        DefaultShadowOffset:=option{vector2{X:=2.0,Y:=2.0}}
        DefaultShadowOpacity:= 1.0
    
    var ExplanationText : stack_box = stack_box{Orientation:=orientation.Vertical}

    HealthText : text_block = text_block:
        DefaultTextColor:=NamedColors.Gold
        DefaultShadowColor:=NamedColors.Black
        DefaultShadowOffset:=option{vector2{X:=2.0,Y:=2.0}}
        DefaultShadowOpacity:= 1.0

    JoinButton : button_loud = button_loud:

    Create():void=
        set Widget = canvas:
            Slots := array:
                canvas_slot:
                    Widget:= LabelText
                    Anchors := anchors{ Minimum := vector2{X := 0.5, Y := 0.2}, Maximum := vector2{X := 0.5, Y := 0.2}}
                    Offsets := margin{Top := -50.0, Left := 0.0, Right := 0.0, Bottom := 0.0}
                    Alignment := vector2{X := 0.5, Y := 0.5}
                    ZOrder := 1
                canvas_slot:
                    Widget:= BossIcon
                    Anchors := anchors{ Minimum := vector2{X := 0.5, Y := 0.2}, Maximum := vector2{X := 0.5, Y := 0.2}}
                    Offsets := margin{Top := 30.0, Left := 0.0, Right := 0.0, Bottom := 0.0}
                    Alignment := vector2{X := 0.5, Y := 0.5}
                    ZOrder := 2
                canvas_slot:
                    Widget:= NameText
                    Anchors := anchors{ Minimum := vector2{X := 0.5, Y := 0.2}, Maximum := vector2{X := 0.5, Y := 0.2}}
                    Offsets := margin{Top := 190.0, Left := 0.0, Right := 0.0, Bottom := 0.0}
                    Alignment := vector2{X := 0.5, Y := 0.5}
                    ZOrder := 3
                canvas_slot:
                    Widget:= TierText
                    Anchors := anchors{ Minimum := vector2{X := 0.5, Y := 0.2}, Maximum := vector2{X := 0.5, Y := 0.2}}
                    Offsets := margin{Top := 240.0, Left := 0.0, Right := 0.0, Bottom := 0.0}
                    Alignment := vector2{X := 0.5, Y := 0.5}
                    ZOrder := 4
                canvas_slot:
                    Widget:= HealthText
                    Anchors := anchors{ Minimum := vector2{X := 0.5, Y := 0.2}, Maximum := vector2{X := 0.5, Y := 0.2}}
                    Offsets := margin{Top := 290.0, Left := 0.0, Right := 0.0, Bottom := 0.0}
                    Alignment := vector2{X := 0.5, Y := 0.5}
                    ZOrder := 5
                canvas_slot:
                    Widget:= JoinButton
                    Anchors := anchors{ Minimum := vector2{X := 0.5, Y := 0.2}, Maximum := vector2{X := 0.5, Y := 0.2}}
                    Offsets := margin{Top := 360.0, Left := 0.0, Right := 0.0, Bottom := 0.0}
                    Alignment := vector2{X := 0.5, Y := 0.5}
                    ZOrder := 6  

    Update(Label: string, Image : texture, Name : string, Tier : int, Health : big_number, StartText : string, ExplanationString : string):void=
        
        BossIcon.SetImage(Image)
        LabelText.SetText(STM(Label))
        NameText.SetText(STM(Name))
        JoinButton.SetText(STM(StartText))
            
        Widget.RemoveWidget(ExplanationText)
        set ExplanationText = WordWrap(ExplanationString, 20)
        Widget.AddWidget(
            canvas_slot:
                Widget:= ExplanationText
                Anchors := anchors{ Minimum := vector2{X := 0.5, Y := 0.2}, Maximum := vector2{X := 0.5, Y := 0.2}}
                Offsets := margin{Top := 420.0, Left := 0.0, Right := 0.0, Bottom := 0.0}
                Alignment := vector2{X := 0.5, Y := 0.0}
                ZOrder := 7
            )
        
        if:
            Tier <= 0
        then:
            TierText.SetText(STM("---"))
            HealthText.SetText(STM("---"))
        else:
            TierText.SetText(STM("Tier {Tier}"))
            HealthText.SetText(STM("{Health} HP"))

MakeBossTeleportMenu<constructor>(GameManagerIn : game_manager, TeleportManagerIn : boss_teleport_manager_device, AgentIn : agent) := boss_teleport_menu: 
    GameManager :=  GameManagerIn
    TeleportManager := TeleportManagerIn 
    Agent := AgentIn

boss_teleport_menu:= class:
    TeleportManager : boss_teleport_manager_device
    GameManager : game_manager
    Agent : agent

    var DialogBackground : texture_block = texture_block{ DefaultImage := Images.StoreBox, DefaultDesiredSize := vector2{X := 1920.0, Y := 1080.0}} 
    #var HeaderPlate : texture_block = texture_block{ DefaultImage := Images.DialogBox, DefaultDesiredSize := vector2{X := 400.0, Y := 80.0}}

    var NameButtons : []button_loud = array{}

    var CurrencyIcon1 : texture_block = texture_block{ DefaultImage:= Images.Empty, DefaultDesiredSize := vector2{X := 50.0, Y := 50.0}}
    var CurrencyIcon2 : texture_block = texture_block{ DefaultImage:= Images.Empty, DefaultDesiredSize := vector2{X := 50.0, Y := 50.0}}
    var CurrencyIcon3 : texture_block = texture_block{ DefaultImage:= Images.Empty, DefaultDesiredSize := vector2{X := 50.0, Y := 50.0}}
    var CurrencyIcon4 : texture_block = texture_block{ DefaultImage:= Images.Empty, DefaultDesiredSize := vector2{X := 50.0, Y := 50.0}}
    var CurrencyIcon5 : texture_block = texture_block{ DefaultImage:= Images.Empty, DefaultDesiredSize := vector2{X := 50.0, Y := 50.0}}
    var CurrencyIcon6 : texture_block = texture_block{ DefaultImage:= Images.Empty, DefaultDesiredSize := vector2{X := 50.0, Y := 50.0}}

    var CurrencyText1 : text_block = text_block{DefaultTextColor:=NamedColors.White,
            DefaultShadowColor:=NamedColors.Black,
            DefaultShadowOffset:=option{vector2{X:=2.0,Y:=2.0}}}
    var CurrencyText2 : text_block = text_block{DefaultTextColor:=NamedColors.White,
            DefaultShadowColor:=NamedColors.Black,
            DefaultShadowOffset:=option{vector2{X:=2.0,Y:=2.0}}}
    var CurrencyText3 : text_block = text_block{DefaultTextColor:=NamedColors.White,
            DefaultShadowColor:=NamedColors.Black,
            DefaultShadowOffset:=option{vector2{X:=2.0,Y:=2.0}}}
    var CurrencyText4 : text_block = text_block{DefaultTextColor:=NamedColors.White,
            DefaultShadowColor:=NamedColors.Black,  
            DefaultShadowOffset:=option{vector2{X:=2.0,Y:=2.0}}}
    var CurrencyText5 : text_block = text_block{DefaultTextColor:=NamedColors.White,
            DefaultShadowColor:=NamedColors.Black,  
            DefaultShadowOffset:=option{vector2{X:=2.0,Y:=2.0}}}
    var CurrencyText6 : text_block = text_block{DefaultTextColor:=NamedColors.White,
            DefaultShadowColor:=NamedColors.Black,
            DefaultShadowOffset:=option{vector2{X:=2.0,Y:=2.0}}}

    OpenMenu():void=
        if:
            InPlayer := player[Agent]
            InPlayer.IsActive[]
            PlayerUI := GetPlayerUI[InPlayer]
            PlayerManager := GameManager.PlayerManagers[InPlayer]
        then:
            CurrencyText1.SetShadowOpacity(1.0)    
            CurrencyText1.SetText(STM("{PlayerManager.GetCollectibleCount("BossKey")}"))  
            CurrencyText2.SetShadowOpacity(1.0)    
            CurrencyText2.SetText(STM("{PlayerManager.GetCollectibleCount("KeySharkey")}"))   
            CurrencyText3.SetShadowOpacity(1.0)    
            CurrencyText3.SetText(STM("{PlayerManager.GetCollectibleCount("KeyGenga")}"))   
            CurrencyText4.SetShadowOpacity(1.0)    
            CurrencyText4.SetText(STM("{PlayerManager.GetCollectibleCount("KeyUrjent")}"))  
            CurrencyText5.SetShadowOpacity(1.0)    
            CurrencyText5.SetText(STM("{PlayerManager.GetCollectibleCount("KeyGiant")}"))  
            CurrencyText6.SetShadowOpacity(1.0)    
            CurrencyText6.SetText(STM("{PlayerManager.GetCollectibleCount("KeyPirateKing")}"))

            CurrencyIcon1.SetImage(GetTextureFromString("BossKey"))
            CurrencyIcon2.SetImage(GetTextureFromString("KeySharkey"))
            CurrencyIcon3.SetImage(GetTextureFromString("KeyGenga"))
            CurrencyIcon4.SetImage(GetTextureFromString("KeyUrjent"))
            CurrencyIcon5.SetImage(GetTextureFromString("KeyGiant"))
            CurrencyIcon6.SetImage(GetTextureFromString("KeyPirateKing"))

            StackBoxOuter := stack_box{Orientation:=orientation.Vertical}
            var CurrentStackBox : stack_box = stack_box{Orientation:=orientation.Horizontal}
            StackBoxOuter.AddWidget(stack_box_slot:
                Widget:= CurrentStackBox
                Padding:= margin{Top := 0.0, Left := 0.0, Right := 0.0, Bottom := 0.0}
            )

            HeaderStackbox := MakeTextBox("Fight Bosses to earn new gear.<br>Teleporting will HEAL you to 100%. Keys are refunded on failure.", ?Alignment := horizontal_alignment.Center)
            
            CloseButton := button_quiet:
            CloseButton.SetText(STM("                                    BACK                                    "))            
            CloseButton.OnClick().Subscribe(PlayerManager.ClosePopup)

            Canvas := canvas:
                Slots := array:
                    canvas_slot:
                        Widget:= stack_box:
                            Orientation:= orientation.Horizontal
                            Slots:= array:
                                stack_box_slot:
                                    Widget:= CurrencyIcon1                                                    
                                    HorizontalAlignment:= horizontal_alignment.Left
                                    Padding:= margin{Left:=2.0, Top:=0.0, Bottom:= 0.0, Right:= -2.0} 
                                    VerticalAlignment:= vertical_alignment.Center
                                stack_box_slot:
                                    Widget:= CurrencyText1
                                    HorizontalAlignment:= horizontal_alignment.Left
                                    Padding:= margin{Left:=2.0, Top:=0.0, Bottom:= 0.0, Right:= 2.0} 
                                stack_box_slot:
                                    Widget:= CurrencyIcon2                                                    
                                    HorizontalAlignment:= horizontal_alignment.Left
                                    Padding:= margin{Left:=12.0, Top:=0.0, Bottom:= 0.0, Right:= -2.0} 
                                    VerticalAlignment:= vertical_alignment.Center
                                stack_box_slot:
                                    Widget:= CurrencyText2
                                    HorizontalAlignment:= horizontal_alignment.Left
                                    Padding:= margin{Left:=2.0, Top:=0.0, Bottom:= 0.0, Right:= 2.0} 
                                stack_box_slot:
                                    Widget:= CurrencyIcon3                                                    
                                    HorizontalAlignment:= horizontal_alignment.Left
                                    Padding:= margin{Left:=12.0, Top:=0.0, Bottom:= 0.0, Right:= -2.0} 
                                    VerticalAlignment:= vertical_alignment.Center
                                stack_box_slot:
                                    Widget:= CurrencyText3
                                    HorizontalAlignment:= horizontal_alignment.Left
                                    Padding:= margin{Left:=2.0, Top:=0.0, Bottom:= 0.0, Right:= 2.0} 
                                stack_box_slot:
                                    Widget:= CurrencyIcon4                                                 
                                    HorizontalAlignment:= horizontal_alignment.Left
                                    Padding:= margin{Left:=12.0, Top:=0.0, Bottom:= 0.0, Right:= -2.0} 
                                    VerticalAlignment:= vertical_alignment.Center
                                stack_box_slot:
                                    Widget:= CurrencyText4
                                    HorizontalAlignment:= horizontal_alignment.Left
                                    Padding:= margin{Left:=2.0, Top:=0.0, Bottom:= 0.0, Right:= 2.0} 
                                # stack_box_slot:
                                #     Widget:= CurrencyIcon5                                              
                                #     HorizontalAlignment:= horizontal_alignment.Left
                                #     Padding:= margin{Left:=12.0, Top:=0.0, Bottom:= 0.0, Right:= -2.0} 
                                #     VerticalAlignment:= vertical_alignment.Center
                                # stack_box_slot:
                                #     Widget:= CurrencyText5
                                #     HorizontalAlignment:= horizontal_alignment.Left
                                #     Padding:= margin{Left:=2.0, Top:=0.0, Bottom:= 0.0, Right:= 2.0} 
                                stack_box_slot:
                                    Widget:= CurrencyIcon6                                                    
                                    HorizontalAlignment:= horizontal_alignment.Left
                                    Padding:= margin{Left:=12.0, Top:=0.0, Bottom:= 0.0, Right:= -2.0} 
                                    VerticalAlignment:= vertical_alignment.Center
                                stack_box_slot:
                                    Widget:= CurrencyText6
                                    HorizontalAlignment:= horizontal_alignment.Left
                                    Padding:= margin{Left:=2.0, Top:=0.0, Bottom:= 0.0, Right:= 2.0} 
                        Anchors := anchors{ Minimum := vector2{X := 0.5, Y := 0.15}, Maximum := vector2{X := 0.5, Y := 0.15}}
                        Offsets := margin{Top := 20.0, Left := -40.0, Right := 0.0, Bottom := 0.0}
                        Alignment := vector2{X := 0.5, Y := 1.0}
                        ZOrder := 5
                    canvas_slot:
                        Widget:= DialogBackground
                        SizeToContent := true
                        Anchors := anchors{ Minimum := vector2{X := 0.1, Y := 0.1}, Maximum := vector2{X := 0.9, Y := 0.9}}  
                        ZOrder := 1
                    canvas_slot:
                        Widget:= HeaderStackbox
                        SizeToContent := true
                        Anchors := anchors{ Minimum := vector2{X := 0.5, Y := 0.25}, Maximum := vector2{X := 0.5, Y := 0.25}}
                        Offsets := margin{Top := 0.0, Left := -150.0, Right := 0.0, Bottom := 0.0}
                        Alignment := vector2{X := 0.5, Y := 0.5}
                        ZOrder := 2
                    canvas_slot:
                        Widget:= StackBoxOuter
                        SizeToContent := true
                        Anchors := anchors{ Minimum := vector2{X := 0.1, Y := 0.3}, Maximum := vector2{X := 0.9, Y := 0.9}}  
                        Alignment := vector2{X := 0.5, Y := 0.0}
                        ZOrder := 3
                    canvas_slot:
                        Widget:= CloseButton
                        Anchors := anchors{ Minimum := vector2{X := 0.5, Y := 0.85}, Maximum := vector2{X := 0.5, Y := 0.85}}
                        Alignment := vector2{X := 0.5, Y := 1.0}
                        ZOrder := 4
                      
            var ItemCount : int = 0
            var TeleporterID : int = 0
            for (Teleporter : TeleportManager.Teleporters):
                if (PlayerManager.GetAchievementLevel[Teleporter.TeleportInfo.Name] > 0 or TeleporterID = 0):
                    if (ItemCount >= 10):
                        set CurrentStackBox = stack_box{Orientation:=orientation.Horizontal}
                        StackBoxOuter.AddWidget(stack_box_slot:
                            Widget:= CurrentStackBox
                            Padding:= margin{Top := 0.0, Left := 0.0, Right := 0.0, Bottom := 00.0}
                        )
                        set ItemCount = 0
                    set ItemCount += 1

                    ItemIcon := texture_block:
                                DefaultImage := TeleportManager.GetTPIcon(TeleporterID)
                                DefaultDesiredSize := vector2{X := 200.0, Y := 200.0}
            
                    NameButton := button_loud:
                        DefaultText:= STM("{TeleportManager.GetTPName(TeleporterID)}")
                    
                    set NameButtons = NameButtons + array{NameButton}

                    #NameButton.OnClick().Subscribe(teleport_wrapper{Teleporter := Teleporter.TeleportInfo.Teleporter, PlayerManager:= PlayerManager, BossKeySpend := 1, BossKeyID := TeleporterID}.Teleport)
                    NameButton.OnClick().Subscribe(open_queue_menu_wrapper{TeleportManagerDevice := TeleportManager, Boss := BossList[Teleporter.ID] or boss_enum.LyraTutorial}.OpenWData)

                    if (not BigNumber_GreaterThanOrEquals(PlayerManager.GetCollectibleCount("BossKey"), big_number{Value:=1.0})?):
                        NameButton.SetEnabled(false)
                    else:
                        NameButton.SetEnabled(true)

                    var DailyBonusText : stack_box = MakeTextBox("<br>")
                    if:
                        Teleporter.ID = PlayerManager.GetDailyBoss()
                    then:
                        set DailyBonusText = MakeTextBox("Daily Bonus!<br>Better Loot!")

                    CostBox := case (TeleporterID):
                        0 => MakeTextBox("<BossKey>")
                        1 => MakeTextBox("<BossKey>or<KeySharkey>x10")
                        2 => MakeTextBox("<BossKey>or<KeyGenga>x10")
                        3 => MakeTextBox("<BossKey>or<KeyUrjent>x10")
                        4 => MakeTextBox("<BossKey>or<KeyGiant>x10")
                        5 => MakeTextBox("<BossKey>or<KeyPirateKing>x10")
                        _ => MakeTextBox("???")

                    Widget := canvas:
                        Slots:= array:
                            canvas_slot:
                                Widget:= DailyBonusText
                                Anchors := anchors{ Minimum := vector2{X := 0.5, Y := 1.0}, Maximum := vector2{X := 0.5, Y := 1.0}}                         
                                Offsets := margin{Top := 190.0, Left := 0.0, Right := 0.0, Bottom := 0.0}
                                Alignment := vector2{X := 0.5, Y := 1.0}
                                ZOrder := 3
                            canvas_slot:
                                Widget:= canvas:
                                    Slots:=array:
                                        canvas_slot:
                                            Widget:=ItemIcon
                                            SizeToContent := true
                                            Anchors := anchors{ Minimum := vector2{X := 0.0, Y := 0.0}, Maximum := vector2{X := 1.0, Y := 1.0}}                        
                                            Alignment := vector2{X := 0.0, Y := 1.0}
                                            ZOrder := 1
                                        #Ability Button
                                        canvas_slot:
                                            Widget:=NameButton
                                            SizeToContent := true
                                            Anchors := anchors{ Minimum := vector2{X := 0.5, Y := 1.0}, Maximum := vector2{X := 0.5, Y := 1.0}}                         
                                            Offsets := margin{Top := 100.0, Left := 0.0, Right := 0.0, Bottom := 0.0}
                                            Alignment := vector2{X := 0.5, Y := 1.0}
                                            ZOrder := 2
                                        canvas_slot:
                                            Widget:=CostBox
                                            Anchors := anchors{ Minimum := vector2{X := 0.5, Y := 1.0}, Maximum := vector2{X := 0.5, Y := 1.0}}                         
                                            Offsets := margin{Top := 140.0, Left := 0.0, Right := 0.0, Bottom := 0.0}
                                            Alignment := vector2{X := 0.5, Y := 1.0}
                                            ZOrder := 3
                    WidgetSlot := stack_box_slot:
                        Widget:= Widget
                        Padding:= margin{Top := 0.0, Left := 2.5, Right := 0.0, Bottom := 0.0}

                    CurrentStackBox.AddWidget(WidgetSlot)

                set TeleporterID += 1
            
            PlayerUI.AddWidget(Canvas, player_ui_slot{InputMode := ui_input_mode.All})
            if (set PlayerManager.UIManager?.Popup = option{Canvas}) {}
            
            spawn:
                CheckNameButtonActivity()

    CheckNameButtonActivity()<suspends>:void=
        loop:
            if:
                InPlayer := player[Agent]
                InPlayer.IsActive[]
                PlayerManager := GameManager.PlayerManagers[InPlayer]
                PlayerManager.UIManager?.Popup?
            then:
                
                for (i := 0..NameButtons.Length - 1; Btn := NameButtons[i]):
                    Remaining := PlayerManager.GetBossKeySecondsRemaining()

                    HasAltKey := case(i):
                        0 => false
                        1 => BigNumber_GreaterThanOrEquals(PlayerManager.GetCollectibleCount("KeySharkey"), big_number{Value:=10.0})
                        2 => BigNumber_GreaterThanOrEquals(PlayerManager.GetCollectibleCount("KeyGenga"), big_number{Value:=10.0})
                        3 => BigNumber_GreaterThanOrEquals(PlayerManager.GetCollectibleCount("KeyUrjent"), big_number{Value:=10.0})
                        #4 => BigNumber_GreaterThanOrEquals(PlayerManager.GetCollectibleCount("KeyGiant"), big_number{Value:=10.0})
                        4 => BigNumber_GreaterThanOrEquals(PlayerManager.GetCollectibleCount("KeyPirateKing"), big_number{Value:=10.0})
                        _ => false

                    if:
                        Remaining <= 0 or HasAltKey?
                    then:
                        Btn.SetText(STM("READY"))
                        Btn.SetEnabled(true)
                    else:
                        Btn.SetText(STM("{Remaining}s"))
                        Btn.SetEnabled(false)
                Sleep(1.0)
            else:
                Log("BREAKING MENU")
                break
