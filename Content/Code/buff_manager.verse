using { /Verse.org/Simulation }

buff := struct <persistable><concrete><public>:
    @editable
    Name <public>: string = ""
    
    @editable
    Amount <public>: big_number = big_number{}

    @editable
    Suffix <public>: string = ""

    @editable
    Duration <public>: float = 0.0

    @editable
    IsSecondary <public>: logic = false