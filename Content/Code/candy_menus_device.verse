using { /Fortnite.com/Devices }
using { /Fortnite.com/UI }
using { /UnrealEngine.com/Temporary/Diagnostics }
using { /UnrealEngine.com/Temporary/SpatialMath }
using { /UnrealEngine.com/Temporary/UI}
using { /Verse.org/Assets }
using { /Verse.org/Colors }
using { /Verse.org/Simulation }


candy_menus_device := class(creative_device, menu_item):

    @editable
    GameManager : ?game_manager = false
    
    OpenMenu<override>(Agent:agent):void=
        if (PlayerManager := GameManager?.PlayerManagers[Agent]):
            
            if (not GameManager?.PlayerManagers[Agent].UIManager?.Popup?):
                if (GM := GameManager?):
                    MakeCandyMenu(GM, Self, Agent).OpenMenu()
            else:
                block:#Log("DOUBLE MENU")

    GetTitle<override>():string=
        return "CANDY"

    GetIcon<override>():texture=
        return Images.CandyIcon

    IsEnabled<override>(PlayerManager : player_manager)<transacts>:logic=    
        return false

MakeCandyMenu<constructor>(GameManagerIn : game_manager, DeviceIn:candy_menus_device, AgentIn : agent) := candy_menu:
    GameManager :=  GameManagerIn
    CandyDevice := DeviceIn
    Agent := AgentIn

candy_menu:= class:
    GameManager : game_manager
    CandyDevice : candy_menus_device
    Agent : agent
    var GachaDisplay : ?gacha_display = false

    var DialogBackground : texture_block = texture_block{ DefaultImage := Images.DialogBox, DefaultDesiredSize := vector2{X := 1800.0, Y := 300.0}}
    var CandyPanel : color_block = color_block{ DefaultColor := NamedColors.Red, DefaultDesiredSize := vector2{X := 480.0 , Y := 1000.0}}
    var EatPanel : color_block = color_block{ DefaultColor := NamedColors.Yellow, DefaultDesiredSize := vector2{X := 480.0 , Y := 1000.0}}

    var StackBoxCandies : stack_box = stack_box{Orientation:=orientation.Vertical}
    var CandyButtons : []candy_button = array{}
    var CloseButton : button_quiet = button_quiet{}

    var StrengthNumbersHostWidget : overlay = overlay{}
    var StrengthNumbers1: []text_block  = array{}
    var StrengthImgs1: []texture_block  = array{}
    var StrengthColors: []color_block  = array{}
    var CurrentStrengthCounter : int = 0
    var ActiveNumbers : []int = array{}
    StrengthNumbersSize : vector2 = vector2{X := 55.0, Y := 55.0}

    IconBounce : float = 1.25
    GatherBounceIterations : float = 8.0
    var ActiveBounces : []texture_block = array{}
    var ActiveBounceSizes : []vector2 = array{}

    ToggleButtons(EnabledIn: logic):void=
        CloseButton.SetEnabled(EnabledIn)
        for (Button : CandyButtons):
            Button.BTN.SetEnabled(EnabledIn)

    MoveUINumbers()<suspends>:void=
        loop:
            Sleep(0.01)
            if:
                not GachaDisplay?
            then:
                return
            LoopStrengthNumbers()
            LoopActiveBounces()
            
    OpenMenu():void =
        GD := MakeGachaDisplay(Self)
        GD.Setup()
        SetupStrengthNumbers()

        spawn:
            MoveUINumbers()

        set GachaDisplay = option{GD}

        if:
            InPlayer := player[Agent]
            InPlayer.IsActive[]
            PlayerUI := GetPlayerUI[InPlayer]
            PlayerManager := GameManager.PlayerManagers[InPlayer]
        then:

            for (i := 1..6):
                if:
                    CT := PlayerManager.GetCollectibleCount("Candy{i}")
                    BigNumber_GreaterThanOrEquals(CT, big_number{Value:=1.0})?
                then:
                    CandyBtn := MakeCandyButton(i, CT, Self)
                    CandyBtn.Setup()
                    set CandyButtons = CandyButtons + array{CandyBtn}
                    StackBoxCandies.AddWidget(
                        
                        stack_box_slot:
                            Widget:= CandyBtn.Widget
                            Padding:= margin{Top := 0.0, Left := 0.0, Right := 0.0, Bottom := 0.0}
                        )

            CloseButton.SetText(STM("                                                                        BACK                                                                        "))            
            CloseButton.OnClick().Subscribe(ClosePopup)

            if:
                GatchaWidget := GachaDisplay?.Widget
                SetGatcha := GachaDisplay?
            then:
                SetGatcha.Update(0)

                Canvas := canvas:
                    Slots := array:
                        canvas_slot:
                            Widget:= DialogBackground
                            SizeToContent := true
                            Anchors := anchors{ Minimum := vector2{X := 0.1, Y := 0.1}, Maximum := vector2{X := 0.9, Y := 0.8}}
                            ZOrder := 1
                        canvas_slot: 
                            Widget:= StrengthNumbersHostWidget
                            Anchors := anchors{ Minimum := vector2{X := 0.4, Y := 0.5}, Maximum := vector2{X := 0.4, Y := 0.5}}
                            Offsets := margin{Top := 0.0, Left := 0.0, Right := 0.0, Bottom := 0.0}
                            Alignment := vector2{X := 0.0, Y := 0.5}
                            ZOrder := 2
                        canvas_slot:
                            Widget:= text_block:
                                DefaultTextColor:=NamedColors.White
                                DefaultShadowColor:=NamedColors.Black   
                                DefaultShadowOffset:=option{vector2{X:=2.0,Y:=2.0}}
                                DefaultShadowOpacity:= 1.0
                                DefaultText:= STM("Choose a candy to eat to gain permanent Strength!")
                            SizeToContent := true
                            Anchors := anchors{ Minimum := vector2{X := 0.5, Y := 0.15}, Maximum := vector2{X := 0.5, Y := 0.15}}
                            Offsets := margin{Top := 0.0, Left := 0.0, Right := 0.0, Bottom := 0.0}
                            Alignment := vector2{X := 0.5, Y := 0.5}
                            ZOrder := 2
                        canvas_slot:
                            Widget:= StackBoxCandies
                            SizeToContent := true
                            Anchors := anchors{ Minimum := vector2{X := 0.1, Y := 0.5}, Maximum := vector2{X := 0.35, Y := 0.5}}
                            Offsets := margin{Top := 25.0, Left := 25.0, Right := 25.0, Bottom := 25.0}
                            Alignment := vector2{X := 0.1, Y := 0.5} 
                            ZOrder := 3
                        # canvas_slot:
                        #     Widget:= color_block{DefaultColor := NamedColors.Red, DefaultDesiredSize := vector2{X := 1920.0, Y := 1080.0}}
                        #     SizeToContent := true
                        #     Anchors := anchors{ Minimum := vector2{X := 0.3, Y := 0.1}, Maximum := vector2{X := 0.9, Y := 0.9}}
                        #     Offsets := margin{Top := 25.0, Left := 25.0, Right := 25.0, Bottom := 25.0}
                        canvas_slot:
                            Widget:= GatchaWidget
                            SizeToContent := true
                            Anchors := anchors{ Minimum := vector2{X := 0.35, Y := 0.1}, Maximum := vector2{X := 0.9, Y := 0.9}}
                            Offsets := margin{Top := 25.0, Left := 25.0, Right := 25.0, Bottom := 25.0}
                            Alignment := vector2{X := 0.5, Y := 0.5}
                            ZOrder := 4
                        canvas_slot:
                            Widget:= CloseButton
                            Anchors := anchors{ Minimum := vector2{X := 0.5, Y := 0.755}, Maximum := vector2{X := 0.5, Y := 0.755}}
                            Alignment := vector2{X := 0.5, Y := 1.0}
                            ZOrder := 5

        
                PlayerUI.AddWidget(Canvas, player_ui_slot{InputMode := ui_input_mode.All})
                if (set PlayerManager.UIManager?.Popup = option{Canvas}) {}

    ClosePopup(Wdata:widget_message):void =
        set GachaDisplay = false
        if:
            InPlayer := player[Agent]
            InPlayer.IsActive[]
            PlayerUI := GetPlayerUI[InPlayer]
            PlayerManager := GameManager.PlayerManagers[InPlayer]
        then:
            PlayerManager.ClosePopup(Wdata)

    SetupStrengthNumbers():void=
        for (X:= 0..20):
            var TempStrengthText : text_block = text_block{DefaultTextColor:=NamedColors.White,
            DefaultShadowColor:=NamedColors.Black,
            DefaultShadowOffset:=option{vector2{X:=2.0,Y:=2.0}}}
            TempStrengthText.SetTextOpacity(0.0) 
            TempStrengthText.SetShadowOpacity(0.0) 

            var TempStrengthImg : texture_block = texture_block:
                DefaultImage := Images.StrengthIcon
                DefaultDesiredSize := StrengthNumbersSize

            var TempColor : color_block = color_block{}
            TempColor.SetOpacity(0.0)
            TempColor.SetDesiredSize(vector2{Y:=3000.0})

            set StrengthNumbers1 = StrengthNumbers1 + array{TempStrengthText}
            set StrengthImgs1 = StrengthImgs1 + array{TempStrengthImg}
            
            set StrengthColors = StrengthColors + array{TempColor}

            TempOverlay:=overlay_slot:
                Widget:=overlay:
                    Slots:=
                        array:
                            overlay_slot:
                                Widget:=TempColor 
                            overlay_slot:
                                Widget:= stack_box:
                                    Orientation:= orientation.Horizontal
                                    Slots:= array:
                                        stack_box_slot:
                                            Widget:= TempStrengthImg
                                            HorizontalAlignment:= horizontal_alignment.Left
                                            Padding:= margin{Left:=2.0, Top:=0.0, Bottom:= 0.0, Right:= 4.0} 
                                        stack_box_slot:
                                            Widget:= TempStrengthText
                                            Padding:= margin{Left:=0.0, Top:=0.0, Bottom:= 0.0, Right:= 8.0} 
    
                                VerticalAlignment:= vertical_alignment.Top
                                HorizontalAlignment:= horizontal_alignment.Left
                HorizontalAlignment:= horizontal_alignment.Left

            StrengthNumbersHostWidget.AddWidget(TempOverlay)
    
    ShowStrengthNumber(Strength : big_number):void=       
    
        if:
            Number1:=StrengthNumbers1[CurrentStrengthCounter]
            Img1:=StrengthImgs1[CurrentStrengthCounter]
            Color:=StrengthColors[CurrentStrengthCounter]

        then:
            Color.SetDesiredSize(vector2{Y:=32.0})

            Number1.SetText(STM(" + {Strength}"))
            Number1.SetTextColor(MakeColorFromHex("FFFFFF"))
            Img1.SetImage(Images.StrengthIcon)

            BeginBounceEffect(Img1, StrengthNumbersSize, IconBounce)

            Number1.SetTextOpacity(1.0)
            Number1.SetShadowOpacity(1.0)

            #Add to ActiveNumbers
            set ActiveNumbers = ActiveNumbers + array{CurrentStrengthCounter}
        
        #IncrementCounter without overflowing
        set CurrentStrengthCounter += 1
        if(CurrentStrengthCounter >= StrengthNumbers1.Length):
            set CurrentStrengthCounter = 0
            
    UpdateCandyCounts():void=
        if:
            InPlayer := player[Agent]
            PlayerManager := GameManager.PlayerManagers[InPlayer]
        then:
            for (btn : CandyButtons):
                btn.UpdateCount(PlayerManager)
    
    LoopStrengthNumbers():void=
        var IndicesToRemove : []int = array{}
        var ActiveNumIndex : int = 0

        #Efficient Loop through Active Numbers
        for (ActiveNumber:ActiveNumbers):
            BeginFadePosition:=400.0
            MaxFadePosition:=500.0
            if:
                Number1:=StrengthNumbers1[ActiveNumber]
                Color:=StrengthColors[ActiveNumber]
                Img1:=StrengthImgs1[ActiveNumber]
            then:
                Temp : float = Color.GetDesiredSize().Y
                if(Temp > BeginFadePosition):
                    SetOpacity(Number1, Number1, MaxFadePosition - Temp)
                    if(Temp > MaxFadePosition):
                        #Bump Position Off-Screen
                        Color.SetDesiredSize(vector2{X:=Color.GetDesiredSize().X,Y:=Color.GetDesiredSize().Y + 3000.0})
                        set IndicesToRemove = IndicesToRemove + array{ActiveNumIndex}
                Color.SetDesiredSize(vector2{X:=Color.GetDesiredSize().X,Y:=Color.GetDesiredSize().Y + 8.0})
                
                #Iterate Index
                set ActiveNumIndex += 1
            
        #Remove from ActiveNumbers
        for(Index:IndicesToRemove):
            if:
                TempValue:=ActiveNumbers[Index]
                TempArray:=ActiveNumbers.RemoveElement[Index]
            then:
                set ActiveNumbers = TempArray

            
    BeginBounceEffect(Img : texture_block, DefaultSize : vector2, BounceAmount : float):void=
        Img.SetDesiredSize(vector2{X:= DefaultSize.X * BounceAmount, Y:= DefaultSize.Y * BounceAmount})
        set ActiveBounces = ActiveBounces + array{Img}
        set ActiveBounceSizes = ActiveBounceSizes + array{DefaultSize}
    
    LoopActiveBounces():void=
        var IndicesToRemove : []int = array{}
        var ActiveNumIndex : int = 0

        #Efficient Loop through ActiveBounces
        for (ActiveBounce:ActiveBounces):
            if:
                ActiveBounceSize := ActiveBounceSizes[ActiveNumIndex]
            then:
                IterateBounceImageEffect(ActiveBounce, ActiveBounceSize, GatherBounceIterations)
                CurrentX := ActiveBounce.GetDesiredSize().X 
                
                if:
                    CurrentX = ActiveBounceSize.X
                then:
                    set IndicesToRemove = IndicesToRemove + array{ActiveNumIndex}
            #Iterate Index
            set ActiveNumIndex += 1

        #Remove from ActiveBounces
        for(Index:IndicesToRemove):
            if:
                TempValue:=ActiveBounces[Index]
                TempArray:=ActiveBounces.RemoveElement[Index]

                TempValue2:=ActiveBounceSizes[Index]
                TempArray2:=ActiveBounceSizes.RemoveElement[Index]
            then:
                set ActiveBounces = TempArray
                set ActiveBounceSizes = TempArray2
    
MakeCandyButton<constructor>(IDIn : int, CountIn : big_number, CandyMenuIn : candy_menu) := candy_button:
    ID := IDIn
    Count := CountIn
    CandyMenu := CandyMenuIn

candy_button := class():
    ID : int = 0
    var Count : big_number = big_number{Value:=0.0}
    CandyMenu : candy_menu

    var Widget : overlay = overlay{}
    var BTN : button_regular = button_regular{}

    Setup():void =

        set BTN = button_regular:
            DefaultText := STM("{Count}")
        
        BTN.OnClick().Subscribe(OnClick)

        set Widget = overlay:
            Slots:= array:
                overlay_slot:
                    Widget:=texture_block:
                        DefaultImage := GetTextureFromString("Candy{ID}")
                        DefaultDesiredSize := vector2{X := 100.0, Y := 100.0}
                    Padding:= margin{Top := 0.0, Left := 0.0, Right := 0.0, Bottom := 0.0}
                    HorizontalAlignment:= horizontal_alignment.Center
                    VerticalAlignment:= vertical_alignment.Center
                overlay_slot:
                    Widget:= BTN
                    Padding:= margin{Top := 0.0, Left := 120.0, Right := 0.0, Bottom := 0.0}
                    HorizontalAlignment:= horizontal_alignment.Center
                    VerticalAlignment:= vertical_alignment.Center

    UpdateCount(PlayerManager : player_manager):void=
        set Count = PlayerManager.GetCollectibleCount("Candy{ID}")
        Print (STM("Updating Count at {ID} with {Count}"))
        BTN.SetText(STM("{Count}"))

    OnClick(Wdata : widget_message):void =
        if:
            PlayerManager := CandyMenu.CandyDevice.GameManager?.PlayerManagers[Wdata.Player]
            BigNumber_GreaterThanOrEquals(big_number{Value:=0.0}, PlayerManager.GetCollectibleCount("Candy{ID}"))?
        then:
            return
        
        Print (STM("Updating GachaDisplay with {ID}"))
        if:
            GD := CandyMenu.GachaDisplay?
        then:
            GD.Update(ID)
            # CandyMenu.ToggleButtons(false)

MakeGachaDisplay<constructor>(CandyMenuIn : candy_menu) := gacha_display:
    CandyMenu := CandyMenuIn

gacha_display:= class():
    CandyMenu : candy_menu

    Widget : canvas = canvas{}
    
    NameText<private> : text_block = text_block:
        DefaultTextColor:=NamedColors.White,
        DefaultShadowColor:=NamedColors.Black,
        DefaultShadowOffset:=option{vector2{X:=4.0,Y:=4.0}}
    Icon<private> : texture_block = texture_block:
        DefaultImage := Images.Empty
        DefaultDesiredSize := vector2{X := 400.0, Y := 400.0}
    BuyButton<private> : button_regular = button_regular: 

    var SelectedCandy: int = 0

    Setup():void=
        # Widget.AddWidget(canvas_slot:
        #     Widget:= NameText
        #     Anchors := anchors{ Minimum := vector2{X := 0.5, Y := 0.1}, Maximum := vector2{X := 0.5, Y := 0.9}}
        #     Offsets := margin{Top := 0.0, Left := 0.0, Right := 0.0, Bottom := 0.0}
        #     Alignment := vector2{X := 0.5, Y := 0.0}
        # )
        Widget.AddWidget(canvas_slot:
            Widget:= Icon
            Anchors := anchors{ Minimum := vector2{X := 0.5, Y := 0.4}, Maximum := vector2{X := 0.5, Y := 0.4}} 
            Offsets := margin{Top := -25.0, Left := 0.0, Right := 0.0, Bottom := 400.0}
            Alignment := vector2{X := 0.5, Y := 0.5}
        )

        BuyButton.SetText(STM("      EAT      "))
        BuyButton.OnClick().Subscribe(OnEat)

    Update(Candy : int):void=
        
        if (Candy > 0):
            Widget.AddWidget(canvas_slot:
                Widget:= BuyButton
                Anchors := anchors{ Minimum := vector2{X := 0.7, Y := 0.2}, Maximum := vector2{X := 0.7, Y := 0.7}}
                Offsets := margin{Top := -25.0, Left := 100.0, Right := 0.0, Bottom := 0.0}
                Alignment := vector2{X := 0.5, Y := 0.5}
            )

        if (Candy = 0):
            # Widget.RemoveWidget(BuyButton)
            #Don't remove Widget

        set SelectedCandy = Candy

        case (Candy):
            1 => NameText.SetText(STM("Common Candy"))
            2 => NameText.SetText(STM("Uncommon Candy"))
            3 => NameText.SetText(STM("Rare Candy"))
            4 => NameText.SetText(STM("Super Rare Candy"))
            5 => NameText.SetText(STM("Epic Candy"))
            6 => NameText.SetText(STM("Legendary Candy"))
            _ => NameText.SetText(STM(""))
        
        Icon.SetImage(GetTextureFromString("Candy{Candy}"))
        Print (STM("Updated Gacha Display with {Candy}"))

    OnEat(Wdata : widget_message):void =
        if:
            PlayerManager := CandyMenu.CandyDevice.GameManager?.PlayerManagers[Wdata.Player]
            PlayerManager.GetCollectibleCount("Candy{SelectedCandy}").Value > 0.0
        then:
            Log("Eating Candy {SelectedCandy}")
            PlayerManager.GameManager.AudioManager.PlayEatCandySFX(PlayerManager.Agent)
            PlayerManager.UpdateQuestProgressForTask(task_event.EatCandy, big_number{Value:=1.0}, false)
            CandyMenu.ShowStrengthNumber(big_number{Value:=1.0, Exponent:=SelectedCandy})
            
            #TODO: Visuals to show what's happening
            PlayerManager.GainStrength(big_number{Value:=1.0, Exponent:=SelectedCandy})
            if (UI :=PlayerManager.UIManager?):
                UI.UpdateLevelUI()

            PlayerManager.GainCollectible(collectible_data{Key:="Candy{SelectedCandy}", Count:=big_number{Value:=-1.0}})
            CandyMenu.UpdateCandyCounts()

            if (BigNumber_GreaterThanOrEquals(big_number{Value:=0.0}, PlayerManager.GetCollectibleCount("Candy{SelectedCandy}"))?):
                Update(0)
        else:
            if:
                GM:=CandyMenu.CandyDevice.GameManager?
            then:
                GM.AudioManager.PlayUpgradeFailSFX(Wdata.Player)