using { /Fortnite.com/Characters }
using { /Fortnite.com/Devices }
using { /Verse.org/Simulation }

cinematic_manager_device := class(creative_device):
    
    @editable 
    BeginningCinematics : []cinematic_sequence_device = array{}

    @editable
    DataLayerLoadCinematic: ?cinematic_sequence_device = false

    PlayBeginningCinematic(Agent:agent, PlayerManager : player_manager)<suspends>:void=
        Sleep(0.0) # Wait 1 frame at game start
        if:
            Player:=player[Agent]
            Player.IsActive[]
            Character := Player.GetFortCharacter[]            
            Cinematic := BeginningCinematics[PlayerManager.CinematicID]
            UI:= PlayerManager.UIManager?
        then:
            set UI.ZoneMsgVisibilityTimer = 0.0
            # spawn:
            #     PlayerManager.HideAllGameUI()
            Character.PutInStasis(stasis_args{AllowFalling:=true})
            Cinematic.Play(Agent)
            Cinematic.StoppedEvent.Await()
            Character.ReleaseFromStasis()
            # spawn:
            #     PlayerManager.RespawnFixUI()
            
    PlayDataLoaderCinematic():void=
        if (Cinematic := DataLayerLoadCinematic?):
            Cinematic.Play()

