using { /Fortnite.com/Characters }
using { /Fortnite.com/Devices }
using { /Fortnite.com/UI }
using { /UnrealEngine.com/Temporary/UI }
using { /UnrealEngine.com/Temporary/SpatialMath }
using { /Verse.org/Colors }
using { /Verse.org/Simulation }

#an individual item that can be purchased
collectible_item:= class<concrete>:
    @editable
    var Name : string = "ITEM"

    @editable
    var Name_Override : string = ""
    
    @editable
    var Description : string = "ITEM"

    @editable
    var Currency : string = "Midas Coins"
    
    @editable
    var Count : big_number = big_number{ Value := 1.0}

    @editable
    var Cost : big_number = big_number{ Value := 50.0}

    var GameManager : game_manager = game_manager{}
    var CollectibleShop : ?player_collectible_shop = false

    Clone() : collectible_item = 
        Result :=collectible_item:
        set Result.Name = Name
        set Result.Name_Override = Name_Override
        set Result.Description = Description
        set Result.Currency = Currency
        set Result.Cost = Cost
        set Result.GameManager = GameManager
        set Result.CollectibleShop = CollectibleShop

        return Result

    OnBuy(Wdata:widget_message, Index : int):void =
        if:
            Player := Wdata.Player
            PlayerUI := GetPlayerUI[Player]
            PlayerProfile := PlayerProfileDataMap[Player]
            PlayerManager := GameManager.PlayerManagers[Player]
            PlayerLevel := PlayerManager.GetLevel[]
            MyShop := CollectibleShop?
        then:
            CurrentCurrency := PlayerManager.GetCollectibleCount(Currency)
            
            if:
                BigNumber_GreaterThanOrEquals(CurrentCurrency, Cost)?
            then:
                if:
                    GetCharsFromString(Name, 5) = "Candy"
                    Fort:= PlayerManager.Agent.GetFortCharacter[]
                    Fort.IsActive[]
                    Transform := Fort.GetTransform()
                then:
                    PlayerManager.UseCollectible(collectible_data{Key:=Name, Count:= Count})
                    GameManager.AudioManager.PickCandyCheerSound.Play(PlayerManager.Agent)
                    GameManager.QueueConfetti(Transform.Translation)
                    if:
                        CS:=CollectibleShop?
                        CandyID := ParseInt[Name.Slice[5, 6]]
                    then:
                        Str := big_number{Value:=1.0, Exponent:=CandyID} * PlayerManager.GetPlayerStat(player_stat.DamageMult)
                        CS.FloatingText.Create(GetTextureFromString("strength"), "+ {Str}") 
                else:
                    PlayerManager.GainCollectible(collectible_data{Key:=Name, Count:= Count})
                    if:
                        CS:=CollectibleShop?
                    then:
                        CS.FloatingText.Create(GetTextureFromString(Name), "{Count}") 

                PlayerManager.GainCollectible(collectible_data{Key:=Currency, Count:= Cost * -1.0})

                spawn:
                    DelayedCost()

                PlayerManager.UpdateQuestProgressForTask(task_event.BoughtItem, Count, false, ?Target := Name, ?IsPersistent := false)
                GameManager.AudioManager.PlayUpgradeSuccessSFX(Player)

                MyShop.UpdateButtonsText(Player)
                MyShop.BeginBounceEffect(Index)
            else:
                #Not enough gems
                GameManager.AudioManager.PlayUpgradeFailSFX(Player)
        else:
            Log("Something went wrong buying item {Name}")


    OnBuyAll(Wdata:widget_message, Index : int):void =
        if:
            Player := Wdata.Player
            PlayerUI := GetPlayerUI[Player]
            PlayerProfile := PlayerProfileDataMap[Player]
            PlayerManager := GameManager.PlayerManagers[Player]
            PlayerLevel := PlayerManager.GetLevel[]
            MyShop := CollectibleShop?
        then:
            CurrentCurrency := PlayerManager.GetCollectibleCount(Currency)

            if:
                CurrentCurrency.Value > 0.0

                GetCharsFromString(Name, 5) = "Candy"
                Fort:= PlayerManager.Agent.GetFortCharacter[]
                Fort.IsActive[]
                Transform := Fort.GetTransform()
            then:
                PlayerManager.GainCollectible(collectible_data{Key:=Currency, Count:= CurrentCurrency * -1.0})
                GameManager.AudioManager.PlayEatCandySFX(PlayerManager.Agent)
                PlayerManager.UpdateQuestProgressForTask(task_event.EatCandy, big_number{Value:=1.0}, false)

                Str := CurrentCurrency / 10.0 
                PlayerManager.GainStrength(Str)

                if (UI := PlayerManager.UIManager?):
                    UI.UpdateLevelUI()
                        
                GameManager.AudioManager.PickCandyCheerSound.Play(PlayerManager.Agent)
                GameManager.QueueConfetti(Transform.Translation)
                if:
                    CS:=CollectibleShop?
                    CandyID := ParseInt[Name.Slice[5, 6]]
                then:
                    CS.FloatingText.Create(GetTextureFromString("strength"), "+ {Str * PlayerManager.GetPlayerStat(player_stat.DamageMult)}") 

                spawn:
                    DelayedCost(?Amount := option{CurrentCurrency})
    
                PlayerManager.UpdateQuestProgressForTask(task_event.BoughtItem, Count, false, ?Target := Name, ?IsPersistent := false)
                GameManager.AudioManager.PlayUpgradeSuccessSFX(Player)
    
                MyShop.UpdateButtonsText(Player)
                MyShop.BeginBounceEffect(Index)
        else:
            Log("Something went wrong buying item {Name}")
    
    DelayedCost(?Amount : ?big_number = false)<suspends>:void=
        Sleep(0.25)
        if:
            CS:=CollectibleShop?
        then:
            if: 
                SetAmt:= Amount?
            then:
                CS.FloatingText.Create(GetTextureFromString(Currency), "{SetAmt * -1.0}") 
            else:
                CS.FloatingText.Create(GetTextureFromString(Currency), "{Cost * -1.0}") 
                

collectible_shop:= class(creative_device):

    @editable:
    GameManager : game_manager = game_manager{}

    @editable
    var Title : string = "SHOP"

    @editable
    var Items : []collectible_item = array{}

    OnBegin<override>()<suspends>:void=
        block:

    TryOpenMenu(Name : string, InPlayer:agent, ?NPCContactMenu : ?npc_contact_device = false):void=
        if:
            PlayerManager:=GameManager.PlayerManagers[InPlayer]
            not (PlayerManager.UIManager?.Popup?)
        then:
            var NewItems : []collectible_item = array{}
            for (Item : Items):
                set NewItems = NewItems + array{Item.Clone()}

            MakeCollectibleShop(GameManager, Name, Title, NewItems, InPlayer, NPCContactMenu).OpenMenu()
        else:
            block:#Log("DOUBLE MENU")
                    
MakeCollectibleShop<constructor>(GameManagerIn : game_manager, InName : string, InTitle : string, ItemsIn : []collectible_item, AgentIn : agent, InMenu : ?npc_contact_device) := player_collectible_shop:
    GameManager :=  GameManagerIn
    Name:= InName
    Title := InTitle
    Items := ItemsIn
    Agent := AgentIn
    Menu := InMenu

player_collectible_shop:= class:
    GameManager : game_manager
    Name : string = "NPC"
    Title : string = "SHOP"
    Items : []collectible_item
    Agent : agent
    Menu : ?npc_contact_device
    
    var ItemButtons : [player][]text_button_base = map{}
    var ItemButtonsAll : [player][]text_button_base = map{}
    var ButtonIcons : [player][]texture_block = map{}
    var ButtonIconsAll : [player][]texture_block = map{}
    StoreBackground : texture_block = texture_block{ DefaultImage := Images.StoreBox, DefaultDesiredSize := vector2{X := 2000.0, Y := 1000.0}}
    GemPlate : texture_block = texture_block{ DefaultImage := Images.DialogBox, DefaultDesiredSize := vector2{X := 300.0, Y := 60.0}}    
    
    var CurrencyIcon1 : texture_block = texture_block{ DefaultImage:= Images.Empty, DefaultDesiredSize := vector2{X := 50.0, Y := 50.0}}
    var CurrencyIcon2 : texture_block = texture_block{ DefaultImage:= Images.Empty, DefaultDesiredSize := vector2{X := 50.0, Y := 50.0}}
    var CurrencyIcon3 : texture_block = texture_block{ DefaultImage:= Images.Empty, DefaultDesiredSize := vector2{X := 50.0, Y := 50.0}}

    FloatingText : ui_bounce_mover = ui_bounce_mover{}

    var CurrencyText1 : text_block = text_block{DefaultTextColor:=NamedColors.White,
            DefaultShadowColor:=NamedColors.Black,
            DefaultShadowOffset:=option{vector2{X:=2.0,Y:=2.0}}}
    var CurrencyText2 : text_block = text_block{DefaultTextColor:=NamedColors.White,
            DefaultShadowColor:=NamedColors.Black,
            DefaultShadowOffset:=option{vector2{X:=2.0,Y:=2.0}}}
    var CurrencyText3 : text_block = text_block{DefaultTextColor:=NamedColors.White,
            DefaultShadowColor:=NamedColors.Black,
            DefaultShadowOffset:=option{vector2{X:=2.0,Y:=2.0}}}

    Size : vector2 = vector2{X := 50.0, Y := 50.0}
    IconBounce : float = 1.25
    GatherBounceIterations : float = 8.0

    BeginBounceEffect(Index : int):void=
        if:
            Item := Items[Index]
        then:
            Txt1 := CurrencyText1.GetText()
            Txt2 := CurrencyText2.GetText()
            Txt3 := CurrencyText3.GetText()

            if (Txt1 = Item.Currency):
                CurrencyIcon1.SetDesiredSize(vector2{X:= Size.X * IconBounce, Y:= Size.Y * IconBounce})
                spawn:
                    IterateBounceEffect(CurrencyIcon1)  
            else if (Txt2 = Item.Currency):
                CurrencyIcon2.SetDesiredSize(vector2{X:= Size.X * IconBounce, Y:= Size.Y * IconBounce})
                spawn:
                    IterateBounceEffect(CurrencyIcon2)  
            else if (Txt3 = Item.Currency):
                CurrencyIcon3.SetDesiredSize(vector2{X:= Size.X * IconBounce, Y:= Size.Y * IconBounce})
                spawn:
                    IterateBounceEffect(CurrencyIcon3)  

    IterateBounceEffect(Img : texture_block)<suspends>:void =
        limit:int=12

        for(X:=1..limit):
            Sleep(0.01)
            CurrentX := Img.GetDesiredSize().X 
            if(CurrentX > Size.X):
                CurrentY := Img.GetDesiredSize().Y
                ShrinkX := Max(CurrentX * ( IconBounce - 1.0 ) / GatherBounceIterations, 0.25)
                ShrinkY := Max(CurrentY * ( IconBounce - 1.0 ) / GatherBounceIterations, 0.25)
                Img.SetDesiredSize(vector2{X := Max(CurrentX - ShrinkX, Size.X), Y:= Max(CurrentY - ShrinkY, Size.Y)})
        
    OnClose(Msg:widget_message):void=
        
        if:
            InPlayer := player[Agent]
            PlayerManager := GameManager.PlayerManagers[InPlayer]
        then:
            PlayerManager.ClosePopup(Msg)
            if:
                SetMenu := Menu?
            then:
                SetMenu.OpenMenu(Agent)

    OpenMenu():void =
        spawn:
            FloatingText.Activate()

        if:
            InPlayer := player[Agent]
            PlayerUI := GetPlayerUI[InPlayer]
            PlayerManager := GameManager.PlayerManagers[InPlayer]
            UIManager := PlayerManager.UIManager?
            Gems:= PlayerManager.GetGems[]
        then:
            CurrencyText1.SetShadowOpacity(1.0)    
            CurrencyText1.SetText(STM(""))  
            CurrencyText2.SetShadowOpacity(1.0)    
            CurrencyText2.SetText(STM(""))   
            CurrencyText3.SetShadowOpacity(1.0)    
            CurrencyText3.SetText(STM(""))   

            AbilityBGStackBox := stack_box{Orientation:=orientation.Horizontal}
                
            if:
                not ItemButtons[InPlayer]
                set ItemButtons[InPlayer] = array{}
                set ItemButtonsAll[InPlayer] = array{}
                set ButtonIcons[InPlayer] = array{}
                set ButtonIconsAll[InPlayer] = array{}
            then:
                var i : int = 0

                for (Item:Items):
                    ItemButton := button_regular:
                    ItemButtonAll := button_regular:

                    set Item.GameManager = GameManager
                    set Item.CollectibleShop = option{Self}
                    
                    ItemButton.OnClick().SubscribeWidget(Item.OnBuy, i)
                    ItemButtonAll.OnClick().SubscribeWidget(Item.OnBuyAll, i)

                    if:
                        TempArray := ItemButtons[InPlayer] 
                        set ItemButtons[InPlayer] = TempArray + array{ItemButton}

                    if:
                        TempArray := ItemButtonsAll[InPlayer] 
                        set ItemButtonsAll[InPlayer] = TempArray + array{ItemButtonAll}
    
                    ButtonIcon := texture_block{DefaultImage := Images.Empty, DefaultDesiredSize := vector2{X := 50.0, Y := 50.0}}
                    ButtonIconAll := texture_block{DefaultImage := Images.Empty, DefaultDesiredSize := vector2{X := 50.0, Y := 50.0}}
                    
                    if:
                        TempArray := ButtonIcons[InPlayer] 
                        set ButtonIcons[InPlayer] = TempArray + array{ButtonIcon}

                    if:
                        TempArray := ButtonIconsAll[InPlayer] 
                        set ButtonIconsAll[InPlayer] = TempArray + array{ButtonIconAll}
                    
                    set i += 1
                    
            var Index : int = 0
            for (Btn:ItemButtons[InPlayer]):
                if:
                    Item := Items[Index]
                    ButtonIcon := ButtonIcons[InPlayer][Index]
                    ButtonIconAll := ButtonIconsAll[InPlayer][Index]
                    BtnAll := ItemButtonsAll[InPlayer][Index]
                then:
                    AbilityStack := stack_box:
                        Orientation:= orientation.Vertical
                    
                    CollectiblePlate : texture_block = texture_block{ DefaultImage := Images.DialogBoxVertical, DefaultDesiredSize := vector2{X := 610.0, Y := 720.0}}

                    TempSlot := stack_box_slot:
                        Widget:= overlay:
                            Slots:= array:
                                overlay_slot:
                                    Widget:=CollectiblePlate
                                    HorizontalAlignment:= horizontal_alignment.Center
                                overlay_slot:
                                    Widget:=AbilityStack
                                    HorizontalAlignment:= horizontal_alignment.Center
                                    VerticalAlignment:= vertical_alignment.Center
                        Padding:= margin{Top := 0.0, Left := 2.5, Right := 0.0, Bottom := 00.0}
                    AbilityBGStackBox.AddWidget(TempSlot)

                    #Item Name
                    var DisplayName : string = ""
                    var TextureLookupString : string = ""
                    var TempArray : []stack_box_slot = array{}
                    
                    PlayerManager.GainAchievementLevel("InteractedWithNPC_{Name}", 1)
                    set DisplayName = Item.Name
                    set TextureLookupString = Item.Name
                    set TempArray = ParseText(Item.Description)
                    if (Item.Name_Override <> "") { set DisplayName = Item.Name_Override }

                    AbilityStack.AddWidget(
                        stack_box_slot:
                            Widget:= text_block:
                                DefaultTextColor:=NamedColors.White
                                DefaultShadowColor:=NamedColors.Black
                                DefaultShadowOffset:=option{vector2{X:=2.0,Y:=2.0}}
                                DefaultShadowOpacity:= 1.0
                                DefaultText:= STM("{DisplayName}")
                        )

                
                    #Ability Icon
                    AbilityStack.AddWidget(
                        stack_box_slot:
                            Widget:= texture_block:
                                DefaultImage:=  GetTextureFromString(TextureLookupString)
                                DefaultDesiredSize := vector2{X:= 250.0, Y:= 250.0}
                            Padding:= margin{Top := 5.0, Left := 0.0, Right := 0.0, Bottom := 0.0}                                
                        )

                    #Ability Description
                    DescBox1 := stack_box{Orientation:=orientation.Horizontal}
                    DescBox2 := stack_box{Orientation:=orientation.Horizontal}
                    DescBox3 := stack_box{Orientation:=orientation.Horizontal}
                    VerticalDescBox : stack_box = stack_box:
                        Orientation := orientation.Vertical
                        Slots:= array:
                            stack_box_slot:
                                Widget:=DescBox1
                                HorizontalAlignment:= horizontal_alignment.Left
                                Padding:= margin{Top := 0.0, Left := 10.0, Right := 0.0, Bottom := 0.0}
                            stack_box_slot:
                                Widget:=DescBox2
                                HorizontalAlignment:= horizontal_alignment.Left
                                Padding:= margin{Top := -5.0, Left := 10.0, Right := 0.0, Bottom := 0.0}
                            stack_box_slot:
                                Widget:=DescBox3
                                HorizontalAlignment:= horizontal_alignment.Left
                                Padding:= margin{Top := -5.0, Left := 10.0, Right := 0.0, Bottom := 0.0}
                    
                    AbilityStack.AddWidget(
                        stack_box_slot:
                            Widget:= VerticalDescBox
                            Padding:= margin{Top := 10.0, Left := 0.0, Right := 0.0, Bottom := 0.0}
                        )
                    
                    var BreakCount : int = 0
                    for(Slot : TempArray):
                        var Skip : logic = false
                        if:
                            TempBlock:=text_block[Slot.Widget]
                        then:
                            TempText:=TempBlock.GetText()
                            if (TempText = "<br>"):
                                set BreakCount += 1
                                set Skip = true
                        
                        if(not(Skip?)):
                            case(BreakCount):
                                1 => DescBox2.AddWidget(Slot)
                                2 => DescBox3.AddWidget(Slot)
                                _ => DescBox1.AddWidget(Slot)

                    #Ability Button
                    AbilityStack.AddWidget(
                        stack_box_slot:
                            Widget:= overlay:
                                Slots:= array:
                                    overlay_slot:
                                        Widget:= Btn
                                        Padding:= margin{Top := 00.0, Left := 0.0, Right := 0.0, Bottom := 00.0}
                                    overlay_slot:
                                        Widget:= ButtonIcon
                                        HorizontalAlignment := horizontal_alignment.Right
                                        Padding:= margin{Top := -20.0, Left := -20.0, Right := 10.0, Bottom := -20.0}
                            Padding:= margin{Top := 20.0, Left := 0.0, Right := 0.0, Bottom := 10.0}
                        )

                    AbilityStack.AddWidget(
                        stack_box_slot:
                            Widget:= overlay:
                                Slots:= array:
                                    overlay_slot:
                                        Widget:= BtnAll
                                        Padding:= margin{Top := 00.0, Left := 0.0, Right := 0.0, Bottom := 00.0}
                                    overlay_slot:
                                        Widget:= ButtonIconAll
                                        HorizontalAlignment := horizontal_alignment.Right
                                        Padding:= margin{Top := -20.0, Left := -20.0, Right := 10.0, Bottom := -20.0}
                            Padding:= margin{Top := 20.0, Left := 0.0, Right := 0.0, Bottom := 10.0}
                        )
                    
                set Index += 1

            UpdateButtonsText(InPlayer)

            CloseButton := button_quiet:
            CloseButton.SetText(STM("                                    BACK                                    "))            
            CloseButton.OnClick().Subscribe(OnClose)

            Canvas := canvas:
                Slots := array:
                    canvas_slot:
                        Widget:= StoreBackground
                        SizeToContent := true
                        Anchors := anchors{ Minimum := vector2{X := 0.005, Y := 0.95}, Maximum := vector2{X := 0.005, Y := 0.95}}
                        Alignment := vector2{X := 0.0, Y := 1.0}
                        ZOrder := 1
                    canvas_slot:
                        Widget := AbilityBGStackBox
                        Anchors := anchors{ Minimum := vector2{X := 0.5, Y := 0.82}, Maximum := vector2{X := 0.5, Y := 0.82}}                        
                        Alignment := vector2{X := 0.5, Y := 1.0}
                        ZOrder := 2
                    canvas_slot:
                        Widget:= FloatingText.Setup(20, vector2{X := 55.0, Y := 55.0}, 2.0, 0.95)
                        Anchors := anchors{ Minimum := vector2{X := 0.64, Y := 0.4}, Maximum := vector2{X := 0.64, Y := 0.4}}  
                        Offsets := margin{Top := 0.0, Left := 0.0, Right := 0.0, Bottom := 0.0}
                        Alignment := vector2{X := 0.5, Y := 0.5}
                        ZOrder := 13
                    canvas_slot:
                        Widget:= text_block:
                            DefaultTextColor:=NamedColors.White
                            DefaultShadowColor:=NamedColors.Black
                            DefaultShadowOffset:=option{vector2{X:=2.0,Y:=2.0}}
                            DefaultShadowOpacity:= 1.0
                            DefaultText:= STM("{Title}")
                        Anchors := anchors{ Minimum := vector2{X := 0.5, Y := 0.95}, Maximum := vector2{X := 0.5, Y := 0.95}}
                        Offsets := margin{Top := -900.0, Left := 0.0, Right := 0.0, Bottom := 0.0}
                        Alignment := vector2{X := 0.5, Y := 1.0}
                        ZOrder := 3
                    canvas_slot:
                        Widget:= GemPlate
                        SizeToContent := true
                        Anchors := anchors{ Minimum := vector2{X := 0.75, Y := 0.95}, Maximum := vector2{X := 0.75, Y := 0.95}}
                        Offsets := margin{Top := -865.0, Left := 0.0, Right := 0.0, Bottom := 0.0}
                        Alignment := vector2{X := 0.0, Y := 1.0}
                        ZOrder := 4
                    canvas_slot:
                        Widget:= stack_box:
                            Orientation:= orientation.Horizontal
                            Slots:= array:
                                stack_box_slot:
                                    Widget:= CurrencyIcon1                                                    
                                    HorizontalAlignment:= horizontal_alignment.Left
                                    Padding:= margin{Left:=2.0, Top:=0.0, Bottom:= 0.0, Right:= -2.0} 
                                    VerticalAlignment:= vertical_alignment.Center
                                stack_box_slot:
                                    Widget:= CurrencyText1
                                    HorizontalAlignment:= horizontal_alignment.Left
                                    Padding:= margin{Left:=2.0, Top:=0.0, Bottom:= 0.0, Right:= 2.0} 
                                stack_box_slot:
                                    Widget:= CurrencyIcon2                                                    
                                    HorizontalAlignment:= horizontal_alignment.Left
                                    Padding:= margin{Left:=12.0, Top:=0.0, Bottom:= 0.0, Right:= -2.0} 
                                    VerticalAlignment:= vertical_alignment.Center
                                stack_box_slot:
                                    Widget:= CurrencyText2
                                    HorizontalAlignment:= horizontal_alignment.Left
                                    Padding:= margin{Left:=2.0, Top:=0.0, Bottom:= 0.0, Right:= 2.0} 
                                stack_box_slot:
                                    Widget:= CurrencyIcon3                                                    
                                    HorizontalAlignment:= horizontal_alignment.Left
                                    Padding:= margin{Left:=12.0, Top:=0.0, Bottom:= 0.0, Right:= -2.0} 
                                    VerticalAlignment:= vertical_alignment.Center
                                stack_box_slot:
                                    Widget:= CurrencyText3
                                    HorizontalAlignment:= horizontal_alignment.Left
                                    Padding:= margin{Left:=2.0, Top:=0.0, Bottom:= 0.0, Right:= 2.0} 
                        Anchors := anchors{ Minimum := vector2{X := 0.75, Y := 0.95}, Maximum := vector2{X := 0.75, Y := 0.95}}
                        Offsets := margin{Top := -870.0, Left := 20.0, Right := 0.0, Bottom := 0.0}
                        Alignment := vector2{X := 0.0, Y := 1.0}
                        ZOrder := 5
                    canvas_slot:
                        Widget:= CloseButton
                        Anchors := anchors{ Minimum := vector2{X := 0.5, Y := 0.89}, Maximum := vector2{X := 0.5, Y := 0.89}}
                        Alignment := vector2{X := 0.5, Y := 1.0}
                        Offsets := margin{Top := 0.0, Left := 0.0, Right := 0.0, Bottom := 0.0}
                        ZOrder := 6

            PlayerUI.AddWidget(Canvas, player_ui_slot{InputMode := ui_input_mode.All})
            if (set GameManager.PlayerManagers[InPlayer].UIManager?.Popup = option{Canvas}) {}
        
    UpdateButtonsText(InPlayer : player):void =
        if:        
            PlayerManager := GameManager.PlayerManagers[InPlayer]
            PlayerLevel := PlayerManager.GetLevel[]
        then:   
            #Update Currency Displays
            var CurrentCurrencies : []string = array{}
            for (Item : Items):
                CurrencyName := Item.Currency
                if (CurrentCurrencies.Find[CurrencyName]):
                    set CurrentCurrencies += array{"SKIP"}
                else:
                    set CurrentCurrencies += array{CurrencyName}

            var CurrencyID : int = 0
            for (Currency : CurrentCurrencies):
                if (Currency = "SKIP"):
                    case (CurrencyID):
                        0 => CurrencyText1.SetVisibility(widget_visibility.Collapsed)
                        1 => CurrencyText2.SetVisibility(widget_visibility.Collapsed)
                        2 => CurrencyText3.SetVisibility(widget_visibility.Collapsed)
                        _ => Log("ERROR: Invalid Currency ID")
                    case (CurrencyID):
                        0 => CurrencyIcon1.SetVisibility(widget_visibility.Collapsed)
                        1 => CurrencyIcon2.SetVisibility(widget_visibility.Collapsed)
                        2 => CurrencyIcon3.SetVisibility(widget_visibility.Collapsed)
                        _ => Log("ERROR: Invalid Currency ID")
                else:
                    case (CurrencyID):
                        0 => CurrencyText1.SetVisibility(widget_visibility.Visible)
                        1 => CurrencyText2.SetVisibility(widget_visibility.Visible)
                        2 => CurrencyText3.SetVisibility(widget_visibility.Visible)
                        _ => Log("ERROR: Invalid Currency ID")
                    case (CurrencyID):
                        0 => CurrencyIcon1.SetVisibility(widget_visibility.Visible)
                        1 => CurrencyIcon2.SetVisibility(widget_visibility.Visible)
                        2 => CurrencyIcon3.SetVisibility(widget_visibility.Visible)
                        _ => Log("ERROR: Invalid Currency ID")
                    case (CurrencyID):
                        0 => CurrencyText1.SetText(STM("{PlayerManager.GetCollectibleCount(Currency)}"))
                        1 => CurrencyText2.SetText(STM("{PlayerManager.GetCollectibleCount(Currency)}"))
                        2 => CurrencyText3.SetText(STM("{PlayerManager.GetCollectibleCount(Currency)}"))
                        _ => Log("ERROR: Invalid Currency ID")
                    case (CurrencyID):
                        0 => CurrencyIcon1.SetImage(GetTextureFromString(Currency))
                        1 => CurrencyIcon2.SetImage(GetTextureFromString(Currency))
                        2 => CurrencyIcon3.SetImage(GetTextureFromString(Currency))
                        _ => Log("ERROR: Invalid Currency ID")
                set CurrencyID += 1

            #Update buttons
            var i : int = 0
            for (Item:Items):
                if:
                    ItemButton := ItemButtons[InPlayer][i]
                    ItemButtonAll := ItemButtonsAll[InPlayer][i]
                    ItemIcon := ButtonIcons[InPlayer][i]
                    ItemIconAll := ButtonIconsAll[InPlayer][i]
                    var CostString: string = "MAXED" 
                then:
                    var Cost : big_number = big_number{}

                    CurrencyCount := PlayerManager.GetCollectibleCount(Item.Currency)
                    set Cost = Item.Cost
                    set CostString = "{Cost} {Item.Currency}"

                    #Update Item Value Texts
                    var ValueText: message = STM("")

                    var TempMsg : message = STM("[{Item.Count}]")

                    if (CostString = "MAXED"):
                        ItemButton.SetText(STM("MAX"))
                        ItemIcon.SetImage(Images.Empty)
                        ItemIconAll.SetImage(Images.Empty)
                    else:
                        ItemButton.SetText(STM("BUY: {Cost}    "))
                        ItemButtonAll.SetText(STM("BUY ALL: {CurrencyCount}    "))
                        ItemIcon.SetImage(GetTextureFromString(Item.Currency))
                        ItemIconAll.SetImage(GetTextureFromString(Item.Currency))
                    set i = i + 1