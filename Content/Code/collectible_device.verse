
using { /Fortnite.com/Characters }
using { /Fortnite.com/Devices }
using { /Verse.org/Simulation }
using { /UnrealEngine.com/Temporary/Diagnostics }
using { /Verse.org/Random }
using { /UnrealEngine.com/Temporary/SpatialMath }
using { /Fortnite.com/AI }
using { /Verse.org/Assets }

collectible_data := class<persistable><concrete><final>:
    @editable
    Name : string = ""
    @editable
    Key : string = ""
    @editable
    Count : big_number = big_number{}
    ArmorData : ?armor_data = false

    WithCount(NewCount : big_number) : collectible_data =
        return collectible_data{
            Name := Name
            Key := Key
            Count := NewCount
            ArmorData := ArmorData
        }

    WithUpgrade()<transacts> : collectible_data =
        if (AD := ArmorData?):
            UpgradedArmor := AD.WithUpgrade()
            return collectible_data{
                Name := Name
                Key := Key
                Count := Count
                ArmorData := option{UpgradedArmor}
            }
        else:
            return collectible_data{
                Name := Name
                Key := Key
                Count := Count
                ArmorData := ArmorData
            }

    GetImageKey() : string = 
        if (Armor := ArmorData?.ToArmor()):
            return Armor.GetImageKey()
        else:
            return case (Key):
                "Empty Head" => "Helmet"
                "Empty Torso" => "Body"
                "Empty Legs" => "Legs"
                "Empty Boots" => "Boots"
                "Empty Gloves" => "Glove"
                "Empty Accessory" => "Ring"
                _ => Key

    GetImage() : texture = 
        GetTextureFromString(GetImageKey())

    GetUpgradeCost()<transacts> : big_number =
        if:
            Armor := ArmorData?.ToArmor()
            Value:=Int[Armor.GetTotalLevel() * 0.35]
        then:
            return big_number{Value:=200.0, Exponent:=0} * MakeBigNumber(1.2, Value)
        else:
            return big_number{}

    GetSellValue()<transacts> : big_number =
        if:
            Armor := ArmorData?.ToArmor()
            Value:=Int[Armor.GetTotalLevel() * 1.0]
        then:
            return MakeBigNumber(Value*1.0, 1.5) * 1000.0
        else:
            return case (Key):
                "Midas Coins" => big_number{}
                "gem" => big_number{}
                "Gems" => big_number{}
                "Candy1" => big_number{Value:=2.0, Exponent:=1}
                "Candy2" => big_number{Value:=2.0, Exponent:=2}
                "Candy3" => big_number{Value:=2.0, Exponent:=3}
                "Candy4" => big_number{Value:=2.0, Exponent:=4}
                "Candy5" => big_number{Value:=2.0, Exponent:=5}
                "Candy6" => big_number{Value:=2.0, Exponent:=6}
                "Beach Attack Plans" => big_number{Value:=1.0, Exponent:=2}
                "Vicious Treat" => big_number{Value:=2.0, Exponent:=2}
                "GoodFlower" => big_number{Value:=2.0, Exponent:=1}
                "Goodflower" => big_number{Value:=2.0, Exponent:=1}
                "Giantflower" => big_number{Value:=2.0, Exponent:=1}
                "BoarHide" => big_number{Value:=3.0, Exponent:=3}
                "WolfLiver" => big_number{Value:=1.0, Exponent:=3}
                "WolfFur" => big_number{Value:=4.0, Exponent:=3}
                "WolfClaw" => big_number{Value:=1.2, Exponent:=4}
                "Coal" => big_number{Value:=1.0, Exponent:=3}
                "RangeMedal" => big_number{}
                "BossKey" => big_number{}
                "UpgradeHammer" => big_number{Value:=1.0, Exponent:=3}
                _ => big_number{Value:=0.0, Exponent:=0}

    GetUseText()<transacts> : string =
        if (Armor := ArmorData?.ToArmor()):
            return ""
        else:
            return case (Key):
                "Candy1" => "Eat"
                "Candy2" => "Eat"
                "Candy3" => "Eat"
                "Candy4" => "Eat"
                "Candy5" => "Eat"
                "Candy6" => "Eat"
                _ => ""
                

    GetUsePopupText() : string =
        if (Armor := ArmorData?.ToArmor()):
            return "Equipped"
        else:
            return case (Key):
                "Candy1" => "+10"
                "Candy2" => "+100"
                "Candy3" => "+1K"
                "Candy4" => "+10K"
                "Candy5" => "+100K"
                "Candy6" => "+1M"
                _ => ""

    GetUseImage() : texture =
        if (Armor := ArmorData?.ToArmor()):
            return GetImage()
        else:
            return case (Key):
                "Candy1" => Images.StrengthIcon
                "Candy2" => Images.StrengthIcon
                "Candy3" => Images.StrengthIcon
                "Candy4" => Images.StrengthIcon
                "Candy5" => Images.StrengthIcon
                "Candy6" => Images.StrengthIcon
                _ => GetImage()

    GetDescription() : string =
        if (Armor := ArmorData?.ToArmor()):
            return Armor.GetDescription()
        else:
            return case (Key):
                "Midas Coins" => "A Golden Coin. You can use them to upgrade items."
                "gem" => "A Power Gem! You can use these to purchase upgrades and guns from NPCs around the world."
                "Gems" => "A Power Gem! You can use these to purchase upgrades and guns from NPCs around the world." #which one of these do we need?
                "Candy1" => "A common candy! Eating it will grant +10 Strength" 
                "Candy2" => "An uncommon candy! Eating it will grant +100 Strength"
                "Candy3" => "A rare candy! Eating it will grant +1K Strength"
                "Candy4" => "A super rare candy! Eating it will grant +10K Strength"
                "Candy5" => "An epic candy! Eating it will grant +100K Strength"
                "Candy6" => "A legendary candy! Eating it will grant +1M Strength"
                "GoodFlower" => "Blossom, near the entrance to Umbral Forest, collects these flowers..." 
                "Goodflower" => "Blossom, near the entrance to Umbral Forest, collects these flowers..." 
                "Giantflower" => "..."
                "BoarHide" => "A hide from a boar. Kiko might trade you something for these."
                "WolfLiver" => "A liver from a wolf? What could you possibly do with this?"
                "BoarTusk" => "An undamaged tusk from a boar. A rare find!"
                "WolfFur" => "..."
                "WolfClaw" => "..."
                "NaturesEssence" => "..."
                "Coal" => "..."
                "UmbralOre" => "..."
                "Diamond" => "..."
                "TribalHeaddress" => "..."
                "TribalBeads" => "..."
                "AncientTotem" => "..."
                "RangeMedal" => "A reward earned at Sarge's target range. He'll exchange these for special rewards."
                "KikoLuckyStick" => "..."
                "EssenceHourglass" => "A special Hourglass that Midas uses to store his essence. He dropped some when you fought him."
                "BossKey" => "Use this key to fight bosses. You gain one every 6 minutes."
                "UpgradeHammer" => "Required to upgrade your gear." 
                "Bomb" => "With this Bomb you can destroy the boulder in the Giantlands."
                "KeyLower" => "Allows access to the Giantlands Raiders' lower store-room."
                "KeyUpper" => "Allows access to the Giantlands Raiders' upper store-room."
                "Garlic" => "Vampires hate this."
                "WoodStake" => "A stake made of wood. Good for killing Vampires."
                "SilverCross" => "Vampires REALLY hate this."
                "BlueKey" => "One of the three keys needed to unseal the Vampire Crypt. Dropped from Vampires on the Lower Floor."
                "RedKey" => "One of the three keys needed to unseal the Vampire Crypt. Dropped from Vampires on the Upper Floor."
                "GreenKey" => "One of the three keys needed to unseal the Vampire Crypt. Dropped from Vampires on the Middle Floor."
                "ForgeHammer" => "A giant hammer used to forge weapons."
                "ForgeApron" => "A giant apron used by a blacksmith."
                "ForgeTongs" => "A giant pair of tongs used in blacksmithing."
                "VampireOre" => "Evil ore found near the Vampire Manor."
                "MithrilOre" => "A magical ore found in the Giantlands."
                "TungstenOre" => "A powerful ore found in the Giantlands."
                "SilverOre" => "Shiny."
                _ => ""

ToString(Collectible : collectible_data) : string =
    s := "{Collectible.Name} | {Collectible.Count} | {Collectible.Key}"
    if (Armor := Collectible.ArmorData?):
        return s + " | {Armor.ToArmor().GetFullName()}"
    return s

collectible_device := class(creative_device):

    @editable
    GameManager : game_manager = game_manager{}

    @editable
    ZoneManager : zone_manager_device = zone_manager_device{}

    @editable
    var Prop : creative_prop = creative_prop{}
    
    @editable
    Interactible : button_device = button_device{}
    
    @editable
    CollectibleObj : ?collectible_object_device = false

    @editable
    var Data : collectible_data = collectible_data{}
    
    @editable
    var Amount : big_number = big_number{Value:=1.0}

    @editable
    Respawn_Min : float = 5.0
    @editable
    Respawn_Max : float = 10.0

    @editable
    Transient : logic = false

    @editable
    RequiredUpgrade : string = ""

    Mat : Mat_Item_material = Mat_Item_material{}

    var InitialPosition : vector3 = vector3{X := 0.0, Y := 0.0, Z := 0.0}

    var IsRespawning : logic = false
         
    var SubpositionTarget : vector3 = vector3{}

    OnBegin<override>()<suspends>:void=
        set InitialPosition = GetTransform().Translation

        if (CO :=CollectibleObj?):
            CO.CollectedEvent.Subscribe(OnInteraction)
        else:
            
            Interactible.InteractedWithEvent.Subscribe(OnInteraction)
            if (Transient?):
                # SpawnPropResult := SpawnProp(PropTemplate, GetTransform())
                # if (TempProp := SpawnPropResult(0)?):
                #     set Prop = TempProp
                Prop.SetMesh(Meshes.Sphere_D84976F8)
                Prop.SetMaterial(Mat)

        # spawn:
        #     CheckSubposition()

    CheckSubposition()<suspends>:void=
        loop:
            Sleep(0.1)
            if:
                SubpositionTarget.X <> 0.0
                SubpositionTarget.Y <> 0.0
                SubpositionTarget.Z <> 0.0
                Co := CollectibleObj?
                ColPos := Co.GetTransform().Translation
                PosIndex := vector3{X:=1.0*Mod[Round[ColPos.X * 5.0], 5], Y:=1.0*Mod[Round[ColPos.Y * 5.0], 5], Z:=1.0*Mod[Round[ColPos.Z * 5.0], 5]}
                SubPosIndex := vector3{X:=1.0*Round[SubpositionTarget.X * 5.0], Y:=1.0*Round[SubpositionTarget.Y * 5.0], Z:=1.0*Round[SubpositionTarget.Z * 5.0]}
                IsAlmostEqual[PosIndex, SubPosIndex, 0.4]
            then:
                Print("Subposition fixing: {SubpositionTarget}  |  {ColPos}")
                TargetPos := AtlasAjustedPosition(ColPos + vector3{X:=0.0, Y:=0.0, Z:=12.0})
                if:
                    Co.TeleportTo[TargetPos, IdentityRotation()]
            # else:
            #     SubpositionTarget = vector3{}

    AtlasAjustedPosition(InTarget : vector3):vector3=
        
        var Target : vector3 = InTarget

        # EPIC GAMESSSSSSSSSSSSSSS
        for (PlayerManager : GameManager.PlayerManagers):
            if:
                PlayerTranslation := PlayerManager.Player.GetFortCharacter[].GetTransform().Translation
                Delta3D := Target - PlayerTranslation
                Delta := vector3{X:=Delta3D.X, Y:=Delta3D.Y, Z:=0.0}
                Delta.Length() < 200.0
            then:
                Print("TOO CLOSE TO PLAYER: {Delta.Length()}")
                if:
                    Delta.Length() > 1.0
                    Delta2 := Delta * 200.0/Delta.Length()
                then:
                    Print("Pushing Away Normally")
                    set Target = vector3{X := PlayerTranslation.X + Delta2.X, Y := PlayerTranslation.Y + Delta2.Y, Z := Target.Z + 30.0}
                else:
                    Print("Div0 Catch")
                    set Target = vector3{X := PlayerTranslation.X, Y := PlayerTranslation.Y + 200.0, Z := Target.Z + 30.0}
                

        var AtlasIndex : int= GetAtlasIDFromString(Data.Key)

        if:
            Data.Key = ""
            Arm := Data.ArmorData?
        then:
            set AtlasIndex =  GetAtlasIDFromString(Arm.ToArmor().GetImageKey())

        if:
            RPosX := 1.0*Round[Target.X]
            RPosY := 1.0*Round[Target.Y]
            RPosZ := 1.0*Round[Target.Z]
            AtlasX := Mod[AtlasIndex, 5] * 0.2
            AtlasY := Floor[1.0*Mod[AtlasIndex,25] / 5.0] * 0.2
            AtlasZ := Floor[1.0*AtlasIndex / 25.0] * 0.2
            AdjTarget := vector3{X:=RPosX + Sign(RPosX)*AtlasX, Y:=RPosY + Sign(RPosY)*AtlasY, Z:=RPosZ + Sign(RPosZ)*AtlasZ}  
        then:
            set SubpositionTarget = vector3{X:=Sign(RPosX)*AtlasX, Y:=Sign(RPosY)*AtlasY, Z:=Sign(RPosZ)*AtlasZ}
            return AdjTarget
        else:
            return Target
            

        # if:
        #     RPosY := 1.0*Round[Target.Y]
        #     RPosZ := 1.0*Round[Target.Z]
        #     AtlasX := Mod[AtlasIndex, 8] * 0.1
        #     AtlasY := Floor[1.0*AtlasIndex / 8.0] * 0.1
        #     AdjTarget := vector3{X:=Target.X, Y:=RPosY + Sign(RPosY)*AtlasX, Z:=RPosZ + Sign(RPosZ)*AtlasY}
        # then:
        #     return AdjTarget
        # else:
        #     return Target

    SetProperties(CollectibleData : collectible_data, InteractionText : string, VisibleForPlayers : []player_manager):void=
        if:
            Transient?
        then:
            set Data = CollectibleData
            if (BigNumber_GreaterThanOrEquals(Data.Count, big_number{Value:=2.0})?):
                Interactible.SetInteractionText(STM("{InteractionText} ({Data.Count})"))
            else:
                Interactible.SetInteractionText(STM(InteractionText))
            
            if (CO := CollectibleObj?):
                TargetPos := AtlasAjustedPosition(GetTransform().Translation)
                if:
                    CO.TeleportTo[TargetPos, IdentityRotation()]
                then:
                    block:
                
                if (VisibleForPlayers.Length > 0):
                    CO.Hide()
                    for (Player : VisibleForPlayers):
                        if:
                            Player.Player.IsActive[]
                        then:
                            CO.Respawn(Player.Player)
                else:
                    CO.RespawnForAll()
                

            else:
                if:
                    Data.Key = ""
                    Arm := Data.ArmorData?
                then:
                    set Mat.Tex = Arm.ToArmor().GetImage()
                else:
                    set Mat.Tex = GetTextureFromString(Data.Key)

    Move(Pos : vector3, ?TargetPos : ?vector3 = false):void=
        if:
            TeleportTo[Pos + vector3{X:=0.0, Y:=0.0, Z:=40.0}, IdentityRotation()]

        if:
            not CollectibleObj?
            Interactible.TeleportTo[Pos + vector3{X:=0.0, Y:=0.0, Z:=40.0}, IdentityRotation()]
        else if:
            Obj := CollectibleObj?
            Obj.TeleportTo[Pos + vector3{X:=0.0, Y:=0.0, Z:=40.0}, IdentityRotation()]

        if:
            Target := TargetPos?
        then:
            spawn{Fall(Target + vector3{X:=0.0, Y:=0.0, Z:=40.0}, Distance(Pos, Target)/2000.0)}
    
    Fall(Pos : vector3, Time : float)<suspends>:void=
        Sleep(0.02)
        if (Col := CollectibleObj?):
            AdjustedPos := AtlasAjustedPosition(Pos)
            Col.MoveTo(AdjustedPos, IdentityRotation(), Time)
            if:
                Col.TeleportTo[AdjustedPos, IdentityRotation()]
        else:
            race:
                MoveTo(Pos, IdentityRotation(), Time)
                Interactible.MoveTo(Pos, IdentityRotation(), Time)
    
    OnInteraction(InPlayer:agent):void=
        if:
            Player := player[InPlayer]
            Player.IsActive[]
            PlayerManager := GameManager.PlayerManagers[Player]
            RequiredUpgrade = "" or PlayerManager.GetUpgradeLevel[RequiredUpgrade] > 0
        then:
            if:
                GetCharsFromString(Data.Name, 5) = "Candy"
                Obj := CollectibleObj?
                Translation := Obj.GetTransform().Translation
            then:
                PlayerManager.UseCollectible(Data)
                GameManager.AudioManager.PickCandyCheerSound.Play(PlayerManager.Agent)
                GameManager.QueueConfetti(Translation)
            else:
                if:
                    Data.Key = "XPCoin"
                then:
                    block:
                else:
                    PlayerManager.GainCollectible(Data)

            if:
                Data.Name = "Midas Coins"
            then:
                GameManager.AudioManager.PlayCollectCoinsSFX(PlayerManager.Agent)
            else:
                if:
                    Data.Key = "XPCoin"
                then:
                    GameManager.AudioManager.XPCoin.Play(PlayerManager.Agent)
                    GameManager.AccoladesManager.XPCoin.Award(PlayerManager.Agent)            
                else:
                    GameManager.AudioManager.PlayPickCandySFX(PlayerManager.Agent)
                
            if (CO := CollectibleObj?):
                block:
            else if:
                Transient?
            then:
                block:
            else:
                Interactible.Disable()
                Prop.Hide()
                spawn {Refresh()}
        else:
            if:
                Player := player[InPlayer]
                Player.IsActive[]
                PlayerManager := GameManager.PlayerManagers[Player] 
                SetUIManager := PlayerManager.UIManager?
            then:
                SetUIManager.MissingUpgradeMessage(RequiredUpgrade)
    
    Refresh()<suspends>:void=
        set IsRespawning = true
        ZoneManager.OnCollectibleTaken()
        Sleep(GetRandomFloat(Respawn_Min, Respawn_Max) * 60.0 / Max((ZoneManager.PlayersInZone * 1.0), 1.0))
        Interactible.Enable()
        Prop.Show()
        set IsRespawning = false