
using { /Fortnite.com/Devices }
using { /Verse.org/Simulation }
using { /UnrealEngine.com/Temporary/Diagnostics }
using { /Fortnite.com/UI }
using { /Verse.org/Assets }
using { /UnrealEngine.com/Temporary/UI }
using { /Verse.org/Colors }
using { /UnrealEngine.com/Temporary/SpatialMath }

collector_device := class(creative_device):

    @editable
    GameManager : ?game_manager = false

    @editable
    CollectibleName : string = ""

    @editable
    EssencePerCollectible : big_number = big_number{Value:=1.0, Exponent:=0}

    @editable
    BuffRequirementAmount : int = 0

    OpenMenu(Agent:agent):void=
        if (PlayerManager := GameManager?.PlayerManagers[Agent]):
            
            if (not GameManager?.PlayerManagers[Agent].UIManager?.Popup?):
                if (GM := GameManager?):
                    MakeCollectorMenu(GM, Self, Agent, CollectibleName, BuffRequirementAmount).OpenMenu()
            else:
                block:#Log("DOUBLE MENU")

MakeCollectorMenu<constructor>(InGameManager : game_manager, InDevice:collector_device, InAgent : agent, InCollectibleName : string, InBuffRequirementAmount : int) := collector_menu:
    GameManager :=  InGameManager
    CollectorDevice := InDevice
    Agent := InAgent
    CollectibleName := InCollectibleName
    BuffRequirementAmount := InBuffRequirementAmount

collector_menu:= class:
    GameManager : game_manager
    CollectorDevice : collector_device
    Agent : agent
    CollectibleName : string
    BuffRequirementAmount : int

    FloatingText : ui_bounce_mover = ui_bounce_mover{}

    var BuffProgress : progress_bar = progress_bar{}
    var BuffProgressText : text_block = text_block:
        DefaultTextColor:=NamedColors.White
        DefaultShadowColor:=NamedColors.Black
        DefaultShadowOffset:=option{vector2{X:=2.0,Y:=2.0}}
        DefaultShadowOpacity:= 1.0

    var Widget : ?canvas = false
    var TextStackBox : stack_box = stack_box{Orientation:= orientation.Horizontal}

    var AmountText : text_block = text_block:
        DefaultTextColor:=NamedColors.White
        DefaultShadowColor:=NamedColors.Black
        DefaultShadowOffset:=option{vector2{X:=2.0,Y:=2.0}}
        DefaultShadowOpacity:= 1.0

    var Btn1 : button_loud = button_loud{}
    var BtnAll : button_loud = button_loud{}

    OpenMenu():void =
        if (not Widget?):
            CreateMenu()

        UpdateMenu()
        DisplayMenu()
    
    FloatingTextSpaced(Out : collectible_data, In : big_number)<suspends>:void =
        FloatingText.Create(Out.GetImage(), "{Out.Count}") 
        Sleep(0.25)
        FloatingText.Create(Images.Essence, "{In}")
        
    CreateMenu():void =
        if:
            PlayerManager := GameManager.PlayerManagers[Agent]
            ExistingAmount := PlayerManager.GetAchievementLevel["Blossom_Goodflower"]
        then:
            set BuffProgress = MakeProgressBar(0.0, vector2{X:= 300.0, Y:= 50.0}, MakeColorFromHex("0072fc"), MakeColorFromHex("000000"), MakeColorFromHex("FFFFFF")) 
            if:
                BuffRequirementAmount > 0
                ProgressAmount := Mod[ExistingAmount, BuffRequirementAmount]
            then:
                Percent := (ProgressAmount * 1.0) / (BuffRequirementAmount * 1.0)
                BuffProgress.SetProgress(Percent, true)
                BuffProgressText.SetText(STM("{ProgressAmount} / {BuffRequirementAmount}"))
                BuffProgress.SetVisible()
            else:
                BuffProgressText.SetText(STM(""))
                BuffProgress.SetInvisible()

            Canvas := MakeUIPanel(?Width := 0.5, ?Height := 0.5)
            set TextStackBox = MakeTextBox("Donate 1 <{CollectorDevice.CollectibleName}> to earn <essence> {CollectorDevice.EssencePerCollectible * (Max(ExistingAmount, 1) * 1.0)}. <br><br><br><br>Donating {BuffRequirementAmount} grants <Healingwinds> Healingwinds<br>(heal 5% max health per second)", ?ImgSize := vector2{X:=60.0,Y:=60.0})
            
            Canvas.AddWidget(
                canvas_slot:
                    Widget:= BuffProgress.GetWidget() 
                    Anchors := anchors{ Minimum := vector2{X := 0.5, Y := 0.4}, Maximum := vector2{X := 0.5, Y := 0.4}}
                    Offsets := margin{Top := 0.0, Left := 120.0, Right := 0.0, Bottom := 0.0}
                    Alignment := vector2{X := 0.5, Y := 0.5}
                    ZOrder := 5
            )

            Canvas.AddWidget(
                canvas_slot:
                    Widget:= BuffProgressText
                    Anchors := anchors{ Minimum := vector2{X := 0.5, Y := 0.4}, Maximum := vector2{X := 0.5, Y := 0.4}}
                    Offsets := margin{Top := 0.0, Left := 120.0, Right := 0.0, Bottom := 0.0}
                    Alignment := vector2{X := 0.5, Y := 0.5}
                    ZOrder := 6
            )
            
            Canvas.AddWidget( 
                canvas_slot:
                    Widget:= TextStackBox 
                    SizeToContent := true
                    Anchors := anchors{ Minimum := vector2{X := 0.5, Y := 0.4}, Maximum := vector2{X := 0.5, Y := 0.4}}
                    Offsets := margin{Top := 40.0, Left := 0.0, Right := 0.0, Bottom := 0.0}
                    Alignment := vector2{X := 0.5, Y := 0.5}
                    ZOrder := 4
            )

            Canvas.AddWidget(
                canvas_slot:
                    Widget:= FloatingText.Setup(20, vector2{X := 55.0, Y := 55.0}, 2.0, 0.95)
                    Anchors := anchors{ Minimum := vector2{X := 0.64, Y := 0.4}, Maximum := vector2{X := 0.64, Y := 0.4}}  
                    Offsets := margin{Top := 0.0, Left := 0.0, Right := 0.0, Bottom := 0.0}
                    Alignment := vector2{X := 0.5, Y := 0.5}
                    ZOrder := 8
            )

            CloseButton := button_quiet:
            CloseButton.SetText(STM("            BACK            "))            
            CloseButton.OnClick().Subscribe(PlayerManager.ClosePopup)

            Btn1.OnClick().Subscribe(CashIn1)
            BtnAll.OnClick().Subscribe(cashInAll)

            Canvas.AddWidget(
                canvas_slot:
                    Widget:= stack_box:
                        Orientation:=orientation.Horizontal
                        Slots := array:
                            stack_box_slot:
                                Widget:= Btn1
                            stack_box_slot:
                                Widget:= BtnAll
                            stack_box_slot:
                                Widget:= CloseButton
                    Anchors := anchors{ Minimum := vector2{X := 0.5, Y := 0.65}, Maximum := vector2{X := 0.5, Y := 0.65}}
                    Alignment := vector2{X := 0.5, Y := 0.5}
                    ZOrder := 3
            )

            Canvas.AddWidget(
                canvas_slot:
                    Widget:= stack_box:
                        Orientation:=orientation.Horizontal
                        Slots := array:
                            stack_box_slot:
                                Widget:= texture_block:
                                    DefaultImage := GetTextureFromString("{CollectorDevice.CollectibleName}")
                                    DefaultDesiredSize := vector2{X:=120.0,Y:=120.0}
                            stack_box_slot:
                                Widget:= AmountText
                    Anchors := anchors{ Minimum := vector2{X := 0.5, Y := 0.4}, Maximum := vector2{X := 0.5, Y := 0.4}}
                    Offsets := margin{Top := 0.0, Left := -250.0, Right := 0.0, Bottom := 0.0}
                    Alignment := vector2{X := 0.5, Y := 0.5}
                    ZOrder := 2
            )

            set Widget = option{Canvas}

    UpdateMenu():void =
        spawn:
            FloatingText.Activate()

        if:
            PlayerManager := GameManager.PlayerManagers[Agent]
            ExistingAmount := Mod[PlayerManager.GetAchievementLevel["Blossom_Goodflower"], BuffRequirementAmount]
            Canvas := Widget?
        then:
            if:
                BuffRequirementAmount > 0
            then:
                Percent := (ExistingAmount * 1.0) / (BuffRequirementAmount * 1.0)
                BuffProgress.SetProgress(Percent, true)
                BuffProgressText.SetText(STM("{ExistingAmount} / {BuffRequirementAmount}"))
                BuffProgress.SetVisible()
            else:
                BuffProgressText.SetText(STM(""))
                BuffProgress.SetInvisible()

            Canvas.RemoveWidget(TextStackBox)
            set TextStackBox = MakeTextBox("Donate 1 <{CollectorDevice.CollectibleName}> to earn <essence> {CollectorDevice.EssencePerCollectible * (Max(ExistingAmount, 1) * 1.0)}. <br><br><br><br>Donating {BuffRequirementAmount} grants <Healingwinds> Healingwinds<br>(heal 5% max health per second)", ?ImgSize := vector2{X:=60.0,Y:=60.0})
            Canvas.AddWidget( 
                canvas_slot:
                    Widget:= TextStackBox 
                    SizeToContent := true
                    Anchors := anchors{ Minimum := vector2{X := 0.5, Y := 0.4}, Maximum := vector2{X := 0.5, Y := 0.4}}
                    Offsets := margin{Top := 40.0, Left := 0.0, Right := 0.0, Bottom := 0.0}
                    Alignment := vector2{X := 0.5, Y := 0.5}
            )
            
            Amt := PlayerManager.GetCollectibleCount(CollectorDevice.CollectibleName)
            Amount:=BuffRequirementAmount-ExistingAmount
            if:
                Amount > 0
            then:
                Btn1.SetText(STM("Donate x{Amount}"))
            else:
                Btn1.SetText(STM("Donate x0"))
            BtnAll.SetText(STM("ALL"))

            AmountText.SetText(STM("x{Amt}"))
          
            Btn1.SetEnabled(true)
            BtnAll.SetEnabled(true)
    
    DisplayMenu():void =
        if:
            InPlayer := player[Agent]
            InPlayer.IsActive[]
            PlayerUI := GetPlayerUI[InPlayer]
            PlayerManager := GameManager.PlayerManagers[InPlayer]
            Canvas := Widget?
        then:
            Canvas.SetEnabled(true)
            Canvas.SetVisibility(widget_visibility.Visible)
            PlayerUI.AddWidget(Canvas, player_ui_slot{InputMode := ui_input_mode.All})
            if (set PlayerManager.UIManager?.Popup = option{Canvas}) {}

    CashIn1(Wdata : widget_message):void=
        if:
            PlayerManager := GameManager.PlayerManagers[Agent]
            ExistingAmount := PlayerManager.GetAchievementLevel["Blossom_Goodflower"]
            Amount:=BuffRequirementAmount-ExistingAmount
            Amount > 0
        then:
            CashIn(big_number{Value:=(Amount) * 1.0})

    cashInAll(Wdata : widget_message):void=
        if (PlayerManager := GameManager.PlayerManagers[Agent]):
            Amt := PlayerManager.GetCollectibleCount(CollectorDevice.CollectibleName)
            CashIn(Amt)

    CashIn(Amount:big_number):void =
        if:
            PlayerManager := GameManager.PlayerManagers[Agent]
            BigNumber_GreaterThanOrEquals(PlayerManager.GetCollectibleCount(CollectorDevice.CollectibleName),Amount)?
            IntAmount := Int[BigNumber_ToFloat(Amount)]
            ExistingAmount := Mod[PlayerManager.GetAchievementLevel["Blossom_Goodflower"], BuffRequirementAmount]
            NewAmount := (ExistingAmount+IntAmount)*1.0/(BuffRequirementAmount*1.0)
        then:
            PlayerManager.GainAchievementLevel("Blossom_Goodflower", IntAmount)
            if:
                NewAmount > 1.0
            then:
                #Do Buff SFX
                PlayerManager.GameManager.AudioManager.BlossomBuff.Play(PlayerManager.Agent)
                BuffAmount := big_number{Value := 5.0}
                Duration := 60.0 * 30.0  * (1.0 + PlayerManager.GetUpgradeValueSafe("P_BuffDuration", false))
                PlayerManager.UpdateBuff("Healingwinds", "%", BuffAmount, false, Duration, false, ?IsSecondary:= true)

            PlayerManager.GainCollectible(collectible_data{Key:=CollectorDevice.CollectibleName, Count:=Amount * -1.0})
            Formula:=(CollectorDevice.EssencePerCollectible * (Max(ExistingAmount,1) * 1.0)) * (IntAmount * 1.0) + (CollectorDevice.EssencePerCollectible * (ExistingAmount * 1.0)) * ((IntAmount *1.0 - 1.0) * (IntAmount * 1.0) / 2.0)
            PlayerManager.GainEssence(Formula, big_number{}, big_number{}, true, false, false, false, false)
            spawn:
                FloatingTextSpaced(collectible_data{Key:=CollectorDevice.CollectibleName, Count:=Amount * -1.0}, Formula)

        UpdateMenu()