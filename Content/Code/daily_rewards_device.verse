
using { /Fortnite.com/Devices }
using { /Verse.org/Simulation }
using { /UnrealEngine.com/Temporary/Diagnostics }
using { /UnrealEngine.com/Temporary/UI }
using { /Verse.org/Assets }
using { /Fortnite.com/UI }
using { /UnrealEngine.com/Temporary/SpatialMath }
using { /Verse.org/Colors }

daily_reward := interface:
    GetText():string
    GetIcon():texture
    OnClaim(PlayerManager:player_manager):void

daily_rewards_device := class(creative_device, menu_openable):

    @editable
    GameManager : ?game_manager = false
    @editable
    ClaimButton: button_device = button_device{}

    var MenuPerPlayer : [player]daily_menu = map{}

    OnBegin<override>()<suspends>:void=
        ClaimButton.InteractedWithEvent.Subscribe(OpenMenu)

    OpenMenu<override>(Agent:agent):void=
        if (PlayerManager := GameManager?.PlayerManagers[Agent]):
            if (not GameManager?.PlayerManagers[Agent].UIManager?.Popup?):
                if: 
                    GM := GameManager?
                then:
                    if (Menu := MenuPerPlayer[PlayerManager.Player]):
                        Menu.OpenMenu()
                    else:
                        Menu := MakeDailyMenu(GM, Self, Agent)
                        if (set MenuPerPlayer[PlayerManager.Player] = Menu){}
                        Menu.OpenMenu()
            else:
                block:#Log("DOUBLE MENU")

    DoClaim(Agent : agent):void=
        if:
            PlayerManager := GameManager?.PlayerManagers[Agent]
            PlayerManager.CanClaimDailyRewards[]
        then:
            PlayerManager.GameManager.AudioManager.UnlockZonePlayer.Play(Agent)
            ProvideDailyReward(PlayerManager, PlayerManager.GetDailyRewardCount())
            PlayerManager.OnDailyRewardsClaimed()
            
    ProvideDailyReward(PlayerManager : player_manager, Day: int):void=
        GetReward(Day).OnClaim(PlayerManager)

    GetReward(Day : int)<transacts>:daily_reward=
        if (GunManager := GameManager?.GunManager):
            case (Day):
                0 => return gun_reward{GunManager := GunManager, GunID := 4}
                1 => return gun_reward{GunManager := GunManager, GunID := 6}
                2 => return gun_reward{GunManager := GunManager, GunID := 3}
                3 => return gem_reward{Amount := big_number{Value:=1.5, Exponent:=1}}
                4 => return gem_reward{Amount := big_number{Value:=3.0, Exponent:=1}}
                5 => return gem_reward{Amount := big_number{Value:=5.0, Exponent:=1}}
                6 => return upgrade_reward{Amount := 1, Title := "+10% DMG", UpgradeKey := "DailyDamageMult"}
                _ => block:
                    return gem_reward{Amount := MakeBigNumber(1.3, Day-7) * 100.0}
        
        return gem_reward{Amount := big_number{Value:=1.0, Exponent:=0}}

gem_reward := class(daily_reward):
    Amount : big_number
    
    GetText<override>():string=
        return "{Amount} Gems"
        
    GetIcon<override>():texture=
        return Images.Gem2
        
    OnClaim<override>(PlayerManager:player_manager):void=
        spawn:
            PlayerManager.GainGems(Amount)

upgrade_reward := class(daily_reward):
    Amount : int
    Title : string
    UpgradeKey : string

    GetText<override>():string=
        return "{Title}"
        
    GetIcon<override>():texture=
        return GetTextureFromString("{UpgradeKey}")
        
    OnClaim<override>(PlayerManager:player_manager):void=
        PlayerManager.GainUpgradeLevel("{UpgradeKey}", Amount)


gun_reward := class(daily_reward):
    GunManager : gun_manager
    GunID : int

    GetStats():gun_stats=
        if (Stats := GunManager.StatsTable[GunID]):
            return Stats
        else:
            return gun_stats{}
    
    GetText<override>():string=
        return "{GetStats().Name}"
        
    GetIcon<override>():texture=
        return GetTextureFromString("Gun_{GunID}")
        
    OnClaim<override>(PlayerManager:player_manager):void=
        PlayerManager.GainUpgradeLevel("Gun_{GunID}", 1)

MakeDailyMenu<constructor>(GameManagerIn : game_manager, DailiesDeviceIn : daily_rewards_device, AgentIn : agent) := daily_menu:
    GameManager :=  GameManagerIn
    DailiesDevice := DailiesDeviceIn
    Agent := AgentIn

daily_menu:= class:
    GameManager : game_manager
    DailiesDevice : daily_rewards_device
    Agent : agent
    
    var DialogBackground : texture_block = texture_block{ DefaultImage := Images.DialogBox, DefaultDesiredSize := vector2{X := 1920.0, Y := 1080.0}}
    var HeaderPlate : texture_block = texture_block{ DefaultImage := Images.DialogBox, DefaultDesiredSize := vector2{X := 400.0, Y := 80.0}}
    var StackBox : stack_box = stack_box{Orientation:=orientation.Horizontal}

    ClaimButton : button_loud = button_loud:    

    var Widget : ?canvas = false
    var RewardButtons : []reward_button = array{}

    OpenMenu():void =
        if (not Widget?):
            CreateMenu()

        UpdateMenu()
        DisplayMenu()
    
    CreateMenu():void =
        if:
            InPlayer := player[Agent]
            InPlayer.IsActive[]
            PlayerUI := GetPlayerUI[InPlayer]
            PlayerManager := GameManager.PlayerManagers[InPlayer]
        then:
            
            CloseButton := button_quiet:
            CloseButton.SetText(STM("            BACK            "))            
            CloseButton.OnClick().Subscribe(PlayerManager.ClosePopup)  

            ClaimButton.SetText(STM("           CLAIM           "))      
            ClaimButton.OnClick().Subscribe(OnClaim)
            
            HeaderText := text_block:
                DefaultTextColor:=NamedColors.White
                DefaultShadowColor:=NamedColors.Black
                DefaultShadowOffset:=option{vector2{X:=2.0,Y:=2.0}}
                DefaultShadowOpacity:= 1.0
                DefaultText:= STM("DAILY REWARDS")

            Canvas := canvas:
                Slots := array:
                    canvas_slot:
                        Widget:= DialogBackground
                        SizeToContent := true
                        Anchors := anchors{ Minimum := vector2{X := 0.1, Y := 0.1}, Maximum := vector2{X := 0.9, Y := 0.9}}                        
                        Alignment := vector2{X := 0.0, Y := 1.0}
                        ZOrder := 1
                    canvas_slot:
                        Widget:= HeaderPlate
                        SizeToContent := true
                        Anchors := anchors{ Minimum := vector2{X := 0.5, Y := 0.1}, Maximum := vector2{X := 0.5, Y := 0.1}}
                        Offsets := margin{Top := 0.0, Left := 0.0, Right := 0.0, Bottom := 0.0}
                        Alignment := vector2{X := 0.5, Y := 0.5}
                        ZOrder := 2
                    canvas_slot:
                        Widget:= HeaderText
                        Anchors := anchors{ Minimum := vector2{X := 0.5, Y := 0.1}, Maximum := vector2{X := 0.5, Y := 0.1}}
                        Offsets := margin{Top := 0.0, Left := 0.0, Right := 0.0, Bottom := 0.0}
                        Alignment := vector2{X := 0.5, Y := 0.5}
                        ZOrder := 3
                    canvas_slot:
                        Widget:= StackBox
                        SizeToContent := true
                        Anchors := anchors{ Minimum := vector2{X := 0.5, Y := 0.5}, Maximum := vector2{X := 0.5, Y := 0.5}}
                        Alignment := vector2{X := 0.5, Y := 1.0}
                        ZOrder := 4
                    canvas_slot:
                        Widget:= ClaimButton
                        Anchors := anchors{ Minimum := vector2{X := 0.5, Y := 0.795}, Maximum := vector2{X := 0.5, Y := 0.795}}
                        Alignment := vector2{X := 0.5, Y := 1.0}
                        ZOrder := 5
                    canvas_slot:
                        Widget:= CloseButton
                        Anchors := anchors{ Minimum := vector2{X := 0.5, Y := 0.86}, Maximum := vector2{X := 0.5, Y := 0.86}}
                        Alignment := vector2{X := 0.5, Y := 1.0}
                        ZOrder := 5
                        
            for (i := 0..6):
                Btn := reward_button{Day := i, Reward := DailiesDevice.GetReward(i)}
                Btn.Create()
                set RewardButtons += array{Btn}
                if (Slot := Btn.Widget?):
                    StackBox.AddWidget(Slot)
            
            set Widget = option{Canvas}

    OnClaim(Wdata:widget_message):void =
        DailiesDevice.DoClaim(Agent)
        UpdateMenu()

    UpdateMenu():void =
        if:
            InPlayer := player[Agent]
            InPlayer.IsActive[]
            PlayerUI := GetPlayerUI[InPlayer]
            PlayerManager := GameManager.PlayerManagers[InPlayer]
        then:
            Day := PlayerManager.GetDailyRewardCount()
            if (PlayerManager.CanClaimDailyRewards[]):
                ClaimButton.SetEnabled(true)
                ClaimButton.SetText(STM("           CLAIM           "))

                for (i := 0..RewardButtons.Length - 1):
                    if (Btn := RewardButtons[i]):
                        if (i < Day):
                            Btn.Update(false, false, false)
                        else if (i = Day):
                            Btn.Update(true, true, false)
                        else:
                            Btn.Update(true, false, false)
            else:
                ClaimButton.SetEnabled(false)
                ClaimButton.SetText(STM("     COME BACK TOMORROW     "))
                
                for (i := 0..RewardButtons.Length - 1):
                    if (Btn := RewardButtons[i]):
                        if (i < Day):
                            Btn.Update(false, false, false)
                        else if (i = Day):
                            Btn.Update(true, false, true)
                        else:
                            Btn.Update(true, false, false)
            
    DisplayMenu():void =
        if:
            InPlayer := player[Agent]
            InPlayer.IsActive[]
            PlayerUI := GetPlayerUI[InPlayer]
            PlayerManager := GameManager.PlayerManagers[InPlayer]
            Canvas := Widget?
        then:
            Canvas.SetEnabled(true)
            Canvas.SetVisibility(widget_visibility.Visible)
            PlayerUI.AddWidget(Canvas, player_ui_slot{InputMode := ui_input_mode.All})
            if (set PlayerManager.UIManager?.Popup = option{Canvas}) {}
            
reward_button := class():
    var Widget : ?stack_box_slot = false

    CheckWidget : texture_block = texture_block:
        DefaultImage := Images.Checkmark
        DefaultDesiredSize := vector2{X := 140.0, Y := 140.0}
    GlowWidget : texture_block = texture_block:
        DefaultImage := Images.Glow
        DefaultDesiredSize := vector2{X := 260.0, Y := 260.0}
    IconWidget : texture_block = texture_block:
        DefaultImage := Images.Empty
        DefaultDesiredSize := vector2{X := 140.0, Y := 140.0}
    TextWidget : text_block = text_block:
        DefaultTextColor:=NamedColors.AntiqueWhite
        DefaultShadowColor:=NamedColors.Black
        DefaultShadowOffset:=option{vector2{X:=2.0,Y:=2.0}}
        DefaultShadowOpacity:= 1.0
    DayTextWidget : text_block = text_block:
        DefaultTextColor:=NamedColors.AntiqueWhite
        DefaultShadowColor:=NamedColors.Black
        DefaultShadowOffset:=option{vector2{X:=2.0,Y:=2.0}}
        DefaultShadowOpacity:= 1.0
    TomorrowTextWidget : text_block = text_block:
        DefaultTextColor:=NamedColors.Grey
        DefaultShadowColor:=NamedColors.Black
        DefaultShadowOffset:=option{vector2{X:=2.0,Y:=2.0}}
        DefaultShadowOpacity:= 1.0

    Reward : daily_reward
    Day : int

    Create():void =
        Icon := Reward.GetIcon()
        Text := STM("{Reward.GetText()}")

        IconWidget.SetImage(Icon)
        TextWidget.SetText(Text)
        DayTextWidget.SetText(STM("Day {Day+1}"))
        TomorrowTextWidget.SetText(STM("Tomorrow"))

        set Widget = option:
            stack_box_slot:
                Widget:= canvas:
                    Slots := array:
                        canvas_slot:
                            Widget:= GlowWidget
                            SizeToContent := true
                            Anchors := anchors{ Minimum := vector2{X := 0.5, Y :=0.5}, Maximum := vector2{X := 0.5, Y := 0.5}}
                            Alignment := vector2{X := 0.5, Y := 0.5}
                            ZOrder := 1
                        canvas_slot:
                            Widget:= IconWidget
                            SizeToContent := true
                            Anchors := anchors{ Minimum := vector2{X := 0.5, Y :=0.5}, Maximum := vector2{X := 0.5, Y := 0.5}}
                            Alignment := vector2{X := 0.5, Y := 0.5}
                            ZOrder := 2
                        canvas_slot:
                            Widget:= CheckWidget
                            SizeToContent := true
                            Anchors := anchors{ Minimum := vector2{X := 0.5, Y :=0.5}, Maximum := vector2{X := 0.5, Y := 0.5}}
                            Alignment := vector2{X := 0.5, Y := 0.5}
                            ZOrder := 3
                        canvas_slot:
                            Widget:= TextWidget
                            SizeToContent := true
                            Anchors := anchors{ Minimum := vector2{X := 0.5, Y :=1.0}, Maximum := vector2{X := 0.5, Y := 1.0}}
                            Alignment := vector2{X := 0.5, Y := 0.0}
                            Offsets := margin{Top := -30.0, Left := 0.0, Right := 0.0, Bottom := 0.0}
                            ZOrder := 4
                        canvas_slot:
                            Widget:= DayTextWidget
                            SizeToContent := true
                            Anchors := anchors{ Minimum := vector2{X := 0.5, Y :=1.0}, Maximum := vector2{X := 0.5, Y := 1.0}}
                            Alignment := vector2{X := 0.5, Y := 0.0}
                            ZOrder := 5
                        canvas_slot:
                            Widget:= TomorrowTextWidget
                            SizeToContent := true
                            Anchors := anchors{ Minimum := vector2{X := 0.5, Y :=0.0}, Maximum := vector2{X := 0.5, Y := 0.0}}
                            Alignment := vector2{X := 0.5, Y := 1.0}
                            Offsets := margin{Top := 50.0, Left := 0.0, Right := 0.0, Bottom := 0.0}
                            ZOrder := 6
                Padding := margin{Top := -58.0, Left := -50.0, Right := -50.0, Bottom := -58.0}

    Update(Enabled: logic, Claimable: logic, Tomorrow: logic):void =
        IconWidget.SetEnabled(Enabled)
        TextWidget.SetEnabled(Enabled)
        DayTextWidget.SetEnabled(Enabled)
        if (Enabled?):
            CheckWidget.SetVisibility(widget_visibility.Hidden)
        else:
            CheckWidget.SetVisibility(widget_visibility.Visible)
            TextWidget.SetText(STM("CLAIMED"))
            DayTextWidget.SetVisibility(widget_visibility.Hidden)
    
        if (Claimable?):
            GlowWidget.SetVisibility(widget_visibility.Visible)
            TextWidget.SetTextColor(NamedColors.Yellow)
            DayTextWidget.SetTextColor(NamedColors.Yellow)
        else:
            GlowWidget.SetVisibility(widget_visibility.Hidden)
            TextWidget.SetTextColor(NamedColors.White)
            DayTextWidget.SetTextColor(NamedColors.White)

        if (Tomorrow?):
            TomorrowTextWidget.SetVisibility(widget_visibility.Visible)
        else:
            TomorrowTextWidget.SetVisibility(widget_visibility.Hidden)