using { /Fortnite.com/Characters}
using { /Fortnite.com/Devices }
using { /Fortnite.com/UI }
using { /UnrealEngine.com/Temporary/SpatialMath }
using { /UnrealEngine.com/Temporary/UI }
using { /Verse.org/Assets }
using { /Verse.org/Colors }
using { /Verse.org/Simulation }

dialog_message := class<concrete>:
    @editable
    Title : ?string = false
    @editable
    Message : string = ""
    @editable
    VO : ?audio_player_device = false

dialog_sequence := class<concrete>:
    @editable:
    var StartTask : int = 0
    @editable:
    var EndTask : int = 0
    @editable:
    var ShowShop : logic = false
    @editable:
    var Messages : []dialog_message = array{}

quest_dialog_map := class<concrete>:
    @editable
    QuestID : int = 0
    @editable
    TaskDialogs : []dialog_sequence = array{}

npc_device:= class(creative_device):
    @editable
    GameManager : game_manager = game_manager{}

    @editable
    Name : string = "NPC"

    @editable
    Shop : ?upgrade_shop = false

    @editable
    CollectibleShop : ?collectible_shop = false

    @editable
    Collector : ?collector_device = false

    @editable : 
    PostDialogMenu : ?menu_device_template = false

    @editable
    Character : character_device = character_device{}
    @editable
    ButtonInteractOverride : ?button_device = false

    @editable
    DefaultDialog : ?dialog_sequence = false

    @editable
    QuestDialogs:[]quest_dialog_map = array{quest_dialog_map{}}

    var DialogMapping : [int]quest_dialog_map = map{}

    OnBegin<override>()<suspends>:void=
        if (Button := ButtonInteractOverride?):
            Button.InteractedWithEvent.Subscribe(OnInteraction)
        else:
            Character.InteractedWithEvent.Subscribe(OnInteraction)

        for (QuestDialog : QuestDialogs):
            if (set DialogMapping[QuestDialog.QuestID] = QuestDialog) {}
    
    OnInteraction(InPlayer:agent):void=
        if:
            Player := player[InPlayer]
            Player.IsActive[]
            PlayerManager := GameManager.PlayerManagers[InPlayer]
            not PlayerManager.UIManager?.Popup?
        then:
            if:
                QuestManager := PlayerManager.QuestManager?
            then:
                #Finds first match and ships it.
                #TODO account for multiple simultaneous quests and tasks
                for (QuestStats : PlayerManager.GetCurrentQuests[]):
                    if:
                        QuestID := QuestStats.QuestID
                        TaskID := QuestStats.TaskID                        
                        DialogsForQuest := DialogMapping[QuestID]
                    then:       
                        Log("QuestID: {QuestID}, TaskID: {TaskID}")
                        for (DialogItem : DialogsForQuest.TaskDialogs):
                            if (DialogItem.StartTask <= TaskID and TaskID <= DialogItem.EndTask):
                                spawn{ShowDialog(PlayerManager, DialogItem)}
                                return
                
            PlayerManager.UpdateQuestProgressForTask(task_event.Interact, big_number{Value := 1.0}, false, ?Target := Name)

            if (Dialog := DefaultDialog?):
                spawn{ShowDialog(PlayerManager, Dialog)}
            else:
                DoPostDialog(PlayerManager)

    DoPostDialog(PlayerManager:player_manager):void=
        if (MyShop := Shop?):
            MyShop.TryOpenMenu(Name, PlayerManager.Player)
        else if (MyShop := CollectibleShop?):
            MyShop.TryOpenMenu(Name, PlayerManager.Player)
        else if (MyCollector := Collector?):
            MyCollector.OpenMenu(PlayerManager.Player)
        else if (MyMenu := PostDialogMenu?):
            MyMenu.OpenMenu(PlayerManager.Player)

    ShowDialog(PlayerManager:player_manager, DialogItem:dialog_sequence)<suspends>:void=
        MakeDialog(Name, DialogItem.Messages).ShowDialog(PlayerManager, 0.0)
    
        if (DialogItem.ShowShop?):
            DoPostDialog(PlayerManager)

dialog_manager := class:
    var Title : string = array{}
    Messages : []dialog_message = array{}
    var CurrentMessage : int = 0
    var DialogBackground : texture_block = texture_block{ DefaultImage := Images.DialogBox, DefaultDesiredSize := vector2{X := 1800.0, Y := 300.0}}
    var NamePlate : texture_block = texture_block{ DefaultImage := Images.DialogBox, DefaultDesiredSize := vector2{X := 300.0, Y := 60.0}}
    var Texture : texture = Images.Empty
    var Profile : texture_block = texture_block{DefaultImage := Images.Empty, DefaultDesiredSize := vector2{X := 250.0, Y := 250.0}}
    
    var TextStackBox : stack_box = stack_box{ Orientation := orientation.Vertical}

    ShowDialog(PlayerManager:player_manager, Delay : float)<suspends>:void=
        Sleep(Delay)
        
        InPlayer := PlayerManager.Player

        if:
            InPlayer.IsActive[]
            PlayerUI := GetPlayerUI[InPlayer]
            FortChar := InPlayer.GetFortCharacter[]
        then:

            CloseButton := button_regular:
            CloseButton.SetText(STM("                                                                    OK                                                                    "))
                
            set Texture = GetTextureFromString(Title)
            Profile.SetImage(Texture)
            TitleWidget := text_block:
                DefaultTextColor:=NamedColors.White
                DefaultShadowColor:=NamedColors.Black
                DefaultShadowOffset:=option{vector2{X:=2.0,Y:=2.0}}
                DefaultShadowOpacity:= 1.0
                DefaultText:= STM("{Title}")

            Canvas := canvas:
                Slots := array:
                    canvas_slot:
                        Widget:= DialogBackground
                        SizeToContent := true
                        Anchors := anchors{ Minimum := vector2{X := 0.03, Y := 0.95}, Maximum := vector2{X := 0.03, Y := 0.95}}                        
                        Alignment := vector2{X := 0.0, Y := 1.0}
                    canvas_slot:
                        Widget:= Profile
                        SizeToContent := true
                        Anchors := anchors{ Minimum := vector2{X := 0.03, Y := 0.95}, Maximum := vector2{X := 0.03, Y := 0.95}}
                        Offsets := margin{Top := -330.0, Left := 125.0, Right := 0.0, Bottom := 0.0}
                        Alignment := vector2{X := 0.0, Y := 1.0}
                    canvas_slot:
                        Widget:= NamePlate
                        SizeToContent := true
                        Anchors := anchors{ Minimum := vector2{X := 0.03, Y := 0.95}, Maximum := vector2{X := 0.03, Y := 0.95}}
                        Offsets := margin{Top := -270.0, Left := 100.0, Right := 0.0, Bottom := 0.0}
                        Alignment := vector2{X := 0.0, Y := 1.0}
                    canvas_slot:
                        Widget:= TitleWidget
                        Anchors := anchors{ Minimum := vector2{X := 0.03, Y := 0.95}, Maximum := vector2{X := 0.03, Y := 0.95}}
                        Offsets := margin{Top := -285.0, Left := 250.0, Right := 0.0, Bottom := 0.0}
                        Alignment := vector2{X := 0.5, Y := 1.0}
                    canvas_slot:
                        Widget:= CloseButton
                        Anchors := anchors{ Minimum := vector2{X := 0.475, Y := 0.92}, Maximum := vector2{X := 0.475, Y := 0.92}}
                        Alignment := vector2{X := 0.5, Y := 1.0}
                        Offsets := margin{Top := 0.0, Left := 40.0, Right := 0.0, Bottom := 0.0}
            
            if (MyUI := PlayerManager.UIManager?.Popup?):
                MyUI.SetEnabled(false)
                MyUI.SetVisibility(widget_visibility.Hidden)
                PlayerUI.RemoveWidget(MyUI)

            PlayerUI.AddWidget(Canvas, player_ui_slot{InputMode := ui_input_mode.All})
            if (set PlayerManager.UIManager?.Popup = option{Canvas}) {}
            var ActiveVO : audio_player_device = audio_player_device{}
            
            for (Message : Messages):

                Canvas.RemoveWidget(TextStackBox)
                set TextStackBox =WordWrap(Message.Message, 102)
                Canvas.AddWidget(
                    canvas_slot:
                        Widget:= TextStackBox
                        SizeToContent := true
                        Anchors := anchors{ Minimum := vector2{X := 0.03, Y := 0.95}, Maximum := vector2{X := 0.03, Y := 0.95}}                        
                        Offsets := margin{Top := -240.0, Left := 60.0, Right := 0.0, Bottom := 0.0}
                        Alignment := vector2{X := 0.0, Y := 0.0}
                )
                
                if:
                    TempTitle:=Message.Title?                
                then:
                    set Texture = GetTextureFromString(TempTitle)
                    Profile.SetImage(Texture)
                    TitleWidget.SetText(STM("{TempTitle}"))
                if:
                    VO := Message.VO?
                then:
                    set ActiveVO = VO
                    VO.Play(PlayerManager.Player)

                CloseButton.OnClick().Await()
                ActiveVO.Stop(PlayerManager.Player)
                
            PlayerManager.UpdateQuestProgressForTask(task_event.Interact, big_number{Value := 1.0}, false, ?Target := Title)
        
            if:
                UIManager := PlayerManager.UIManager?
                UIManager.Popup? = Canvas
            then:
                UIManager.ClosePopup()


MakeDialog<constructor>(InTitle : string, InMessages : []dialog_message) := dialog_manager:
    Title := InTitle
    Messages := InMessages