
using { /Fortnite.com/Devices }
using { /Verse.org/Simulation }
using { /UnrealEngine.com/Temporary/Diagnostics }
using { /UnrealEngine.com/Temporary/SpatialMath }
using { /Verse.org/Random }

drop_item := class<concrete>:
    @editable
    Name : string = ""
    @editable
    PickupText : string = ""
    @editable
    Chance : float = 0.0
    @editable
    Unique : logic = false
    @editable
    MinCount : big_number = big_number{Value:=1.0}
    @editable
    MaxCount : big_number = big_number{Value:=1.0}
    @editable
    ArmorInfo : ?armor_gen_properties = false
    @editable
    RequiredQuest : ?quest_stats = false

drop_table := class<concrete>:
    @editable
    Items : []drop_item = array{}

drop_table_map_entry := class<concrete>:
    @editable
    Key : string = ""
    @editable
    Table : drop_table = drop_table{}


drop_manager_device := class(creative_device):

    @editable
    GameManager : ?game_manager = false

    ArmorGenerator : armor_generator = armor_generator{}

    @editable
    GlobalDropTables : []drop_table_map_entry = array{}
    var GlobalDropTablesMap : [string]drop_table = map{}

    var ItemsDB : []collectible_device = array{}
    var Available : []collectible_device = array{}

    var Coins : []collectible_device = array{}
    var AvailableCoins : []collectible_device = array{}
    
    OnBegin<override>()<suspends>:void=
        ItemsInit()
        set Available = ItemsDB
        set AvailableCoins = Coins
        for (Entry : GlobalDropTables):
            if:
                set GlobalDropTablesMap[Entry.Key] = Entry.Table
    
    ItemsInit():void=
        ItemsTemp := FindCreativeObjectsWithTag(item{})
        for (Obj : ItemsTemp):
            if (Item := collectible_device[Obj]):
                set ItemsDB = ItemsDB + array{Item}

        CoinsTemp := FindCreativeObjectsWithTag(coin{})
        for (Obj : CoinsTemp):
            if (Item := collectible_device[Obj]):
                set Coins = Coins + array{Item}

    SelectFromDropTable(Table : drop_table, NumPlayers : int):?drop_item =
        var Selector : float = GetRandomFloat(0.0,1.0)
        var CandyDropped : logic = false
        for (i := 0..Table.Items.Length - 1):
            if:
                not CandyDropped?
                Item := Table.Items[i]
                Chance := Item.Chance
            then:

                if (Item.Unique? and NumPlayers > 0):
                    set Selector -= Chance * NumPlayers
                else:
                    set Selector -= Chance
                if:
                    Selector <= 0.0
                then:
                    return option{Item}
        return false

    SelectFromDropTable(TableID : string, NumPlayers : int):?drop_item =
        if:
            Table := GlobalDropTablesMap[TableID]
        then:
            return SelectFromDropTable(Table, NumPlayers)

        return false

    SpawnItems(DropItems : []drop_item, Pos : vector3, ?TargetPos : ?vector3 = false, ?DespawnTime : float = 60.0, ?ParticipatingPlayers : []player_manager = array{}):void =
        if (DropItems.Length < 0): 
            return
            
        Rad := Max(DropItems.Length-1, 1) * 40.0 + 80.0 
        AngleOffset := GetRandomFloat(0.0, PiFloat * 2.0)
        for (i := 0..DropItems.Length - 1):
            if:
                Item := DropItems[i]
                Angle := (1.0 * i) / (1.0 * DropItems.Length) * PiFloat * 2.0 + AngleOffset 
            then:
                InnerRad := GetRandomFloat(Rad/2.0, Rad)
                Offset := vector3{X := InnerRad * Cos(Angle), Y := InnerRad * Sin(Angle), Z := 0.0}

                var TP : ?vector3 = TargetPos
                if (T := TP?):
                    set TP = option{T + Offset}

                spawn {SpawnItem(Item, Pos + Offset, ?TargetPos := TP, ?DespawnTime := DespawnTime)}

    SpawnItem(Item : drop_item, Pos : vector3, ?TargetPos : ?vector3 = false, ?DespawnTime : float = 60.0, ?ParticipatingPlayers : []player_manager = array{})<suspends>:void =

        var PickupText : string = Item.PickupText
        var Col : collectible_data = collectible_data{
            Name := Item.Name
            Key := Item.Name
            Count := GetRandomBigNumber(Item.MinCount, Item.MaxCount)
        }

        if (Info := Item.ArmorInfo?):
            Armor := ArmorGenerator.GenArmor(Info)
            set Col = collectible_data{
                Name := Item.Name
                Key := Armor.GetImageKey()
                Count := GetRandomBigNumber(Item.MinCount, Item.MaxCount)
                ArmorData := option{Armor.ToData()}
            }
            set PickupText = "{Armor.GetName()}"

        if (Req :=Item.RequiredQuest?):
            var FilteredParticipatingPlayers : []player_manager = ParticipatingPlayers

            if (FilteredParticipatingPlayers.Length < 0):
                for (Key->Value : GameManager?.PlayerManagers):
                    set FilteredParticipatingPlayers = FilteredParticipatingPlayers + array{Value}

            var Temp : []player_manager = array{}
            for (Player : FilteredParticipatingPlayers):
            
                if (CurrentQuests := Player.GetCurrentQuests[]):
                    for (i := 0..CurrentQuests.Length - 1):
                        if:
                            Quest := CurrentQuests[i]
                            Quest.QuestID = Req.QuestID
                            Quest.TaskID = Req.TaskID
                        then:
                            set Temp = Temp + array{Player}

            if (Temp.Length > 0):
                SpawnItem(Col, PickupText, Pos, ?TargetPos := TargetPos, ?DespawnTime := DespawnTime, ?ParticipatingPlayers := Temp)
        else if:
            Item.Unique?
            PlayerID := GetRandomInt(0, ParticipatingPlayers.Length - 1)
            Player := ParticipatingPlayers[PlayerID]
        then:
            SpawnItem(Col, PickupText, Pos, ?TargetPos := TargetPos, ?DespawnTime := DespawnTime, ?ParticipatingPlayers := array{Player})
        else:
            SpawnItem(Col, PickupText, Pos, ?TargetPos := TargetPos, ?DespawnTime := DespawnTime)
    
    SpawnItem(CollectibleData : collectible_data, InteractionText : string, Pos : vector3, ?TargetPos : ?vector3 = false, ?DespawnTime : float = 60.0, ?ParticipatingPlayers : []player_manager = array{})<suspends>:void =  
        
        if (CollectibleData.Key = "XPCoin"):
            SpawnCoin(CollectibleData, InteractionText, Pos, ?TargetPos := TargetPos, ?DespawnTime := DespawnTime, ?ParticipatingPlayers := ParticipatingPlayers)
            return
        
        if:
            Collectible := Available[0]
            Temp := Available.RemoveElement[0]
            set Available = Temp
        then:
            Collectible.Move(Pos, ?TargetPos := TargetPos)
            Collectible.SetProperties(CollectibleData, InteractionText, ParticipatingPlayers)
            # race:
            #     Collectible.Interactible.InteractedWithEvent.Await()
            #     Sleep(DespawnTime)
            Sleep(DespawnTime)
            set Collectible.SubpositionTarget = vector3{}
            Collectible.Move(Collectible.InitialPosition)
            set Available = Available + array{Collectible}

    SpawnCoin(CollectibleData : collectible_data, InteractionText : string, Pos : vector3, ?TargetPos : ?vector3 = false, ?DespawnTime : float = 60.0, ?ParticipatingPlayers : []player_manager = array{})<suspends>:void =  
        if:
            Collectible := AvailableCoins[0]
            Temp := AvailableCoins.RemoveElement[0]
            set AvailableCoins = Temp
        then:
            Collectible.Move(Pos, ?TargetPos := TargetPos)
            Collectible.SetProperties(CollectibleData, InteractionText, ParticipatingPlayers)
            # race:
            #     Collectible.Interactible.InteractedWithEvent.Await()
            #     Sleep(DespawnTime)
            Sleep(DespawnTime)
            set Collectible.SubpositionTarget = vector3{}
            Collectible.Move(Collectible.InitialPosition)
            set AvailableCoins = AvailableCoins + array{Collectible}