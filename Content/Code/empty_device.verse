using { /Fortnite.com/Devices }
empty_device := class(creative_device):