using { /Fortnite.com/Devices }
using { /Fortnite.com/UI }
using { /Fortnite.com/Characters }
using { /Verse.org/Simulation }
using { /UnrealEngine.com/Temporary/UI }
using { /Verse.org/Colors }
using { /UnrealEngine.com/Temporary/SpatialMath }
using { /Verse.org/Assets }

equipment_menu_device := class(menu_device_template):
    
    @editable
    GameManager : game_manager = game_manager{}

    GetGameManager<override>():game_manager=
        return GameManager

    MakeMenu<override>(GM : game_manager, PM: player_manager):menu_template=
        return equipment_menu{GameManager:=option{GM}, Agent:=option{PM.Player}}

    GetTitle<override>():string=
        return "Inventory"

    GetIcon<override>():texture=
        return Images.Backpack

    IsEnabled<override>(PlayerManager : player_manager)<transacts>:logic=
        if:
            PlayerManager.GetAchievementLevel["InteractedWithNPC_Samira"] > 0
        then:
            return true 

        if (PlayerManager.GetPrestigeCount() > 0):
            return true
        
        return false

equipment_menu := class(menu_template):
    FloatingText : ui_bounce_mover = ui_bounce_mover{}

    var DialogBackground : texture_block = texture_block{ DefaultImage := Images.StoreBox, DefaultDesiredSize := vector2{X := 1920.0, Y := 1080.0}} 
    var ItemDisplays : []list_item_display = array{}
    var EquipDisplays : []list_item_display = array{}
    var InfoDisplay : ?item_info_display = false

    var ItemListBox : stack_box = stack_box{Orientation:=orientation.Vertical}
    var EquipListBox : stack_box = stack_box{Orientation:=orientation.Vertical}

    var Filters : []item_filter = array{everything_filter{}, collectible_filter{}, armor_filter{},
                                        sellable_filter{}, consumable_filter{}, 
                                        head_filter{}, body_filter{}, legs_filter{}, 
                                        boots_filter{}, gloves_filter{}, accessory_filter{} }
    
    var CurrentFilter : ?item_filter = false
    var FilterSelector : texture_block = texture_block{ DefaultImage := Images.SelectBorder, DefaultDesiredSize := vector2{X := 62.0, Y := 72.0}}
    var FilterSelectorPos : vector2 = vector2{X := 0.14, Y := 0.82}

    var CloseButton : button_quiet = button_quiet:
    var PageBackButton : button_quiet = button_quiet:
    var PageForwardButton : button_quiet = button_quiet:
    var PageText : text_block = text_block:
        DefaultTextColor:=NamedColors.White
        DefaultShadowColor:=NamedColors.Black   
        DefaultShadowOffset:=option{vector2{X:=2.0,Y:=2.0}}
        DefaultShadowOpacity:= 1.0
        
    TitleTextInventory : text_block = text_block:
        DefaultTextColor:=NamedColors.Black
        DefaultShadowColor:=NamedColors.Orange   
        DefaultShadowOffset:=option{vector2{X:=2.0,Y:=2.0}}
        DefaultShadowOpacity:= 1.0
    TitleTextEquipment : text_block = text_block:
        DefaultTextColor:=NamedColors.Black
        DefaultShadowColor:=NamedColors.Green   
        DefaultShadowOffset:=option{vector2{X:=2.0,Y:=2.0}}
        DefaultShadowOpacity:= 1.0
    InventoryCutIn : texture_block = texture_block:
        DefaultImage := Images.InvCutIn
    EquipmentCutIn : texture_block= texture_block:
        DefaultImage := Images.InvCutIn
    
    var LastPage : int = 0
    var Page : int = 0
    ItemsPerPage : int = 9
    EquipSlots : []armor_slot = array{armor_slot.Head, armor_slot.Torso, armor_slot.Legs, armor_slot.Boots, armor_slot.Gloves, armor_slot.Accessory}

    UpdateMenu<override>(Widget:canvas):void=
        spawn:
            FloatingText.Activate()

        set CurrentFilter = option{sellable_filter{}}

        if:
            PlayerManager := GameManager?.PlayerManagers[Agent?]
            PlayerCollectibles := PlayerManager.GetCollectibleMap[]
            IDisp := InfoDisplay?
        then:
            var ItemsToShow : []collectible_data = array{}

            for (Info : PlayerCollectibles):
                if:
                    (not CurrentFilter?) or CurrentFilter?.Filter[Info, PlayerManager]
                    BigNumber_GreaterThanOrEquals(Info.Count, big_number{Value:=1.0})?
                then:
                    set ItemsToShow = ItemsToShow + array{Info}
            
            set LastPage = Floor((ItemsToShow.Length-1) / ItemsPerPage) or 0

            if (Page < 0):
                set Page = 0
            if (Page > LastPage):
                set Page = LastPage
        
            for (i := 0..ItemDisplays.Length - 1):
                if (ItemDisplay := ItemDisplays[i]):
                    

                    var Selected : logic = false
                    if (IDisp.CurrentItem? = ItemDisplay):
                        set Selected = true

                    if:
                        ToShow := ItemsToShow[i + Page*ItemsPerPage]
                    then:

                        if (not ItemDisplay.CurrentItem?):
                            ItemListBox.AddWidget(
                                stack_box_slot:
                                    Widget := ItemDisplay.Widget
                                    HorizontalAlignment := horizontal_alignment.Fill
                                    Padding:= margin{Top := 0.0, Left := 0.0, Right := 0.0, Bottom := -20.0}
                            )

                        ItemDisplay.Update(option{ToShow}, Selected)
                    else:

                        ItemListBox.RemoveWidget(ItemDisplay.Widget)
                        ItemDisplay.Update(false, Selected)

                        
            for (i := 0..ItemDisplays.Length - 1):
                if (ItemDisplay := ItemDisplays[i]):

                    if:
                        not IDisp.CurrentItem?.CurrentItem?
                        ID1 := i + Page*ItemsPerPage 
                        ID2 := ItemsToShow.Length - 1
                    then:
                        Log("ID1: {ID1}, ID2: {ID2}")
                        if:
                            ID1 = ID2
                        then:
                            IDisp.Update(option{ItemDisplay})
                            ItemDisplay.Update(option{ItemsToShow[ID1]}, true)


            
        
            for (i := 0..EquipDisplays.Length - 1):
                if:
                    ItemDisplay := EquipDisplays[i]
                    ArmorSlot := EquipSlots[i]
                then:
                    
                    var Selected : logic = false
                    if (IDisp.CurrentItem? = ItemDisplay):
                        set Selected = true

                    if:
                        ToShow := PlayerManager.GetArmorInSlot[ArmorSlot]
                    then:
                        ArmorAsCollectible := collectible_data{
                            Name := ToShow.GetName()
                            Key := "Unique_{ToShow.UID}"
                            Count := big_number{Value:=1.0, Exponent:=0}
                            ArmorData := option{ToShow.ToData()}
                        }
                        ItemDisplay.Update(option{ArmorAsCollectible}, Selected)
                    else:
                        ArmorAsCollectible := collectible_data{
                            Name := "{ArmorSlot}: None"
                            Key := "Empty {ArmorSlot}"
                            Count := big_number{Value:=0.0, Exponent:=0}
                            ArmorData := false
                        }
                        ItemDisplay.Update(option{ArmorAsCollectible}, Selected)

            PageText.SetText(STM("Page {Page+1}/{LastPage+1}"))

            #Breaks controllers :(

            # if (Page <= 0):
            #     PageBackButton.SetEnabled(false)
            # else:
            #     PageBackButton.SetEnabled(true) 
            
            # if (Page >= LastPage):
            #     PageForwardButton.SetEnabled(false)
            # else:
            #     PageForwardButton.SetEnabled(true)

            IDisp.Update(IDisp.CurrentItem)

            if (W := _Widget?):
                W.RemoveWidget(FilterSelector)
                W.AddWidget(
                    canvas_slot:
                        Widget:= FilterSelector
                        Anchors := anchors { Minimum := FilterSelectorPos, Maximum := FilterSelectorPos}
                        Alignment := vector2{X := 0.5, Y := 0.5}
                        Offsets := margin{Top := 0.0, Left := 0.0, Right := 0.0, Bottom := 0.0}
                        ZOrder := 2
                ) 
        
        

    CreateMenu<override>():canvas=
        IDisp := item_info_display{Menu:=Self}
        IDisp.Create()

        set InfoDisplay = option{IDisp}

        Result := canvas:
            Slots := array:
                canvas_slot:
                    Widget:= DialogBackground
                    SizeToContent := true
                    Anchors := anchors{ Minimum := vector2{X := 0.06, Y := 0.06}, Maximum := vector2{X := 0.94, Y := 0.94}}  
                    ZOrder := 1
                canvas_slot:
                    Widget:= CloseButton
                    Anchors := anchors{ Minimum := vector2{X := 0.5, Y := 0.915}, Maximum := vector2{X := 0.5, Y := 0.915}}
                    Alignment := vector2{X := 0.5, Y := 1.0}
                    ZOrder := 3
                canvas_slot:
                    Widget := InventoryCutIn
                    Anchors := anchors{ Minimum := vector2{X := 0.076, Y := 0.32}, Maximum := vector2{X := 0.4, Y := 0.36}}  
                    Offsets := margin{Top := 50.0, Left := 50.0, Right := -50.0, Bottom := -50.0}
                    Alignment := vector2{X := 0.0, Y := 0.0}
                    ZOrder := 4
                canvas_slot:
                    Widget := TitleTextInventory
                    Anchors := anchors{ Minimum := vector2{X := 0.076, Y := 0.32}, Maximum := vector2{X := 0.4, Y := 0.36}}  
                    Offsets := margin{Top := 50.0, Left := 50.0, Right := -50.0, Bottom := -50.0}
                    Alignment := vector2{X := 0.0, Y := 0.0}
                    ZOrder := 5
                canvas_slot:
                    Widget := EquipmentCutIn
                    Anchors := anchors{ Minimum := vector2{X := 0.076, Y := 0.06}, Maximum := vector2{X := 0.4, Y := 0.1}}  
                    Offsets := margin{Top := 50.0, Left := 50.0, Right := -50.0, Bottom := -50.0}
                    Alignment := vector2{X := 0.0, Y := 0.0}
                    ZOrder := 6
                canvas_slot:
                    Widget := TitleTextEquipment
                    Anchors := anchors{ Minimum := vector2{X := 0.076, Y := 0.06}, Maximum := vector2{X := 0.4, Y := 0.1}}  
                    Offsets := margin{Top := 50.0, Left := 50.0, Right := -50.0, Bottom := -50.0}
                    Alignment := vector2{X := 0.0, Y := 0.0}
                    ZOrder := 7
                canvas_slot:
                    Widget := ItemListBox
                    Anchors := anchors{ Minimum := vector2{X := 0.1, Y := 0.36}, Maximum := vector2{X := 0.6, Y := 0.84}}  
                    Offsets := margin{Top := 50.0, Left := 50.0, Right := 50.0, Bottom := 50.0}
                    Alignment := vector2{X := 0.0, Y := 0.0}
                    ZOrder := 8
                canvas_slot:
                    Widget:= PageBackButton
                    Anchors := anchors{ Minimum := vector2{X := 0.50, Y := 0.686}, Maximum := vector2{X := 0.50, Y := 0.686}}  
                    Offsets := margin{Top := 50.0, Left := 10.0, Right := 10.0, Bottom := 10.0}
                    Alignment := vector2{X := 0.0, Y := 0.0}
                    ZOrder := 9
                canvas_slot:
                    Widget:= PageForwardButton
                    Anchors := anchors{ Minimum := vector2{X := 0.54, Y := 0.686}, Maximum := vector2{X := 0.54, Y := 0.686}}  
                    Offsets := margin{Top := 50.0, Left := 10.0, Right := 10.0, Bottom := 10.0}
                    Alignment := vector2{X := 0.0, Y := 0.0}
                    ZOrder := 10
                canvas_slot:
                    Widget:= PageText
                    Anchors := anchors{ Minimum := vector2{X := 0.47, Y := 0.706}, Maximum := vector2{X := 0.47, Y := 0.706}}  
                    Offsets := margin{Top := 60.0, Left := 10.0, Right := 10.0, Bottom := 10.0}
                    Alignment := vector2{X := 1.0, Y := 0.5}
                    ZOrder := 11
                canvas_slot:
                    Widget:= IDisp.Widget
                    Anchors := anchors{ Minimum := vector2{X := 0.6, Y := 0.1}, Maximum := vector2{X := 0.9, Y := 0.9}}  
                    Offsets := margin{Top := 50.0, Left := 50.0, Right := 50.0, Bottom := 50.0}
                    Alignment := vector2{X := 0.5, Y := 0.5}
                    ZOrder := 12
                canvas_slot:
                    Widget:= FloatingText.Setup(20, vector2{X := 55.0, Y := 55.0}, 2.0, 0.95)
                    Anchors := anchors{ Minimum := vector2{X := 0.64, Y := 0.4}, Maximum := vector2{X := 0.64, Y := 0.4}}  
                    Offsets := margin{Top := 0.0, Left := 0.0, Right := 0.0, Bottom := 0.0}
                    Alignment := vector2{X := 0.5, Y := 0.5}
                    ZOrder := 13
                canvas_slot:
                    Widget := EquipListBox
                    Anchors := anchors{ Minimum := vector2{X := 0.1, Y := 0.1}, Maximum := vector2{X := 0.6, Y := 0.2}}  
                    Offsets := margin{Top := 50.0, Left := 50.0, Right := 50.0, Bottom := 50.0}
                    Alignment := vector2{X := 0.0, Y := 0.0}
                    ZOrder := 14
            
        # for (i := 0..Filters.Length - 1; Filter := Filters[i];
        # rangeI1 := type{_X:int where 0 <= _X, _X <= 2147483647}[100+i*2];
        # rangeI2 := type{_X:int where 0 <= _X, _X <= 2147483647}[101+i*2]):
        #     FilterImg := texture_block:
        #         DefaultImage := Filter.GetImage()
        #         DefaultDesiredSize := vector2{X := 40.0, Y := 40.0}
        #     FilterBtn := button_quiet:
            
        #     Pos := vector2{X := 0.14 + 0.04229*i, Y := 0.82}

        #     FilterBtn.OnClick().Subscribe(filter_setter{Filter:=Filter, Menu:=Self, Position:=Pos}.Set)
        #     Result.AddWidget(
        #         canvas_slot:
        #             Widget := FilterImg
        #             Anchors := anchors{ Minimum := Pos, Maximum := Pos}  
        #             Offsets := margin{Top := 0.0, Left := 0.0, Right := 0.0, Bottom := 0.0}
        #             Alignment := vector2{X := 0.5, Y := 0.5}
        #             ZOrder := rangeI1
        #     )
        #     Result.AddWidget(
        #         canvas_slot:
        #             Widget := FilterBtn
        #             Anchors := anchors{ Minimum := Pos, Maximum := Pos}  
        #             Offsets := margin{Top := 0.0, Left := 0.0, Right := 0.0, Bottom := 0.0}
        #             Alignment := vector2{X := 0.5, Y := 0.5}
        #             ZOrder := rangeI2
        #     )
        
        if (PlayerManager := GetPlayerManager[]):
            for (i := 0..ItemsPerPage-1):
                set ItemDisplays = ItemDisplays + array{list_item_display{Menu:=Self}}

            for (i := 0..EquipSlots.Length-1):
                set EquipDisplays = EquipDisplays + array{list_item_display{Menu:=Self}}
            
            for (ItemDisplay : ItemDisplays):
                ItemDisplay.Create()
            
            for (ItemDisplay : EquipDisplays):
                ItemDisplay.Create()
                EquipListBox.AddWidget(
                    stack_box_slot:
                        Widget := ItemDisplay.Widget
                        HorizontalAlignment := horizontal_alignment.Fill
                        Padding:= margin{Top := 0.0, Left := 0.0, Right := 0.0, Bottom := -20.0}
                )
            
            TitleTextInventory.SetText(STM("Inventory"))
            TitleTextEquipment.SetText(STM("Equipment"))

            CloseButton.SetText(STM("                                                BACK                                                "))            
            CloseButton.OnClick().Subscribe(PlayerManager.ClosePopup)

            PageBackButton.SetText(STM("<"))
            PageBackButton.OnClick().Subscribe(PageBack)

            PageForwardButton.SetText(STM(">"))
            PageForwardButton.OnClick().Subscribe(PageForward)

        return Result


    SelectItem(Item : collectible_data):void=
        for (Disp : ItemDisplays; Disp.CurrentItem?.Key = Item.Key):
            if (IDisp := InfoDisplay?):
                IDisp.Update(option{Disp})
        Update()

    PageBack(Wdata:widget_message):void=
        if (Page <= 0):
            return
        
        set Page = Page - 1
        if (IDisp := InfoDisplay?):
            IDisp.Update(false)
        Update()
    PageForward(Wdata:widget_message):void=
        if (Page >= LastPage):
            return

        set Page = Page + 1
        if (IDisp := InfoDisplay?):
            IDisp.Update(false)
        Update()

list_item_display := class<unique>:
    Menu : menu_template

    var CurrentItem : ?collectible_data = false

    Widget : canvas = canvas:
    
    ItemBackground : color_block = color_block:
        DefaultColor:=NamedColors.White
        DefaultOpacity:=0.4
    SelectedBorder : color_block = color_block:
        DefaultColor:=NamedColors.Green
        DefaultOpacity:=0.4

    IconDisplay : texture_block = texture_block:
        DefaultImage := Images.Empty
    NameDisplay : text_block = text_block:
        DefaultTextColor:=NamedColors.White
        DefaultShadowColor:=NamedColors.Black   
        DefaultShadowOffset:=option{vector2{X:=2.0,Y:=2.0}}
        DefaultShadowOpacity:= 1.0
    UseButton : button_quiet = button_quiet:
    InfoButton : button_quiet = button_quiet:
            
            
    Create():void=

        StackBoxL := stack_box{Orientation:=orientation.Horizontal}
        StackBoxR := stack_box{Orientation:=orientation.Horizontal}

        Widget.AddWidget(
            canvas_slot:
                Widget := StackBoxL
                Anchors := anchors{ Minimum := vector2{X := 0.0, Y := 0.0}, Maximum := vector2{X := 1.0, Y := 0.5}}  
                Alignment := vector2{X := 0.0, Y := 0.5}
                Offsets := margin{Top := -5.0, Left := 0.0, Right := 0.0, Bottom := -5.0}
                ZOrder := 1
        )

        Widget.AddWidget(
            canvas_slot:
                Widget := StackBoxR
                Anchors := anchors{ Minimum := vector2{X := 1.0, Y := 0.0}, Maximum := vector2{X := 1.0, Y := 0.5}}  
                Alignment := vector2{X := 1.0, Y := 1.0}
                Offsets := margin{Top := 0.0, Left := 0.0, Right := 0.0, Bottom := 0.0}
                ZOrder := 2
        )

        Widget.AddWidget(
            canvas_slot:
                Widget := ItemBackground
                Anchors := anchors{ Minimum := vector2{X := 0.0, Y := 0.0}, Maximum := vector2{X := 1.0, Y := 0.5}}  
                Offsets := margin{Top := -2.0, Left := -2.0, Right := -2.0, Bottom := -2.0}
                Alignment := vector2{X := 1.0, Y := 1.0}
                ZOrder := 0
        )

        Widget.AddWidget(
            canvas_slot:
                Widget := SelectedBorder
                Anchors := anchors{ Minimum := vector2{X := 0.0, Y := 0.0}, Maximum := vector2{X := 1.0, Y := 0.5}}  
                Offsets := margin{Top := -2.0, Left := -2.0, Right := -2.0, Bottom := -2.0}
                Alignment := vector2{X := 1.0, Y := 1.0}
                ZOrder := 0
        )

        StackBoxL.AddWidget(
            stack_box_slot:
                Widget := IconDisplay
                HorizontalAlignment := horizontal_alignment.Left
                VerticalAlignment := vertical_alignment.Bottom
        )
        StackBoxL.AddWidget(
            stack_box_slot:
                Widget := NameDisplay
                HorizontalAlignment := horizontal_alignment.Left
                VerticalAlignment := vertical_alignment.Bottom
                Padding:= margin{Top := 0.0, Left := 10.0, Right := 10.0, Bottom := 0.0}
        )
        # StackBoxR.AddWidget(
        #     stack_box_slot:
        #         Widget := UseButton
        #         HorizontalAlignment := horizontal_alignment.Right
        #         VerticalAlignment := vertical_alignment.Bottom
        #         Padding:= margin{Top := 0.0, Left := 0.0, Right := 3.0, Bottom := 0.0}
        # )
        StackBoxR.AddWidget(
            stack_box_slot:
                Widget := InfoButton
                HorizontalAlignment := horizontal_alignment.Right
                VerticalAlignment := vertical_alignment.Bottom
                Padding:= margin{Top := 0.0, Left := 0.0, Right := 3.0, Bottom := 0.0}
        )

        UseButton.SetText(STM("Use"))
        InfoButton.SetText(STM("Info"))

        UseButton.OnClick().Subscribe(OnUse)
        InfoButton.OnClick().Subscribe(OnInfo)

        SelectedBorder.SetVisibility(widget_visibility.Hidden)


    Update(Item:?collectible_data, Selected:logic):void=
        if:
            It := Item?
            PlayerManager := Menu.GetPlayerManager[]
        then:
            if (Selected?):
                SelectedBorder.SetVisibility(widget_visibility.Visible)
            else:
                SelectedBorder.SetVisibility(widget_visibility.Hidden)

            set CurrentItem = option{It}
            IconDisplay.SetImage(It.GetImage())
            if (Armor := It.ArmorData?.ToArmor()):
                NameDisplay.SetText(STM("{Armor.GetName()}"))
                NameDisplay.SetTextColor(Armor.Rarity.ToColor())

                if (PlayerManager.IsArmorEquipped(Armor)?):
                    ItemBackground.SetColor(NamedColors.Gold)
                else:
                    ItemBackground.SetColor(NamedColors.Grey)
            else:
                var ItemName : string = It.Name
                if (ItemName = ""):         #TODO: Why does this happen?
                    set ItemName = It.Key

                if (BigNumber_GreaterThanOrEquals(It.Count, big_number{Value:=2.0})?):
                    NameDisplay.SetText(STM("{ItemName}({It.Count})"))
                else:
                    NameDisplay.SetText(STM("{ItemName}"))

                NameDisplay.SetTextColor(NamedColors.White)
                ItemBackground.SetColor(NamedColors.White)

            Widget.SetVisibility(widget_visibility.Visible) 
            InfoButton.SetVisibility(widget_visibility.Visible)
            Value := It.GetSellValue()

            UseTxt := It.GetUseText()
            if (UseTxt = ""):
                UseButton.SetVisibility(widget_visibility.Hidden)
            else:
                UseButton.SetVisibility(widget_visibility.Visible)

            UseButton.SetText(STM(UseTxt))
        else:
            set CurrentItem = false
            # Widget.SetVisibility(widget_visibility.Collapsed)
            # InfoButton.SetVisibility(widget_visibility.Collapsed)

    OnUpgrade(Wdata:widget_message):void =
        if:
            PlayerManager := Menu.GetPlayerManager[]
            It := CurrentItem?
        then:
            UpgradeCost := It.GetUpgradeCost()
            if (BigNumber_GreaterThanOrEquals(PlayerManager.GetCollectibleCount("Midas Coins"), UpgradeCost)?):
                if (BigNumber_GreaterThanOrEquals(PlayerManager.GetCollectibleCount("UpgradeHammer"), big_number{Value:=1.0})?):
                    if (It.ArmorData?.Potential > 0):
                    
                        PlayerManager.GainCollectible(
                            collectible_data{
                                Name := "Midas Coins"
                                Key := "Midas Coins"
                                Count := UpgradeCost * -1.0
                            }
                        )
                        
                        PlayerManager.GainCollectible(
                            collectible_data{
                                Name := "Upgrade Hammer"
                                Key := "UpgradeHammer"
                                Count := big_number{Value:=-1.0, Exponent:=0}
                            }
                        )

                        if:
                            Item := PlayerManager.UpgradeItem(It)?
                            Ar2 := Item.ToArmor()
                            Ar := It.ArmorData?.ToArmor()
                        then:
                            spawn:
                                FloatingTextSpaced(It, Ar, Ar2)

                            if:
                                PlayerManager.IsArmorEquipped(Ar2)?
                            then:
                                PlayerManager.EquipArmor(Ar2)
                                Menu.Update()
                                PlayerManager.EquipArmor(Ar2)
                                Menu.Update()
                            
                    else:
                        if:
                            EqMenu := equipment_menu[Menu]
                        then:
                            EqMenu.FloatingText.Create(Images.QuestionMark, "Not Enough Potential") 
                else:
                    if:
                        EqMenu := equipment_menu[Menu]
                    then:
                        EqMenu.FloatingText.Create(Images.QuestionMark, "Not Enough Upgrade Hammers") 
            else:            
                if:
                    EqMenu := equipment_menu[Menu]
                then:
                    EqMenu.FloatingText.Create(Images.QuestionMark, "Not Enough Midas Coins")
            
            Menu.Update()
            if:
                EqMenu := equipment_menu[Menu]
            then:
                EqMenu.SelectItem(It)

    FloatingTextSpaced(It : collectible_data, Armor1 : armor, Armor2 : armor)<suspends>:void =
        UpgradeCost := It.GetUpgradeCost()
        if (EqMenu := equipment_menu[Menu]):
            EqMenu.FloatingText.Create(It.GetImage(), GetStatDiff(Armor2, Armor1)) 
            Sleep(0.2)
            EqMenu.FloatingText.Create(Images.MidasCoins, "{UpgradeCost * -1.0}") 
            Sleep(0.2)
            EqMenu.FloatingText.Create(Images.UpgradeHammer, "-1")

    OnUse(Wdata:widget_message):void =
        if:
            PlayerManager := Menu.GetPlayerManager[]
            It := CurrentItem?
        then:
            UseText := PlayerManager.UseCollectible(It)
            Menu.Update()
            
            if:
                EqMenu := equipment_menu[Menu]
            then:
                EqMenu.FloatingText.Create(It.GetUseImage(), It.GetUsePopupText())
    OnEquip(Wdata:widget_message):void =
        if:
            PlayerManager := Menu.GetPlayerManager[]
            It := CurrentItem?
        then:
            PlayerManager.UseCollectible(It)
            Menu.Update()
    OnSell(Wdata:widget_message):void =
        if:
            PlayerManager := Menu.GetPlayerManager[]
            It := CurrentItem?
        then:
            if (BigNumber_GreaterThanOrEquals(big_number{Value:=0.0}, It.GetSellValue())?):
                return

            if (PlayerManager.IsArmorEquipped(It.ArmorData?.ToArmor())?):
                return
            
            PlayerManager.GainCollectible(It.WithCount(big_number{Value:=-1.0, Exponent:=0}), ?ForceKey := true)
            PlayerManager.GainCollectible(
                collectible_data{
                    Name := "Midas Coins"
                    Key := "Midas Coins"
                    Count := It.GetSellValue()
                }
            )
            Log("OnSell Update")
            Menu.Update()

    OnSellAll(Wdata:widget_message):void =
        if:
            PlayerManager := Menu.GetPlayerManager[]
            It := CurrentItem?
            Count := PlayerManager.GetCollectibleCount(It.Key)
        then:
            if (BigNumber_GreaterThanOrEquals(big_number{Value:=0.0}, It.GetSellValue())?):
                return

            if (PlayerManager.IsArmorEquipped(It.ArmorData?.ToArmor())?):
                return
            
            PlayerManager.GainCollectible(It.WithCount(big_number{Value:= -Count.Value, Exponent:= Count.Exponent}), ?ForceKey := true)
            PlayerManager.GainCollectible(
                collectible_data{
                    Name := "Midas Coins"
                    Key := "Midas Coins"
                    Count := It.GetSellValue() * Count
                }
            )
            Log("OnSellAll Update")
            Menu.Update()
            
    OnInfo(Wdata:widget_message):void =
        if:
            IDisp := equipment_menu[Menu].InfoDisplay?
        then:
            IDisp.Update(option{Self})
        # else if:
        #     IDisp := loot_popup[Menu].InfoDisplay?
        # then:
        #     IDisp.Update(option{Self})
        Menu.Update()

            

item_info_display := class():
    Menu : menu_template

    var CurrentItem : ?list_item_display = false

    Widget : canvas = canvas:
    
    ItemBackground : color_block = color_block:
        DefaultColor:=NamedColors.White
        DefaultOpacity:=0.4

    IconDisplay : texture_block = texture_block:
        DefaultImage := Images.Empty
        DefaultDesiredSize := vector2{X := 200.0, Y := 200.0}
    RarityDisplay : text_block = text_block:
            DefaultTextColor:=NamedColors.White
            DefaultShadowColor:=NamedColors.Black   
            DefaultShadowOffset:=option{vector2{X:=2.0,Y:=2.0}}
            DefaultShadowOpacity:= 1.0
    NameDisplay : text_block = text_block:
        DefaultTextColor:=NamedColors.White
        DefaultShadowColor:=NamedColors.Black   
        DefaultShadowOffset:=option{vector2{X:=2.0,Y:=2.0}}
        DefaultShadowOpacity:= 1.0
    var DescDisplay : stack_box = stack_box {Orientation:=orientation.Vertical}
    UpgradeButton : button_quiet = button_quiet:
    EquipButton : button_quiet = button_quiet:
    SellButton : button_quiet = button_quiet:
    SellAllButton : button_quiet = button_quiet:

    SellValueStackBox : stack_box = stack_box:
        Orientation := orientation.Horizontal
    SellValue : text_block = text_block:
        DefaultTextColor:=NamedColors.Gold
        DefaultShadowColor:=NamedColors.Black   
        DefaultShadowOffset:=option{vector2{X:=2.0,Y:=2.0}}
        DefaultShadowOpacity:= 1.0
    SellIcon : texture_block = texture_block:
        DefaultImage := Images.MidasCoins
        DefaultDesiredSize := vector2{X := 40.0, Y := 40.0}
        

    UpgradeCostStackBox : stack_box = stack_box:
        Orientation := orientation.Vertical
    UpgradeCostStackBox2 : stack_box = stack_box:
        Orientation := orientation.Horizontal
    UpgradeCostStackBox3 : stack_box = stack_box:
        Orientation := orientation.Horizontal

    UpgradeCostCoins : text_block = text_block:
        DefaultTextColor:=NamedColors.Gold
        DefaultShadowColor:=NamedColors.Black   
        DefaultShadowOffset:=option{vector2{X:=2.0,Y:=2.0}}
        DefaultShadowOpacity:= 1.0
    UpgradeIconCoins : texture_block = texture_block:
        DefaultImage := Images.MidasCoins
        DefaultDesiredSize := vector2{X := 40.0, Y := 40.0}
    UpgradeCostHammer : text_block = text_block:
        DefaultTextColor:=NamedColors.Gold
        DefaultShadowColor:=NamedColors.Black   
        DefaultShadowOffset:=option{vector2{X:=2.0,Y:=2.0}}
        DefaultShadowOpacity:= 1.0
    UpgradeIconHammer : texture_block = texture_block:
        DefaultImage := Images.UpgradeHammer
        DefaultDesiredSize := vector2{X := 40.0, Y := 40.0}
            
            
    Create():void=

        Widget.AddWidget(
            canvas_slot:
                Widget := ItemBackground
                Anchors := anchors{ Minimum := vector2{X := 0.0, Y := 0.0}, Maximum := vector2{X := 1.0, Y := 1.0}}  
                Offsets := margin{Top := 0.0, Left := 0.0, Right := 0.0, Bottom := 0.0}
                Alignment := vector2{X := 0.5, Y := 0.5}
                ZOrder := 0
        )

        Widget.AddWidget(
            canvas_slot:
                Widget := NameDisplay
                Anchors := anchors{ Minimum := vector2{X := 0.5, Y := 0.01}, Maximum := vector2{X := 0.5, Y := 0.01}}  
                Offsets := margin{Top := 0.0, Left := 0.0, Right := 0.0, Bottom := 0.0}
                Alignment := vector2{X := 0.5, Y := 0.0}
                ZOrder := 0
        )

        Widget.AddWidget(
            canvas_slot:
                Widget := RarityDisplay
                Anchors := anchors{ Minimum := vector2{X := 0.5, Y := 0.01}, Maximum := vector2{X := 0.5, Y := 0.01}}  
                Offsets := margin{Top := 35.0, Left := 0.0, Right := 0.0, Bottom := 0.0}
                Alignment := vector2{X := 0.5, Y := 0.0}
                ZOrder := 2
        )

        Widget.AddWidget(
            canvas_slot:
                Widget := IconDisplay
                Anchors := anchors{ Minimum := vector2{X := 0.5, Y := 0.28}, Maximum := vector2{X := 0.5, Y := 0.28}}  
                Offsets := margin{Top := 0.0, Left := 0.0, Right := 0.0, Bottom := 0.0}
                Alignment := vector2{X := 0.5, Y := 0.5}
                SizeToContent := true
                ZOrder := 0
        )

        Widget.AddWidget(
            canvas_slot:
                Widget := UpgradeButton
                Anchors := anchors{ Minimum := vector2{X := 0.2, Y := 0.77}, Maximum := vector2{X := 0.2, Y := 0.83}}  
                Offsets := margin{Top := 0.0, Left := 260.0, Right := 0.0, Bottom := 0.0}
                Alignment := vector2{X := 0.5, Y := 0.5}
                ZOrder := 0
        )

        Widget.AddWidget(
            canvas_slot:
                Widget := EquipButton
                Anchors := anchors{ Minimum := vector2{X := 0.2, Y := 0.77}, Maximum := vector2{X := 0.2, Y := 0.83}}  
                Offsets := margin{Top := 0.0, Left := 0.0, Right := 0.0, Bottom := 0.0}
                Alignment := vector2{X := 0.5, Y := 0.5}
                ZOrder := 0
        )

        Widget.AddWidget(
            canvas_slot:
                Widget := SellValueStackBox
                Anchors := anchors{ Minimum := vector2{X := 0.5, Y := 0.83}, Maximum := vector2{X := 0.5, Y := 0.87}}  
                Offsets := margin{Top := 80.0, Left := 0.0, Right := 0.0, Bottom := 0.0}
                Alignment := vector2{X := 0.5, Y := 0.5}
                ZOrder := 0
        )

        Widget.AddWidget(
            canvas_slot:
                Widget := UpgradeCostStackBox
                Anchors := anchors{ Minimum := vector2{X := 0.8, Y := 0.815}, Maximum := vector2{X := 0.8, Y := 0.93}}  
                Offsets := margin{Top := -140.0, Left := 10.0, Right := 0.0, Bottom := 0.0}
                Alignment := vector2{X := 0.5, Y := 0.5}
                ZOrder := 0
        )

        Widget.AddWidget(
            canvas_slot:
                Widget := SellButton
                Anchors := anchors{ Minimum := vector2{X := 0.2, Y := 0.87}, Maximum := vector2{X := 0.2, Y := 0.93}}  
                Offsets := margin{Top := 0.0, Left := 20.0, Right := 0.0, Bottom := 0.0}
                Alignment := vector2{X := 0.5, Y := 0.5}
                ZOrder := 0
        )

        Widget.AddWidget(
            canvas_slot:
                Widget := SellAllButton
                Anchors := anchors{ Minimum := vector2{X := 0.2, Y := 0.87}, Maximum := vector2{X := 0.2, Y := 0.93}}  
                Offsets := margin{Top := 0.0, Left := 260.0, Right := 0.0, Bottom := 0.0}
                Alignment := vector2{X := 0.5, Y := 0.5}
                ZOrder := 0
        )

        SellValueStackBox.AddWidget(
            stack_box_slot:
                Widget := SellIcon
                HorizontalAlignment:= horizontal_alignment.Left
                Padding:= margin{Left:=2.0, Top:=0.0, Bottom:= 0.0, Right:= 2.0} 
        )
        SellValueStackBox.AddWidget(
            stack_box_slot:
                Widget := SellValue
                HorizontalAlignment:= horizontal_alignment.Right
                Padding:= margin{Left:=2.0, Top:=0.0, Bottom:= 0.0, Right:= 2.0} 
        )

        UpgradeCostStackBox.AddWidget(
            stack_box_slot:
                Widget := UpgradeCostStackBox2
                HorizontalAlignment:= horizontal_alignment.Center
                Padding:= margin{Left:=2.0, Top:=0.0, Bottom:= 0.0, Right:= 2.0} 
        )
        UpgradeCostStackBox.AddWidget(
            stack_box_slot:
                Widget := UpgradeCostStackBox3
                HorizontalAlignment:= horizontal_alignment.Center
                Padding:= margin{Left:=2.0, Top:=0.0, Bottom:= 0.0, Right:= 2.0} 
        )

        UpgradeCostStackBox3.AddWidget(
            stack_box_slot:
                Widget := UpgradeIconCoins
                HorizontalAlignment:= horizontal_alignment.Left
                Padding:= margin{Left:=2.0, Top:=0.0, Bottom:= 0.0, Right:= 2.0} 
        )
        UpgradeCostStackBox3.AddWidget(
            stack_box_slot:
                Widget := UpgradeCostCoins
                HorizontalAlignment:= horizontal_alignment.Left
                Padding:= margin{Left:=2.0, Top:=0.0, Bottom:= 0.0, Right:= 2.0} 
        )
        UpgradeCostStackBox2.AddWidget(
            stack_box_slot:
                Widget := UpgradeIconHammer
                HorizontalAlignment:= horizontal_alignment.Right
                Padding:= margin{Left:=2.0, Top:=0.0, Bottom:= 0.0, Right:= 2.0} 
        )
        UpgradeCostStackBox2.AddWidget(
            stack_box_slot:
                Widget := UpgradeCostHammer
                HorizontalAlignment:= horizontal_alignment.Right
                Padding:= margin{Left:=2.0, Top:=0.0, Bottom:= 0.0, Right:= 2.0} 
        )

        UpgradeCostHammer.SetText(STM("1"))

        UpgradeButton.SetText(STM("Upgrade"))
        EquipButton.SetText(STM(" Equip "))
        SellButton.SetText(STM("Sell"))
        SellAllButton.SetText(STM("Sell All"))

        UpgradeButton.OnClick().Subscribe(OnUpgrade)
        EquipButton.OnClick().Subscribe(OnEquip)
        SellButton.OnClick().Subscribe(OnSell)
        SellAllButton.OnClick().Subscribe(OnSellAll)


    Update(Item:?list_item_display):void=
        if:
            IDisp := Item?
            It := IDisp.CurrentItem?
            PlayerManager := Menu.GetPlayerManager[]
        then:
            set CurrentItem = option{IDisp}
            IconDisplay.SetImage(It.GetImage())
            RarityDisplay.SetText(STM(""))
            if (Armor := It.ArmorData?.ToArmor()):
                NameDisplay.SetText(STM("{Armor.GetName()}"))
                RarityDisplay.SetText(STM("Level {Armor.GetTotalLevel()} {Armor.Rarity}"))
                RarityDisplay.SetTextColor(Armor.Rarity.ToColor())
                UpgradeButton.SetVisibility(widget_visibility.Visible)
                EquipButton.SetVisibility(widget_visibility.Visible)
                    
                Widget.RemoveWidget(DescDisplay)
                set DescDisplay = MakeTextBox(It.GetDescription())

                if (PlayerManager.IsArmorEquipped(Armor)?):
                    EquipButton.SetText(STM("Unequip"))
                    # ItemBackground.SetColor(NamedColors.Gold)
                else:
                    EquipButton.SetText(STM("  Equip  "))
                    # ItemBackground.SetColor(NamedColors.Grey)
                ItemBackground.SetColor(Armor.Rarity.ToColor())
                
                    
                UpgradeCost := It.GetUpgradeCost()
                if:
                    BigNumber_GreaterThanOrEquals(PlayerManager.GetCollectibleCount("Midas Coins"), UpgradeCost)?
                    BigNumber_GreaterThanOrEquals(PlayerManager.GetCollectibleCount("UpgradeHammer"), big_number{Value:=1.0})?
                    It.ArmorData?.Potential > 0
                then:
                    if (BigNumber_GreaterThanOrEquals(UpgradeCost, big_number{Value:=1.0})?):
                        UpgradeCostCoins.SetText(STM("{UpgradeCost}"))
                        UpgradeCostStackBox.SetVisibility(widget_visibility.Visible)
                        UpgradeButton.SetVisibility(widget_visibility.Visible)
                        UpgradeButton.SetText(STM("Upgrade"))
                else:
                    if: 
                        It.ArmorData?.Potential > 0
                    then:
                        UpgradeButton.SetText(STM("Can't Upgrade"))
                    else:
                        UpgradeButton.SetText(STM("Maxed"))

                UpgradeButton.SetEnabled(true)

            else:
                var ItemName : string = It.Name
                if (ItemName = ""):         #TODO: Why does this happen?
                    set ItemName = It.Key

                if (BigNumber_GreaterThanOrEquals(It.Count, big_number{Value:=2.0})?):
                    NameDisplay.SetText(STM("{ItemName}({It.Count})"))
                else:
                    NameDisplay.SetText(STM("{ItemName}"))
                    
                Widget.RemoveWidget(DescDisplay)
                set DescDisplay = WordWrap(It.GetDescription(), 25)

                
                UpgradeButton.SetVisibility(widget_visibility.Hidden)
                EquipButton.SetVisibility(widget_visibility.Hidden)
                ItemBackground.SetColor(NamedColors.White)

            Widget.AddWidget(
                canvas_slot:
                    Widget := DescDisplay
                    Anchors := anchors{ Minimum := vector2{X := 0.1, Y := 0.48}, Maximum := vector2{X := 0.9, Y := 0.9}}  
                    Offsets := margin{Top := -160.0, Left := 0.0, Right := 0.0, Bottom := 0.0}
                    Alignment := vector2{X := 0.0, Y := 0.5}
                    ZOrder := 0
            )

            SellAllButton.SetEnabled(true)
            SellButton.SetEnabled(true)
            SellButton.SetText(STM("Sell"))
            SellAllButton.SetText(STM("Sell All"))

            Widget.SetVisibility(widget_visibility.Visible) 
            Value := It.GetSellValue()
            if:
                BigNumber_GreaterThanOrEquals(Value, big_number{Value:=1.0})?
                not PlayerManager.IsArmorEquipped(It.ArmorData?.ToArmor())?
            then:
                SellValue.SetText(STM("{Value}"))
                SellValueStackBox.SetVisibility(widget_visibility.Visible)
            else:
                SellValueStackBox.SetVisibility(widget_visibility.Hidden)
                SellButton.SetText(STM("Can't Sell"))
                SellAllButton.SetText(STM("Can't Sell"))

        else:
            Widget.SetVisibility(widget_visibility.Collapsed)

    OnUpgrade(Wdata:widget_message):void =
        if (CI := CurrentItem?):
            CI.OnUpgrade(Wdata)
    
    OnEquip(Wdata:widget_message):void =
        if (CI := CurrentItem?):
            CI.OnEquip(Wdata)
    OnSell(Wdata:widget_message):void =
        if (CI := CurrentItem?):
            CI.OnSell(Wdata)
    OnSellAll(Wdata:widget_message):void =
        if (CI := CurrentItem?):
            CI.OnSellAll(Wdata)

filter_setter := class:
    Filter : item_filter
    Menu : equipment_menu
    Position : vector2

    Set(Wdata:widget_message):void=
        set Menu.CurrentFilter = option{Filter}
        set Menu.FilterSelectorPos = Position
        Menu.Update()


item_filter := interface:
    Filter(Item : collectible_data, PlayerManager : player_manager)<decides><transacts>:void
    GetImage():texture

everything_filter := class(item_filter):
    Filter<override>(Item : collectible_data, PlayerManager : player_manager)<decides><transacts>:void=
        block:

    GetImage<override>():texture=
        return Images.Backpack

collectible_filter := class(item_filter):
    Filter<override>(Item : collectible_data, PlayerManager : player_manager)<decides><transacts>:void=
        if (Item.ArmorData?):
            false?

    GetImage<override>():texture=
        return Images.Chest

armor_filter := class(item_filter):
    Filter<override>(Item : collectible_data, PlayerManager : player_manager)<decides><transacts>:void=
        if (not Item.ArmorData?):
            false?

    GetImage<override>():texture=
        return Images.Sword

sellable_filter := class(item_filter):
    Filter<override>(Item : collectible_data, PlayerManager : player_manager)<decides><transacts>:void=
        SellValue := Item.GetSellValue()
        if ((not BigNumber_GreaterThanOrEquals(SellValue, big_number{Value:=1.0})?) or (PlayerManager.IsArmorEquipped(Item.ArmorData?.ToArmor())?)):
            false?

    GetImage<override>():texture=
        return Images.MidasCoins

consumable_filter := class(item_filter):
    Filter<override>(Item : collectible_data, PlayerManager : player_manager)<decides><transacts>:void=
        UseText := Item.GetUseText()
        if (UseText = ""): 
            false?

    GetImage<override>():texture=
        return Images.Potion

head_filter := class(item_filter):
    Filter<override>(Item : collectible_data, PlayerManager : player_manager)<decides><transacts>:void=
        if (not Item.ArmorData?.ToArmor().Slot = armor_slot.Head):
            false?

    GetImage<override>():texture=
        return Images.Helm

body_filter := class(item_filter):
    Filter<override>(Item : collectible_data, PlayerManager : player_manager)<decides><transacts>:void=
        if (not Item.ArmorData?.ToArmor().Slot = armor_slot.Torso):
            false?

    GetImage<override>():texture=
        return Images.Body

legs_filter := class(item_filter):
    Filter<override>(Item : collectible_data, PlayerManager : player_manager)<decides><transacts>:void=
        if (not Item.ArmorData?.ToArmor().Slot = armor_slot.Legs):
            false?

    GetImage<override>():texture=
        return Images.Legs

boots_filter := class(item_filter):
    Filter<override>(Item : collectible_data, PlayerManager : player_manager)<decides><transacts>:void=
        if (not Item.ArmorData?.ToArmor().Slot = armor_slot.Boots):
            false?

    GetImage<override>():texture=
        return Images.Boots

gloves_filter := class(item_filter):
    Filter<override>(Item : collectible_data, PlayerManager : player_manager)<decides><transacts>:void=
        if (not Item.ArmorData?.ToArmor().Slot = armor_slot.Gloves):
            false?

    GetImage<override>():texture=
        return Images.Glove

accessory_filter := class(item_filter):
    Filter<override>(Item : collectible_data, PlayerManager : player_manager)<decides><transacts>:void=
        if (not Item.ArmorData?.ToArmor().Slot = armor_slot.Accessory):
            false?

    GetImage<override>():texture=
        return Images.Ring

