using { /Fortnite.com/AI }
using { /Fortnite.com/Characters }
using { /Fortnite.com/Devices }
using { /Fortnite.com/Game }
using { /Verse.org }
using { /Verse.org/Random }
using { /Verse.org/Simulation }
using { /Verse.org/Assets }
using { /UnrealEngine.com/Temporary/Diagnostics }
using { /UnrealEngine.com/Temporary/SpatialMath }
using { /UnrealEngine.com/Assets }
using { /Fortnite.com/Animation/PlayAnimation }

floating_orb_manager := class(creative_device): 
    @editable
    FloatingOrbs : []creative_prop = array{}

    @editable
    Waypoints : []creative_prop = array{}

    var ZoneManager : ?zone_manager_device = false

    FloatOrbs()<suspends>:void=
        loop:
            if:
                SetZoneManger := ZoneManager?
                SetZoneManger.PlayersInZone > 0  
                FloatingOrbs.Length > 0
            then:
                var PointsToExclude : []int = array{}
                for (Orb : FloatingOrbs):
                    T := Orb.GetTransform()
                    R := T.Rotation
                    NewR := R.ApplyYaw(3.0)
                    if:
                        Result := GetRandomWayPoint[PointsToExclude]
                        P := Result(0).GetTransform().Translation
                        set PointsToExclude = PointsToExclude + array{Result(1)}
                    then:
                        spawn:
                            Orb.MoveTo(P, NewR, 15.0)
                Sleep(15.0)
                for (Orb : FloatingOrbs):
                    T := Orb.GetTransform()
                    R := T.Rotation
                    NewR := R.ApplyYaw(1.0)
                    P := Orb.GetTransform().Translation 
                    spawn:
                        Orb.MoveTo(P - vector3{Z:= 30.0}, NewR, 2.5)
                Sleep(2.5)    
                for (Orb : FloatingOrbs):
                    T := Orb.GetTransform()
                    R := T.Rotation
                    NewR := R.ApplyYaw(1.0)
                    P := Orb.GetTransform().Translation 
                    spawn:
                        Orb.MoveTo(P + vector3{Z:= 30.0}, NewR, 2.5)
            Sleep(2.5)

    GetRandomWayPoint(PointsToExclude : []int)<transacts><decides>:tuple(creative_prop, int)=
        var Num : int = GetRandomInt(0, Waypoints.Length - 1) 
        for(X:= 0..Waypoints.Length - 1):
            if:
                PointsToExclude.Find[Num]
                not(PointsToExclude.Find[X])
                set Num = X
        (Waypoints[Num], Num)