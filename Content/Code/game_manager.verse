using { /Fortnite.com/AI }
using { /Fortnite.com/Characters }
using { /Fortnite.com/Devices }
using { /Fortnite.com/Game }
using { /Fortnite.com/UI }
using { /UnrealEngine.com/Temporary/SpatialMath }
using { /Verse.org/Simulation }

game_manager := class(creative_device):
    @editable 
    AccoladesManager : accolades_manager = accolades_manager{}

    @editable 
    AudioManager : audio_manager = audio_manager{}

    @editable
    CinematicManager : cinematic_manager_device = cinematic_manager_device{}
    
    @editable
    ParticleManager : moving_particle_manager = moving_particle_manager{}
    
    @editable 
    GunManager : gun_manager = gun_manager{}

    @editable 
    DebugClearTracker : trigger_device = trigger_device{}
    
    @editable 
    ResetGameTrigger : trigger_device = trigger_device{}

    @editable 
    PlayerDataDebug : trigger_device = trigger_device{}
    
    @editable 
    CandyDevice : candy_menus_device = candy_menus_device{}
    
    @editable 
    DropManager : drop_manager_device = drop_manager_device{}

    @editable
    LootPopupDevice : loot_popup_device = loot_popup_device{}

    @editable
    LevelPopupDevice : level_popup_device = level_popup_device{}

    @editable
    UpgradePopupDevice : upgrade_popup_device = upgrade_popup_device{}
    
    @editable 
    QuestModel : quest_model_device = quest_model_device{}
    
    @editable 
    ItemGranter : item_granter_device = item_granter_device{}
    
    @editable 
    ItemGranterNoSwitch : item_granter_device = item_granter_device{}

    @editable
    PlayerLevelMaterialSetters: []cinematic_sequence_device = array{}

    @editable
    PlayerLevelVFXPowerps: []visual_effect_powerup_device = array{}
    
    @editable 
    ItemRemovers : []item_remover_device = array{}
    
    @editable 
    ItemCheckers : []conditional_button_device = array{}
    
    @editable 
    SwitchToTeam1 : class_and_team_selector_device = class_and_team_selector_device{}

    @editable 
    SwitchToTeam2 : class_and_team_selector_device = class_and_team_selector_device{}

    @editable 
    SwitchToClassNone : class_and_team_selector_device = class_and_team_selector_device{}

    @editable 
    SwitchToClass1 : class_and_team_selector_device = class_and_team_selector_device{}

    @editable 
    InnTeleporter : teleporter_device = teleporter_device{}

    @editable 
    PirateIslandTeleporter : teleporter_device = teleporter_device{}

    @editable 
    PrestigeTeleporter : teleporter_device = teleporter_device{}

    @editable
    MovementModulators : []movement_modulator_device = array{}

    @editable
    PeelyManagers : []peely_doll_device = array{}

    @editable
    ActorsToHide : []peely_doll_device = array{}

    @editable
    ConfettiPool : []vfx_spawner_device = array{}

    @editable
    EssenceVFXPool : []vfx_spawner_device = array{}

    @editable
    TeleporterGoodtownSkydive : teleporter_device = teleporter_device{}

    @editable
    HideAllUITrigger : trigger_device = trigger_device{}

    @editable
    ShowAllUITrigger : trigger_device = trigger_device{}

    @editable
    IncreaseStrength : trigger_device = trigger_device{}

    @editable
    DecreaseStrength : trigger_device = trigger_device{}

    @editable
    OrbitCamera : trigger_device = trigger_device{}
    
    @editable
    OrbitCameraDevice : gameplay_camera_orbit_device = gameplay_camera_orbit_device{}

    @editable
    WTFMutator : mutator_zone_device = mutator_zone_device{}

    @editable
    WTFMutator2 : mutator_zone_device = mutator_zone_device{}

    @editable
    Leaderboards : []leaderboard_device = array{}
    
    @editable
    DisableOnStartDamageVolumes : []damage_volume_device = array{}

    @editable
    AnalyticsManager : analytics_manager = analytics_manager{}

    @editable
    BossInfo : boss_info_device = boss_info_device{}

    @editable
    BossTeleportManager : boss_teleport_manager_device = boss_teleport_manager_device{}

    @editable
    TeleportManager : teleport_manager_device = teleport_manager_device{}
    
    PlayerDamageManager : player_damage_manager = player_damage_manager{}

    CurrentVersion : float = 1.4
    LevelReqScaleRatio : float = 1.3
    GuardRespawnTime : float = 1.0
    ZoneLevelCap : int = 5
    PassiveIncomeTimer : int = 60

    var ConfettiIndex : int = 0

    var EssenceVFXIndex : int = 0
    var EssenceVFXAssignment : []guard_manager = array{}

    var QueuedPeelyDolls : []queued_peely_doll = array{}
    
    var UpgradeValues : [string][]float = map{}
    var UpgradeLevelCosts : [string][]int = map{}

    var UpgradeDatas : [string]upgrade_item = map{}

    var BossAgents : []agent = array{}

    #Persistable Playerdata
    var InitialSavedPlayerData : player_profile_data = player_profile_data{}

    #Player Functions That Manipulate Persistable PlayerData
    var PlayerManagers : [player]player_manager = map{}

    #ZoneManagers which host many GuardManagers for each Zone (1 Zone Manager per Zone)
    var ZoneManagers : []zone_manager_device = array{}

    OnBegin<override>()<suspends>:void=
        Print("GAME START PHASE 1")
        InitSpawners()
        Print("GAME START PHASE 2")
        InitZones()
        Print("GAME START PHASE 3")
        SetupDataLayers()
        Print("GAME START PHASE 4")
        Sleep(0.0)
        Print("GAME START PHASE 5")
        GetPlayspace().PlayerRemovedEvent().Subscribe(OnPlayerLeftServer)
        Print("GAME START PHASE 6")
        DebugClearTracker.TriggeredEvent.Subscribe(OnDebugClearTriggered)
        ResetGameTrigger.TriggeredEvent.Subscribe(OnResetGame)
        Print("GAME START PHASE 7")
        PlayerDataDebug.TriggeredEvent.Subscribe(OnPlayerDataDebug)
        Print("GAME START PHASE 8")
        # ShowAllUITrigger.TriggeredEvent.Subscribe(OnShowAllUI)
        # HideAllUITrigger.TriggeredEvent.Subscribe(OnHideAllUI)
        # IncreaseStrength.TriggeredEvent.Subscribe(OnIncreaseStrength)
        # DecreaseStrength.TriggeredEvent.Subscribe(OnDecreaseStrength)
        # OrbitCamera.TriggeredEvent.Subscribe(OnOrbitCamera)
        Print("GAME START PHASE 9")
        OnBeginIntializePersistableData()
        Print("GAME START PHASE 10")
        InitLeaderboards()
        Print("GAME START PHASE 11")
        TeleporterGoodtownSkydive.TeleportedEvent.Subscribe(OnGoodtownSkydive)
        Print("GAME START PHASE 12")
        for (Volume : DisableOnStartDamageVolumes):
            Volume.Disable()
        Print("GAME START PHASE 13")
        
    OnGoodtownSkydive(Agent:agent):void=
        spawn { DoStasis(Agent) }
        if:
            Player := player[Agent]
            Player.IsActive[]
            PlayerManager := PlayerManagers[Player]
        then:
            PlayerManager.OnEscapedMidasCastle()
            
    DoStasis(Agent:agent)<suspends>:void=
        if:
            Player := player[Agent]
            Player.IsActive[]
            PlayerManager := PlayerManagers[Player]
        then:
            CinematicManager.PlayBeginningCinematic(Agent, PlayerManager)

    InitLeaderboards():void=
        for(Board : Leaderboards):
            set Board.GameManager = option{Self}
            Board.Init()

    AssignEssenceVFX(GuardManager : guard_manager):void=
        if:
            EssenceVFX := EssenceVFXPool[EssenceVFXIndex]
            set EssenceVFXIndex += 1
            if:
                EssenceVFXIndex > EssenceVFXPool.Length - 1
            then:
                set EssenceVFXIndex = 0
        then:
            #Unassign
            if:
                EssenceVFXAssignment.Length < EssenceVFXPool.Length
            then:
                set EssenceVFXAssignment = EssenceVFXAssignment + array{GuardManager}
            else:
                if:
                    set EssenceVFXAssignment[EssenceVFXIndex].EssenceVFX = false
                    NewEssenceVFXAssignment := EssenceVFXAssignment.ReplaceElement[EssenceVFXIndex, GuardManager]
                    set EssenceVFXAssignment = NewEssenceVFXAssignment

            #Reassign
            set GuardManager.EssenceVFX = option{EssenceVFX}

                

    QueueConfetti(Position : vector3):void=
        if:
            Confetti:= ConfettiPool[ConfettiIndex]
            set ConfettiIndex += 1
            if:
                ConfettiIndex > ConfettiPool.Length - 1
            then:
                set ConfettiIndex = 0
            Confetti.TeleportTo[Position, Confetti.GetTransform().Rotation]
        then:
            Confetti.Restart()

    QueuePeely(QueuedPeely : queued_peely_doll, ?SetState : logic = false):void=
        if:
            QueuedPeelyDolls.Length = 0
        then:
            spawn:
                StartPeely()        
        
        set QueuedPeelyDolls = QueuedPeelyDolls + array{QueuedPeely}

        if:
            SetState?
            Agent := QueuedPeely.Agent?
        then:
            QueuedPeely.Peely.Peely.SetState(Agent, QueuedPeely.NewState)

    StartPeely()<suspends>:void=
        loop:
            Sleep(0.01)
            if:
                QueuedPeely:= QueuedPeelyDolls[0]
                HideSequencer := QueuedPeely.Peely.HideSequencer
                ShowSequencer := QueuedPeely.Peely.ShowSequencer
                Agent := QueuedPeely.Agent?
                NewState := QueuedPeely.NewState
            then:
                ActivatePeely(Agent, NewState, HideSequencer, ShowSequencer)
                
                if:
                    Temp1 := QueuedPeelyDolls.RemoveElement[0]
                    set QueuedPeelyDolls = Temp1
                
                if:
                    QueuedPeelyDolls.Length = 0
                then:
                    return
            
    ActivatePeely(Agent : agent, NewState : logic, HideSequencer : cinematic_sequence_device, ShowSequencer : cinematic_sequence_device)<suspends>:void=
            if:
                NewState?
            then:
                Sleep(0.1)
                ShowSequencer.Play(Agent)
                Sleep(0.1)
                ShowSequencer.Stop(Agent)
                Sleep(0.1)
                ShowSequencer.Play(Agent)
                Sleep(0.1)
                ShowSequencer.Stop(Agent)
            else:
                Sleep(0.1)
                HideSequencer.Play(Agent)
                Sleep(0.1)
                HideSequencer.Stop(Agent)
                Sleep(0.1)
                HideSequencer.Play(Agent)
                Sleep(0.1)
                HideSequencer.Stop(Agent)

    SetupDataLayers():void=
        CinematicManager.PlayDataLoaderCinematic()

    RemovePlayerFromAllBeacons(Agent:agent):void=
        Beacons := FindCreativeObjectsWithTag(beacon{})
        for (Obj : Beacons):
            if (Beacon := beacon_device[Obj]):
                Beacon.RemoveFromShowList(Agent)

    GetZoneManagerFromZoneTag(ZoneTag : fetchable_tag)<transacts><decides>:?zone_manager_device=
        var TempZoneManager : ?zone_manager_device = false
        for (ZoneManager : ZoneManagers):
            if:
                ZoneManager.ZoneTag = ZoneTag
            then:
                set TempZoneManager = option{ZoneManager}
        return TempZoneManager

    InitSpawners():void=
        Spawners := FindCreativeObjectsWithTag(spawner{})
        for (Obj : Spawners):
            if (Spawner := player_spawner_device[Obj]):
                Spawner.SpawnedEvent.Subscribe(OnPlayerSpawned)

        WTFMutator.AgentEntersEvent.Subscribe(OnBadPlayerBadSpawned)
        WTFMutator2.AgentEntersEvent.Subscribe(OnPlayerSpawned)
        
        BossSpawners := FindCreativeObjectsWithTag(boss_spawner{})
        for (Obj : BossSpawners):
            if (Spawner := npc_spawner_device[Obj]):
                Spawner.SpawnedEvent.Subscribe(OnBossSpawned)
                #Spawner.Spawn()
        
        GuardSpawners:= FindCreativeObjectsWithTag(guard_spawner{})
        for (Obj : GuardSpawners):
            if (Spawner := guard_spawner_device[Obj]):
                Spawner.SpawnedEvent.Subscribe(OnGuardSpawned)

    InitZones():void=
        Zones := FindCreativeObjectsWithTag(zone{})
        for (Obj : Zones):
            if:
                Zone : []zone_manager_device = array{zone_manager_device[Obj]}
                set ZoneManagers = ZoneManagers + Zone
        for (ZoneManager : ZoneManagers):
            ZoneManager.Init(Self)

    OnBeginIntializePersistableData():void=
        set InitialSavedPlayerData = player_profile_data:
            CreationTime := GetSecondsSinceEpoch()
            Version := CurrentVersion
            LastTimeOnline := GetSecondsSinceEpoch()

        Players := GetPlayspace().GetPlayers()
        for (Player : Players):
            if:
                Player.IsActive[]
                not PlayerProfileDataMap[Player]
                set PlayerProfileDataMap[Player] = InitialSavedPlayerData    

    CheckSaveDataForPlayer(Agent:agent)<decides><transacts>:void=
        if:
            set InitialSavedPlayerData = player_profile_data:
                CreationTime := GetSecondsSinceEpoch()
                Version := CurrentVersion
                LastTimeOnline := GetSecondsSinceEpoch()
            true?
        then:
            if:
                Player := player[Agent]
                Player.IsActive[]
            then:
                if(PlayerProfileDataMap[Player]):
                    # no-op
                else:
                    # Log("No Data Found, Setting New Player Data! (CheckSaveData)")
                    set PlayerProfileDataMap[Player] = InitialSavedPlayerData

    OnPlayerAdded(Player:player):void=
        set InitialSavedPlayerData = player_profile_data:
            CreationTime := GetSecondsSinceEpoch()
            Version := CurrentVersion
            LastTimeOnline := GetSecondsSinceEpoch()
            
        if:
            not PlayerProfileDataMap[Player]
            # Log("No Data Found, Setting New Player Data! (OnPlayerAdded)")
            set PlayerProfileDataMap[Player] = InitialSavedPlayerData

    OnPlayerSpawned(Agent:agent):void=
        spawn:
            SetupPlayer(Agent, false)

    OnBadPlayerBadSpawned(Agent:agent):void=
        spawn:
            SetupPlayer(Agent, true)
    
    SetupPlayer(Agent : agent, NeedsRespawn : logic)<suspends>:void=
        Sleep(0.01)
        SwitchToTeam1.ChangeTeam(Agent)
        if:
            Player:=player[Agent]
            Player.IsActive[]
            FortChar := Agent.GetFortCharacter[]
            not PlayerManagers[Player]
        then:
            RemovePlayerFromAllBeacons(Agent)
            FortChar.EliminatedEvent().Subscribe(OnPlayerEliminated)
            OnPlayerAdded(Player)
            InitPlayer(Player)

            if:
                NeedsRespawn?
                FortChar.TeleportTo[vector3{X:=-84667.427559,Y:=-50962.648180,Z:=24251.062454}, FortChar.GetTransform().Rotation]
                PlayerManager:=PlayerManagers[Player]
            then:
                spawn:
                    PlayerManager.SkipPrison()

    #NEED TO FIX (Don't hard-code)
    OnBossSpawned(Agent:agent):void=
        Print("OnBossSpawned")
        if:
            Boss:= Agent            
            Generic:=Boss.GetNPCBehavior[]
        then:
            set BossAgents = BossAgents + array{Boss}
            if:
                BossBehavior := boss_midas_behavior[Generic]
                ZoneManager := GetZoneManagerFromZoneTag[BossBehavior.ZoneTag]
                set BossBehavior.GameManager = option{Self}
                set BossBehavior.ZoneManager = ZoneManager
                set BossBehavior.Mutator = option{BossBehavior.ZoneManager?.ZoneMutuator}
                ZM := BossBehavior.ZoneManager?
            then:
                for (Volume : BossBehavior.DamageVolumes):
                    Volume.Disable()
                Print("Boss Behavior Set : {BossBehavior.ZoneTag}, {ZM.ZoneTag}")
                spawn {BossBehavior.ActualOnBegin()}

    OnPlayerEliminated(Elimination: elimination_result):void=
        if:
            Character := Elimination.EliminatedCharacter
            Character.IsActive[]
            Agent := Character.GetAgent[]
        then: 
            SwitchToTeam1.ChangeTeam(Agent)

            if:
                Player:=player[Agent]
                PlayerManager:=PlayerManagers[Player]
                PlayerManager.GetAchievementLevel["EscapedMidasCastle"] > 0
            then:
                SwitchToClass1.ChangeClass(Agent)            

    OnGuardSpawned(Agent:agent):void=
        if:
            Guard:= Agent
            Character := Guard.GetFortCharacter[]
            Character.IsActive[]
        then:
            Character.DamagedEvent().Subscribe(OnGuardDamaged)

    OnGuardDamaged(Damage : damage_result):void=
        block:    

    InitPlayer(Player:player):void=
        if:
            Agent:=agent[Player]
            Fort:= Agent.GetFortCharacter[]
        then:
            if(Existing:= PlayerManagers[Player]):
                PlayerManager := player_manager{Agent:=Agent, Player:=Player, GameManager:=Self}
                set PlayerManager.CinematicID = PlayerManagers.Length
                if(set PlayerManagers[Player] = PlayerManager){}
                PlayerManager.Init()
            else:
                PlayerManager := player_manager{Agent:=Agent, Player:=Player, GameManager:=Self}
                set PlayerManager.CinematicID = PlayerManagers.Length
                if(set PlayerManagers[Player] = PlayerManager){}
                PlayerManager.Init()

    OnPlayerLeftServer(LeavingPlayer:player):void=        
        RemovePlayerFromQuestMobs(LeavingPlayer)
        # OnPlayerLeaveCheckBossSpawners()

        var NewPlayerManagersMap:[player]player_manager = map{}
        for (Key -> Value : PlayerManagers, Key <> LeavingPlayer):
            set NewPlayerManagersMap = ConcatenateMaps(NewPlayerManagersMap, map{Key => Value})
        set PlayerManagers = NewPlayerManagersMap
    
        var count : int = 0
        for (PlayerManager : PlayerManagers):
            set PlayerManager.CinematicID = count
            set count += 1

    ResetPlayerDamageTrackers(Player:player):void=
        for (Zone : ZoneManagers):
            for (Guard : Zone.GuardManagers):
                if:
                    Guard.PlayersHittingUsNow[Player] = big_number{}
                    Print("Reset ZM Guards")

        for (Quest : QuestModel.Quests):
            for (Task : Quest.Tasks):
                if:
                    Guard := Task.SpawnGuardOnStart?
                    Guard.GuardManager.PlayersHittingUsNow[Player] = big_number{}
                    Print("Reset Quest Guards")
                then:
                    Guard.GuardManager.RemoveQuestOwner(Player)

    RemovePlayerFromQuestMobs(LeavingPlayer:player):void=
        for (Quest : QuestModel.Quests):
            for (Task : Quest.Tasks):
                if:
                    GuardManager := Task.SpawnGuardOnStart?
                then:
                    GuardManager.GuardManager.RemoveQuestOwner(LeavingPlayer)

    CheckBarrierLimit(Agent:agent, PlayerManager:player_manager):void=
        for (ZoneManager : ZoneManagers):
            ZoneManager.CheckBarrierLimit(Agent,PlayerManager)
    
    OnIncreaseStrength(Agent:?agent):void=
        if:
            Player:=player[Agent?]
            Player.IsActive[]
            PlayerManager:=PlayerManagers[Player]
        then: 
            PlayerManager.IncreaseStrength()
    
    OnDecreaseStrength(Agent:?agent):void=
        if:
            Player:=player[Agent?]
            Player.IsActive[]
            PlayerManager:=PlayerManagers[Player]
        then: 
            PlayerManager.DecreaseStrength()
    
    OnOrbitCamera(Agent:?agent):void=
        if:
            SetAgent := Agent?
        then:
            OrbitCameraDevice.AddTo(SetAgent)

    OnShowAllUI(Agent:?agent):void=
        if:
            Player:=player[Agent?]
            Player.IsActive[]
            PlayerManager:=PlayerManagers[Player]
        then: 
            spawn:
                PlayerManager.RespawnFixUI()
            
        # if:
        #     Quest:= PlayerManager.GetCurrentQuests[][0]
        #     SetQuestManager:= PlayerManager.QuestManager?
        #     SetQuestManager.BeginStartingQuests()
        #     SetQuestManager.CompleteTask(Quest) #Skip Cell Wall
    
    

        Playspace := Self.GetPlayspace()
        HUDController := Playspace.GetHUDController()
 
        HUDElementsToHide := array:
            creative_hud_identifier_health_numbers{}
            creative_hud_identifier_mimimap{}
            creative_hud_identifier_minimap{}
            creative_hud_identifier_map_prompts{}
            hud_identifier_world_resource_wood{}
            hud_identifier_world_resource_stone{}
            hud_identifier_world_resource_metal{}
            hud_identifier_world_resource_permanite{}
            hud_identifier_world_resource_gold_currency{}
            hud_identifier_world_resource_ingredient{}
            creative_hud_identifier_team_info{}
 
        HUDElementsToShow := array:
            creative_hud_identifier_player_inventory{}
            creative_hud_identifier_interaction_prompts{}
            creative_hud_identifier_experience_level{}
            creative_hud_identifier_experience_supercharged{}
            creative_hud_identifier_experience_ui{}
            creative_hud_identifier_health{}
            creative_hud_identifier_hud_info{}


        HUDController.HideElements(HUDElementsToHide)
        HUDController.ShowElements(HUDElementsToShow)

    OnHideAllUI(Agent:?agent):void=
        if:
            Player:=player[Agent?]
            Player.IsActive[]
            PlayerManager:=PlayerManagers[Player]
        then: 
            spawn:
                PlayerManager.HideAllGameUI()

        Playspace := Self.GetPlayspace()
        HUDController := Playspace.GetHUDController()
    
        HUDElementsToHide := array:
            creative_hud_identifier_health_numbers{}
            creative_hud_identifier_mimimap{}
            creative_hud_identifier_minimap{}
            player_hud_identifier_all{}
            creative_hud_identifier_map_prompts{}
            hud_identifier_world_resource_wood{}
            creative_hud_identifier_shield_numbers{}
            creative_hud_identifier_shileds{}
            creative_hud_identifier_shields{}
            creative_hud_identifier_equipped_item{}
            creative_hud_identifier_player_inventory{}
            hud_identifier_world_resource_stone{}
            hud_identifier_world_resource_metal{}
            hud_identifier_world_resource_permanite{}
            hud_identifier_world_resource_gold_currency{}
            hud_identifier_world_resource_ingredient{}
            creative_hud_identifier_team_info{}
            creative_hud_identifier_player_inventory{}
            creative_hud_identifier_interaction_prompts{}
            # creative_hud_identifier_map_prompts{}
            creative_hud_identifier_experience_level{}
            creative_hud_identifier_experience_supercharged{}
            creative_hud_identifier_experience_ui{}
            creative_hud_identifier_health{}
            #creative_hud_identifier_health_numbers{}
            creative_hud_identifier_hud_info{}

        HUDController.HideElements(HUDElementsToHide)

    OnResetGame(Agent:?agent):void=
        if:
            Player:=player[Agent?]
            Player.IsActive[]
            PlayerManager:=PlayerManagers[Player]
        then: 
            PlayerManager.ResetGame()

    OnDebugClearTriggered(Agent:?agent):void=
        if:
            Player:=player[Agent?]
            Player.IsActive[]
            PlayerManager:=PlayerManagers[Player]
        then: 
            PlayerManager.DebugResetPlayerData()

    OnPlayerDataDebug(Agent:?agent):void=
        if:
            Player:=player[Agent?]
            Player.IsActive[]
            PlayerManager:=PlayerManagers[Player]
        then: 
            PlayerManager.DebugPlayerData()
    
    GetZoneLevelCap(ZoneTag:fetchable_tag):int=
        if (ZoneTag.IsBossTag()?):
            return 99999
        else:
            return ZoneLevelCap + ZoneTag.ToZoneValue()
            
    GetTeamNumber(TeamToCheck:team)<decides><transacts>:int=
        var TeamIndex:?int = false

        for (Index -> Team : GetPlayspace().GetTeamCollection().GetTeams()):
            if (Team = TeamToCheck):
                set TeamIndex = option{ Index }

        TeamIndex?

    GetBossHealth(BossIndex:int, BossLevel:int)<decides><transacts>:big_number=
        case (BossIndex):
            1 => big_number{Value:=BossLevel * 5.0, Exponent:=2 + BossLevel}
            _ => big_number{Value:=BossLevel * 3.0, Exponent:=6}
