using { /Fortnite.com/AI }
using { /Fortnite.com/Devices }
using { /Verse.org/Simulation }
using { /UnrealEngine.com/Temporary/SpatialMath }
using { /Fortnite.com/Characters }
using { /Fortnite.com/Game }
using { /Verse.org/Random }
using { /UnrealEngine.com/Assets }

# See https://dev.epicgames.com/documentation/en-us/uefn/create-your-own-device-in-verse for how to create a verse device.

guard_manager := class<concrete><public>:
    @editable 
    Manipulator : prop_manipulator_device = prop_manipulator_device{}
    
    @editable
    GuardProp : creative_prop = creative_prop{}

    @editable   
    SpawnPool : ?spawn_pooler = false
    
    @editable
    SpawnPosition : ?creative_device = false

    @editable
    SpawnTarget : ?creative_device = false
    
    @editable
    CustomName : ?string = false

    @editable
    CustomMaxHealth : ?big_number = false

    @editable
    IsQuestEnemy : logic = false

    @editable
    AlwaysWeakspot : logic = false

    @editable 
    Classes : []string = array{}

    @editable
    RespawnTimerMax : float = 3.0

    @editable
    Weakspot : trigger_device = trigger_device{}

    @editable
    GuardMultiplier : float = 1.0

    @editable
    DefaultWeakness : float = 1.0

    @editable
    GunWeaknesses : []gun_weakness = array{}

    @editable
    HandcannonMultiplier : float = 1.0

    @editable
    Unkillable : logic = false

    @editable
    IgnoreCandyDrops : logic = false

    @editable
    IgnoreXPCoinDrop : logic = false

    @editable
    DropsShared : []string = array{}

    @editable
    DropsCustom : []drop_table = array{}

    @editable
    OwningSpawner : ?player_spawner_device = false

    @editable
    VFXLocation : vector3 = vector3{}

    @editable
    RequiresSuperHandcannon : logic = false

    @editable
    ArmorValue : big_number = big_number{}
    
    @editable
    HealthIsPerPlayer : logic = false

    @editable
    WeaponExperienceData : ?weapon_experience_data = false

    @editable
    BigRockData : ?big_rock_data = false

    var WeaknessesMap : [int]float = map{}
    var AllowedPlayers : []agent = array{}
    
    var EssenceVFX : ?vfx_spawner_device = false

    var GameManager : game_manager = game_manager{}
    var ZoneManager : zone_manager_device = zone_manager_device{}
    var MaxHealth : big_number = big_number{Value:=1.0, Exponent:=3}
    var CurrentHealth : big_number = big_number{Value:=1.0, Exponent:=3}
    
    var GuardID : string = ""
    var GuardName : string = ""
    var RespawnTimer : float = 0.0
    var StartPosition : vector3 = vector3{}
    var StartRotation : rotation = rotation{}
    var StartScale : vector3 = vector3{}
    var DeathTransform : transform = transform{}

    var PlayersHittingUsNow : [player]big_number = map{}

    var PlayerInResetRange : logic = false

    var PendingEndAgentThreads : []agent = array{}

    var PlayersInZoneMultiplier : float = 1.0

    var EntityAgent : ?agent = false
    
    var PlayerQuestOwners : []player = array{}

    var PlayerScaling : float = 0.0

    var RespawnID : int = 0

    var IsActive : logic = true
    #Unused
    OnBegin()<suspends>:void =
        if (Spawner := OwningSpawner?):
            Spawner.SpawnedEvent.Subscribe(OnOwningPlayerSpawned)

    OnOwningPlayerSpawned(Player:agent):void=
        set AllowedPlayers = AllowedPlayers + array{Player}

    # Runs when the device is started in a running game
    Init(GM : game_manager, ZM: zone_manager_device, ZoneTag : fetchable_tag, GuardIDIn : string):void=
        spawn {ActualInit(GM, ZM, ZoneTag, GuardIDIn)}

    SetMaxHealth():big_number=
        var Amt : big_number = big_number{}
        # NewPlayerScaling : float = Max(1.0 + (ZoneManager.PlayersInZone - 1) * ZoneManager.ZoneTag.ToPlayerScaling() , 1.0)
        # OldPlayerScaling : float = PlayerScaling
        #Scale Up, Sometimes
        if: 
            true?
            # NewPlayerScaling >= PlayerScaling
            # CurrentHealth.Value <= 0.0 or BigNumber_Equals(CurrentHealth, MaxHealth)?
        then:
            # set PlayerScaling = NewPlayerScaling 
            
            set Amt = ZoneManager.ZoneTag.ToGuardHealth() * GuardMultiplier #* Max(PlayerScaling * 1.0,1.0)
            if:
                CurrentHealth.Value > 0.0
            then:
                # Don't set current health if its at 0.0 because then you're reviving a guard
                set CurrentHealth = Amt

            set MaxHealth = Amt

        # #Scale Down, Always
        # if: 
        #     NewPlayerScaling < PlayerScaling
        # then:
        #     set PlayerScaling = NewPlayerScaling
        
        #     set Amt = ZoneManager.ZoneTag.ToGuardHealth() * GuardMultiplier * Max(PlayerScaling * 1.0,1.0)

        #     OldPercent := BigNumber_Percentage(CurrentHealth, MaxHealth)

        #     if:
        #         CurrentHealth.Value > 0.0
        #     then:
        #         # Don't set current health if its at 0.0 because then you're reviving a guard
        #         set CurrentHealth = Amt * OldPercent

        #     set MaxHealth = Amt

        #Custom Max Health Overrides these
        if:
            SetCustomMaxHealth := CustomMaxHealth?
        then:
            set Amt = SetCustomMaxHealth
            set MaxHealth = Amt

            if:
                CurrentHealth.Value > 0.0
            then:
                # Don't set current health if its at 0.0 because then you're reviving a guard
                set CurrentHealth = Amt

        for (Player : GameManager.PlayerManagers):
            if: 
                SetUI := Player.UIManager?
            then:
                SetUI.UpdateGuardPlayerCount(GuardID, PlayerScaling, IsQuestEnemy)

        return Amt

    ActualInit(GM : game_manager, ZM: zone_manager_device, ZoneTag : fetchable_tag, GuardIDIn : string)<suspends>:void=
        Sleep(5.0)
        set GuardID = GuardIDIn
        set GameManager = GM
        set ZoneManager = ZM
        SetMaxHealth()

        set WeaknessesMap = map{}

        for (i := 0..99):
            if:
                set WeaknessesMap[i] = DefaultWeakness
        
        for (Weakness : GunWeaknesses):
            if:
                set WeaknessesMap[Weakness.ID] = Weakness.Weakness

        if:
            SetCustomName := CustomName?
        then:
            set GuardName = SetCustomName
        else:
            set GuardName = ZoneTag.ToGuardName()
        set StartPosition = GuardProp.GetTransform().Translation
        set StartRotation = GuardProp.GetTransform().Rotation
        set StartScale =  GuardProp.GetTransform().Scale
        set DeathTransform = transform{Translation:= {StartPosition}, Scale:={vector3{X:=0.001, Y:=0.001, Z:=0.001}}, Rotation:=StartRotation}
        
        if (SpawnPool?):
            spawn{TrackAgent()}

        else:
            Manipulator.DamagedEvent.Subscribe(OnGuardDamaged)
            Weakspot.TriggeredEvent.Subscribe(OnWeakspotDamaged)
    
    GetClosestPlayerDist()<transacts>: float=
        var PlayerDist : float = 9999999.0
        var ClosestPlayer : ?player = false
        var Players : [player]player_manager = GameManager.PlayerManagers

        if:
            IsQuestEnemy?
        then:
            var NewMap : [player]player_manager = map{}
            for (QuestOwner : PlayerQuestOwners):
                if:
                    Temp := GameManager.PlayerManagers[QuestOwner]
                    set NewMap[QuestOwner] = Temp
            set Players = NewMap

        for (Player : Players):
            if:
                T := SpawnPosition?.GetTransform()
                NewPlayerDist := Min(PlayerDist, Distance(Player.Player.GetFortCharacter[].GetTransform().Translation, T.Translation))
                NewPlayerDist < PlayerDist
            then:
                set PlayerDist = NewPlayerDist
                set ClosestPlayer = option{Player.Player}
        # if:
        #     GuardName = "Deranged Boar"
        # then:
        #     Log("PlayerDist: {PlayerDist}")
        return PlayerDist

    PrepareDoAgentSpawn()<suspends>:void=
        if(GetClosestPlayerDist() > ZoneManager.ZoneTag.GuardSpawnDistance()):
            if :
                Agent := EntityAgent?
                FortChar := Agent.GetFortCharacter[]
            then:
                spawn {DespawnChar(FortChar)}
        else:
            if:
                IsActive?
            then:
                if:
                    not(EntityAgent?)
                then:
                    RequestAgentSpawn()

    SetLeash(FortChar:fort_character):void=
        if:
            Leashable := FortChar.GetFortLeashable[]
            T := SpawnPosition?.GetTransform()
        then:
            if(IsQuestEnemy?):
                Leashable.SetLeashPosition(T.Translation, 0.0, 100.0)
            else:
                Leashable.SetLeashPosition(T.Translation, 0.0, 0.1)


    TrackAgent()<suspends>:void=
        loop:
            Sleep(1.0)
            if:
                GuardAgent := EntityAgent?
                FortChar := GuardAgent.GetFortCharacter[]
            then:
                FortChar.Show()
                SetLeash(FortChar)
                
            if (CurrentHealth.Value > 0.0):
                if:
                    not(IsQuestEnemy?)
                then:
                    PrepareDoAgentSpawn()
                else:
                    if:
                        IsQuestEnemy?
                        PlayerQuestOwners.Length > 0
                    then:
                        PrepareDoAgentSpawn()  
            else:
                if:
                    IsQuestEnemy? 
                    PlayerQuestOwners.Length > 0
                    CurrentHealth.Value <= 0.0                    
                then:
                    set CurrentHealth = MaxHealth
                else:
                    if(RespawnTimerMax > 0.0):
                        set RespawnTimer -= 1.0 
                        if(RespawnTimer <= 0.0):
                            SetMaxHealth()
                            if:
                                not(IsQuestEnemy?)
                            then:
                                set CurrentHealth = MaxHealth
                                set PlayersHittingUsNow = map{}          
              
    # No way to despawn a character, so we make it vulnerable, and do damage so it's dead
    # The damage amount passed in doesn't actually do anything, it's overwritten by the spawner's outgoing damage vaue (engine bug)
    # So we have to make sure the character's max health is lower than its outgoing damage value
    DespawnChar(Char:fort_character)<suspends>:void=
        Char.SetVulnerability(true)
        Char.SetMaxHealth(1.0)
        Char.SetHealth(1.0)
        Char.Damage(9999999999.0)
        Char.SetVulnerability(false)
        set EntityAgent = false
    
    NotifyBlockedAttack(PlayerManager : player_manager):void=  
        if:
            UI := PlayerManager.UIManager?
            UI.ZoneMsgVisibilityTimer <= 0.0
        then:
            AdjustedArmorValue := ArmorValue * (1.0 - PlayerManager.GetUpgradeValueSafe("P_ArmorPiercing", false))
            UI.SetZoneOpenText("You need to do {AdjustedArmorValue} damage to hurt this.")
            set UI.ZoneMsgVisibilityTimer = 3.0

    CheckBigRockData(PlayerManager : player_manager):void=  
        if:
            SetBigRockData:=BigRockData?
            UI := PlayerManager.UIManager?
            Level := PlayerManager.GetBigRockLevel()
        then:
            if:
                SetBigRockData.ID > Level
                PlayerManager.SetBigRockLevel[SetBigRockData.ID]
            then:
                PlayerManager.CheckBigRockAnalytics(SetBigRockData.ID)
                UI.SetZoneOpenText("Big Rock Level {SetBigRockData.ID + 1} Destroyed!<br>+ {(SetBigRockData.ID+1) * 15}% damage to Bosses!")
                set UI.ZoneMsgVisibilityTimer = 5.0
                GameManager.AudioManager.UnlockZonePlayer.Play(PlayerManager.Agent)
            else:
                if:
                    UI.ZoneMsgVisibilityTimer <= 0.0
                then:
                    UI.SetZoneOpenText("Destroy Big Rock Level {Level + 2}<br>To Gain Boss Damage!")
                    set UI.ZoneMsgVisibilityTimer = 5.0

    var IsShowingWeaponExperience : logic = false
    CheckWeaponExperienceData(PlayerManager : player_manager, ?DisplayNotice : logic = true):void=
        if:
            WepExpData:=WeaponExperienceData?
            PlayerManager.GetCooldown(WepExpData.Name) <= 0.0
        then:
            PlayerManager.SetCooldown(WepExpData.Name, WepExpData.CooldownSeconds * 1.0)
            #Pick a Gun ID
            var GunsToShow : []int = array{}
            for (i := 0..99):
                if:
                    i = 1 or PlayerManager.GetUpgradeLevel["Gun_{i}"] > 0
                then:
                    set GunsToShow += array{i}
            
            LevelupGun:= GetRandomInt(0, GunsToShow.Length-1)

            if:
                ID:=GunsToShow[LevelupGun]
                OldLevel:= PlayerManager.GetWeaponLevel(ID)
                PlayerManager.GainWeaponExperience[ID, WepExpData.GainAmount] #Updated
                NewLevel:= PlayerManager.GetWeaponLevel(ID)
                NewExp:= PlayerManager.GetWeaponExperience(ID)
                UI:= PlayerManager.UIManager?
            then:
                Req:=PlayerManager.GetWeaponLevelRequirement(NewLevel)
                spawn:
                    UI.UpdateGunExp(ID, NewLevel, NewExp, Req, OldLevel, WepExpData.GainAmount)
                PlayerManager.UpdateWeaponExperienceUI()
        else if (DisplayNotice?):
            if:
                WepExpData:=WeaponExperienceData?
                UI := PlayerManager.UIManager?
                CDSeconds := Int[PlayerManager.GetCooldown(WepExpData.Name)]
            then:
                UI.SetZoneOpenText("You've already collected this recently. Come back in {FormatTime(CDSeconds)}.")
                set UI.ZoneMsgVisibilityTimer = 5.0

    OnFortCharDamaged(Event : damage_result):void=
        if:
            Player:=player[Event.Instigator?.GetInstigatorAgent[]]
            Player.IsActive[]
            PlayerManager := GameManager.PlayerManagers[Player]
            GunStats := PlayerManager.GetCurrentGunStats[]
            DamagedChar := fort_character[Event.Target]
            PlayerChar := Player.GetFortCharacter[]
            Damaged := DamagedChar.GetAgent[]  
        then:
            Dist := Distance(PlayerChar.GetTransform().Translation, DamagedChar.GetTransform().Translation)
            FalloffMult := GunStats.GetFallofMult(Dist)
            # Log("OnFortCharDamaged INTERNAL - EVENT AMOUNT : {Event.Amount}")
            
            if:
                GunStats.FortBaseDamage > 0.0
                Event.Amount > GunStats.FortBaseDamage * 1.3
                PlayerWeakspotMult := PlayerManager.GetPlayerStat(player_stat.Crit_Mult)
            then:
                ProcessDamage(PlayerManager, BigNumber_Multiply(GunStats.WeakspotMult, PlayerWeakspotMult), true, FalloffMult, ?DamagedAgent := option{Damaged}) 
                PlayerManager.OnWeakspotHit()
            else:
                ProcessDamage(PlayerManager, big_number{Value:=1.0}, false, FalloffMult, ?DamagedAgent := option{Damaged}) 
    

    OnWeakspotDamaged(Agent:?agent):void=
        if:
            SetAgent:=Agent?
            Player:=player[SetAgent]
            Player.IsActive[]
            PlayerManager := GameManager.PlayerManagers[Player]
            WeakSpotSFX := PlayerManager.GetWeakSpotSFX()
            GunStats := PlayerManager.GetCurrentGunStats[]
            PlayerChar := Player.GetFortCharacter[]
            PlayerWeakspotMult := PlayerManager.GetPlayerStat(player_stat.Crit_Mult)
        then:
            Dist := Distance(PlayerChar.GetTransform().Translation, GuardProp.GetTransform().Translation)
            FalloffMult := GunStats.GetFallofMult(Dist)
            
            GameManager.AudioManager.PlayWeakSpotHitSFX(SetAgent, WeakSpotSFX)
            ProcessDamage(PlayerManager, BigNumber_Multiply(GunStats.WeakspotMult, PlayerWeakspotMult), true, FalloffMult)
            PlayerManager.OnWeakspotHit()


    OnGuardDamaged(Agent:agent):void=
        if:
            Player:=player[Agent]
            Player.IsActive[]
            PlayerManager := GameManager.PlayerManagers[Player]
            GunStats := PlayerManager.GetCurrentGunStats[]
            PlayerChar := Player.GetFortCharacter[]
        then:
            Dist := Distance(PlayerChar.GetTransform().Translation, GuardProp.GetTransform().Translation)
            FalloffMult := GunStats.GetFallofMult(Dist)

            if:
                AlwaysWeakspot?
                PlayerWeakspotMult := PlayerManager.GetPlayerStat(player_stat.Crit_Mult)
            then:
                ProcessDamage(PlayerManager, GunStats.WeakspotMult * PlayerWeakspotMult, true, FalloffMult)
                PlayerManager.OnWeakspotHit()
            else:
                ProcessDamage(PlayerManager, big_number{Value:=1.0}, false, FalloffMult)
                
    IsAlive()<transacts>:logic=
        return BigNumber_GreaterThanOrEquals(CurrentHealth, big_number{Value:=1.0})
            
    ProcessDamage(PlayerManager : player_manager, WeakspotDamageMult : big_number, IsWeakHit:logic, FalloffMult :big_number, ?DamagedAgent:?agent = false):void=        
        if:
            CurrentHealth.Value <= 0.0 
        then:
            return

        if:
            HealthIsPerPlayer?
            Temp := PlayersHittingUsNow[PlayerManager.Player]
            BigNumber_GreaterThanOrEquals(PlayersHittingUsNow[PlayerManager.Player], MaxHealth)?
        then:
            CheckWeaponExperienceData(PlayerManager)
            CheckBigRockData(PlayerManager)
            return
            
        PlayerManager.UpdateQuestProgressForTask(task_event.HitGuard, big_number{Value:=1.0}, false, ?Target := GuardName)

        PlayerManager.GrantWeaponExperience(false)

        if(BigNumber_GreaterThanOrEquals(CurrentHealth, MaxHealth)?):
            set CurrentHealth = MaxHealth

        if:
            BaseStrength := PlayerManager.GetAdjustedStrength[Classes, ?EnemyWeaknesses := WeaknessesMap, ?DefaultWeakness := DefaultWeakness]
            TeamworkDamage := (1.0 + (PlayerManager.GetUpgradeValue["Teamwork", false] / 100.0 * PlayersHittingUsNow.Length * 1.0))
        then:
            var IsBlocked : logic = false
            var HandcannonDmgMult : big_number = big_number{Value:= 1.0}
            if (PlayerManager.Handcannon  = 3):
                set HandcannonDmgMult = big_number{Value:= HandcannonMultiplier}

            var AmbushDmgMult : big_number = big_number{Value:= 1.0}
            if:
                not(PlayersHittingUsNow[PlayerManager.Player])
                Temp := PlayerManager.GetUpgradeValue["Ambush", false] / 100.0
                set AmbushDmgMult = AmbushDmgMult + Temp
                WepMult := PlayerManager.GetCurrentGunStats[].FirstStrikeMult
                set AmbushDmgMult = AmbushDmgMult * WepMult

            var Strength : big_number = (BaseStrength * WeakspotDamageMult * FalloffMult * TeamworkDamage * AmbushDmgMult * HandcannonDmgMult)
            AdjustedArmorValue := ArmorValue * (1.0 - PlayerManager.GetUpgradeValueSafe("P_ArmorPiercing", false))
            set Strength = Strength - AdjustedArmorValue
            if(Strength.Value < 0.0):
                set Strength = big_number{}
                set IsBlocked = true

            var Gain : big_number = big_number{}
            var Overkill : big_number = big_number{}
            var PerPlayerHealth : big_number = big_number{}

            if:
                HealthIsPerPlayer?
                if:
                    Temp:=PlayersHittingUsNow[PlayerManager.Player]
                then:
                    set PerPlayerHealth = MaxHealth - Temp
                else:
                    set PerPlayerHealth = MaxHealth
                if:
                    BigNumber_GreaterThanOrEquals(Strength, PerPlayerHealth)? and not Unkillable?
                then:
                    set Overkill = Strength - PerPlayerHealth
                    set Gain = PerPlayerHealth
                else:
                    set Gain = Strength
            else:
                if(BigNumber_GreaterThanOrEquals(Strength, CurrentHealth)? and not Unkillable?):
                    set Overkill = Strength - CurrentHealth
                    set Gain = CurrentHealth
                else:
                    set Gain = Strength
           
            if:
                Gain.Value > 0.0
                not EssenceVFX?
            then:
                GameManager.AssignEssenceVFX(Self)

            if:
                Gain.Value > 0.0
                AgentLoc := DamagedAgent?.GetFortCharacter[].GetTransform()
                EssenceVFX?.TeleportTo[AgentLoc]
            else:
                if:
                    EssenceVFX?.TeleportTo[GuardProp.GetTransform().Translation + VFXLocation, GuardProp.GetTransform().Rotation]
            
            if:
                Gain.Value > 0.0
                SetVFX := EssenceVFX?
            then:
                SetVFX.Disable()
                SetVFX.Enable()
                SetVFX.Restart()

            var IsSuperHandcannon : logic = PlayerManager.IsPreSuperHandcannon()
            if:
                RequiresSuperHandcannon?
                not IsSuperHandcannon?
            then:
                set Gain = big_number{}

            if (not Unkillable?):
                ReduceGuardHealth(PlayerManager, Gain)

            for (Player: GameManager.PlayerManagers):
                if:
                    ZoneAgent:=agent[Player.Player]
                    Player = PlayerManager or ZoneManager.ZoneMutuator.IsInVolume[ZoneAgent]
                then:
                    Player.OnDamageDealt(PlayerManager.Agent, Gain)

            var IsHandcannon : logic = PlayerManager.UpdateHandcannon()

            if (IsHandcannon?):
                if:
                    AgentLoc := DamagedAgent?.GetFortCharacter[].GetTransform()
                then:
                    SpawnParticleSystem(VFX_HandCannon, AgentLoc.Translation + VFXLocation)
                else:
                    SpawnParticleSystem(VFX_HandCannon, GuardProp.GetTransform().Translation + VFXLocation)
                

            var IsAmbush : logic = false
            if(BigNumber_GreaterThanOrEquals(AmbushDmgMult, big_number{Value:= 1.001})?):
                set IsAmbush = true
            else:
                set IsAmbush = false            
                
            if (CurrentHealth.Value <= 0.0 and not HealthIsPerPlayer?):
                KillGuard(PlayerManager, Gain, Overkill, IsWeakHit, IsHandcannon, IsAmbush, ?KillAgent := DamagedAgent)
            else:
                PlayerManager.ProcessEssence(ZoneManager.ZoneTag, Gain, Overkill, big_number{}, true, IsWeakHit, IsHandcannon, IsAmbush, IsBlocked)           
                if:
                    IsBlocked?
                then:
                    NotifyBlockedAttack(PlayerManager)
                else:
                    if:
                        HealthIsPerPlayer?
                        Temp:=PlayersHittingUsNow[PlayerManager.Player]
                        set PerPlayerHealth = MaxHealth - Temp
                        PerPlayerHealth.Value <= 0.0
                    then:
                        CheckWeaponExperienceData(PlayerManager)
                        CheckBigRockData(PlayerManager)
              
    GetKOBonus(PlayerManager : player_manager)<transacts>:big_number=
        if:
            DamageDealt:=PlayersHittingUsNow[PlayerManager.Player] 
        then:
            return MaxHealth * GuardMultiplier * PlayerManager.GetKOBonusMultiplier() * (DamageDealt / (MaxHealth * GuardMultiplier))
        
        return big_number{}
            
    ReduceGuardHealth(PlayerManager : player_manager, Damage : big_number):void=
        var OldHealth : big_number = CurrentHealth
        if:
            HealthIsPerPlayer?
        then:
            if:
                Temp := PlayersHittingUsNow[PlayerManager.Player]
                set OldHealth = MaxHealth - Temp
            block:
        else:
            set CurrentHealth = BigNumber_Subtract(CurrentHealth, Damage)

        if:
            not(PlayersHittingUsNow[PlayerManager.Player])
            set PlayersHittingUsNow[PlayerManager.Player] = Damage
        else:
            if:
                Temp := PlayersHittingUsNow[PlayerManager.Player] + Damage
                set PlayersHittingUsNow[PlayerManager.Player] = Temp

        if:
            HealthIsPerPlayer?
            NewHealth := MaxHealth - PlayersHittingUsNow[PlayerManager.Player]
            SetUIManager:=PlayerManager.UIManager?
        then:
            KOBonus : big_number = GetKOBonus(PlayerManager)
            ClearBonus : big_number = ZoneManager.GetClearBonus(PlayerManager)
            AdjustedArmorValue := ArmorValue * (1.0 - PlayerManager.GetUpgradeValueSafe("P_ArmorPiercing", false))
            SetUIManager.QueueUpdateGuardUI(NewHealth, OldHealth, MaxHealth, GuardName, PlayersHittingUsNow.Length, KOBonus, ClearBonus, GuardID, IsUpdateFromOwner := true, IsBoss :=  false, IsDamageMastery := false, ?ArmorValue := option{AdjustedArmorValue})
        else:
            for (Player: GameManager.PlayerManagers):
                if:
                    ZoneAgent:=agent[Player.Player]
                    Player = PlayerManager or ZoneManager.ZoneMutuator.IsInVolume[ZoneAgent]
                    SetUIManager:=Player.UIManager?
                then:
                    KOBonus : big_number = GetKOBonus(Player)
                    ClearBonus : big_number = ZoneManager.GetClearBonus(Player)
            
                    if:
                        Player.GetAchievementLevel["Damage Mastery"] > 0
                    then:
                        ZoneManager.OnZoneDamageDealt(Damage)
                        if:
                            ZoneManager.DamageMasteryTimer > 0
                        then:
                            if:
                                MasteryLevel := ZoneManager.GetDamageMasteryLevel()
                                MasteryMax := ZoneManager.ZoneTag.ToDamageMasteryArray()[MasteryLevel]
                                MasteryMult : float = ZoneManager.ZoneTag.ToDamageMasteryMultipliers()[MasteryLevel]
                            then:
                                MasteryHealth := BigNumber_Subtract(MasteryMax, ZoneManager.DamageMasteryDamageDealt)
                                MasteryName := "Damage Mastery Level {MasteryLevel + 1}"
                                SetUIManager.QueueUpdateGuardUI(MasteryHealth, MasteryHealth, MasteryMax, MasteryName,  PlayersHittingUsNow.Length, BigNumber_Multiply(KOBonus, MasteryMult), BigNumber_Multiply(ClearBonus, MasteryMult), GuardID, IsUpdateFromOwner := true, IsBoss :=  false, IsDamageMastery := true, ?DamageMasteryMult := MasteryMult)
                    else:        
                        if(Player = PlayerManager): 
                            #Update Guard UI per Player
                            SetUIManager.QueueUpdateGuardUI(CurrentHealth, OldHealth, MaxHealth, GuardName, PlayersHittingUsNow.Length, KOBonus, ClearBonus, GuardID, IsUpdateFromOwner := true, IsBoss :=  false, IsDamageMastery := false)
                        else:
                            SetUIManager.QueueUpdateGuardUI(CurrentHealth, OldHealth, MaxHealth, GuardName, PlayersHittingUsNow.Length, KOBonus, ClearBonus, GuardID, IsUpdateFromOwner := false, IsBoss := false, IsDamageMastery := false)
                
    DebugHealth(Health:big_number):void=
        Log("{GuardName} Health: {FormatBigNumber(Health)}")

    DespawnGuard():void=
        set CurrentHealth = big_number{}
        GuardProp.Hide()

        if:
            GuardProp.IsValid[]
            GuardProp.TeleportTo[DeathTransform]
            
        if (FortChar := EntityAgent?.GetFortCharacter[]):
            spawn {DespawnChar(FortChar)}
        
        if(RespawnTimerMax > 0.0):
            set RespawnTimer = RespawnTimerMax
            if (not SpawnPool?):
                spawn {GrowGuard(true)}

    KillGuard(PlayerManagerIn:player_manager, Gain:big_number, Overkill:big_number, IsWeakHit:logic, IsHandcannon : logic, IsAmbush : logic, ?KillAgent:?agent = false):void=
        GuardProp.Hide()
        # if:
        #     GuardName = "Deranged Boar"
        # then:
        #     Log("KILLGUARD for Respawn {RespawnID} {GuardID}")

        var ParticipatingPlayers : []player_manager = array{}
        for (PlayerManager: GameManager.PlayerManagers):
            if:
                ZoneAgent:=agent[PlayerManager.Player]
                DidPlayerHitGuardRecently := ZoneManager.DidPlayerHitGuardRecently(PlayerManager.Player, PlayersHittingUsNow)
                PlayerManager = PlayerManagerIn or ZoneManager.ZoneMutuator.IsInVolume[ZoneAgent] or DidPlayerHitGuardRecently?
            then:
                set ParticipatingPlayers = ParticipatingPlayers + array{PlayerManager}
        
        if:
            GuardProp.IsValid[]
            GuardProp.TeleportTo[DeathTransform]
        
        if:
            KillChar := KillAgent?.GetFortCharacter[]
        then:
            spawn {DespawnChar(KillChar)}
        
        # DROPS
        #from shared tables
        var Items : []drop_item = array{}
        for (Table : DropsShared):
            Item := GameManager.DropManager.SelectFromDropTable(Table, ParticipatingPlayers.Length)
            if (ExistingItem := Item?): 
                set Items = Items + array{ExistingItem}
        #from custom tables
        for (Table : DropsCustom):
            Item := GameManager.DropManager.SelectFromDropTable(Table, ParticipatingPlayers.Length)
            if (ExistingItem := Item?): 
                set Items = Items + array{ExistingItem}
        #from candy table   
        if (not IgnoreCandyDrops?):
            Item := GameManager.DropManager.SelectFromDropTable(ZoneManager.ZoneTag.CandyChances(), ParticipatingPlayers.Length)
            if (ExistingItem := Item?): 
                set Items = Items + array{ExistingItem}
        #XP coin
        if (not IgnoreXPCoinDrop?):
            if (GetRandomFloat(0.0, 1.0) < 0.5):
                Item := drop_item { Name := "XPCoin", PickupText := "XP Coin", Chance := 0.2, Unique := false, MinCount := big_number{Value:=1.0}, MaxCount := big_number{Value:=1.0} }
                set Items = Items + array{Item}

        if:
            TF := SpawnTarget?.GetTransform()
            SetVFX := EssenceVFX?
        then:
            GameManager.DropManager.SpawnItems(Items, SetVFX.GetTransform().Translation, ?TargetPos := option{TF.Translation}, ?ParticipatingPlayers := ParticipatingPlayers)
        else:
            if:
                SetVFX := EssenceVFX?
            then:
                GameManager.DropManager.SpawnItems(Items, SetVFX.GetTransform().Translation, ?ParticipatingPlayers := ParticipatingPlayers)
        

        #AwardBonus
        for (PlayerManager : ParticipatingPlayers):
            DidPlayerHitGuardRecently := ZoneManager.DidPlayerHitGuardRecently(PlayerManager.Player, PlayersHittingUsNow)

            for (Item:Items):
                if:
                    Item.Name = "Midas Coins"
                then:
                    GameManager.AudioManager.PlayDropCoinsSFX(PlayerManager.Agent)
                else:
                    GameManager.AudioManager.PlayDropCandySFX(PlayerManager.Agent)
            
            PlayerManager.OnZoneGuardKO(PlayerManagerIn, ZoneManager.ZoneTag, ZoneManager.GuardsAlive, ZoneManager.MaxGuards, DidPlayerHitGuardRecently, ?GuardName := GuardName)
            
            var KOBonus : big_number = GetKOBonus(PlayerManager)
            
            if:
                PlayerManager.GetAchievementLevel["Damage Mastery"] > 0
                MasteryLevel := ZoneManager.GetDamageMasteryLevel()
                MasteryMax := ZoneManager.ZoneTag.ToDamageMasteryArray()[MasteryLevel]
                MasteryMult : float = ZoneManager.ZoneTag.ToDamageMasteryMultipliers()[MasteryLevel]
            then:
                set KOBonus = BigNumber_Multiply(KOBonus, MasteryMult)

            if(PlayerManager = PlayerManagerIn or DidPlayerHitGuardRecently?):
                if:
                    PlayerQuestOwners.Find[PlayerManager.Player]
                then:
                    RemoveQuestOwner(PlayerManager.Player)
                    
                PlayerManager.ProcessEssence(ZoneManager.ZoneTag, Gain, Overkill, KOBonus, true, IsWeakHit, IsHandcannon, IsAmbush, false)
                PlayerManager.OnGuardKO()
                set PlayerManager.DamageMasteryEssenceGained = BigNumber_Add(PlayerManager.DamageMasteryEssenceGained, KOBonus)
            else:
                PlayerManager.StashEssence(KOBonus)
                set PlayerManager.DamageMasteryEssenceStashed = BigNumber_Add(PlayerManager.DamageMasteryEssenceStashed, KOBonus)
       
        if(RespawnTimerMax > 0.0):
            set RespawnTimer = RespawnTimerMax
            if (not SpawnPool?):
                spawn {GrowGuard(true)}

        spawn:
            ZoneManager.OnGuardDestroyed(GuardID, RespawnTimerMax)
        
        # if (PlayerManagerIn.GetUpgradeLevel["stash"] > 0):
        #     spawn{GameManager.ParticleManager.MoveTo(StartPosition + vector3{X:=0.0, Y:=0.0, Z:=200.0}, ZoneManager.Stash.GetTransform().Translation)}        

    RemoveQuestOwner(Player : player):void=
        for (QuestOwnerIndex := 0..PlayerQuestOwners.Length - 1):
            if:
                PlayerQuestOwners[QuestOwnerIndex] = Player
                Temp:= PlayerQuestOwners.RemoveElement[QuestOwnerIndex]                        
                set PlayerQuestOwners = Temp

    RequestAgentSpawn():void=
        if (Pool := SpawnPool?):
            Pool.RequestSpawn(Self)
        
    OnAgentSpawn():logic=
        if:
            FortChar := EntityAgent?.GetFortCharacter[]
        then:
            if:
                T := SpawnPosition?.GetTransform()
                FortChar.TeleportTo[T.Translation + vector3{X:=0.0, Y:=0.0, Z:=50.0}, T.Rotation]
            then:
                SetLeash(FortChar)
                FortChar.SetMaxHealth(Inf)
                FortChar.DamagedEvent().Subscribe(OnFortCharDamaged)
                FortChar.EliminatedEvent().Subscribe(OnElimination)
                set RespawnID += 1
                if:
                    IsQuestEnemy?
                then:
                    if:
                        CurrentHealth.Value <= 0.0
                    then:
                        set CurrentHealth = MaxHealth
                else:
                    set CurrentHealth = MaxHealth
                    set PlayersHittingUsNow = map{}
                return true
        return false

    OnElimination(Result:elimination_result):void=
        if:
            
            Agent := EntityAgent?
            # Log("ELIMINATION B - for Respawn {RespawnID} {GuardID}")
            FortChar := Agent.GetFortCharacter[]
        then:
            # if:
            #     GuardName = "Deranged Boar"
            # then:
            #     Log("ELIMINATION - for Respawn {RespawnID} {GuardID}")
            spawn {DespawnChar(FortChar)}
        else:
            set EntityAgent = false

    GrowGuard(UseCountdown : logic)<suspends> : void =  
        loop:
            if(UseCountdown = true):
                Sleep(1.0)
            set RespawnTimer -= 1.0 
            if(RespawnTimer <= 0.0):
                GuardProp.Show()
                set CurrentHealth = MaxHealth
                set PlayersHittingUsNow = map{}
                GuardProp.MoveTo(transform{Translation:= StartPosition, Scale:=StartScale, Rotation:=StartRotation}, GameManager.GuardRespawnTime)                
                return

    MatchesClass(ClassName:string)<computes>:logic=
        for (Class : Classes):
            if (Class = ClassName):
                return true
        return false


guard_manager_device := class(creative_device):
    @editable 
    GuardManager : guard_manager = guard_manager{}

    Init(GM : game_manager, ZM: zone_manager_device, ZoneTag : fetchable_tag, GuardIDIn : string):void=
        GuardManager.Init(GM, ZM, ZoneTag, GuardIDIn)
    
gun_weakness := class<concrete>:
    @editable
    ID : int = 0
    @editable
    Weakness : float = 1.0
