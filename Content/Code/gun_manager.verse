# fill this file with your verse-script

using { /Fortnite.com/Devices }
using { /Fortnite.com/UI }
using { /Fortnite.com/Characters }
using { /Verse.org/Simulation }
using { /UnrealEngine.com/Temporary/UI }
using { /Verse.org/Colors }
using { /UnrealEngine.com/Temporary/SpatialMath }
using { /Verse.org/Assets }

cooldown_state := class<concrete>:
    ID : int = -1
    Slot : int = 0
    RemainingCooldown : int = 0
    RemainingDuration : int = 0

cooldown_updater := class<unique>:
    Button : gun_button


    UpdateText()<suspends>:void =
        loop:
            Sleep(0.5) #We sometimes miss a tick if we do 1.0
            
            if:
                CD := Button.PlayerManager.GetRemainingCooldown[Button.GunID]
                CD > 0
            then:
                Button.EquippedText.SetText(STM("{CD}s"))
            else:
                Button.Update(Button.GunID)
                return
            
            if:
                PlayerUIManager := Button.PlayerManager.UIManager?
                MyUI := PlayerUIManager.Popup?
            then:
                block:
            else:
                return
            

gun_granter:= class<concrete>:
    ID : int = 0
    GameManager : game_manager = game_manager{}
    GunManager : ?player_gun_manager = false

    ForceGrant(Wdata:widget_message):void=
        if:
            Player := Wdata.Player
            PlayerManager := GameManager.PlayerManagers[Player]
        then:
            PlayerManager.MakeRoomForGun()
            PlayerManager.EquipItem(ID, ?Equip := true)
            PlayerManager.UpdateQuestProgressForTask(task_event.EquipItem, big_number{Value := 1.0}, false, ?Target := "{ID}", ?IsPersistent := true)
            PlayerManager.ClosePopup(Wdata)

    Grant(Wdata:widget_message):void =
        ForceGrant(Wdata)
        # if:
        #     Player := Wdata.Player
        #     PlayerUI := GetPlayerUI[Player]
        #     PlayerManager := GameManager.PlayerManagers[Player]
        #     UIManager := PlayerManager.UIManager?
        #     GM := GunManager?
        # then:
            

            # if (PlayerManager.HasEmptySlot[]):
            # else:
            #     Canvas := canvas:
            #     Canvas.AddWidget( 
            #         canvas_slot:
            #             Widget:= GM.HeaderPlate
            #             SizeToContent := true
            #             Anchors := anchors{ Minimum := vector2{X := 0.3, Y := 0.3}, Maximum := vector2{X := 0.7, Y := 0.7}}
            #             Offsets := margin{Top := 0.0, Left := 0.0, Right := 0.0, Bottom := 0.0}
            #             Alignment := vector2{X := 0.5, Y := 0.5}
            #     )
            #     Canvas.AddWidget( 
            #         canvas_slot:
            #             Widget:= text_block:
            #                 DefaultTextColor:=NamedColors.White
            #                 DefaultShadowColor:=NamedColors.Black
            #                 DefaultShadowOffset:=option{vector2{X:=2.0,Y:=2.0}}
            #                 DefaultShadowOpacity:= 1.0
            #                 DefaultText:= STM("This will replace your current weapon.")
            #             SizeToContent := true
            #             Anchors := anchors{ Minimum := vector2{X := 0.3, Y := 0.3}, Maximum := vector2{X := 0.7, Y := 0.7}}
            #             Offsets := margin{Top := 50.0, Left := 50.0, Right := 0.0, Bottom := 0.0}
            #             Alignment := vector2{X := 0.5, Y := 0.5}
            #     )
            #     Canvas.AddWidget( 
            #         canvas_slot:
            #             Widget:= text_block:
            #                 DefaultTextColor:=NamedColors.White
            #                 DefaultShadowColor:=NamedColors.Black
            #                 DefaultShadowOffset:=option{vector2{X:=2.0,Y:=2.0}}
            #                 DefaultShadowOpacity:= 1.0
            #                 DefaultText:= STM("Are you sure?")
            #             SizeToContent := true
            #             Anchors := anchors{ Minimum := vector2{X := 0.3, Y := 0.3}, Maximum := vector2{X := 0.7, Y := 0.7}}
            #             Offsets := margin{Top := 100.0, Left := 50.0, Right := 0.0, Bottom := 0.0}
            #             Alignment := vector2{X := 0.5, Y := 0.5}
            #     )
                
            #     YesButton := button_quiet:
            #     YesButton.SetText(STM("YES"))            
            #     YesButton.OnClick().Subscribe(ForceGrant)
                
            #     NoButton := button_quiet:
            #     NoButton.SetText(STM("NO"))            
            #     NoButton.OnClick().Subscribe(PlayerManager.ClosePopup)

            #     Canvas.AddWidget( 
            #         canvas_slot:
            #             Widget:= YesButton
            #             Anchors := anchors{ Minimum := vector2{X := 0.3, Y := 0.6}, Maximum := vector2{X := 0.5, Y := 0.65}}
            #             Offsets := margin{Top := 0.0, Left := 20.0, Right := 20.0, Bottom := 0.0}
            #             Alignment := vector2{X := 0.5, Y := 0.5}
            #     )

            #     Canvas.AddWidget( 
            #         canvas_slot:
            #             Widget:= NoButton
            #             Anchors := anchors{ Minimum := vector2{X := 0.5, Y := 0.6}, Maximum := vector2{X := 0.7, Y := 0.65}}
            #             Offsets := margin{Top := 0.0, Left := 20.0, Right := 20.0, Bottom := 0.0}
            #             Alignment := vector2{X := 0.5, Y := 0.5}
            #     )
                
            #     UIManager.ClosePopup()
            #     PlayerUI.AddWidget(Canvas, player_ui_slot{InputMode := ui_input_mode.All})
            #     if (set PlayerManager.UIManager?.Popup = option{Canvas}) {}
        # else:
        #     Log("Something went wrong granting gun {ID}")

weapon_class := enum:
    None
    Melee
    Rapid
    Precision

gun_stats := class<concrete>: 
    FallofCategory<public><localizes>:message = "Settings"

    @editable
    Name : string = "Gun"
    @editable
    WeaponClass : weapon_class = weapon_class.None
    @editable
    Cooldown : int = 0
    @editable
    Duration : int = 0
    @editable
    DamageMult : big_number = big_number{Value:=1.0, Exponent:=0}
    @editable
    WeakspotMult : big_number = big_number{Value:=1.5, Exponent:=0}
    @editable
    FirstStrikeMult : big_number = big_number{Value:=1.0, Exponent:=0}
    @editable
    HandcannonMult : big_number = big_number{Value:=1.0, Exponent:=0}
    @editable
    FocusFireMult : float = 1.0
    @editable:
        Categories:=array{FallofCategory}
    FalloffStart : float = 99999.0
    @editable:
        Categories:=array{FallofCategory}
    FalloffEnd : float = 999999.0
    @editable:
        Categories:=array{FallofCategory}
    FalloffMin : float = 1.0
    @editable
    FortBaseDamage : float = 0.0
    @editable
    DamageStatType : player_stat = player_stat.None

    GetFallofMult(Dist : float)<transacts>:big_number=
        if (Dist < FalloffStart):
            return big_number{Value:=1.0, Exponent:=0}
        else:
            if (Dist > FalloffEnd):
                return big_number{Value:=FalloffMin, Exponent:=0}
            else:
                return big_number{Value:=(1.0-(Dist - FalloffStart) / (FalloffEnd - FalloffStart)) * (1.0-FalloffMin) + FalloffMin, Exponent:=0}

    GetDuration(PlayerManager : player_manager)<transacts>:int=
        return Round[Duration * (1.0 + PlayerManager.GetUpgradeValueSafe("P_GunDuration", false))] or 0

gun_manager:= class(creative_device, menu_item):

    @editable
    StatsTable : []gun_stats = array{}
    
    @editable
    GameManager : ?game_manager = false

    var MenuPerPlayer : [player]player_gun_manager = map{}
    
    OpenMenu<override>(Agent:agent, ?Force:logic = false, ?RequiredID:int=0):void=
        if (PlayerManager := GameManager?.PlayerManagers[Agent]):
            if (Force?):
                if:
                    UI:=PlayerManager.UIManager?
                then:
                    UI.ClosePopup()

            if (not GameManager?.PlayerManagers[Agent].UIManager?.Popup?):
                if: 
                    GM := GameManager?
                then:
                    if (Menu := MenuPerPlayer[PlayerManager.Player]):
                        Menu.OpenMenu(RequiredID)
                    else:
                        Menu := MakeGunManager(GM, StatsTable, Agent)
                        if:
                            set MenuPerPlayer[PlayerManager.Player] = Menu
                        Menu.OpenMenu(RequiredID)
            else:
                block:#Log("DOUBLE MENU")

    GetTitle<override>():string=
        return "GUNS"

    GetIcon<override>():texture=
        return Images.Gun_Rifle

    IsEnabled<override>(PlayerManager : player_manager)<transacts>:logic=
        return true

MakeGunManager<constructor>(GameManagerIn : game_manager, StatsTableIn : []gun_stats, AgentIn : agent) := player_gun_manager:
    GameManager :=  GameManagerIn
    StatsTable := StatsTableIn
    Agent := AgentIn

player_gun_manager:= class:
    StatsTable : []gun_stats = array{}
    GameManager : game_manager
    SignalDevice : signal_remote_manager_device = signal_remote_manager_device{}

    Agent : agent

    var DialogBackground : texture_block = texture_block{ DefaultImage := Images.StoreBox, DefaultDesiredSize := vector2{X := 1800.0, Y := 700.0}}
    var HeaderPlate : texture_block = texture_block{ DefaultImage := Images.DialogBox, DefaultDesiredSize := vector2{X := 400.0, Y := 80.0}}

    var Widget : ?canvas = false
    var StackBox : stack_box = stack_box{Orientation:=orientation.Horizontal}
    var StackBox2 : stack_box = stack_box{Orientation:=orientation.Horizontal}
    var StackBoxOuter : stack_box = stack_box{Orientation:=orientation.Vertical}
    var GunButtons : []gun_button = array{}

    OpenMenu(RequiredID:int):void =
        if (not Widget?):
            CreateMenu()

        UpdateMenu(RequiredID)
        DisplayMenu()

    UpdateMenu(RequiredID:int):void =
        if:
            InPlayer := player[Agent]
            PlayerManager := GameManager.PlayerManagers[InPlayer]
        then:
            var GunsToShow : []int = array{}
            for (i := 0..99):
                if:
                    i = 1 or PlayerManager.GetUpgradeLevel["Gun_{i}"] > 0
                    RequiredID = 0 or RequiredID = i
                then:
                    set GunsToShow += array{i}
            
            if:
                GunsToShow.Length < GunButtons.Length
            then:
                for (i := GunButtons.Length - 1..GunsToShow.Length):
                    if:
                        Temp:=GunButtons[i]
                        Slot := Temp.Widget?
                    then:
                        StackBox.RemoveWidget(Slot.Widget)
                        StackBox2.RemoveWidget(Slot.Widget)
                        if:
                            Temp2:=GunButtons.RemoveElement[i]
                        then:
                            set GunButtons = Temp2

            for (i := GunButtons.Length..GunsToShow.Length-1):
                if (GunToShow := GunsToShow[i]):
                    Btn := MakeGunButton(GunToShow, PlayerManager, Self)
                    Btn.Create()
                    set GunButtons = GunButtons + array{Btn}
                    if (Slot := Btn.Widget?):
                        if (i >= 5):
                            StackBox2.AddWidget(Slot)
                        else:
                            StackBox.AddWidget(Slot)

            for (i := 0..GunButtons.Length-1):
                if:
                    Btn := GunButtons[i]
                    ID := GunsToShow[i]
                then:
                    Btn.Update(ID)
    
    OnClose(Wdata : widget_message):void=
        if:
            InPlayer := player[Agent]
            InPlayer.IsActive[]
            PlayerManager := GameManager.PlayerManagers[InPlayer]
        then:
            PlayerManager.ClosePopup(Wdata)

    CreateMenu():void =
        if:
            InPlayer := player[Agent]
            InPlayer.IsActive[]
            PlayerUI := GetPlayerUI[InPlayer]
            PlayerManager := GameManager.PlayerManagers[InPlayer]
        then:

            StackBoxOuter.AddWidget(stack_box_slot:
                Widget:= StackBox
                Padding:= margin{Top := 0.0, Left := 0.0, Right := 0.0, Bottom := 5.0}
            )
            StackBoxOuter.AddWidget(stack_box_slot:
                Widget:= StackBox2
                Padding:= margin{Top := 60.0, Left := 0.0, Right := 0.0, Bottom := 00.0}
            )
            
            CloseButton := button_quiet:
            CloseButton.SetText(STM("                                                                        BACK                                                                        "))            
            CloseButton.OnClick().Subscribe(OnClose)
            
            HeaderText := text_block:
                DefaultTextColor:=NamedColors.White
                DefaultShadowColor:=NamedColors.Black
                DefaultShadowOffset:=option{vector2{X:=2.0,Y:=2.0}}
                DefaultShadowOpacity:= 1.0
                DefaultText:= STM("Choose a weapon:")


            Canvas := canvas:
                Slots := array:
                    canvas_slot:
                        Widget:= DialogBackground
                        SizeToContent := true
                        Anchors := anchors{ Minimum := vector2{X := 0.03, Y := 0.95}, Maximum := vector2{X := 0.03, Y := 0.95}}                        
                        Alignment := vector2{X := 0.0, Y := 1.0}
                        ZOrder := 1
                    canvas_slot:
                        Widget:= HeaderPlate
                        SizeToContent := true
                        Anchors := anchors{ Minimum := vector2{X := 0.03, Y := 0.935}, Maximum := vector2{X := 0.03, Y := 0.935}}
                        Offsets := margin{Top := -640.0, Left := 100.0, Right := 0.0, Bottom := 0.0}
                        Alignment := vector2{X := 0.0, Y := 1.0}
                        ZOrder := 2
                    canvas_slot:
                        Widget:= HeaderText
                        Anchors := anchors{ Minimum := vector2{X := 0.03, Y := 0.935}, Maximum := vector2{X := 0.03, Y := 0.935}}
                        Offsets := margin{Top := -665.0, Left := 300.0, Right := 0.0, Bottom := 0.0}
                        Alignment := vector2{X := 0.5, Y := 1.0}
                        ZOrder := 3
                    canvas_slot:
                        Widget:= StackBoxOuter
                        SizeToContent := true
                        Anchors := anchors{ Minimum := vector2{X := 0.5, Y := 0.80}, Maximum := vector2{X := 0.5, Y := 0.80}}
                        Offsets := margin{Top := -190.0, Left := 0.0, Right := 0.0, Bottom := 0.0}
                        Alignment := vector2{X := 0.5, Y := 0.5}
                        ZOrder := 5
                    canvas_slot:
                        Widget:= CloseButton
                        Anchors := anchors{ Minimum := vector2{X := 0.5, Y := 0.915}, Maximum := vector2{X := 0.5, Y := 0.915}}
                        Alignment := vector2{X := 0.5, Y := 1.0}
                        ZOrder := 6
                        
            set Widget = option{Canvas}

    DisplayMenu():void =
        if:
            InPlayer := player[Agent]
            InPlayer.IsActive[]
            PlayerUI := GetPlayerUI[InPlayer]
            PlayerManager := GameManager.PlayerManagers[InPlayer]
            Canvas := Widget?
        then:
            Canvas.SetEnabled(true)
            Canvas.SetVisibility(widget_visibility.Visible)
            PlayerUI.AddWidget(Canvas, player_ui_slot{InputMode := ui_input_mode.All})
            if (set PlayerManager.UIManager?.Popup = option{Canvas}) {}

MakeGunButton(GunIDIn:int, PlayerManagerIn:player_manager, PlayerGunManagerIn:player_gun_manager):gun_button = gun_button:
    GunID := GunIDIn
    PlayerManager := PlayerManagerIn
    PlayerGunManager := PlayerGunManagerIn

gun_button:= class():

    PlayerManager : player_manager
    PlayerGunManager : player_gun_manager

    var GunID : int = 0

    var Widget : ?stack_box_slot = false

    var Overlay : canvas = canvas: 
    UpgradePlate : texture_block = texture_block{ DefaultImage := Images.DialogBoxVertical, DefaultDesiredSize := vector2{X := 180.0, Y := 160.0}}

    var WasOnCooldown : logic = true

    GunDisplay : texture_block = texture_block:
        DefaultImage := Images.Empty
        DefaultDesiredSize := vector2{X := 100.0, Y := 100.0}
    UseItemButton : button_loud = button_loud:
    EquippedText : text_block = text_block:
        DefaultTextColor:=NamedColors.AntiqueWhite
        DefaultShadowColor:=NamedColors.Black
        DefaultShadowOffset:=option{vector2{X:=2.0,Y:=2.0}}
        DefaultShadowOpacity:= 1.0
    var GunExpProgressBar : progress_bar = progress_bar{}
    GunExpLevelText :text_block = text_block:
        DefaultTextColor:=NamedColors.White
        DefaultShadowColor:=NamedColors.Black
        DefaultShadowOffset:=option{vector2{X:=2.0,Y:=2.0}}
        DefaultShadowOpacity:= 1.0
    GunExpProgressText :text_block = text_block:
        DefaultTextColor:=NamedColors.White
        DefaultShadowColor:=NamedColors.Black
        DefaultShadowOffset:=option{vector2{X:=2.0,Y:=2.0}}
        DefaultShadowOpacity:= 1.0

    Create():void =
        set Widget = option:
            stack_box_slot:
                Widget:= Overlay
                Padding:= margin{Top := 0.0, Left := 15.0, Right := 15.0, Bottom := 0.0}

        set GunExpProgressBar = MakeProgressBar(0.0, vector2{X:= 140.0, Y:= 50.0}, MakeColorFromHex("0072fc"), MakeColorFromHex("000000"), NamedColors.Goldenrod) #Exp Blue

        GunDisplay.SetImage(GetTextureFromString("Gun_{GunID}"))
        
        UseItemButton.SetText(STM("EQUIP"))
        UseItemButton.OnClick().Subscribe(OnEquip)

        Level:=PlayerManager.GetWeaponLevel(GunID)
        WepExp:= MakeBigNumber(PlayerManager.GetWeaponExperience(GunID) * 1.0,1)
        Req:= MakeBigNumber(PlayerManager.GetWeaponLevelRequirement(Level) * 1.0,1)

        GunExpLevelText.SetText(STM("Lv. {Level}:"))
        GunExpProgressText.SetText(STM("{WepExp} / {Req}"))
        GunExpProgressBar.SetProgress((PlayerManager.GetWeaponExperience(GunID) * 1.0)/(PlayerManager.GetWeaponLevelRequirement(Level) * 1.0),true)
        
        #Keep these AddWidgets Seperate.
        Overlay.AddWidget(
            canvas_slot:
                Widget:=UpgradePlate
                Alignment:= vector2{X:=0.5,Y:=0.5}
                ZOrder:=0
        )

        #Keep these AddWidgets Seperate.
        Overlay.AddWidget(
            canvas_slot:
                Widget:=GunDisplay
                Offsets:= margin{Top:=0.0,Left:=0.0,Right:=20.0,Bottom:=30.0}
                Alignment:= vector2{X:=0.5,Y:=0.5}
                ZOrder:=1
        )

        #Keep these AddWidgets Seperate.
        Overlay.AddWidget(
            canvas_slot:
                Widget:= GunExpLevelText
                Offsets:= margin{Top:=-110.0,Left:=0.0,Right:=0.0,Bottom:=0.0}
                Alignment:= vector2{X:=0.5,Y:=0.5}
                ZOrder:=2
        )
       
        Overlay.AddWidget(
            canvas_slot:
                Widget:= GunExpProgressBar.GetWidget()
                Offsets:= margin{Top:=-70.0,Left:=0.0,Right:=0.0,Bottom:=0.0}
                Alignment:= vector2{X:=0.5,Y:=0.5}
                ZOrder:=3
        )

        #Keep these AddWidgets Seperate.
        Overlay.AddWidget(
            canvas_slot:
                Widget:= GunExpProgressText
                Offsets:= margin{Top:=-70.0,Left:=0.0,Right:=0.0,Bottom:=0.0}
                Alignment:= vector2{X:=0.5,Y:=0.5}
                ZOrder:=4
        )
       
    Update(ID:int):void =
        if (not GunID = ID):
            set GunID = ID
            GunDisplay.SetImage(GetTextureFromString("Gun_{GunID}"))

        Level:=PlayerManager.GetWeaponLevel(GunID)
        WepExp:= MakeBigNumber(PlayerManager.GetWeaponExperience(GunID) * 1.0,1)
        Req:= MakeBigNumber(PlayerManager.GetWeaponLevelRequirement(Level) * 1.0,1)
        GunExpLevelText.SetText(STM("Lv. {Level}:"))
        GunExpProgressText.SetText(STM("{WepExp} / {Req}"))
        GunExpProgressBar.SetProgress((PlayerManager.GetWeaponExperience(GunID) * 1.0)/(PlayerManager.GetWeaponLevelRequirement(Level) * 1.0),true)

        if:
            CD := PlayerManager.GetRemainingCooldown[GunID]
            Duration := PlayerManager.GetRemainingDuration[GunID]
            CD > 0
        then:
            if (not WasOnCooldown?):
                Overlay.RemoveWidget(UseItemButton)

                Overlay.AddWidget(
                    canvas_slot:
                        Widget:=EquippedText
                        Anchors:= anchors{Minimum:=vector2{X:=0.5,Y:=0.5},Maximum:=vector2{X:=0.5,Y:=0.5}}
                        Offsets:= margin{Top:=-80.0,Left:=-80.0,Right:=0.0,Bottom:=0.0}
                        Alignment:= vector2{X:=0.5,Y:=0.5}
                        ZOrder:=10
                )


            if (PlayerManager.GetEquippedItemIDs[].Find[GunID]):
                EquippedText.SetText(STM("EQUIPPED"))
            else:
                EquippedText.SetText(STM("{CD}s"))

                #Update the cooldown text
                spawn{cooldown_updater{
                    Button:=Self}.UpdateText()}
            
            set WasOnCooldown = true
        else:
            if (WasOnCooldown?):
                Overlay.RemoveWidget(EquippedText)

                Overlay.AddWidget(
                    canvas_slot:
                        Widget:=UseItemButton
                        Anchors:= anchors{Minimum:=vector2{X:=0.0,Y:=0.75},Maximum:=vector2{X:=0.5,Y:=0.7525}}
                        Offsets:= margin{Top:=-50.0,Left:=-90.0,Right:=0.0,Bottom:=0.0}
                        Alignment:= vector2{X:=0.5,Y:=0.5}
                        ZOrder:=11
                )
            
            if (GunStats := PlayerManager.GetGunStats[GunID]):
                var DmgStr : string = FormatBigNumber(GunStats.DamageMult)
                if (GunStats.DamageMult.Exponent < 1):
                    set DmgStr = FloatToTruncatedString(GunStats.DamageMult.Value, 1)

                
            set WasOnCooldown = false

    OnEquip(Wdata:widget_message):void =
        gun_granter{
            ID := GunID
            GameManager:= PlayerManager.GameManager
            GunManager := option{PlayerGunManager}
        }.Grant(Wdata)