using { /Fortnite.com/Devices }
using { /Fortnite.com/UI }

hide_default_hud := class(creative_device):
    
    OnBegin<override>()<suspends> : void =
        Playspace := Self.GetPlayspace()
        HUDController := Playspace.GetHUDController()
 
        HUDElementsToHide := array:
            creative_hud_identifier_health_numbers{}
            creative_hud_identifier_mimimap{}
            creative_hud_identifier_minimap{}
            creative_hud_identifier_map_prompts{}
            hud_identifier_world_resource_wood{}
            hud_identifier_world_resource_stone{}
            hud_identifier_world_resource_metal{}
            hud_identifier_world_resource_permanite{}
            hud_identifier_world_resource_gold_currency{}
            hud_identifier_world_resource_ingredient{}
            creative_hud_identifier_team_info{}
            creative_hud_identifier_equipped_item{}
 
        HUDElementsToShow := array:
            creative_hud_identifier_player_inventory{}
            creative_hud_identifier_interaction_prompts{}
            # creative_hud_identifier_map_prompts{}
            creative_hud_identifier_experience_level{}
            creative_hud_identifier_experience_supercharged{}
            creative_hud_identifier_experience_ui{}
            creative_hud_identifier_health{}
            #creative_hud_identifier_health_numbers{}
            creative_hud_identifier_hud_info{}


        HUDController.HideElements(HUDElementsToHide)
        HUDController.ShowElements(HUDElementsToShow)