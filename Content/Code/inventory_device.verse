# fill this file with your verse-script

using { /Fortnite.com/Devices }
using { /Fortnite.com/UI }
using { /Fortnite.com/Characters }
using { /Verse.org/Simulation }
using { /UnrealEngine.com/Temporary/UI }
using { /Verse.org/Colors }
using { /UnrealEngine.com/Temporary/SpatialMath }
using { /Verse.org/Assets }

collectible_info := class<concrete>:
    @editable
    Name : ?string = false
    @editable
    Description : string = ""
    @editable
    VO : ?audio_player_device = false

inventory_device:= class(creative_device, menu_item):

    @editable
    StatsTable : []collectible_info = array{}
    
    @editable
    GameManager : ?game_manager = false
    
    OpenMenu<override>(Agent:agent):void=
        if (PlayerManager := GameManager?.PlayerManagers[Agent]):
            
            if (not GameManager?.PlayerManagers[Agent].UIManager?.Popup?):
                if (GM := GameManager?):
                    MakeInventoryMenu(GM, StatsTable, Agent).OpenMenu()
            else:
                block:#Log("DOUBLE MENU")

    GetTitle<override>():string=
        return "Inventory"

    GetIcon<override>():texture=
        return Images.Backpack

    IsEnabled<override>(PlayerManager : player_manager)<transacts>:logic=
        if:
            Temp:= PlayerManager.GetAchievementLevel["Landed in Goodtown"]
            Temp > 0
        then:
            return true
        else:
            return false

MakeInventoryMenu<constructor>(GameManagerIn : game_manager, StatsTableIn : []collectible_info, AgentIn : agent) := inventory_menu:
    GameManager :=  GameManagerIn
    StatsTable := StatsTableIn
    Agent := AgentIn

inventory_menu:= class:
    StatsTable : []collectible_info = array{}
    GameManager : game_manager
    Agent : agent

    var DialogBackground : texture_block = texture_block{ DefaultImage := Images.DialogBox, DefaultDesiredSize := vector2{X := 1920.0, Y := 1080.0}} 
    #var HeaderPlate : texture_block = texture_block{ DefaultImage := Images.DialogBox, DefaultDesiredSize := vector2{X := 400.0, Y := 80.0}}

    OpenMenu():void =
        if:
            InPlayer := player[Agent]
            InPlayer.IsActive[]
            PlayerUI := GetPlayerUI[InPlayer]
            PlayerManager := GameManager.PlayerManagers[InPlayer]
            Keys := PlayerManager.GetCollectibleMap[]
        then:
            StackBoxOuter := stack_box{Orientation:=orientation.Vertical}
            var CurrentStackBox : stack_box = stack_box{Orientation:=orientation.Horizontal}
            StackBoxOuter.AddWidget(stack_box_slot:
                Widget:= CurrentStackBox
                Padding:= margin{Top := 0.0, Left := 0.0, Right := 0.0, Bottom := 00.0}
            )

            CloseButton := button_quiet:
            CloseButton.SetText(STM("            BACK            "))            
            CloseButton.OnClick().Subscribe(PlayerManager.ClosePopup)

            Canvas := canvas:
                Slots := array:
                    canvas_slot:
                        Widget:= DialogBackground
                        SizeToContent := true
                        Anchors := anchors{ Minimum := vector2{X := 0.1, Y := 0.1}, Maximum := vector2{X := 0.9, Y := 0.9}}  
                        ZOrder := 1
                    canvas_slot:
                        Widget:= StackBoxOuter
                        SizeToContent := true
                        Anchors := anchors{ Minimum := vector2{X := 0.5, Y := 0.45}, Maximum := vector2{X := 0.5, Y := 0.45}}  
                        Alignment := vector2{X := 0.5, Y := 0.5}
                        ZOrder := 2
                    canvas_slot:
                        Widget:= CloseButton
                        Anchors := anchors{ Minimum := vector2{X := 0.5, Y := 0.855}, Maximum := vector2{X := 0.5, Y := 0.855}}
                        Alignment := vector2{X := 0.5, Y := 1.0}
                        ZOrder := 3
                      
            var ItemCount : int = 0#
            for (Key -> Value : Keys):
                Count := Value.Count
                if (BigNumber_GreaterThanOrEquals(Count, big_number{Value:=1.0})?):
                    if (ItemCount >= 10):
                        set CurrentStackBox = stack_box{Orientation:=orientation.Horizontal}
                        StackBoxOuter.AddWidget(stack_box_slot:
                            Widget:= CurrentStackBox
                            Padding:= margin{Top := 0.0, Left := 0.0, Right := 0.0, Bottom := 00.0}
                        )
                        set ItemCount = 0
                    set ItemCount += 1

                    CollectibleDisplay := texture_block:
                                DefaultImage := Value.GetImage()
                                DefaultDesiredSize := vector2{X := 120.0, Y := 120.0}
            
                    NameText := text_block:
                        DefaultTextColor:=NamedColors.White
                        DefaultShadowColor:=NamedColors.Black
                        DefaultShadowOffset:=option{vector2{X:=2.0,Y:=2.0}}
                        DefaultShadowOpacity:= 1.0
                        DefaultText:= STM("{Value.Name}")
                    
            
                    CountText := text_block:
                        DefaultTextColor:=NamedColors.White
                        DefaultShadowColor:=NamedColors.Black
                        DefaultShadowOffset:=option{vector2{X:=2.0,Y:=2.0}}
                        DefaultShadowOpacity:= 1.0
                        DefaultText:= STM("{Count}")

                    if (not BigNumber_GreaterThanOrEquals(Count, big_number{Value:=1.0})?):
                        CountText.SetVisibility(widget_visibility.Hidden)
                    
                    Widget := overlay:
                        Slots:= array:
                            overlay_slot:
                                Widget:=CollectibleDisplay
                                Padding:= margin{Top := 0.0, Left := 0.0, Right := 0.0, Bottom := 0.0}
                            overlay_slot:
                                Widget:=NameText
                                Padding:= margin{Top := -75.0, Left := 20.0, Right := 0.0, Bottom := 0.0}
                                HorizontalAlignment:= horizontal_alignment.Left
                                VerticalAlignment:= vertical_alignment.Top
                            overlay_slot:
                                Widget:=CountText
                                Padding:= margin{Top := 0.0, Left := 5.0, Right := 0.0, Bottom := 0.0}
                                HorizontalAlignment:= horizontal_alignment.Left
                                VerticalAlignment:= vertical_alignment.Top

                    if (PlayerManager.CanUseCollectible(Value)?):
                        UseBtn := button_quiet:
                            DefaultText:= STM("Use")
                        Widget.AddWidget(overlay_slot:
                            Widget:= UseBtn
                            Padding:= margin{Top := 60.0, Left := 0.0, Right := 0.0, Bottom := 0.0}
                            HorizontalAlignment:= horizontal_alignment.Center
                            VerticalAlignment:= vertical_alignment.Center
                        )
                        UseBtn.OnClick().Subscribe(collectible_user{PlayerManager:=PlayerManager, Collectible:=Value, CountText:=CountText, UseBtn:=UseBtn, Container:=Widget}.Use)
                    
                    WidgetSlot := stack_box_slot:
                        Widget:= Widget
                        Padding:= margin{Top := 0.0, Left := 2.5, Right := 0.0, Bottom := 34.0}

                    CurrentStackBox.AddWidget(WidgetSlot)
            
            PlayerUI.AddWidget(Canvas, player_ui_slot{InputMode := ui_input_mode.All})
            if (set PlayerManager.UIManager?.Popup = option{Canvas}) {}

collectible_user:= class():
    PlayerManager : player_manager
    Collectible : collectible_data
    CountText : text_block
    UseBtn : button_quiet
    Container : overlay

    Use(Wdata : widget_message):void =
        PlayerManager.UseCollectible(Collectible)
        NewCount := PlayerManager.GetCollectibleCount(Collectible.Key)
        CountText.SetText(STM("{NewCount}"))
        
        if (not BigNumber_GreaterThanOrEquals(NewCount, big_number{Value:=1.0})?):
            Container.SetVisibility(widget_visibility.Collapsed)
            
        if (not BigNumber_GreaterThanOrEquals(NewCount, big_number{Value:=2.0})?):
            CountText.SetVisibility(widget_visibility.Hidden)

        if (not PlayerManager.CanUseCollectible(Collectible)?):
            UseBtn.SetVisibility(widget_visibility.Hidden)