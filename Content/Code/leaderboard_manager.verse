using { /Fortnite.com/Devices }
using { /Fortnite.com/UI }
using { /UnrealEngine.com/Temporary/SpatialMath }
using { /UnrealEngine.com/Temporary/UI }
using { /Verse.org/Colors }
using { /Verse.org/Simulation}

leaderboard_device := class(creative_device): 
    @editable
    PlayerSlots : []billboard_device = array{}

    @editable
    Type : string = ""

    var GameManager : ?game_manager = false
    
    Leaderboard : player_leaderboard_manager = player_leaderboard_manager{}

    Init():void=
        spawn:
            LoopingMinuteCheck()

    LoopingMinuteCheck()<suspends>:void=
        loop:
            Sleep(30.0)
            OnUpdate()

    OnUpdate():void=
        #Leaderboard
        if:
            GM := GameManager?
        then:
            case (Type):
                "GuardsKilled" => block:
                    for (InPlayerManager : GM.PlayerManagers):
                        if:
                            Amount := InPlayerManager.GetGuardsDestroyed[]
                        then:
                            Leaderboard.UpdateLeaderboard(InPlayerManager.Agent, Amount)    
                "PeelysFound" => block:
                    for (InPlayerManager : GM.PlayerManagers):
                        Amount := InPlayerManager.GetCollectibleCount("Peely")
                        Leaderboard.UpdateLeaderboard(InPlayerManager.Agent, Amount)    
                "PlayerLevel" => block:
                    for (InPlayerManager : GM.PlayerManagers):
                        if:
                            Amount := InPlayerManager.GetLevel[]
                        then:
                            Leaderboard.UpdateLeaderboard(InPlayerManager.Agent, big_number{Value:=Amount*1.0})    
                "BigRockLevel" => block:
                    for (InPlayerManager : GM.PlayerManagers):
                        Amount := InPlayerManager.GetBigRockLevel()
                        Leaderboard.UpdateLeaderboard(InPlayerManager.Agent, big_number{Value:=Amount*1.0})    
                "FlowersFound" => block:
                    for (InPlayerManager : GM.PlayerManagers):
                        var Amount : int = 0
                        if:
                            BestAchieve := InPlayerManager.GetAchievementLevel["Blossom_Goodflower"]
                            BestAchieve > 0
                        then:
                            set Amount = BestAchieve-1
                        Leaderboard.UpdateLeaderboard(InPlayerManager.Agent, big_number{Value:=Amount*1.0})    
                "LifetimeEssence" => block:
                    for (InPlayerManager : GM.PlayerManagers):
                        if:
                            Amount := InPlayerManager.GetLifetimeEssence[]
                        then:
                            Leaderboard.UpdateLeaderboard(InPlayerManager.Agent, Amount)    
                _ => block:

            UpdateBillboards()

    UpdateBillboards():void=
        for(Slot : Leaderboard.PlayerScores):
            # Print("{Type} Slot Loop {Slot.Rank} | {Slot.Score}")
            Rank := Slot.Rank - 1
            if:
                Billboard:=PlayerSlots[Rank]
                Agent := Slot.Agent?
            then:
                Billboard.SetText(PlayerNameWithString(Agent," {Slot.Score}"))
                Billboard.UpdateDisplay()


MakePlayerLeaderboard(AgentIn : agent):player_leaderboard_slot=
    
    PlayerScore := player_leaderboard_slot{}
    set PlayerScore.Agent = option{AgentIn}

    set PlayerScore.PlayerName = text_block{DefaultTextColor:=NamedColors.White,
    DefaultShadowColor:=NamedColors.Black,
    DefaultShadowOffset:=option{vector2{X:=2.0,Y:=2.0}}}
    PlayerScore.PlayerName.SetShadowOpacity(1.0)    
    PlayerScore.PlayerName.SetText(GetPlayerName(AgentIn))

    set PlayerScore.ScoreText = text_block{DefaultTextColor:=NamedColors.White,
    DefaultShadowColor:=NamedColors.Black,
    DefaultShadowOffset:=option{vector2{X:=2.0,Y:=2.0}}}
    PlayerScore.ScoreText.SetShadowOpacity(1.0)    
    PlayerScore.ScoreText.SetText(STM(""))

    set PlayerScore.RankText = text_block{DefaultTextColor:=NamedColors.White,
    DefaultShadowColor:=NamedColors.Black,
    DefaultShadowOffset:=option{vector2{X:=2.0,Y:=2.0}}}
    PlayerScore.RankText.SetShadowOpacity(1.0)    
    PlayerScore.RankText.SetText(STM(""))

    set PlayerScore.ProgressBar  = MakeProgressBar(0.0, vector2{X:= 500.0, Y:= 35.0}, MakeColorFromHex("CC362F"), MakeColorFromHex("000000"), MakeColorFromHex("FFFFFF")) #Red

    set PlayerScore.Widget = 
        overlay:   
            Slots:= array:
                overlay_slot:
                    Widget:= overlay:
                        Slots:= array:
                            overlay_slot:
                                Widget := PlayerScore.ColorBlock
                            overlay_slot:  
                                Widget := PlayerScore.ProgressBar.GetWidget()
                                VerticalAlignment := vertical_alignment.Top
                            overlay_slot:
                                Widget:= stack_box:
                                    Orientation:= orientation.Horizontal
                                    Slots:= array:
                                        stack_box_slot:
                                            Widget:= PlayerScore.RankText  
                                            HorizontalAlignment:= horizontal_alignment.Left   
                                            Padding:= margin{Left:=15.0, Top:=0.0, Bottom:= 0.0, Right:= 0.0}                                                             
                                        stack_box_slot:
                                            Widget:= PlayerScore.PlayerName  
                                            HorizontalAlignment:= horizontal_alignment.Left     
                                        stack_box_slot:
                                            Widget:=  PlayerScore.ScoreText
                                            HorizontalAlignment:= horizontal_alignment.Left
                                HorizontalAlignment:= horizontal_alignment.Left
                                VerticalAlignment := vertical_alignment.Top
                    VerticalAlignment := vertical_alignment.Top
                    HorizontalAlignment:= horizontal_alignment.Left
                    
    PlayerScore.Widget.SetVisibility(widget_visibility.Hidden)

    return PlayerScore 

player_leaderboard_slot := class: 
    var Agent : ?agent = false
    
    var PlayerName : text_block = text_block{}

    var ProgressBar : progress_bar = progress_bar{}

    var Score : big_number = big_number{}
    var ScoreText : text_block = text_block{}

    var ColorBlock : color_block =  color_block:
        DefaultColor:= NamedColors.White
        DefaultOpacity:= 0.0
        DefaultDesiredSize := vector2{X:=0.01, Y:= 0.00}        
    
    var Widget : widget = overlay{}

    var Rank : int = 0
    var RankText : text_block = text_block{}
    
    UpdatePosition():void=
        case (Rank):
            1 => ColorBlock.SetDesiredSize(vector2{Y:= (70.0) * 8})
            2 => ColorBlock.SetDesiredSize(vector2{Y:= (70.0) * 7})
            3 => ColorBlock.SetDesiredSize(vector2{Y:= (70.0) * 6})
            4 => ColorBlock.SetDesiredSize(vector2{Y:= (70.0) * 5})
            5 => ColorBlock.SetDesiredSize(vector2{Y:= (70.0) * 4})
            6 => ColorBlock.SetDesiredSize(vector2{Y:= (70.0) * 3})
            7 => ColorBlock.SetDesiredSize(vector2{Y:= (70.0) * 2})
            8 => ColorBlock.SetDesiredSize(vector2{Y:= (70.0)})
            _ => block:
                    ColorBlock.SetDesiredSize(vector2{Y:= 0.0})
                    Widget.SetVisibility(widget_visibility.Hidden)
                    ScoreText.SetText(STM(""))    
                    RankText.SetText(STM(""))    

player_leaderboard_manager := class:
    var PlayerScores : []player_leaderboard_slot = array{}
    var LargestScore : big_number = big_number{}

    var QueuedPlayers : []agent = array{}
    var QueuedScores : []big_number = array{}

    IsContainingAgent(Agent:agent)<transacts>:logic=
        for (PlayerScore : PlayerScores):
            if: 
                PlayerScore.Agent? = Agent
            then: 
                return true
        
        return false
                
    AddToLeaderboard(InPlayerScore : player_leaderboard_slot):void=
        set PlayerScores = PlayerScores + array{InPlayerScore}

    UpdateLeaderboardPositions():void=
        for (PlayerScore : PlayerScores):
            PlayerScore.UpdatePosition()

    QueueUpdateLeaderboard(Agent:agent, Score:big_number):void=
        if:
            Score.Value > 0.0
            QueuedPlayers.Length = 0
        then:
            spawn:
                StartUpdateLeaderboard()
        
        if:
            i:= QueuedPlayers.Find[Agent]
            NewQueuedScores := QueuedScores.ReplaceElement[i, Score]
            set QueuedScores = NewQueuedScores
        then:
            block:
        else:
            set QueuedPlayers = QueuedPlayers + array{Agent}
            set QueuedScores = QueuedScores + array{Score}
    
    StartUpdateLeaderboard()<suspends>:void=
        Sleep(0.01)

        for(i := 0..QueuedPlayers.Length -1):
            if:
                Player:= QueuedPlayers[i]
                Score:= QueuedScores[i]
            then:
                UpdateLeaderboard(Player, Score)

        set QueuedPlayers = array{}
        set QueuedScores = array{}
        
    UpdateLeaderboard(Agent:agent, Score:big_number):void=
        if:
            IsContainingAgent(Agent)?
        then:
            for (PlayerScore : PlayerScores):
                if: 
                    PlayerScore.Agent? = Agent
                then: 
                    set PlayerScore.Score = Score

                    PlayerScore.ScoreText.SetText(STM(": {FormatBigNumber(PlayerScore.Score)}"))
                    UpdateRankings(Agent, PlayerScore.Score)
                    PlayerScore.Widget.SetVisibility(widget_visibility.Visible)
        else:
            AddToLeaderboard(MakePlayerLeaderboard(Agent))
            UpdateLeaderboard(Agent, Score)

    ResetDamagePlayerScore():void=
        set LargestScore = big_number{}

        for (PlayerScore : PlayerScores):
            set PlayerScore.Score = big_number{}
            PlayerScore.ScoreText.SetText(STM(""))    
            PlayerScore.RankText.SetText(STM(""))    
            PlayerScore.ProgressBar.SetProgress(0.0, true)
            PlayerScore.Widget.SetVisibility(widget_visibility.Hidden)
            set PlayerScore.Rank = 0

    UpdateRankings(Agent:agent, Score:big_number):void=
        var TempLowestRank : int = 99 
        var OldRank : int = 99
        var RankedPlayersCount : int = 0
        
        for (PlayerScore : PlayerScores):
            if: 
                PlayerScore.Agent? = Agent
            then: 
                set OldRank = PlayerScore.Rank

            if:
                PlayerScore.Rank > 0
            then:
                set RankedPlayersCount += 1

                if:
                    BigNumber_GreaterThanOrEquals(Score, PlayerScore.Score)?
                    PlayerScore.Rank < TempLowestRank
                then:
                    set TempLowestRank = PlayerScore.Rank

        if(TempLowestRank = 99):
            set TempLowestRank = RankedPlayersCount + 1

        if:
            BigNumber_GreaterThanOrEquals(Score, LargestScore)?
        then:
            set LargestScore = Score

        for (PlayerScore : PlayerScores):
            PercentLargest := BigNumber_Percentage(PlayerScore.Score, LargestScore) * 0.95
            PlayerScore.ProgressBar.SetProgress(PercentLargest, true)
            
            if:
                PlayerScore.Rank >= TempLowestRank
                OldRank > TempLowestRank or OldRank = 0
                PlayerScore.Rank < OldRank or OldRank = 0
            then:
                set PlayerScore.Rank += 1
            if: 
                PlayerScore.Agent? = Agent
            then: 
                set PlayerScore.Rank = TempLowestRank