
using { /Fortnite.com/Devices }
using { /Verse.org/Simulation }
using { /UnrealEngine.com/Temporary/Diagnostics }
using { /Verse.org/Assets }
using { /UnrealEngine.com/Temporary/UI}
using { /Fortnite.com/UI }
using { /UnrealEngine.com/Temporary/SpatialMath }
using { /Verse.org/Colors }

upgrade_option := class:
    Key : string = ""

    GetDescription(PlayerManager : player_manager):string=

        var TempItemLevel : int = 0
        if:
            Temp := PlayerProfileDataMap[PlayerManager].UpgradeLevels[Key]
            set TempItemLevel = Temp

        if:
            Data := PlayerManager.GameManager.UpgradeDatas[Key]
            ValueTemp := "{FloatToTruncatedString(PlayerManager.GetUpgradeValue[Data.Name, false],Data.ValueDecimals)}{Data.ValueUnit} -> {FloatToTruncatedString(PlayerManager.GetUpgradeValue[Data.Name, option{TempItemLevel + 1}],Data.ValueDecimals)}{Data.ValueUnit}"
        then:
            return "{ValueTemp}<br><br>{Data.Description}"
            
        return ""

level_popup_device := class(menu_device_template):
    
    @editable
    GameManager : ?game_manager = false

    GetGameManager<override>():game_manager=
        if (GM := GameManager?):
            return GM

        return game_manager{}

    MakeMenu<override>(GM : game_manager, PM: player_manager):menu_template=
        return level_popup{GameManager:=option{GM}, Agent:=option{PM.Player}}  

    GetTitle<override>():string=
        return "LEVEL UP"

    GetIcon<override>():texture=
        return Images.Helm

level_popup := class(menu_template):
    DialogBackground : texture_block = texture_block{ DefaultImage := Images.DialogBox, DefaultDesiredSize := vector2{X := 1920.0, Y := 1080.0}} 
    var ChoiceDisplays : []upgrade_choice_display = array{}

    Title : text_block = text_block:
        DefaultTextColor:=NamedColors.White,
        DefaultShadowColor:=NamedColors.Black,
        DefaultShadowOffset:=option{vector2{X:=2.0,Y:=2.0}}

    LevelNum : text_block = text_block:
        DefaultTextColor:=NamedColors.White,
        DefaultShadowColor:=NamedColors.Black,
        DefaultShadowOffset:=option{vector2{X:=2.0,Y:=2.0}}

    CloseButton : button_quiet = button_quiet:
    

            
    UpdateMenu<override>(Widget:canvas):void=
        for (i := 0..ChoiceDisplays.Length - 1):
            if (ChoiceDisplay := ChoiceDisplays[i]):
                ChoiceDisplay.Update()

            
    CreateMenu<override>():canvas=
        Title.SetText(STM("LEVEL UP!"))
        LevelNum.SetText(STM("You reached level {GetPlayerManager[].GetLevel[] or 1}!"))

        Result := canvas:
            Slots := array:
                canvas_slot:
                    Widget:= DialogBackground
                    SizeToContent := true
                    Anchors := anchors{ Minimum := vector2{X := 0.1, Y := 0.1}, Maximum := vector2{X := 0.9, Y := 0.9}}  
                    ZOrder := 1
                canvas_slot:
                    Widget:= Title
                    Anchors := anchors{ Minimum := vector2{X := 0.5, Y := 0.14}, Maximum := vector2{X := 0.5, Y := 0.14}}
                    Alignment := vector2{X := 0.5, Y := 0.5}
                    ZOrder := 2
                canvas_slot:
                    Widget:= LevelNum
                    Anchors := anchors{ Minimum := vector2{X := 0.5, Y := 0.2}, Maximum := vector2{X := 0.5, Y := 0.2}}
                    Alignment := vector2{X := 0.5, Y := 0.5}
                    ZOrder := 3

        for (i := ChoiceDisplays.Length..2; rangeI := type{_X:int where 0 <= _X, _X <= 2147483647}[i+3]):

            GeneratedUpgrade := upgrade_option:
                Key := case (i):    
                    0 => "Focus Fire"
                    1 => "Momentum"
                    _ => "Handcannon"
                    

            Log("Create CD {i}")
            CD := upgrade_choice_display{Menu:=Self, Index:=i, Upgrade := GeneratedUpgrade}
            CD.Create()
            set ChoiceDisplays = ChoiceDisplays + array{CD}
            Result.AddWidget(
                canvas_slot:
                    Widget := CD.Widget
                    Anchors := anchors{ Minimum := vector2{X := 0.125 + i*0.25, Y := 0.2}, Maximum := vector2{X := 0.375 + 0.25*i, Y := 0.85}}  
                    Offsets := margin{Top := 20.0, Left := 20.0, Right := 20.0, Bottom := 20.0}
                    Alignment := vector2{X := 0.5, Y := 0.5}
                    ZOrder := rangeI
            )

        return Result

    ChooseItem(Index : int):void=
        if:
            PlayerManager := GetPlayerManager[]
        then:
            PlayerManager.UpdateQuestProgressForTask(task_event.LootChosen, big_number{Value := 1.0}, false, ?IsPersistent := true, ?Target := "{Index+1}")   
            if(UI := PlayerManager.UIManager?):
                UI.ClosePopup()

upgrade_choice_display := class():
    Menu : level_popup
    Index : int = 0
    var Upgrade : upgrade_option

    Widget : canvas = canvas:

    ItemBackground : color_block = color_block:
        DefaultColor:=NamedColors.White
        DefaultOpacity:=0.4

    IconDisplay : texture_block = texture_block:
        DefaultImage := Images.Empty
        DefaultDesiredSize := vector2{X := 300.0, Y := 300.0}
    NameDisplay : text_block = text_block:
        DefaultTextColor:=NamedColors.White
        DefaultShadowColor:=NamedColors.Black   
        DefaultShadowOffset:=option{vector2{X:=2.0,Y:=2.0}}
        DefaultShadowOpacity:= 1.0
    
    ChooseButton : button_loud = button_loud:
    var DescDisplay : stack_box = stack_box {Orientation:=orientation.Vertical}
    
    Create():void=

        Widget.AddWidget(
            canvas_slot:
                Widget := ItemBackground
                Anchors := anchors{ Minimum := vector2{X := 0.0, Y := 0.0}, Maximum := vector2{X := 1.0, Y := 1.0}}  
                Offsets := margin{Top := 0.0, Left := 0.0, Right := 0.0, Bottom := 0.0}
                Alignment := vector2{X := 0.5, Y := 0.5}
                ZOrder := 0
        )

        Widget.AddWidget(
            canvas_slot:
                Widget := NameDisplay
                Anchors := anchors{ Minimum := vector2{X := 0.5, Y := 0.01}, Maximum := vector2{X := 0.5, Y := 0.01}}  
                Offsets := margin{Top := 40.0, Left := 0.0, Right := 0.0, Bottom := 0.0}
                Alignment := vector2{X := 0.5, Y := 0.0}
                ZOrder := 2
        )

        Widget.AddWidget(
            canvas_slot:
                Widget := IconDisplay
                Anchors := anchors{ Minimum := vector2{X := 0.5, Y := 0.28}, Maximum := vector2{X := 0.5, Y := 0.28}}  
                Offsets := margin{Top := 50.0, Left := 0.0, Right := 0.0, Bottom := 0.0}
                Alignment := vector2{X := 0.5, Y := 0.5}
                SizeToContent := true
                ZOrder := 1
        )

        Widget.AddWidget(
            canvas_slot:
                Widget := ChooseButton
                Anchors := anchors{ Minimum := vector2{X := 0.5, Y := 0.91}, Maximum := vector2{X := 0.5, Y := 0.98}}  
                Offsets := margin{Top := 0.0, Left := 0.0, Right := 0.0, Bottom := 0.0}
                Alignment := vector2{X := 0.5, Y := 0.5}
                ZOrder := 4
        )
        
        ChooseButton.SetText(STM("Choose"))

        ChooseButton.OnClick().Subscribe(OnChoose)

    OnChoose(Wdata:widget_message):void=
        Menu.ChooseItem(Index)


    Update():void=
        if:
            PlayerManager := Menu.GetPlayerManager[]
        then:
            IconDisplay.SetImage(GetTextureFromString(Upgrade.Key))
            ItemBackground.SetColor(NamedColors.White)
            NameDisplay.SetText(STM("{Upgrade.Key}"))

            Widget.RemoveWidget(DescDisplay)
            set DescDisplay = MakeTextBox(Upgrade.GetDescription(PlayerManager)) 
            Widget.AddWidget(
                canvas_slot:
                    Widget := DescDisplay
                    Anchors := anchors{ Minimum := vector2{X := 0.1, Y := 0.48}, Maximum := vector2{X := 0.9, Y := 0.9}}  
                    Offsets := margin{Top := 20.0, Left := -30.0, Right := 0.0, Bottom := 0.0}
                    Alignment := vector2{X := 0.0, Y := 0.5}
                    ZOrder := 0
            )

            Widget.SetVisibility(widget_visibility.Visible) 