
using { /Fortnite.com/Devices }
using { /Verse.org/Simulation }
using { /UnrealEngine.com/Temporary/Diagnostics }
using { /Verse.org/Assets }
using { /UnrealEngine.com/Temporary/UI}
using { /Fortnite.com/UI }
using { /UnrealEngine.com/Temporary/SpatialMath }
using { /Verse.org/Colors }


loot_choice := class<concrete>:
    @editable
    Item1 : ?armor = false
    @editable
    Item2 : ?armor = false
    @editable
    Item3 : ?armor = false

    PopupFor(PlayerManager : player_manager):void=
        GM := PlayerManager.GameManager
        
        if:
            SetUIManager := PlayerManager.UIManager?
        then:
            SetUIManager.ClosePopup()

        LootPopupDevice := GM.LootPopupDevice
        LootPopupDevice.OpenMenu(PlayerManager.Player, ?Force := true)
        if (Menu := loot_popup[LootPopupDevice.GetOpenMenu[PlayerManager.Player]]):
            set Menu.GearChoices = array{}
            
            set Menu.Items = array{
            collectible_data{
            Name := "Essence Hourglass"
            Key := "EssenceHourglass"
            Count := big_number{Value:=1.0}
            }
            }
            
            if (I1 := Item1?):
                set Menu.GearChoices = Menu.GearChoices + array{collectible_data{
                    Name := "Gear"
                    Key := ""
                    Count := big_number{Value:=1.0, Exponent:=0}
                    ArmorData := option{I1.ToData()}
                }}
            if (I2 := Item2?):
                set Menu.GearChoices = Menu.GearChoices + array{collectible_data{
                    Name := "Gear"
                    Key := ""
                    Count := big_number{Value:=1.0, Exponent:=0}
                    ArmorData := option{I2.ToData()}
                }}
            if (I3 := Item3?):
                set Menu.GearChoices = Menu.GearChoices + array{collectible_data{
                    Name := "Gear"
                    Key := ""
                    Count := big_number{Value:=1.0, Exponent:=0}
                    ArmorData := option{I3.ToData()}
                }}
            Menu.Update()

loot_popup_device := class(menu_device_template):
    
    @editable
    GameManager : ?game_manager = false

    GetGameManager<override>():game_manager=
        if (GM := GameManager?):
            return GM

        return game_manager{}

    MakeMenu<override>(GM : game_manager, PM: player_manager):menu_template=
        return loot_popup{GameManager:=option{GM}, Agent:=option{PM.Player}}  

    GetTitle<override>():string=
        return "???"

    GetIcon<override>():texture=
        return Images.Helm

upgrade_popup_device := class(menu_device_template):
    
    @editable
    GameManager : ?game_manager = false

    GetGameManager<override>():game_manager=
        if (GM := GameManager?):
            return GM

        return game_manager{}

    MakeMenu<override>(GM : game_manager, PM: player_manager):menu_template=
        return upgrade_level_popup{GameManager:=option{GM}, Agent:=option{PM.Player}}  

    GetTitle<override>():string=
        return "???"

    GetIcon<override>():texture=
        return Images.Helm

upgrade_level_popup := class(menu_template):
    var DialogBackground : texture_block = texture_block{ DefaultImage := Images.DialogBox, DefaultDesiredSize := vector2{X := 1920.0, Y := 1080.0}} 

    var Title : text_block = text_block:
        DefaultTextColor:=NamedColors.White,
        DefaultShadowColor:=NamedColors.Black,
        DefaultShadowOffset:=option{vector2{X:=2.0,Y:=2.0}}

    var Title2 : text_block = text_block:
        DefaultTextColor:=NamedColors.White,
        DefaultShadowColor:=NamedColors.Black,
        DefaultShadowOffset:=option{vector2{X:=2.0,Y:=2.0}}

    var LeaderboardTitle : text_block = text_block:
        DefaultTextColor:=NamedColors.White,
        DefaultShadowColor:=NamedColors.Black,
        DefaultShadowOffset:=option{vector2{X:=2.0,Y:=2.0}}

    var Icon : texture_block = texture_block:
        DefaultImage := Images.Empty
        DefaultDesiredSize := vector2{X := 300.0, Y := 300.0}

    var Description : stack_box = stack_box{Orientation:=orientation.Vertical}

    var CloseButton : button_loud = button_loud:

    var GearChoices: []collectible_data = array{}
    var StartLevel : int = 0
    var EndLevel : int = 0
    var BossType : boss_enum = boss_enum.Sharky

    var Leaderboard : player_leaderboard_manager = player_leaderboard_manager{}

    UpdateMenu<override>(Widget:canvas):void=
        if:
            GM := GameManager?
            SetAgent := Agent?
            PlayerManager := GM.PlayerManagers[SetAgent]
            Widget = _Widget?
        then:

            #IF NEW RECORD
            Title.SetText(STM("New {BossType.GetBossName()} Record: Level {EndLevel}!"))
            
            #IF CURRENT

            Title2.SetText(STM("Previous Best: Level {StartLevel}!"))
            Title2.SetTextColor(MakeColorFromHex("CC362F"))

            T1 := PlayerManager.GetUpgradeValueSafe(BossType.RewardStatName(), option{StartLevel})
            T2 := PlayerManager.GetUpgradeValueSafe(BossType.RewardStatName(), option{EndLevel})

            Val1 := MakeBigNumber(T1, 1) + BossType.RewardStatBaseValue()
            Val2 := MakeBigNumber(T2, 1) + BossType.RewardStatBaseValue()
            # Val1 := "TEST"
            # Val2 := "TEST2"

            Widget.RemoveWidget(Description)
            case(BossType):
                boss_enum.Sharky => set Description = MakeTextBox("Passive Income Increased! Now {Val2} Essence per minute")
                boss_enum.Genga => set Description = MakeTextBox("Max Health Increased! Now {Val2} HP") 
                boss_enum.Urjent => set Description = MakeTextBox("Weakpoint Damage Increased! Now {Val2}%")
                boss_enum.PirateKing => set Description = MakeTextBox("Damage Mult Increased! Now {Val2}%")
                _ => set Description = MakeTextBox("---")
            

            Widget.AddWidget(
                canvas_slot:
                    Widget:= Description
                    Anchors := anchors{ Minimum := vector2{X := 0.5, Y := 0.2}, Maximum := vector2{X := 0.5, Y := 0.2}}
                    Offsets := margin{Top := 80.0, Left := 0.0, Right := 0.0, Bottom := 0.0}
                    Alignment := vector2{X := 0.5, Y := 0.5}
                    ZOrder := 4
            )

            # LeaderboardWidget : overlay = overlay{}
            # #Leaderboard
            # for (InPlayerManager : GM.PlayerManagers):
            #     var BestTier : int = 0
            #     if:
            #         BestAchieve := InPlayerManager.GetAchievementLevel[BossType.AchievementName()]
            #         BestAchieve > 0
            #     then:
            #         set BestTier = BestAchieve-1

            #     Leaderboard.UpdateLeaderboard(InPlayerManager.Agent, big_number{Value:=BestTier*1.0})    

            # for (Slot : Leaderboard.PlayerScores):
            #     TempSlot:= overlay_slot:
            #         Widget := Slot.Widget

            #     LeaderboardWidget.AddWidget(TempSlot)

            # LeaderboardTitle.SetText(STM("{BossType.GetBossName()} Level Leaderboard"))

            # Widget.AddWidget(
            #     canvas_slot:
            #         Widget:= LeaderboardTitle
            #         Anchors := anchors{ Minimum := vector2{X := 0.5, Y := 0.2}, Maximum := vector2{X := 0.5, Y := 0.2}}
            #         Offsets := margin{Top := 190.0, Left := 0.0, Right := 0.0, Bottom := 0.0}
            #         Alignment := vector2{X := 0.5, Y := 0.5}
            #         ZOrder := 3
            # )

            # Widget.AddWidget(
            #     canvas_slot:
            #         Widget:= LeaderboardWidget
            #         Anchors := anchors{ Minimum := vector2{X := 0.5, Y := 0.2}, Maximum := vector2{X := 0.5, Y := 0.2}}
            #         Offsets := margin{Top := 670.0, Left := 0.0, Right := 0.0, Bottom := 0.0}
            #         Alignment := vector2{X := 0.5, Y := 0.0}
            #         ZOrder := 3
            # )
        


            
    CreateMenu<override>():canvas=
        Log("Creating Upgrade Popup")
        Result := canvas:
            Slots := array:
                canvas_slot:
                    Widget:= DialogBackground
                    SizeToContent := true
                    Anchors := anchors{ Minimum := vector2{X := 0.1, Y := 0.1}, Maximum := vector2{X := 0.9, Y := 0.9}}  
                    ZOrder := 1
                canvas_slot:
                    Widget:= Title
                    Anchors := anchors{ Minimum := vector2{X := 0.5, Y := 0.2}, Maximum := vector2{X := 0.5, Y := 0.2}}
                    Alignment := vector2{X := 0.5, Y := 0.5}
                    ZOrder := 2
                canvas_slot:
                    Widget:= Title2
                    Anchors := anchors{ Minimum := vector2{X := 0.5, Y := 0.2}, Maximum := vector2{X := 0.5, Y := 0.2}}
                    Offsets := margin{Top := 30.0, Left := 0.0, Right := 0.0, Bottom := 0.0}
                    Alignment := vector2{X := 0.5, Y := 0.5}
                    ZOrder := 3
                canvas_slot:
                    Widget:= Icon
                    Anchors := anchors{ Minimum := vector2{X := 0.5, Y := 0.5}, Maximum := vector2{X := 0.5, Y := 0.5}}
                    Alignment := vector2{X := 0.5, Y := 0.5}
                    ZOrder := 4
                canvas_slot:
                    Widget:= Description
                    Anchors := anchors{ Minimum := vector2{X := 0.5, Y := 0.65}, Maximum := vector2{X := 0.5, Y := 0.65}}
                    Alignment := vector2{X := 0.5, Y := 0.0}
                    ZOrder := 5
                canvas_slot:
                    Widget:= CloseButton
                    Anchors := anchors{ Minimum := vector2{X := 0.5, Y := 0.85}, Maximum := vector2{X := 0.5, Y := 0.85}}
                    Alignment := vector2{X := 0.5, Y := 1.0}
                    ZOrder := 6
        

        CloseButton.SetText(STM("                              Collect Loot!                              "))            
        CloseButton.OnClick().Subscribe(NextPopupWData)

        return Result

    NextPopupWData(Wdata:widget_message):void=
        NextPopup()

    NextPopup():void=
        if:
            GM := GameManager?
            SetAgent := Agent?
            PlayerManager := GM.PlayerManagers[SetAgent]
        then:
            PlayerManager.ClosePopupUI() 
            GM.LootPopupDevice.OpenMenu(PlayerManager.Player, ?Force := true)
            
            if (Menu := loot_popup[GM.LootPopupDevice.GetOpenMenu[PlayerManager.Player]]):
                Log("Got Loot Popup Open. Gear Count: {GearChoices.Length}")
                set Menu.GearChoices = GearChoices
                set Menu.Items = array{collectible_data{
                    Name := "Essence Hourglass"
                    Key := "EssenceHourglass"
                    Count := big_number{Value:=1.0}
                    }}
                Menu.Update()

loot_popup := class(menu_template):
    var DialogBackground : texture_block = texture_block{ DefaultImage := Images.DialogBox, DefaultDesiredSize := vector2{X := 1920.0, Y := 1080.0}} 
    var ItemDisplays : []list_item_display = array{}
    var ChoiceDisplays : []item_choice_display = array{}

    var Title : text_block = text_block:
        DefaultTextColor:=NamedColors.White,
        DefaultShadowColor:=NamedColors.Black,
        DefaultShadowOffset:=option{vector2{X:=2.0,Y:=2.0}}

    var CloseButton : button_quiet = button_quiet:

    var Items: []collectible_data = array{}
    var GearChoices: []collectible_data = array{}

            
    UpdateMenu<override>(Widget:canvas):void=
        for (i := 0..ItemDisplays.Length - 1):
            if (ItemDisplay := ItemDisplays[i]):

                if:
                    ToShow := Items[i]
                then:
                    ItemDisplay.Update(option{ToShow}, false)
                else:
                    ItemDisplay.Update(false, false)

        for (i := 0..ChoiceDisplays.Length - 1):
            Log("Update CD {i}")
            if (ChoiceDisplay := ChoiceDisplays[i]; Item := GearChoices[i]):
                set ChoiceDisplay.Item = option{Item}
                ChoiceDisplay.Update()

            
    CreateMenu<override>():canvas=
        ItemListBox := stack_box{Orientation:=orientation.Vertical}

        Title.SetText(STM("Choose your loot!"))

        Result := canvas:
            Slots := array:
                canvas_slot:
                    Widget:= DialogBackground
                    SizeToContent := true
                    Anchors := anchors{ Minimum := vector2{X := 0.1, Y := 0.1}, Maximum := vector2{X := 0.9, Y := 0.9}}  
                    ZOrder := 1
                # canvas_slot:
                #     Widget := ItemListBox
                #     Anchors := anchors{ Minimum := vector2{X := 0.1, Y := 0.2}, Maximum := vector2{X := 0.6, Y := 0.6}}  
                #     Offsets := margin{Top := 50.0, Left := 50.0, Right := 50.0, Bottom := 50.0}
                #     Alignment := vector2{X := 0.0, Y := 0.0}
                #     #SizeToContent := true
                #     ZOrder := 2
                canvas_slot:
                    Widget:= Title
                    Anchors := anchors{ Minimum := vector2{X := 0.5, Y := 0.14}, Maximum := vector2{X := 0.5, Y := 0.14}}
                    Alignment := vector2{X := 0.5, Y := 0.5}
                    ZOrder := 4
                # canvas_slot:
                #     Widget:= CloseButton
                #     Anchors := anchors{ Minimum := vector2{X := 0.5, Y := 0.915}, Maximum := vector2{X := 0.5, Y := 0.915}}
                #     Alignment := vector2{X := 0.5, Y := 1.0}
                #     ZOrder := 5
            
        
        for (i := 0..6):
            set ItemDisplays = ItemDisplays + array{list_item_display{Menu:=Self}}
        
        for (ItemDisplay : ItemDisplays):
            ItemDisplay.Create()
            ItemListBox.AddWidget(
                stack_box_slot:
                    Widget := ItemDisplay.Widget
                    HorizontalAlignment := horizontal_alignment.Fill
                    Padding:= margin{Top := 0.0, Left := 0.0, Right := 0.0, Bottom := -20.0}
            )

        for (i := ChoiceDisplays.Length..2; rangeI := type{_X:int where 0 <= _X, _X <= 2147483647}[i+6]):

            Log("Create CD {i}")
            CD := item_choice_display{Menu:=Self, Index:=i}
            CD.Create()
            set ChoiceDisplays = ChoiceDisplays + array{CD}
            Result.AddWidget(
                canvas_slot:
                    Widget := CD.Widget
                    Anchors := anchors{ Minimum := vector2{X := 0.125 + i*0.25, Y := 0.2}, Maximum := vector2{X := 0.375 + 0.25*i, Y := 0.85}}  
                    Offsets := margin{Top := 20.0, Left := 20.0, Right := 20.0, Bottom := 20.0}
                    Alignment := vector2{X := 0.5, Y := 0.5}
                    ZOrder := rangeI
            )

        return Result

    ChooseItem(Index : int):void=
        if:
            PlayerManager := GetPlayerManager[]
            Item := GearChoices[Index]
        then:
            Log("Choosing {Item.Key}")
            UID := PlayerManager.GainCollectible(Item)

            if:
                Log("Choosing A")
                Armor := Item.ArmorData?.WithUID(UID).ToArmor()
                Log("Choosing B")
                EquippedArmor := PlayerManager.GetArmorInSlot[Armor.Slot]
                Log("Choosing C")
                EqLevel := EquippedArmor.GetTotalLevel()
            then:
                if:
                    Log("Choosing D")
                    EqLevel < Armor.GetTotalLevel()
                then:
                    Log("Choosing E")
                    PlayerManager.EquipArmor(Armor)
            else:
                if:
                    Log("Choosing F")
                    Armor := Item.ArmorData?.WithUID(UID).ToArmor()
                then:
                    Log("Choosing G")
                    PlayerManager.EquipArmor(Armor)
            
            PlayerManager.UpdateQuestProgressForTask(task_event.LootChosen, big_number{Value := 1.0}, false, ?IsPersistent := true, ?Target := "{Index+1}")   
            if(UI := PlayerManager.UIManager?):
                UI.ClosePopup()

item_choice_display := class():
    Menu : loot_popup
    Index : int = 0
    var Item: ?collectible_data = false

    Widget : canvas = canvas:

    ItemBackground : color_block = color_block:
        DefaultColor:=NamedColors.White
        DefaultOpacity:=0.4

    IconDisplay : texture_block = texture_block:
        DefaultImage := Images.Empty
        DefaultDesiredSize := vector2{X := 300.0, Y := 300.0}
    NameDisplay : text_block = text_block:
        DefaultTextColor:=NamedColors.White
        DefaultShadowColor:=NamedColors.Black   
        DefaultShadowOffset:=option{vector2{X:=2.0,Y:=2.0}}
        DefaultShadowOpacity:= 1.0

    RarityDisplay : text_block = text_block:
        DefaultTextColor:=NamedColors.White
        DefaultShadowColor:=NamedColors.Black   
        DefaultShadowOffset:=option{vector2{X:=2.0,Y:=2.0}}
        DefaultShadowOpacity:= 1.0
    
    ChooseButton : button_loud = button_loud:
    var DescDisplay : stack_box = stack_box {Orientation:=orientation.Vertical}

        
    SellValueStackBox : stack_box = stack_box:
        Orientation := orientation.Horizontal
    SellValue : text_block = text_block:
        DefaultTextColor:=NamedColors.Gold
        DefaultShadowColor:=NamedColors.Black   
        DefaultShadowOffset:=option{vector2{X:=2.0,Y:=2.0}}
        DefaultShadowOpacity:= 1.0
    SellIcon : texture_block = texture_block:
        DefaultImage := Images.MidasCoins
        DefaultDesiredSize := vector2{X := 40.0, Y := 40.0}
            
    EquippedIndicator : text_block = text_block:
        DefaultTextColor:=NamedColors.White
        DefaultShadowColor:=NamedColors.Black   
        DefaultShadowOffset:=option{vector2{X:=2.0,Y:=2.0}}
        DefaultShadowOpacity:= 1.0
    
    Create():void=

        Widget.AddWidget(
            canvas_slot:
                Widget := ItemBackground
                Anchors := anchors{ Minimum := vector2{X := 0.0, Y := 0.0}, Maximum := vector2{X := 1.0, Y := 1.0}}  
                Offsets := margin{Top := 0.0, Left := 0.0, Right := 0.0, Bottom := 0.0}
                Alignment := vector2{X := 0.5, Y := 0.5}
                ZOrder := 0
        )

        Widget.AddWidget(
            canvas_slot:
                Widget := NameDisplay
                Anchors := anchors{ Minimum := vector2{X := 0.5, Y := 0.01}, Maximum := vector2{X := 0.5, Y := 0.01}}  
                Offsets := margin{Top := 40.0, Left := 0.0, Right := 0.0, Bottom := 0.0}
                Alignment := vector2{X := 0.5, Y := 0.0}
                ZOrder := 2
        )

        Widget.AddWidget(
            canvas_slot:
                Widget := RarityDisplay
                Anchors := anchors{ Minimum := vector2{X := 0.5, Y := 0.01}, Maximum := vector2{X := 0.5, Y := 0.01}}  
                Offsets := margin{Top := 75.0, Left := 0.0, Right := 0.0, Bottom := 0.0}
                Alignment := vector2{X := 0.5, Y := 0.0}
                ZOrder := 2
        )

        Widget.AddWidget(
            canvas_slot:
                Widget := IconDisplay
                Anchors := anchors{ Minimum := vector2{X := 0.5, Y := 0.28}, Maximum := vector2{X := 0.5, Y := 0.28}}  
                Offsets := margin{Top := 50.0, Left := 0.0, Right := 0.0, Bottom := 0.0}
                Alignment := vector2{X := 0.5, Y := 0.5}
                SizeToContent := true
                ZOrder := 1
        )

        # Widget.AddWidget(
        #     canvas_slot:
        #         Widget := SellValueStackBox
        #         Anchors := anchors{ Minimum := vector2{X := 0.5, Y := 0.87}, Maximum := vector2{X := 0.5, Y := 0.93}}  
        #         Offsets := margin{Top := 0.0, Left := 0.0, Right := 0.0, Bottom := 0.0}
        #         Alignment := vector2{X := 0.5, Y := 0.5}
        #         ZOrder := 3
        # )

        Widget.AddWidget(
            canvas_slot:
                Widget := ChooseButton
                Anchors := anchors{ Minimum := vector2{X := 0.5, Y := 0.91}, Maximum := vector2{X := 0.5, Y := 0.98}}  
                Offsets := margin{Top := 0.0, Left := 0.0, Right := 0.0, Bottom := 0.0}
                Alignment := vector2{X := 0.5, Y := 0.5}
                ZOrder := 4
        )

        Widget.AddWidget(
            canvas_slot:
                Widget := EquippedIndicator
                Anchors := anchors{ Minimum := vector2{X := 0.0, Y := 0.0}, Maximum := vector2{X := 0.0, Y := 0.0}}  
                Offsets := margin{Top := 20.0, Left := 0.0, Right := 0.0, Bottom := 0.0}
                Alignment := vector2{X := 0.0, Y := 0.5}
                ZOrder := 5
        )
        

        SellValueStackBox.AddWidget(
            stack_box_slot:
                Widget := SellIcon
                HorizontalAlignment:= horizontal_alignment.Left
                Padding:= margin{Left:=2.0, Top:=0.0, Bottom:= 0.0, Right:= 2.0} 
        )
        SellValueStackBox.AddWidget(
            stack_box_slot:
                Widget := SellValue
                HorizontalAlignment:= horizontal_alignment.Right
                Padding:= margin{Left:=2.0, Top:=0.0, Bottom:= 0.0, Right:= 2.0} 
        )

        ChooseButton.SetText(STM("Equip"))

        ChooseButton.OnClick().Subscribe(OnChoose)

    OnChoose(Wdata:widget_message):void=
        Menu.ChooseItem(Index)


    Update():void=
        if:
            It := Item?
            PlayerManager := Menu.GetPlayerManager[]
        then:
            IconDisplay.SetImage(It.GetImage())
            if (Armor := It.ArmorData?.ToArmor()):
                ItemBackground.SetColor(Armor.Rarity.ToColor())
                RarityDisplay.SetText(STM("Level {Armor.GetTotalLevel()} {Armor.Rarity}"))
                RarityDisplay.SetTextColor(Armor.Rarity.ToColor())
                NameDisplay.SetText(STM("{Armor.GetName()}"))
                
                if (EquippedArmor := PlayerManager.GetArmorInSlot[Armor.Slot]):
                    EqLevel := EquippedArmor.GetTotalLevel()
                    CompareTxt := "Currently Equipped: Lv{EqLevel}"

                    EquippedIndicator.SetVisibility(widget_visibility.Visible)
                    EquippedIndicator.SetTextColor(EquippedArmor.Rarity.ToColor())
                    EquippedIndicator.SetText(STM(CompareTxt))
                    
                    if:
                        Armor.GetTotalLevel() <= EqLevel
                    then:
                        ChooseButton.SetText(STM("Take"))
                    else:
                        ChooseButton.SetText(STM("Equip"))

                else:
                    ChooseButton.SetText(STM("Equip"))
                    EquippedIndicator.SetTextColor(MakeColorFromHex("CC362F"))
                    EquippedIndicator.SetText(STM("Currently Equipped: None"))


            Widget.RemoveWidget(DescDisplay)
            set DescDisplay = MakeTextBox(It.GetDescription()) 
            Widget.AddWidget(
                canvas_slot:
                    Widget := DescDisplay
                    Anchors := anchors{ Minimum := vector2{X := 0.1, Y := 0.48}, Maximum := vector2{X := 0.9, Y := 0.9}}  
                    Offsets := margin{Top := 20.0, Left := -30.0, Right := 0.0, Bottom := 0.0}
                    Alignment := vector2{X := 0.0, Y := 0.5}
                    ZOrder := 0
            )

            Widget.SetVisibility(widget_visibility.Visible) 
            Value := It.GetSellValue()
            if (BigNumber_GreaterThanOrEquals(Value, big_number{Value:=1.0})?):
                SellValue.SetText(STM("{Value}"))