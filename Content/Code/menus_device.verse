
using { /Fortnite.com/Devices }
using { /Verse.org/Simulation }
using { /UnrealEngine.com/Temporary/Diagnostics }
using { /UnrealEngine.com/Temporary/UI }
using { /Verse.org/Assets }
using { /Fortnite.com/UI }
using { /UnrealEngine.com/Temporary/SpatialMath }
using { /Verse.org/Colors }

menu_item := interface(menu_openable):
    GetTitle():string
    GetIcon():texture
    IsEnabled(PlayerManager : player_manager)<transacts>:logic

menu_openable := interface():
    OpenMenu(Agent:agent):void

menus_device:= class(creative_device):
    @editable
    GameManager : ?game_manager = false
    @editable
    SignalDevice : signal_remote_manager_device = signal_remote_manager_device{}
    @editable
    PrimaryMenu : creative_device = gun_manager{}
    @editable
    SecondaryMenu : creative_device = inventory_device{}

    OnBegin<override>()<suspends>:void=
        SignalDevice.PrimarySignalEvent.Subscribe(OnPrimary)
        SignalDevice.SecondarySignalEvent.Subscribe(OnSecondary)
    
    OnPrimary(Agent:agent):void=
        if (PlayerManager := GameManager?.PlayerManagers[Agent]):        
            if:
                PlayerManager.HasCompletedTask(0, 6)?
            then:
                if:
                    PlayerManager.HasCompletedTask(0, 11)?
                then:
                    if:
                        PlayerManager.CurrentZone.IsBossTag()? or PlayerManager.CurrentZone = fetchable_tag.prestigeareatag
                    then:
                        if (Menu := menu_openable[PrimaryMenu]):
                            Menu.OpenMenu(Agent)
                    else:
                        if (Menu := menu_openable[SecondaryMenu]):
                            Menu.OpenMenu(Agent)
                else:
                    if (Menu := menu_openable[PrimaryMenu]):
                        Menu.OpenMenu(Agent)
            else:    
                if (PlayerManager.HasEmptySlot[]):
                    PlayerManager.MakeRoomForGun()
                    PlayerManager.EquipItem(1, ?Equip := true)
                    PlayerManager.UpdateQuestProgressForTask(task_event.EquipItem, big_number{Value := 1.0}, false, ?Target := "{1}", ?IsPersistent := true)
                    
    OnSecondary(Agent:agent):void=
        if (PlayerManager := GameManager?.PlayerManagers[Agent]):        
            
            if (Menu := menu_openable[PrimaryMenu]):
                Menu.OpenMenu(Agent)
                
list_menu := class(creative_device, menu_item):
    @editable
    GameManager : ?game_manager = false
    @editable
    Items : []creative_device = array{}
    @editable
    Title : string = "LIST"
    @editable
    IconName : string = "EMPTY"


    OpenMenu<override>(Agent:agent):void=
        if (PlayerManager := GameManager?.PlayerManagers[Agent]):
            
            if (not GameManager?.PlayerManagers[Agent].UIManager?.Popup?):
                if:
                    GM := GameManager?
                then:
                    var Items_Casted : []menu_item = array{}
                    for (Item : Items):
                        if (Item_Casted := menu_item[Item]): 
                            set Items_Casted = Items_Casted + array{Item_Casted}
                        else:
                            Log("Item in list menu is not a menu_item")
                    MakeListMenu(GM, Items_Casted, Agent).OpenMenu()
            else:
                block:#Log("DOUBLE MENU")
    
    GetTitle<override>():string=
        return Title

    GetIcon<override>():texture=
        return GetTextureFromString(IconName)

    IsEnabled<override>(PlayerManager : player_manager)<transacts>:logic=
        return true


MakeListMenu<constructor>(GameManagerIn : game_manager, ItemsIn : []menu_item, AgentIn : agent) := list_menu_instance:
    GameManager :=  GameManagerIn
    Items := ItemsIn
    Agent := AgentIn

ranking_player_stat := class: 
    var Name : string = ""
    var Score : big_number = big_number{}
    var Rank : int = 0

list_menu_instance:= class:
    GameManager : game_manager
    Items : []menu_item
    Agent : agent

    var DialogBackground : texture_block = texture_block{ DefaultImage := Images.StoreBox, DefaultDesiredSize := vector2{X := 1800.0, Y := 500.0}}
    var StatsBackground : texture_block = texture_block{ DefaultImage := Images.StoreBox, DefaultDesiredSize := vector2{X := 200.0, Y := 500.0}}
    
    var EnemyStats : []ranking_player_stat = array{}
    var WeaponStats : []ranking_player_stat = array{}

    OpenMenu():void=
        if:
            InPlayer := player[Agent]
            InPlayer.IsActive[]
            PlayerUI := GetPlayerUI[InPlayer]
            PlayerManager := GameManager.PlayerManagers[InPlayer]
        then:
            StatsStackBox := stack_box{Orientation:=orientation.Vertical}

            StackBoxOuter := stack_box{Orientation:=orientation.Vertical}
            var CurrentStackBox : stack_box = stack_box{Orientation:=orientation.Horizontal}
            StackBoxOuter.AddWidget(stack_box_slot:
                Widget:= CurrentStackBox
                Padding:= margin{Top := 0.0, Left := 0.0, Right := 0.0, Bottom := 00.0}
            )

            CloseButton := button_quiet:
            CloseButton.SetText(STM("                                                BACK                                                "))            
            CloseButton.OnClick().Subscribe(PlayerManager.ClosePopup)

            Canvas := canvas:
                Slots := array:
                    canvas_slot:
                        Widget:= StatsBackground
                        SizeToContent := true
                        Anchors := anchors{ Minimum := vector2{X := 0.0, Y := 0.1}, Maximum := vector2{X := 0.2, Y := 0.9}}  
                        Offsets:= margin{Left:=0.0, Top:= 0.0, Bottom:= 0.0, Right:= 0.0} 
                        ZOrder := 1
                    canvas_slot:
                        Widget:= StatsStackBox
                        SizeToContent := true
                        Anchors := anchors{ Minimum := vector2{X := 0.0, Y := 0.1}, Maximum := vector2{X := 0.2, Y := 0.9}}  
                        Offsets:= margin{Left:=0.0, Top:= 0.0, Bottom:= 0.0, Right:= 0.0} 
                        ZOrder := 2
                    canvas_slot:
                        Widget:= DialogBackground
                        SizeToContent := true
                        Anchors := anchors{ Minimum := vector2{X := 0.2, Y := 0.1}, Maximum := vector2{X := 1.0, Y := 0.9}}  
                        ZOrder := 1
                    canvas_slot:
                        Widget:= StackBoxOuter
                        SizeToContent := true
                        Anchors := anchors{ Minimum := vector2{X := 0.5, Y := 0.15}, Maximum := vector2{X := 0.5, Y := 0.7}}  
                        Alignment := vector2{X := 0.5, Y := 0.0}
                        Offsets:= margin{Left:=150.0, Top:= -50.0, Bottom:= 0.0, Right:= 0.0} 
                        ZOrder := 2
                    canvas_slot:
                        Widget:= CloseButton
                        Anchors := anchors{ Minimum := vector2{X := 0.5, Y := 0.855}, Maximum := vector2{X := 0.5, Y := 0.855}}
                        Alignment := vector2{X := 0.5, Y := 1.0}
                        Offsets:= margin{Left:=150.0, Top:= 0.0, Bottom:= 0.0, Right:= 0.0} 
                        ZOrder := 3

            StatsStackBox.AddWidget(stack_box_slot:
                Widget:= stack_box:
                    Orientation := orientation.Horizontal
                    Slots := array:
                        stack_box_slot:
                            Widget:= text_block{DefaultText:= STM("Player Stats"), DefaultShadowOpacity := 1.0, DefaultTextColor:= NamedColors.Goldenrod, DefaultShadowColor:=NamedColors.Black, DefaultShadowOffset:=option{vector2{X:=2.0,Y:=2.0}}}
                Padding:= margin{Top := 25.0, Left := 0.0, Right := 0.0, Bottom := 0.0}
            )

            StatsStackBox.AddWidget(stack_box_slot:
                Widget:= stack_box:
                    Orientation := orientation.Horizontal
                    Slots := array:
                        stack_box_slot:
                            Widget:= text_block{DefaultText:= STM("Health: {PlayerManager.GetPlayerStat(player_stat.MaxHealth)}"), DefaultShadowOpacity := 1.0, DefaultTextColor:= NamedColors.White, DefaultShadowColor:=NamedColors.Black, DefaultShadowOffset:=option{vector2{X:=2.0,Y:=2.0}}}
                Padding:= margin{Top := 10.0, Left := 0.0, Right := 0.0, Bottom := 0.0}
            )

            StatsStackBox.AddWidget(stack_box_slot:
                Widget:= stack_box:
                    Orientation := orientation.Horizontal
                    Slots := array:
                        stack_box_slot:
                            Widget:= text_block{DefaultText:= STM("Regen: {PlayerManager.GetPlayerStat(player_stat.Regen)}"), DefaultShadowOpacity := 1.0, DefaultTextColor:= NamedColors.White, DefaultShadowColor:=NamedColors.Black, DefaultShadowOffset:=option{vector2{X:=2.0,Y:=2.0}}}
                Padding:= margin{Top := 10.0, Left := 0.0, Right := 0.0, Bottom := 0.0}
            )

            StatsStackBox.AddWidget(stack_box_slot:
                Widget:= stack_box:
                    Orientation := orientation.Horizontal
                    Slots := array:
                        stack_box_slot:
                            Widget:= text_block{DefaultText:= STM("Dmg Mult: {PlayerManager.GetPlayerStat(player_stat.DamageMult) * 100.0}%"), DefaultShadowOpacity := 1.0, DefaultTextColor:= NamedColors.Goldenrod, DefaultShadowColor:=NamedColors.Black, DefaultShadowOffset:=option{vector2{X:=2.0,Y:=2.0}}}
                Padding:= margin{Top := 30.0, Left := 0.0, Right := 0.0, Bottom := 00.0}
            )

            var Strength : big_number = big_number{}
            if:
                Temp := PlayerManager.GetDisplayStrength[]
                set Strength = Temp
            StatsStackBox.AddWidget(stack_box_slot:
                Widget:= stack_box:
                    Orientation := orientation.Horizontal
                    Slots := array:
                        stack_box_slot:
                            Widget:= text_block{DefaultText:= STM("Strength: {Strength}"), DefaultShadowOpacity := 1.0, DefaultTextColor:= NamedColors.Goldenrod, DefaultShadowColor:=NamedColors.Black, DefaultShadowOffset:=option{vector2{X:=2.0,Y:=2.0}}}
                Padding:= margin{Top := 5.0, Left := 0.0, Right := 0.0, Bottom := 00.0}
            )
            
            block:
                Temp1 := ranking_player_stat{Name:="VsBoss", Score:=PlayerManager.GetPlayerStat(player_stat.Damage_Boss) * 100.0, Rank:=0}
                set EnemyStats = EnemyStats + array{Temp1}
                UpdateRankings(Temp1.Name, Temp1.Score, EnemyStats)
                
                Temp2 := ranking_player_stat{Name:="VsGuard", Score:=PlayerManager.GetPlayerStat(player_stat.Damage_Guard) * 100.0, Rank:=0}
                set EnemyStats = EnemyStats + array{Temp2}
                UpdateRankings(Temp2.Name, Temp2.Score, EnemyStats)
                
                Temp3 := ranking_player_stat{Name:="VsMineral", Score:=PlayerManager.GetPlayerStat(player_stat.Damage_Mineral) * 100.0, Rank:=0}
                set EnemyStats = EnemyStats + array{Temp3}
                UpdateRankings(Temp3.Name, Temp3.Score, EnemyStats)

                Temp4 := ranking_player_stat{Name:="VsBeast", Score:=PlayerManager.GetPlayerStat(player_stat.Damage_Beast) * 100.0, Rank:=0}
                set EnemyStats = EnemyStats + array{Temp4}
                UpdateRankings(Temp4.Name, Temp4.Score, EnemyStats)

                for (Stat : EnemyStats):
                    if:
                        Stat.Rank = 1
                    then:
                        StatsStackBox.AddWidget(stack_box_slot:
                            Widget:= stack_box:
                                Orientation := orientation.Horizontal
                                Slots := array:
                                    stack_box_slot:
                                        Widget:= text_block{DefaultText:= STM("{Stat.Name}: {Stat.Score}%"), DefaultShadowOpacity := 1.0, DefaultTextColor:= NamedColors.White, DefaultShadowColor:=NamedColors.Black, DefaultShadowOffset:=option{vector2{X:=2.0,Y:=2.0}}}
                            Padding:= margin{Top := 30.0, Left := 0.0, Right := 0.0, Bottom := 00.0}
                        )

                for (Stat : EnemyStats):
                    if:
                        Stat.Rank = 2
                    then:
                        StatsStackBox.AddWidget(stack_box_slot:
                            Widget:= stack_box:
                                Orientation := orientation.Horizontal
                                Slots := array:
                                    stack_box_slot:
                                        Widget:= text_block{DefaultText:= STM("{Stat.Name}: {Stat.Score}%"), DefaultShadowOpacity := 1.0, DefaultTextColor:= NamedColors.White, DefaultShadowColor:=NamedColors.Black, DefaultShadowOffset:=option{vector2{X:=2.0,Y:=2.0}}}
                            Padding:= margin{Top := 5.0, Left := 0.0, Right := 0.0, Bottom := 00.0}
                        )

                for (Stat : EnemyStats):
                    if:
                        Stat.Rank = 3
                    then:
                        StatsStackBox.AddWidget(stack_box_slot:
                            Widget:= stack_box:
                                Orientation := orientation.Horizontal
                                Slots := array:
                                    stack_box_slot:
                                        Widget:= text_block{DefaultText:= STM("{Stat.Name}: {Stat.Score}%"), DefaultShadowOpacity := 1.0, DefaultTextColor:= NamedColors.White, DefaultShadowColor:=NamedColors.Black, DefaultShadowOffset:=option{vector2{X:=2.0,Y:=2.0}}}
                            Padding:= margin{Top := 5.0, Left := 0.0, Right := 0.0, Bottom := 00.0}
                        )
                
                for (Stat : EnemyStats):
                    if:
                        Stat.Rank = 4
                    then:
                        StatsStackBox.AddWidget(stack_box_slot:
                            Widget:= stack_box:
                                Orientation := orientation.Horizontal
                                Slots := array:
                                    stack_box_slot:
                                        Widget:= text_block{DefaultText:= STM("{Stat.Name}: {Stat.Score}%"), DefaultShadowOpacity := 1.0, DefaultTextColor:= NamedColors.White, DefaultShadowColor:=NamedColors.Black, DefaultShadowOffset:=option{vector2{X:=2.0,Y:=2.0}}}
                            Padding:= margin{Top := 5.0, Left := 0.0, Right := 0.0, Bottom := 00.0}
                        )
            
                                    
            block:
                Temp1 := ranking_player_stat{Name:="Pickaxe", Score:=PlayerManager.GetPlayerStat(player_stat.Damage_With_Pickaxe) * 100.0, Rank:=0}
                set WeaponStats = WeaponStats + array{Temp1}
                UpdateRankings(Temp1.Name, Temp1.Score, WeaponStats)
                
                Temp2 := ranking_player_stat{Name:="Pistol", Score:=PlayerManager.GetPlayerStat(player_stat.Damage_With_Pistol) * 100.0, Rank:=0}
                set WeaponStats = WeaponStats + array{Temp2}
                UpdateRankings(Temp2.Name, Temp2.Score, WeaponStats)
                
                Temp3 := ranking_player_stat{Name:="Rifle", Score:=PlayerManager.GetPlayerStat(player_stat.Damage_With_AR) * 100.0, Rank:=0}
                set WeaponStats = WeaponStats + array{Temp3}
                UpdateRankings(Temp3.Name, Temp3.Score, WeaponStats)

                Temp4 := ranking_player_stat{Name:="Shotgun", Score:=PlayerManager.GetPlayerStat(player_stat.Damage_With_Shotgun) * 100.0, Rank:=0}
                set WeaponStats = WeaponStats + array{Temp4}
                UpdateRankings(Temp4.Name, Temp4.Score, WeaponStats)

                Temp5 := ranking_player_stat{Name:="Shotgun", Score:=PlayerManager.GetPlayerStat(player_stat.Damage_With_Shotgun) * 100.0, Rank:=0}
                set WeaponStats = WeaponStats + array{Temp5}
                UpdateRankings(Temp5.Name, Temp5.Score, WeaponStats)

                Temp6 := ranking_player_stat{Name:="Railgun", Score:=PlayerManager.GetPlayerStat(player_stat.Damage_With_Railgun) * 100.0, Rank:=0}
                set WeaponStats = WeaponStats + array{Temp6}
                UpdateRankings(Temp6.Name, Temp6.Score, WeaponStats)

                Temp7 := ranking_player_stat{Name:="Bow", Score:=PlayerManager.GetPlayerStat(player_stat.Damage_With_Bow) * 100.0, Rank:=0}
                set WeaponStats = WeaponStats + array{Temp7}
                UpdateRankings(Temp7.Name, Temp7.Score, WeaponStats)

                Temp8 := ranking_player_stat{Name:="Minigun", Score:=PlayerManager.GetPlayerStat(player_stat.Damage_With_Minigun) * 100.0, Rank:=0}
                set WeaponStats = WeaponStats + array{Temp8}
                UpdateRankings(Temp8.Name, Temp8.Score, WeaponStats)

                Temp9 := ranking_player_stat{Name:="Stake Gun", Score:=PlayerManager.GetPlayerStat(player_stat.Damage_With_Stake_Gun) * 100.0, Rank:=0}
                set WeaponStats = WeaponStats + array{Temp9}
                UpdateRankings(Temp9.Name, Temp9.Score, WeaponStats)

                Temp10 := ranking_player_stat{Name:="Scythe", Score:=PlayerManager.GetPlayerStat(player_stat.Damage_With_Scythe) * 100.0, Rank:=0}
                set WeaponStats = WeaponStats + array{Temp10}
                UpdateRankings(Temp10.Name, Temp10.Score, WeaponStats)

                for (Stat : WeaponStats):
                    if:
                        Stat.Rank = 1
                    then:
                        StatsStackBox.AddWidget(stack_box_slot:
                            Widget:= stack_box:
                                Orientation := orientation.Horizontal
                                Slots := array:
                                    stack_box_slot:
                                        Widget:= text_block{DefaultText:= STM("{Stat.Name}: {Stat.Score}%"), DefaultShadowOpacity := 1.0, DefaultTextColor:= NamedColors.White, DefaultShadowColor:=NamedColors.Black, DefaultShadowOffset:=option{vector2{X:=2.0,Y:=2.0}}}
                            Padding:= margin{Top := 30.0, Left := 0.0, Right := 0.0, Bottom := 00.0}
                        )

                for (Stat : WeaponStats):
                    if:
                        Stat.Rank = 2
                    then:
                        StatsStackBox.AddWidget(stack_box_slot:
                            Widget:= stack_box:
                                Orientation := orientation.Horizontal
                                Slots := array:
                                    stack_box_slot:
                                        Widget:= text_block{DefaultText:= STM("{Stat.Name}: {Stat.Score}%"), DefaultShadowOpacity := 1.0, DefaultTextColor:= NamedColors.White, DefaultShadowColor:=NamedColors.Black, DefaultShadowOffset:=option{vector2{X:=2.0,Y:=2.0}}}
                            Padding:= margin{Top := 5.0, Left := 0.0, Right := 0.0, Bottom := 00.0}
                        )

                for (Stat : WeaponStats):
                    if:
                        Stat.Rank = 3
                    then:
                        StatsStackBox.AddWidget(stack_box_slot:
                            Widget:= stack_box:
                                Orientation := orientation.Horizontal
                                Slots := array:
                                    stack_box_slot:
                                        Widget:= text_block{DefaultText:= STM("{Stat.Name}: {Stat.Score}%"), DefaultShadowOpacity := 1.0, DefaultTextColor:= NamedColors.White, DefaultShadowColor:=NamedColors.Black, DefaultShadowOffset:=option{vector2{X:=2.0,Y:=2.0}}}
                            Padding:= margin{Top := 5.0, Left := 0.0, Right := 0.0, Bottom := 00.0}
                        )
                
                for (Stat : WeaponStats):
                    if:
                        Stat.Rank = 4
                    then:
                        StatsStackBox.AddWidget(stack_box_slot:
                            Widget:= stack_box:
                                Orientation := orientation.Horizontal
                                Slots := array:
                                    stack_box_slot:
                                        Widget:= text_block{DefaultText:= STM("{Stat.Name}: {Stat.Score}%"), DefaultShadowOpacity := 1.0, DefaultTextColor:= NamedColors.White, DefaultShadowColor:=NamedColors.Black, DefaultShadowOffset:=option{vector2{X:=2.0,Y:=2.0}}}
                            Padding:= margin{Top := 5.0, Left := 0.0, Right := 0.0, Bottom := 00.0}
                        )

                for (Stat : WeaponStats):
                    if:
                        Stat.Rank = 5
                    then:
                        StatsStackBox.AddWidget(stack_box_slot:
                            Widget:= stack_box:
                                Orientation := orientation.Horizontal
                                Slots := array:
                                    stack_box_slot:
                                        Widget:= text_block{DefaultText:= STM("{Stat.Name}: {Stat.Score}%"), DefaultShadowOpacity := 1.0, DefaultTextColor:= NamedColors.White, DefaultShadowColor:=NamedColors.Black, DefaultShadowOffset:=option{vector2{X:=2.0,Y:=2.0}}}
                            Padding:= margin{Top := 5.0, Left := 0.0, Right := 0.0, Bottom := 00.0}
                        )
                        
                for (Stat : WeaponStats):
                    if:
                        Stat.Rank = 6
                    then:
                        StatsStackBox.AddWidget(stack_box_slot:
                            Widget:= stack_box:
                                Orientation := orientation.Horizontal
                                Slots := array:
                                    stack_box_slot:
                                        Widget:= text_block{DefaultText:= STM("{Stat.Name}: {Stat.Score}%"), DefaultShadowOpacity := 1.0, DefaultTextColor:= NamedColors.White, DefaultShadowColor:=NamedColors.Black, DefaultShadowOffset:=option{vector2{X:=2.0,Y:=2.0}}}
                            Padding:= margin{Top := 5.0, Left := 0.0, Right := 0.0, Bottom := 00.0}
                        )
            
            var ItemCount : int = 0
            for (Item : Items, Item.IsEnabled(PlayerManager)?):
                if (ItemCount >= 4):
                    set CurrentStackBox = stack_box{Orientation:=orientation.Horizontal}
                    StackBoxOuter.AddWidget(stack_box_slot:
                        Widget:= CurrentStackBox
                        Padding:= margin{Top := -20.0, Left := 0.0, Right := 0.0, Bottom := 0.0}
                    )
                    set ItemCount = 0
                set ItemCount += 1

                ItemIcon := texture_block:
                            DefaultImage := Item.GetIcon()
                            DefaultDesiredSize := vector2{X := 120.0, Y := 120.0}
        
                NameButton := button_loud:
                    DefaultText:= STM(Item.GetTitle())

                NameButton.OnClick().Subscribe(menu_opener{Item := Item, GameManager := GameManager}.Open)
                
                Widget := overlay:
                    Slots:= array:
                        overlay_slot:
                            Widget:=ItemIcon
                            Padding:= margin{Top := 0.0, Left := 0.0, Right := 0.0, Bottom := 0.0}
                        overlay_slot:
                            Widget:=NameButton
                            Padding:= margin{Top := 210.0, Left := 10.0, Right := 0.0, Bottom := 0.0}
                            HorizontalAlignment:= horizontal_alignment.Left
                            VerticalAlignment:= vertical_alignment.Top
                
                WidgetSlot := stack_box_slot:
                    Widget:= Widget
                    Padding:= margin{Top := 0.0, Left := 2.5, Right := 0.0, Bottom := 00.0}

                CurrentStackBox.AddWidget(WidgetSlot)
            
            PlayerUI.AddWidget(Canvas, player_ui_slot{InputMode := ui_input_mode.All})
            if (set PlayerManager.UIManager?.Popup = option{Canvas}) {}

    UpdateRankings(Name:string, Score:big_number, PlayerStats:[]ranking_player_stat):void=
        var TempLowestRank : int = 99 
        var OldRank : int = 99
        var RankedPlayersCount : int = 0
        
        for (PlayerScore : PlayerStats):
            if: 
                PlayerScore.Name = Name
            then: 
                set OldRank = PlayerScore.Rank

            if:
                PlayerScore.Rank > 0
            then:
                set RankedPlayersCount += 1

                if:
                    BigNumber_GreaterThanOrEquals(Score, PlayerScore.Score)?
                    PlayerScore.Rank < TempLowestRank
                then:
                    set TempLowestRank = PlayerScore.Rank

        if(TempLowestRank = 99):
            set TempLowestRank = RankedPlayersCount + 1

        for (PlayerScore : PlayerStats):
            if:
                PlayerScore.Rank >= TempLowestRank
                OldRank > TempLowestRank or OldRank = 0
                PlayerScore.Rank < OldRank or OldRank = 0
            then:
                set PlayerScore.Rank += 1
            if: 
                PlayerScore.Name = Name
            then: 
                set PlayerScore.Rank = TempLowestRank

menu_opener := class():
    Item : menu_openable
    GameManager : game_manager
    
    Open(Wdata:widget_message):void=
        if:
            Player := Wdata.Player
            UIManager := GameManager.PlayerManagers[Player].UIManager?
        then:
            UIManager.ClosePopup()
        Item.OpenMenu(Wdata.Player)