using { /Fortnite.com/Devices }
using { /Fortnite.com/UI }
using { /Fortnite.com/Characters }
using { /Verse.org/Simulation }
using { /UnrealEngine.com/Temporary/UI }
using { /Verse.org/Colors }
using { /UnrealEngine.com/Temporary/SpatialMath }
using { /Verse.org/Assets }

menu_template := class<abstract>():
    Agent : ?agent = false
    GameManager : ?game_manager = false

    var _Widget : ?canvas = false

    OpenMenu():void =
        if (not _Widget?):
            Menu := CreateMenu()
            set _Widget = option{Menu}

        if (W := _Widget?):
            UpdateMenu(W)
            DisplayMenu(W)

    Update():void=
        if (W := _Widget?):
            UpdateMenu(W)

    UpdateMenu(Widget:canvas):void
    CreateMenu():canvas

    DisplayMenu(Widget:canvas):void=
        if:
            InPlayer := player[Agent?]
            InPlayer.IsActive[]
            PlayerUI := GetPlayerUI[InPlayer]
            PlayerManager := GameManager?.PlayerManagers[InPlayer]
            Canvas := Widget
        then:
            Canvas.SetEnabled(true)
            Canvas.SetVisibility(widget_visibility.Visible)
            PlayerUI.AddWidget(Canvas, player_ui_slot{InputMode := ui_input_mode.All})
            if (set PlayerManager.UIManager?.Popup = option{Canvas}) {}
            Log("Canvas Launched")

    GetPlayerManager()<decides><transacts>:player_manager= #ABSTRACT
        return GameManager?.PlayerManagers[Agent?]

dummy_menu := class(menu_template):
    UpdateMenu<override>(Widget:canvas):void=
        block:
    CreateMenu<override>():canvas=
        return canvas{}

menu_device_template := class(creative_device, menu_item):
    
    var MenuPerPlayer : [player]menu_template = map{}
    
    OpenMenu<override>(Agent:agent, ?Force:logic = false):void=
        GM := GetGameManager()
        if:
            PlayerManager := GM.PlayerManagers[Agent]
            UIManager := PlayerManager.UIManager?
        then:
            if (Force?):
                UIManager.ClosePopup()

            if:
                not PlayerManager.UIManager?.Popup?
            then:
                if (Menu := MenuPerPlayer[PlayerManager.Player]):
                    Menu.OpenMenu()
                else:
                    Menu := MakeMenu(GM, PlayerManager)
                    if:
                        set MenuPerPlayer[PlayerManager.Player] = Menu
                    Menu.OpenMenu()

    GetGameManager():game_manager= #ABSTRACT
        return game_manager{}

    GetOpenMenu(Agent:agent)<decides><transacts>:menu_template=
        return MenuPerPlayer[Agent]

    MakeMenu(GM : game_manager, PM: player_manager):menu_template= #ABSTRACT
        return dummy_menu{}

    GetTitle<override>():string= #ABSTRACT
        return "???"

    GetIcon<override>():texture= #ABSTRACT
        return Images.QuestionMark

    IsEnabled<override>(PlayerManager : player_manager)<transacts>:logic=
        return true