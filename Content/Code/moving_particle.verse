using { /UnrealEngine.com/Temporary/SpatialMath }
using { /Verse.org/Assets }
using { /UnrealEngine.com/Assets }
using { /Fortnite.com/Devices }
using { /Verse.org/Simulation }

moving_particle_manager:=class(creative_device):
    
    @editable
    Particles : []vfx_spawner_device = array{}
    var Available: []vfx_spawner_device = array{}
    
    OnBegin<override>()<suspends>:void=
        set Available = Particles
        var num:int = 0
        for (Part : Particles):
            set num = num + 1
            Part.Disable()

    MoveTo(StartPos:vector3, EndPos:vector3)<suspends>:void=
        if:
            Part := Available[0]
            OrigPos := Part.GetTransform().Translation 
            Temp := Available.RemoveElement[0]
            set Available = Temp
        then:
            if:
                Part.TeleportTo[StartPos, IdentityRotation()]
            Sleep(0.1)
            Part.Enable()
            Part.Restart()
            MoveResult := Part.MoveTo((StartPos*3.0+EndPos)/4.0 + vector3{X:=0.0, Y:=0.0, Z:=100.0}, IdentityRotation(), 0.3)
            MoveResult2 := Part.MoveTo((StartPos+EndPos)/2.0 + vector3{X:=0.0, Y:=0.0, Z:=200.0}, IdentityRotation(), 0.3)
            MoveResult3 := Part.MoveTo((StartPos+EndPos*3.0)/4.0 + vector3{X:=0.0, Y:=0.0, Z:=200.0}, IdentityRotation(), 0.3) 
            MoveResult4 := Part.MoveTo(EndPos, IdentityRotation(), 0.3)
            Sleep(2.0)
            if (Part.TeleportTo[OrigPos, IdentityRotation()]) {}
            Part.Disable()
            set Available = Available + array{Part}