using { /Fortnite.com/Devices }
using { /Fortnite.com/UI }
using { /UnrealEngine.com/Temporary/Diagnostics }
using { /UnrealEngine.com/Temporary/SpatialMath }
using { /UnrealEngine.com/Temporary/UI}
using { /Verse.org/Assets }
using { /Verse.org/Colors }
using { /Verse.org/Simulation }

npc_contact_device := class(creative_device, menu_item):

    @editable
    GameManager : ?game_manager = false

    @editable 
    NPCs: []npc_device = array{}
    
    OpenMenu<override>(Agent:agent):void=
        if (PlayerManager := GameManager?.PlayerManagers[Agent]):
            
            if (not GameManager?.PlayerManagers[Agent].UIManager?.Popup?):
                if (GM := GameManager?):
                    MakeNPCSMenu(GM, Self, Agent).OpenMenu()
                    PlayerManager.GainAchievementLevel("OpenedUpgrades",1)
            else:
                block:#Log("DOUBLE MENU")#

    GetTitle<override>():string=
        return "UPGRADES"

    GetIcon<override>():texture=
        return Images.UpgradeIcon

    IsEnabled<override>(PlayerManager : player_manager)<transacts>:logic=
        return PlayerManager.HasCompletedTask(0, 21)

MakeNPCSMenu<constructor>(GameManagerIn : game_manager, DeviceIn:npc_contact_device, AgentIn : agent) := npc_contact_menu:
    GameManager :=  GameManagerIn
    ContactDevice := DeviceIn
    Agent := AgentIn

npc_contact_menu:= class:
    GameManager : game_manager
    ContactDevice : npc_contact_device 
    Agent : agent

    var DialogBackground : texture_block = texture_block{ DefaultImage := Images.StoreBox, DefaultDesiredSize := vector2{X := 1920.0, Y := 1080.0}}   

    OpenMenu():void =
        if:
            InPlayer := player[Agent]
            InPlayer.IsActive[]
            PlayerUI := GetPlayerUI[InPlayer]
            PlayerManager := GameManager.PlayerManagers[InPlayer]
        then:
            var Items : []widget = array{}

            for (NPC : ContactDevice.NPCs):
                if (PlayerManager.GetAchievementLevel["InteractedWithNPC_{NPC.Name}"] > 0):
                    Display := MakeNPCDisplay(NPC, PlayerManager, ContactDevice)
                    Display.Setup()
                    set Items = Items + array{Display.Widget}

            CloseButton := button_quiet:
            CloseButton.SetText(STM("                                                                        BACK                                                                        "))            
            CloseButton.OnClick().Subscribe(PlayerManager.ClosePopup)

            Grid := MakeWidgetGrid(Items, 3, vector2{X:=10.0, Y:=40.0})

            Canvas := canvas:
                Slots := array:
                    canvas_slot:
                        Widget := DialogBackground
                        SizeToContent := true
                        Anchors := anchors{ Minimum := vector2{X := 0.1, Y := 0.1}, Maximum := vector2{X := 0.9, Y := 0.9}}
                        ZOrder := 1
                    canvas_slot:
                        Widget:= text_block:
                            DefaultTextColor:=NamedColors.White
                            DefaultShadowColor:=NamedColors.Black   
                            DefaultShadowOffset:=option{vector2{X:=2.0,Y:=2.0}}
                            DefaultShadowOpacity:= 1.0
                            DefaultText:= STM("Pick a Character to view their upgrades!")
                        SizeToContent := true
                        Anchors := anchors{ Minimum := vector2{X := 0.5, Y := 0.15}, Maximum := vector2{X := 0.5, Y := 0.15}}
                        Offsets := margin{Top := 0.0, Left := 0.0, Right := 0.0, Bottom := 0.0}
                        Alignment := vector2{X := 0.5, Y := 0.5}
                        ZOrder := 2
                    canvas_slot:
                        Widget := Grid
                        SizeToContent := true
                        Anchors := anchors{ Minimum := vector2{X := 0.1, Y := 0.1}, Maximum := vector2{X := 0.9, Y := 0.9}}
                        Offsets := margin{ Top := 50.0, Left := 50.0, Right := 50.0, Bottom := 50.0}
                        Alignment := vector2{ X := 0.5, Y := 1.0}
                        ZOrder := 3
                    canvas_slot:
                        Widget := CloseButton
                        Anchors := anchors{ Minimum := vector2{X := 0.5, Y := 0.855}, Maximum := vector2{X := 0.5, Y := 0.855}}
                        Alignment := vector2{ X := 0.5, Y := 1.0}
                        ZOrder := 4
            
            PlayerUI.AddWidget(Canvas, player_ui_slot{InputMode := ui_input_mode.All})
            if (set PlayerManager.UIManager?.Popup = option{Canvas}) {}

MakeNPCDisplay<constructor>(InNPC : npc_device, InPM : player_manager, InMenu : npc_contact_device) := npc_display: 
    NPC := InNPC
    PM := InPM
    Menu := InMenu

npc_display:= class():
    NPC : npc_device
    PM : player_manager
    Menu : npc_contact_device
    

    var Widget : overlay = overlay{}

    Setup():void =
        Btn := button_loud:
            DefaultText := STM("{NPC.Name}")
        
        Btn.OnClick().Subscribe(OnClick)

        set Widget = overlay:
            Slots:= array:
                overlay_slot:
                    Widget:=texture_block:
                        DefaultImage := GetTextureFromString(NPC.Name)
                        DefaultDesiredSize := vector2{X := 200.0, Y := 200.0}
                    Padding:= margin{Top := 0.0, Left := 0.0, Right := 0.0, Bottom := 0.0}
                overlay_slot:
                    Widget:= Btn
                    Padding:= margin{Top := 130.0, Left := 40.0, Right := 40.0, Bottom := 0.0}
                    HorizontalAlignment:= horizontal_alignment.Center
        
        var i : int = 0
        for (Item : NPC.CollectibleShop?.Items):
            Widget.AddWidget(
                overlay_slot:
                    Widget:= texture_block:
                        DefaultImage := Images.IconBorder
                        DefaultDesiredSize := vector2{X := 60.0, Y := 60.0}
                    Padding:= margin{Top := 70.0, Left := 60.0 - 60.0*i, Right := 0.0, Bottom := 0.0} 
                    HorizontalAlignment:= horizontal_alignment.Center
            )
            Widget.AddWidget(
                overlay_slot:
                    Widget:= texture_block:
                        DefaultImage := GetTextureFromString(Item.Name)
                        DefaultDesiredSize := vector2{X := 50.0, Y := 50.0}
                    Padding:= margin{Top := 70.0, Left := 60.0 - 60.0*i, Right := 0.0, Bottom := 0.0} 
                    HorizontalAlignment:= horizontal_alignment.Center
            )
            set i += 1
        

        set i = 0
        for (Item : NPC.CollectibleShop?.Items):            
            var Cost : int = 0
            CurrencyName := Item.Currency

            Widget.AddWidget(
                overlay_slot:
                    Widget:= texture_block:
                        DefaultImage := GetTextureFromString(CurrencyName)
                        DefaultDesiredSize := vector2{X := 30.0, Y := 30.0}
                    Padding:= margin{Top := 70.0 - 20.0, Left := 60.0 - 60.0*i - 20.0, Right := 0.0, Bottom := 0.0}  
                    HorizontalAlignment:= horizontal_alignment.Center
            )
            set i += 1
    
        for (Upgrade : NPC.Shop?.Items):
            Widget.AddWidget(
                overlay_slot:
                    Widget:= texture_block:
                        DefaultImage := Images.IconBorder
                        DefaultDesiredSize := vector2{X := 60.0, Y := 60.0}
                    Padding:= margin{Top := 70.0, Left := 60.0 - 60.0*i, Right := 0.0, Bottom := 0.0} 
                    HorizontalAlignment:= horizontal_alignment.Center
            )
            Widget.AddWidget(
                overlay_slot:
                    Widget:= texture_block:
                        DefaultImage := GetTextureFromString(Upgrade.Name)
                        DefaultDesiredSize := vector2{X := 50.0, Y := 50.0}
                    Padding:= margin{Top := 70.0, Left := 60.0 - 60.0*i, Right := 0.0, Bottom := 0.0} 
                    HorizontalAlignment:= horizontal_alignment.Center
            )
            set i += 1
        
        set i = 0
        for (Upgrade : NPC.Shop?.Items):
            
            var ItemLevel : int = 0
            var Cost : int = 0
            if:
                TempItemLevel := PlayerProfileDataMap[PM.Player].UpgradeLevels[Upgrade.Name]
                set ItemLevel = TempItemLevel
            CurrencyName := Upgrade.GetCurrencyName(ItemLevel)

            Widget.AddWidget(
                overlay_slot:
                    Widget:= texture_block:
                        DefaultImage := GetTextureFromString(CurrencyName)
                        DefaultDesiredSize := vector2{X := 30.0, Y := 30.0}
                    Padding:= margin{Top := 70.0 - 20.0, Left := 60.0 - 60.0*i - 20.0, Right := 0.0, Bottom := 0.0}  
                    HorizontalAlignment:= horizontal_alignment.Center
            )
            set i += 1

    OnClick(Wdata : widget_message):void =
        PM.ClosePopup(Wdata)
        if (NPCShop := NPC.Shop?):
            NPCShop.TryOpenMenu(NPC.Name, PM.Player, ?NPCContactMenu:= option{Menu})
        if (NPCShop := NPC.CollectibleShop?):
            NPCShop.TryOpenMenu(NPC.Name, PM.Player, ?NPCContactMenu:= option{Menu})