using { /Fortnite.com/Devices }
using { /Verse.org/Assets }
using { /UnrealEngine.com/Assets }
using { /Fortnite.com/Devices }
using { /Verse.org/Simulation }

particle_prop:=class(creative_device):
    @editable
    VFX:vfx_spawner_device = vfx_spawner_device{}