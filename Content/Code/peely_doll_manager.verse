using { /Fortnite.com/Devices }
using { /Fortnite.com/UI }
using { /Fortnite.com/Characters }
using { /Verse.org/Simulation }
using { /UnrealEngine.com/Temporary/UI }
using { /Verse.org/Colors }
using { /UnrealEngine.com/Temporary/SpatialMath }
using { /Verse.org/Assets }

queued_peely_doll := struct<concrete><public>:
    Peely  : peely_doll_device = peely_doll_device{}
    Agent : ?agent = false
    NewState : logic = false

peely_doll_device := class(creative_device):
    @editable
    Peely : switch_device = switch_device{}

    @editable
    Index : int = -1

    @editable
    HideSequencer : cinematic_sequence_device = cinematic_sequence_device{}

    @editable
    ShowSequencer : cinematic_sequence_device = cinematic_sequence_device{}

    @editable
    ZoneTag : fetchable_tag = fetchable_tag.none_tag

    @editable
    IsQuestPeely : logic = false

    @editable
    CustomCollectible : ?collectible_data = false

peely_doll_manager := class: 
    var GameManager: ?game_manager = false
    var Player : ?player = false

    var PeelyMidas : []peely_doll_device = array{}
    var PeelyGoodtown : []peely_doll_device = array{}
    var PeelyBeach : []peely_doll_device = array{}
    var PeelyUmbralForest : []peely_doll_device = array{}
    var PeelyGiantlands : []peely_doll_device = array{}
    var PeelyZephyrsCanyon : []peely_doll_device = array{}
    var PeelyPirateIsland : []peely_doll_device = array{}

    var PeelyIndex : []peely_doll_device = array{}

    var IGNORESUBSCRIPTION : logic = false

    Init():void=
        if:
            GM := GameManager?
            SetPlayer := Player?
            SetPlayer.IsActive[]
            PlayerManager := GM.PlayerManagers[SetPlayer]
            Agent := PlayerManager.Agent
        then:
            for(Peely : GM.PeelyManagers):
                AssignPeelyToZoneGroup(Peely)

        SubscribeToAllPeelys()

    ClearPeelyPersistence(Agent:agent):void=
        for (Peely : PeelyIndex):
            Peely.Peely.ClearPersistenceData(Agent)
            spawn:
                ResetStateCarefully(Peely, Agent, true)

    EnsurePeelysVisibility(Agent : agent, Peelys : []peely_doll_device):void=
        for (Peely : Peelys):
            if:
                GM := GameManager?
                InPlayer := player[Agent]
                InPlayer.IsActive[]
                PlayerManager := GM.PlayerManagers[InPlayer]
            then:
                if:
                    PlayerManager.GetPeelysFound[].Find[Peely.Index]
                then:
                    GM.QueuePeely(queued_peely_doll{Peely := Peely, Agent := option{Agent}, NewState := false})
                else:
                    GM.QueuePeely(queued_peely_doll{Peely := Peely, Agent := option{Agent}, NewState := true})

    PlayerEntersZone(ZoneTag:fetchable_tag):void=
        if:
            GM := GameManager?
            SetPlayer := Player?
            SetPlayer.IsActive[]
            PlayerManager := GM.PlayerManagers[SetPlayer]
            Agent := PlayerManager.Agent
        then:
            CallVisbilityForGroup(Agent, ZoneTag)
            
    SubscribeToAllPeelys():void=
        #Check Player Peely Date
        
        if:
            GM := GameManager?
            SetPlayer := Player?
            SetPlayer.IsActive[]
            PlayerManager := GM.PlayerManagers[SetPlayer]
            Agent := PlayerManager.Agent
        then:
            if:
                PlayerManager.IsNewPeelyDate[]
                PlayerManager.ResetPeelysFound[]
            then:
                for(Peely : GM.PeelyManagers):
                    if(Peely.IsQuestPeely?):
                    else:
                        spawn:
                            ResetStateCarefully(Peely, Agent, true)
                        GM.QueuePeely(queued_peely_doll{Peely := Peely, Agent := option{Agent}, NewState := true})
            else:
                for(Peely : GM.PeelyManagers):
                    if:
                        PlayerManager.GetPeelysFound[].Find[Peely.Index]
                    then:
                        GM.QueuePeely(queued_peely_doll{Peely := Peely, Agent := option{Agent}, NewState := false})
                        spawn:
                            ResetStateCarefully(Peely, Agent, false)
                    

        var PeelyCount : int = 0
        #Subscribe to all Peelys according to class
        for (Peely : PeelyMidas):
            Peely.Peely.TurnedOffEvent.SubscribeAgent(OnToggle, PeelyCount) #
            Peely.Peely.TurnedOnEvent.SubscribeAgent(StopThat, PeelyCount)
            set PeelyIndex = PeelyIndex + array{Peely}
            set PeelyCount += 1

        for (Peely : PeelyGoodtown):
            Peely.Peely.TurnedOffEvent.SubscribeAgent(OnToggle, PeelyCount) #
            Peely.Peely.TurnedOnEvent.SubscribeAgent(StopThat, PeelyCount)
            set PeelyIndex = PeelyIndex + array{Peely}
            set PeelyCount += 1

        for (Peely : PeelyBeach):
            Peely.Peely.TurnedOffEvent.SubscribeAgent(OnToggle, PeelyCount) #
            Peely.Peely.TurnedOnEvent.SubscribeAgent(StopThat, PeelyCount)
            set PeelyIndex = PeelyIndex + array{Peely}
            set PeelyCount += 1
    
        for (Peely : PeelyUmbralForest):
            Peely.Peely.TurnedOffEvent.SubscribeAgent(OnToggle, PeelyCount) #
            Peely.Peely.TurnedOnEvent.SubscribeAgent(StopThat, PeelyCount)
            set PeelyIndex = PeelyIndex + array{Peely}
            set PeelyCount += 1
    
        for (Peely : PeelyGiantlands):
            Peely.Peely.TurnedOffEvent.SubscribeAgent(OnToggle, PeelyCount) #
            Peely.Peely.TurnedOnEvent.SubscribeAgent(StopThat, PeelyCount)
            set PeelyIndex = PeelyIndex + array{Peely}
            set PeelyCount += 1
    
        for (Peely : PeelyZephyrsCanyon):
            Peely.Peely.TurnedOffEvent.SubscribeAgent(OnToggle, PeelyCount) #
            Peely.Peely.TurnedOnEvent.SubscribeAgent(StopThat, PeelyCount)
            set PeelyIndex = PeelyIndex + array{Peely}
            set PeelyCount += 1

        for (Peely : PeelyPirateIsland):
            Peely.Peely.TurnedOffEvent.SubscribeAgent(OnToggle, PeelyCount) #
            Peely.Peely.TurnedOnEvent.SubscribeAgent(StopThat, PeelyCount)
            set PeelyIndex = PeelyIndex + array{Peely}
            set PeelyCount += 1

    
    StopThat(InteractAgent:agent, Index : int):void=
        if:
            IGNORESUBSCRIPTION?
        then:
            return

        #Confirm agent is ours
        if:
            GM := GameManager?
            SetPlayer := Player?
            SetPlayer.IsActive[]
            PlayerManager := GM.PlayerManagers[SetPlayer]
            Agent := PlayerManager.Agent
            Agent = InteractAgent
            Peely := PeelyIndex[Index]
        then:
            Location := GetLocationForZone(Peely.ZoneTag)

            LocalPeelysFound := GetPeelysFound(Location) + 1
            LocalMaxPeelys := GetMaxPeelys(Location)
            TotalPeelysFound := PlayerManager.GetCollectibleCount("Peely")

            #Display UI based on X out of Y found in area
            if:
                SetUIManager := PlayerManager.UIManager?
            then:
                SetUIManager.FoundPeely(Location, LocalPeelysFound, LocalMaxPeelys, TotalPeelysFound)

            spawn:
                ResetStateCarefully(Peely, Agent, false)
            
    ResetStateCarefully(Peely : peely_doll_device, Agent : agent, NewState : logic)<suspends>:void=
        loop:
            if:
                IGNORESUBSCRIPTION?
            then:
                block:
            else:
                break
            Sleep(0.1)

        set IGNORESUBSCRIPTION = true

        Peely.Peely.SetState(Agent, NewState)

        Sleep(0.1)
        set IGNORESUBSCRIPTION = false


    OnToggle(InteractAgent:agent, Index : int):void=
        if:
            IGNORESUBSCRIPTION?
        then:
            return

        #Confirm agent is ours
        if:
            GM := GameManager?
            SetPlayer := Player?
            SetPlayer.IsActive[]
            PlayerManager := GM.PlayerManagers[SetPlayer]
            Agent := PlayerManager.Agent
            Agent = InteractAgent
            Peely := PeelyIndex[Index]
            #Set Player Peely Date
            PlayerManager.SetPeelyDate[]
        then:
            #PlayersHasn't Found it yet
            if:
                PlayerManager.GetPeelysFound[].Find[Peely.Index]
            then:
                return

            if:
                PlayerManager.AddPeelysFound[Peely.Index]

            GM.AccoladesManager.FoundPeelyAward(InteractAgent)
            GM.QueuePeely(queued_peely_doll{Peely := Peely, Agent := option{Agent}, NewState := false})

            Location := GetLocationForZone(Peely.ZoneTag)

            TotalPeelysFound := PlayerManager.GetCollectibleCount("Peely") + big_number{Value := 1.0}
            LocalPeelysFound := GetPeelysFound(Location)
            LocalMaxPeelys := GetMaxPeelys(Location)
            var LocalPeelyBonus : float = 0.0

            if(BigNumber_Equals(TotalPeelysFound, big_number{Value:= 1.0})?):
                GM.AnalyticsManager.peely_found_first.Submit(Agent)
            else:
                if(BigNumber_Equals(TotalPeelysFound, big_number{Value:= 10.0})?):
                    GM.AnalyticsManager.peely_found_ten.Submit(Agent)

            if(LocalPeelysFound = LocalMaxPeelys):
                set LocalPeelyBonus = LocalMaxPeelys * 1.0
            #Payout Rewards according to false count
            
            if (CC := Peely.CustomCollectible?):
                PlayerManager.GainCollectible(CC)
            else:
                if:
                    Print("A")
                    not Peely.IsQuestPeely?
                then:
                    spawn:
                        GainPeelyCollectibles(PlayerManager, GM, TotalPeelysFound, LocalPeelyBonus)

            if:
                Print("A")
                not  Peely.IsQuestPeely?
            then:
                #Display UI based on X out of Y found in area
                if:
                    SetUIManager := PlayerManager.UIManager?

                then:
                    SetUIManager.FoundPeely(Location, LocalPeelysFound, LocalMaxPeelys, TotalPeelysFound)

    GainPeelyCollectibles(PlayerManager : player_manager, GM : game_manager, TotalPeelysFound : big_number, LocalPeelyBonus : float)<suspends>:void=
        PlayerManager.GainCollectible(collectible_data{Key:="Peely", Count:=big_number{Value:= 1.0}})
        Sleep(0.25)
        if:
            Amt := 100.0 * MakeBigNumber(1.25, Int[BigNumber_ToFloat(TotalPeelysFound)])
        then:
            PlayerManager.GainCollectible(collectible_data{Key:="Midas Coins", Count:= Amt})
        GM.AudioManager.PlayCollectCoinsSFX(PlayerManager.Agent)
        Sleep(0.25)
        if:
            LocalPeelyBonus > 0.0
            Amt := 100.0 * MakeBigNumber(1.25, Int[BigNumber_ToFloat(TotalPeelysFound)])
        then:
            PlayerManager.GainCollectible(collectible_data{Key:="Midas Coins", Count:= LocalPeelyBonus * (Amt)})        

    GetPeelysFound(Location : string) : int=
        if:
            SetPlayer := Player?
            GM := GameManager?
            PlayerManager := GM.PlayerManagers[SetPlayer]
        then:
            case (Location):
                "All" =>
                    var Count : int = 0
                    set Count += GetPeelysFound("Midas Castle")
                    set Count += GetPeelysFound("the Goodtown Area")
                    set Count += GetPeelysFound("Umbral Forest")
                    set Count += GetPeelysFound("the Giantlands")
                    set Count += GetPeelysFound("Zephyr's Canyon")
                "Midas Castle" => 
                    var Count : int = 0
                    for (Peely : PeelyMidas):
                        if:
                            PlayerManager.GetPeelysFound[].Find[Peely.Index]
                        then:
                            set Count += 1
                    Count
                "the Goodtown Area" => 
                    var Count : int = 0
                    for (Peely : PeelyGoodtown):
                        if:
                            PlayerManager.GetPeelysFound[].Find[Peely.Index]
                        then:
                            set Count += 1
                    Count
                "Goodtown Beach" => 
                    var Count : int = 0
                    for (Peely : PeelyBeach):
                        if:
                            PlayerManager.GetPeelysFound[].Find[Peely.Index]
                        then:
                            set Count += 1
                    Count
                "Umbral Forest" => 
                    var Count : int = 0
                    for (Peely : PeelyUmbralForest):
                        if:
                            PlayerManager.GetPeelysFound[].Find[Peely.Index]
                        then:
                            set Count += 1
                    Count
                "the Giantlands" => 
                    var Count : int = 0
                    for (Peely : PeelyGiantlands):
                        if:
                            PlayerManager.GetPeelysFound[].Find[Peely.Index]
                        then:
                            set Count += 1
                    Count
                "Zephyr's Canyon" => 
                    var Count : int = 0
                    for (Peely : PeelyGiantlands):
                        if:
                            PlayerManager.GetPeelysFound[].Find[Peely.Index]
                        then:
                            set Count += 1
                    Count
                _ => 0
        else:
            0

            # s
    GetMaxPeelys(Location : string) : int=
        case (Location):
            "Midas Castle" => PeelyMidas.Length
            "the Goodtown Area" => PeelyGoodtown.Length
            "Goodtown Beach" => PeelyBeach.Length
            "Umbral Forest" => PeelyUmbralForest.Length
            "the Giantlands" => PeelyGiantlands.Length
            "Zephyr's Canyon" => PeelyZephyrsCanyon.Length
            "Pirate Island" => PeelyPirateIsland.Length
            _ => 0

     
    GetLocationForZone(ZoneTag : fetchable_tag):string= 
        case (ZoneTag):
            fetchable_tag.zone1tag => "Midas Castle"
            fetchable_tag.zone2tag => "Midas Castle"
            fetchable_tag.zone3tag => "Midas Castle"
            fetchable_tag.midastag => "Midas Castle"

            fetchable_tag.towntag => "the Goodtown Area"
            fetchable_tag.afk_zonetag => "the Goodtown Area"
            fetchable_tag.range_zonetag => "the Goodtown Area"

            fetchable_tag.beach_zone1tag => "Goodtown Beach"
            fetchable_tag.beach_zone2tag => "Goodtown Beach"
            fetchable_tag.beach_zone3tag => "Goodtown Beach"

            fetchable_tag.forest_zone1tag => "Umbral Forest"
            fetchable_tag.forest_zone2tag => "Umbral Forest"
            fetchable_tag.forest_zone3tag => "Umbral Forest"
            fetchable_tag.forest_zone4tag => "Umbral Forest"

            fetchable_tag.canyon_zone1tag => "Zephyr's Canyon"
            fetchable_tag.canyon_zone2tag => "Zephyr's Canyon"
            fetchable_tag.canyon_zone3tag => "Zephyr's Canyon"
            fetchable_tag.canyon_zone4tag => "Zephyr's Canyon"
            fetchable_tag.canyon_zone5tag => "Zephyr's Canyon"
            fetchable_tag.canyon_zone6tag => "Zephyr's Canyon"
            
            fetchable_tag.giantlands_zone1tag => "the Giantlands"
            fetchable_tag.giantlands_zone2tag => "the Giantlands"
            fetchable_tag.giantlands_zone3tag => "the Giantlands"
            fetchable_tag.giantlands_zone4tag => "the Giantlands"
            fetchable_tag.giantlands_zone5tag => "the Giantlands"
            fetchable_tag.giantlands_zone6tag => "the Giantlands"

            fetchable_tag.pirates_zone1tag => "Pirate Island"
            fetchable_tag.pirates_zone2tag => "Pirate Island"
            fetchable_tag.pirates_zone3tag => "Pirate Island"
            fetchable_tag.pirates_zone4tag => "Pirate Island"
            fetchable_tag.pirates_zone5tag => "Pirate Island"
            _ => ""

    CallVisbilityForGroup(Agent : agent, ZoneTag : fetchable_tag):void= 
        case (ZoneTag):
            fetchable_tag.zone1tag => EnsurePeelysVisibility(Agent, PeelyMidas)
            fetchable_tag.zone2tag => EnsurePeelysVisibility(Agent, PeelyMidas)
            fetchable_tag.zone3tag => EnsurePeelysVisibility(Agent, PeelyMidas)
            fetchable_tag.midastag => EnsurePeelysVisibility(Agent, PeelyMidas)
            fetchable_tag.towntag => EnsurePeelysVisibility(Agent, PeelyGoodtown)
            fetchable_tag.afk_zonetag => EnsurePeelysVisibility(Agent, PeelyGoodtown)
            fetchable_tag.range_zonetag => EnsurePeelysVisibility(Agent, PeelyGoodtown)

            fetchable_tag.beach_zone1tag => EnsurePeelysVisibility(Agent, PeelyBeach)
            fetchable_tag.beach_zone2tag => EnsurePeelysVisibility(Agent, PeelyBeach)
            fetchable_tag.beach_zone3tag => EnsurePeelysVisibility(Agent, PeelyBeach)

            fetchable_tag.forest_zone1tag => EnsurePeelysVisibility(Agent, PeelyBeach)
            fetchable_tag.forest_zone2tag => EnsurePeelysVisibility(Agent, PeelyBeach)
            fetchable_tag.forest_zone3tag => EnsurePeelysVisibility(Agent, PeelyBeach)
            fetchable_tag.forest_zone4tag => EnsurePeelysVisibility(Agent, PeelyBeach)

            fetchable_tag.canyon_zone1tag => EnsurePeelysVisibility(Agent, PeelyZephyrsCanyon)
            fetchable_tag.canyon_zone2tag => EnsurePeelysVisibility(Agent, PeelyZephyrsCanyon)
            fetchable_tag.canyon_zone3tag => EnsurePeelysVisibility(Agent, PeelyZephyrsCanyon)
            fetchable_tag.canyon_zone4tag => EnsurePeelysVisibility(Agent, PeelyZephyrsCanyon)
            fetchable_tag.canyon_zone5tag => EnsurePeelysVisibility(Agent, PeelyZephyrsCanyon)
            fetchable_tag.canyon_zone6tag => EnsurePeelysVisibility(Agent, PeelyZephyrsCanyon)
            
            fetchable_tag.giantlands_zone1tag => EnsurePeelysVisibility(Agent, PeelyGiantlands)
            fetchable_tag.giantlands_zone2tag => EnsurePeelysVisibility(Agent, PeelyGiantlands)
            fetchable_tag.giantlands_zone3tag => EnsurePeelysVisibility(Agent, PeelyGiantlands)
            fetchable_tag.giantlands_zone4tag => EnsurePeelysVisibility(Agent, PeelyGiantlands)
            fetchable_tag.giantlands_zone5tag => EnsurePeelysVisibility(Agent, PeelyGiantlands)
            fetchable_tag.giantlands_zone6tag => EnsurePeelysVisibility(Agent, PeelyGiantlands)

            fetchable_tag.pirates_zone1tag => EnsurePeelysVisibility(Agent, PeelyPirateIsland)
            fetchable_tag.pirates_zone2tag => EnsurePeelysVisibility(Agent, PeelyPirateIsland)
            fetchable_tag.pirates_zone3tag => EnsurePeelysVisibility(Agent, PeelyPirateIsland)
            fetchable_tag.pirates_zone4tag => EnsurePeelysVisibility(Agent, PeelyPirateIsland)
            fetchable_tag.pirates_zone5tag => EnsurePeelysVisibility(Agent, PeelyPirateIsland)

            _ => block:

    AssignPeelyToZoneGroup(Peely : peely_doll_device):void=
        case (Peely.ZoneTag):
            fetchable_tag.zone1tag => set PeelyMidas = PeelyMidas + array{Peely}
            fetchable_tag.zone2tag => set PeelyMidas = PeelyMidas + array{Peely}
            fetchable_tag.zone3tag => set PeelyMidas = PeelyMidas + array{Peely}
            fetchable_tag.midastag => set PeelyMidas = PeelyMidas + array{Peely}

            fetchable_tag.towntag => set PeelyGoodtown = PeelyGoodtown + array{Peely}
            fetchable_tag.afk_zonetag => set PeelyGoodtown = PeelyGoodtown + array{Peely}
            fetchable_tag.range_zonetag => set PeelyGoodtown = PeelyGoodtown + array{Peely}

            fetchable_tag.forest_zone1tag => set PeelyUmbralForest = PeelyUmbralForest + array{Peely}
            fetchable_tag.forest_zone2tag => set PeelyUmbralForest = PeelyUmbralForest + array{Peely}
            fetchable_tag.forest_zone3tag => set PeelyUmbralForest = PeelyUmbralForest + array{Peely}
            fetchable_tag.forest_zone4tag => set PeelyUmbralForest = PeelyUmbralForest + array{Peely}

            fetchable_tag.beach_zone1tag => set PeelyBeach = PeelyBeach + array{Peely}
            fetchable_tag.beach_zone2tag => set PeelyBeach = PeelyBeach + array{Peely}
            fetchable_tag.beach_zone3tag => set PeelyBeach = PeelyBeach + array{Peely}

            fetchable_tag.canyon_zone1tag => set PeelyZephyrsCanyon = PeelyZephyrsCanyon + array{Peely}
            fetchable_tag.canyon_zone2tag => set PeelyZephyrsCanyon = PeelyZephyrsCanyon + array{Peely}
            fetchable_tag.canyon_zone3tag => set PeelyZephyrsCanyon = PeelyZephyrsCanyon + array{Peely}
            fetchable_tag.canyon_zone4tag => set PeelyZephyrsCanyon = PeelyZephyrsCanyon + array{Peely}
            fetchable_tag.canyon_zone5tag => set PeelyZephyrsCanyon = PeelyZephyrsCanyon + array{Peely}
            fetchable_tag.canyon_zone6tag => set PeelyZephyrsCanyon = PeelyZephyrsCanyon + array{Peely}
            
            fetchable_tag.giantlands_zone1tag => set PeelyGiantlands = PeelyGiantlands + array{Peely}
            fetchable_tag.giantlands_zone2tag => set PeelyGiantlands = PeelyGiantlands + array{Peely}
            fetchable_tag.giantlands_zone3tag => set PeelyGiantlands = PeelyGiantlands + array{Peely}
            fetchable_tag.giantlands_zone4tag => set PeelyGiantlands = PeelyGiantlands + array{Peely}
            fetchable_tag.giantlands_zone5tag => set PeelyGiantlands = PeelyGiantlands + array{Peely}
            fetchable_tag.giantlands_zone6tag => set PeelyGiantlands = PeelyGiantlands + array{Peely}

            fetchable_tag.pirates_zone1tag => set PeelyPirateIsland = PeelyPirateIsland + array{Peely}
            fetchable_tag.pirates_zone2tag => set PeelyPirateIsland = PeelyPirateIsland + array{Peely}
            fetchable_tag.pirates_zone3tag => set PeelyPirateIsland = PeelyPirateIsland + array{Peely}
            fetchable_tag.pirates_zone4tag => set PeelyPirateIsland = PeelyPirateIsland + array{Peely}
            fetchable_tag.pirates_zone5tag => set PeelyPirateIsland = PeelyPirateIsland + array{Peely}
        
            _ => block: