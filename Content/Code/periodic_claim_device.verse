
using { /Fortnite.com/Devices }
using { /Verse.org/Simulation }
using { /UnrealEngine.com/Temporary/Diagnostics }

periodic_claim_device := class(creative_device):
    @editable
    GameManager : game_manager = game_manager{}

    @editable
    Button: button_device = button_device{}

    @editable
    CooldownSeconds : float = 900.0

    @editable
    CooldownID : string = ""

    @editable
    GrantCollectible : collectible_data = collectible_data{}

    OnBegin<override>()<suspends>:void=
        Button.InteractedWithEvent.Subscribe(OnButtonInteracted)
    
    OnButtonInteracted(Agent:agent):void=
        if:
            PlayerManager := GameManager.PlayerManagers[Agent]
            UIManager := PlayerManager.UIManager?
        then:
            
            CD := PlayerManager.GetCooldown(CooldownID)
            if:
                CD <= 0.0
            then:
                PlayerManager.SetCooldown(CooldownID, CooldownSeconds)
                PlayerManager.GainCollectible(GrantCollectible)
            else if:
                CDSeconds := Int[CD]
            then:
                UIManager.SetZoneOpenText("You've already claimed this recently. Come back in {FormatTime(CDSeconds)}.")
                set UIManager.ZoneMsgVisibilityTimer = 5.0
