
using { /Fortnite.com/Devices }
using { /Verse.org/Simulation }
using { /UnrealEngine.com/Temporary/Diagnostics }
using { /Fortnite.com/Game }
using { /Fortnite.com/Characters }

# See https://dev.epicgames.com/documentation/en-us/uefn/create-your-own-device-in-verse for how to create a verse device.

# A Verse-authored creative device that can be placed in a level
player_damage_manager := class:

    var PerPlayerRegisters : [float]([player]float) = map{}
    var Processing : logic = false

    RegisterDamage(Volume:damage_volume_device, Player:player, Amount:float):void=
        RegisterDamage(1.0*Volume.GetDamage(), Player, Amount)

    RegisterDamage(Sentinel: float,Player:player,  Amount:float):void=
            if:
                not PerPlayerRegisters[Sentinel]
                set PerPlayerRegisters[Sentinel] = map{}

            if:
                not PerPlayerRegisters[Sentinel][Player]
                FortChar := Player.GetFortCharacter[]
            then:
                FortChar.DamagedEvent().Subscribe(OnDamage)

            if:
                set PerPlayerRegisters[Sentinel][Player] = Amount
            

    # Runs when the device is started in a running game
    OnDamage(DamageResult:damage_result):void=
        if (Processing?):
            return

        set Processing = true
        InAmount := DamageResult.Amount
        if:
            Target := fort_character[DamageResult.Target]
            TargetPlayer := player[Target.GetAgent[]]
        then:
            if:
                PlayerMap := PerPlayerRegisters[InAmount]
                Amount := PlayerMap[TargetPlayer]
            then:
                AdjustDamage(Target, Amount, InAmount)
        set Processing = false

    AdjustDamage(FortChar:fort_character, AdjustedAmount:float, BaseAmount:float):void=
        if (AdjustedAmount >= BaseAmount):
            FortChar.Damage(AdjustedAmount - BaseAmount)
        else:
            FortChar.Heal(BaseAmount - AdjustedAmount)
