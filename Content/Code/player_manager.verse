using { /Fortnite.com/Characters }
using { /Fortnite.com/Devices }
using { /Fortnite.com/FortPlayerUtilities }
using { /UnrealEngine.com/Temporary/SpatialMath }
using { /Verse.org/Random }
using { /Verse.org/Simulation }
using { /UnrealEngine.com/Temporary/UI }
using { /Fortnite.com/Game }
using { /Verse.org/Colors }

#Persistable Player Data Must be comprised of persistable components.
# Saves player data on Fortnite servers between player sessions of our game.
# This is dangerously applied to the Current Quests & Completed Quests. Need to be careful. (Can't use enums)
# https://dev.epicgames.com/documentation/es-es/uefn/using-persistable-data-in-verse
player_profile_data := class<final><persistable>:
    Version:float = 0.0
    CreationTime : float = 0.0
    PlayerLevel:int = 1
    PlayerStrength : big_number = big_number{Value:=1.0}
    LifetimeEssence : big_number = big_number{}
    CurrentEssence : big_number = big_number{}
    StashedEssence : big_number = big_number{}
    #CurrentGems : big_number = big_number{}
    GuardsDestroyed : big_number = big_number{}
    ZoneClearArray : []int = array{ 0,0,0,0,0,0,0,0,0,0,
                                    0,0,0,0,0,0,0,0,0,0 }
    ZoneLevelArray : []int = array{ 1,1,1,1,1,1,1,1,1,1,
                                    1,1,1,1,1,1,1,1,1,1 }
    FailableQuest : quest_stats = quest_stats{}
    CurrentQuests : []quest_stats = array{} #QuestIndex    
    CompletedQuests : []int = array{} #QuestIndex
    Gear : []armor_data = array{}
    LastKnownLocation : vector3 = vector3{}
    LastKnownRotation : vector3 = vector3{}
    LastTimeOnline : float = 0.0
    LastDailyClaim : float = 0.0
    LastPeelyTime : float = 0.0
    BuffStates : [string]buff = map{}
    UpgradeLevels : [string]int = map{}
    Achievements : [string]int = map{}
    Collectibles : [string]collectible_data = map{}
    EquippedItems : []int = array{-1,-1,-1,-1,-1}
    ItemCooldowns : []int = array  {0,0,0,0,0,0,0,0,0,0, #100 items
                                    0,0,0,0,0,0,0,0,0,0,
                                    0,0,0,0,0,0,0,0,0,0,
                                    0,0,0,0,0,0,0,0,0,0,
                                    0,0,0,0,0,0,0,0,0,0,
                                    0,0,0,0,0,0,0,0,0,0,
                                    0,0,0,0,0,0,0,0,0,0,
                                    0,0,0,0,0,0,0,0,0,0,
                                    0,0,0,0,0,0,0,0,0,0,
                                    0,0,0,0,0,0,0,0,0,0}
    PrestigePoints : big_number = big_number{}
    PrestigeCount : int = 0
    PeelysFound : []int = array{}
    MonotonicID : int = 0
    BigRockLevel : int = -1
    BossKeySeconds : int = 0
    ClaimCooldowns : [string]float = map{}
    WeaponLevel : [int]int = map{}
    WeaponExperience : [int]int = map{}
    # CodesUnlocked : []codes_enum = array{}

#Always use this constructor when updating PlayerData values (See examples below)
MakePlayerProfileData<constructor>(Src:player_profile_data)<transacts> := player_profile_data:
    Version := Src.Version
    CreationTime := Src.CreationTime
    PlayerLevel := Src.PlayerLevel
    PlayerStrength := Src.PlayerStrength
    LifetimeEssence := Src.LifetimeEssence
    CurrentEssence := Src.CurrentEssence
    StashedEssence := Src.StashedEssence
    #CurrentGems  := Src.CurrentGems
    GuardsDestroyed := Src.GuardsDestroyed
    ZoneClearArray := Src.ZoneClearArray
    ZoneLevelArray := Src.ZoneLevelArray
    FailableQuest := Src.FailableQuest
    CurrentQuests := Src.CurrentQuests
    CompletedQuests := Src.CompletedQuests
    Gear := Src.Gear
    LastKnownLocation := Src.LastKnownLocation
    LastKnownRotation := Src.LastKnownRotation
    LastTimeOnline := Src.LastTimeOnline
    LastDailyClaim := Src.LastDailyClaim
    LastPeelyTime := Src.LastPeelyTime
    BuffStates := Src.BuffStates
    UpgradeLevels := Src.UpgradeLevels
    Achievements := Src.Achievements
    Collectibles := Src.Collectibles
    EquippedItems := Src.EquippedItems
    ItemCooldowns := Src.ItemCooldowns
    PrestigePoints := Src.PrestigePoints
    PrestigeCount := Src.PrestigeCount
    PeelysFound := Src.PeelysFound
    MonotonicID := Src.MonotonicID
    BigRockLevel := Src.BigRockLevel
    BossKeySeconds := Src.BossKeySeconds
    ClaimCooldowns := Src.ClaimCooldowns
    WeaponLevel := Src.WeaponLevel
    WeaponExperience := Src.WeaponExperience

#Global Variable
var PlayerProfileDataMap:weak_map(player, player_profile_data) = map{}

player_manager:=class<unique>():
    Agent:agent
    Player:player
    GameManager:game_manager
    var CinematicID : int = 0
    var UIManager: ?ui_manager = false #'?' before 'ui_manager' and '= false' makes this an optional variable.  Otherwise, it'll create a circular loop and crash UEFN.
    var QuestManager: ?quest_manager = false #'?' before 'quest_manager' and '= false' makes this an optional variable.  Otherwise, it'll create a circular loop and crash UEFN.
    var PetProp<private>:creative_prop = creative_prop{}
    var ZoneOverkill : float = 0.0 
    var BonusChestContentByZone : [fetchable_tag]big_number = map{fetchable_tag.zone1tag=>big_number{}}
    var PassiveIncomeTimer : int = 60 #Does not count down, used as a reference in a Sleep() function
    var IsLevelLooping : logic = false #Aka IsLevelLooping? Don't override existing leveling ups
    var CheckedOfflineProgression : logic = false
    var GuardSFXResetTimer : int = 0;
    var LastQuarterSecondPosition : vector3 = vector3{}
    var IsSliding : logic = false #god why do we have to make this...
    var GuardComboSFX : int = 1;
    var WeakSpotSFX : int = 1;
    var CurrentZone : fetchable_tag = fetchable_tag.none_tag
    var EliminatedZone : fetchable_tag = fetchable_tag.none_tag
    var CurrentMusicDevice : ?audio_player_wrapper = false

    var Handcannon : int = 0
    var SuperHandcannon : int = 0

    var DamageMasteryEssenceGained : big_number = big_number{}
    var DamageMasteryEssenceStashed : big_number = big_number{}

    var QuestOnUI : quest_stats = quest_stats{}

    var PeelyManager : peely_doll_manager = peely_doll_manager{}

    LastGuardHitTimerAmount : float = 30.0

    #Functions Grouped by: (Ctrl+F'able)
        #Level Functions
        #Strength Functions
        #Essence Functions
        #Stash Functions
        #Passive Income Functions
        #Guard Functions
        #Zone Functions
        #Gem Functions
        #Item Functions
        #Upgrade Functions
        #Quest Functions
        #Damage Functions
        #Buff Functions
        #Daily Functions
        #Misc Functions
        #Debug Functions

    GetSourceDataSafe()<transacts><decides>:player_profile_data=
        Player.IsActive[]
        GameManager.CheckSaveDataForPlayer[Agent]
        SourceData := PlayerProfileDataMap[Player]

    ##Init 
    Init():void=
        Log("Initializing Player!")
        CheckVersionDeltas() #Is this the right place?

        set QuestManager = option{quest_manager{QuestModel:= GameManager.QuestModel, Agent:=Agent, Player:=Player, PlayerManager:=Self}}
        set UIManager = option{ui_manager{Player:=Player, PlayerManager:=Self}}
        GameManager.SwitchToTeam1.ChangeTeam(Agent)
        if:
            GetCurrentQuests[].Length = 0  or not GetCurrentQuests[]
            false?
        then:
            spawn:
                GameManager.CinematicManager.PlayBeginningCinematic(Agent, Self)
        else:
            spawn:
                PostCinematicInit()
        spawn:
            LoopToResetGuardComboSFX()
        spawn:
            LoopingMinuteCheck()
        spawn:
            LoopingSecondCheck()

        spawn:
            LoopingQuarterSecondCheck()

        EquipItemsAtGameStart()
        CheckOfflineProgression()

        set PeelyManager.GameManager = option{GameManager}
        set PeelyManager.Player = option{Player}
        PeelyManager.Init()

        PlayMusic()

        if:
            Char := Player.GetFortCharacter[]
        then:
            Char.DamagedEvent().Subscribe(OnDamageTaken)
            Char.EliminatedEvent().Subscribe(OnPlayerEliminated)

        CheckModifierUpgrades()

        HideObjectsOnStart()
        if (Level := GetLevel[]):
            UpdateLevelOnMaterial(Level)

    HideObjectsOnStart():void=
        for(Actor : GameManager.ActorsToHide):
            GameManager.QueuePeely(queued_peely_doll{Peely := Actor, Agent := option{Agent}, NewState := false})

    CheckSessionStartAnalytics():void=
        if:
            Now:= GetSecondsSinceEpoch()
            LastLogin:= GetRealtime[]
            Created := GetCreationTime[]
            LifetimeDays := Int[(Now - LastLogin) / (60.0 * 60.0 * 24.0)]
        then:
            if(LifetimeDays < 2):
                GameManager.AnalyticsManager.session_start_d0_1.Submit(Agent)
            else:
                if(LifetimeDays < 8):
                    GameManager.AnalyticsManager.session_start_d2_7.Submit(Agent)
                else:
                    if(LifetimeDays < 15):
                        GameManager.AnalyticsManager.session_start_d8_14.Submit(Agent)
                    else:
                        if(LifetimeDays < 31):
                            GameManager.AnalyticsManager.session_start_d15_30.Submit(Agent)
                        else:
                            GameManager.AnalyticsManager.session_start_d_gte_31.Submit(Agent)

    SkipPrison()<suspends>:void= 
        Sleep(0.1)
        GameManager.AnalyticsManager.session_start_bad_spawn.Submit(Agent)
        CheckSessionStartAnalytics()
        if:
            SetQuestManager:= QuestManager?
            Quest:=GetCurrentQuests[][0]
            Quest.QuestID = 0
            Quest.TaskID = 0
        then:
            SetQuestManager.CompleteTask(Quest) #Skip Cell Wall


    PostCinematicInit()<suspends>:void=
        if:
            Player.IsActive[]
        then:
            if(SetUIManager:= UIManager?):
                SetUIManager.Init()

            if(SetQuestManager:= QuestManager?):
                SetQuestManager.Init()

                if:
                    GetCurrentQuests[].Length = 0 or not GetCurrentQuests[]
                then:
                    GameManager.AnalyticsManager.session_start_new_player.Submit(Agent)
                    SetQuestManager.BeginStartingQuests()
                    GainUpgradeLevel("Momentum", 1)
                    GameManager.SwitchToClassNone.ChangeClass(Agent)
                else:   
                    GameManager.AnalyticsManager.session_start_old_player.Submit(Agent)
                    CheckSessionStartAnalytics()

                    spawn:
                        GoToPersistentLocation()
        
            Sleep(0.25)
            GameManager.CheckBarrierLimit(Agent, Self)
            if(SetUIManager:= UIManager?):
                if:
                    Gems := GetGems[]
                then:
                    SetUIManager.UpdateGemUI(Gems, big_number{})
          
    GoToPersistentLocation()<suspends>:void=
        Sleep(0.01)
        if: 
            not IsAlmostEqual[GetLocation[], vector3{X := 0.0, Y := 0.0, Z := 0.0}, 0.0001] #if we haven't set the location yet, don't teleport 
            set CurrentZone = fetchable_tag.none_tag
            Player.GetFortCharacter[].TeleportTo[GetLocation[], GetRotation[]]
        else:
            #If fails, go to inn or stay put.
            Print("GoToPersistentLocation failed")
            
            if:
                GetAchievementLevel["EscapedMidasCastle"] > 0
            then:
                GainAchievementLevel("Inn",1)
                GainAchievementLevel("Midas Castle",1)
                GameManager.SwitchToClass1.ChangeClass(Agent)
                if:
                    InnTransform:= GameManager.InnTeleporter.GetTransform()
                    Player.GetFortCharacter[].TeleportTo[InnTransform.Translation, InnTransform.Rotation]
            else:
                GameManager.SwitchToClassNone.ChangeClass(Agent)

            #Rename Boar Area into Umbral Forest
            if:
                GetAchievementLevel["Boar Area"] > 0
            then:
                GainAchievementLevel("Umbral Forest",1)

    #Big Rock Level Functions
    GetBigRockLevel()<transacts>:int=
        if:
            SourceData := GetSourceDataSafe[]
            value := SourceData.BigRockLevel
        then:
            return value
        else:
            return 0

    SetBigRockLevel(Level : int)<transacts><decides>:void=
        SourceData := GetSourceDataSafe[]
        set PlayerProfileDataMap[Player] = player_profile_data:
            MakePlayerProfileData<constructor>(SourceData)
            BigRockLevel := Level

    # Weapon Exp Functions
    GainWeaponLevel(ID : int)<transacts><decides>:void=
        SourceData := GetSourceDataSafe[]
        CurrentLevel := GetWeaponLevel(ID)
        var NewMap : [int]int := SourceData.WeaponLevel
        set NewMap[ID] = CurrentLevel + 1
        set PlayerProfileDataMap[Player] = player_profile_data:
            MakePlayerProfileData<constructor>(SourceData)
            WeaponLevel := NewMap    

    GetWeaponLevel(ID : int)<transacts>:int=
        if:
            SourceData := GetSourceDataSafe[]
            value := SourceData.WeaponLevel[ID]
        then:
            return value
        else:
            return 1

    GetWeaponExperience(ID : int)<transacts>:int=
        if:
            SourceData := GetSourceDataSafe[]
            value := SourceData.WeaponExperience[ID]
            temp := GetWeaponLevelRequirement(GetWeaponLevel(ID))
        then:
            #overflow handling
            if:
                value > temp
            then:
                return temp - 1
            return value
        else:
            return 0
    
    GrantWeaponExperience(IsKO:logic):void=
        if:
            ID:= GetHeldGun[]
            OldLevel:= GetWeaponLevel(ID)
        then:
            var WeaponExpAdj : int = 1

            case(ID):
                3 => set WeaponExpAdj = 15 #Rocket
                4 => set WeaponExpAdj = 4 #Shotgun
                5 => set WeaponExpAdj = 8 #Bow
                6 => set WeaponExpAdj = 15 #Railgun
                9 => set WeaponExpAdj = 4 #Stakegun
                10 => set WeaponExpAdj = 4 #Scythe
                _ => 
            
            if:
                IsKO?
            then:
                set WeaponExpAdj = WeaponExpAdj * 10

            if:
                IsDailyWeapon(ID)?
            then:
                set WeaponExpAdj = WeaponExpAdj * 2

            if:
                GainWeaponExperience[ID, WeaponExpAdj]

            NewLevel := GetWeaponLevel(ID)
            UpdateWeaponExperienceUI()
            if:
                NewLevel > OldLevel
            then:
                GameManager.AudioManager.GunLevelUp.Play(Agent)
                GameManager.AccoladesManager.GunLevelUp.Award(Agent)

    GainWeaponExperience(ID:int, WeaponExperience:int)<transacts><decides>:void=
        OldExp := GetWeaponExperience(ID)
        NewExperience :=  OldExp + WeaponExperience
        NewExperience >= 0

        SourceData := GetSourceDataSafe[]
        var NewMap : [int]int := SourceData.WeaponExperience
        set NewMap[ID] = NewExperience

        set PlayerProfileDataMap[Player] = player_profile_data:
            MakePlayerProfileData<constructor>(SourceData)
            WeaponExperience := NewMap

        if:
            CheckWeaponLevelUp[ID, NewExperience]

    CheckWeaponLevelUp(ID : int, NewExp : int)<transacts><decides>:void=
        CurrentWeaponLevel := GetWeaponLevel(ID)
        Req := GetWeaponLevelRequirement(CurrentWeaponLevel)
        NewExp >= Req
        GainWeaponLevel[ID]
        GainWeaponExperience[ID, -Req]
            
    GetWeaponLevelRequirement(Level : int)<transacts>:int=
        Number : float = (50 * Pow(1.50, (Level - 1) *1.0))
        MultipleOf5 : float = Number / 5.0
        if:
            Result : int = Round[MultipleOf5] * 5 
        then:
            return Result
        return 5

    #Level Functions
    GetLevel()<transacts><decides>:int=
        SourceData := GetSourceDataSafe[]
        return SourceData.PlayerLevel

    GetLevelProgress()<transacts><decides>: tuple(float, tuple(big_number, big_number)) =    
        CurrentEssence : big_number = GetEssence[]
        TargetEssence : big_number = GetLevelRequirement(GetLevel[])
        Percent : float = Min(BigNumber_Percentage(CurrentEssence, TargetEssence), 1.0)
        return (Percent, (CurrentEssence, TargetEssence))

    GetLevelRequirement(Level : int) <transacts>: big_number=
        
        return (1.0-GetUpgradeValueSafe("P_ReducedXPRequirement", false)) * case (Level):
            1 => big_number{Value:=1.0, Exponent:=0}
            2 => big_number{Value:=2.0, Exponent:=0}
            3 => big_number{Value:=3.0, Exponent:=0}
            4 => big_number{Value:=1.0, Exponent:=1}
            5 => big_number{Value:=2.0, Exponent:=1}
            6 => big_number{Value:=4.0, Exponent:=1}
            7 => big_number{Value:=6.0, Exponent:=1}
            8 => big_number{Value:=7.0, Exponent:=1}
            9 => big_number{Value:=8.0, Exponent:=1}
            10 => big_number{Value:=9.0, Exponent:=1}
            11 => big_number{Value:=2.0, Exponent:=2}
            12 => big_number{Value:=3.5, Exponent:=2}
            13 => big_number{Value:=5.5, Exponent:=2}
            14 => big_number{Value:=7.0, Exponent:=2}
            15 => big_number{Value:=1.0, Exponent:=4}
            16 => big_number{Value:=3.0, Exponent:=4}
            17 => big_number{Value:=4.0, Exponent:=4}
            18 => big_number{Value:=5.0, Exponent:=4}
            19 => big_number{Value:=7.5, Exponent:=4}
            20 => big_number{Value:=1.0, Exponent:=5}
            _  => MakeBigNumber(GameManager.LevelReqScaleRatio, Level - 20) * GetLevelRequirement(20)

    LoopLevelUps()<suspends>:void= 
        set IsLevelLooping = true
        var count : int = 0
        var HighSpeedSetting:int = 0
        var MaxIterationSpeed : float = 0.60
        var ShowUIUpdates : logic = true
        var MultiLevelUp : int = 1
        var EssenceRequirement : big_number = big_number{}
        var DidLevelUp : logic = false
        loop:

            if:
                CurrentEssence := GetEssence[]
                CurrentLevel:=GetLevel[]
                CurrentStash:=GetStashedEssence[]
            then:
                set EssenceRequirement = BigNumber_CalculateCostRange(GetLevelRequirement(CurrentLevel), 0, MultiLevelUp, GameManager.LevelReqScaleRatio)
                set DidLevelUp = BigNumber_GreaterThanOrEquals(CurrentEssence, EssenceRequirement)

                if (DidLevelUp?):
                    SpendEssence(EssenceRequirement, DidLevelUp) #Good
                    GainLevel(CurrentLevel, MultiLevelUp, EssenceRequirement, HighSpeedSetting) #Good
                    if(ShowUIUpdates?): 
                        if(SetUIManager:=UIManager?):
                            SetUIManager.UpdateEssenceUIOnLevelUp() #Good
                            SetUIManager.GainLevel(MultiLevelUp) #Good
                            SetUIManager.UpdateLevelUI() #Good
                            SetUIManager.UpdateStashUI(CurrentStash, big_number{}, CurrentStash)
                else:
                    # if 1, then exit out.
                    if(MultiLevelUp = 1):
                        set IsLevelLooping = false
                        UpdateQuestProgressForTask(task_event.PlayerLevel, big_number{Value:= CurrentLevel * 1.0}, false, ?IsPersistent := true)
                        return
                    #Lower MultiLevelUp in Stages, 100 -> 25 -> 5 -> 1
                    if(MultiLevelUp = 5):
                        set MultiLevelUp = 1
                    if(MultiLevelUp = 25):
                        set MultiLevelUp = 5
                    if(MultiLevelUp = 100):
                        set MultiLevelUp = 25

                #Increment count number at MultiLevelUp
                set count += 1
                if(count = 5):
                    if(MultiLevelUp = 25):
                        set HighSpeedSetting = 2
                        set MultiLevelUp = 100
                        set count = 3
                    if(MultiLevelUp = 5):
                        set HighSpeedSetting = 2
                        set MultiLevelUp = 25
                        set count = 3
                    if(MultiLevelUp = 1):
                        set HighSpeedSetting = 1
                        set MultiLevelUp = 5
                        set count = 3
                        
                Sleep(0.75 - Min((count * 1.0 / 15.0), MaxIterationSpeed))

    CheckSkillPointOnLevelUp(Level : int):void=
        if:
            Level > 1
        then:
            spawn:
                GainGems(big_number{Value:=1.0}, ?DidLevelUp:=true)

    GainLevel(CurrentLevel:int, LevelsGained:int, Essence_Req:big_number, SpeedSetting:int):void=
        spawn:
            FullHeal()

        GameManager.AccoladesManager.PlayerLevelUp.Award(Agent)

        #AudioCues
        if(SpeedSetting = 1):
            GameManager.AudioManager.PlayHighSpeedLevelUpSFX(Agent)
        else:
            if(SpeedSetting = 2):
                GameManager.AudioManager.PlayVeryHighSpeedLevelUpSFX(Agent)
            else: 
                GameManager.AudioManager.PlayLevelUpSFX(Agent)
    
        #Iterate Through Strength Gains
        for(X:=1..LevelsGained):
            GainStrength(GetStrengthGainedPerLevel(CurrentLevel + X))
            CheckSkillPointOnLevelUp(CurrentLevel + X)
        
        UpdateLevelOnMaterial(CurrentLevel + LevelsGained)
        
        #Changing Values
        if:
            SourceData := GetSourceDataSafe[]
            set PlayerProfileDataMap[Player] = player_profile_data:
                MakePlayerProfileData<constructor>(SourceData)
                PlayerLevel := SourceData.PlayerLevel + LevelsGained

        # ClosePopupUI()
        # GameManager.LevelPopupDevice.OpenMenu(Player, ?Force := true)

        GameManager.CheckBarrierLimit(Agent, Self)
    
    UpdateLevelOnMaterial(Level : int):void=
        spawn:
            UpdateLevelOnMaterial_Interal(Level)
        
    UpdateLevelOnMaterial_Interal(Level : int)<suspends>:void=
        var Idx : int = -1
        var Counter : int = 0
        for (Key -> Value : GameManager.PlayerManagers):
            if (Value = Self):
                set Idx = Counter
            set Counter += 1
            
        if:
            not Idx = -1
            PowerupDevice := GameManager.PlayerLevelVFXPowerps[Idx]
            CinematicDevice := GameManager.PlayerLevelMaterialSetters[Idx]
        then:
            # Log("setting playback frame to {Level}")
            CinematicDevice.SetPlaybackFrame(Level)
            CinematicDevice.Play()

            if (not PowerupDevice.HasEffect[Agent]):
                PowerupDevice.Pickup(Agent)

            Sleep(0.1)
            CinematicDevice.SetPlaybackFrame(Level)
            CinematicDevice.Play()
            Sleep(0.1)
            CinematicDevice.SetPlaybackFrame(Level)
            CinematicDevice.Play()
            Sleep(0.1)
            CinematicDevice.SetPlaybackFrame(Level)
            CinematicDevice.Play()
            CinematicDevice.Pause()
            
    #Strength Functions
    #Includes all multipliers
    GetAdjustedStrength(EnemyClasses : []string, ?EnemyWeaknesses : [int]float = map{}, ?DefaultWeakness : float = 1.0)<decides><transacts>:big_number=
        var DamageMult : big_number = big_number{Value:= 1.0}
        if (GunMult := GetCurrentGunStats[].DamageMult):
            set DamageMult = DamageMult * GunMult

        if:
            #HasPistol
            GetHeldGun[] = 1
            #IsCrouching
            Agent.GetFortCharacter[].IsCrouching[]
            #HasStealthAttackUpgrade
            Temp := 1.0 + (GetUpgradeValue["Stealth Attack", false] / 100.0)
        then:
            set DamageMult = DamageMult * Temp 

        if:
            #HasRifle
            GetHeldGun[] = 2
            #IsSliding
            IsSliding?
            #HasStealthAttackUpgrade
            Temp := 1.0 + (GetUpgradeValue["Sliding Rifle", false] / 100.0)
        then:
            set DamageMult = DamageMult * Temp 

        if (GunID := GetHeldGun[]):

            if (EnemyWeaknesses.Length > 0):
                if(Weakness := EnemyWeaknesses[GunID]):
                    set DamageMult = DamageMult * (1.0*Weakness)
                else:
                    set DamageMult = DamageMult * DefaultWeakness
        
        if:
            GunStatType := GetCurrentGunStats[].DamageStatType
            not GunStatType = player_stat.None
        then:
            set DamageMult = DamageMult * GetPlayerStat(GunStatType)
        else:
            set DamageMult = DamageMult * GetPlayerStat(player_stat.Damage_With_Pickaxe)

        if:
            WeaponClass := GetCurrentGunStats[].WeaponClass
        then:
            case (WeaponClass):
                weapon_class.Melee => set DamageMult = DamageMult * GetPlayerStat(player_stat.Damage_Melee)
                weapon_class.Rapid => set DamageMult = DamageMult * GetPlayerStat(player_stat.Damage_Rapid)
                weapon_class.Precision => set DamageMult = DamageMult * GetPlayerStat(player_stat.Damage_Precision)
                _ => block:
        
        if:
            Handcannon = 3
            Temp := 1.0 + (GetUpgradeValue["Handcannon", false] / 100.0)
            GunMult := GetCurrentGunStats[].HandcannonMult
        then:
            set DamageMult = DamageMult * Temp * GunMult
            
        if:
            SuperHandcannon = 3
            Temp := 1.0 + (GetUpgradeValue["SuperHandcannon", false] / 100.0)
        then:
            set DamageMult = DamageMult * Temp  
        
        if (MasteryMult := GetMasteryMultiplier[EnemyClasses]):
            set DamageMult = DamageMult * MasteryMult
        if:
            BuffStates := GetBuffStates[]
            Temp := (1.0 + (BuffStates["Focus Fire"].Amount / 100.0))
            set DamageMult = DamageMult * Temp

        if:
            BuffStates := GetBuffStates[]
            Temp := (1.0 + (BuffStates["Momentum"].Amount / 100.0))
            set DamageMult = DamageMult * Temp

        return GetDisplayStrength[] * DamageMult

    #Base strength * global multipliers
    GetDisplayStrength()<decides><transacts>:big_number=
        var Result : big_number = GetStrength[] * GetPlayerStat(player_stat.DamageMult)
        return Result

    #Base strength
    GetStrength()<decides><transacts>:big_number=
        SourceData := GetSourceDataSafe[]
        return SourceData.PlayerStrength

    GetMasteryMultiplier(EnemyClasses : []string)<decides><transacts>:big_number=
        var DamageMult : big_number = big_number{Value:=1.0, Exponent:=0}
        if (EnemyClasses.Find["Beast"]):
            set DamageMult = DamageMult * GetPlayerStat(player_stat.Damage_Beast)
            
        if (EnemyClasses.Find["Guard"] or EnemyClasses.Find["Humanoid"]):
            set DamageMult = DamageMult * GetPlayerStat(player_stat.Damage_Guard)
            
        if (EnemyClasses.Find["Boss"]):
            set DamageMult = DamageMult * GetPlayerStat(player_stat.Damage_Boss)
        
        if (EnemyClasses.Find["Mineral"]):
            set DamageMult = DamageMult * GetPlayerStat(player_stat.Damage_Mineral)
            if (GetHeldGun[] = -1):
                set DamageMult = DamageMult * GetPlayerStat(player_stat.Empowered_Mining)

        return DamageMult


    GainStrength(Gain: big_number):void=
        if:
            NewStrength := BigNumber_Add(GetStrength[], Gain)
            SourceData := GetSourceDataSafe[]
            set PlayerProfileDataMap[Player] = player_profile_data:
                MakePlayerProfileData<constructor>(SourceData)
                PlayerStrength := NewStrength
            
    GetStrengthGainedPerLevel(Level:int):big_number=
        var TenLevelFloor : int = 0
        if:
            Temp := Floor(Level/10)
            set TenLevelFloor = Temp

        return case (Level):
            1 => big_number{Value:=1.0, Exponent:=0}  #1
            2 => big_number{Value:=2.0, Exponent:=0}  #3
            3 => big_number{Value:=3.0, Exponent:=0}  #6
            4 => big_number{Value:=4.0, Exponent:=0}  #10
            5 => big_number{Value:=4.0, Exponent:=0}  #14
            6 => big_number{Value:=4.0, Exponent:=0}  #18
            7 => big_number{Value:=5.0, Exponent:=0}  #23
            8 => big_number{Value:=5.0, Exponent:=0}  #28
            9 => big_number{Value:=5.0, Exponent:=0}  #33
            10 => big_number{Value:=7.0, Exponent:=0} #40
            11 => big_number{Value:=1.0, Exponent:=1} #50
            _ => MakeBigNumber(1.13, Level) * MakeBigNumber(2.0, TenLevelFloor) + 4.0
          
    #Essence Functions
    GetEssence()<decides><transacts>:big_number=
        SourceData := GetSourceDataSafe[]
        return SourceData.CurrentEssence

    GetLifetimeEssence()<transacts><decides>:big_number=
        SourceData := GetSourceDataSafe[]
        return SourceData.LifetimeEssence

    ProcessEssence(ZoneTag:fetchable_tag, Gain:big_number, Overkill:big_number, KOBonusIn:big_number, DisplayGain: logic, IsWeakHit:logic, IsHandcannon : logic, IsAmbush : logic, IsBlocked : logic):void=
        var OverkillBonus : big_number = BigNumber_Multiply(Overkill, ZoneOverkill)
        var KOBonus : big_number = KOBonusIn

        if (GetUpgradeValue["KO Bonus", false] <= 0.0):
            set OverkillBonus = big_number{}
            set KOBonus = big_number{}

        var EarnedBonus : big_number = BigNumber_Add(KOBonus, OverkillBonus)

        if (IsHandcannon?):
            GameManager.AudioManager.PlayHandcannon(Agent)
        

        GainEssence(Gain, EarnedBonus, big_number{}, DisplayGain, IsWeakHit, IsHandcannon, IsAmbush, IsBlocked)

        # Too frequent
        # GameManager.AccoladesManager.HarvestEssenceAward(Agent)

    GainEssence(Gain:big_number, KOBonus:big_number, ClearBonus:big_number, DisplayGain:logic, IsWeakHit:logic, IsHandcannon : logic, IsAmbush : logic, IsBlocked : logic):void=
        if:
            TotalGain := (Gain + KOBonus + ClearBonus)
            NewCurrentEssence := GetEssence[] + TotalGain
            NewLifetimeEssence := GetLifetimeEssence[] + TotalGain
            CurrentLevel := GetLevel[]
            CurrentStash:=GetStashedEssence[]

            SourceData := GetSourceDataSafe[]
            set PlayerProfileDataMap[Player] = player_profile_data:
                MakePlayerProfileData<constructor>(SourceData)
                LifetimeEssence := NewLifetimeEssence
                CurrentEssence := NewCurrentEssence
        then:
            DidLevelUp := BigNumber_GreaterThanOrEquals(NewCurrentEssence, GetLevelRequirement(CurrentLevel))

            UpdateQuestProgressForTask(task_event.GatherEssence, TotalGain, false)

            if(SetUIManager:=UIManager?):
                if(DisplayGain?):
                    SetUIManager.QueueDamageNumber(Gain, KOBonus, ClearBonus, IsWeakHit, IsHandcannon, IsAmbush, IsBlocked)   
                    SetUIManager.UpdateStashUI(CurrentStash, big_number{}, CurrentStash)

            if(not(IsLevelLooping?)):
                spawn:
                    LoopLevelUps()
        
    SpendEssence(Spend:big_number, DidLevelUp:logic):void=     
        if:   
            NewCurrentEssence := BigNumber_Subtract(GetEssence[], Spend)
            SourceData := GetSourceDataSafe[]
            set PlayerProfileDataMap[Player] = player_profile_data:
                MakePlayerProfileData<constructor>(SourceData)
                CurrentEssence := NewCurrentEssence

        if(not(DidLevelUp?)):
            if(SetUIManager:=UIManager?):
                SetUIManager.UpdateEssenceUI()
        
    #Stash Functions
    GetStashedEssence()<decides><transacts>:big_number=
        SourceData := GetSourceDataSafe[]
        return SourceData.StashedEssence

    StashEssence(Gain:big_number): void=
        if:
            Player.IsActive[]
            StashedEssence:=GetStashedEssence[] 
            Gain.Value > 0.0
            BigNumber_GreaterThanOrEquals(Gain, big_number{})?
        then:
            NewStashedEssence:= BigNumber_Add(StashedEssence,Gain)
            if:
                SourceData := GetSourceDataSafe[]
                set PlayerProfileDataMap[Player] = player_profile_data:
                    MakePlayerProfileData<constructor>(SourceData)
                    StashedEssence := NewStashedEssence

            if(SetUIManager:=UIManager?):
                SetUIManager.UpdateStashUI(NewStashedEssence, Gain, NewStashedEssence)
            
            UpdateQuestProgressForTask(task_event.FillStash, Gain, false) 

    CollectStash():void=
        if:
            Player.IsActive[]
            CurrentValue:= GetStashedEssence[]
        then:
            if:        
                SourceData := GetSourceDataSafe[]
                set PlayerProfileDataMap[Player] = player_profile_data:
                    MakePlayerProfileData<constructor>(SourceData)
                    StashedEssence := big_number{}
            then:
                GainEssence(CurrentValue, big_number{}, big_number{}, true, false, false, false, false)
                UpdateQuestProgressForTask(task_event.CollectStash, CurrentValue, false)
                if:
                    CurrentValue.Value > 0.0
                then:
                    GameManager.AccoladesManager.CollectStashAward(Agent)
            if(SetUIManager:=UIManager?):
                SetUIManager.UpdateStashUI(big_number{}, big_number{}, CurrentValue)

    #When IsSliding Is added, delete this please.
    LoopingQuarterSecondCheck()<suspends>:void=
        loop:
            Sleep(0.25)
            if:
                Player.IsActive[]
                Fort:= Agent.GetFortCharacter[]
                NewPosition := Fort.GetTransform().Translation   
            then:
                RecentDistance : float = Distance(LastQuarterSecondPosition, NewPosition)
                RecentSpeed : float = RecentDistance / 0.25
                MSMult := GetMoveSpeedMult().Value
                if:
                    Fort.IsCrouching[]
                    RecentSpeed > 270.0 * MSMult
                then:
                    set IsSliding = true
                else:
                    set IsSliding = false

                set LastQuarterSecondPosition = NewPosition

    LoopingSecondCheck()<suspends>:void=
        loop:
            Sleep(0.01)
            if (Player.IsActive[]):
                # if (FC := Player.GetFortCharacter[]):
                #     Log("{FC.GetHealth()} Health")

                if:
                    SetUIManager := UIManager?
                then:
                    SetUIManager.UpdateStashPassiveUI()

                UpdateItemCooldowns()
                CheckForOpenGunSlots()
                LoopBuffDurations()
                UpdateQuestProgressForTask(task_event.None, big_number{Value:=1.0}, false)

                #Note (Tazz): This is kinda silly but whatever
                CheckSwimming()
                Sleep(0.25)
                CheckSwimming()
                Sleep(0.25)
                CheckSwimming()
                Sleep(0.25)
                CheckSwimming()
                Sleep(0.25)
                CheckSwimming()

                CheckAFKEssence()
                TickRegen()

                CheckAutoFarmChests()
                
                AddBossKeySeconds(1)

                spawn{SetLocationAndRotation(false)}

    OnRespawn():void=
        if:
            GetAchievementLevel["EscapedMidasCastle"] > 0
        then:
            GameManager.SwitchToClass1.ChangeTeam(Agent)
        else:
            GameManager.SwitchToClassNone.ChangeTeam(Agent)

        GameManager.SwitchToTeam1.ChangeTeam(Agent)
        CheckModifierUpgrades()
        if (GetEquippedItemIDs[].Find[0]):
            RemoveItem(0)
            EquipItem(0)

        if:
            Print("AAA {EliminatedZone.ToZoneName()} {CurrentZone.ToZoneName()}")
            EliminatedZone = fetchable_tag.prestigeareatag or CurrentZone = fetchable_tag.prestigeareatag
        then:
            if:
                Player.GetFortCharacter[].TeleportTo[GameManager.PrestigeTeleporter.GetTransform().Translation, GameManager.PrestigeTeleporter.GetTransform().Rotation]
            Print("Teleported Prestige Area")
        else:
            if:
                EliminatedZone.IsBossTag()?
                WaitingRoomTransform := case (EliminatedZone):
                    fetchable_tag.lyratag => GameManager.BossInfo.SargeWaitingRoom.GetTransform()
                    fetchable_tag.sharkeytag => GameManager.BossInfo.SharkyWaitingRoom.GetTransform()
                    fetchable_tag.gengatag => GameManager.BossInfo.GengaWaitingRoom.GetTransform()
                    fetchable_tag.urjenttag => GameManager.BossInfo.UrjentWaitingRoom.GetTransform()
                    fetchable_tag.piratekingtag => GameManager.BossInfo.PirateKingWaitingRoom.GetTransform()
                    _ => Player.GetFortCharacter[].GetTransform()
                Player.GetFortCharacter[].TeleportTo[WaitingRoomTransform.Translation, WaitingRoomTransform.Rotation]
                Print("Teleported to waiting room")
            else:
                Print("Teleported to Persistent Location")
                spawn {GoToPersistentLocation() }

            
        for (Teleporter : GameManager.BossTeleportManager.Teleporters):
            Teleporter.RemovedFromFight(Self)

        if:
            EliminatedZone = fetchable_tag.prestigeareatag or CurrentZone = fetchable_tag.prestigeareatag
        then:
            block:
        else:
            set EliminatedZone = fetchable_tag.none_tag

        spawn:
            RespawnFixUI()

    RespawnFixUI()<suspends>:void=
        Sleep(0.1)
        if:
            SetUIManager := UIManager?
            Popup := SetUIManager.Popup?
            PlayerUI := GetPlayerUI[Player]
        then:
            PlayerUI.RemoveWidget(SetUIManager.MainCanvas)
            NewUIManager := ui_manager{Player:=Player, PlayerManager:=Self}
            set UIManager = option{NewUIManager}
            NewUIManager.Init()
            NewUIManager.UpdateUICounts()

            PlayerUI.RemoveWidget(Popup)
            set NewUIManager.Popup = option{Popup}
            PlayerUI.AddWidget(Popup, player_ui_slot{InputMode := ui_input_mode.All})

        PlayMusic(?IsOnRespawn := true)

    HideAllGameUI()<suspends>:void=
        if:
            SetUIManager := UIManager?            
            PlayerUI := GetPlayerUI[Player]
        then:
            PlayerUI.RemoveWidget(SetUIManager.MainCanvas)
            if:
                Popup := SetUIManager.Popup?
            then:
                PlayerUI.RemoveWidget(Popup)

    PlayMusic(?IsOnRespawn : logic = false):void=
        if(SetQuestManager:=QuestManager?):
            for (QuestStats : GetCurrentQuests[]):
                if:
                    QuestMusicDevice := SetQuestManager.QuestModel.Quests[QuestStats.QuestID].Tasks[QuestStats.TaskID].PlayMusicOnStart?
                then:
                    if:
                        SetCurrentMusic  := CurrentMusicDevice?
                        not QuestMusicDevice.ID = SetCurrentMusic.ID
                    then:
                        SetCurrentMusic.APD.Stop(Agent)
                        set CurrentMusicDevice = option{QuestMusicDevice}
                    else:
                        set CurrentMusicDevice = option{QuestMusicDevice}

                    if:
                        IsOnRespawn?
                    then:
                        QuestMusicDevice.APD.Stop(Agent)    
                    QuestMusicDevice.APD.Register(Agent)
                    QuestMusicDevice.APD.Play(Agent)
                    return

        NewMusicDevice := GameManager.AudioManager.GetMusicForZone(Agent, CurrentZone)

        if:
            SetCurrentMusic := CurrentMusicDevice?
            not NewMusicDevice.ID = SetCurrentMusic.ID
        then:
            SetCurrentMusic.APD.Stop(Agent)
            set CurrentMusicDevice = option{NewMusicDevice}
        else:
            set CurrentMusicDevice = option{NewMusicDevice}

        if:
            IsOnRespawn?
        then:
            NewMusicDevice.APD.Stop(Agent)
        
        NewMusicDevice.APD.Register(Agent)
        NewMusicDevice.APD.Play(Agent)

    DoStasis():void=
        if:
            Character := Player.GetFortCharacter[]          
            Character.IsActive[]  
        then:
            Character.PutInStasis(stasis_args{AllowFalling:=true, AllowTurning:=false})

    ReleaseStasis():void=
        if:
            Character := Player.GetFortCharacter[]            
            Character.IsActive[]
        then:
            Character.ReleaseFromStasis()

    CheckSwimming()<suspends>:void=
        if:
            Player.IsActive[]
            Char := Player.GetFortCharacter[]
            Char.IsInWater[]
            not Char.IsOnGround[]
            not(CurrentZone = fetchable_tag.forest_zone1tag or CurrentZone = fetchable_tag.forest_zone2tag or CurrentZone.IsBossTag()? or CurrentZone = fetchable_tag.forest_bandit_zone1tag or CurrentZone = fetchable_tag.forest_bandit_zone2tag)
        then:
            if (GetAchievementLevel["Swimming_Unlocked"] <= 0):
                Char.PutInStasis(stasis_args{AllowTurning := true, AllowFalling := false, AllowEmotes := false})
            
                if:
                    Swim0:= GameManager.AudioManager.LyraSwim[0]
                    Swim1:= GameManager.AudioManager.LyraSwim[1]
                then:
                    spawn:
                        MakeDialog("", array{
                            dialog_message{
                                Title := option{"Lyra"}
                                Message := "Woah! You don't know how to swim yet!<br>Hold on, I'm teleporting you back to the land!"
                                VO := option{Swim0}
                            },
                            dialog_message{
                                Title := option{"Lyra"}
                                Message := "That was a close one.<br>You'll have to learn to swim before you can traverse the water."
                                VO := option{Swim1}
                            }
                        }
                        ).ShowDialog(Self, 0.01)
                Sleep(2.0)
                GoToPersistentLocation()
                Char.ReleaseFromStasis()

    CheckAFKEssence():void=
        if:
            Player.IsActive[]
            CurrentZone = fetchable_tag.afk_zonetag
            CurrentStrength := GetDisplayStrength[]
        then:
            GainEssence(0.9*CurrentStrength, big_number{}, big_number{}, true, false, false, false, false)

    ShouldAutoFarmChest(Name:string)<transacts>:logic=
        UpgradeLevel := GetUpgradeLevel["P_AutoFarmChest"] or 0
        return case (Name):
            "BeachXP" => logic{UpgradeLevel > 0}
            "BeachCaveXP" => logic{UpgradeLevel > 1}
            "ForestXPChest" => logic{UpgradeLevel > 2}
            "BanditXPChest" => logic{UpgradeLevel > 3}
            "RaidersXPChest" => logic{UpgradeLevel > 4}
            "GiantsXPChest" => logic{UpgradeLevel > 5}
            "PirateIslandXPChest" => logic{UpgradeLevel > 6}
            "PirateBaseXPChest" => logic{UpgradeLevel > 7}
            _ => false

    CheckAutoFarmChests():void=
        if:
            Player.IsActive[]
        then:
            if:
                ZoneManager := GameManager.GetZoneManagerFromZoneTag[fetchable_tag.gunxpchesttag]?
            then:
                for (GuardManager : ZoneManager.GuardManagers):
                    if (ShouldAutoFarmChest(GuardManager.WeaponExperienceData?.Name)?):
                        GuardManager.CheckWeaponExperienceData(Self, ?DisplayNotice := false)

    TickRegen():void=
        if:
            Player.IsActive[]
            FortChar := Agent.GetFortCharacter[]
        then:
            FortChar.SetHealth(FortChar.GetHealth() + BigNumber_ToFloat(GetPlayerStat(player_stat.Regen)))

    ResetItemCooldown(GunID : int):void=
        if:
            SourceData := GetSourceDataSafe[]
            var NewCooldowns : []int = SourceData.ItemCooldowns
        then:
            for (i := 0..GameManager.GunManager.StatsTable.Length-1):
                if:
                    var UpdatedCD : int = NewCooldowns[i]
                    if:
                        i = GunID
                    then:
                        set UpdatedCD = 0

                    UpdatedCD >= 0
                    set NewCooldowns[i] = UpdatedCD
                    GunStats := GameManager.GunManager.StatsTable[i]
                then:
                    if:
                        GetRemainingDuration[GunID] > 0
                        i = GunID
                    then:
                        if: 
                            SetUIManager := UIManager?
                        then:
                            RemoveItem(i)
                            SetUIManager.HideGunDurationWidget()
                    if:
                        GameManager.CheckSaveDataForPlayer[Agent]
                        SourceData2 := PlayerProfileDataMap[Player]
                        set PlayerProfileDataMap[Player] = player_profile_data:
                            MakePlayerProfileData<constructor>(SourceData2)
                            ItemCooldowns := NewCooldowns

    CheckForOpenGunSlots():void=
        if:
            Player.IsActive[]
            HasCompletedTask(0,2)?
            SourceData := GetSourceDataSafe[]
            var NewCooldowns : []int = SourceData.ItemCooldowns
        then:
            for (i := 0..GameManager.GunManager.StatsTable.Length-1):
                if:
                    i > 0
                    GetUpgradeLevel["Gun_{i}"] > 0  or i = 1
                    NewCooldowns[i] <= 0
                then:
                    EquipItem(i)


    UpdateItemCooldowns(?DeltaSeconds : int = 1, ?TargetCD : ?int = false, ?IDTargeted : ?int = false):void=
        if:
            Player.IsActive[]
            HasCompletedTask(0,2)?
            SourceData := GetSourceDataSafe[]
            var NewCooldowns : []int = SourceData.ItemCooldowns
        then:
            var DidUpdate : logic = false
            var CDStates : []cooldown_state = array{}
            for (i := 0..GameManager.GunManager.StatsTable.Length-1):
                if:
                    var UpdatedCD : int= NewCooldowns[i] - DeltaSeconds
                    if:
                        ID:=IDTargeted?
                        TargetedCD := TargetCD?
                        ID = i
                    then:
                        set UpdatedCD = TargetedCD

                    GunStats := GameManager.GunManager.StatsTable[i]

                    Expended := GunStats.Cooldown - UpdatedCD
                    Remaining := GunStats.GetDuration(Self) - Expended

                    if:
                        GetEquippedItemIDs[].Find[i]
                        not ShouldTickCooldown[i]
                    then:
                        Temp:=NewCooldowns[i]
                        set  UpdatedCD = Temp

                    UpdatedCD >= 0
                    set NewCooldowns[i] = UpdatedCD
                    
                then:
                    set DidUpdate = true

                    if (Remaining = 0):
                        RemoveItem(0)
                        RemoveItem(i)
                        EquipItem(0, ?Equip := false)
                        UpdateQuestProgressForTask(task_event.ItemExpired, big_number{Value := 1.0}, false, ?Target := "{i}", ?IsPersistent := true)
                        
                        # #OpenMenu Back (ONLY IF THERE'S NO GUNS - Does this ever happen now?)
                        # GameManager.GunManager.OpenMenu(Agent)

                    var Slot : int = -1
                    if:
                        Temp:= SourceData.EquippedItems.Find[i]
                        set Slot = Temp

                    if (Remaining > 0):
                        set CDStates = CDStates + array{cooldown_state{ID := i, RemainingCooldown := UpdatedCD, RemainingDuration := Remaining, Slot := Slot}}
                    else:
                        set CDStates = CDStates + array{cooldown_state{ID := i, RemainingCooldown := UpdatedCD, RemainingDuration := 0, Slot := Slot}}


            if:
                Player.IsActive[]
                DidUpdate?
                GameManager.CheckSaveDataForPlayer[Agent]
                SourceData2 := PlayerProfileDataMap[Player]
                set PlayerProfileDataMap[Player] = player_profile_data:
                    MakePlayerProfileData<constructor>(SourceData2)
                    ItemCooldowns := NewCooldowns
            
            if (SetUIManager := UIManager?):
                SetUIManager.OnCooldownsUpdated(CDStates) 

    ShouldTickCooldown(ID:int)<decides><transacts>:void=
        not (CurrentZone = fetchable_tag.zone1tag or
        CurrentZone = fetchable_tag.zone2tag or
        CurrentZone = fetchable_tag.towntag or
        CurrentZone = fetchable_tag.afk_zonetag or
        CurrentZone = fetchable_tag.range_zonetag or
        not (GetHeldGun[] = ID) or
        CurrentZone.IsBossTag()? or
        UIManager?.Popup?)

    CheckMapIndicators():void=
        block:

    #Passive Income Functions (aka Stash)
    LoopingMinuteCheck()<suspends>:void=
        SFXOffset:=2.0
        loop:
            if(SetUIManager:=UIManager?):
                SetUIManager.ResetStashCountdownTimer()

            Sleep(GameManager.PassiveIncomeTimer * 1.0 - SFXOffset)
            if:
                Player.IsActive[]
            then:
                #Passive Income
                spawn:
                    GainPassiveIncome(SFXOffset, 1)

                CheckQuestTrackers()

                GameManager.AccoladesManager.ThankYou.Award(Agent)
                GameManager.AccoladesManager.PleaseLikeAndRecommend.Award(Agent)

                if(CheckedOfflineProgression?):
                    #Location
                    spawn{SetLocationAndRotation(true)}
                    
                    #Realtime
                    SetRealtime()
    
    #In multiplayer games, sometimes these don't appear. Try, try again.
    CheckQuestTrackers():void=
        if:
            QM := QuestManager?
            Quests := GetCurrentQuests[]
        then:
            for (QuestStats : Quests):
                spawn:
                    QM.TurnOnQuestTrackers(QuestStats, ?IsGameStart:=true)

    GainPassiveIncome(SFXOffset:float, Minutes: int)<suspends>:void=
        if:
            var IncomeAmount : big_number = GetPassiveIncomeAmount[]
        then:
            if(IncomeAmount.Value > 0.0):
                #This audio has a 2 second delay for some reason
                GameManager.AudioManager.PlayPassiveIncomeCollectedSFX(Agent)
                Sleep(SFXOffset)
                if:
                    not(Player.IsActive[])
                then:
                    return

                if: 
                    Temp:= GetPassiveIncomeAmount[]
                    set IncomeAmount = Temp
                
                
                StashEssence(IncomeAmount * (1.0 *Minutes))
                
                if(SetUIManager:=UIManager?):
                    # SetUIManager.QueueDamageNumber(big_number{}, IncomeAmount, false)

    AddBossKeySeconds(Seconds:int):void=
        if:
            SourceData := GetSourceDataSafe[]
            GetAchievementLevel["Sarge_NextTier"] > 0
        then:    
            var NewSeconds : int = Seconds + SourceData.BossKeySeconds
            # if (Mod[NewSeconds, 10] = 0):
            #     Log("Boss Key Seconds: {NewSeconds}")
            if:
                KeysGained := Floor[NewSeconds * 1.0 / (GetBossKeyRefreshRate() * 1.0)]
                TempNewSeconds := Mod[NewSeconds, GetBossKeyRefreshRate()]
                set NewSeconds = TempNewSeconds

                set PlayerProfileDataMap[Player] = player_profile_data:
                    MakePlayerProfileData<constructor>(SourceData)
                    BossKeySeconds := NewSeconds
            then:
                if (KeysGained > 0):
                    CurrentKeys := GetCollectibleCount("BossKey")
                    if:
                        KeyCap := GetUpgradeValue["BossKeyCap", false]
                        KeysGained * 1.0 + CurrentKeys.Value > KeyCap * 1.0
                    then:
                        GainCollectible(collectible_data{Key := "BossKey", Name := "Boss Key", Count := big_number{Value:=KeyCap-CurrentKeys.Value, Exponent:=0}})
                    else:
                        GainCollectible(collectible_data{Key := "BossKey", Name := "Boss Key", Count := big_number{Value:=KeysGained * 1.0, Exponent:=0}})

                    VOIndex:= GetRandomInt(0, 2)
                    GameManager.AudioManager.PlayBossEntranceVO(Player, VOIndex, boss_enum.Sharky)
                    # GainCollectible(collectible_data{Key := "UpgradeHammer", Name := "Uprgade Hammer", Count := big_number{Value:=KeysGained * 1.0, Exponent:=0}})

                if:
                    SetUIManager := UIManager?
                then:
                    SetUIManager.UpdateBossKeyTimer(GetBossKeyRefreshRate() - NewSeconds)
  
    GetBossKeySecondsRemaining()<transacts>:int=
        if (BigNumber_GreaterThanOrEquals(GetCollectibleCount("BossKey"), big_number{Value:=1.0})?):
            return 0
        
        if:
            SourceData := GetSourceDataSafe[]
        then:
            return GetBossKeyRefreshRate() - SourceData.BossKeySeconds
        else:
            return 0

    GetBossKeyRefreshRate()<transacts>:int=
        return 720
  
    GetPassiveIncomeAmount()<decides><transacts>:big_number=
        #Average Fortnite Swing is 500ms or 2x per second
        return GetPlayerStat(player_stat.Passive_Income_Flat) * GetPlayerStat(player_stat.Passive_Income_Percent)
  
    EquipItemsAtGameStart():void=
        var ItemIDs : []int = GetEquippedItemIDs[] or array{-1,-1,-1,-1,-1}
        if:
            not ItemIDs.Find[0]
            set ItemIDs[0] = 0

        if:
            HasCompletedTask(0,2)?
        then:
            var i : int = 0
            for (ItemID : ItemIDs):
                if (ItemID <> -1):
                    EquipItem(ItemID, ?Slot:=i)
                set i = i + 1

    CheckOfflineProgression():void=
        if:
            Now:= GetSecondsSinceEpoch()
            LastLogin:= GetRealtime[]
            Created := GetCreationTime[]
            LifetimeMinutes := Int[(Now - LastLogin) / (60.0)]
            #Get 1/10th the credit for offline time
        then:
            spawn:
                GainPassiveIncome(2.0, LifetimeMinutes)
            AddBossKeySeconds(60 * LifetimeMinutes)
            UpdateItemCooldowns(?DeltaSeconds := 60 * LifetimeMinutes)

            set CheckedOfflineProgression = true

    CheckVersionDeltas():void=
        if:
            var SourceData : player_profile_data = GetSourceDataSafe[]
            CurrentVersion := SourceData.Version
        then:

            var AchievementsToGrant : []string = array{}
            var AchievementsToReset : []string = array{}
            var CollectiblesToGrant : []collectible_data = array{}

            #Adjust quest ID for Lyra, Urjent bossfights
            if (CurrentVersion < 1.0):
                Print("Updating Player Data from {CurrentVersion} to Version 1.0")
                set SourceData = UpdateQuestDataAdd(SourceData, quest_stats{QuestID:=1, TaskID:=5})
                set SourceData = UpdateQuestDataAdd(SourceData, quest_stats{QuestID:=1, TaskID:=23})
                set SourceData = player_profile_data:
                    MakePlayerProfileData<constructor>(SourceData)
                    Version := 1.0

                if(HasCompletedTask(1,4)?):
                    set AchievementsToGrant = AchievementsToGrant + array{"Midas_TP_Unlocked", "LyraTutorial"}

                if(HasCompletedTask(3,35)?):
                    set AchievementsToGrant = AchievementsToGrant + array{"Urjent"}

            #Fix missing Sharkey unlock achievement for people who had already passed it 
            if (CurrentVersion < 1.1):
                if(HasCompletedTask(2,8)?):
                    set AchievementsToGrant = AchievementsToGrant + array{"Sharkey"}

                set SourceData = player_profile_data:
                    MakePlayerProfileData<constructor>(SourceData)
                    Version := 1.1
            
            #Fix missing "Sarge_NextTier" Achieve and missing boss key
            if (CurrentVersion < 1.2):
                if(HasCompletedTask(1,4)?):
                    set CollectiblesToGrant = CollectiblesToGrant + array{collectible_data{Key := "BossKey", Count := big_number{Value:=1.0, Exponent:=0}}}

                set SourceData = player_profile_data:
                    MakePlayerProfileData<constructor>(SourceData)
                    Version := 1.2


            #Boss Rework 11/15/2024
            if (CurrentVersion < 1.3):
                if(HasCompletedTask(1,1)?):
                    set AchievementsToGrant = AchievementsToGrant + array{"AllowDailyHarvest"}

                if(HasCompletedTask(1,5)?):
                    set AchievementsToGrant = AchievementsToGrant + array{"Midas_TP_Unlocked"}

                if(HasCompletedTask(1,7)?):
                    set AchievementsToReset = AchievementsToReset + array{"Sarge_NextTier"}
                    set AchievementsToGrant = AchievementsToGrant + array{"LyraTutorial"}
                    set AchievementsToGrant = AchievementsToGrant + array{"SargeTutorial"}

                if(HasCompletedTask(1,24)?):
                    set AchievementsToReset = AchievementsToReset + array{"Sharkey_NextTier"}
                    set AchievementsToGrant = AchievementsToGrant + array{"Sharkey"}

                if(HasCompletedTask(2,19)?):
                    set AchievementsToReset = AchievementsToReset + array{"Genga_NextTier"}
                    set AchievementsToGrant = AchievementsToGrant + array{"Genga"}

                if(HasCompletedTask(3,24)?):
                    set AchievementsToReset = AchievementsToReset + array{"Urjent_NextTier"}
                    set AchievementsToGrant = AchievementsToGrant + array{"Urjent"}

                if(HasCompletedTask(4,29)?):
                    set AchievementsToReset = AchievementsToReset + array{"PirateKing_NextTier"}
                    set AchievementsToGrant = AchievementsToGrant + array{"Pirate King"}

                set SourceData = player_profile_data:
                    MakePlayerProfileData<constructor>(SourceData)
                    Version := 1.3


            if (CurrentVersion < 1.4):
                if:
                    var IDs : []int = GetEquippedItemIDs[]
                then:
                    loop:
                        if:
                            IDs.Length < 5
                        then:
                            set IDs = IDs + array{-1}
                        else:
                            break
                    
                    #Rebalance levels
                    if:
                        GetLevel[] > 320
                    then:
                        set SourceData = player_profile_data:
                            MakePlayerProfileData<constructor>(SourceData)
                            PlayerLevel := 320
                            PlayerStrength := big_number{Value := 3.31, Exponent:= 28}

                    set SourceData = player_profile_data:
                        MakePlayerProfileData<constructor>(SourceData)
                        EquippedItems := IDs
                        Version := 1.4

            if (CurrentVersion < 1.5):
                if:
                    var IDs : []int = GetEquippedItemIDs[]
                then:
                    #Rebalance levels Again (With Stash)
                    if:
                        GetLevel[] > 320
                    then:
                        set SourceData = player_profile_data:
                            MakePlayerProfileData<constructor>(SourceData)
                            PlayerLevel := 320
                            PlayerStrength := big_number{Value := 3.31, Exponent:= 28}
                            CurrentEssence := big_number{Value := 0.0, Exponent:= 0}
                            StashedEssence := big_number{Value := 0.0, Exponent:= 0}
                        
                    set SourceData = player_profile_data:
                        MakePlayerProfileData<constructor>(SourceData)
                        EquippedItems := IDs
                        Version := 1.5

            if:
                set PlayerProfileDataMap[Player] = SourceData
            
            for (Achievement : AchievementsToGrant):
                GainAchievementLevel(Achievement, 1)
            
            for (Achievement : AchievementsToReset):
                SetAchievementLevel(Achievement, 1)

            for (Collectible : CollectiblesToGrant):
                GainCollectible(Collectible)

            

    UpdateQuestDataAdd(SourceData : player_profile_data, Added:quest_stats):player_profile_data=
        var NewCurrentQuests : []quest_stats = array{}
        for (Quest : SourceData.CurrentQuests):
            if:
                Added.QuestID = Quest.QuestID
                Added.TaskID < Quest.TaskID
            then:
                #Same quest but ID offset by 1
                set NewCurrentQuests = NewCurrentQuests + array{quest_stats{QuestID:=Quest.QuestID, TaskID:=Quest.TaskID+1, Progress:=Quest.Progress}}
            else if:
                Added.QuestID = Quest.QuestID
                Added.TaskID = Quest.TaskID
            then:
                #Same quest ID, but the new quest is in this slot -- reset progress to 0
                set NewCurrentQuests = NewCurrentQuests + array{quest_stats{QuestID:=Quest.QuestID, TaskID:=Quest.TaskID}}
            else:
                #Same quest and quest ID
                set NewCurrentQuests = NewCurrentQuests + array{Quest}

        return player_profile_data:
            MakePlayerProfileData<constructor>(SourceData)
            CurrentQuests := NewCurrentQuests

    UpdateQuestDataRemove(SourceData : player_profile_data, Removed:quest_stats):player_profile_data=
        var NewCurrentQuests : []quest_stats = array{}
        for (Quest : SourceData.CurrentQuests):
            if:
                Removed.QuestID = Quest.QuestID
                Removed.TaskID < Quest.TaskID
            then:
                #Same quest but ID offset by 1
                set NewCurrentQuests = NewCurrentQuests + array{quest_stats{QuestID:=Quest.QuestID, TaskID:=Quest.TaskID-1, Progress:=Quest.Progress}}
            else if:
                Removed.QuestID = Quest.QuestID
                Removed.TaskID = Quest.TaskID
            then:
                #Same quest ID, but the the quest we were on was removed -- reset progress to 0
                set NewCurrentQuests = NewCurrentQuests + array{quest_stats{QuestID:=Quest.QuestID, TaskID:=Quest.TaskID}}
            else:
                #Same quest and quest ID
                set NewCurrentQuests = NewCurrentQuests + array{Quest}

        return player_profile_data:
            MakePlayerProfileData<constructor>(SourceData)
            CurrentQuests := NewCurrentQuests

    #Armor Functions
    GetArmorInSlot(Slot : armor_slot)<decides><transacts>:armor=
        var Armor : ?armor = false
        if:
            SourceData := GetSourceDataSafe[]
        then:
            for (i := 0..SourceData.Gear.Length - 1):
                if:
                    Gear := SourceData.Gear[i]
                    ArmorSlotFromString(Gear.Slot) = Slot
                then:
                    set Armor = option{Gear.ToArmor()}
        
        return Armor?

    IsArmorEquipped(Armor : armor)<transacts>:logic=
        if:
            Equipped := GetArmorInSlot[Armor.Slot]
            Equipped.Equals[Armor]
        then:
            return true
        else:
            return false

    UpgradeItem(Item : collectible_data)<transacts>:?armor_data=
        if:
            SourceData := GetSourceDataSafe[]
            ItemArmor := Item.ArmorData?
        then:
            UpgradedItem := ItemArmor.WithUpgrade()

            NewCollectible := collectible_data{
                Name := Item.Name
                Key := Item.Key
                Count := Item.Count
                ArmorData := option{UpgradedItem}
            }
            var NewMap : [string]collectible_data := SourceData.Collectibles
            
            if:
                NewArmor:=NewCollectible.ArmorData?
            then:
                EquipArmor(NewArmor.ToArmor())

            if:
                set NewMap[Item.Key] = NewCollectible 

                set PlayerProfileDataMap[Player] = player_profile_data:
                    MakePlayerProfileData<constructor>(SourceData)
                    Collectibles := NewMap
            
            return option{UpgradedItem}
        
        return false

    EquipArmor(Armor : armor)<transacts>:void=
        RemoveArmor(Armor.Slot)

        if:
            SourceData := GetSourceDataSafe[]
            set PlayerProfileDataMap[Player] = player_profile_data:
                MakePlayerProfileData<constructor>(SourceData)
                Gear := SourceData.Gear + array{Armor.ToData()}

    RemoveArmor(Slot : armor_slot)<transacts>:void=
        if:
            SourceData := GetSourceDataSafe[]
        then:
            for (i := 0..SourceData.Gear.Length - 1):
                if:
                    Gear := SourceData.Gear[i]
                    ArmorSlotFromString(Gear.Slot) = Slot
                    NewGear := SourceData.Gear.RemoveElement[i]
                    set PlayerProfileDataMap[Player] = player_profile_data:
                        MakePlayerProfileData<constructor>(SourceData)
                        Gear := NewGear
                then:
                    return

    GetGearStatTotal(Stat : player_stat)<transacts>:big_number=
        var Result : big_number = big_number{}
        if:
            SourceData := GetSourceDataSafe[]
        then:
            for (GearData : SourceData.Gear):
                Gear := GearData.ToArmor()
                for (i := 0..Gear.StatLevels.Length - 1; GearStat := Gear.Stats[i]; Level := Gear.StatLevels[i]):
                    if (GearStat = Stat):
                        set Result = Result + Stat.ToArmorStatValue(Level)

        return Result
    
    UpdateWeaponExperienceUI():void=
        if:
            ID := GetHeldGun[]
            UI := UIManager?
            EXP:= GetWeaponExperience(ID)
            Level:= GetWeaponLevel(ID)
            Req:=GetWeaponLevelRequirement(Level)
        then:
            UI.QueueGainWeaponExperience(ID, Level, EXP, Req)

    #Guard Functions
    GetGuardsDestroyed()<decides><transacts>:big_number=
        SourceData := GetSourceDataSafe[]
        return SourceData.GuardsDestroyed
        
    #When *THIS* player KOs a guard    
    OnGuardKO():void=
        set GuardSFXResetTimer = 10
        GameManager.AudioManager.PlayGuardDestroyedSFX(Agent, GuardComboSFX)
        GameManager.AccoladesManager.GuardKOAward(Agent)

        GrantWeaponExperience(true)

        if:
            SourceData := GetSourceDataSafe[]
            set PlayerProfileDataMap[Player] = player_profile_data:
                MakePlayerProfileData<constructor>(SourceData)
                GuardsDestroyed := BigNumber_Add(SourceData.GuardsDestroyed, big_number{Value:=1.0})
          
        
        if:
            Destroyed:=GetGuardsDestroyed[]
            Remainder := Mod[Int[BigNumber_ToFloat(Destroyed)], 25]
            UI:= UIManager?
        then:
            if:
                Remainder = 0
            then:
                GainCollectible(collectible_data{Key := "UpgradeHammer", Count := big_number{Value:=1.0}})
            UI.QueueGainUpgradeHammerProgress(Remainder, 25)

        set GuardComboSFX += 1
        if(GuardComboSFX > 4):
            ResetGuardComboSFX()

    GrantUniqueBossItem(ZoneTag : fetchable_tag, count : int) : void =
        case (ZoneTag):
            fetchable_tag.midastag => GainCollectible(collectible_data{Key := "EssenceHourglass", Count := big_number{Value:=1.0*count}})
            _ => block:
        
    OnEscapedMidasCastle():void=
        GainAchievementLevel("EscapedMidasCastle", 1)
        GainAchievementLevel("Inn", 1)
        GainAchievementLevel("Midas Castle",1)
        GameManager.SwitchToClass1.ChangeClass(Agent)
    
    #When *a* player KOs a guard
    OnZoneGuardKO(GuardKiller : player_manager, ZoneTag:fetchable_tag, GuardsAlive:int, MaxGuards:int, DidHitRecently:logic, ?GuardName : string = ""):void=
        if (DidHitRecently?):
            UpdateQuestProgressForTask(task_event.GuardKO, big_number{Value:=1.0}, false, ?Target := GuardName)
            CheckMomentum()

    LoopToResetGuardComboSFX()<suspends>:void=
        loop:
            Sleep(1.0)
            set GuardSFXResetTimer -= 1
            if(GuardSFXResetTimer<= 0):
                ResetGuardComboSFX()

    ResetGuardComboSFX():void=
        set GuardComboSFX = 1
    
    #Zone Functions
    GetZoneClears(ZoneTag:fetchable_tag)<transacts><decides>:int=
        SourceData := GetSourceDataSafe[]

        # Log("Getting Zone Clears for {ZoneTag.ToZoneName()}")

        if (ZoneTag = fetchable_tag.midastag or ZoneTag = fetchable_tag.zone3tag):
            ZoneIndexValue1 := fetchable_tag.midastag.ToZoneIndex()
            ZoneIndexValue2 := fetchable_tag.zone3tag.ToZoneIndex()
            ZoneClears := SourceData.ZoneClearArray[ZoneIndexValue1] + SourceData.ZoneClearArray[ZoneIndexValue2]
            
            # Log("Boss Zone Hack: {ZoneClears}")
            return ZoneClears

        else:
            ZoneIndexValue := ZoneTag.ToZoneIndex()
            ZoneClears := SourceData.ZoneClearArray[ZoneIndexValue]
            
            # Log("No Boss Zone Hack: {ZoneClears}")
            return ZoneClears

    GetZoneLevel(ZoneTag:fetchable_tag)<transacts><decides>:int=
        ZoneIndexValue := ZoneTag.ToZoneIndex()
        SourceData := GetSourceDataSafe[]
        ZoneLevel := SourceData.ZoneLevelArray[ZoneIndexValue]

        UpdateZoneOverkill(ZoneLevel)

        return ZoneLevel

    SetZoneLevel(ZoneTag:fetchable_tag, Value:int):void=
        ZoneIndexValue := ZoneTag.ToZoneIndex()

        if:
            SourceData := GetSourceDataSafe[]
            NewZoneLevelArray : []int = SourceData.ZoneLevelArray.ReplaceElement[ZoneIndexValue, Value]

            set PlayerProfileDataMap[Player] = player_profile_data:
                MakePlayerProfileData<constructor>(SourceData)
                ZoneLevelArray := NewZoneLevelArray
        
        UpdateZoneOverkill(Value)

    #TODO: Need to hook this up to actual upgrade check
    UpdateZoneOverkill(ZoneLevel : int)<transacts>: void=
        OverkillUnlocked := false
        if(OverkillUnlocked?):
            set ZoneOverkill = Max((ZoneLevel - 1) * 1.0 / 10.0,0.0)

    OnZoneCleared(ZoneTag : fetchable_tag, DamageMasteryTimer : int, ZoneBonus : big_number, PlayersInZone : int):void=
        PlayersZoneBonusToStash := PlayersInZone - 1 #Dont count yourself
        AmountToStash := BigNumber_Multiply(ZoneBonus, PlayersZoneBonusToStash * 1.0)
        StashEssence(AmountToStash)
        GainEssence(big_number{}, big_number{}, ZoneBonus, true, false, false, false, false)
        
        if:
            DamageMasteryTimer > 0
        then:
            set DamageMasteryEssenceStashed = BigNumber_Add(DamageMasteryEssenceStashed, AmountToStash)
            set DamageMasteryEssenceGained = BigNumber_Add(DamageMasteryEssenceGained, AmountToStash)

        GainZoneClear(ZoneTag)

        if:
            SetUIManager:=UIManager?
            PlayerClears:= GetZoneClears[ZoneTag]
            ZoneLevel:=GetZoneLevel[ZoneTag]
        then:
            UpdateQuestProgressForTask(task_event.ZoneClear, big_number{Value := PlayerClears * 1.0}, false, ?IsPersistent:=true)
            ClearsNeeded := GetClearsNeeded(ZoneTag, ZoneLevel, PlayerClears)
            SetUIManager.UpdateZoneClears(ZoneTag, PlayerClears, ClearsNeeded, ZoneLevel, false)


    GainZoneClear(ZoneTag:fetchable_tag):void=
        ZoneIndexValue := ZoneTag.ToZoneIndex()

        if:
            PlayerClears:= GetZoneClears[ZoneTag]
            ZoneLevel:=GetZoneLevel[ZoneTag]

            SourceData := GetSourceDataSafe[]
            NewZoneClears := SourceData.ZoneClearArray[ZoneIndexValue] + 1
            NewZoneClearArray : []int = SourceData.ZoneClearArray.ReplaceElement[ZoneIndexValue, NewZoneClears]
            
            set PlayerProfileDataMap[Player] = player_profile_data:
                MakePlayerProfileData<constructor>(SourceData)
                ZoneClearArray := NewZoneClearArray
        then:
            ZoneLevelCap := GameManager.GetZoneLevelCap(ZoneTag)
            
            ClearsNeeded := GetClearsNeeded(ZoneTag, ZoneLevel, PlayerClears)
            if:
                PlayerClears + 1 >= ClearsNeeded
                SetUIManager:=UIManager?
            then:
                if(ZoneLevel >= ZoneLevelCap):
                    if:
                        GetAchievementLevel["Damage Mastery"] > 0
                    else:
                        SetUIManager.ShowZoneClearMessage(ZoneTag, false)
                    return

                SetUIManager.ShowZoneLevelUpMessageUI(ZoneTag, ClearsNeeded, ZoneLevel + 1)
                if:
                    GetAchievementLevel["Damage Mastery"] > 0
                else:
                    SetUIManager.ShowZoneClearMessage(ZoneTag, true)
                OnZoneLevelUp(ZoneTag)
            else:
                if(SetUIManager:=UIManager?):
                    if (GetAchievementLevel["Damage Mastery"] > 0):
                    else:
                        SetUIManager.ShowZoneClearMessage(ZoneTag, false)

    GetClearsNeeded(ZoneTag : fetchable_tag, ZoneLevel : int, Clears : int):int=
        return ZoneLevel
        #Always give gem on Zone Clear

        # if (ZoneTag.IsBossTag()?):
        #     return ZoneLevel

        # var ClearsNeeded : int = 1
        
        # StartingClearsThreshold:int = 9
        # StartingClearsBase : int = 1
        # StartingClearsBaseAdder : int = 2
        # var StartingClearsBaseMultiplier : int = 0

        # BaseAdder: int = 2 + ZoneTag.ToZoneValue()
        # var BaseMultiplier : int = 0

        # if: 
        #     LevelThreshold:= Int[(StartingClearsThreshold - StartingClearsBase) * 1.0 / (StartingClearsBaseAdder * 1.0)] + 1
        # then:
        #     if(ZoneLevel <= LevelThreshold):
        #         set StartingClearsBaseMultiplier = ZoneLevel - 1
        #         set ClearsNeeded = StartingClearsBase + StartingClearsBaseAdder * StartingClearsBaseMultiplier
        #         return ClearsNeeded
        #     else:
        #         set BaseMultiplier = ZoneLevel - LevelThreshold
        #         set ClearsNeeded = StartingClearsThreshold + BaseAdder * BaseMultiplier
        #         return ClearsNeeded

        # return ClearsNeeded
    
    UnlockZone(ZoneTag:fetchable_tag):void=
        if(ZoneTag = fetchable_tag.none_tag):
            return

        GameManager.AudioManager.PlayUnlockZoneSFX(Agent)
        if(SetUIManager:=UIManager?):
            SetUIManager.ShowZoneUnlockMessageUI(ZoneTag)

    OnEnterZone(ZoneTag:fetchable_tag, GuardsAlive:int, MaxGuards:int):void=
        if:
            not CurrentZone = ZoneTag
        then:
            set CurrentZone = ZoneTag
            if:
                not CurrentZone = fetchable_tag.none_tag
            then:
                OnEnterZoneUpdateUI(ZoneTag, GuardsAlive, MaxGuards)
                GameManager.AudioManager.PlayEnterZoneSFX(Agent)
                PeelyManager.PlayerEntersZone(ZoneTag)
                PlayMusic()

        if (not ZoneTag.IsBossTag()?):
            for (Teleporter : GameManager.BossTeleportManager.Teleporters):
                Teleporter.RemovedFromFight(Self)

        if:
            ZoneTag = fetchable_tag.zone3tag
        then:
            GainAchievementLevel("EnteredMidas", 1)

        if:
            ZoneTag = fetchable_tag.towntag
        then:
            GainAchievementLevel("Landed in Goodtown", 1)

        UpdateQuestProgressForTask(task_event.EnterZone, big_number{Value:=1.0}, option{ZoneTag})


    OnExitZone(ZoneTag:fetchable_tag):void=
        if:
            ZoneTag = fetchable_tag.afk_zonetag or ZoneTag = fetchable_tag.range_zonetag
        then:
            set CurrentZone = fetchable_tag.towntag
                
    OnZoneLevelUp(ZoneTag:fetchable_tag):void=
        if:
            CurrentZoneLevel := GetZoneLevel[ZoneTag]
            PlayerClears:int= GetZoneClears[ZoneTag]
        then:
            SetZoneLevel(ZoneTag, CurrentZoneLevel + 1)
        
            GameManager.AudioManager.PlayZoneLevelUpSFX(Agent)
            
            UpdateQuestProgressForTask(task_event.ZoneLevel, big_number{Value:=CurrentZoneLevel * 1.0}, option{ZoneTag}, ?IsPersistent := true)
            spawn:
                GainGems(big_number{Value:= ZoneTag.ToZoneGemGain()})
        
    OnEnterZoneUpdateUI(ZoneTag:fetchable_tag, CurrentKOs:int, MaxKOs:int):void=
        if:
            PlayerClears:int= GetZoneClears[ZoneTag]
            ZoneLevel:int= GetZoneLevel[ZoneTag]
        then:
            ClearsNeeded:=GetClearsNeeded(ZoneTag, ZoneLevel, PlayerClears)

            if(SetUIManager:=UIManager?):
                SetUIManager.ShowEnterZoneMessageUI(ZoneTag, CurrentKOs)

    ExitZone(ZoneTag:fetchable_tag):void=
        if(SetUIManager:=UIManager?):
            SetUIManager.ExitZone(ZoneTag)

    #Gem Functions
    GetGems()<decides><transacts>:big_number=
        return GetCollectibleCount("Gems")

    GainGems(Gain:big_number, ?DidLevelUp:logic = false)<suspends>:void=
        Sleep(0.5)
        GameManager.AudioManager.PlayCollectGemSFX(Agent)
        GainCollectible(collectible_data{Key := "Gems", Count := Gain}, ?DidLevelUp := DidLevelUp)
            
    SpendGems(Spend:big_number):void=     
        GainCollectible(collectible_data{Key := "Gems", Count := Spend * -1.0})

    #Collectible Functions
    GetCollectibleCount(Name:string)<transacts>:big_number=
        if:
            SourceData := GetSourceDataSafe[]
            value := SourceData.Collectibles[Name].Count
        then:
            return value
        else:
            return big_number{Value:=0.0, Exponent:=0}


    GetCandyCount()<decides><transacts>:big_number=
        return  GetCollectibleCount("Candy1") +
                GetCollectibleCount("Candy2") +
                GetCollectibleCount("Candy3") +
                GetCollectibleCount("Candy4") +
                GetCollectibleCount("Candy5") +
                GetCollectibleCount("Candy6")

    GainCollectible(CollectibleIn : collectible_data, ?ForceKey : logic = false, ?DidLevelUp : logic =false):int=
        var Collectible : collectible_data = CollectibleIn

        if:
            Collectible.Key = "" or Collectible.ArmorData?
            not ForceKey?
            UID := TakeUniqueID()
        then:
            set Collectible = collectible_data{
                Name := Collectible.Name
                Key := "Unique_{UID}"
                Count := Collectible.Count
                ArmorData := option{Collectible.ArmorData?.WithUID(UID)}
            }
        else if (Collectible.Name = ""):
            set Collectible = collectible_data{
                Name := Collectible.Key
                Key := Collectible.Key
                Count := Collectible.Count
                ArmorData := Collectible.ArmorData
            }

        #Log(ToString(Collectible))

        var NewAmount : big_number = GetCollectibleCount(Collectible.Key) + Collectible.Count

        if:
            SourceData := GetSourceDataSafe[]
            var NewMap : [string]collectible_data := SourceData.Collectibles
            set NewMap[Collectible.Key] = collectible_data{
                Key := Collectible.Key
                Count := NewAmount
                ArmorData := Collectible.ArmorData
            }
        then:
            if:
                BigNumber_GreaterThanOrEquals(big_number{}, NewAmount)?
            then:
                set NewAmount = big_number{}
                set NewMap = RemoveKeyFromMap2(NewMap, Collectible.Key)

            if:
                set PlayerProfileDataMap[Player] = player_profile_data:
                    MakePlayerProfileData<constructor>(SourceData)
                    Collectibles := NewMap
        
        UpdateQuestProgressForTask(task_event.GetCollectibleCount, NewAmount, false, ?Target := Collectible.Key, ?IsPersistent := true)

        if:
            SetUIManager:=UIManager?
        then:
            SetUIManager.QueueCollectionItem(CollectibleIn.Count, CollectibleIn.GetImageKey())

            if:
                GetCharsFromString(Collectible.Name, 5) = "Candy"
                CandyCount := GetCandyCount[]
            then:
                SetUIManager.UpdateCandyUI(CandyCount)

            if:
                Collectible.Name = "Peely"
                PeelyCount := GetCollectibleCount(Collectible.Name)
            then:
                SetUIManager.UpdatePeelyUI(PeelyCount)

            if:
                Collectible.Key = "UpgradeHammer"
                HammerCount := GetCollectibleCount(Collectible.Key)
            then:
                SetUIManager.UpdateHammerUI(HammerCount)
    

            if:
                Collectible.Name = "Midas Coins"
                CoinCount := GetCollectibleCount(Collectible.Name)
            then:
                SetUIManager.UpdateCoinsUI(CoinCount)
    
            if: 
                Collectible.Name = "Gems"
                NewGems := GetGems[]
            then:
                SetUIManager.UpdateGemUI(NewGems, CollectibleIn.Count, ?DidLevelUp:= DidLevelUp)

            if:
                Collectible.Key = "BossKey"
                KeyCount := GetCollectibleCount(Collectible.Key)
                KeyCap := GetUpgradeValue["BossKeyCap", false]
            then:
                
                SetUIManager.UpdateKeyUI(KeyCount, CollectibleIn.Count, KeyCap)
        
        if (ColArmor := Collectible.ArmorData?.ToArmor()):
            return ColArmor.UID
        else:
            return 0

    GetCollectibleMap()<decides><transacts>:[string]collectible_data=
        if:
            SourceData := GetSourceDataSafe[]
        then:
            return SourceData.Collectibles
        else:
            return map{}

    CanUseCollectible(Collectible : collectible_data)<transacts>:logic=
        if (Armor := Collectible.ArmorData?):
            if (IsArmorEquipped(Armor.ToArmor())?):
                return false
            else:
                return true 
        if (Collectible.Key.Slice[0, 5] = "Candy"):
            return true
        return false

    UseCollectible(Collectible : collectible_data):void=
        if (Armor := Collectible.ArmorData?.ToArmor()):
            if (IsArmorEquipped(Armor)?):
                RemoveArmor(Armor.Slot)
            else:
                UpdateQuestProgressForTask(task_event.EquipArmor, big_number{Value:=1.0}, false, ?Target := "{Armor.Prefix}")
                EquipArmor(Armor)
            return

        if:
            Collectible.Key.Slice[0, 5] = "Candy"
            CandyID := ParseInt[Collectible.Key.Slice[5, 6]]
        then:
            GameManager.AudioManager.PlayEatCandySFX(Agent)
            UpdateQuestProgressForTask(task_event.EatCandy, big_number{Value:=1.0}, false)

            Str := big_number{Value:=1.0, Exponent:=CandyID}
            GainStrength(Str)
            if (UI :=UIManager?):
                UI.UpdateLevelUI()
                spawn:
                    UI.CreateStrengthFloatingText(Collectible)
            # GainCollectible(collectible_data{Key:=Collectible.Key, Count:=big_number{Value:=-1.0}})
    
            return

    #Item Functions
    GetEquippedItemIDs()<decides><transacts>:[]int=
        SourceData := GetSourceDataSafe[]
        return SourceData.EquippedItems
    
    GetEquippedItemIDInSlot(Slot:int)<decides><transacts>:int=
        return GetEquippedItemIDs[][Slot]

    HasEmptySlot()<decides><transacts>:void=

        var Found : logic = false

        if:
            SourceData := GetSourceDataSafe[]
        then:
            for (ii := 0..(SourceData.EquippedItems.Length - 1); i := SourceData.EquippedItems.Length - 1 - ii):
                if (ItemI := SourceData.EquippedItems[i]):
                    if (ItemI = -1):
                        set Found = true
        
        if (not Found?):
            false?
    
    MakeRoomForGun():void=
        if:
            SourceData := GetSourceDataSafe[]
        then:
            var IDtoRemove : int = -1
            var lowestCD : int = 99999999999
            var IsThereRoom : logic = false

            for (ii := 0..(SourceData.EquippedItems.Length - 1); i := SourceData.EquippedItems.Length - 1 - ii):
                if:
                    ItemI := SourceData.EquippedItems[i]
                then:
                    if:
                        ItemI = -1
                    then:
                        #There's room. Don't remove anything
                        set IsThereRoom = true
                    if:
                        ItemI > 0
                        Temp := GetRemainingCooldown[ItemI]
                        Temp < lowestCD
                    then:
                        set lowestCD = Temp
                        set IDtoRemove = ItemI
            if:
                IsThereRoom?
            then:
                return
                

            RemoveItem(IDtoRemove)

            if:
                TargetCD := GetGunStats[IDtoRemove].Cooldown - GetGunStats[IDtoRemove].GetDuration(Self)
                TargetCD < GetRemainingCooldown[IDtoRemove]
            then:
                UpdateItemCooldowns(?DeltaSeconds := 0, ?TargetCD := option{TargetCD}, ?IDTargeted := option{IDtoRemove})

    EquipItem(id:int, ?Slot:int = -1, ?Equip:logic = true):void= 
        var _Slot : int = Slot

        if:
            SourceData := GetSourceDataSafe[]
        then:
            if (_Slot = -1):
                for (ii := 0..(SourceData.EquippedItems.Length - 1); i := SourceData.EquippedItems.Length - 1 - ii):
                    if (ItemI := SourceData.EquippedItems[i]):
                        if (ItemI = id):
                            return #Already Equipped
                        if (ItemI = -1):
                            set _Slot = i

                if (_Slot = -1): #No Empty Slots
                    return

                if (SourceData.ItemCooldowns[id] > 0): #On Cooldown
                    return

                if:
                    var NewCooldowns : []int = SourceData.ItemCooldowns
                    Temp:=GameManager.GunManager.StatsTable[id].Cooldown
                    set NewCooldowns[id] = Temp

                    var NewEquippedItems : []int = SourceData.EquippedItems
                    set NewEquippedItems[_Slot] = id

                    set PlayerProfileDataMap[Player] = player_profile_data:
                        MakePlayerProfileData<constructor>(SourceData)
                        EquippedItems := NewEquippedItems
                        ItemCooldowns := NewCooldowns

            if (id = -1):
                block:
            else:
                GameManager.ItemGranter.SetNextItem(id)
                if (Equip?):
                    GameManager.ItemGranter.GrantItem(Player)
                else:
                    GameManager.ItemGranterNoSwitch.GrantItem(Player)
        
    UpdateGunUI():void=
        if:
            id := GetHeldGun[]
            SetUIManager:=UIManager?
            Dmg:=GetDamageMultiplier[id]
        then:
            SetUIManager.UpdateGunIcon(id , Dmg)
            UpdateWeaponExperienceUI()

    RemoveItem(id:int):void=
        if:
            SourceData := GetSourceDataSafe[]
        then:
            if (Slot := SourceData.EquippedItems.Find[id]):
                if:
                    var NewEquippedItems : []int = SourceData.EquippedItems
                    set NewEquippedItems[Slot] = -1
                    
                    set PlayerProfileDataMap[Player] = player_profile_data:
                        MakePlayerProfileData<constructor>(SourceData)
                        EquippedItems := NewEquippedItems
                then:
                    if (Remover := GameManager.ItemRemovers[id]):
                        Remover.Remove(Player)
                    
            if:
                SetUIManager:=UIManager?
            then:
                SetUIManager.RemoveGunIcon()   
                UpdateWeaponExperienceUI() 
    
    GetGunStats(ItemID:int)<decides><transacts>:gun_stats=
        return GameManager.GunManager.StatsTable[ItemID]

    GetRemainingCooldown(ItemID:int)<decides><transacts>:int=
        var CD : int = 0
        if:
            SourceData := GetSourceDataSafe[]
            SourceData.ItemCooldowns[ItemID]
        then:
            Temp := SourceData.ItemCooldowns[ItemID]
            set CD = Temp
        return CD
    
    GetRemainingDuration(ItemID:int)<decides><transacts>:int=
        var Remaining : int = 0
        if:
            SourceData := GetSourceDataSafe[]
            GunStats := GetGunStats[ItemID]
        then:
            Expended := GunStats.Cooldown - SourceData.ItemCooldowns[ItemID]
            set Remaining = GunStats.GetDuration(Self) - Expended
        return Remaining

    GetDamageMultiplier(ItemID:int)<decides><transacts>:big_number=
        return GetGunStats[ItemID].DamageMult
    
    GetCurrentGunStats()<decides><transacts>:gun_stats=
        var Stats : gun_stats = gun_stats{}
        for (id : GetEquippedItemIDs[]):
            if:
                not id = -1
                GameManager.ItemCheckers[id].IsHoldingItem[Player, 0]
                Temp := GetGunStats[id]
                set Stats = Temp
        return Stats
    
    GetHeldGun()<decides><transacts>:int=
        var ReturnID : int = -1
        for (id : GetEquippedItemIDs[]):
            if:
                not id = -1
                GameManager.ItemCheckers[id].IsHoldingItem[Player, 0]
                set ReturnID = id
        return ReturnID

    #Upgrade Functions        
    GetUpgradeLevel(Name:string)<decides><transacts>:int=
        if:
            SourceData := GetSourceDataSafe[]
            value := SourceData.UpgradeLevels[Name]
        then:
            return value
        else:
            return 0

    GetCumulativeUpgradeLevels()<decides><transacts>:int=
        if:
            var CumulativeUpgrades : int = 0
            SourceData := GetSourceDataSafe[]
            for (UpgradeLevel : SourceData.UpgradeLevels):
                set CumulativeUpgrades += UpgradeLevel
        then:
            return CumulativeUpgrades
        else:
            return 0

    GetUpgradeLevelCost(Name:string, level:?int)<decides><transacts>:int=
        if:
            SourceData := GetSourceDataSafe[]

            var finalLevel : int = GetUpgradeLevel[Name]
            if (tempLevel := level?):
                set finalLevel = tempLevel

            Mod[finalLevel,2]
            value : int = case (Name):
                "Focus Fire" => Floor(finalLevel/2) + 1
                "Momentum" => Floor(finalLevel/2) + 1
                "Teamwork" => Floor(finalLevel/2) + 1
                "Stealth Attack" => Floor(finalLevel/2) + 1
                "Sliding Rifle" => Floor(finalLevel/2) + 1
                "Handcannon" => Floor(finalLevel/2) + 1
                "Ambush" => Floor(finalLevel/2) + 1
                "SuperHandcannon" => block:
                    if:
                        finalLevel = 1
                    then:
                        200
                    else:
                        Floor(finalLevel/2) + 1
                _ => GameManager.UpgradeLevelCosts[Name][finalLevel]
            
        then:
            return value
        else:
            false?
            return 0
            
    GetUpgradeValue(Name:string, level:?int)<decides><transacts>:float=
        if:
            SourceData := GetSourceDataSafe[]

            var finalLevel : int = GetUpgradeLevel[Name]
        then:
            if (tempLevel := level?):
                set finalLevel = tempLevel

            if (Name.Slice[0, 2] = "P_"):
                return GetPrestigeUpgradeFromKey(Name).GetValue(finalLevel)
            else:
                return case (Name):
                    "BossKeyCap" => 10.0 + finalLevel * 1.0
                    "Heartofthemountain" => finalLevel * 20.0
                    "Passive Income I" => 1000.0 * Pow(2.0,1.0*finalLevel) * Min(1.0,1.0*finalLevel)
                    "Urjent_CritMult" => 2.0 * finalLevel
                    "Charon_DamageMult" => 1.0 * finalLevel
                    "Focus Fire" => finalLevel * 10.0 + Min(finalLevel*1.0, 1.0) * 20.0
                    "Momentum" => finalLevel * 5.0 + Min(finalLevel*1.0, 1.0) * 10.0
                    "Teamwork" => finalLevel * 0.5 + Min(finalLevel*1.0, 1.0) * 4.5
                    "Stealth Attack" => finalLevel * 25.0 + Min(finalLevel*1.0, 1.0) * 225.0
                    "Sliding Rifle" => finalLevel * 25.0 + Min(finalLevel*1.0, 1.0) * 225.0
                    "Handcannon" => finalLevel * 20.0 + Min(finalLevel*1.0, 1.0) * 80.0
                    "Ambush" => finalLevel * 20.0 + Min(finalLevel*1.0, 1.0) * 80.0
                    "SuperHandcannon" => finalLevel * 20.0 + Min(finalLevel*1.0, 1.0) * 80.0
                    _ => GameManager.UpgradeValues[Name][finalLevel] or 0.0
        else:
            false?
            return 0.0

    GetUpgradeValueSafe(Name:string, level:?int)<transacts>:float=
        if (value := GetUpgradeValue[Name, level]):
            return value
        else:
            return 0.0

    GainUpgradeLevelTo(Name : string, TargetLevel : int):void=
        if (GetUpgradeLevel[Name] >= TargetLevel):
            return

        if:
            SourceData := GetSourceDataSafe[]

            var NewMap : [string]int := SourceData.UpgradeLevels
            set NewMap[Name] = TargetLevel

            set PlayerProfileDataMap[Player] = player_profile_data:
                MakePlayerProfileData<constructor>(SourceData)
                UpgradeLevels := NewMap
        
        CheckModifierUpgrades()
        if (SetUIManager := UIManager?):
            SetUIManager.UpdateLevelUI()

    GainUpgradeLevel(Name:string, Gain: int):void=
        var NewLevel : int = 1
        if:
            TempNewLevel := GetUpgradeLevel[Name] + Gain
            set NewLevel = TempNewLevel

        if:
            SourceData := GetSourceDataSafe[]

            var NewMap : [string]int := SourceData.UpgradeLevels
            set NewMap[Name] = NewLevel

            set PlayerProfileDataMap[Player] = player_profile_data:
                MakePlayerProfileData<constructor>(SourceData)
                UpgradeLevels := NewMap
        
        CheckModifierUpgrades()
        if (SetUIManager := UIManager?):
            SetUIManager.UpdateLevelUI()
    
    #Buff Functions
    GetMoveSpeedMult()<transacts>:big_number=
        var SpeedMult : big_number = big_number{Value := 1.0}
        if:
            BuffStates := GetBuffStates[]
            Temp := (1.0 + (BuffStates["Momentum"].Amount / 100.0))
            set SpeedMult = SpeedMult * Temp

        SpeedMult

    CheckModifierUpgrades():void=
        if:
            Player.IsActive[]
            FortChar := Agent.GetFortCharacter[]
        then:
            SpeedMult := GetMoveSpeedMult()
            if:
                Temp := SpeedMult.Value / 1.05
                Temp >= 1.0
                Index : int = Int[Temp]
                Modulator := GameManager.MovementModulators[Index]
            then:
                Modulator.Activate(Agent)
            else:
                if:
                    #This resets their speed.
                    Modulator := GameManager.MovementModulators[0]
                then:
                    Modulator.Activate(Agent)


            FortChar.SetMaxHealth(BigNumber_ToFloat(GetPlayerStat(player_stat.MaxHealth)))
            
    #Stat functions
    FullHeal()<suspends>:void=
        if:
            Player.IsActive[]
            FortChar := Agent.GetFortCharacter[]
        then:
            FortChar.SetHealth(BigNumber_ToFloat(GetPlayerStat(player_stat.MaxHealth)))

    GetPlayerStat(Stat:player_stat)<transacts>:big_number=
        case (Stat):
            player_stat.MaxHealth => block:
                2000.0 + GetGearStatTotal(player_stat.MaxHealth) + GetUpgradeValueSafe("Heartofthemountain", false)

            player_stat.Regen => block:
                var Healingwinds : big_number = big_number{}
                if:
                    BuffStates := GetBuffStates[]
                    Temp := ((BuffStates["Healingwinds"].Amount / 100.0) * GetPlayerStat(player_stat.MaxHealth))
                then:
                    set Healingwinds = Temp
                GetGearStatTotal(player_stat.Regen) + Healingwinds
                
            player_stat.Crit_Mult => block:
                1.0 + 
                GetUpgradeValueSafe("Urjent_CritMult", false)/100.0 +
                GetGearStatTotal(player_stat.Crit_Mult)/100.0

            player_stat.DamageMult => block:
                (1.0 + GetUpgradeValueSafe("Tribalwrath", false)/100.0) * 
                (1.0 + GetUpgradeValueSafe("Charon_DamageMult", false)/100.0) *
                (1.0 + GetGearStatTotal(player_stat.DamageMult) / 100.0) 

            player_stat.Damage_Beast => block:
                GetGearStatTotal(player_stat.Damage_Beast)/100.0 + GetUpgradeValueSafe("BeastMastery", false)

            player_stat.Damage_Guard => block:
                GetGearStatTotal(player_stat.Damage_Guard)/100.0 + GetUpgradeValueSafe("GuardMastery", false)

            player_stat.Damage_Mineral => block:
                GetGearStatTotal(player_stat.Damage_Mineral)/100.0 + GetUpgradeValueSafe("MiningMastery", false)

            player_stat.Damage_Boss => block:
                GetGearStatTotal(player_stat.Damage_Boss)/100.0 + GetUpgradeValueSafe("BossMastery", false) + 0.15 * (GetBigRockLevel() + 1)

            player_stat.Damage_With_Pickaxe => block:
                GetGearStatTotal(player_stat.Damage_With_Pickaxe)/100.0 + 1.0

            player_stat.Damage_With_Pistol => block:
                GetGearStatTotal(player_stat.Damage_With_Pistol)/100.0 + 0.1 * GetWeaponLevel(1) + 1.0

            player_stat.Damage_With_AR => block:
                GetGearStatTotal(player_stat.Damage_With_AR)/100.0 + 0.1 * GetWeaponLevel(2) + 1.0

            player_stat.Damage_With_Shotgun => block:
                GetGearStatTotal(player_stat.Damage_With_Shotgun)/100.0 + 0.1 * GetWeaponLevel(4) + 1.0

            player_stat.Damage_With_Railgun => block:
                GetGearStatTotal(player_stat.Damage_With_Railgun)/100.0 + 0.1 * GetWeaponLevel(6) + 1.0

            player_stat.Damage_With_Bow => block:
                GetGearStatTotal(player_stat.Damage_With_Bow)/100.0 + 0.1 * GetWeaponLevel(5) + 1.0

            player_stat.Damage_With_Minigun => block:
                GetGearStatTotal(player_stat.Damage_With_Minigun)/100.0 + 0.1 * GetWeaponLevel(7) + 1.0

            player_stat.Damage_With_Hand_Cannon => block:
                GetGearStatTotal(player_stat.Damage_With_Hand_Cannon)/100.0 + 0.1 * GetWeaponLevel(8) + 1.0

            player_stat.Damage_With_Stake_Gun => block:
                GetGearStatTotal(player_stat.Damage_With_Stake_Gun)/100.0 + 0.1 * GetWeaponLevel(9) + 1.0

            player_stat.Damage_With_Scythe => block:
                GetGearStatTotal(player_stat.Damage_With_Scythe)/100.0 + 0.1 * GetWeaponLevel(10) + 1.0

            player_stat.Damage_Melee => block:
                GetGearStatTotal(player_stat.Damage_Melee)/100.0 + GetUpgradeValueSafe("P_DamageMelee", false) + 1.0
            
            player_stat.Damage_Rapid => block:
                GetGearStatTotal(player_stat.Damage_Rapid)/100.0 + GetUpgradeValueSafe("P_DamageRapid", false) + 1.0

            player_stat.Damage_Precision => block:
                GetGearStatTotal(player_stat.Damage_Precision)/100.0 + GetUpgradeValueSafe("P_DamagePrecision", false) + 1.0

            player_stat.Empowered_Mining => block:
                GetGearStatTotal(player_stat.Empowered_Mining)/100.0 + GetUpgradeValueSafe("EmpoweredMining", false)

            player_stat.Passive_Income_Flat => block:
                GetGearStatTotal(player_stat.Passive_Income_Flat) + MakeBigNumber(GetUpgradeValueSafe("Passive Income I", false), 1)

            player_stat.Passive_Income_Percent => block:
                1.0 + GetGearStatTotal(player_stat.Passive_Income_Percent)/100.0

            player_stat.MoveSpeed => block:
                big_number{}
                #TODO: This is tricky because of modulators
            _ => big_number{}
    
    #Daily Functions
    GetDailyWeapon()<transacts>:int=
        CurrentTime := GetSecondsSinceEpoch()
        var ID : int = 0
        if:
            Days := Int[CurrentTime / 86400.0]
            Temp :=  Mod[Days, GameManager.GunManager.StatsTable.Length - 1]
            set ID = Temp + 1
            if:
                ID = 8 or ID = 9
            then:
                set ID += 1
        then:
            return ID
        
        return 1
            
    IsDailyWeapon(Gun:int)<transacts>:logic=
        if:
            Gun = GetDailyWeapon()
        then:
            return true
        
        return false

    GetDailyBoss()<transacts>:int=
        CurrentTime := GetSecondsSinceEpoch()
        var ID : int = 0
        if:
            Days := Int[CurrentTime / 86400.0]
            Temp :=  Mod[Days, GameManager.BossTeleportManager.Teleporters.Length]
            set ID = Temp
            if:
                ID = 4
            then:
                set ID += 1
            
        then:
            return ID
        
        return 1
            
    IsDailyBoss(Boss:int)<transacts>:logic=
        if:
            Boss = GetDailyBoss()
        then:
            return true
        
        return false

    #Achievement Functions
    GetAchievementLevel(Name:string)<decides><transacts>:int=
        if:
            Name = "Damage Mastery"
        then:
            return 0
        else:
            if:
                SourceData := GetSourceDataSafe[]
                value := SourceData.Achievements[Name]
            then:
                return value
            else:
                return 0

    GainAchievementLevelTo(Name:string, TargetLevel : int):void=
        if (GetAchievementLevel[Name] >= TargetLevel):
            return

        if:
            SourceData := GetSourceDataSafe[]

            var NewMap : [string]int := SourceData.Achievements
            set NewMap[Name] = TargetLevel

            set PlayerProfileDataMap[Player] = player_profile_data:
                MakePlayerProfileData<constructor>(SourceData)
                Achievements := NewMap
        then:
            UpdateQuestProgressForTask(task_event.RequireAchievement, big_number{Value:= TargetLevel * 1.0}, false, ?Target := Name, ?IsPersistent := true)

    SetAchievementLevel(Name:string, TargetLevel : int):void=
        if (GetAchievementLevel[Name] = TargetLevel):
            return

        if:
            SourceData := GetSourceDataSafe[]

            var NewMap : [string]int := SourceData.Achievements
            set NewMap[Name] = TargetLevel

            set PlayerProfileDataMap[Player] = player_profile_data:
                MakePlayerProfileData<constructor>(SourceData)
                Achievements := NewMap
        then:
            UpdateQuestProgressForTask(task_event.RequireAchievement, big_number{Value:= TargetLevel * 1.0}, false, ?Target := Name, ?IsPersistent := true)

    GainAchievementLevel(Name:string, Gain: int):void=
        var NewLevel : int = Gain
        if:
            TempNewLevel := GetAchievementLevel[Name] + Gain
            set NewLevel = TempNewLevel

        if:
            SourceData := GetSourceDataSafe[]

            var NewMap : [string]int := SourceData.Achievements
            set NewMap[Name] = NewLevel

            set PlayerProfileDataMap[Player] = player_profile_data:
                MakePlayerProfileData<constructor>(SourceData)
                Achievements := NewMap
        then:
            UpdateQuestProgressForTask(task_event.RequireAchievement, big_number{Value:= Gain * 1.0}, false, ?Target := Name, ?IsPersistent := false)

    #Specific Upgrade Functions
    GetKOBonusMultiplier()<transacts>:float= 
        if (Bonus := GetUpgradeValue["KO Bonus", false]):
            return Bonus / 100.0
        else:
            return 0.0

    GetClearBonusMultiplier()<transacts>:float= 
        if (Bonus := GetUpgradeValue["Clear Bonus", false]):
            return Bonus / 100.0
        else:
            return 0.0
        
    #Quest Functions
    GetCurrentQuests()<transacts><decides>:[]quest_stats=
        SourceData := GetSourceDataSafe[]
        return SourceData.CurrentQuests

    GetCompletedQuests()<transacts><decides>:[]int=
        SourceData := GetSourceDataSafe[]
        return SourceData.CompletedQuests

    GetFailableTask()<transacts><decides>:quest_stats=
        SourceData := GetSourceDataSafe[]
        return SourceData.FailableQuest

    SetFailableTask(QuestStats : quest_stats):void=
        if:
            SourceData := GetSourceDataSafe[]
        
            Temp := quest_stats{QuestID:= QuestStats.QuestID, TaskID:= QuestStats.TaskID}

            set PlayerProfileDataMap[Player] = player_profile_data:
                MakePlayerProfileData<constructor>(SourceData)
                FailableQuest := Temp

    GetQuestStats(QuestID:int)<transacts><decides>:quest_stats=
        var Result : quest_stats = quest_stats{QuestID := -1}
        for (Event : GetCurrentQuests[]):
            if:
                Event.QuestID = QuestID
            then:
                set Result = Event
        return Result

    GetQuestStats(QuestID:int, TaskID:int)<transacts><decides>:quest_stats=
        var Result : quest_stats = quest_stats{QuestID := -1}
        for (Event : GetCurrentQuests[]):
            if:
                Event.QuestID = QuestID
                Event.TaskID = TaskID
            then:
                set Result = Event
        return Result

    GetRebirthTeleportLocation()<decides><transacts>:transform=
        var Result : transform = transform{}

        PrestigeValue := GetUpgradeLevel["P_AutoClearQuest"] or 0

        if (PrestigeValue > 12): #pirate island
            #Not this one
            set Result = GameManager.PirateIslandTeleporter.GetTransform()
        else if (PrestigeValue > 9): #giantlands
            GiantlandsTP := GameManager.TeleportManager.Teleporters[4].Teleporter.GetTransform()
            set Result = GiantlandsTP
        else if (PrestigeValue > 5): #goodtown
            set Result = GameManager.InnTeleporter.GetTransform()
        else:
            Spawners := GameManager.FindCreativeObjectsWithTag(spawner{})
            var NumSpawners : int = 0
            for (Obj : Spawners):
                set NumSpawners += 1
                
            var SelectedSpawner : ?creative_object_interface = false
            var Selection : int = GetRandomInt(0, NumSpawners - 1)
            var Count : int = 0
            for (Obj : Spawners):
                if (Count = Selection):
                    set SelectedSpawner = option{Obj}
                set Count += 1
            
            SetSpawner := SelectedSpawner?
            set Result = SetSpawner.GetTransform()

        return Result

    ShouldSkipQuestGroup(GroupKey : string)<transacts>:logic=
        PrestigeValue := GetUpgradeLevel["P_AutoClearQuest"] or 0
        PrestigeCount := GetSourceDataSafe[].PrestigeCount or 0

        case (GroupKey):
            "FirstPrestige" => return logic{PrestigeCount > 0}
            "Pistol" => return logic{PrestigeValue > 0}
            "AR" => return logic{PrestigeValue > 0}
            "Shotgun" => return logic{PrestigeValue > 0}
            "LyraUpgrade" => return logic{PrestigeValue > 1}
            "SamiraUpgrade" => return logic{PrestigeValue > 2}
            "Blossom" => return logic{PrestigeValue > 3}
            "Bow" => return logic{PrestigeValue > 4}
            "Minigun" => return logic{PrestigeValue > 4}
            "MidasCastle" => return logic{PrestigeValue > 5}
            "Beach" => return logic{PrestigeValue > 6}
            "Forest" => return logic{PrestigeValue > 7}
            "StakeGun" => return logic{PrestigeValue > 8}
            "Bandit" => return logic{PrestigeValue > 9}
            "Raiders" => return logic{PrestigeValue > 10}
            "Vampires" => return logic{PrestigeValue > 11}
            "Giant" => return logic{PrestigeValue > 12}
            _ => return false

    #Add QuestID (QuestManager Index) to PlayerData CurrentQuests, TaskIndex to CurrentTasks, Progress to CurrentTasksProgress
    AddQuest(QuestID:int, TaskID:int)<decides><transacts>:quest_stats=
        SourceData := GetSourceDataSafe[]
        
        Temp := quest_stats{QuestID:= QuestID, TaskID:= TaskID}

        NewQuestsArray := SourceData.CurrentQuests + array{Temp}

        set PlayerProfileDataMap[Player] = player_profile_data:
            MakePlayerProfileData<constructor>(SourceData)
            CurrentQuests := NewQuestsArray    

        return Temp

    UpdateQuestProgressForTask(Task : task_event, Gain : big_number, ZoneTag : ?fetchable_tag, ?Target : string := "", ?IsPersistent : logic := false):void=
        if(SetQuestManager:=QuestManager?):
            for (X := 0..SetQuestManager.ActiveTrackedQuests.Length - 1):
                if:
                    ActiveTask := SetQuestManager.ActiveTrackedQuests[X].TaskEvent
                    ActiveTask = Task
                    TrackedQuestStats := SetQuestManager.ActiveTrackedQuests[X].QuestStats
                    QuestID := TrackedQuestStats.QuestID
                    TaskID := TrackedQuestStats.TaskID
                    CurrentQuestStats := GetQuestStats[QuestID, TaskID]
                    CurrentProgress := CurrentQuestStats.Progress
                    RequiredZone := SetQuestManager.QuestModel.Quests[QuestID].Tasks[TaskID].ZoneRequired
                    TaskTarget := SetQuestManager.QuestModel.Quests[QuestID].Tasks[TaskID].TargetName
                    TaskTarget = Target or TaskTarget = ""
                then:
                    var NewProgress : big_number = BigNumber_Add(CurrentProgress,Gain)
                    if(IsPersistent?):
                        set NewProgress = Gain

                    NewQuestStats := quest_stats{QuestID:= QuestID, TaskID := TaskID, Progress:= NewProgress}

                    if:
                        FailableTask := GetFailableTask[]
                        FailableTask.QuestID = QuestID
                        FailableTask.TaskID = TaskID
                        TaskID + 1 <= CurrentQuestStats.TaskID
                        RequiredProgress := SetQuestManager.QuestModel.Quests[FailableTask.QuestID].Tasks[FailableTask.TaskID].Progress
                        QuestStatsToRevert := GetQuestStats[QuestID]
                    then:
                        IsNewProgressEnough := BigNumber_GreaterThanOrEquals(NewProgress, RequiredProgress)
                        if (IsNewProgressEnough?):
                            #Do nothing
                        else:
                            spawn:
                                SetQuestManager.RevertToFailedTask(QuestStatsToRevert, FailableTask)
                            UpdateQuestPlayerData(NewQuestStats, TaskID) #These are our stats now
                    else:
                        # Print("Updating Progress for {TaskID} to {NewProgress}")
                        if(SetZoneTag := ZoneTag?):
                            if(SetZoneTag = RequiredZone):
                                UpdateQuestPlayerData(NewQuestStats, TaskID)
                        else:
                            UpdateQuestPlayerData(NewQuestStats, TaskID)

    #Find PlayerData.CurrentQuests with the same value to get the Index Position within PlayerData
    #Then, with Index Position, update CurrentTasks and CurrentTasksProgress accordingly
    UpdateQuestPlayerData(QuestStats : quest_stats, OldTaskID : int, ?DebugOverride : logic = false):void=
        var QuestFound : logic = false
        if:
            SourceData := GetSourceDataSafe[]
            
            #Find CurrentQuest's Index Position within PlayerData
            var PlayerDataIndex : int = 0
            for (X := 0..SourceData.CurrentQuests.Length - 1):
                if:
                    SourceData.CurrentQuests[X].QuestID = QuestStats.QuestID
                    SourceData.CurrentQuests[X].TaskID = OldTaskID
                then:
                    set PlayerDataIndex = X
                    set QuestFound = true
        
            if: 
                QuestFound?
            then:
                NewQuestsArray : []quest_stats = SourceData.CurrentQuests.ReplaceElement[PlayerDataIndex, QuestStats]

                if:
                    set PlayerProfileDataMap[Player] = player_profile_data:
                        MakePlayerProfileData<constructor>(SourceData)
                        CurrentQuests := NewQuestsArray
            
        if:
            DebugOverride?
            not(QuestFound?)
            AddQuest[QuestStats.QuestID, QuestStats.TaskID]     
        then:    
            UpdateQuestPlayerData(QuestStats, QuestStats.TaskID)

        if (DebugOverride? or QuestFound?):
            CheckQuestProgress(QuestStats)

    CheckQuestProgress(QuestStats : quest_stats) : void=
        if(SetQuestManager:=QuestManager?):
            SetQuestManager.OnValuesUpdated(QuestStats)

    #Find PlayerData.CurrentQuests with the same value as QuestID to get the Index Position within PlayerData
    #Then, with Index Position, remove that position for CurrentQuests, CurrentTasks, and CurrentTasksProgress
    #Then add this QuestID to CompletedQuests
    CompleteQuest(QuestID : int, ?DebugOverride : logic = false):void=
        var QuestFound : logic = false

        if:
            SourceData := GetSourceDataSafe[]
        then:
        
            #Find CurrentQuest's Index Position within PlayerData
            var PlayerDataIndex : int = 0
            for (X := 0..SourceData.CurrentQuests.Length - 1):
                if(SourceData.CurrentQuests[X].QuestID = QuestID):
                    set PlayerDataIndex = X
                    set QuestFound = true
            
            if:
                QuestFound?
                NewQuestArray := SourceData.CurrentQuests.RemoveElement[PlayerDataIndex]
            then:
                #Create Temp Arrays to Update PlayerData
                NewCompletedQuestsArray := SourceData.CompletedQuests + array{QuestID}

                if:
                    set PlayerProfileDataMap[Player] = player_profile_data:
                        MakePlayerProfileData<constructor>(SourceData)
                        CurrentQuests := NewQuestArray
                        CompletedQuests := NewCompletedQuestsArray
                
                # if (QM := QuestManager?):
                #     Print("Primary Task Completed, Checking Persistent Quests")
                #     QM.CheckPersistentQuests()
        
        if:
            DebugOverride?
            not(QuestFound?)
            AddQuest[QuestID, 0]     
        then:    
            CompleteQuest(QuestID)

    CompleteSecondaryTask(QuestID : int, TaskID : int):void=
        var QuestFound : logic = false

        if:
            SourceData := GetSourceDataSafe[]
        then:
            #Find CurrentQuest's Index Position within PlayerData
            var PlayerDataIndex : int = 0
            for (X := 0..SourceData.CurrentQuests.Length - 1):
                if:
                    SourceData.CurrentQuests[X].QuestID = QuestID
                    SourceData.CurrentQuests[X].TaskID = TaskID
                then:
                    set PlayerDataIndex = X
                    set QuestFound = true
            
            if:
                QuestFound?
                NewQuestArray := SourceData.CurrentQuests.RemoveElement[PlayerDataIndex]
            then:
                #DO NOT DO THIS
                #NewCompletedQuestsArray := SourceData.CompletedQuests + array{QuestID}

                if:
                    set PlayerProfileDataMap[Player] = player_profile_data:
                        MakePlayerProfileData<constructor>(SourceData)
                        CurrentQuests := NewQuestArray
                
                # if (QM := QuestManager?):
                #     Print("Secondary Task Completed, Checking Persistent Quests")
                #     QM.CheckPersistentQuests()
                

    AreQuestArraysValid(QuestsLength : int, TasksLength : int, TasksProgress : int)<transacts>:logic=
        var Value : int = QuestsLength
        if(TasksLength = Value):
            if(TasksProgress = TasksLength):
                return true
        return false

    UpdateQuestUI(QuestStats : []quest_stats, QuestDescription : string, ?IsNewQuest:logic=false, ?DoesSkipCompleteFlair : logic = false):void=
        if:
            SetUIManager:=UIManager?
            SetQuestManager := QuestManager?
        then:
            if:
                IsNewQuest?
                FirstQuest := QuestStats[0]
                set QuestOnUI = FirstQuest
            then:
                SetUIManager.SetQuestText(QuestDescription, ?DoesSkipCompleteFlair :=DoesSkipCompleteFlair)
            
            var ShouldShow : logic = false
            for (QS : QuestStats):
                if:
                    QS.QuestID = QuestOnUI.QuestID
                    QS.TaskID = QuestOnUI.TaskID
                then:
                    set ShouldShow = true

            if (ShouldShow?):
                var QuestInfos : []quest_progress_info = array{}
                for (ID := 0..QuestStats.Length - 1; Quest := QuestStats[ID]):
                    if:
                        Task := SetQuestManager.QuestModel.Quests[Quest.QuestID].Tasks[Quest.TaskID]
                        Required := Task.Progress
                        TextureString := Task.IconTexture
                    then:
                        # Print("Queueing for Task: {Quest.TaskID}")
                        QuestInfo := quest_progress_info{Current := Quest.Progress, Total := Required, IconText := TextureString}
                        set QuestInfos = QuestInfos + array{QuestInfo}
                SetUIManager.QueueQuestProgress(QuestInfos)


    UpdateQuestUI(QuestStats : quest_stats, QuestDescription : string, ?IsNewQuest:logic=false, ?DoesSkipCompleteFlair : logic = false):void=
        if:
            SetUIManager:=UIManager?
            CurrentProgress := GetQuestStats[QuestStats.QuestID, QuestStats.TaskID].Progress
            SetQuestManager := QuestManager?
            Task := SetQuestManager.QuestModel.Quests[QuestStats.QuestID].Tasks[QuestStats.TaskID]
            RequiredProgress := Task.Progress
            TextureString := Task.IconTexture
        then:
            if:
                IsNewQuest?
            then:
                set QuestOnUI = QuestStats 
                SetUIManager.SetQuestText(QuestDescription, ?DoesSkipCompleteFlair := DoesSkipCompleteFlair)

            if: 
                QuestStats.QuestID = QuestOnUI.QuestID
                QuestStats.TaskID = QuestOnUI.TaskID
            then:
                SetUIManager.QueueQuestProgress(array{quest_progress_info{Current := CurrentProgress, Total := RequiredProgress, IconText := TextureString}})
            
    HasCompletedTask(InQuestID : int, InTaskID : int)<transacts>:logic=
        if:
            PlayerManager := GameManager.PlayerManagers[Agent]
        then:
            if:
                InQuestID = 0
                InTaskID = -1
            then:
                return true

            #Check completed quests
            if:
                CompletedQuests := PlayerManager.GetCompletedQuests[]
            then:
                for (QuestID : CompletedQuests):
                    if (InQuestID <= QuestID):
                        return true

            if:
                CurrentQuests := PlayerManager.GetCurrentQuests[]
            then:
                var HighestActiveTask : int = -1
                for (i := 0..CurrentQuests.Length - 1):
                    if:
                        CurrentQuests[i].QuestID = InQuestID
                        CurrentTaskID := CurrentQuests[i].TaskID
                    then:
                        if:
                            CurrentTaskID > HighestActiveTask
                        then:
                            set HighestActiveTask = CurrentTaskID
                
                if:
                    HighestActiveTask <= InTaskID
                then:
                    return false

                if:
                    not HighestActiveTask = -1
                then:
                    if:
                        HighestActiveTask > InTaskID
                    then:
                        return true
        
        return false
    
    #Damage Dealt
    OnDamageDealt(AgentIn: agent, Damage : big_number):void=
        if(SetUIManager:=UIManager?):
            SetUIManager.QueueDamageDealt(AgentIn, Damage)

    OnPlayerEliminated(Result : elimination_result):void=
        Print("Setting EliminatedZone to {CurrentZone}")
        
        set EliminatedZone = CurrentZone

        # if:
        #     RespawningAgent := Result.EliminatedCharacter.GetAgent[]
        # then:
        #     RespawningAgent.Respawn(vector3{}, IdentityRotation())

    #Damage Taken
    OnDamageTaken(DamageResult:damage_result):void=
        block:
        # Log("Damage Taken: {DamageResult.Amount}")

    #Prestige Functions

    GetPrestigeCount()<transacts>:int=
        if:
            SourceData := GetSourceDataSafe[]
        then:
            return SourceData.PrestigeCount
        else:
            return 0

    CalculatePrestigeGain():big_number=

            Base := big_number{Value:=EssenceLevel()*5.0}
            GunMult := 1.0 + 0.01 * TotalGunLevels()
            PeelyMult := 1.0 + 0.01 * PeeliesFound()
            BossMult := MakeBigNumber(1.1, TotalBossLevels())

            return Base * GunMult * PeelyMult * BossMult

 
    EssenceLevel():int=
        if (Level := GetLevel[]):
            return Level
        return 0

    TotalGunLevels():int=
        var Total : int = 0
        for (ID := 0..99):
            set Total += GetWeaponLevel(ID)
        return Total

    TotalBossLevels():int=
        var Total : int = 0
        for (Boss : BossList):
            if:
                BossLevel := GetAchievementLevel[Boss.AchievementName()]
                BossLevel > 0
            then:
                set Total += BossLevel - 1
        return Total

    PeeliesFound():big_number=
        GetCollectibleCount("Peely")

    #Peely Persistence
    GetPeelysFound()<decides><transacts>:[]int=
        SourceData := GetSourceDataSafe[]
        return SourceData.PeelysFound

    AddPeelysFound(Index : int)<decides><transacts>:void=
        SourceData := GetSourceDataSafe[]
        var Temp : []int =  SourceData.PeelysFound
        set Temp = Temp + array{Index}
        set PlayerProfileDataMap[Player] = player_profile_data:
            MakePlayerProfileData<constructor>(SourceData)
            PeelysFound := Temp

    ResetPeelysFound()<decides><transacts>:void=
        SourceData := GetSourceDataSafe[]
        var Temp : []int =  array{}
        set PlayerProfileDataMap[Player] = player_profile_data:
            MakePlayerProfileData<constructor>(SourceData)
            PeelysFound := Temp

    DoPrestigeReset(?Retries : int = 2):void=
        if:
            SourceData := GetSourceDataSafe[]
            SetQuestManager := QuestManager?
        then:
            if:
                TeleportLocation := GetRebirthTeleportLocation[]
                Player.GetFortCharacter[].TeleportTo[TeleportLocation.Translation, TeleportLocation.Rotation]
            then:

                UnsubscribeFromQuests()
                ClearInventory()
                ResetPlayerDamageTrackers()
        

                var NewUpgrades : [string]int = map{}
                var NewCollectibles : [string]collectible_data = map{}
                var NewAchievements : [string]int = map{}
                var NewGear : []armor_data = array{}

                for (Key -> Value : SourceData.UpgradeLevels):
                    if:
                        Key.Slice[0, 2] = "P_"
                        set NewUpgrades[Key] = Value
                
                for (Key -> Value : SourceData.Collectibles):
                    if:
                        Key.Slice[0, 2] = "P_"
                        set NewCollectibles[Key] = Value
                
                for (Key -> Value : SourceData.Achievements):
                    if:
                        Key.Slice[0, 2] = "P_"
                        set NewAchievements[Key] = Value


                PeelyCollectible := collectible_data{
                    Name := "Peely"
                    Key := "Peely"
                    Count := PeeliesFound()
                    ArmorData := false
                }
                if:
                    set NewCollectibles[PeelyCollectible.Key] = PeelyCollectible

                KeepGearLevel := GetUpgradeLevel["P_KeepGearOnPrestige"] or 0
                for (Item : SourceData.Gear):
                    Print("Checking Gear Slot {Item.Slot}")
                    if:
                        ArmorSlotFromString(Item.Slot) = armor_slot.Head
                        KeepGearLevel >= 1
                    then:
                        set NewGear = NewGear + array{Item}
                        
                        ItemCollectible := collectible_data{
                            Name := "Gear"
                            Key := "Unique_{Item.UID}"
                            Count := big_number{Value:=1.0, Exponent:=0}
                            ArmorData := option{Item}
                        }
                        if:
                            set NewCollectibles[ItemCollectible.Key] = ItemCollectible
                        Print("Keeping Helmet")

                    if:
                        ArmorSlotFromString(Item.Slot) = armor_slot.Torso
                        KeepGearLevel >= 2
                    then:
                        set NewGear = NewGear + array{Item}
                        
                        ItemCollectible := collectible_data{
                            Name := "Gear"
                            Key := "Unique_{Item.UID}"
                            Count := big_number{Value:=1.0, Exponent:=0}
                            ArmorData := option{Item}
                        }
                        if:
                            set NewCollectibles[ItemCollectible.Key] = ItemCollectible
                        Print("Keeping Chest")
                    
                    if:
                        ArmorSlotFromString(Item.Slot) = armor_slot.Legs
                        KeepGearLevel >= 3
                    then:
                        set NewGear = NewGear + array{Item}
                        
                        ItemCollectible := collectible_data{
                            Name := "Gear"
                            Key := "Unique_{Item.UID}"
                            Count := big_number{Value:=1.0, Exponent:=0}
                            ArmorData := option{Item}
                        }
                        if:
                            set NewCollectibles[ItemCollectible.Key] = ItemCollectible
                        Print("Keeping Legs")
                    
                    if:
                        ArmorSlotFromString(Item.Slot) = armor_slot.Boots
                        KeepGearLevel >= 4
                    then:
                        set NewGear = NewGear + array{Item}
                        
                        ItemCollectible := collectible_data{
                            Name := "Gear"
                            Key := "Unique_{Item.UID}"
                            Count := big_number{Value:=1.0, Exponent:=0}
                            ArmorData := option{Item}
                        }
                        if:
                            set NewCollectibles[ItemCollectible.Key] = ItemCollectible
                        Print("Keeping Boots")
                    
                    if:
                        ArmorSlotFromString(Item.Slot) = armor_slot.Gloves
                        KeepGearLevel >= 5
                    then:
                        set NewGear = NewGear + array{Item}
                        
                        ItemCollectible := collectible_data{
                            Name := "Gear"
                            Key := "Unique_{Item.UID}"
                            Count := big_number{Value:=1.0, Exponent:=0}
                            ArmorData := option{Item}
                        }
                        if:
                            set NewCollectibles[ItemCollectible.Key] = ItemCollectible
                        Print("Keeping Gloves")
                    
                    if:
                        ArmorSlotFromString(Item.Slot) = armor_slot.Accessory
                        KeepGearLevel >= 6
                    then:
                        set NewGear = NewGear + array{Item}
                        
                        ItemCollectible := collectible_data{
                            Name := "Gear"
                            Key := "Unique_{Item.UID}"
                            Count := big_number{Value:=1.0, Exponent:=0}
                            ArmorData := option{Item}
                        }
                        if:
                            set NewCollectibles[ItemCollectible.Key] = ItemCollectible
                        Print("Keeping Accessory")

                if:
                    set PlayerProfileDataMap[Player] = player_profile_data:
                        MakePlayerProfileData<constructor>(SourceData)
                        PlayerLevel := 1
                        PrestigeCount := SourceData.PrestigeCount + 1
                        PlayerStrength := big_number{Value:=1.0, Exponent:=0}
                        LifetimeEssence := SourceData.LifetimeEssence + SourceData.CurrentEssence + SourceData.StashedEssence
                        CurrentEssence := big_number{Value:=0.0, Exponent:=0}
                        StashedEssence := big_number{Value:=0.0, Exponent:=0}
                        #CurrentGems  := big_number{Value:=0.0, Exponent:=0}
                        GuardsDestroyed := big_number{Value:=0.0, Exponent:=0} # We should rack lifetime guards destroyed
                        ZoneClearArray := array{ 0,0,0,0,0,0,0,0,0,0,
                                                0,0,0,0,0,0,0,0,0,0 }
                        ZoneLevelArray := array{ 1,1,1,1,1,1,1,1,1,1,
                                                1,1,1,1,1,1,1,1,1,1 }
                        CurrentQuests := array{} #QuestIndex    
                        CompletedQuests := array{} #QuestIndex
                        Gear := NewGear
                        LastKnownLocation := vector3{}
                        LastKnownRotation := vector3{}
                        LastTimeOnline := SourceData.LastTimeOnline
                        LastDailyClaim := SourceData.LastDailyClaim
                        LastPeelyTime := SourceData.LastPeelyTime
                        BuffStates := map{}
                        UpgradeLevels := NewUpgrades
                        Collectibles := NewCollectibles
                        Achievements := NewAchievements
                        EquippedItems := array{-1,-1,-1,-1,-1}
                        ItemCooldowns := array  {0,0,0,0,0,0,0,0,0,0, #100 items
                                                        0,0,0,0,0,0,0,0,0,0,
                                                        0,0,0,0,0,0,0,0,0,0,
                                                        0,0,0,0,0,0,0,0,0,0,
                                                        0,0,0,0,0,0,0,0,0,0,
                                                        0,0,0,0,0,0,0,0,0,0,
                                                        0,0,0,0,0,0,0,0,0,0,
                                                        0,0,0,0,0,0,0,0,0,0,
                                                        0,0,0,0,0,0,0,0,0,0,
                                                        0,0,0,0,0,0,0,0,0,0}
                        PrestigePoints := SourceData.PrestigePoints
                        BigRockLevel := -1
                        BossKeySeconds := 0
                        ClaimCooldowns := map{}
                        PeelysFound := array{}
                        WeaponLevel := map{}
                        WeaponExperience := map{}
                        MonotonicID := SourceData.MonotonicID

            #weapon unlocks
            if (ShouldSkipQuestGroup("Pistol")?):
                GainUpgradeLevel("Gun_0Fixed", 1)
                GainUpgradeLevel("Gun_1", 1)

            if (ShouldSkipQuestGroup("AR")?):
                GainUpgradeLevel("Gun_2", 1)
                
            if (ShouldSkipQuestGroup("Shotgun")?):
                GainUpgradeLevel("Gun_4", 1)

            if (ShouldSkipQuestGroup("Bow")?):
                GainUpgradeLevel("Gun_5", 1)

            if (ShouldSkipQuestGroup("Minigun")?):
                GainUpgradeLevel("Gun_7", 1)

            if (ShouldSkipQuestGroup("StakeGun")?):
                GainCollectible(collectible_data{Key:="Garlic", Count:=big_number{Value:=1.0}})
                GainCollectible(collectible_data{Key:="SilverCross", Count:=big_number{Value:=1.0}})
                GainCollectible(collectible_data{Key:="WoodStake", Count:=big_number{Value:=1.0}})

            if (ShouldSkipQuestGroup("Vampires")?):
                GainUpgradeLevel("Gun_9", 1)

            #carry over weapons
            if (SourceData.UpgradeLevels["Gun_3"] > 0):
                GainUpgradeLevel("Gun_3", 1)
                
            if (SourceData.UpgradeLevels["Gun_6"] > 0):
                GainUpgradeLevel("Gun_6", 1)

            #upgrade unlocks
            if (ShouldSkipQuestGroup("LyraUpgrade")?):
                GainUpgradeLevel("Momentum", 2)
                GainUpgradeLevel("Teamwork", 2)
                GainUpgradeLevel("Focus Fire", 2)

            if (ShouldSkipQuestGroup("SamiraUpgrade")?):
                GainUpgradeLevel("Handcannon", 2)
                GainUpgradeLevel("Ambush", 2)

            #misc fixes
            if (ShouldSkipQuestGroup("MidasCastle")?):
                OnEscapedMidasCastle()

            GainCollectible(collectible_data{Key := "Gems", Count := big_number{Value:=1.0} * GetUpgradeValueSafe("P_BonusStartingGems", false)})
            GainCollectible(collectible_data{Key := "BossKey", Count := big_number{Value:=1.0} * GetUpgradeValueSafe("P_BonusStartingBossKeys", false)})
            
            SetQuestManager.BeginStartingQuests()
            SetQuestManager.UnhideAll()
            CheckQuestProgress(quest_stats{QuestID := 0, TaskID := 0})


            PeelyManager.ClearPeelyPersistence(Agent)
            CheckModifierUpgrades()
            ForceUpdateUI()

        else:
            if (Retries > 0):
                Print("Prestige teleport failed, retrying")
                DoPrestigeReset(?Retries := Retries - 1)
            else:
                Print("Prestige teleport failed, aborting")
        

            

    #Buff Functions
    GetBuffStates()<decides><transacts>:[string]buff=
        SourceData := GetSourceDataSafe[]
        return SourceData.BuffStates

    ClearBuffUI():void=
        var NewMap : [string]buff := map{}
        if:
            set NewMap["Momentum"] = buff{Name := "Momentum", Amount := big_number{}, Suffix := "", Duration := 0.0}
            set NewMap["Focus Fire"] = buff{Name := "Focus Fire", Amount := big_number{}, Suffix := "", Duration := 0.0}
            set NewMap["Healingwinds"] = buff{Name := "Healingwinds", Amount := big_number{}, Suffix := "", Duration := 0.0}
        
        if:
            SetUIManager := UIManager?
        then:
            SetUIManager.UpdateBuffDurations(UpdateBuffDuration:=NewMap, IsBounce :=false)

    LoopBuffDurations():void=
        var NewMap : [string]buff := map{}
        if:
            ActiveBuffs := GetBuffStates[]
        then:
            for (Buff : ActiveBuffs):
                OldDuration := Buff.Duration
                NewDuration := Max(Buff.Duration - 1.0, 0.0)
                var NewAmount : big_number = Buff.Amount
                if:
                    OldDuration > 0.0
                    NewDuration = 0.0
                then:
                    set NewAmount = big_number{}
                    CheckModifierUpgrades()

                if:
                    set NewMap[Buff.Name] = buff{Name := Buff.Name, Amount := NewAmount, Suffix := Buff.Suffix, Duration := NewDuration}
        
        if:
            SourceData := GetSourceDataSafe[]
            set PlayerProfileDataMap[Player] = player_profile_data:
                MakePlayerProfileData<constructor>(SourceData)
                BuffStates := NewMap

        if:
            SetUIManager := UIManager?
        then:
            SetUIManager.UpdateBuffDurations(UpdateBuffDuration:=NewMap, IsBounce :=false)
 
    
    CheckMomentum():void=        
        if:
            GetUpgradeValue["Momentum", false] > 0.0
            MaxAmount := big_number{Value := GetUpgradeValue["Momentum", false]}
            Amount := big_number{Value := 5.0}
            Duration := 195.0 * (1.0 + GetUpgradeValueSafe("P_BuffDuration", false))
        then:
            UpdateBuff("Momentum", "%", Amount, true, Duration, false, ?MaxAmount := option{MaxAmount})
            CheckModifierUpgrades()

    OnWeakspotHit():void=
        if:
            GetUpgradeValue["Focus Fire", false] > 0.0
            MaxAmount := big_number{Value := GetUpgradeValue["Focus Fire", false]} * GetCurrentGunStats[].FocusFireMult
            Amount := big_number{Value := 1.0 * GetCurrentGunStats[].FocusFireMult}
            Duration := 195.0 / GetCurrentGunStats[].FocusFireMult * (1.0 + GetUpgradeValueSafe("P_BuffDuration", false))
        then:
            UpdateBuff("Focus Fire", "%", Amount, true, Duration, false, ?MaxAmount := option{MaxAmount})
        
    GetBuff(Name : string)<decides><transacts>:buff=
        var TargetBuff : buff = buff{}
        ActiveBuffs := GetBuffStates[]

        for (Buff : ActiveBuffs):
            if:
                Buff.Name = Name
            then:
                set TargetBuff = Buff
        
        return TargetBuff

    UpdateBuff(Name : string, Suffix:string, Amount : big_number, IsAmountAdditive : logic, Duration : float, IsDurationAdditive : logic, ?MaxAmount : ?big_number = false, ?IsSecondary:logic=false):void=
        
        var ActiveBuffAmount : big_number = big_number{}
        var ActiveBuffDuration : float = 0.0

        if:
            var PotentialActiveBuff : buff = GetBuff[Name]
            PotentialActiveBuff.Name = Name
            set ActiveBuffAmount = PotentialActiveBuff.Amount
            set ActiveBuffDuration = PotentialActiveBuff.Duration
        then:
            if:
                IsAmountAdditive?
            then:
                set ActiveBuffAmount = ActiveBuffAmount + Amount
            else:
                set ActiveBuffAmount = Amount
            
            if:
                SetMaxAmount := MaxAmount?
                BigNumber_GreaterThanOrEquals(ActiveBuffAmount, SetMaxAmount)?
            then:
                set ActiveBuffAmount = SetMaxAmount

            if:
                IsDurationAdditive?
            then:
                set ActiveBuffDuration += Duration
            else:
                set ActiveBuffDuration = Duration
        else:
            set ActiveBuffAmount = Amount
            set ActiveBuffDuration = Duration
        
        UpdatedBuff := buff{Name := Name, Suffix := Suffix, Amount := ActiveBuffAmount, Duration := ActiveBuffDuration, IsSecondary := IsSecondary}
        var NewMap : [string]buff := map{}        

        if:
            SourceData := GetSourceDataSafe[]

            set NewMap = SourceData.BuffStates
            set NewMap[Name] = UpdatedBuff

            set PlayerProfileDataMap[Player] = player_profile_data:
                MakePlayerProfileData<constructor>(SourceData)
                BuffStates := NewMap

        if:
            SetUIManager := UIManager?
        then:
            SetUIManager.UpdateBuffDurations(UpdateBuffDuration:=NewMap, IsBounce :=true)
                
    #Misc Functions
    UpdateHandcannon():logic=
        if:
            GetUpgradeValue["Handcannon", false] = 0.0
        then:    
            return false

        if: 
            Handcannon = 4
        then:
            set Handcannon = 1

            if (GetUpgradeValue["SuperHandcannon", false] > 0.0):
                if (SuperHandcannon = 4):
                    set SuperHandcannon = 1
                else:
                    set SuperHandcannon += 1

            # Print("Handcannon: {Handcannon} ({SuperHandcannon})")
        else:
            set Handcannon += 1

        if:
            UI := UIManager?
        then:
            if (GetUpgradeValue["SuperHandcannon", false] > 0.0):
                UI.UpdateHandcannon(Handcannon, SuperHandcannon, true)
            else:
                UI.UpdateHandcannon(Handcannon, SuperHandcannon, false)

        if:
            Handcannon = 4
        then:
            return true
        else:
            return false

    IsSuperHandcannon():logic=
        if (SuperHandcannon = 4 and Handcannon = 4):
            return true
        else:
            return false

    IsPreSuperHandcannon():logic=
        if (SuperHandcannon = 3 and Handcannon = 3):
            return true
        else:
            return false


    TakeUniqueID()<transacts>:int=
        if:
            SourceData := GetSourceDataSafe[]
            set PlayerProfileDataMap[Player] = player_profile_data:
                MakePlayerProfileData<constructor>(SourceData)
                MonotonicID := SourceData.MonotonicID + 1
        then:
            return SourceData.MonotonicID + 1
        return 0

    UnlockTeleporter(TeleporterID : int, Name : string):void=
        if:
            SetUIManager := UIManager?
        then:
            GainAchievementLevel(Name, 1)
            SetUIManager.ShowNewTeleporterUnlocked(Name)
            GameManager.AudioManager.PlayUnlockZoneSFX(Agent)

    SetLocationAndRotation(WaitForGround:logic)<suspends>:void=
        loop:
            if:
                Player.IsActive[]
                Player.GetFortCharacter[].IsOnGround[]
                not Player.GetFortCharacter[].IsInWater[]
                not CurrentZone.IsBossTag()?
            then:
                SetLocation()
                SetRotation()
                break

            if:
                WaitForGround?
                Player.IsActive[]
            then:
                Sleep(1.0)
            else:
                break

    SetLocation():void=
        if:
            Fort:= Agent.GetFortCharacter[]
            SourceData := GetSourceDataSafe[]
            set PlayerProfileDataMap[Player] = player_profile_data:
                MakePlayerProfileData<constructor>(SourceData)
                LastKnownLocation := Fort.GetTransform().Translation

    SetRotation():void=
        if:
            Fort:= Agent.GetFortCharacter[]
            SourceData := GetSourceDataSafe[]

            YawPitchRoll := Fort.GetTransform().Rotation.GetYawPitchRollDegrees()
            RotVec := vector3{X := YawPitchRoll[0], Y := YawPitchRoll[1], Z := YawPitchRoll[2]} 

            set PlayerProfileDataMap[Player] = player_profile_data:
                MakePlayerProfileData<constructor>(SourceData)
                LastKnownRotation := RotVec

    GetLocation()<decides><transacts>:vector3=
        SourceData := GetSourceDataSafe[]
        return SourceData.LastKnownLocation

    GetRotation()<decides><transacts>:rotation=
        SourceData := GetSourceDataSafe[]
        return MakeRotationFromYawPitchRollDegrees(SourceData.LastKnownRotation.X, SourceData.LastKnownRotation.Y, SourceData.LastKnownRotation.Z)

    SetRealtime():void=
        Realtime:= GetSecondsSinceEpoch()
        if:
            SourceData := GetSourceDataSafe[]
            set PlayerProfileDataMap[Player] = player_profile_data:
                MakePlayerProfileData<constructor>(SourceData)
                LastTimeOnline := Realtime

    GetRealtime()<decides><transacts>:float=
        SourceData := GetSourceDataSafe[]
        return SourceData.LastTimeOnline

    GetCreationTime()<decides><transacts>:float=
        SourceData := GetSourceDataSafe[]
        return SourceData.CreationTime

    GetWeakSpotSFX()<transacts>:int=
        Num:= GetRandomInt(1, 5) 
        if(Num = WeakSpotSFX):
            if(WeakSpotSFX = 5):
                set WeakSpotSFX = 1
            else:
                set WeakSpotSFX += 1
        else:
            set WeakSpotSFX = Num
        
        return WeakSpotSFX

    SetPeelyDate()<decides><transacts>:void=
        SourceData := GetSourceDataSafe[]
        CurrentTime := GetSecondsSinceEpoch()
        set PlayerProfileDataMap[Player] = player_profile_data:
            MakePlayerProfileData<constructor>(SourceData)
            LastPeelyTime := CurrentTime
        
    IsNewPeelyDate()<decides><transacts>:void=
        SourceData := GetSourceDataSafe[]
        CurrentTime := GetSecondsSinceEpoch()

        if:
            OldDays := Int[SourceData.LastPeelyTime / 86400.0]
            NewDays := Int[CurrentTime / 86400.0]
            not(OldDays = NewDays)
        then:
            block:
        else:
            false?

    CanClaimDailyRewards()<decides><transacts>:void=
        SourceData := GetSourceDataSafe[]
        CurrentTime := GetSecondsSinceEpoch()
        if:
            GetAchievementLevel["AllowDailyHarvest"] > 0
            OldDays := Int[SourceData.LastDailyClaim / 86400.0]
            NewDays := Int[CurrentTime / 86400.0]
            GetDailyRewardCount() = 0 or not(OldDays = NewDays)
        then:
            block:
        else:
            false?

    OnDailyRewardsClaimed():void=
        if:
            SourceData := GetSourceDataSafe[]
            CurrentTime := GetSecondsSinceEpoch()
            set PlayerProfileDataMap[Player] = player_profile_data:
                MakePlayerProfileData<constructor>(SourceData)
                LastDailyClaim := CurrentTime
        then:
            GainAchievementLevel("CollectedHarvest", 1)

    GetDailyRewardCount()<transacts>:int=
        if:
            Temp := GetAchievementLevel["CollectedHarvest"]
        then:
            return Temp
        else:
            return 0
    
    ClosePopup(Wdata:widget_message):void =
        ClosePopupUI()

    ClosePopupUI():void =
        if (SetUIManager := UIManager?):
            SetUIManager.ClosePopup()

    GetCooldown(ID:string)<transacts>:float =
        if:
            SourceData := GetSourceDataSafe[]
        then:
            if:
                CDTimestamp := SourceData.ClaimCooldowns[ID]
                Delta := CDTimestamp - GetSecondsSinceEpoch()
                Delta > 0.0
            then:
                return Delta
            else:
                return 0.0
        return 0.0

    SetCooldown(ID:string, Seconds:float)<transacts>:void =
        if:
            SourceData := GetSourceDataSafe[]

            var NewClaimCooldowns : [string]float = SourceData.ClaimCooldowns
            set NewClaimCooldowns[ID] = GetSecondsSinceEpoch() + Seconds

            set PlayerProfileDataMap[Player] = player_profile_data:
                MakePlayerProfileData<constructor>(SourceData)
                ClaimCooldowns := NewClaimCooldowns

    Heal(?Percent : float = 100.0):void=
        if (FortChar := Player.GetFortCharacter[]):
            FortChar.SetHealth(FortChar.GetMaxHealth() * Percent / 100.0)

    ###Debug for Trackers
    IncreaseStrength():void=
        if:
            Amt:=GetStrength[]
        then:
            GainStrength(Amt)
        if (UI :=UIManager?):
            UI.UpdateLevelUI()
            
    DecreaseStrength():void=
        if:
            Amt:=big_number{}-GetStrength[]
        then:
            GainStrength(Amt/2.0)
        if (UI :=UIManager?):
            UI.UpdateLevelUI()

    DebugPlayerArmor():void=
        if:
            SourceData := GetSourceDataSafe[]
        then:
            var Diag:[]dialog_message = array{}
            var Index : int = 1
            for (Item : SourceData.Collectibles):
                var IsArmor : string = "False"
                var ArmorName : string = ""
                if:
                    Item.ArmorData?
                    Armor:=Item.ArmorData?.ToArmor()
                then:
                    set IsArmor = "True"
                    set ArmorName = "{Armor.Prefix} {Armor.Base} {Armor.Suffix} | Armor Stats Length: {Armor.Stats.Length}"
                set Diag = Diag + array{dialog_message{
                    Title := option{"{Index} of {SourceData.Collectibles.Length}"}
                    Message := "WEEPING28. SHOW THIS TO THE DEVS:<br> #{Index}: Name: {Item.Name}  Key: {Item.Key} Count: {Item.Count} IsArmor : {IsArmor}<br>Armor Stats (If Applicable): {ArmorName}"
                }}
            
            spawn:
                MakeDialog("", Diag).ShowDialog(Self, 0.01)
            
            var Keys : []string = array{}
            for (Item : SourceData.Collectibles):
                if:
                    Item.Count.Value = 0.0
                then:
                    set Keys = Keys + array{Item.Key}
            
            var NewMap : [string]collectible_data := SourceData.Collectibles
            for(Key : Keys):
                set NewMap = RemoveKeyFromMap2(NewMap, Key)

            if:
                set PlayerProfileDataMap[Player] = player_profile_data:
                    MakePlayerProfileData<constructor>(SourceData)
                    Collectibles := NewMap
    
    DebugPlayerData():void=
        if:
            SourceData := GetSourceDataSafe[]
        then:
            var Diag:[]dialog_message = array{}
            var Index : int = 1
            var Message : string = ""
            
            set Message = Message + "Current Quests / Tasks:"
            for (Quest : SourceData.CurrentQuests):
                set Message = Message + " {Quest.QuestID} / {Quest.TaskID} ||"
            
            set Message = Message + "<br> Completed Quests:"
            for (Quest : SourceData.CompletedQuests):
                set Message = Message + " {Quest} ||"

            set Diag = Diag + array{dialog_message{
                Title := option{"Imthekat"}
                Message := Message
            }}
            
            spawn:
                MakeDialog("", Diag).ShowDialog(Self, 0.01)

            
        
                
    CheckBigRockAnalytics(Index : int):void=
        case (Index):
            0 => GameManager.AnalyticsManager.big_rock_1.Submit(Agent)
            1 => GameManager.AnalyticsManager.big_rock_2.Submit(Agent)
            2 => GameManager.AnalyticsManager.big_rock_3.Submit(Agent)
            _ => block:
        
    ResetGame():void=    
        if:
            StartingTransform:= GameManager.AnalyticsManager.GetTransform()
            Player.GetFortCharacter[].TeleportTo[StartingTransform.Translation, StartingTransform.Rotation]
        UnsubscribeFromQuests()
        ClearInventory()
        ResetPlayerDamageTrackers()
        PeelyManager.ClearPeelyPersistence(Agent)
        ResetToBeginning()
        CheckModifierUpgrades()
        ForceUpdateUI()

    DebugResetPlayerData():void=
        UnsubscribeFromQuests()
        ClearInventory()
        ResetPlayerDamageTrackers()
        PeelyManager.ClearPeelyPersistence(Agent)
        
        #CHOOSE ONE:
        ResetGame()
        # ResetFastArea1Clear()
        # ResetFastArea2Clear()
        # ResetPostMidas(?DmgExp := 0) 
        # TestChest(?DmgExp := 0) 
        # ResetPostYoungBogloc(?DmgExp := 0) 
        # ResetPostWolves(?DmgExp := 0)
        # ResetGiantLands(?DmgExp := 6)
        # ResetPirateIsland(?DmgExp := 1)
        # ResetPostPirateIsland(?DmgExp := 3)
        # ResetPrestigeTest()
        CheckModifierUpgrades()

        ForceUpdateUI()

    UnsubscribeFromQuests():void=
        if(SetQuestManager:= QuestManager?):
            for (X:= 0..SetQuestManager.QuestModel.Quests.Length - 1):
                if:
                    QuestStats:=GetQuestStats[X]
                then:
                    SetQuestManager.UnsubscribeFromTask(QuestStats)
                    SetQuestManager.TurnOffQuestTrackers(QuestStats)

    ResetPlayerDamageTrackers():void=
        GameManager.ResetPlayerDamageTrackers(Player)

    ClearInventory():void=     
        #clear inventory
        for (Remover: GameManager.ItemRemovers):
            Remover.Remove(Player)
        
    ForceUpdateUI():void=
        if(SetUIManager:=UIManager?):
            SetUIManager.UpdateEssenceUI()
            SetUIManager.UpdateLevelUI()
            SetUIManager.UpdateUICounts()
            SetUIManager.UpdateStashUI(big_number{}, big_number{}, big_number{})
            SetUIManager.UpdateKeyUI(big_number{}, big_number{}, 0.0)
            ClearBuffUI()

            if (Level := GetLevel[]):
                UpdateLevelOnMaterial(Level)

    ResetToBeginning():void=             
        #Update PlayerProfile
            if:
                SetQuestManager:= QuestManager?
                set PlayerProfileDataMap[Player] = player_profile_data{}
            then:
                #Update Quests
                SetQuestManager.BeginStartingQuests()
                SetQuestManager.CompleteTask(quest_stats{QuestID:=0,TaskID:=0}) #Skip Cell Wall

                #unhide barriers
                SetQuestManager.UnhideAll()

    ResetFastArea1Clear():void=
        #Update PlayerProfile
        if:
            set PlayerProfileDataMap[Player] = player_profile_data{}
            SourceData := GetSourceDataSafe[]
            set PlayerProfileDataMap[Player] = player_profile_data:
                MakePlayerProfileData<constructor>(SourceData)
                PlayerLevel:=20
                PlayerStrength:= big_number{Value:=233.0, Exponent:= 0}
            PlayerManager := GameManager.PlayerManagers[Player]
            SetQuestManager:= QuestManager?
        then:
            #Update Quests
            SetQuestManager.BeginStartingQuests()
            SetQuestManager.CompleteTask(quest_stats{QuestID:=0,TaskID:=1}) 
            
            #Grant Gems
            PlayerManager.GainCollectible(collectible_data{Key := "Gems", Count := big_number{Value:=1.0, Exponent:=2}})

            #unhide barriers
            SetQuestManager.UnhideAll()
            
    ResetFastArea2Clear():void=
        #Update PlayerProfile
        if:
            set PlayerProfileDataMap[Player] = player_profile_data{}
            SourceData := GetSourceDataSafe[]

            set PlayerProfileDataMap[Player] = player_profile_data:
                MakePlayerProfileData<constructor>(SourceData)
                PlayerLevel:=20
                PlayerStrength:= big_number{Value:=233.0, Exponent:= 0}
            SetQuestManager:= QuestManager?
        then:
            #Grant Gems
            GainCollectible(collectible_data{Key := "Gems", Count := big_number{Value:=1.0, Exponent:=2}})

            # SetQuestManager.CompleteTask(quest_stats{QuestID:=0,TaskID:=10}) #Remove Barrier
            QuestToComplete := quest_stats{QuestID:=0, TaskID:= 18}
            SetQuestManager.CompleteTask(QuestToComplete, ?DebugOverride := true)

            SetQuestManager.CheckPersistentQuests()

        EquipItem(0)
        GainUpgradeLevel("KO Bonus", 1) 
        GainUpgradeLevel("Gun_0Fixed", 1) 
        GainUpgradeLevel("Clear Bonus", 1) 
        GainUpgradeLevel("stash", 1) 
        GainUpgradeLevel("Gun_2", 1) 
        GainAchievementLevel("InteractedWithNPC_Mel", 1)
        GainAchievementLevel("InteractedWithNPC_Lyra", 1)

    TestChest(?DmgExp : int = 0):void= 
        if:
            set PlayerProfileDataMap[Player] = player_profile_data{}
            SourceData := GetSourceDataSafe[]

            set PlayerProfileDataMap[Player] = player_profile_data:
                MakePlayerProfileData<constructor>(SourceData)
                PlayerLevel:=19
                PlayerStrength:= big_number{Value:=500.0, Exponent:= DmgExp}
                # PlayerLevel:= 40
                # PlayerStrength:= big_number{Value:=16350.0, Exponent:= DmgExp}
                # FOR Starting at Quest 41
                # PlayerLevel:= 32
                # PlayerStrength:= big_number{Value:=3.6, Exponent:= 3}
            SetQuestManager:= QuestManager?
        then:

            QuestToComplete := quest_stats{QuestID:=1, TaskID:= 3}
            #15 post Midas
            #25 Collect Stash
            #41 Collect Daily Harvest
            SetQuestManager.CompleteTask(QuestToComplete, ?DebugOverride := true)
# 
        #Grant Gems
        GainCollectible(collectible_data{Key := "Gems", Count := big_number{Value:=10.0, Exponent:=0}})
        #10 gems post midas, level 10
        #32 gems at level 32

        EquipItem(0)
        GainUpgradeLevel("Gun_0Fixed", 1) 
        GainUpgradeLevel("Gun_2", 1) 
        GainUpgradeLevel("Gun_4", 1) 

        GainAchievementLevel("Sarge_NextTier", 1)

        GainAchievementLevel("InteractedWithNPC_Lyra", 1)
        # GainAchievementLevel("InteractedWithNPC_Samira", 1)
        GainAchievementLevel("Midas_TP_Unlocked", 1)

        # DebugGainAllCollectibles(big_number{Value:= 30.0, Exponent:=0})

        GainAchievementLevel("EnteredMidas", 1)
        GainAchievementLevel("EscapedMidasCastle", 1)
        GameManager.SwitchToClass1.ChangeClass(Agent)

        if:
            InnTransform:= GameManager.InnTeleporter.GetTransform()
            Player.GetFortCharacter[].TeleportTo[InnTransform.Translation, InnTransform.Rotation]

    ResetPostMidas(?DmgExp : int = 0):void= 
        #Update PlayerProfile
        
        if:
            set PlayerProfileDataMap[Player] = player_profile_data{}
            SourceData := GetSourceDataSafe[]

            set PlayerProfileDataMap[Player] = player_profile_data:
                MakePlayerProfileData<constructor>(SourceData)
                PlayerLevel:=30
                PlayerStrength:= big_number{Value:=50.0, Exponent:= 3 + DmgExp}
                # PlayerLevel:= 40
                # PlayerStrength:= big_number{Value:=16350.0, Exponent:= DmgExp}
                # FOR Starting at Quest 41
                # PlayerLevel:= 32
                # PlayerStrength:= big_number{Value:=3.6, Exponent:= 3}
            SetQuestManager:= QuestManager?
        then:

            QuestToComplete := quest_stats{QuestID:=0, TaskID:= 9}
            #15 post Midas
            #25 Collect Stash
            #41 Collect Daily Harvest
            SetQuestManager.CompleteTask(QuestToComplete, ?DebugOverride := true)
# 
        #Grant Gems
        GainCollectible(collectible_data{Key := "Gems", Count := big_number{Value:=30.0, Exponent:=0}})
        #10 gems post midas, level 10
        #32 gems at level 32

        EquipItem(0)
        GainUpgradeLevel("Gun_0Fixed", 1) 
        GainUpgradeLevel("Gun_2", 1) 
        GainUpgradeLevel("Gun_4", 1) 

        GainAchievementLevel("InteractedWithNPC_Lyra", 1)
        GainAchievementLevel("InteractedWithNPC_Samira", 1)
        GainAchievementLevel("Midas_TP_Unlocked", 1)

        DebugGainAllCollectibles(big_number{Value:= 30.0, Exponent:=0})

        GainAchievementLevel("EnteredMidas", 1)
        GainAchievementLevel("EscapedMidasCastle", 1)
        GameManager.SwitchToClass1.ChangeClass(Agent)

        if:
            InnTransform:= GameManager.InnTeleporter.GetTransform()
            Player.GetFortCharacter[].TeleportTo[InnTransform.Translation, InnTransform.Rotation]
    
    ResetPostYoungBogloc(?DmgExp : int = 0):void= 
        #Update PlayerProfile
        
        if:
            set PlayerProfileDataMap[Player] = player_profile_data{}
            SourceData := GetSourceDataSafe[]

            set PlayerProfileDataMap[Player] = player_profile_data:
                MakePlayerProfileData<constructor>(SourceData)
                PlayerLevel:=30
                PlayerStrength:= big_number{Value:=2.0, Exponent:= 5}
            SetQuestManager:= QuestManager?
        then:
            CompleteQuest(0, ?DebugOverride := true)
            CompleteQuest(1, ?DebugOverride := true)

            QuestToComplete := quest_stats{QuestID:=2, TaskID:= 0}
            
            SetQuestManager.CompleteTask(QuestToComplete, ?DebugOverride := true)
            
        #Grant Gems
        GainCollectible(collectible_data{Key := "Gems", Count := big_number{Value:=30.0, Exponent:=0}})
        GainCollectible(collectible_data{Key := "BossKey", Count := big_number{Value:=1.0, Exponent:=0}})

        EquipItem(0)

        GainUpgradeLevel("Gun_0Fixed", 1) 
        GainUpgradeLevel("Gun_2", 1) 
        GainUpgradeLevel("Gun_4", 1) 
        GainUpgradeLevel("Gun_6", 1) 
        
        GainUpgradeLevel("Focus Fire", 3) 
        GainUpgradeLevel("Teamwork", 3) 
        GainUpgradeLevel("Momentum", 3) 
        GainUpgradeLevel("Hand Cannon", 3) 
        GainUpgradeLevel("Ambush", 3) 

        GainAchievementLevel("Sarge_NextTier", 1)

        GainAchievementLevel("InteractedWithNPC_Lyra", 1)
        GainAchievementLevel("InteractedWithNPC_Samira", 1)
        GainAchievementLevel("Midas_TP_Unlocked", 1)
        GainAchievementLevel("Inn", 1)
        GainAchievementLevel("Midas Castle",1)
        GainAchievementLevel("Umbral Quarry", 1)
        GainAchievementLevel("Giantlands", 1)
        GainAchievementLevel("Zephyr's Canyon", 1)
        GainAchievementLevel("Goodtown Beach", 1)
        GainAchievementLevel("Bandit Camp", 1)

        # DebugGainAllCollectibles(big_number{Value:= 30.0, Exponent:=0})

        GainAchievementLevel("EnteredMidas", 1)
        GainAchievementLevel("EscapedMidasCastle", 1)
        GameManager.SwitchToClass1.ChangeClass(Agent)

        if:
            InnTransform:= GameManager.InnTeleporter.GetTransform()
            Player.GetFortCharacter[].TeleportTo[InnTransform.Translation, InnTransform.Rotation]

    ResetPostWolves(?DmgExp : int = 0):void= 
        #Update PlayerProfile
        
        if:
            set PlayerProfileDataMap[Player] = player_profile_data{}
            SourceData := GetSourceDataSafe[]

            set PlayerProfileDataMap[Player] = player_profile_data:
                MakePlayerProfileData<constructor>(SourceData)
                PlayerLevel:=70
                PlayerStrength:= big_number{Value:=4.5, Exponent:= 6 + DmgExp}
            SetQuestManager:= QuestManager?
        then:

            QuestToComplete := quest_stats{QuestID:=0, TaskID:= 55}
            SetQuestManager.CompleteTask(QuestToComplete, ?DebugOverride := true)
            SetQuestManager.StartQuest(3)
            
        #Grant Gems
        GainCollectible(collectible_data{Key := "Gems", Count := big_number{Value:=70.0, Exponent:=0}})
        GainCollectible(collectible_data{Key := "BossKey", Count := big_number{Value:=1.0, Exponent:=0}})


        EquipItem(0)
        GainUpgradeLevel("Gun_0Fixed", 1) 
        GainUpgradeLevel("Gun_2", 1) 
        GainUpgradeLevel("Gun_4", 1) 
        GainUpgradeLevel("Gun_5", 1) 
        GainUpgradeLevel("Gun_6", 1) 
        
        GainUpgradeLevel("Focus Fire", 6) 
        GainUpgradeLevel("Teamwork", 6) 
        GainUpgradeLevel("Momentum", 6) 
        GainUpgradeLevel("Stealth Attack", 3) 
        GainUpgradeLevel("Hand Cannon", 6) 
        GainUpgradeLevel("Ambush", 6) 

        GainAchievementLevel("Sarge_NextTier", 1)

        GainAchievementLevel("InteractedWithNPC_Lyra", 1)
        GainAchievementLevel("InteractedWithNPC_Samira", 1)
        GainAchievementLevel("InteractedWithNPC_Germit", 1)
        GainAchievementLevel("Midas_TP_Unlocked", 1)
        GainAchievementLevel("Inn", 1)
        GainAchievementLevel("Midas Castle",1)
        GainAchievementLevel("Umbral Quarry", 1)
        GainAchievementLevel("Giantlands", 1)
        GainAchievementLevel("Zephyr's Canyon", 1)
        GainAchievementLevel("Goodtown Beach", 1)
        GainAchievementLevel("Bandit Camp", 1)


        # DebugGainAllCollectibles(big_number{Value:= 30.0, Exponent:=0})

        GainAchievementLevel("EnteredMidas", 1)
        GainAchievementLevel("EscapedMidasCastle", 1)
        GameManager.SwitchToClass1.ChangeClass(Agent)

        if:
            InnTransform:= GameManager.InnTeleporter.GetTransform()
            Player.GetFortCharacter[].TeleportTo[InnTransform.Translation, InnTransform.Rotation]
                    

    ResetGiantLands(?DmgExp : int = 0):void= 
        #Update PlayerProfile
        
        if:
            set PlayerProfileDataMap[Player] = player_profile_data{}
            SourceData := GetSourceDataSafe[]

            set PlayerProfileDataMap[Player] = player_profile_data:
                MakePlayerProfileData<constructor>(SourceData)
                PlayerLevel:=69
                PlayerStrength:= big_number{Value:=4.2, Exponent:= 7+DmgExp}
            SetQuestManager:= QuestManager?
        then:
            # Game Start
            # SetQuestManager.BeginStartingQuests()
            # SetQuestManager.CompleteTask(quest_stats{QuestID:=0,TaskID:=0}) #Skip Cell Wall

            # Giantlands
            # CompleteQuest(0, ?DebugOverride := true)
            # CompleteQuest(1, ?DebugOverride := true)
            # CompleteQuest(2, ?DebugOverride := true)

            # QuestToComplete := quest_stats{QuestID:=3, TaskID:= 0}
            # SetQuestManager.CompleteTask(QuestToComplete, ?DebugOverride := true)

            # Blossom
            CompleteQuest(0, ?DebugOverride := true)
            CompleteQuest(1, ?DebugOverride := true)
            CompleteQuest(2, ?DebugOverride := true)

            QuestToComplete := quest_stats{QuestID:=3, TaskID:= 6}
            
            SetQuestManager.CompleteTask(QuestToComplete, ?DebugOverride := true)

        #Grant Gems
        GainCollectible(collectible_data{Key := "Gems", Count := big_number{Value:=70.0, Exponent:=0}})
        GainCollectible(collectible_data{Key := "BossKey", Count := big_number{Value:=1.0, Exponent:=0}})


        EquipItem(0)
        GainUpgradeLevel("Gun_0Fixed", 1) 
        GainUpgradeLevel("Gun_1", 1) 
        GainUpgradeLevel("Gun_2", 1) 
        GainUpgradeLevel("Gun_3", 1) 
        GainUpgradeLevel("Gun_4", 1) 
        GainUpgradeLevel("Gun_5", 1) 
        GainUpgradeLevel("Gun_6", 1) 
        GainUpgradeLevel("Gun_7", 1) 
        GainUpgradeLevel("Gun_8", 1) 
        GainUpgradeLevel("Gun_9", 1) 
        GainUpgradeLevel("Gun_10", 1) 
        
        GainUpgradeLevel("Focus Fire", 6) 
        GainUpgradeLevel("Teamwork", 6) 
        GainUpgradeLevel("Momentum", 6) 
        GainUpgradeLevel("Stealth Attack", 6) 
        GainUpgradeLevel("Hand Cannon", 6) 
        GainUpgradeLevel("Ambush", 6) 

        GainAchievementLevel("Sarge_NextTier", 1)

        GainAchievementLevel("InteractedWithNPC_Lyra", 1)
        GainAchievementLevel("InteractedWithNPC_Samira", 1)
        GainAchievementLevel("InteractedWithNPC_Germit", 1)
        GainAchievementLevel("Midas_TP_Unlocked", 1)
        GainAchievementLevel("Inn", 1)
        GainAchievementLevel("Midas Castle",1)
        GainAchievementLevel("Umbral Quarry", 1)
        GainAchievementLevel("Giantlands", 1)
        GainAchievementLevel("Zephyr's Canyon", 1)
        GainAchievementLevel("Goodtown Beach", 1)
        GainAchievementLevel("Bandit Camp", 1)
        
        #GainAchievementLevel("Genga", 1)


        # DebugGainAllCollectibles(big_number{Value:= 30.0, Exponent:=0})

        GainAchievementLevel("EnteredMidas", 1)
        GainAchievementLevel("EscapedMidasCastle", 1)
        GameManager.SwitchToClass1.ChangeClass(Agent)

        if:
            InnTransform:= GameManager.InnTeleporter.GetTransform()
            Player.GetFortCharacter[].TeleportTo[InnTransform.Translation, InnTransform.Rotation]
            

    ResetPirateIsland(?DmgExp : int = 0):void= 
        #Update PlayerProfile
        
        if:
            set PlayerProfileDataMap[Player] = player_profile_data{}
            SourceData := GetSourceDataSafe[]

            set PlayerProfileDataMap[Player] = player_profile_data:
                MakePlayerProfileData<constructor>(SourceData)
                PlayerLevel:=100
                PlayerStrength:= big_number{Value:=8.2, Exponent:= 8 + DmgExp}
            SetQuestManager:= QuestManager?
        then:

            # Blossom
            CompleteQuest(0, ?DebugOverride := true)
            CompleteQuest(1, ?DebugOverride := true)
            CompleteQuest(2, ?DebugOverride := true)
            CompleteQuest(3, ?DebugOverride := true)
            QuestToComplete := quest_stats{QuestID:=4, TaskID:= 0}
            # Testing
            
            SetQuestManager.CompleteTask(QuestToComplete, ?DebugOverride := true)

        #Grant Gems
        GainCollectible(collectible_data{Key := "Gems", Count := big_number{Value:=100.0, Exponent:=0}})
        GainCollectible(collectible_data{Key := "BossKey", Count := big_number{Value:=20.0, Exponent:=0}})


        EquipItem(0)
        GainUpgradeLevel("Gun_0Fixed", 1) 
        GainUpgradeLevel("Gun_1", 1) 
        GainUpgradeLevel("Gun_2", 1) 
        GainUpgradeLevel("Gun_3", 1) 
        GainUpgradeLevel("Gun_4", 1) 
        GainUpgradeLevel("Gun_5", 1) 
        GainUpgradeLevel("Gun_6", 1) 
        GainUpgradeLevel("Gun_7", 1) 
        GainUpgradeLevel("Gun_8", 1) 
        GainUpgradeLevel("Gun_9", 1) 
        GainUpgradeLevel("Gun_10", 1) 
        
        GainUpgradeLevel("Focus Fire", 6) 
        GainUpgradeLevel("Teamwork", 6) 
        GainUpgradeLevel("Momentum", 6) 
        GainUpgradeLevel("Stealth Attack", 6) 
        GainUpgradeLevel("Hand Cannon", 6) 
        GainUpgradeLevel("Ambush", 6) 

        GainAchievementLevel("Sarge_NextTier", 1)

        GainAchievementLevel("InteractedWithNPC_Lyra", 1)
        GainAchievementLevel("InteractedWithNPC_Samira", 1)
        GainAchievementLevel("InteractedWithNPC_Germit", 1)
        GainAchievementLevel("Midas_TP_Unlocked", 1)
        GainAchievementLevel("Inn", 1)
        GainAchievementLevel("Midas Castle",1)
        GainAchievementLevel("Umbral Quarry", 1)
        GainAchievementLevel("Giantlands", 1)
        GainAchievementLevel("Zephyr's Canyon", 1)
        GainAchievementLevel("Goodtown Beach", 1)
        GainAchievementLevel("Bandit Camp", 1)
        GainAchievementLevel("Pirate Island", 1)
        
        GainAchievementLevel("Sharkey", 1)
        GainAchievementLevel("Genga", 1)
        GainAchievementLevel("Urjent", 1)
        
        # DebugGainAllCollectibles(big_number{Value:= 30.0, Exponent:=0})

        GainAchievementLevel("EnteredMidas", 1)
        GainAchievementLevel("EscapedMidasCastle", 1)
        GameManager.SwitchToClass1.ChangeClass(Agent)

        if:
            InnTransform:= GameManager.InnTeleporter.GetTransform()
            Player.GetFortCharacter[].TeleportTo[InnTransform.Translation, InnTransform.Rotation]
    

    ResetPostPirateIsland(?DmgExp : int = 0):void= 
        #Update PlayerProfile
        
        if:
            set PlayerProfileDataMap[Player] = player_profile_data{}
            SourceData := GetSourceDataSafe[]

            set PlayerProfileDataMap[Player] = player_profile_data:
                MakePlayerProfileData<constructor>(SourceData)
                PlayerLevel:=130
                PlayerStrength:= big_number{Value:=8.2, Exponent:= 11 + DmgExp}
            SetQuestManager:= QuestManager?
        then:

            # Blossom
            CompleteQuest(0, ?DebugOverride := true)
            CompleteQuest(1, ?DebugOverride := true)
            CompleteQuest(2, ?DebugOverride := true)
            CompleteQuest(3, ?DebugOverride := true)
            # Testing
            
            QuestToComplete := quest_stats{QuestID:=4, TaskID:= 31}
            SetQuestManager.CompleteTask(QuestToComplete, ?DebugOverride := true)

        #Grant Gems
        GainCollectible(collectible_data{Key := "Gems", Count := big_number{Value:=100.0, Exponent:=0}})
        GainCollectible(collectible_data{Key := "BossKey", Count := big_number{Value:=20.0, Exponent:=0}})


        EquipItem(0)
        GainUpgradeLevel("Gun_0Fixed", 1) 
        GainUpgradeLevel("Gun_1", 1) 
        GainUpgradeLevel("Gun_2", 1) 
        GainUpgradeLevel("Gun_3", 1) 
        GainUpgradeLevel("Gun_4", 1) 
        GainUpgradeLevel("Gun_5", 1) 
        GainUpgradeLevel("Gun_6", 1) 
        GainUpgradeLevel("Gun_7", 1) 
        GainUpgradeLevel("Gun_8", 1) 
        GainUpgradeLevel("Gun_9", 1) 
        GainUpgradeLevel("Gun_10", 1) 
        
        GainUpgradeLevel("Focus Fire", 6) 
        GainUpgradeLevel("Teamwork", 6) 
        GainUpgradeLevel("Momentum", 6) 
        GainUpgradeLevel("Stealth Attack", 6) 
        GainUpgradeLevel("Hand Cannon", 6) 
        GainUpgradeLevel("Ambush", 6) 
        GainUpgradeLevel("Super Cannon", 6) 

        GainAchievementLevel("Sarge_NextTier", 1)

        GainAchievementLevel("InteractedWithNPC_Lyra", 1)
        GainAchievementLevel("InteractedWithNPC_Samira", 1)
        GainAchievementLevel("InteractedWithNPC_Germit", 1)
        GainAchievementLevel("Midas_TP_Unlocked", 1)
        GainAchievementLevel("Inn", 1)
        GainAchievementLevel("Midas Castle",1)
        GainAchievementLevel("Umbral Quarry", 1)
        GainAchievementLevel("Giantlands", 1)
        GainAchievementLevel("Zephyr's Canyon", 1)
        GainAchievementLevel("Goodtown Beach", 1)
        GainAchievementLevel("Bandit Camp", 1)
        GainAchievementLevel("Pirate Island", 1)

        GainAchievementLevel("Sarge_NextTier", 1)
        GainAchievementLevel("Sharkey_NextTier", 1)
        GainAchievementLevel("Genga_NextTier", 1)
        GainAchievementLevel("Urjent_NextTier", 1)
        GainAchievementLevel("PirateKing_NextTier", 1)
        
        GainAchievementLevel("Sharkey", 1)
        GainAchievementLevel("Genga", 1)
        GainAchievementLevel("Urjent", 1)
        GainAchievementLevel("Pirate King", 1)
        
        DebugGainAllCollectibles(big_number{Value:= 30.0, Exponent:=0})

        GainAchievementLevel("EnteredMidas", 1)
        GainAchievementLevel("EscapedMidasCastle", 1)
        GameManager.SwitchToClass1.ChangeClass(Agent)

        if:
            InnTransform:= GameManager.InnTeleporter.GetTransform()
            Player.GetFortCharacter[].TeleportTo[InnTransform.Translation, InnTransform.Rotation]
    

    ResetPrestigeTest(?DmgExp : int = 0):void= 
        #Update PlayerProfile
        
        if:
            set PlayerProfileDataMap[Player] = player_profile_data{}
            SourceData := GetSourceDataSafe[]

            set PlayerProfileDataMap[Player] = player_profile_data:
                MakePlayerProfileData<constructor>(SourceData)
                PlayerLevel:=1000
                PlayerStrength:= big_number{Value:=8.2, Exponent:= 30 + DmgExp}
            SetQuestManager:= QuestManager?
        then:

            # Blossom
            CompleteQuest(0, ?DebugOverride := true)
            CompleteQuest(1, ?DebugOverride := true)
            CompleteQuest(2, ?DebugOverride := true)
            CompleteQuest(3, ?DebugOverride := true)
            # Testing
            
            QuestToComplete := quest_stats{QuestID:=4, TaskID:= 32}
            SetQuestManager.CompleteTask(QuestToComplete, ?DebugOverride := true)

        #Grant Gems
        GainCollectible(collectible_data{Key := "Gems", Count := big_number{Value:=100.0, Exponent:=10}})
        GainCollectible(collectible_data{Key := "BossKey", Count := big_number{Value:=20.0, Exponent:=0}})
        GainCollectible(collectible_data{Key := "P_PrestigeGem", Count := big_number{Value:=2.0, Exponent:=50}}) 


        EquipItem(0)
        GainUpgradeLevel("Gun_0Fixed", 1) 
        GainUpgradeLevel("Gun_1", 1) 
        GainUpgradeLevel("Gun_2", 1) 
        GainUpgradeLevel("Gun_3", 1) 
        GainUpgradeLevel("Gun_4", 1) 
        GainUpgradeLevel("Gun_5", 1) 
        GainUpgradeLevel("Gun_6", 1) 
        GainUpgradeLevel("Gun_7", 1) 
        GainUpgradeLevel("Gun_8", 1) 
        GainUpgradeLevel("Gun_9", 1) 
        GainUpgradeLevel("Gun_10", 1) 
        
        GainUpgradeLevel("Focus Fire", 6) 
        GainUpgradeLevel("Teamwork", 6) 
        GainUpgradeLevel("Momentum", 6) 
        GainUpgradeLevel("Stealth Attack", 6) 
        GainUpgradeLevel("Hand Cannon", 6) 
        GainUpgradeLevel("Ambush", 6) 
        GainUpgradeLevel("Super Cannon", 6) 

        GainAchievementLevel("Sarge_NextTier", 1)

        GainAchievementLevel("InteractedWithNPC_Lyra", 1)
        GainAchievementLevel("InteractedWithNPC_Samira", 1)
        GainAchievementLevel("InteractedWithNPC_Germit", 1)
        GainAchievementLevel("Midas_TP_Unlocked", 1)
        GainAchievementLevel("Inn", 1)
        GainAchievementLevel("Midas Castle",1)
        GainAchievementLevel("Umbral Quarry", 1)
        GainAchievementLevel("Giantlands", 1)
        GainAchievementLevel("Zephyr's Canyon", 1)
        GainAchievementLevel("Goodtown Beach", 1)
        GainAchievementLevel("Bandit Camp", 1)
        GainAchievementLevel("Pirate Island", 1)

        GainAchievementLevel("Sarge_NextTier", 1)
        GainAchievementLevel("Sharkey_NextTier", 1)
        GainAchievementLevel("Genga_NextTier", 1)
        GainAchievementLevel("Urjent_NextTier", 1)
        GainAchievementLevel("PirateKing_NextTier", 1)
        
        GainAchievementLevel("Sharkey", 1)
        GainAchievementLevel("Genga", 1)
        GainAchievementLevel("Urjent", 1)
        GainAchievementLevel("Pirate King", 1)
        
        DebugGainAllCollectibles(big_number{Value:= 30.0, Exponent:=0})

        GainAchievementLevel("EnteredMidas", 1)
        GainAchievementLevel("EscapedMidasCastle", 1)
        GameManager.SwitchToClass1.ChangeClass(Agent)

        if:
            InnTransform:= GameManager.InnTeleporter.GetTransform()
            Player.GetFortCharacter[].TeleportTo[InnTransform.Translation, InnTransform.Rotation]

    DebugGainAllCollectibles(Amount : big_number):void=
        GainCollectible(collectible_data{Key := "Gems", Count := Amount})
        GainCollectible(collectible_data{Key := "Candy1", Count := Amount})
        GainCollectible(collectible_data{Key := "Candy2", Count := Amount})
        GainCollectible(collectible_data{Key := "Candy3", Count := Amount})
        GainCollectible(collectible_data{Key := "Candy4", Count := Amount})
        GainCollectible(collectible_data{Key := "Candy5", Count := Amount})
        GainCollectible(collectible_data{Key := "Candy6", Count := Amount})
        GainCollectible(collectible_data{Key := "Goodflower", Count := Amount})
        GainCollectible(collectible_data{Key := "Beach Attack Plans", Count := Amount})
        GainCollectible(collectible_data{Key := "KeySharkey", Count := Amount})
        GainCollectible(collectible_data{Key := "BoarHide", Count := Amount})
        GainCollectible(collectible_data{Key := "WolfLiver", Count := Amount})
        GainCollectible(collectible_data{Key := "WolfFur", Count := Amount})
        GainCollectible(collectible_data{Key := "BossKey", Count := Amount})
        GainCollectible(collectible_data{Key := "UpgradeHammer", Count := Amount})
