prestige_upgrade := enum:
    None
    ReducedXPRequirement
    KeepGearOnPrestige
    BonusStartingGems
    BonusStartingBossKeys
    AutoClearQuest
    AutoFarmChest
    BonusItemLevelDrops
    ItemRarityImprovement
    ArmorPiercing
    GunDuration
    BuffDuration
    ReducedUpgradeLevelRequirement
    DamageMelee #Pickaxe, Scythe
    DamageRapid #Pistol, AR, Minigun
    DamagePrecision #Shotgun, Railgun, Bow, StakeGun

PrestigeUpgradesList: []prestige_upgrade = array:
    prestige_upgrade.ReducedXPRequirement
    prestige_upgrade.KeepGearOnPrestige
    prestige_upgrade.BonusStartingGems
    prestige_upgrade.BonusStartingBossKeys
    prestige_upgrade.AutoClearQuest
    #prestige_upgrade.AutoFarmChest
    prestige_upgrade.BonusItemLevelDrops
    prestige_upgrade.ItemRarityImprovement
    #prestige_upgrade.ArmorPiercing
    prestige_upgrade.GunDuration
    prestige_upgrade.BuffDuration
    prestige_upgrade.ReducedUpgradeLevelRequirement
    prestige_upgrade.DamageMelee
    prestige_upgrade.DamageRapid
    prestige_upgrade.DamagePrecision

StepScale(Value : int)<transacts>:float = 
    if (Value < 1):
        return 0.0
    Pow(10.0, Floor((Value-1)/9)*1.0) * (Mod[(Value-1), 9] + 1) or 0.0

(Upgrade: prestige_upgrade).GetValue(Level : int)<transacts>:float =
    return case (Upgrade):
        prestige_upgrade.ReducedXPRequirement => 1.0 - 1.0/(1.0+StepScale(Level)*0.1)
        prestige_upgrade.BonusStartingGems => 10.0 * Level
        prestige_upgrade.BonusStartingBossKeys => 1.0 * Level
        prestige_upgrade.BonusItemLevelDrops => StepScale(Level)
        prestige_upgrade.ArmorPiercing => case (Level):
            0 => 0.0
            1 => 0.1
            2 => 0.2
            3 => 0.3
            4 => 0.4
            5 => 0.5
            6 => 0.55
            7 => 0.6
            8 => 0.65
            9 => 0.7
            10 => 0.75
            11 => 0.775
            12 => 0.8
            13 => 0.825
            14 => 0.85  
            15 => 0.875
            16 => 0.9
            17 => 0.91
            18 => 0.92
            19 => 0.93
            20 => 0.94
            21 => 0.95
            22 => 0.96
            23 => 0.965
            24 => 0.97
            25 => 0.975
            26 => 0.98
            27 => 0.985
            28 => 0.99
            _ => 0.99
        prestige_upgrade.GunDuration => 0.1 * StepScale(Level)
        prestige_upgrade.BuffDuration => 0.1 * StepScale(Level)
        prestige_upgrade.ReducedUpgradeLevelRequirement => StepScale(Level)
        prestige_upgrade.DamageMelee => 0.1 * StepScale(Level)
        prestige_upgrade.DamageRapid => 0.1 * StepScale(Level)
        prestige_upgrade.DamagePrecision => 0.1 * StepScale(Level)
        _ => 0.0

(Upgrade: prestige_upgrade).GetCost(Level : int):big_number=
    case (Upgrade):
        prestige_upgrade.ReducedXPRequirement => 200.0 * MakeBigNumber(StepScale(Level), 2)
        prestige_upgrade.KeepGearOnPrestige => big_number{Value:=Level*1.0, Exponent:=0+2*Level}
        prestige_upgrade.BonusStartingGems => 100.0 * MakeBigNumber(2.0, Level)
        prestige_upgrade.BonusStartingBossKeys => 100.0 * MakeBigNumber(3.0, Level)
        prestige_upgrade.AutoClearQuest => case (Level):
            1 => big_number{Value:=1.0, Exponent:=2}
            2 => big_number{Value:=1.0, Exponent:=3}
            3 => big_number{Value:=1.0, Exponent:=3}
            4 => big_number{Value:=2.0, Exponent:=3}
            5 => big_number{Value:=2.0, Exponent:=3}
            6 => big_number{Value:=1.0, Exponent:=4}
            7 => big_number{Value:=5.0, Exponent:=4}
            8 => big_number{Value:=1.0, Exponent:=5}
            9 => big_number{Value:=2.0, Exponent:=5}
            10 => big_number{Value:=5.0, Exponent:=5}
            11 => big_number{Value:=1.0, Exponent:=6}
            12 => big_number{Value:=2.0, Exponent:=6}
            13 => big_number{Value:=3.0, Exponent:=6}
            _ => big_number{Value:=9.99, Exponent:=999}
        prestige_upgrade.AutoFarmChest => case (Level):
            1 => big_number{Value:=1.0, Exponent:=3}
            2 => big_number{Value:=2.0, Exponent:=3}
            3 => big_number{Value:=5.0, Exponent:=3}
            4 => big_number{Value:=1.0, Exponent:=4}
            5 => big_number{Value:=1.0, Exponent:=5}
            6 => big_number{Value:=1.0, Exponent:=6}
            7 => big_number{Value:=1.0, Exponent:=7}
            8 => big_number{Value:=1.0, Exponent:=8}
            _ => big_number{Value:=9.99, Exponent:=999}
        prestige_upgrade.BonusItemLevelDrops => 100.0 * MakeBigNumber(2.0, StepScale(Level))
        prestige_upgrade.ItemRarityImprovement => case (Level):
            1 => big_number{Value:=1.0, Exponent:=3}#"Common Items no longer drop from Bosses"
            2 => big_number{Value:=2.0, Exponent:=3}#"2x Drop Rate for Epic and Legendary Items" 
            3 => big_number{Value:=5.0, Exponent:=3}#"Mythic Rarity items can drop from Bosses"
            4 => big_number{Value:=5.0, Exponent:=3}#"3x Drop Rate for Epic or higher Items"
            5 => big_number{Value:=2.0, Exponent:=4}#"Unlock Godly Rarity Item Drops"
            6 => big_number{Value:=3.0, Exponent:=4}#"30% Chance for Items to drop at a higher Rarity"
            7 => big_number{Value:=1.0, Exponent:=5}#"Unlock Exalted Rarity Item Drops"
            8 => big_number{Value:=2.0, Exponent:=5}#"Uncommon Items no longer drop from Bosses"
            9 => big_number{Value:=5.0, Exponent:=5}#"2x Drop Rate for Mythic or higher Items"
            10 => big_number{Value:=1.0, Exponent:=6}#"2x Drop Rate for Exalted Items"
            11 => big_number{Value:=1.0, Exponent:=7}#"Unlock Ultimate Rarity Item Drops"
            12 => big_number{Value:=5.0, Exponent:=7}#"3x Drop Rate for Godly and Exalted Items"
            13 => big_number{Value:=1.0, Exponent:=9}#"Unlock Supreme Rarity Item Drops"
            14 => big_number{Value:=1.0, Exponent:=10}#"3x Drop Rate for Legendary and higher Items"
            15 => big_number{Value:=1.0, Exponent:=11}#"50% Chance for Items to drop at a higher Rarity"
            16 => big_number{Value:=1.0, Exponent:=12}#"Unlock Transcendent Rarity Item Drops"
            17 => big_number{Value:=1.0, Exponent:=13}#"Rare Items no longer drop from Bosses"
            18 => big_number{Value:=1.0, Exponent:=15}#"Epic Items no longer drop from Bosses"
            19 => big_number{Value:=1.0, Exponent:=17}#"2x Drop Rate for Supreme and Transcendent Items"
            20 => big_number{Value:=1.0, Exponent:=18}#"2x Drop Rate for Transcendent Items"
            21 => big_number{Value:=1.0, Exponent:=20}#"Unlock Infinite Rarity Item Drops"
            22 => big_number{Value:=1.0, Exponent:=21}#"Legendary/Godly/Ultimate Items always drop as the next Rarity"
            23 => big_number{Value:=1.0, Exponent:=22}#"2x Drop Rate for Transcendent and Infinite Items"
            24 => big_number{Value:=1.0, Exponent:=23}#"4x Drop Rate for Transcendent Items"
            25 => big_number{Value:=1.0, Exponent:=25}#"6x Drop Rate for Infinite Items"
            _ => big_number{Value:=9.99, Exponent:=999}
        prestige_upgrade.ArmorPiercing => big_number{Value:=1.0, Exponent:=1+Level}
        prestige_upgrade.GunDuration => 50.0 * MakeBigNumber(StepScale(Level), 2)
        prestige_upgrade.BuffDuration => 50.0 * MakeBigNumber(4.0, StepScale(Level)) 
        prestige_upgrade.ReducedUpgradeLevelRequirement => 50.0 * MakeBigNumber(2.0, StepScale(Level)) 
        prestige_upgrade.DamageMelee => 50.0 * MakeBigNumber(StepScale(Level)*1.0, 1.5)
        prestige_upgrade.DamageRapid => 50.0 * MakeBigNumber(StepScale(Level)*1.0, 1.5)
        prestige_upgrade.DamagePrecision => 50.0 * MakeBigNumber(StepScale(Level)*1.0, 1.5)
        _ => big_number{Value:=9.99, Exponent:=999}

(Upgrade: prestige_upgrade).GetDescription(Level : int):string =
    case (Upgrade):
        prestige_upgrade.ReducedXPRequirement => "Reduce XP Requirements by {FormatPercent(Upgrade.GetValue(Level), 2)}"
        prestige_upgrade.KeepGearOnPrestige => case (Level):
            1 => "Keep Helmet on Rebirth"
            2 => "Keep Chest Armor on Rebirth"
            3 => "Keep Leg Armor on Rebirth"
            4 => "Keep Boots on Rebirth"
            5 => "Keep Gloves on Rebirth"
            6 => "Keep Accessory on Rebirth"
            _ => "..."
        prestige_upgrade.BonusStartingGems => "Start with {Round[Upgrade.GetValue(Level)] or 0} Extra Gems"
        prestige_upgrade.BonusStartingBossKeys => "Start with {Round[Upgrade.GetValue(Level)] or 0} Extra Boss Keys"
        prestige_upgrade.AutoClearQuest => case (Level):
            1 => "Start with Pistol/AR/Shotgun Unlocked"
            2 => "Start with 2 points in Lyra's Upgrades"
            3 => "Start with 2 points in Hand Cannon and First Strike"
            4 => "Skip Blossom and Candy's Quests"
            5 => "Start with Bow/Minigun Unlocked"
            6 => "Start in Goodtown"
            7 => "Skip Samira's Quests on the Beach" 
            8 => "Skip Germit's Quests in the Forest"
            9 => "Start with one Garlic, Stake, and Silver Cross"
            10 => "Start in the Giantlands"
            11 => "Skip Barbariana's Quests in the Raider Outpost"
            12 => "Skip Vampire Hunter G's Quests"
            13 => "Start at Pirate Island"
            _ => "???"
        prestige_upgrade.AutoFarmChest => case (Level):
            1 => "Automatically farm the Beach Entrance EXP chest"
            2 => "Automatically farm the Beach Cave EXP chest"
            3 => "Automatically farm the Forest Entrance EXP chest"
            4 => "Automatically farm the Bandit Camp EXP chest"
            5 => "Automatically farm the Raider Outpost EXP chest"
            6 => "Automatically farm the Giant's Fortress EXP chest"
            7 => "Automatically farm the Pirate Cove EXP chest"
            8 => "Automatically farm the Pirate Warehouse EXP chest"
            _ => "???"
        prestige_upgrade.BonusItemLevelDrops => "+{Round[Upgrade.GetValue(Level)] or 0} to max level of items dropped by bosses"
        prestige_upgrade.ItemRarityImprovement => case (Level):
            1 => "Common Items no longer drop from Bosses"
            2 => "2x Drop Rate for Epic and Legendary Items" 
            3 => "Mythic Rarity items can drop from Bosses"
            4 => "3x Drop Rate for Epic or higher Items"
            5 => "Unlock Godly Rarity Item Drops"
            6 => "30% Chance for Items to drop at a higher Rarity"
            7 => "Unlock Exalted Rarity Item Drops"
            8 => "Uncommon Items no longer drop from Bosses"
            9 => "2x Drop Rate for Mythic or higher Items"
            10 => "2x Drop Rate for Exalted Items"
            11 => "Unlock Ultimate Rarity Item Drops"
            12 => "3x Drop Rate for Godly and Exalted Items"
            13 => "Unlock Supreme Rarity Item Drops"
            14 => "3x Drop Rate for Legendary and higher Items"
            15 => "50% Chance for Items to drop at a higher Rarity"
            16 => "Unlock Transcendent Rarity Item Drops"
            17 => "Rare Items no longer drop from Bosses"
            18 => "Epic Items no longer drop from Bosses"
            19 => "2x Drop Rate for Supreme and Transcendent Items"
            20 => "2x Drop Rate for Transcendent Items"
            21 => "Unlock Infinite Rarity Item Drops"
            22 => "Legendary/Godly/Ultimate Items always drop as the next Rarity"
            23 => "2x Drop Rate for Transcendent and Infinite Items"
            24 => "4x Drop Rate for Transcendent Items"
            25 => "6x Drop Rate for Infinite Items"
            _ => "???"
        prestige_upgrade.ArmorPiercing => "Ignore {FormatPercent(Upgrade.GetValue(Level))} of Enemy Armor"
        prestige_upgrade.GunDuration => "Guns last {FormatPercent(Upgrade.GetValue(Level))} longer"
        prestige_upgrade.BuffDuration => "Buffs last {FormatPercent(Upgrade.GetValue(Level))} longer"
        prestige_upgrade.ReducedUpgradeLevelRequirement => "Upgrade tiers are unlocked {Round[Upgrade.GetValue(Level)] or 0} levels earlier"
        prestige_upgrade.DamageMelee => "Increase damage with melee weapons by {FormatPercent(Upgrade.GetValue(Level))}"
        prestige_upgrade.DamageRapid => "Increase damage with rapid fire weapons by {FormatPercent(Upgrade.GetValue(Level))}"
        prestige_upgrade.DamagePrecision => "Increase damage with precision weapons by {FormatPercent(Upgrade.GetValue(Level))}"
        _ => "???"

(Upgrade: prestige_upgrade).GetValueDescription(Level : int):string =
    case (Upgrade):
        prestige_upgrade.ReducedXPRequirement => "{FormatPercent(Upgrade.GetValue(Level), 2)}"
        prestige_upgrade.KeepGearOnPrestige => " "
        prestige_upgrade.BonusStartingGems => "{Round[Upgrade.GetValue(Level)] or 0}"
        prestige_upgrade.BonusStartingBossKeys => "{Round[Upgrade.GetValue(Level)] or 0}"
        prestige_upgrade.AutoClearQuest => " "
        prestige_upgrade.AutoFarmChest => " "
        prestige_upgrade.BonusItemLevelDrops => "+{Round[Upgrade.GetValue(Level)] or 0}"
        prestige_upgrade.ItemRarityImprovement => " "
        prestige_upgrade.ArmorPiercing => "{FormatPercent(Upgrade.GetValue(Level))}"
        prestige_upgrade.GunDuration => "{FormatPercent(Upgrade.GetValue(Level))}"
        prestige_upgrade.BuffDuration => "{FormatPercent(Upgrade.GetValue(Level))}"
        prestige_upgrade.ReducedUpgradeLevelRequirement => "{Round[Upgrade.GetValue(Level)] or 0}"
        prestige_upgrade.DamageMelee => "{FormatPercent(Upgrade.GetValue(Level))}"
        prestige_upgrade.DamageRapid => "{FormatPercent(Upgrade.GetValue(Level))}"
        prestige_upgrade.DamagePrecision => "{FormatPercent(Upgrade.GetValue(Level))}"
        _ => "???"

(Upgrade: prestige_upgrade).GetKey()<transacts>:string =
    case (Upgrade):
        prestige_upgrade.None => ""
        prestige_upgrade.ReducedXPRequirement => "P_ReducedXPRequirement"
        prestige_upgrade.KeepGearOnPrestige => "P_KeepGearOnPrestige"
        prestige_upgrade.BonusStartingGems => "P_BonusStartingGems"
        prestige_upgrade.BonusStartingBossKeys => "P_BonusStartingBossKeys"
        prestige_upgrade.AutoClearQuest => "P_AutoClearQuest"
        prestige_upgrade.AutoFarmChest => "P_AutoFarmChest"
        prestige_upgrade.BonusItemLevelDrops => "P_BonusItemLevelDrops"
        prestige_upgrade.ItemRarityImprovement => "P_ItemRarityImprovement"
        prestige_upgrade.ArmorPiercing => "P_ArmorPiercing"
        prestige_upgrade.GunDuration => "P_GunDuration"
        prestige_upgrade.BuffDuration => "P_BuffDuration"
        prestige_upgrade.ReducedUpgradeLevelRequirement => "P_ReducedUpgradeLevelRequirement"
        prestige_upgrade.DamageMelee => "P_DamageMelee"
        prestige_upgrade.DamageRapid => "P_DamageRapid"
        prestige_upgrade.DamagePrecision => "P_DamagePrecision"

GetPrestigeUpgradeFromKey(Key:string)<transacts>:prestige_upgrade=
    case (Key):
        "P_ReducedXPRequirement" => prestige_upgrade.ReducedXPRequirement
        "P_KeepGearOnPrestige" => prestige_upgrade.KeepGearOnPrestige
        "P_BonusStartingGems" => prestige_upgrade.BonusStartingGems
        "P_BonusStartingBossKeys" => prestige_upgrade.BonusStartingBossKeys
        "P_AutoClearQuest" => prestige_upgrade.AutoClearQuest
        "P_AutoFarmChest" => prestige_upgrade.AutoFarmChest
        "P_BonusItemLevelDrops" => prestige_upgrade.BonusItemLevelDrops
        "P_ItemRarityImprovement" => prestige_upgrade.ItemRarityImprovement
        "P_ArmorPiercing" => prestige_upgrade.ArmorPiercing
        "P_GunDuration" => prestige_upgrade.GunDuration
        "P_BuffDuration" => prestige_upgrade.BuffDuration
        "P_ReducedUpgradeLevelRequirement" => prestige_upgrade.ReducedUpgradeLevelRequirement
        "P_DamageMelee" => prestige_upgrade.DamageMelee
        "P_DamageRapid" => prestige_upgrade.DamageRapid
        "P_DamagePrecision" => prestige_upgrade.DamagePrecision
        _ => prestige_upgrade.None

GetRarityWeights(PrestigeUpgradeLevel : int)<transacts>:[]int = case (PrestigeUpgradeLevel):

    #             C | UC| R | Ep| L | M | G | Ex| U | S | T | I
    #            ---+---+---+---+---+---+---+---+---+---+---+---
    0  =>  array{200,550,200, 40, 10,  0,  0,  0,  0,  0,  0,  0}   #base
    1  =>  array{  0,550,200, 40, 10,  0,  0,  0,  0,  0,  0,  0}   #no commons
    2  =>  array{  0,550,200, 80, 20,  0,  0,  0,  0,  0,  0,  0}   #x2 epic+
    3  =>  array{  0,550,200, 80, 20, 10,  0,  0,  0,  0,  0,  0}   #mythic
    4  =>  array{  0,550,200,240, 60, 30,  0,  0,  0,  0,  0,  0}   #x3 epic+
    5  =>  array{  0,550,200,240, 60, 30,  5,  0,  0,  0,  0,  0}   #godly
    6  =>  array{  0,380,300,220,110, 40, 11,  0,  0,  0,  0,  0}   #30% upgrade
    7  =>  array{  0,380,300,220,110, 40, 11,  5,  0,  0,  0,  0}   #exalted
    8  =>  array{  0,  0,300,220,110, 40, 11,  5,  0,  0,  0,  0}   #no uncommons
    9  =>  array{  0,  0,300,220,110, 80, 22, 10,  0,  0,  0,  0}   #x2 mythic+
    10 =>  array{  0,  0,300,220,110, 80, 22, 20,  0,  0,  0,  0}   #x3 exalted
    11 =>  array{  0,  0,300,220,110, 80, 22, 20,  4,  0,  0,  0}   #ultimate
    12 =>  array{  0,  0,300,220,110, 80, 66, 60,  4,  0,  0,  0}   #x3 godly+
    13 =>  array{  0,  0,300,220,110, 80, 66, 60,  4,  2,  0,  0}   #supreme
    14 =>  array{  0,  0,300,220,330,240,198,180, 12,  6,  0,  0}   #x3 legendary+
    15 =>  array{  0,  0,150,260,275,285,219,189, 96, 12,  0,  0}   #50% upgrade
    16 =>  array{  0,  0,150,260,275,285,219,189, 96, 12,  2,  0}   #transcendent
    17 =>  array{  0,  0,  0,260,275,285,219,189, 96, 12,  2,  0}   #no rares
    18 =>  array{  0,  0,  0,  0,275,285,219,189, 96, 12,  2,  0}   #no epics
    19 =>  array{  0,  0,  0,  0,275,285,219,189, 96, 36,  6,  0}   #x3 supreme+
    20 =>  array{  0,  0,  0,  0,275,285,219,189, 96, 36, 12,  0}   #x2 transcendent+
    21 =>  array{  0,  0,  0,  0,275,285,219,189, 96, 36, 12,  1}   #infinite
    22 =>  array{  0,  0,  0,  0,  0,560,  0,408,  0,132, 12,  1}   #legendary/godly/ultimate guaranteed upgrade
    23 =>  array{  0,  0,  0,  0,  0,560,  0,408,  0,132, 12,  2}   #x2 transcendent+
    24 =>  array{  0,  0,  0,  0,  0,560,  0,408,  0,132, 48,  2}   #x4 transcendent
    25 =>  array{  0,  0,  0,  0,  0,560,  0,408,  0,132, 38, 12}   #x6 infinite
    _  =>  array{  1,  0,  0,  0,  0,  0,  0,  0,  0,  0,  0,  0}