
using { /Fortnite.com/Devices }
using { /Verse.org/Simulation }
using { /UnrealEngine.com/Temporary/Diagnostics }
using { /Verse.org/Assets}
using { /Fortnite.com/UI }
using { /UnrealEngine.com/Temporary/UI}
using { /UnrealEngine.com/Temporary/SpatialMath }    
using { /Verse.org/Colors }

prestige_menu_device := class(menu_device_template):
    
    @editable
    GameManager : ?game_manager = false
    
    @editable
    Teleporter : teleporter_device = teleporter_device{}

    @editable 
    SpendPointsButton : button_device = button_device{}

    GetGameManager<override>():game_manager=
        if (GM := GameManager?):
            return GM

        return game_manager{}

    IsEnabled<override>(PlayerManager:player_manager)<transacts>:logic=
        if:
            PlayerManager.GetPrestigeCount() > 0
            PlayerManager.HasCompletedTask(5,3)?
        then:
            return true
        else:
            return false

    MakeMenu<override>(GM : game_manager, PM: player_manager):menu_template=
        return prestige_confirm_popup{GameManager:=option{GM}, Agent:=option{PM.Player}, Teleporter:=Teleporter}  

    GetTitle<override>():string=
        return "Prestige"

    GetIcon<override>():texture=
        return Images.PrestigeGem

    OnBegin<override>()<suspends>:void=
        SpendPointsButton.InteractedWithEvent.Subscribe(OpenPointsMenu)

    OpenPointsMenu(Agent : agent):void=
        Menu := prestige_menu{GameManager:=GameManager, Agent:=option{Agent}}  
        Menu.OpenMenu()

prestige_final_confirmation_device := class(menu_device_template):
    
    @editable
    GameManager : ?game_manager = false

    GetGameManager<override>():game_manager=
        if (GM := GameManager?):
            return GM

        return game_manager{}

    IsEnabled<override>(PlayerManager:player_manager)<transacts>:logic=
        return false

    MakeMenu<override>(GM : game_manager, PM: player_manager):menu_template=
        return prestige_final_confirm_popup{GameManager:=option{GM}, Agent:=option{PM.Player}}  

    GetTitle<override>():string=
        return "Prestige"

    GetIcon<override>():texture=
        return Images.PrestigeGem

prestige_confirm_popup := class(menu_template):
    Teleporter : teleporter_device

    DialogBackground : texture_block = texture_block{ DefaultImage := Images.DialogBox, DefaultDesiredSize := vector2{X := 1920.0, Y := 1080.0}} 
    var TextBox : stack_box = stack_box{Orientation := orientation.Vertical}

    var NoButton : button_regular = button_regular:

    var YesButton : button_loud = button_loud:
    
    UpdateMenu<override>(Widget:canvas):void=
        if:
            GM := GameManager?
            SetAgent := Agent?
            PlayerManager := GM.PlayerManagers[SetAgent]
            Widget = _Widget?
        then:
            TimeCrystalCount := PlayerManager.CalculatePrestigeGain()

            var ExplanationStr : string = "Are you sure you want to rebirth?<br><br>This will reset your progress and provide you with Time Crystals,<br>which can be used to purchase powerful upgrades<br><br>You will gain {TimeCrystalCount} Time Crystals:<br><br>"
                
            Base := big_number{Value:=PlayerManager.EssenceLevel()*5.0}
            GunMult := 1.0 + 0.01 * PlayerManager.TotalGunLevels()
            PeelyMult := 1.0 + 0.01 * PlayerManager.PeeliesFound()
            BossMult := MakeBigNumber(1.1, PlayerManager.TotalBossLevels())

            set ExplanationStr += "Player Level ({PlayerManager.EssenceLevel()}) x5:   {Base}"
            set ExplanationStr += "<br>Gun Levels ({PlayerManager.TotalGunLevels()}) x 1%:   x{FormatPercent(GunMult)}"
            set ExplanationStr += "<br>Peelies Collected ({PlayerManager.PeeliesFound()}) x 1%:   x{FormatPercent(BigNumber_ToFloat(PeelyMult))}" 
            if (BigNumber_GreaterThanOrEquals(BossMult, big_number{Value:=100.0})?):
                set ExplanationStr += "<br>Total Boss Tiers ({PlayerManager.TotalBossLevels()}) ^ 1.1:   x{BossMult}"
            else: 
                set ExplanationStr += "<br>Total Boss Tiers ({PlayerManager.TotalBossLevels()}) ^ 1.1:   x{FormatPercent(BigNumber_ToFloat(BossMult))}"

            Widget.RemoveWidget(TextBox)
            set TextBox = MakeTextBox(ExplanationStr)
            Widget.AddWidget(
                canvas_slot:
                    Widget := TextBox
                    Anchors := anchors{ Minimum := vector2{X := 0.5, Y := 0.45}, Maximum := vector2{X := 0.5, Y := 0.45}}
                    Alignment := vector2{X := 0.5, Y := 0.5}
                    ZOrder := 1
            )

    CreateMenu<override>():canvas=
        Result := canvas:
            Slots := array:
                canvas_slot:
                    Widget := DialogBackground
                    Anchors := anchors{ Minimum := vector2{X := 0.1, Y := 0.1}, Maximum := vector2{X := 0.9, Y := 0.9}}
                    Alignment := vector2{X := 0.5, Y := 0.5}
                    ZOrder := 0
                canvas_slot:
                    Widget := NoButton
                    Anchors := anchors{ Minimum := vector2{X := 0.3, Y := 0.8}, Maximum := vector2{X := 0.3, Y := 0.8}}
                    Alignment := vector2{X := 0.5, Y := 0.5}
                    ZOrder := 2
                canvas_slot:
                    Widget := YesButton
                    Anchors := anchors{ Minimum := vector2{X := 0.7, Y := 0.8}, Maximum := vector2{X := 0.7, Y := 0.8}}
                    Alignment := vector2{X := 0.5, Y := 0.5}
                    ZOrder := 3

        YesButton.SetText(STM("LET'S GO"))
        NoButton.SetText(STM("CANCEL"))

        YesButton.OnClick().Subscribe(OnYes)
        NoButton.OnClick().Subscribe(OnNo)

        return Result

    OnYes(Wdata:widget_message):void=
        if:
            SetAgent := Agent?
            PlayerManager := GameManager?.PlayerManagers[SetAgent]
        then:
            Teleporter.Teleport(SetAgent)
            PlayerManager.ClosePopupUI()
            spawn {GainTimeCrystals(PlayerManager)}

    OnNo(Wdata:widget_message):void=
        if (PlayerManager := GameManager?.PlayerManagers[Agent?]):
            PlayerManager.ClosePopupUI()

    GainTimeCrystals(PlayerManager:player_manager)<suspends>:void=
        Sleep(0.2)
        TimeCrystalCount := PlayerManager.CalculatePrestigeGain()
        PlayerManager.GainCollectible(collectible_data{Key := "P_PrestigeGem", Count := TimeCrystalCount})
        PlayerManager.SetLocationAndRotation(false)

prestige_final_confirm_popup := class(menu_template):
    DialogBackground : texture_block = texture_block{ DefaultImage := Images.DialogBox, DefaultDesiredSize := vector2{X := 1920.0, Y := 1080.0}} 
    var TextBox : stack_box = stack_box{Orientation := orientation.Vertical}

    var NoButton : button_regular = button_regular:

    var YesButton : button_loud = button_loud:
    
    UpdateMenu<override>(Widget:canvas):void=
        if:
            GM := GameManager?
            SetAgent := Agent?
            PlayerManager := GM.PlayerManagers[SetAgent]
            Widget = _Widget?
        then:
            TimeCrystalCount := PlayerManager.GetCollectibleCount("P_PrestigeGem")

            Widget.RemoveWidget(TextBox)
            set TextBox = MakeTextBox("Starchild is about to send you back to the beginning of the game.<br>Make sure you have spent your Time Crystals before you continue.<br><br>You have {TimeCrystalCount} Time Crystals left.")
            Widget.AddWidget(
                canvas_slot:
                    Widget := TextBox
                    Anchors := anchors{ Minimum := vector2{X := 0.5, Y := 0.5}, Maximum := vector2{X := 0.5, Y := 0.5}}
                    Alignment := vector2{X := 0.5, Y := 0.5}
                    ZOrder := 1
            )

    CreateMenu<override>():canvas=
        Result := canvas:
            Slots := array:
                canvas_slot:
                    Widget := DialogBackground
                    Anchors := anchors{ Minimum := vector2{X := 0.1, Y := 0.1}, Maximum := vector2{X := 0.9, Y := 0.9}}
                    Alignment := vector2{X := 0.5, Y := 0.5}
                    ZOrder := 0
                canvas_slot:
                    Widget := NoButton
                    Anchors := anchors{ Minimum := vector2{X := 0.3, Y := 0.8}, Maximum := vector2{X := 0.3, Y := 0.8}}
                    Alignment := vector2{X := 0.5, Y := 0.5}
                    ZOrder := 2
                canvas_slot:
                    Widget := YesButton
                    Anchors := anchors{ Minimum := vector2{X := 0.7, Y := 0.8}, Maximum := vector2{X := 0.7, Y := 0.8}}
                    Alignment := vector2{X := 0.5, Y := 0.5}
                    ZOrder := 3

        YesButton.SetText(STM("REBIRTH"))
        NoButton.SetText(STM("CANCEL"))

        YesButton.OnClick().Subscribe(OnYes)
        NoButton.OnClick().Subscribe(OnNo)

        return Result

    OnYes(Wdata:widget_message):void=
        if:
            SetAgent := Agent?
            PlayerManager := GameManager?.PlayerManagers[SetAgent]
        then:
            PlayerManager.DoPrestigeReset()
            PlayerManager.ClosePopupUI()

    OnNo(Wdata:widget_message):void=
        if (PlayerManager := GameManager?.PlayerManagers[Agent?]):
            PlayerManager.ClosePopupUI()

        
prestige_menu := class(menu_template):
    DialogBackground : texture_block = texture_block{ DefaultImage := Images.DialogBox, DefaultDesiredSize := vector2{X := 1920.0, Y := 1080.0}} 
    
    CloseButton :button_loud = button_loud:

    var UpgradeItems : []prestige_upgrade_item = array{}

    PrestigeGemsText : text_block = text_block:
        DefaultTextColor:=NamedColors.Gold
        DefaultShadowColor:=NamedColors.RoyalBlue
        DefaultShadowOffset:=option{vector2{X:=2.0,Y:=2.0}}
        DefaultShadowOpacity:= 1.0

    UpdateGems() :void=
        if:
            GM := GameManager?
            SetAgent := Agent?
            PlayerManager := GM.PlayerManagers[SetAgent]
        then:
            PrestigeGemsText.SetText(STM("{PlayerManager.GetCollectibleCount("P_PrestigeGem")}"))

    UpdateMenu<override>(Widget:canvas):void=
        if:
            GM := GameManager?
            SetAgent := Agent?
            PlayerManager := GM.PlayerManagers[SetAgent]
            Widget = _Widget?
        then:
            UpdateGems()

    CreateMenu<override>():canvas=
        StackBox := stack_box{Orientation := orientation.Vertical}

        Result := canvas:
            Slots := array:
                canvas_slot:
                    Widget := DialogBackground
                    Anchors := anchors{ Minimum := vector2{X := 0.1, Y := 0.1}, Maximum := vector2{X := 0.9, Y := 0.9}}
                    Alignment := vector2{X := 0.5, Y := 0.5}
                    ZOrder := 0
                canvas_slot:
                    Widget := StackBox
                    Anchors := anchors{ Minimum := vector2{X := 0.15, Y := 0.15}, Maximum := vector2{X := 0.85, Y := 0.7}}  
                    Alignment := vector2{X := 0.0, Y := 0.0}
                    Offsets := margin{Top := 0.0, Left := 0.0, Right := 0.0, Bottom := 0.0}
                    ZOrder := 1
                canvas_slot:
                    Widget := CloseButton
                    Anchors := anchors{ Minimum := vector2{X := 0.15, Y := 0.745}, Maximum := vector2{X := 0.15, Y := 0.745}}
                    Alignment := vector2{X := 0.0, Y := 0.0}
                    ZOrder := 2
                canvas_slot:
                    Widget := texture_block { DefaultImage := Images.PrestigeGem, DefaultDesiredSize := vector2{X := 160.0, Y := 160.0}}  
                    Anchors := anchors{ Minimum := vector2{X := 0.85, Y := 0.815}, Maximum := vector2{X := 0.85, Y := 0.745}}
                    Alignment := vector2{X := 1.0, Y := 0.5}
                    ZOrder := 3
                canvas_slot:
                    Widget := PrestigeGemsText
                    Anchors := anchors{ Minimum := vector2{X := 0.8, Y := 0.805}, Maximum := vector2{X := 0.80, Y := 0.735}}
                    Offsets := margin{Top := 0.0, Left := -200.0, Right := 0.0, Bottom := 0.0}
                    Alignment := vector2{X := 0.5, Y := 0.5}
                    ZOrder := 4

        if:
            GM := GameManager?
            SetAgent := Agent?
            PlayerManager := GM.PlayerManagers[SetAgent]
        then:
            var i : int = 0
            for (Upgrade : PrestigeUpgradesList):
                UpgradeItem := prestige_upgrade_item{PlayerManager := PlayerManager, Menu := Self}
                UpgradeItem.Create()
                UpgradeItem.Update(Upgrade)
                set UpgradeItems = UpgradeItems + array{UpgradeItem}

                var TopPadding : float = 0.0
                case (Upgrade):
                    prestige_upgrade.BonusStartingGems => set TopPadding = 10.0
                    prestige_upgrade.AutoClearQuest => set TopPadding = 10.0
                    prestige_upgrade.BonusItemLevelDrops  => set TopPadding = 10.0
                    prestige_upgrade.GunDuration => set TopPadding = 10.0
                    prestige_upgrade.ReducedUpgradeLevelRequirement => set TopPadding = 10.0
                    prestige_upgrade.DamageMelee => set TopPadding = 10.0
                    _   => block:

                StackBox.AddWidget(stack_box_slot:
                    Widget := UpgradeItem.Widget
                    HorizontalAlignment := horizontal_alignment.Fill
                    Padding:= margin{Top := TopPadding, Left := 0.0, Right := 0.0, Bottom := -17.0}
                )
                set i = i + 1

            CloseButton.SetText(STM("CLOSE"))
            CloseButton.OnClick().Subscribe(PlayerManager.ClosePopup)

        return Result

prestige_upgrade_menu := class(menu_template):
    DialogBackground : texture_block = texture_block{ DefaultImage := Images.DialogBox, DefaultDesiredSize := vector2{X := 1920.0, Y := 1080.0}} 
    
    
    UpdateMenu<override>(Widget:canvas):void=
        if:
            GM := GameManager?
            SetAgent := Agent?
            PlayerManager := GM.PlayerManagers[SetAgent]
            Widget = _Widget?
        then:
            block:

    CreateMenu<override>():canvas=
        Result := canvas:

        return Result

prestige_upgrade_item := class():
    Menu : prestige_menu
    PlayerManager : player_manager

    var Widget : canvas = canvas:

    var Upgrade : prestige_upgrade = prestige_upgrade.None

    NameText : text_block = text_block:
        DefaultTextColor:=NamedColors.White
        DefaultShadowColor:=NamedColors.Black   
        DefaultShadowOffset:=option{vector2{X:=2.0,Y:=2.0}}
        DefaultShadowOpacity:= 1.0
    ValueText : text_block = text_block:
        DefaultTextColor:=NamedColors.White
        DefaultShadowColor:=NamedColors.Black   
        DefaultShadowOffset:=option{vector2{X:=2.0,Y:=2.0}}
        DefaultShadowOpacity:= 1.0
    CostText : text_block = text_block:
        DefaultTextColor:=NamedColors.RoyalBlue
        DefaultShadowColor:=NamedColors.White
        DefaultShadowOffset:=option{vector2{X:=2.0,Y:=2.0}}
        DefaultShadowOpacity:= 1.0
    Icon : texture_block = texture_block:
        DefaultImage := Images.PrestigeGem
        #DefaultDesiredSize := vector2{X := 40.0, Y := 40.0}  
    Background : color_block = color_block:
        DefaultColor:=NamedColors.Gold
        DefaultOpacity:=0.4
    BuyButton : button_quiet = button_quiet:

    Create():void=

        StackBoxL := stack_box{Orientation:=orientation.Horizontal}
        StackBoxR := stack_box{Orientation:=orientation.Horizontal}


        Widget.AddWidget(
            canvas_slot:
                Widget := Background
                Anchors := anchors{ Minimum := vector2{X := 0.0, Y := 0.0}, Maximum := vector2{X := 1.0, Y := 0.5}}  
                Offsets := margin{Top := -2.0, Left := -2.0, Right := -2.0, Bottom := -2.0}
                Alignment := vector2{X := 1.0, Y := 1.0}
                ZOrder := 0
        )

        Widget.AddWidget(
            canvas_slot:
                Widget := StackBoxL
                Anchors := anchors{ Minimum := vector2{X := 0.0, Y := 0.0}, Maximum := vector2{X := 1.0, Y := 0.5}}  
                Alignment := vector2{X := 0.0, Y := 0.5}
                Offsets := margin{Top := -5.0, Left := 0.0, Right := 0.0, Bottom := -5.0}
                ZOrder := 1
        )

        Widget.AddWidget(
            canvas_slot:
                Widget := StackBoxR
                Anchors := anchors{ Minimum := vector2{X := 1.0, Y := 0.0}, Maximum := vector2{X := 1.0, Y := 0.5}}  
                Alignment := vector2{X := 1.0, Y := 1.0}
                Offsets := margin{Top := 0.0, Left := 0.0, Right := 0.0, Bottom := 0.0}
                ZOrder := 2
        )
        
        StackBoxL.AddWidget(
            stack_box_slot:
                Widget := BuyButton
                HorizontalAlignment := horizontal_alignment.Left
                VerticalAlignment := vertical_alignment.Bottom
                Padding:= margin{Top := 0.0, Left := 0.0, Right := 3.0, Bottom := 0.0}
        )
        StackBoxL.AddWidget(
            stack_box_slot:
                Widget := Icon
                HorizontalAlignment := horizontal_alignment.Left
                VerticalAlignment := vertical_alignment.Bottom
        )
        StackBoxL.AddWidget(
            stack_box_slot:
                Widget := CostText
                HorizontalAlignment := horizontal_alignment.Left
                VerticalAlignment := vertical_alignment.Bottom
                Padding:= margin{Top := 0.0, Left := 0.0, Right := 10.0, Bottom := 0.0}
        )
        StackBoxL.AddWidget(
            stack_box_slot:
                Widget := NameText
                HorizontalAlignment := horizontal_alignment.Left
                VerticalAlignment := vertical_alignment.Bottom
                Padding:= margin{Top := 0.0, Left := 10.0, Right := 10.0, Bottom := 0.0}
        )
        StackBoxR.AddWidget(
            stack_box_slot:
                Widget := ValueText
                HorizontalAlignment := horizontal_alignment.Right
                VerticalAlignment := vertical_alignment.Bottom
                Padding:= margin{Top := 0.0, Left := 10.0, Right := 10.0, Bottom := 0.0}
        )

        BuyButton.OnClick().Subscribe(OnBuy)

    Update(UpgradeIn : prestige_upgrade):void=
        set Upgrade = UpgradeIn

        Key := Upgrade.GetKey()
        Level := (PlayerManager.GetUpgradeLevel[Key] + 1) or 1 #looking at the next level
        Cost := Upgrade.GetCost(Level)

        var ValueDesc : string = Upgrade.GetValueDescription(Level-1)
        if (not ValueDesc = " "):
            set ValueDesc = "Current: {ValueDesc}" 

        ValueText.SetText(STM(ValueDesc))
        NameText.SetText(STM(Upgrade.GetDescription(Level)))
        CostText.SetText(STM("{Cost}"))

        if (BigNumber_GreaterThanOrEquals(PlayerManager.GetCollectibleCount("P_PrestigeGem"), Cost)?):
            BuyButton.SetText(STM("BUY"))
        else:
            BuyButton.SetText(STM("   "))


    OnBuy(Wdata:widget_message):void=
        Key := Upgrade.GetKey()
        Level := (PlayerManager.GetUpgradeLevel[Key] + 1) or 1
        Cost := Upgrade.GetCost(Level)
        if (BigNumber_GreaterThanOrEquals(PlayerManager.GetCollectibleCount("P_PrestigeGem"), Cost)?):
            PlayerManager.GainUpgradeLevel(Key, 1)
            PlayerManager.GainCollectible(collectible_data{Key := "P_PrestigeGem", Count := Cost * -1.0})

        Update(Upgrade)
        Menu.UpdateGems()
        
