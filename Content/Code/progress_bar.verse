using { /UnrealEngine.com/Temporary/SpatialMath }
using { /UnrealEngine.com/Temporary/UI }
using { /Verse.org/Colors }
using { /Verse.org/Simulation}

MakeProgressBar(InProgress:float,InSize:vector2,InFrontColor:color,InBackColor:color,InProgressColor:color, ?InMaxColor:?color=false, ?InLimit:int=20):progress_bar=
    return progress_bar{Progress:=InProgress, Size:=InSize, FrontColor:=InFrontColor, BackColor:=InBackColor, ProgressColor:=InProgressColor, MaxColor := InMaxColor, Limit:= InLimit }

progress_bar:=class:
    var LastProgress : float = 0.0
    var Progress : float = 0.0
    Orientation: tuple(horizontal_alignment, vertical_alignment) = (horizontal_alignment.Left, vertical_alignment.Fill)
    var Size : vector2 = vector2{}
    var FrontColor : color = color{}
    var ProgressColor : color = color{}
    var BackColor : color = color{}
    var MaxColor : ?color = false
    var Limit : int = 20


    var CoverColorBlock: color_block = color_block{}
    var FrontColorBlock: color_block = color_block{}
    var ProgressColorBlock: color_block = color_block{}
    var BackColorBlock: color_block = color_block{}
    var ID : int = 0

    IconBounce : float = 1.25
    GatherBounceIterations : float = 8.0

    GetWidget():overlay=
        set FrontColorBlock = color_block{DefaultColor:= FrontColor, DefaultDesiredSize:= vector2{X:=0.0, Y:=0.0}, DefaultOpacity:= 0.0}
        set ProgressColorBlock = color_block{DefaultColor:= ProgressColor, DefaultDesiredSize:= vector2{X:=0.0, Y:=0.0}, DefaultOpacity:= 0.0}
        set BackColorBlock = color_block{DefaultColor:= BackColor, DefaultDesiredSize:= vector2{X:=0.0, Y:=0.0}, DefaultOpacity:= 0.0}
        set CoverColorBlock = color_block{DefaultColor:= BackColor, DefaultDesiredSize:= vector2{X:=0.0, Y:=0.0}, DefaultOpacity:= 0.0}
                
        return overlay:
            Slots:=array:
                overlay_slot:
                    Widget:=BackColorBlock
                    HorizontalAlignment:=horizontal_alignment.Fill
                    VerticalAlignment:=vertical_alignment.Fill
                overlay_slot:
                    Widget:=ProgressColorBlock
                    HorizontalAlignment:= Orientation(0)
                    VerticalAlignment:= Orientation(1)
                overlay_slot:
                    Widget:=FrontColorBlock
                    HorizontalAlignment:= Orientation(0)
                    VerticalAlignment:= Orientation(1)    
                overlay_slot:
                    Widget:= CoverColorBlock
                    HorizontalAlignment:= Orientation(0)
                    VerticalAlignment:= vertical_alignment.Bottom

    PrepareProgressForNextBar(InProgress:float):void= 
        set Progress = InProgress  

    SetInvisible():void=
        FrontColorBlock.SetOpacity(0.0)
        FrontColorBlock.SetDesiredSize(vector2{})
        ProgressColorBlock.SetOpacity(0.0)
        ProgressColorBlock.SetDesiredSize(vector2{})
        BackColorBlock.SetOpacity(0.0)
        BackColorBlock.SetDesiredSize(vector2{X:=0.01, Y:=0.01})
        CoverColorBlock.SetOpacity(0.0)
        CoverColorBlock.SetDesiredSize(vector2{})

    SetVisible():void=
        FrontColorBlock.SetOpacity(0.8)
        ProgressColorBlock.SetOpacity(0.8)
        BackColorBlock.SetOpacity(0.5)
        BackColorBlock.SetDesiredSize(Size)
        CoverColorBlock.SetOpacity(0.5)
        CoverColorBlock.SetDesiredSize(vector2{X:= 0.0, Y:=Size.Y * 0.45})
        
    SetProgressWhen(InProgress:float, DoUpdateProgress:logic, When : float)<suspends>:void=
        limit:int=120
        for(X:=1..limit):

            Sleep(0.01)

            Temp:= ProgressColorBlock.GetDesiredSize().X
            Temp2:= FrontColorBlock.GetDesiredSize().X

            if:
                Temp = When
                Temp2 = When
            then:
                SetProgress(InProgress, DoUpdateProgress)

    SetProgress(InProgress:float, DoUpdateProgress:logic, ?IsNewTarget : logic = false, ?DoColorUpdate : logic = true):void=
        SetVisible()
        set LastProgress = Progress #Store Most Recent Progress  
        set Progress = InProgress #
        set ID += 1

        if:
            DoColorUpdate?
        then:
            if:
                InProgress = 1.0
                SetMaxColor := MaxColor?
            then:
                FrontColorBlock.SetColor(SetMaxColor)
            else:
                FrontColorBlock.SetColor(FrontColor)

        var Delta:float = Progress - LastProgress 
        if:
            IsNewTarget?
        then:
            if(DoUpdateProgress?):
                spawn:
                    UpdateProgress(LastProgress, Delta, IsNewTarget, ID)
        else:
            if(Delta > 0.0):
                LowerNum := FrontColorBlock.GetDesiredSize().X / Size.X
                set Delta = Progress - LowerNum 
                if(DoUpdateProgress?):
                    spawn:
                        UpdateProgress(LowerNum, Delta, IsNewTarget, ID)
            else:
                HigherNum := ProgressColorBlock.GetDesiredSize().X / Size.X
                set Delta = Progress - HigherNum 
                if(DoUpdateProgress?):
                    spawn:
                        UpdateProgress(HigherNum, Delta, IsNewTarget, ID)

    GetProgress()<transacts>:float=
        return Progress
        
    BeginBounceEffect(?BounceSize : float = IconBounce):void=
        BackColorBlock.SetDesiredSize(vector2{X:= Size.X * BounceSize, Y:= Size.Y * BounceSize})
        spawn:
            IterateBounceEffect()        

    IterateBounceEffect()<suspends>:void =
        limit:int=12

        for(X:=1..limit):
            Sleep(0.01)
            CurrentX := BackColorBlock.GetDesiredSize().X 
            if(CurrentX > Size.X):
                CurrentY := BackColorBlock.GetDesiredSize().Y
                ShrinkX := Max(CurrentX * ( IconBounce - 1.0 ) / GatherBounceIterations, 0.25)
                ShrinkY := Max(CurrentY * ( IconBounce - 1.0 ) / GatherBounceIterations, 0.25)
                BackColorBlock.SetDesiredSize(vector2{X := Max(CurrentX - ShrinkX, Size.X), Y:= Max(CurrentY - ShrinkY, Size.Y)})
        
    UpdateProgress(LastProgressSnapshot: float, InDelta:float, IsNewTarget : logic, IDIn : int)<suspends>:void=
        if(InDelta > 0.0):
            if(IsNewTarget?):
                if(IDIn = ID):
                    CoverColorBlock.SetDesiredSize(vector2{X:=Size.X*Progress, Y:=Size.Y * 0.45})
                    ProgressColorBlock.SetDesiredSize(vector2{X:=Size.X*Progress})
                    # if(Size.X = 650.0):
                    #     Log("A ID{IDIn} C{Size.X*Progress} F{} P{Size.X*Progress}")
            else:
                if(IDIn = ID):
                    FrontColorBlock.SetDesiredSize(vector2{X:=Size.X*LastProgressSnapshot})
                    CoverColorBlock.SetDesiredSize(vector2{X:=Size.X*Progress,  Y:=Size.Y * 0.45})
                    ProgressColorBlock.SetDesiredSize(vector2{X:=Size.X*Progress})
                    # if(Size.X = 650.0):
                    #     Log("B ID{IDIn} C{Size.X*Progress} F{Size.X*Progress} P{Size.X*LastProgress}")
        else:
            if(IsNewTarget?):
                if(IDIn = ID):
                    FrontColorBlock.SetDesiredSize(vector2{X:=Size.X*Progress})
                    # if(Size.X = 650.0):
                    #     Log("C ID{IDIn} C{} F{Size.X*Progress} P{}")
            else:
                if(IDIn = ID):
                    FrontColorBlock.SetDesiredSize(vector2{X:=Size.X*Progress})
                    CoverColorBlock.SetDesiredSize(vector2{X:=Size.X*LastProgressSnapshot, Y:=Size.Y * 0.45})
                    ProgressColorBlock.SetDesiredSize(vector2{X:=Size.X*LastProgressSnapshot})
                    # if(Size.X = 650.0):
                    #     Log("D ID{IDIn} C{Size.X*LastProgress} F{Size.X*Progress} P{Size.X*LastProgress}")
        
        for(X:=1..Limit):
            if(not IDIn = ID):
                return
            Sleep(0.01)
            if(not IDIn = ID):
                return
            if(InDelta>0.0):
                FrontColorBlock.SetDesiredSize(vector2{X:= Size.X *(LastProgressSnapshot + (InDelta * X/ (Limit * 1.0)))})
                # if(Size.X = 650.0):
                #     Log("1 ID{IDIn} C{} F{Size.X *(LastProgressSnapshot + (InDelta * X/ (limit * 1.0)))} P{}")
            else:
                CoverColorBlock.SetDesiredSize(vector2{X:= Size.X *(LastProgressSnapshot + (InDelta * X/(Limit * 1.0))), Y:=Size.Y * 0.45})
                ProgressColorBlock.SetDesiredSize(vector2{X:= Size.X *(LastProgressSnapshot + (InDelta * X/(Limit * 1.0)))})            
                # if(Size.X = 650.0):
                #     Log("2 ID{IDIn} C{Size.X *(LastProgressSnapshot + (InDelta * X/ (limit * 1.0)))} F{} P{Size.X *(LastProgressSnapshot + (InDelta * X/ (limit * 1.0)))}")
                

            