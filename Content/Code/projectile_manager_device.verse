
using { /Fortnite.com/Devices }
using { /Verse.org/Simulation }
using { /UnrealEngine.com/Temporary/Diagnostics }
using { /UnrealEngine.com/Temporary/SpatialMath }
using { /Fortnite.com/Characters }
using { /Verse.org/Assets }
using { /UnrealEngine.com/Assets }

SpawnProjectile(GameManager : game_manager, Prop : creative_prop_asset, Position: vector3, Radius : float, Velocity : vector3, Lifetime : float, Damage : float, ?Explosion : ?particle_system = false, ?Scale : vector3 = vector3{X:=1.0, Y:=1.0, Z:=1.0}, ?Rotation : rotation = IdentityRotation())<suspends>:?projectile= 
    
    if:
        PropInstance := SpawnProp(Prop, transform{Translation:=Position, Rotation:=Rotation, Scale := Scale})(0)?
    then:
    
        var NewProjectile : projectile = projectile{
            GameManager := GameManager
            Prop := PropInstance
            Radius := Radius
            Velocity := Velocity
            Lifetime := Lifetime
            Damage := Damage
            Explosion := Explosion
        }

        NewProjectile.Start()
        return option{NewProjectile}
    else:
        return false

projectile := class:
    GameManager : game_manager

    Prop : creative_prop
    Radius : float
    Velocity : vector3
    Lifetime : float
    Damage : float
    Explosion : ?particle_system = false

    TickDistanceBuffer : float = 600.0
    MinTickPeriod : float = 0.01
    MaxTickPeriod : float = 1.0

    var DestroyAfterMovement : logic = true
    var DestroyOnCollision : logic = true
    var RepeatHitTime : float = 1.0

    var HitPlayers : [player]float = map{}

    Start():void=
        spawn { Tick() }

    Tick()<suspends>:void=
        race:
            HandleMovement()
            HandleCollision()

        if (Prop.IsValid[]):
            Prop.Dispose()

    HandleMovement()<suspends>:void=
        if (Prop.IsValid[]):
        Prop.MoveTo(Prop.GetTransform().Translation + Velocity * Lifetime, Prop.GetTransform().Rotation, Lifetime)
        loop:
            if (DestroyAfterMovement?):
                break
            Sleep(0.1)
        

    HandleCollision()<suspends>:void=
        loop:
            if (Prop.IsValid[]):

                var MinDist : float = 999999999.0

                for (Player : GameManager.PlayerManagers):
                    if:
                        Player.Player.IsActive[]
                        FortChar := Player.Player.GetFortCharacter[]
                        Dist := Distance(Prop.GetTransform().Translation, FortChar.GetTransform().Translation + vector3{X:=0.0, Y:=0.0, Z:= 100.0})
                        not HitPlayers[Player.Player] > 0.0
                    then:
                        if (Dist < MinDist):
                            set MinDist = Dist

                        if (Dist < Radius):
                            FortChar.Damage(Damage)
                            if (Expl := Explosion?):
                                SpawnParticleSystem(Expl, Prop.GetTransform().Translation, ?Rotation := Prop.GetTransform().Rotation)
                            Player.GameManager.AudioManager.PlayHandcannon(Player.Agent)
                            if (DestroyOnCollision?):
                                return
                            else if:
                                set HitPlayers[Player.Player] = RepeatHitTime
                
                SleepTime := Clamp(MinDist / (TickDistanceBuffer + Velocity.Length()), MinTickPeriod, MaxTickPeriod)
                Sleep(SleepTime)

                if (not DestroyOnCollision?):
                    for (K -> V : HitPlayers):
                        if:
                            set HitPlayers[K] -= SleepTime 
            else:
                break