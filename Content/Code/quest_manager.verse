using { /Fortnite.com/Characters }
using { /Fortnite.com/Devices }
using { /Verse.org/Simulation }

quest_stats := struct <persistable><concrete><public>:
    @editable
    QuestID <public>: int = 0
    
    @editable
    TaskID <public>: int = -1

    @editable
    Progress <public>: big_number = big_number{}

    @editable
    SubTasks <public>: []subtask_stats = array{}

subtask_stats := struct <persistable><concrete><public>:
    @editable
    SubTaskID <public>: int = 0

    @editable
    SubTaskProgress <public>: big_number = big_number{}

tracked_quest := class <concrete><public>: 
    var TaskEvent : task_event = task_event.None
    var QuestStats : quest_stats = quest_stats{}
    var SubscribedEvents : []cancelable = array{}

MakeTrackedQuest<constructor>(Src:tracked_quest)<transacts> := tracked_quest:
    TaskEvent := Src.TaskEvent
    QuestStats := Src.QuestStats
    SubscribedEvents := Src.SubscribedEvents

quest_manager:=class<unique>:
    QuestModel:quest_model_device
    Agent:agent
    Player:player
    PlayerManager : player_manager
    var ActiveTrackedQuests : []tracked_quest = array{}
    
    Init():void=
        if:
            CurrentQuests:=PlayerManager.GetCurrentQuests[]
        then:
            for (X := 0..CurrentQuests.Length - 1):
                if:
                    QuestStats := CurrentQuests[X]
                then:
                    spawn:
                        TurnOnQuestTrackers(QuestStats, ?IsGameStart := true)
                    SubscribeToTask(QuestStats)
                    UpdateQuestUI(QuestStats, ?IsNewQuest:=true)
        
        CheckPersistentQuests()

    BeginStartingQuests():void=
        var QuestID : int = 0
        for (Quest : QuestModel.Quests):
            if: 
                Quest.IsStartingQuest?
            then:
                StartQuest(QuestID, ?DoesSkipCompleteFlair := true)
            set QuestID += 1

    StartQuest(QuestID : int, ?DoesSkipCompleteFlair:logic = false):void=
        if:
            IsQuestValid(QuestID)?
            Quest:= QuestModel.Quests[QuestID]
            QuestStats:= PlayerManager.AddQuest[QuestID, Quest.StartingTask]
        then:
            StartTask(QuestStats, ?DoesSkipCompleteFlair := DoesSkipCompleteFlair)
            if (PlayerManager.ShouldSkipQuestGroup(QuestModel.Quests[QuestStats.QuestID].Tasks[QuestStats.TaskID].GroupKey)?):
                OnValuesUpdated(QuestStats)

    StartQuestSecondaryTask(QuestID : int, TaskID : int):void=
        if:
            QuestStats:= PlayerManager.AddQuest[QuestID, TaskID]
        then:
            StartTask(QuestStats, ?DoesSkipCompleteFlair := true)

    StartTask(QuestStats : quest_stats, ?DoesSkipCompleteFlair:logic = false):void=

        Print("StartTask for {QuestStats.TaskID} of Quest {QuestStats.QuestID}")

        spawn:
            TurnOnQuestTrackers(QuestStats)
        SubscribeToTask(QuestStats)

        if:
            Task:= QuestModel.Quests[QuestStats.QuestID].Tasks[QuestStats.TaskID]
            Task.CompoundWithNext?
        then:
            StartQuestSecondaryTask(QuestStats.QuestID, QuestStats.TaskID + 1)
        else:
            UpdateQuestUI(GetRelevantCompoundTasks(QuestStats), ?IsNewQuest:=true, ?DoesSkipCompleteFlair := DoesSkipCompleteFlair)
            CheckPersistentQuests()
            PlayerManager.PlayMusic()

    CheckPersistentQuests():void=
        if:
            CurrentEssence := PlayerManager.GetStashedEssence[]
            CurrentPlayerLevel := PlayerManager.GetLevel[]
            CumulativeUpgrades := PlayerManager.GetCumulativeUpgradeLevels[]
        then:
            PlayerManager.UpdateQuestProgressForTask(task_event.FillStash, CurrentEssence, false, ?IsPersistent := true)
            PlayerManager.UpdateQuestProgressForTask(task_event.PlayerLevel, big_number{Value:=CurrentPlayerLevel * 1.0}, false, ?IsPersistent := true)

            for (Zone : PlayerManager.GameManager.ZoneManagers):
                ZoneTag := Zone.ZoneTag
                if (CurrentZoneLevel := PlayerManager.GetZoneLevel[ZoneTag]):
                    PlayerManager.UpdateQuestProgressForTask(task_event.ZoneLevel, big_number{Value:= CurrentZoneLevel * 1.0}, option{ZoneTag}, ?IsPersistent := true)
                
                if (CurrentZoneClears := PlayerManager.GetZoneClears[ZoneTag]):
                    PlayerManager.UpdateQuestProgressForTask(task_event.ZoneClear, big_number{Value:= CurrentZoneClears * 1.0}, option{ZoneTag}, ?IsPersistent := true)
            
            for (Key -> Value : PlayerProfileDataMap[Player].UpgradeLevels):
                PlayerManager.UpdateQuestProgressForTask(task_event.CollectUpgrades, big_number{Value:=Value * 1.0}, false, ?Target := Key, ?IsPersistent := true) 

            for (Key -> Value : PlayerProfileDataMap[Player].Collectibles):
                PlayerManager.UpdateQuestProgressForTask(task_event.GetCollectibleCount, Value.Count, false, ?Target := Value.Key, ?IsPersistent := true)

            PlayerManager.UpdateQuestProgressForTask(task_event.CumulativeUpgrades, big_number{Value := CumulativeUpgrades * 1.0}, false, ?IsPersistent := true)                
                
        #Check completed quests and hide objects / grant achievements
        if:
            CompletedQuests := PlayerManager.GetCompletedQuests[]
        then:
            for (QuestID : CompletedQuests):
                if:
                    Quest := QuestModel.Quests[QuestID]
                then:
                    for (Task : Quest.Tasks):
                        for (Achievement : Task.CompleteAchievements):
                            


                        Task.MapIndicator.DeactivateObjectivePulse(Agent)

                        if(HideObjects := Task.HideObjects?):
                            for (Hide : HideObjects):
                                Hide.AddToIgnoreList(Agent)

                        if(HideObjects := Task.HideObjectsAlt?):
                            for (Hide : HideObjects):
                                # Print("Hiding AltObject for {QuestID} Task ? -- Prior Quest")
                                PlayerManager.GameManager.QueuePeely(queued_peely_doll{Peely := Hide, Agent := option{PlayerManager.Agent}, NewState := false}, ?SetState := true)
                            

        #Check completed tasks in current quests and hide objects
        if:
            CurrentQuests := PlayerManager.GetCurrentQuests[]
        then:
            for (i := 0..CurrentQuests.Length - 1):
                if:
                    QuestID := CurrentQuests[i].QuestID
                    TaskID := CurrentQuests[i].TaskID
                    Quest := QuestModel.Quests[QuestID]
                then:
                    if:
                        PrevQuest := Quest.Tasks[TaskID - 1]
                        PrevQuest.CompoundWithNext?
                    then:
                        block:
                    else:
                        for (j := 0..TaskID - 1):
                            if:
                                Task := Quest.Tasks[j]
                            then:
                                Task.MapIndicator.DeactivateObjectivePulse(Agent)
                                if:
                                    HideObjects := Task.HideObjects?
                                then:
                                    for (Hide : HideObjects):                                
                                        Hide.AddToIgnoreList(Agent)
                            if:
                                Task := Quest.Tasks[j]
                                HideObjects := Task.HideObjectsAlt?
                            then:
                                for (Hide : HideObjects):
                                    # Print("Hiding AltObject for {QuestID} {j} -- Prior Task")
                                    PlayerManager.GameManager.QueuePeely(queued_peely_doll{Peely := Hide, Agent := option{PlayerManager.Agent}, NewState := false}, ?SetState := true)

        #Find current quest and make sure trackers are on
        if:
            CurrentQuests := PlayerManager.GetCurrentQuests[]
        then:
            for (i := 0..CurrentQuests.Length - 1):
                if:
                    QuestID := CurrentQuests[i].QuestID
                    TaskID := CurrentQuests[i].TaskID
                    Quest := QuestModel.Quests[QuestID]
                    Task := Quest.Tasks[TaskID]
                then:
                    NewQuestStats := quest_stats{QuestID:= QuestID, TaskID := TaskID}
                    spawn:
                        TurnOnQuestTrackers(NewQuestStats)
                    
                    if (PlayerManager.ShouldSkipQuestGroup(Task.GroupKey)?):
                        OnValuesUpdated(NewQuestStats)

    UnhideAll():void=
        for (Quest : QuestModel.Quests):
            for (Task : Quest.Tasks):
                if(HideObjects := Task.HideObjects?):
                    for (Hide : HideObjects):
                        Hide.RemoveFromIgnoreList(Agent)
                if:
                    HideObjects := Task.HideObjectsAlt?
                then:
                    for (Hide : HideObjects):
                        PlayerManager.GameManager.QueuePeely(queued_peely_doll{Peely := Hide, Agent := option{PlayerManager.Agent}, NewState := true})
    
                if:
                    SetShowObjects := Task.ShowObjectsAlt?
                then:
                    for (Show : SetShowObjects):
                        PlayerManager.GameManager.QueuePeely(queued_peely_doll{Peely := Show, Agent := option{PlayerManager.Agent}, NewState := false})
    
    
    OnValuesUpdated(QuestStats : quest_stats):void=
        if:
            Quest := QuestModel.Quests[QuestStats.QuestID]
            Task := Quest.Tasks[QuestStats.TaskID]
            TaskRequirement := (Task.Progress)
            TaskProgress := QuestStats.Progress
        then:

            if:
                Quest.IsShowTask?
            then:
                UpdateQuestUI(GetRelevantCompoundTasks(QuestStats))

            IsComplete := IsQuestComplete(QuestStats)
            if:
                IsComplete?
                not PlayerManager.HasCompletedTask(QuestStats.QuestID, QuestStats.TaskID)?
            then:
                CompleteTask(QuestStats)


    IsQuestComplete(QuestStats : quest_stats)<transacts>:logic=
        if:
            Quest := QuestModel.Quests[QuestStats.QuestID]
            Task := Quest.Tasks[QuestStats.TaskID]
            TaskRequirement := (Task.Progress)
            TaskProgress := QuestStats.Progress
        then:

            # Print("Checking QuestComplete for {QuestStats.TaskID}")
            # Print("===ACTIVE QUESTS===")
            # for (Q: PlayerManager.GetCurrentQuests[]):
            #     Print("QuestID: {Q.QuestID}, TaskID: {Q.TaskID}")
            # Print("===================")

            #if the quest is in a skip group, then it is complete
            if(PlayerManager.ShouldSkipQuestGroup(Task.GroupKey)?):
                return true

            #this quest is not complete
            if (not BigNumber_GreaterThanOrEquals(TaskProgress, TaskRequirement)?):
                Print("Not Complete: {QuestStats.TaskID} : {TaskProgress} / {TaskRequirement}")
                return false

            var WorkingTask : int = QuestStats.TaskID

            #a prior linked quest is not complete
            loop:
                set WorkingTask = WorkingTask - 1
                if:
                    PrevTask := Quest.Tasks[WorkingTask]
                    PrevTask.CompoundWithNext?
                    PrevTaskRequirement := (PrevTask.Progress)
                    PrevTaskProgress := PlayerManager.GetQuestStats[QuestStats.QuestID, WorkingTask].Progress
                then:
                    if:
                        not BigNumber_GreaterThanOrEquals(PrevTaskProgress, PrevTaskRequirement)?
                        not PlayerManager.HasCompletedTask(QuestStats.QuestID, WorkingTask)?
                    then:
                        Print("Not Complete: {WorkingTask}: {PrevTaskProgress} / {PrevTaskRequirement}")
                        return false
                else:
                    break
            
            set WorkingTask = QuestStats.TaskID

            #a future linked quest is not complete
            loop:

                if:
                    CheckTask := Quest.Tasks[WorkingTask]
                    CheckTask.CompoundWithNext?
                    set WorkingTask = WorkingTask + 1
                    NextTaskRequirement := (Quest.Tasks[WorkingTask].Progress)
                    NextTaskProgress := PlayerManager.GetQuestStats[QuestStats.QuestID, WorkingTask].Progress
                then:
                    if:
                        not BigNumber_GreaterThanOrEquals(NextTaskProgress, NextTaskRequirement)?
                        not PlayerManager.HasCompletedTask(QuestStats.QuestID, WorkingTask)?
                    then:
                        Print("Not Complete: {WorkingTask} : {NextTaskProgress} / {NextTaskRequirement}")
                        return false
                else:
                    break

            return true
        return false

    GetRelevantCompoundTasks(QuestStats : quest_stats):[]quest_stats=
        var Result : []quest_stats = array{}

        if:
            Quest := QuestModel.Quests[QuestStats.QuestID]
        then:
            var CurrentTaskID : int = QuestStats.TaskID 

            loop:
                if:
                    Task := Quest.Tasks[CurrentTaskID-1]
                    Task.CompoundWithNext?
                then:
                    set CurrentTaskID = CurrentTaskID - 1
                else:
                    break

            loop:
                if:
                    AppendQuestStats := PlayerManager.GetQuestStats[QuestStats.QuestID, CurrentTaskID]
                then:
                    # Print("Appending Task {AppendQuestStats.TaskID} with Progress {AppendQuestStats.Progress}")
                    set Result = Result + array{AppendQuestStats}
                
                if:
                    Task := Quest.Tasks[CurrentTaskID]
                    Task.CompoundWithNext?
                then:
                    set CurrentTaskID = CurrentTaskID + 1
                else:
                    break
        
        return Result

    GetTrackedQuestStats(QuestID : int, TaskID : int)<transacts>:?quest_stats=
        for(ActiveTrackedQuest : ActiveTrackedQuests):
            if:
                ActiveTrackedQuest.QuestStats.QuestID = QuestID
                ActiveTrackedQuest.QuestStats.TaskID = TaskID
            then:
                return option{ActiveTrackedQuest.QuestStats}

        return false

    CheckAnalytics(QuestID : int, TaskID : int):void=
        if(QuestID = 0):
            case (TaskID):
                0 => PlayerManager.GameManager.AnalyticsManager.quest_complete_0_0.Submit(Agent)
                1 => PlayerManager.GameManager.AnalyticsManager.quest_complete_0_1.Submit(Agent)
                2 => PlayerManager.GameManager.AnalyticsManager.quest_complete_0_2.Submit(Agent)
                3 => PlayerManager.GameManager.AnalyticsManager.quest_complete_0_3.Submit(Agent)
                4 => PlayerManager.GameManager.AnalyticsManager.quest_complete_0_4.Submit(Agent)
                5 => PlayerManager.GameManager.AnalyticsManager.quest_complete_0_5.Submit(Agent)
                6 => PlayerManager.GameManager.AnalyticsManager.quest_complete_0_6.Submit(Agent)
                7 => PlayerManager.GameManager.AnalyticsManager.quest_complete_0_7.Submit(Agent)
                8 => PlayerManager.GameManager.AnalyticsManager.quest_complete_0_8.Submit(Agent)
                9 => PlayerManager.GameManager.AnalyticsManager.quest_complete_0_9.Submit(Agent)
                _ => block:
        
        if(QuestID = 1):
            case (TaskID):
                0 => PlayerManager.GameManager.AnalyticsManager.quest_complete_1_0.Submit(Agent)
                1 => PlayerManager.GameManager.AnalyticsManager.quest_complete_1_1.Submit(Agent)
                2 => PlayerManager.GameManager.AnalyticsManager.quest_complete_1_2.Submit(Agent)
                3 => PlayerManager.GameManager.AnalyticsManager.quest_complete_1_3.Submit(Agent)
                4 => PlayerManager.GameManager.AnalyticsManager.quest_complete_1_4.Submit(Agent)
                5 => PlayerManager.GameManager.AnalyticsManager.quest_complete_1_5.Submit(Agent)
                6 => PlayerManager.GameManager.AnalyticsManager.quest_complete_1_6.Submit(Agent)
                7 => PlayerManager.GameManager.AnalyticsManager.quest_complete_1_7.Submit(Agent)
                8 => PlayerManager.GameManager.AnalyticsManager.quest_complete_1_8.Submit(Agent)
                9 => PlayerManager.GameManager.AnalyticsManager.quest_complete_1_9.Submit(Agent)
                10 => PlayerManager.GameManager.AnalyticsManager.quest_complete_1_10.Submit(Agent)
                11 => PlayerManager.GameManager.AnalyticsManager.quest_complete_1_11.Submit(Agent)
                12 => PlayerManager.GameManager.AnalyticsManager.quest_complete_1_12.Submit(Agent)
                13 => PlayerManager.GameManager.AnalyticsManager.quest_complete_1_13.Submit(Agent)
                14 => PlayerManager.GameManager.AnalyticsManager.quest_complete_1_14.Submit(Agent)
                15 => PlayerManager.GameManager.AnalyticsManager.quest_complete_1_15.Submit(Agent)
                16 => PlayerManager.GameManager.AnalyticsManager.quest_complete_1_16.Submit(Agent)
                17 => PlayerManager.GameManager.AnalyticsManager.quest_complete_1_17.Submit(Agent)
                18 => PlayerManager.GameManager.AnalyticsManager.quest_complete_1_18.Submit(Agent)
                19 => PlayerManager.GameManager.AnalyticsManager.quest_complete_1_19.Submit(Agent)
                20 => PlayerManager.GameManager.AnalyticsManager.quest_complete_1_20.Submit(Agent)
                21 => PlayerManager.GameManager.AnalyticsManager.quest_complete_1_21.Submit(Agent)
                22 => PlayerManager.GameManager.AnalyticsManager.quest_complete_1_22.Submit(Agent)
                23 => PlayerManager.GameManager.AnalyticsManager.quest_complete_1_23.Submit(Agent)
                24 => PlayerManager.GameManager.AnalyticsManager.quest_complete_1_24.Submit(Agent)
                25 => PlayerManager.GameManager.AnalyticsManager.quest_complete_1_25.Submit(Agent)
                26 => PlayerManager.GameManager.AnalyticsManager.quest_complete_1_26.Submit(Agent)
                _ => block:

        if(QuestID = 2):
            case (TaskID):
                0 => PlayerManager.GameManager.AnalyticsManager.quest_complete_2_0.Submit(Agent)
                1 => PlayerManager.GameManager.AnalyticsManager.quest_complete_2_1.Submit(Agent)
                2 => PlayerManager.GameManager.AnalyticsManager.quest_complete_2_2.Submit(Agent)
                3 => PlayerManager.GameManager.AnalyticsManager.quest_complete_2_3.Submit(Agent)
                4 => PlayerManager.GameManager.AnalyticsManager.quest_complete_2_4.Submit(Agent)
                5 => PlayerManager.GameManager.AnalyticsManager.quest_complete_2_5.Submit(Agent)
                6 => PlayerManager.GameManager.AnalyticsManager.quest_complete_2_6.Submit(Agent)
                7 => PlayerManager.GameManager.AnalyticsManager.quest_complete_2_7.Submit(Agent)
                8 => PlayerManager.GameManager.AnalyticsManager.quest_complete_2_8.Submit(Agent)
                9 => PlayerManager.GameManager.AnalyticsManager.quest_complete_2_9.Submit(Agent)
                10 => PlayerManager.GameManager.AnalyticsManager.quest_complete_2_10.Submit(Agent)
                11 => PlayerManager.GameManager.AnalyticsManager.quest_complete_2_11.Submit(Agent)
                12 => PlayerManager.GameManager.AnalyticsManager.quest_complete_2_12.Submit(Agent)
                13 => PlayerManager.GameManager.AnalyticsManager.quest_complete_2_13.Submit(Agent)
                14 => PlayerManager.GameManager.AnalyticsManager.quest_complete_2_14.Submit(Agent)
                15 => PlayerManager.GameManager.AnalyticsManager.quest_complete_2_15.Submit(Agent)
                16 => PlayerManager.GameManager.AnalyticsManager.quest_complete_2_16.Submit(Agent)
                17 => PlayerManager.GameManager.AnalyticsManager.quest_complete_2_17.Submit(Agent)
                18 => PlayerManager.GameManager.AnalyticsManager.quest_complete_2_18.Submit(Agent)
                19 => PlayerManager.GameManager.AnalyticsManager.quest_complete_2_19.Submit(Agent)
                20 => PlayerManager.GameManager.AnalyticsManager.quest_complete_2_20.Submit(Agent)
                _ => block:

        if(QuestID = 3):
            case (TaskID):
                0 => PlayerManager.GameManager.AnalyticsManager.quest_complete_3_0.Submit(Agent)
                1 => PlayerManager.GameManager.AnalyticsManager.quest_complete_3_1.Submit(Agent)
                2 => PlayerManager.GameManager.AnalyticsManager.quest_complete_3_2.Submit(Agent)
                3 => PlayerManager.GameManager.AnalyticsManager.quest_complete_3_3.Submit(Agent)
                4 => PlayerManager.GameManager.AnalyticsManager.quest_complete_3_4.Submit(Agent)
                5 => PlayerManager.GameManager.AnalyticsManager.quest_complete_3_5.Submit(Agent)
                6 => PlayerManager.GameManager.AnalyticsManager.quest_complete_3_6.Submit(Agent)
                7 => PlayerManager.GameManager.AnalyticsManager.quest_complete_3_7.Submit(Agent)
                8 => PlayerManager.GameManager.AnalyticsManager.quest_complete_3_8.Submit(Agent)
                9 => PlayerManager.GameManager.AnalyticsManager.quest_complete_3_9.Submit(Agent)
                10 => PlayerManager.GameManager.AnalyticsManager.quest_complete_3_10.Submit(Agent)
                11 => PlayerManager.GameManager.AnalyticsManager.quest_complete_3_11.Submit(Agent)
                12 => PlayerManager.GameManager.AnalyticsManager.quest_complete_3_12.Submit(Agent)
                13 => PlayerManager.GameManager.AnalyticsManager.quest_complete_3_13.Submit(Agent)
                14 => PlayerManager.GameManager.AnalyticsManager.quest_complete_3_14.Submit(Agent)
                15 => PlayerManager.GameManager.AnalyticsManager.quest_complete_3_15.Submit(Agent)
                16 => PlayerManager.GameManager.AnalyticsManager.quest_complete_3_16.Submit(Agent)
                20 => PlayerManager.GameManager.AnalyticsManager.quest_complete_3_20.Submit(Agent)
                25 => PlayerManager.GameManager.AnalyticsManager.quest_complete_3_25.Submit(Agent)
                _ => block:

        if(QuestID = 4):
            case (TaskID):
                0 => PlayerManager.GameManager.AnalyticsManager.quest_complete_4_0.Submit(Agent)
                3 => PlayerManager.GameManager.AnalyticsManager.quest_complete_4_3.Submit(Agent)
                6 => PlayerManager.GameManager.AnalyticsManager.quest_complete_4_6.Submit(Agent)
                9 => PlayerManager.GameManager.AnalyticsManager.quest_complete_4_9.Submit(Agent)
                12 => PlayerManager.GameManager.AnalyticsManager.quest_complete_4_12.Submit(Agent)
                22 => PlayerManager.GameManager.AnalyticsManager.quest_complete_4_22.Submit(Agent)
                29 => PlayerManager.GameManager.AnalyticsManager.quest_complete_4_29.Submit(Agent)
                _ => block:

    CompleteTask(QuestStats : quest_stats, ?DebugOverride : logic = false):void=
        if:
            Quest := QuestModel.Quests[QuestStats.QuestID]
            Task := Quest.Tasks[QuestStats.TaskID]
            Created := PlayerManager.GetCreationTime[]
            Now:= GetSecondsSinceEpoch()

            not PlayerManager.HasCompletedTask(QuestStats.QuestID, QuestStats.TaskID)?

        then:
            if:
                Created / 86400.0 + 1.0 >= Now / 86400.0
                PlayerManager.GetSourceDataSafe[].PrestigeCount <= 0
            then:
                CheckAnalytics(QuestStats.QuestID, QuestStats.TaskID)

            spawn:
                PlayerManager.FullHeal()

            if:
                Task.WaitForNoPopup?
            then:
                spawn:
                    WaitForNoPopup(QuestStats)
            else:
                spawn:
                    ProcessCompleteTask(QuestStats, ?DebugOverride := DebugOverride)

    WaitForNoPopup(QuestStats : quest_stats)<suspends>:void=
        loop:
            Sleep(1.0)
            if:
                Fort:= PlayerManager.Agent.GetFortCharacter[]
                not(Fort.IsActive[])
            then:
                return

            if (not PlayerManager.UIManager?.Popup?):
                ProcessCompleteTask(QuestStats)            
                return

    RevertToFailedTask(QuestStatsToUndo : quest_stats, FailedTask : quest_stats)<suspends>:void=
        Sleep(0.1)
        if:
            Task:=PlayerManager.GetFailableTask[]    
            Task.QuestID = FailedTask.QuestID
            Task.TaskID = FailedTask.TaskID
        then:
            block:
        else:
            return

        UnsubscribeFromTask(QuestStatsToUndo)
        TurnOffQuestTrackers(QuestStatsToUndo)
        PlayerManager.SetFailableTask(quest_stats{})
        StartTask(FailedTask)

    ProcessCompleteTask(QuestStats : quest_stats, ?DebugOverride : logic = false)<suspends>:void=
        Print("ProcessCompleteTask Q{QuestStats.QuestID}T{QuestStats.TaskID}")
        if:
            Quest := QuestModel.Quests[QuestStats.QuestID]
            Task := Quest.Tasks[QuestStats.TaskID]
            DebugOverride? or not(PlayerManager.HasCompletedTask(QuestStats.QuestID, QuestStats.TaskID)?)
        then:
            Print("Completing Task {QuestStats.TaskID} of Quest {QuestStats.QuestID}")
            if:
                Task.IsFailable?
            then:
                PlayerManager.SetFailableTask(QuestStats)
            else:
                UnsubscribeFromTask(QuestStats)
                if:
                    Temp:=PlayerManager.GetFailableTask[]
                then:
                    UnsubscribeFromTask(Temp)
                    PlayerManager.SetFailableTask(quest_stats{})

            # Print("Turning off Trackers")
            TurnOffQuestTrackers(QuestStats)


            if (ItemList := Task.GrantItemID?):
                for(ItemID : ItemList):
                    PlayerManager.MakeRoomForGun()
                    PlayerManager.EquipItem(ItemID, ?Equip:=true)

            # Grant Achievements
            for (Achievement : Task.CompleteAchievements):
                PlayerManager.GainAchievementLevel(Achievement, 1)

            #Give Essence Reward
            if:
                Task.CompleteStashReward.Value > 0.0
            then:
                PlayerManager.StashEssence(Task.CompleteStashReward)

            #Drop Items
            if:
                Item := Task.CompleteDropItem?
                not PlayerManager.ShouldSkipQuestGroup(Task.GroupKey)?
                Fort:= Agent.GetFortCharacter[]
                Fort.IsActive[]
            then:
                PlayerManager.GameManager.DropManager.SpawnItems(array{Item}, Fort.GetTransform().Translation, ?ParticipatingPlayers := array{PlayerManager})

            # Unlock Zone
            PlayerManager.UnlockZone(Task.CompleteZoneUnlock)

            # Hide Objects
            if: 
                SetHideObjects := Task.HideObjects?
            then:
                for (Hide : SetHideObjects):
                    Hide.AddToIgnoreList(Agent)

            if: 
                SetHideObjects := Task.HideObjectsAlt?
            then:
                for (Hide : SetHideObjects):
                    PlayerManager.GameManager.QueuePeely(queued_peely_doll{Peely := Hide, Agent := option{PlayerManager.Agent}, NewState := false}, ?SetState := true)
            
            # Show Objects    
            if: 
                SetShowObjects := Task.ShowObjectsAlt?
            then:
                for (Show : SetShowObjects):
                    PlayerManager.GameManager.QueuePeely(queued_peely_doll{Peely := Show, Agent := option{PlayerManager.Agent}, NewState := false}, ?SetState := true)
                    
            if:
                Task.TaskEvent = task_event.Interact
                not Task.TargetName = ""
                not Task.InteractableSwitch?
            then:
                PlayerManager.GainAchievementLevel("InteractedWithNPC_{Task.TargetName}", 1)
            
            #ResetGun CDs
            if:
                ItemsToReset := Task.CompleteResetItemCDs?
            then:
                for (Item : ItemsToReset):
                    PlayerManager.ResetItemCooldown(Item)

            #Grant Upgrade
            if:
                not(Task.CompleteUpgrade = "")
            then:
                PlayerManager.GainUpgradeLevel(Task.CompleteUpgrade, 1)

            #Grant Collectible
            if:
                not(Task.CompleteCollectible = "")
            then:
                PlayerManager.GainCollectible(collectible_data{Key:=Task.CompleteCollectible, Count:=Task.CompleteCollectibleCount})

            # Make Sounds
            if:
                Sounds := Task.CompleteSounds?
            then: 
                for (Sound : Sounds):
                    Sound.Play(Agent)

            # Make Dialog
            if:
                Dialogs := Task.CompleteDialogs?
                not PlayerManager.ShouldSkipQuestGroup(Task.GroupKey)?
            then:
                MakeDialog("", Dialogs).ShowDialog(PlayerManager, 0.01)

            # Start new Quests
            for (NewQuest : Task.QuestsToStartOnCompletion):
                StartQuest(NewQuest.QuestID)

            if:
                QuestStats.TaskID < Quest.Tasks.Length - 1
            then:
                if:
                    Task.CompoundWithNext?
                then:
                    PlayerManager.CompleteSecondaryTask(QuestStats.QuestID, QuestStats.TaskID)
                    Log("Completed Secondary Task {QuestStats.QuestID} {QuestStats.TaskID}")
                else:
                    NewTaskID := QuestStats.TaskID + 1
                    
                    NewQuestStats := quest_stats{QuestID:= QuestStats.QuestID, TaskID := NewTaskID}
                    #UpdatePlayerData
                    PlayerManager.UpdateQuestPlayerData(NewQuestStats, QuestStats.TaskID, ?DebugOverride := DebugOverride)

                    var SkipCompleteFlair : logic = false
                    if:
                        Task.DoesSkipCompleteFlair? or PlayerManager.ShouldSkipQuestGroup(Task.GroupKey)?
                    then:
                        set SkipCompleteFlair = true
                        
                    
                    #StartNextTask
                    Log("Completed Quest {QuestStats.QuestID} {QuestStats.TaskID}, Starting Task {NewQuestStats.TaskID}")
                    StartTask(NewQuestStats, ?DoesSkipCompleteFlair := SkipCompleteFlair)
            else:
                #UpdatePlayerData
                PlayerManager.CompleteQuest(QuestStats.QuestID, ?DebugOverride := DebugOverride)

            PlayerManager.PlayMusic()
            CheckPersistentQuests()

    #EventID = "{QuestIndex}.{TaskIndex}"
    OnButtonInteract(InteractAgent:agent, EventIDIn : string):void=
        if:
            Agent = InteractAgent
            CurrentQuests := PlayerManager.GetCurrentQuests[]
        then:
            for (QuestStats : CurrentQuests):
                if:
                    QuestID := QuestStats.QuestID
                    Quest := QuestModel.Quests[QuestID]
                    TaskID := QuestStats.TaskID
                    Task := Quest.Tasks[TaskID]
                    ProgressValue := QuestStats.Progress
                then:
                    ExpectedEventID := GetEventIDString(QuestID, TaskID)
                    if(Task.TaskEvent = task_event.Interact and EventIDIn = ExpectedEventID):
                        #Interact are always a 1 and done for tasks
                        CompleteTask(QuestStats)

    #EventID = "{QuestIndex}.{TaskIndex}"
    OnMutatorEnter(InteractAgent:agent, EventIDIn : string):void=
        if:
            Agent = InteractAgent
            CurrentQuests := PlayerManager.GetCurrentQuests[]
        then:
            for (QuestStats : CurrentQuests):
                if:
                    QuestID := QuestStats.QuestID
                    Quest := QuestModel.Quests[QuestID]
                    TaskID := QuestStats.TaskID
                    Task := Quest.Tasks[TaskID]
                    ProgressValue := QuestStats.Progress
                then:
                    ExpectedEventID := GetEventIDString(QuestID, TaskID)
                    if(Task.TaskEvent = task_event.EnterMutator and EventIDIn = ExpectedEventID):
                        #EnterMutator are always a 1 and done for tasks
                        CompleteTask(QuestStats)
    
    SubscribeToTask(QuestStats : quest_stats):void=
        # Log("SubscribeToTask!")
        QuestID := QuestStats.QuestID
        TaskID := QuestStats.TaskID

        EventIDString := GetEventIDString(QuestID, TaskID)
        if:
            Task := QuestModel.Quests[QuestID].Tasks[TaskID]
            TaskEvent := Task.TaskEvent
        then: 
            TrackedQuest := tracked_quest{QuestStats := quest_stats{QuestID:= QuestID, TaskID:= TaskID}, TaskEvent := TaskEvent}
    
            if:
                Loot := Task.CompleteLoot?
                not PlayerManager.ShouldSkipQuestGroup(Task.GroupKey)?
            then:
                Loot.PopupFor(PlayerManager)

            if:
                Task.MakeGunManagerPopupOnStart?
                not PlayerManager.ShouldSkipQuestGroup(Task.GroupKey)?
            then:
                PlayerManager.GameManager.GunManager.OpenMenu(PlayerManager.Agent, ?Force := true, ?RequiredID:=Task.GunIDRequired)

            if:
                Menu:=Task.UpgradeMenuPopupOnStart?
                not PlayerManager.ShouldSkipQuestGroup(Task.GroupKey)?
            then:
                Menu.TryOpenMenu("", PlayerManager.Agent)    

            if:
                GuardManagerDevice := Task.SpawnGuardOnStart?
                not(GuardManagerDevice.GuardManager.PlayerQuestOwners.Find[PlayerManager.Player])
            then:
                set GuardManagerDevice.GuardManager.PlayerQuestOwners = GuardManagerDevice.GuardManager.PlayerQuestOwners + array{PlayerManager.Player}

            if:
                Teleporter := Task.TeleportOnStart?
                not PlayerManager.ShouldSkipQuestGroup(Task.GroupKey)?
            then:
                Teleporter.Teleport(Player)

            if:
                not(Task.IsFailable?)
            then:
                if:
                    Interactable := Task.InteractableSwitch?
                then:
                    set TrackedQuest.SubscribedEvents = TrackedQuest.SubscribedEvents + array{Interactable.TurnedOffEvent.SubscribeAgent(OnButtonInteract, EventIDString)}
                else:
                    set TrackedQuest.SubscribedEvents = TrackedQuest.SubscribedEvents + array{switch_device{}.TurnedOffEvent.SubscribeAgent(OnButtonInteract, "")}

                if:
                    Interactable := Task.InteractableButton?
                then:
                    set TrackedQuest.SubscribedEvents = TrackedQuest.SubscribedEvents + array{Interactable.InteractedWithEvent.SubscribeAgent(OnButtonInteract, EventIDString)}
                else:
                    set TrackedQuest.SubscribedEvents = TrackedQuest.SubscribedEvents + array{button_device{}.InteractedWithEvent.SubscribeAgent(OnButtonInteract, "")}
                
                if:
                    Mutator := Task.Mutator?
                then:
                    set TrackedQuest.SubscribedEvents = TrackedQuest.SubscribedEvents + array{Mutator.AgentEntersEvent.SubscribeAgent(OnMutatorEnter, EventIDString)}
                else:
                    set TrackedQuest.SubscribedEvents = TrackedQuest.SubscribedEvents + array{mutator_zone_device{}.AgentEntersEvent.SubscribeAgent(OnMutatorEnter, "")}

            var AlreadyExists : logic = false

            for(ActiveTrackedQuest : ActiveTrackedQuests):
                if:
                    ActiveTrackedQuest.QuestStats.QuestID = QuestID
                    ActiveTrackedQuest.QuestStats.TaskID = TaskID
                then:
                    set AlreadyExists = true

            if (AlreadyExists?):
            else:
                set ActiveTrackedQuests = ActiveTrackedQuests + array{TrackedQuest}

    UnsubscribeFromTask(QuestStats : quest_stats):void=
        QuestID := QuestStats.QuestID
        TaskID := QuestStats.TaskID
        # Log("Unsubscribe: Q{QuestID}T{TaskID}, ActiveTrackedQuests Length: {ActiveTrackedQuests.Length}")
        var IndexToRemove : int = -1
        for(Index := 0..ActiveTrackedQuests.Length - 1):
            if:
                TrackedQuest := ActiveTrackedQuests[Index]
                TrackedQuest.QuestStats.QuestID = QuestID
                TrackedQuest.QuestStats.TaskID = TaskID                
            then:
                set IndexToRemove = Index
                for (SubscriptionToCancel : TrackedQuest.SubscribedEvents):
                    SubscriptionToCancel.Cancel()

        if:
            TempTaskEvents := ActiveTrackedQuests.RemoveElement[IndexToRemove]
            set ActiveTrackedQuests = TempTaskEvents

        # for(TrackedQuest : ActiveTrackedQuests):
            # Log("Remaining: {TrackedQuest.TaskEvent.ToString()}")
        

    #Confirm Quest does not already exist or was completed already
    IsQuestValid(QuestID : int)<transacts>:logic=
        if(not(PlayerManager.GetQuestStats[QuestID].QuestID = -1)):
            return false
        if(PlayerManager.GetCompletedQuests[].Find[QuestID]):
            return false
        return true

    TurnOnQuestTrackers(QuestStats : quest_stats, ?IsGameStart : logic = false)<suspends>:void=
        QuestID := QuestStats.QuestID
        TaskID := QuestStats.TaskID
        # Log("Turning on Trackers for QuestIndex: {QuestID} TaskIndex: {TaskID}!")
        if:
            Task := QuestModel.Quests[QuestID].Tasks[TaskID]
        then:            
            #Why is Verse like this
            #Wont work for first 2 minutes without REALLY turning it on
            # https://forums.unrealengine.com/t/objective-pulse-broken/1797066/3
            var Iterations : int = 2
            var SleepDuration : float = 1.0
            
            if:
                not(PlayerManager.HasCompletedTask(QuestID, TaskID)?)
            then:
                Task.Beacon.AddToShowList(Agent)

            if:
                IsGameStart?
            then:
                set Iterations = 20
                set SleepDuration = 0.01
            
            for(X:=0..Iterations):
                if:
                    not(PlayerManager.HasCompletedTask(QuestID, TaskID)?)
                then:
                    Task.MapIndicator.ActivateObjectivePulse(Agent)
                Sleep(SleepDuration)
            


    TurnOffQuestTrackers(QuestStats : quest_stats):void=
        QuestID := QuestStats.QuestID
        TaskID := QuestStats.TaskID
        
        if:
            Task := QuestModel.Quests[QuestID].Tasks[TaskID]
        then:            
            Task.MapIndicator.DeactivateObjectivePulse(Agent)
            spawn:
                ReallyRemove(Task)
        

    ReallyRemove(Task : task)<suspends>:void=
        Task.Beacon.RemoveFromShowList(Agent)
        Sleep(0.01)
        Task.Beacon.RemoveFromShowList(Agent)


    UpdateQuestUI(CompoundTasks : []quest_stats, ?IsNewQuest : logic = false, ?DoesSkipCompleteFlair : logic = false):void=
        if:
            FirstTaskStats := CompoundTasks[0]
            FirstTask := QuestModel.Quests[FirstTaskStats.QuestID].Tasks[FirstTaskStats.TaskID]
        then:
            if:
                CompoundTasks.Length = 1
            then:
                UpdateQuestUI(FirstTaskStats, ?IsNewQuest := IsNewQuest, ?DoesSkipCompleteFlair := DoesSkipCompleteFlair)
            else:
                PlayerManager.UpdateQuestUI(CompoundTasks, FirstTask.Description, ?IsNewQuest:=IsNewQuest, ?DoesSkipCompleteFlair:=DoesSkipCompleteFlair)
            
    UpdateQuestUI(QuestStats : quest_stats, ?IsNewQuest : logic = false, ?DoesSkipCompleteFlair : logic = false):void=
        QuestID := QuestStats.QuestID
        TaskID := QuestStats.TaskID
        
        #Get the Task Name
        if:
            Quest := QuestModel.Quests[QuestID]
            Quest.IsShowTask?
            Task := QuestModel.Quests[QuestID].Tasks[TaskID]
            not Task.Description = ""
        then:
            PlayerManager.UpdateQuestUI(QuestStats, Task.Description, ?IsNewQuest := IsNewQuest, ?DoesSkipCompleteFlair := DoesSkipCompleteFlair)
        else:
            if: 
                TempQuestStats := PlayerManager.GetQuestStats[0]
                Task := QuestModel.Quests[TempQuestStats.QuestID].Tasks[TempQuestStats.TaskID]
            then:
                PlayerManager.UpdateQuestUI(TempQuestStats, Task.Description, ?IsNewQuest := IsNewQuest, ?DoesSkipCompleteFlair := DoesSkipCompleteFlair)

    # UpdateQuestUI(CompoundTasks : []quest_stats, ?IsNewQuest : logic = false):void=
    #     if:
    #         FirstTaskStats := CompoundTasks[0]
    #         FirstTask := QuestModel.Quests[FirstTaskStats.QuestID].Tasks[FirstTaskStats.TaskID]
    #     then:
    #         PlayerManager.UpdateQuestUI(CompoundTasks, FirstTask.Description, ?IsNewQuest:=IsNewQuest)
          
    # UpdateQuestUI(QuestStats : quest_stats, ?IsNewQuest : logic = false):void=
    #     UpdateQuestUI(GetRelevantCompoundTasks(QuestStats), ?IsNewQuest:=IsNewQuest)


    GetEventIDString(QuestID : int, TaskID : int):string=
        return "{QuestID}.{TaskID}"

quest_progress_info := class :
    Current : big_number
    Total : big_number
    IconText : string
