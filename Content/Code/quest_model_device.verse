using { /Fortnite.com/Characters }
using { /Fortnite.com/Devices }
using { /UnrealEngine.com/Temporary/SpatialMath }
using { /Verse.org/Simulation }

task_event := enum: 
    None
    GatherEssence #Supported
    GatherRuby #Not Supported
    Interact #Supported
    GuardKO #Supported
    ZoneClear #Supported
    ZoneLevel #Supported
    CollectStash #Supported
    PlayerLevel #Supported
    EnterZone #Supported
    CollectUpgrades #Supported
    CumulativeUpgrades #Supported
    FillStash #Supported
    EquipItem #Supported
    DefeatBoss #Supported
    ItemExpired #Supported
    RequireAchievement #Supported
    UseTeleport #Supported
    EatCandy #Supported
    GetCollectibleCount #Supported
    LootChosen #Supported
    EquipArmor #Supported
    BoughtItem
    EnterMutator
    HitGuard
    #ZoneLevelGain #Can cause softlocks
    #

(Tag:task_event).ToString()<transacts>:string =
    case (Tag):
        task_event.None => "None"
        task_event.GatherEssence => "GatherEssence"
        task_event.GatherRuby => "GatherRuby"
        task_event.Interact => "Interact"
        task_event.GuardKO => "GuardKO"
        task_event.ZoneClear => "ZoneClear"
        task_event.ZoneLevel => "ZoneLevel"
        task_event.CollectStash => "CollectStash"
        task_event.PlayerLevel => "PlayerLevel"
        task_event.EnterZone => "EnterZone"
        task_event.CollectUpgrades => "CollectUpgrades"
        task_event.CumulativeUpgrades => "CumulativeUpgrades"
        task_event.FillStash => "FillStash"
        task_event.EquipItem => "EquipItem"
        task_event.DefeatBoss => "DefeatBoss"
        task_event.ItemExpired => "ItemExpired"
        task_event.RequireAchievement => "RequireAchievement"
        task_event.UseTeleport => "UseTeleport"
        task_event.EatCandy => "EatCandy"
        task_event.GetCollectibleCount => "GetCollectibleCount"
        task_event.LootChosen => "LootChosen"
        task_event.EquipArmor => "EquipArmor"
        task_event.BoughtItem => "BoughtItem"
        task_event.EnterMutator => "EnterMutator"
        task_event.HitGuard => "HitGuard"

condition:=class<concrete>:

quest:=class<concrete>:
    @editable
    Name : string = ""
    
    @editable
    Tasks : []task = array{}
    
    @editable
    Conditions : []condition = array{}
    
    @editable
    IsStartingQuest : logic = false

    @editable
    StartingTask : int = 0
    
    @editable
    IsShowTask : logic = false

task:=class<concrete>:
    @editable
    Description : string = ""

    @editable
    GroupKey : string = ""
    
    @editable
    DoesSkipCompleteFlair : logic = false

    #Note: We don't allow Failable quests to use UEFN's Subscribe functionality. 
    #Failable Quests cannot be interact task_events
    #Failable Quests also MUST use Persistent Data (Ex: collectible counts)
    @editable
    IsFailable : logic = false
    
    @editable
    CompoundWithNext : logic = false

    @editable
    IconTexture : string = ""

    @editable
    Progress : big_number = big_number{}
    
    @editable 
    Beacon : beacon_device = beacon_device{}
    
    @editable
    MapIndicator : map_indicator_device = map_indicator_device{}

    @editable
    TaskEvent : task_event = task_event.None

    @editable
    InteractableButton : ?button_device = false

    @editable
    InteractableSwitch : ?switch_device = false
    
    @editable
    TargetName : string = ""

    @editable
    ZoneRequired : fetchable_tag = fetchable_tag.none_tag

    @editable
    SpawnGuardOnStart : ?guard_manager_device = false
    
    @editable
    TeleportOnStart : ?teleporter_device = false

    @editable
    MakeGunManagerPopupOnStart : logic = false

    @editable
    GunIDRequired : int = 0

    @editable
    UpgradeMenuPopupOnStart : ?upgrade_shop = false

    @editable
    PlayMusicOnStart : ?audio_player_wrapper = false

    @editable
    HideObjects : ?[]barrier_device = false

    @editable
    HideObjectsAlt : ?[]peely_doll_device = false

    @editable
    ShowObjectsAlt : ?[]peely_doll_device = false

    @editable
    GrantItemID : ?[]int = false 

    @editable
    WaitForNoPopup : logic = false 

    @editable
    CompleteStashReward : big_number = big_number{}
    
    @editable
    CompleteUpgrade : string = ""

    @editable
    CompleteCollectible : string = ""

    @editable
    CompleteCollectibleCount : big_number = big_number{Value:=1.0}

    @editable
    CompleteResetItemCDs : ?[]int = false 

    @editable
    CompleteDialogs : ?[]dialog_message = false 

    @editable
    CompleteSounds : ?[]audio_player_device = false 

    @editable
    CompleteZoneUnlock : fetchable_tag = fetchable_tag.none_tag

    @editable
    CompleteAchievements : []string = array{}

    @editable
    CompleteDropItem : ?drop_item = false

    @editable
    CompleteLoot : ?loot_choice = false

    @editable
    QuestsToStartOnCompletion : []quest_stats = array{}

    @editable
    Mutator : ?mutator_zone_device = false

quest_model_device := class(creative_device):
    @editable
    Quests : []quest = array{}