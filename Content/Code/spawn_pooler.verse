
using { /Fortnite.com/Devices }
using { /Verse.org/Simulation }
using { /UnrealEngine.com/Temporary/Diagnostics }
using { /Fortnite.com/Characters }

# See https://dev.epicgames.com/documentation/en-us/uefn/create-your-own-device-in-verse for how to create a verse device.

# A Verse-authored creative device that can be placed in a level
spawn_pooler := class(creative_device):

    @editable
    WildlifeSpawner : ?wildlife_spawner_device = false
    @editable
    GuardSpawner : ?guard_spawner_device = false
    @editable
    CreaturePlacers : []creature_placer_device = array{}
    
    var QueuedGuardManagers : [string]?guard_manager = map{}
    var QueuedAgents : []agent = array{}

    var AvailableCreaturePlacers : []creature_placer_device = array{}
    
    OnBegin<override>():void=
        if:
            Wildlife:=WildlifeSpawner?
        then:
            Wildlife.SpawnedEvent.Subscribe(OnSpawned)

        if (Guard := GuardSpawner?):        
            Guard.SpawnedEvent.Subscribe(OnSpawned)

        spawn:
            LoopSpawn()

    LoopSpawn()<suspends>:void=
        loop:#
            Sleep(0.23)
            if:
                QueuedAgents[0]
            then:
                for(Guard:QueuedGuardManagers):
                    if:
                        Available := QueuedAgents[0]
                        SetGuard := Guard?
                        not SetGuard.EntityAgent?
                    then:
                        #Assign to QueuedGuardManager
                        var GuardIDtoRemove : string = ""
                        set SetGuard.EntityAgent = option{Available}
                        var DidSucceed : logic = SetGuard.OnAgentSpawn()
                        if(DidSucceed?):
                            #Remove GuardManager from Queue
                            set GuardIDtoRemove = SetGuard.GuardID
                            if:
                                set QueuedGuardManagers[GuardIDtoRemove] = false
                        else:
                            set SetGuard.EntityAgent = false
                            
                        #Success or fail, this Agent needs to be removed
                        if:
                            Temp:= QueuedAgents.RemoveElement[0]                        
                        then:
                            set QueuedAgents = Temp
            else:
                var ShouldDoSpawn : logic = false
                for(Guard:QueuedGuardManagers):
                    if:
                        SetGuard := Guard?
                        not SetGuard.EntityAgent?
                    then:
                        set ShouldDoSpawn = true                        
                if(ShouldDoSpawn?):
                    DoSpawn()
            
    OnSpawned(Agent:agent):void=
        set QueuedAgents = QueuedAgents + array{Agent}
            
    DoSpawn()<suspends>:void=
        if (Wildlife := WildlifeSpawner?):
            Wildlife.Spawn()
            
        else if (Guard := GuardSpawner?):
            Guard.Spawn()

        # else if (Creature := AvailableCreaturePlacers[0]):
        #     set SpawnAttempted = true
        #     Creature.Spawn()
    #         set AvailableCreaturePlacers = AvailableCreaturePlacers.RemoveElement[0]
        #     Creature.EliminatedEvent.Subscribe(creature_reclaimer{Placer := Creature, Pooler := Self}.Reclaim)

    RequestSpawn(GuardManager : guard_manager):void=
        if(not QueuedGuardManagers[GuardManager.GuardID]?.GuardID = GuardManager.GuardID):
            if:
                set QueuedGuardManagers[GuardManager.GuardID] = option{GuardManager}
            then:
                block:
    
creature_reclaimer := class():
    Placer : creature_placer_device
    Pooler : spawn_pooler

    Reclaim(Agent : ?agent):void =
        set Pooler.AvailableCreaturePlacers = Pooler.AvailableCreaturePlacers + array{Placer}