using { /Fortnite.com/Devices }
using { /Fortnite.com/UI }
using { /UnrealEngine.com/Temporary/Diagnostics }
using { /UnrealEngine.com/Temporary/SpatialMath }
using { /UnrealEngine.com/Temporary/UI}
using { /Verse.org/Assets }
using { /Verse.org/Colors }
using { /Verse.org/Simulation }

stash_menus_device := class(creative_device, menu_item):

    @editable
    GameManager : ?game_manager = false

    OnBegin<override>()<suspends>:void=
        block:

    OpenMenu<override>(Agent:agent):void=
        if (PlayerManager := GameManager?.PlayerManagers[Agent]):
            PlayerManager.CollectStash()

    GetTitle<override>():string=
        return "STASH"

    GetIcon<override>():texture=
        return Images.Chest

    IsEnabled<override>(PlayerManager : player_manager)<transacts>:logic=
        if (PlayerManager.GetStashedEssence[].Value > 0.0):
            return true
        else:
            return false