using { /Verse.org/Assets }
using { /Verse.org/Simulation/Tags }

guard_spawner := class(tag){}
boss_spawner := class(tag){}
spawner := class(tag){}
beacon := class(tag){}
item := class(tag){}
coin := class(tag){}
mapindicator := class(tag){}
t_button := class(tag){}
zone := class(tag){}

none := class(tag){}
gunxpchest := class(tag){}
prestigearea := class(tag){}

#midas castle
zone1 := class(tag){}
zone2 := class(tag){}
zone3 := class(tag){}
zone4 := class(tag){}

#town
town_zone := class(tag){}

#beach
beach_zone1 := class(tag){}
beach_zone2 := class(tag){}
beach_zone3 := class(tag){}

#forest
forest_zone1 := class(tag){}
forest_zone2 := class(tag){}
forest_zone3 := class(tag){}
forest_zone4 := class(tag){}

#forest_bandits
forest_bandit_zone1 := class(tag){}
forest_bandit_zone2 := class(tag){}
forest_bandit_zone3 := class(tag){}

#shooting range
range_zone := class(tag){}

#AFK Zone
afk_zone := class(tag){}

#no Zone
no_zone := class(tag){}

#pirates
pirates_zone1 := class(tag){}
pirates_zone2 := class(tag){}
pirates_zone3 := class(tag){}
pirates_zone4 := class(tag){}
pirates_zone5 := class(tag){}

#canyon
canyon_zone1 := class(tag){}
canyon_zone2 := class(tag){}
canyon_zone3 := class(tag){}
canyon_zone4 := class(tag){}
canyon_zone5 := class(tag){}
canyon_zone6 := class(tag){}

#giantlands
giantlands_zone1 := class(tag){}
giantlands_zone2 := class(tag){}
giantlands_zone3 := class(tag){}
giantlands_zone4 := class(tag){}
giantlands_zone5 := class(tag){}
giantlands_zone6 := class(tag){}

fetchable_tag := enum:
    none_tag
    zone1tag
    zone2tag
    zone3tag
    towntag
    forest_zone1tag
    forest_zone2tag
    forest_zone3tag
    forest_zone4tag
    range_zonetag
    afk_zonetag
    canyon_zone1tag
    canyon_zone2tag
    canyon_zone3tag
    canyon_zone4tag
    canyon_zone5tag
    canyon_zone6tag
    no_zonetag
    midastag
    giantlands_zone1tag
    giantlands_zone2tag
    giantlands_zone3tag
    giantlands_zone4tag
    giantlands_zone5tag
    giantlands_zone6tag
    beach_zone1tag
    beach_zone2tag
    beach_zone3tag
    forest_bandit_zone1tag
    forest_bandit_zone2tag
    forest_bandit_zone3tag
    pirates_zone1tag
    pirates_zone2tag
    pirates_zone3tag
    pirates_zone4tag
    pirates_zone5tag
    gengatag
    urjenttag
    gianttag
    jenkinstag
    piratekingtag
    lyratag
    sharkeytag
    gunxpchesttag
    prestigeareatag

tags_list : []fetchable_tag = array{fetchable_tag.none_tag, fetchable_tag.zone1tag, fetchable_tag.zone2tag, fetchable_tag.zone3tag,
                                    fetchable_tag.towntag, fetchable_tag.forest_zone1tag, fetchable_tag.forest_zone2tag,
                                    fetchable_tag.forest_zone3tag, fetchable_tag.forest_zone4tag, fetchable_tag.range_zonetag, 
                                    fetchable_tag.no_zonetag, fetchable_tag.midastag, fetchable_tag.gunxpchesttag}

GetZoneTagFromIndex(Index:int)<transacts>: fetchable_tag=
    case (Index):
        0 => fetchable_tag.zone1tag
        1 => fetchable_tag.zone2tag
        2 => fetchable_tag.zone3tag
        3 => fetchable_tag.towntag
        4 => fetchable_tag.forest_zone1tag
        5 => fetchable_tag.forest_zone2tag
        6 => fetchable_tag.forest_zone3tag
        7 => fetchable_tag.forest_zone4tag
        8 => fetchable_tag.range_zonetag
        9 => fetchable_tag.no_zonetag
        10 => fetchable_tag.midastag
        11 => fetchable_tag.beach_zone1tag
        12 => fetchable_tag.beach_zone2tag 
        13 => fetchable_tag.beach_zone3tag
        14 => fetchable_tag.forest_bandit_zone1tag
        15 => fetchable_tag.forest_bandit_zone2tag
        16 => fetchable_tag.forest_bandit_zone3tag
        17 => fetchable_tag.pirates_zone1tag
        18 => fetchable_tag.pirates_zone2tag
        19 => fetchable_tag.pirates_zone3tag
        20 => fetchable_tag.pirates_zone4tag
        21 => fetchable_tag.pirates_zone5tag
        22 => fetchable_tag.gengatag
        23 => fetchable_tag.urjenttag
        24 => fetchable_tag.gianttag
        25 => fetchable_tag.jenkinstag
        26 => fetchable_tag.piratekingtag
        27 => fetchable_tag.lyratag
        28 => fetchable_tag.sharkeytag
        29 => fetchable_tag.gunxpchesttag
        _ => fetchable_tag.none_tag

(Tag:fetchable_tag).ToTagClass():tag =
    case (Tag):
        fetchable_tag.zone1tag => zone1{}
        fetchable_tag.zone2tag => zone2{}
        fetchable_tag.zone3tag => zone3{}
        fetchable_tag.midastag => zone3{}
        fetchable_tag.towntag => town_zone{}
        fetchable_tag.forest_zone1tag => forest_zone1{}
        fetchable_tag.forest_zone2tag => forest_zone2{}
        fetchable_tag.forest_zone3tag => forest_zone3{}
        fetchable_tag.forest_zone4tag => forest_zone4{}
        fetchable_tag.range_zonetag => range_zone{}
        fetchable_tag.afk_zonetag => afk_zone{}
        fetchable_tag.none_tag => none{}
        fetchable_tag.canyon_zone1tag => canyon_zone1{}
        fetchable_tag.canyon_zone2tag => canyon_zone2{}
        fetchable_tag.canyon_zone3tag => canyon_zone3{}
        fetchable_tag.canyon_zone4tag => canyon_zone4{}
        fetchable_tag.canyon_zone5tag => canyon_zone5{}
        fetchable_tag.canyon_zone6tag => canyon_zone6{}
        fetchable_tag.giantlands_zone1tag => giantlands_zone1{}
        fetchable_tag.giantlands_zone2tag => giantlands_zone2{}
        fetchable_tag.giantlands_zone3tag => giantlands_zone3{}
        fetchable_tag.giantlands_zone4tag => giantlands_zone4{}
        fetchable_tag.giantlands_zone5tag => giantlands_zone5{}
        fetchable_tag.giantlands_zone6tag => giantlands_zone6{}
        fetchable_tag.beach_zone1tag => beach_zone1{}
        fetchable_tag.beach_zone2tag => beach_zone2{}
        fetchable_tag.beach_zone3tag => beach_zone3{}
        fetchable_tag.forest_bandit_zone1tag => forest_bandit_zone1{}
        fetchable_tag.forest_bandit_zone2tag => forest_bandit_zone2{}
        fetchable_tag.forest_bandit_zone3tag => forest_bandit_zone3{}
        fetchable_tag.pirates_zone1tag => pirates_zone1{}
        fetchable_tag.pirates_zone2tag => pirates_zone2{}
        fetchable_tag.pirates_zone3tag => pirates_zone3{}
        fetchable_tag.pirates_zone4tag => pirates_zone4{}
        fetchable_tag.pirates_zone5tag => pirates_zone5{}
        fetchable_tag.gengatag => zone3{}  #IDK What this is used for
        fetchable_tag.urjenttag => zone3{}
        fetchable_tag.sharkeytag => zone3{}
        fetchable_tag.gianttag => zone3{}
        fetchable_tag.jenkinstag => zone3{}
        fetchable_tag.piratekingtag => zone3{}
        fetchable_tag.lyratag => zone3{}
        fetchable_tag.no_zonetag => no_zone{}
        fetchable_tag.gunxpchesttag => gunxpchest{}
        fetchable_tag.prestigeareatag => prestigearea{}

(Tag:fetchable_tag).ToGuardImage()<transacts>:string =
    case (Tag):
        fetchable_tag.zone1tag => "Midas Guard"
        fetchable_tag.zone2tag => "Midas Elite"
        fetchable_tag.zone3tag => "MidasIcon"
        fetchable_tag.forest_zone1tag => "BoarExample"
        fetchable_tag.forest_zone2tag => "UmbralOreExample"
        fetchable_tag.forest_zone3tag => "KikoIcon"
        fetchable_tag.beach_zone1tag => "Bogstick"
        fetchable_tag.beach_zone2tag => "BeachInvader"
        fetchable_tag.beach_zone3tag => "Funguy"
        fetchable_tag.pirates_zone1tag => "Cutthroat"
        fetchable_tag.pirates_zone2tag => "Cutthroat"
        fetchable_tag.pirates_zone3tag => "Cutthroat"
        fetchable_tag.pirates_zone4tag => "Cutthroat"
        fetchable_tag.pirates_zone5tag => "Cutthroat"
        _ => ""

(Tag:fetchable_tag).ToGuardHealth()<transacts>:big_number =
    case (Tag):
        fetchable_tag.zone1tag => big_number{Value:=2.0, Exponent:=1}
        fetchable_tag.zone2tag => big_number{Value:=2.0, Exponent:=2}   
        fetchable_tag.zone3tag => big_number{Value:=6.0, Exponent:=4}
        fetchable_tag.towntag => big_number{Value:=1.0, Exponent:=0}
        fetchable_tag.beach_zone1tag => big_number{Value:=2.5, Exponent:=4}
        fetchable_tag.beach_zone2tag => big_number{Value:=1.0, Exponent:=5}
        fetchable_tag.beach_zone3tag => big_number{Value:=3.0, Exponent:=5}
        fetchable_tag.forest_zone1tag => big_number{Value:=3.0, Exponent:=6}
        fetchable_tag.forest_zone2tag => big_number{Value:=1.0, Exponent:=7}
        fetchable_tag.forest_bandit_zone1tag => big_number{Value:=1.0, Exponent:= 8}
        fetchable_tag.forest_bandit_zone2tag => big_number{Value:=2.5, Exponent:= 8}
        fetchable_tag.forest_bandit_zone3tag => big_number{Value:=2.5, Exponent:=8}
        fetchable_tag.giantlands_zone1tag => big_number{Value:=0.75, Exponent:=9}
        fetchable_tag.giantlands_zone2tag => big_number{Value:=0.1, Exponent:=10}
        fetchable_tag.giantlands_zone3tag => big_number{Value:=1.3, Exponent:=11}
        fetchable_tag.pirates_zone1tag => big_number{Value:=1.0, Exponent:=12}
        fetchable_tag.pirates_zone2tag => big_number{Value:=4.0, Exponent:=12}
        fetchable_tag.pirates_zone3tag => big_number{Value:=1.2, Exponent:=13}
        fetchable_tag.pirates_zone4tag => big_number{Value:=5.0, Exponent:=13}
        #fetchable_tag.pirates_zone5tag => big_number{Value:=1.0, Exponent:=12}
        fetchable_tag.forest_zone3tag => big_number{Value:=3.2, Exponent:=7}
        fetchable_tag.forest_zone4tag => big_number{Value:=1.5, Exponent:=8}
        
        fetchable_tag.range_zonetag => big_number{Value:=1.0, Exponent:=3}
        _ => big_number{Value:=1.0, Exponent:=0}

(Tag:fetchable_tag).ToGuardKOBonus():float = 0.0
        
(Tag:fetchable_tag).ToGuardName():string =
    case (Tag):
        fetchable_tag.zone1tag => "Gun Storage"
        fetchable_tag.zone2tag => "Midas Statue"
        fetchable_tag.zone3tag => "Midas"
        fetchable_tag.towntag => "Town Guard"
        fetchable_tag.forest_zone1tag => "Forest Spirit"
        fetchable_tag.forest_zone2tag => "Umbral Ore Deposit"
        fetchable_tag.forest_zone3tag => "Forest Spirit"
        fetchable_tag.forest_zone4tag => "Forest Keeper"
        fetchable_tag.forest_bandit_zone1tag => "Bandit"
        fetchable_tag.forest_bandit_zone2tag =>"Veteran Bandit"
        fetchable_tag.forest_bandit_zone3tag => "Elite Bandit"
        fetchable_tag.range_zonetag => "Target"
        fetchable_tag.no_zonetag => "???"
        fetchable_tag.beach_zone1tag => "Young Bogloc"
        fetchable_tag.beach_zone2tag => "Beach Invader"
        fetchable_tag.beach_zone3tag => "Funguy"
        fetchable_tag.pirates_zone1tag => "Pirate"
        fetchable_tag.pirates_zone2tag => "Elite Pirate"
        fetchable_tag.pirates_zone3tag => "Hulking Pirate"
        fetchable_tag.pirates_zone4tag => "Raider Guard"
        #fetchable_tag.pirates_zone5tag => "Veteran Destroyer"
        fetchable_tag.giantlands_zone1tag => "Raider"
        _ => "..."
# 
(Tag:fetchable_tag).ToZoneName()<transacts>:string =
    case (Tag):
        fetchable_tag.zone1tag => "Midas Antechamber"
        fetchable_tag.zone2tag => "Midas Throne Room"
        fetchable_tag.zone3tag => ""
        fetchable_tag.midastag => "Midas Sanctum"
        fetchable_tag.towntag => "Goodtown"
        fetchable_tag.forest_zone1tag => "Umbral Forest River"
        fetchable_tag.forest_zone2tag => "Umbral Forest Quarry"
        fetchable_tag.forest_zone3tag => "Heart of Umbral Forest"
        fetchable_tag.forest_bandit_zone1tag => "Bandit Camps"
        fetchable_tag.forest_bandit_zone2tag => "Bandit Midas Orbs"
        fetchable_tag.forest_bandit_zone3tag => "Genga's Tent"
        fetchable_tag.forest_zone4tag => "Keeper's Grove"
        fetchable_tag.range_zonetag => "Shooting Range"
        fetchable_tag.afk_zonetag => "AFK Zone"
        fetchable_tag.canyon_zone1tag => "Zephyr's Canyon"
        fetchable_tag.no_zonetag => "???"
        fetchable_tag.beach_zone1tag => "Goodtown Beachhead"
        fetchable_tag.beach_zone2tag => "The Goodtown Beach"
        fetchable_tag.beach_zone3tag => "The Beach Shacks"
        fetchable_tag.pirates_zone1tag => "Cutthroat Island"
        fetchable_tag.pirates_zone2tag => "Cutthroat Watchtower"
        fetchable_tag.pirates_zone3tag => "Cutthroat Keep"
        fetchable_tag.pirates_zone4tag => "Secret Pirate Outpost"
        #fetchable_tag.pirates_zone5tag => "Cutthroat Keep"
        fetchable_tag.giantlands_zone1tag => "Giantlands Raider Outpost"
        fetchable_tag.giantlands_zone2tag => "Vampire Manor"
        fetchable_tag.giantlands_zone3tag => "Giant's Keep"
        fetchable_tag.gengatag => "Genga's Hideout"
        fetchable_tag.urjenttag => "Urjent's Lair"
        fetchable_tag.gianttag => "Giant's Fortress"
        fetchable_tag.jenkinstag => "Jenkins' Cove"
        fetchable_tag.piratekingtag => "Pirate King's Cove"
        fetchable_tag.lyratag => "Lyra's Training Ground"
        fetchable_tag.sharkeytag => "Sharkey's Pier"
        fetchable_tag.prestigeareatag => "Beginning of Everything"
        _ => "..."

ToString(Tag:fetchable_tag):string = Tag.ToZoneName()

(Tag:fetchable_tag).ToZoneValue():int =
    case (Tag):
        fetchable_tag.zone1tag => 1
        fetchable_tag.zone2tag => 2
        fetchable_tag.zone3tag => 3
        fetchable_tag.towntag => 4
        fetchable_tag.forest_zone1tag => 5
        fetchable_tag.forest_zone2tag => 6
        fetchable_tag.forest_zone3tag => 7
        fetchable_tag.forest_zone4tag => 8
        fetchable_tag.range_zonetag => 9
        fetchable_tag.no_zonetag => 10
        fetchable_tag.beach_zone1tag => 11
        fetchable_tag.beach_zone2tag => 12
        fetchable_tag.beach_zone3tag => 13
        fetchable_tag.pirates_zone1tag => 14
        fetchable_tag.pirates_zone2tag => 15
        fetchable_tag.pirates_zone3tag => 16
        fetchable_tag.pirates_zone4tag => 17
        fetchable_tag.pirates_zone5tag => 18
        fetchable_tag.giantlands_zone1tag => 19
        fetchable_tag.giantlands_zone2tag => 20
        fetchable_tag.giantlands_zone3tag => 21
        _ => 0

(Tag:fetchable_tag).ToZoneIndex()<transacts>:int =
    case (Tag):
        fetchable_tag.zone1tag => 0
        fetchable_tag.zone2tag => 1
        fetchable_tag.zone3tag => 2
        fetchable_tag.towntag => 3
        fetchable_tag.forest_zone1tag => 4
        fetchable_tag.forest_zone2tag => 5
        fetchable_tag.forest_zone3tag => 6
        fetchable_tag.forest_zone4tag => 7
        fetchable_tag.range_zonetag => 8
        fetchable_tag.no_zonetag => 9
        fetchable_tag.midastag => 10
        fetchable_tag.beach_zone1tag => 11
        fetchable_tag.beach_zone2tag => 12
        fetchable_tag.beach_zone3tag => 13
        fetchable_tag.pirates_zone1tag => 14
        fetchable_tag.pirates_zone2tag => 15
        fetchable_tag.pirates_zone3tag => 16
        fetchable_tag.pirates_zone4tag => 17
        fetchable_tag.pirates_zone5tag => 18
        fetchable_tag.giantlands_zone1tag => 19
        fetchable_tag.giantlands_zone2tag => 20
        fetchable_tag.giantlands_zone3tag => 21
        fetchable_tag.gengatag => 22
        fetchable_tag.urjenttag => 23
        fetchable_tag.gianttag => 24
        fetchable_tag.jenkinstag => 25
        fetchable_tag.piratekingtag => 26
        fetchable_tag.lyratag => 27
        fetchable_tag.sharkeytag => 28
        _ => -1

(Tag:fetchable_tag).ToPlayerScaling()<transacts>:float =
    case (Tag):
        fetchable_tag.zone1tag => 1.0
        fetchable_tag.zone2tag => 1.0
        fetchable_tag.forest_zone1tag => 0.25
        fetchable_tag.forest_zone2tag => 0.25
        fetchable_tag.forest_zone3tag => 0.25
        fetchable_tag.beach_zone1tag => 0.25
        fetchable_tag.beach_zone2tag => 0.25
        fetchable_tag.beach_zone3tag => 0.25
        _ => 0.25

#Unused
(Tag:fetchable_tag).ToZoneGemGain()<transacts>:float =
    case (Tag):
        fetchable_tag.zone1tag => 1.0
        fetchable_tag.zone2tag => 1.0
        fetchable_tag.forest_zone1tag => 2.0
        fetchable_tag.forest_zone2tag => 2.0
        fetchable_tag.forest_zone3tag => 2.0
        fetchable_tag.forest_zone4tag => 2.0
        fetchable_tag.beach_zone1tag => 1.0
        fetchable_tag.beach_zone2tag => 1.0
        fetchable_tag.beach_zone3tag => 1.0
        _ => 1.0

#Unused
(Tag:fetchable_tag).ToDamageMasteryArray()<transacts>: []big_number =
    case (Tag):
        fetchable_tag.zone1tag => array{}
        fetchable_tag.zone2tag => array:
            big_number{Value:= 2.5, Exponent:= 4}
            big_number{Value:= 1.0, Exponent:= 5}
            big_number{Value:= 5.0, Exponent:= 5}
            big_number{Value:= 2.5, Exponent:= 6}
            big_number{Value:= 1.25, Exponent:= 7}
            big_number{Value:= 5.25, Exponent:= 7}
            big_number{Value:= 2.5, Exponent:= 8}
            big_number{Value:= 1.25, Exponent:= 9}
        fetchable_tag.zone3tag => array{}
        fetchable_tag.towntag => array{}
        fetchable_tag.forest_zone1tag => array{}
        fetchable_tag.forest_zone2tag => array{}
        fetchable_tag.forest_zone3tag => array{}
        fetchable_tag.range_zonetag => array{}
        fetchable_tag.forest_zone4tag => array{}
        _ => array{}

#Unused
(Tag:fetchable_tag).ToDamageMasteryMultipliers()<transacts>: []float =
    case (Tag):
        _ => array:
            1.0
            10.0
            40.0
            160.0
            640.0
            2560.0
            10000.0
            40000.0
            16000.0
            64000.0
            256000.0

(Tag:fetchable_tag).GuardSpawnDistance()<transacts>:float =
    case (Tag):
        fetchable_tag.no_zonetag => 4500.0
        _ => 5500.0

(Tag:fetchable_tag).IsBossTag()<transacts>:logic =
    case (Tag):
        fetchable_tag.zone3tag => true
        fetchable_tag.midastag => true
        fetchable_tag.gengatag => true
        fetchable_tag.urjenttag => true
        fetchable_tag.gianttag => true
        fetchable_tag.jenkinstag => true
        fetchable_tag.piratekingtag => true
        fetchable_tag.lyratag => true
        fetchable_tag.sharkeytag => true
        _ => false

(Tag:fetchable_tag).BossID()<transacts>:int =
    case (Tag):
        fetchable_tag.zone3tag => 1
        _ => 0

MakeCandyTable(Chances : []float) : drop_table =
    var Items : []drop_item = array{}
    if:
        C0 := Chances[0]
        C0 > 0.0
    then:
        set Items = Items + array{drop_item{Name := "Candy1", PickupText := "Common Candy", Chance := C0, MinCount := big_number{Value:=1.0}, MaxCount := big_number{Value:=1.0}}}
    if:
        C1 := Chances[1]
        C1 > 0.0
    then:
        set Items = Items + array{drop_item{Name := "Candy2", PickupText := "Uncommon Candy", Chance := C1, MinCount := big_number{Value:=1.0}, MaxCount := big_number{Value:=1.0}}}
    if:
        C2 := Chances[2]
        C2 > 0.0
    then:
        set Items = Items + array{drop_item{Name := "Candy3", PickupText := "Rare Candy", Chance := C2, MinCount := big_number{Value:=1.0}, MaxCount := big_number{Value:=1.0}}}
    if:
        C3 := Chances[3]
        C3 > 0.0
    then:
        set Items = Items + array{drop_item{Name := "Candy4", PickupText := "Super Rare Candy", Chance := C3, MinCount := big_number{Value:=1.0}, MaxCount := big_number{Value:=1.0}}}
    if:
        C4 := Chances[4]
        C4 > 0.0
    then:
        set Items = Items + array{drop_item{Name := "Candy5", PickupText := "Epic Candy", Chance := C4, MinCount := big_number{Value:=1.0}, MaxCount := big_number{Value:=1.0}}}
    if:
        C5 := Chances[5]
        C5 > 0.0
    then:
        set Items = Items + array{drop_item{Name := "Candy6", PickupText := "Legendary Candy", Chance := C5, MinCount := big_number{Value:=1.0}, MaxCount := big_number{Value:=1.0}}}
    
    return drop_table{Items := Items}

(Tag:fetchable_tag).CandyChances():drop_table =
    case (Tag):
        fetchable_tag.zone1tag =>        MakeCandyTable(array{0.01, 0.00, 0.00,  0.00,  0.00,  0.00}) 
        fetchable_tag.zone2tag =>        MakeCandyTable(array{0.01, 0.00, 0.00,  0.00,  0.00,  0.00})

        fetchable_tag.beach_zone1tag =>        MakeCandyTable(array{0.05, 0.00, 0.00,  0.00,  0.00,  0.00})
        fetchable_tag.beach_zone2tag =>        MakeCandyTable(array{0.03, 0.02, 0.00,  0.00,  0.00,  0.00})
        fetchable_tag.beach_zone3tag =>        MakeCandyTable(array{0.00, 0.05, 0.00,  0.00,  0.00,  0.00})

        fetchable_tag.forest_zone1tag => MakeCandyTable(array{0.00, 0.03, 0.02, 0.0,  0.00,  0.00})
        fetchable_tag.forest_zone2tag => MakeCandyTable(array{0.00, 0.00, 0.05,  0.0, 0.00,  0.00})

        fetchable_tag.forest_bandit_zone1tag => MakeCandyTable(array{0.00, 0.00, 0.04,  0.01, 0.00,  0.00})
        fetchable_tag.forest_bandit_zone2tag => MakeCandyTable(array{0.00, 0.00, 0.03,  0.02, 0.00,  0.00})
        fetchable_tag.forest_bandit_zone3tag => MakeCandyTable(array{0.00, 0.00, 0.02,  0.03, 0.00,  0.00})
        
        fetchable_tag.giantlands_zone1tag => MakeCandyTable(array{0.00, 0.00, 0.01,  0.04, 0.0,  0.00})
        fetchable_tag.giantlands_zone2tag => MakeCandyTable(array{0.00, 0.00, 0.00,  0.05, 0.00,  0.00})
        fetchable_tag.giantlands_zone3tag => MakeCandyTable(array{0.00, 0.00, 0.00,  0.04, 0.01,  0.00})

        fetchable_tag.pirates_zone1tag => MakeCandyTable(array{0.00, 0.00, 0.00,  0.03, 0.02,  0.00})
        fetchable_tag.pirates_zone2tag => MakeCandyTable(array{0.00, 0.00, 0.00,  0.02, 0.03,  0.00})
        fetchable_tag.pirates_zone3tag => MakeCandyTable(array{0.00, 0.00, 0.00,  0.01, 0.04,  0.0})
        fetchable_tag.pirates_zone4tag => MakeCandyTable(array{0.00, 0.00, 0.00,  0.00, 0.05,  0.00})

        fetchable_tag.forest_zone3tag => MakeCandyTable(array{0.00, 0.00, 0.05,  0.01,  0.0001, 0.00})         
        _ => MakeCandyTable(array{0.0, 0.0, 0.0, 0.0, 0.0, 0.0})
