
using { /Fortnite.com/Devices }
using { /Verse.org/Simulation }
using { /UnrealEngine.com/Temporary/Diagnostics }
using { /Verse.org/Assets }
using { /UnrealEngine.com/Temporary/UI}
using { /Fortnite.com/UI }
using { /UnrealEngine.com/Temporary/SpatialMath }
using { /Verse.org/Colors }
using { /Fortnite.com/Characters }

# See https://dev.epicgames.com/documentation/en-us/uefn/create-your-own-device-in-verse for how to create a verse device.

teleport_info :=class<concrete>:
    @editable
    Teleporter : teleporter_device = teleporter_device{}
    @editable
    Name : string = "[...]"
    @editable
    Icon : string = "Portal"

binding_stone :=class<concrete>:
    @editable
    Teleporter : teleporter_device = teleporter_device{}
    @editable
    Button : button_device = button_device{}
    @editable
    Name : string = "[...]"

# A Verse-authored creative device that can be placed in a level
teleport_manager_device := class(creative_device, menu_item):

    @editable
    GameManager : ?game_manager = false

    @editable
    Teleporters : []teleport_info = array{}

    @editable
    BindingStones : []binding_stone = array{}

    OnBegin<override>():void =        
        for (BindingStone : BindingStones):
            BindingStone.Button.InteractedWithEvent.SubscribeAgent(OnBinding, BindingStone.Name)

    OnBinding(InteractAgent:agent, Name : string):void=  
        var TeleporterID : int = -1
        for (Teleporter : Teleporters):
            set TeleporterID += 1
            if:
                Teleporter.Name = Name
                Player := player[InteractAgent] 
                PlayerManager := GameManager?.PlayerManagers[InteractAgent]
            then:
                PlayerManager.UnlockTeleporter(TeleporterID, Name)
                return

    TeleportAgent(Agent : agent, TeleporterID : int) : void =
        if:
            Teleporter := Teleporters[TeleporterID].Teleporter
        then:
            Teleporter.Teleport(Agent)
    
    OpenMenu<override>(Agent:agent):void=
        if (PlayerManager := GameManager?.PlayerManagers[Agent]):
            
            if (not GameManager?.PlayerManagers[Agent].UIManager?.Popup?):
                if (GM := GameManager?):
                    MakeTeleportMenu(GM, Self, Agent).OpenMenu()
            else:
                block:#Log("DOUBLE MENU")

    GetTitle<override>():string=
        return "TELEPORT"

    GetIcon<override>():texture=
        return Images.Portal

    IsEnabled<override>(PlayerManager : player_manager)<transacts>:logic=
        if:
            PlayerManager.GetAchievementLevel["InteractedWithNPC_Samira"] > 0
        then:
            return true 

        if (PlayerManager.GetPrestigeCount() > 0):
            return true
        
        return false

    GetTPIcon(TeleporterID : int) : texture = 
        if (Temp := Teleporters[TeleporterID].Icon):
            return GetTextureFromString(Temp)
        else:
            return Images.Portal

    GetTPName(TeleporterID : int) : string = 
        if (Temp := Teleporters[TeleporterID].Name):
            return Temp
        else:
            return "TP #{TeleporterID}"

MakeTeleportMenu<constructor>(GameManagerIn : game_manager, TeleportManagerIn : teleport_manager_device, AgentIn : agent) := teleport_menu: 
    GameManager :=  GameManagerIn
    TeleportManager := TeleportManagerIn 
    Agent := AgentIn

teleport_menu:= class:
    TeleportManager : teleport_manager_device
    GameManager : game_manager
    Agent : agent

    var DialogBackground : texture_block = texture_block{ DefaultImage := Images.DialogBox, DefaultDesiredSize := vector2{X := 1920.0, Y := 1080.0}} 
    #var HeaderPlate : texture_block = texture_block{ DefaultImage := Images.DialogBox, DefaultDesiredSize := vector2{X := 400.0, Y := 80.0}}

    
    OpenMenu():void=
        if:
            InPlayer := player[Agent]
            InPlayer.IsActive[]
            PlayerUI := GetPlayerUI[InPlayer]
            PlayerManager := GameManager.PlayerManagers[InPlayer]
        then:
            StackBoxOuter := stack_box{Orientation:=orientation.Vertical}
            var CurrentStackBox : stack_box = stack_box{Orientation:=orientation.Horizontal}
            StackBoxOuter.AddWidget(stack_box_slot:
                Widget:= CurrentStackBox
                Padding:= margin{Top := 0.0, Left := 0.0, Right := 0.0, Bottom := 00.0}
            )

            CloseButton := button_quiet:
            CloseButton.SetText(STM("            BACK            "))            
            CloseButton.OnClick().Subscribe(PlayerManager.ClosePopup)

            Canvas := canvas:
                Slots := array:
                    canvas_slot:
                        Widget:= DialogBackground
                        SizeToContent := true
                        Anchors := anchors{ Minimum := vector2{X := 0.1, Y := 0.1}, Maximum := vector2{X := 0.9, Y := 0.9}}  
                        ZOrder := 1
                    canvas_slot:
                        Widget:= text_block:
                            DefaultTextColor:=NamedColors.White
                            DefaultShadowColor:=NamedColors.Black   
                            DefaultShadowOffset:=option{vector2{X:=2.0,Y:=2.0}}
                            DefaultShadowOpacity:= 1.0
                            DefaultText:= STM("Choose a location to teleport!")
                        SizeToContent := true
                        Anchors := anchors{ Minimum := vector2{X := 0.5, Y := 0.15}, Maximum := vector2{X := 0.5, Y := 0.15}}
                        Offsets := margin{Top := 40.0, Left := 0.0, Right := 0.0, Bottom := 0.0}
                        Alignment := vector2{X := 0.5, Y := 0.5}
                        ZOrder := 2
                    canvas_slot:
                        Widget:= StackBoxOuter
                        SizeToContent := true
                        Anchors := anchors{ Minimum := vector2{X := 0.5, Y := 0.3}, Maximum := vector2{X := 0.5, Y := 0.75}}  
                        Alignment := vector2{X := 0.5, Y := 0.0}
                        ZOrder := 3
                    canvas_slot:
                        Widget:= CloseButton
                        Anchors := anchors{ Minimum := vector2{X := 0.5, Y := 0.825}, Maximum := vector2{X := 0.5, Y := 0.825}}
                        Alignment := vector2{X := 0.5, Y := 1.0}
                        ZOrder := 4
                      
            var ItemCount : int = 0
            var TeleporterID : int = 0
            for (Teleporter : TeleportManager.Teleporters):
                if (PlayerManager.GetAchievementLevel[Teleporter.Name] > 0):
                    if (ItemCount >= 5):
                        set CurrentStackBox = stack_box{Orientation:=orientation.Horizontal}
                        StackBoxOuter.AddWidget(stack_box_slot:
                            Widget:= CurrentStackBox
                            Padding:= margin{Top := 0.0, Left := 0.0, Right := 0.0, Bottom := 00.0}
                        )
                        set ItemCount = 0
                    set ItemCount += 1

                    ItemIcon := texture_block:
                                DefaultImage := TeleportManager.GetTPIcon(TeleporterID)
                                DefaultDesiredSize := vector2{X := 200.0, Y := 200.0}
            
                    NameButton := button_quiet:
                        DefaultText:= STM("{TeleportManager.GetTPName(TeleporterID)}")

                    NameButton.OnClick().Subscribe(teleport_wrapper{Teleporter := Teleporter.Teleporter, PlayerManager:= PlayerManager}.Teleport)

                    Widget := overlay:
                        Slots:= array:
                            overlay_slot:
                                Widget:=ItemIcon
                                Padding:= margin{Top := 0.0, Left := 0.0, Right := 0.0, Bottom := 0.0}
                            overlay_slot:
                                Widget:=NameButton
                                Padding:= margin{Top := 60.0, Left := 10.0, Right := 0.0, Bottom := 0.0}
                                HorizontalAlignment:= horizontal_alignment.Left
                                VerticalAlignment:= vertical_alignment.Top
                    
                    WidgetSlot := stack_box_slot:
                        Widget:= Widget
                        Padding:= margin{Top := 0.0, Left := 2.5, Right := 0.0, Bottom := 00.0}

                    CurrentStackBox.AddWidget(WidgetSlot)

                set TeleporterID += 1
            
            PlayerUI.AddWidget(Canvas, player_ui_slot{InputMode := ui_input_mode.All})
            if (set PlayerManager.UIManager?.Popup = option{Canvas}) {}

teleport_wrapper := class():
    Teleporter : teleporter_device
    PlayerManager : player_manager
    BossType : boss_enum = boss_enum.None
    
    Teleport(Wdata:widget_message):void=
        Teleport_NoWData(Wdata.Player)

    Teleport_NoWData(Player:player):logic=
        var DidTeleport : logic = false
        if (not BossType = boss_enum.None):

            HasKey := BossType.CheckHasKey(PlayerManager, ?Spend := false)
            if (HasKey?):
                
                PlayerManager.Heal()
                Teleporter.Teleport(Player)
                set DidTeleport = true
                PlayerManager.UpdateQuestProgressForTask(task_event.UseTeleport, big_number{Value:= 1.0}, false, ?IsPersistent := true)
        
                #Close Menu
                PlayerManager.ClosePopupUI()
        else:
            Teleporter.Teleport(Player)
            set DidTeleport = true
            PlayerManager.UpdateQuestProgressForTask(task_event.UseTeleport, big_number{Value:= 1.0}, false, ?IsPersistent := true)
    
            #Close Menu
            if (SetUIManager := PlayerManager.UIManager?):
                SetUIManager.ClosePopup()
        
        return DidTeleport