using { /Fortnite.com/UI }
using { /UnrealEngine.com/Temporary/UI }
using { /Verse.org/Assets }
using { /UnrealEngine.com/Temporary/SpatialMath }
using { /Verse.org/Colors }
using { /Verse.org/Simulation }

ui_bounce_mover := class():
    HostWidget : overlay = overlay{}

    var FadeStart : float = 400.0
    var FadeEnd : float = 500.0

    var Texts : []text_block = array{}
    var Textures : []texture_block = array{}
    var Blocks : []color_block = array{}
    var Active : []logic = array{}
    var BounceActive : []logic = array{}

    var Activated : logic = false
    var DefaultSize : vector2 = vector2{X:=55.0, Y:=55.0}
    var BounceRatio : float = 2.0
    var DecayRatio : float = 0.98

    Setup(Count : int, ImageSize : vector2, BounceRatioIn : float, DecayRatioIn : float):overlay=
        set DefaultSize = ImageSize
        set BounceRatio = BounceRatioIn
        set DecayRatio = DecayRatioIn
        for (X:= 1..Count):
            var TempText : text_block = text_block:
                DefaultTextColor:=NamedColors.White,
                DefaultShadowColor:=NamedColors.Black,
                DefaultShadowOffset:=option{vector2{X:=2.0,Y:=2.0}}
            TempText.SetTextOpacity(0.0) 
            TempText.SetShadowOpacity(0.0) 

            var TempImg : texture_block = texture_block:
                DefaultImage := Images.Empty
                DefaultDesiredSize := DefaultSize

            var TempBlock : color_block = color_block{DefaultColor:=NamedColors.Blue}
            # TempBlock.SetOpacity(0.5)
            TempBlock.SetDesiredSize(vector2{Y:=3000.0})

            set Texts = Texts + array{TempText}
            set Textures = Textures + array{TempImg}
            set Blocks = Blocks + array{TempBlock}
            set Active = Active + array{false}
            set BounceActive = BounceActive + array{false}

            TempOverlay:=overlay_slot:
                Widget:=overlay:
                    Slots:=
                        array:
                            overlay_slot:
                                Widget:=TempBlock 
                                VerticalAlignment:= vertical_alignment.Bottom
                                HorizontalAlignment:= horizontal_alignment.Left
                            overlay_slot:
                                Widget:= stack_box:
                                    Orientation:= orientation.Horizontal
                                    Slots:= array:
                                        stack_box_slot:
                                            Widget:= TempImg
                                            HorizontalAlignment:= horizontal_alignment.Left
                                            Padding:= margin{Left:=2.0, Top:=0.0, Bottom:= 0.0, Right:= 4.0} 
                                        stack_box_slot:
                                            Widget:= TempText
                                            Padding:= margin{Left:=0.0, Top:=0.0, Bottom:= 0.0, Right:= 8.0} 
    
                                VerticalAlignment:= vertical_alignment.Top
                                HorizontalAlignment:= horizontal_alignment.Left
                HorizontalAlignment:= horizontal_alignment.Left

            HostWidget.AddWidget(TempOverlay)

        spawn:
            Activate()

        return HostWidget

    Deactivate():void=
        set Activated = false

    Activate()<suspends>:void =
        if (Activated?):
            return
        set Activated = true

        loop:
            Sleep(0.01)
            if (not HostWidget.GetRootWidget[]):
                set Activated = false
            if (not Activated?): 
                return

            TickBounces()
            TickMovements()

    Create(BounceImage: texture, Text: string):void=
        if:
            Index := Active.Find[false]
            set Active[Index] = true
            set BounceActive[Index] = true
            TextBlock := Texts[Index]
            Texture := Textures[Index]
            Block := Blocks[Index]
        then:
            Block.SetDesiredSize(vector2{Y:=32.0})

            TextBlock.SetText(STM(Text))
            TextBlock.SetTextOpacity(1.0)
            TextBlock.SetShadowOpacity(1.0)

            Texture.SetImage(BounceImage)
            Texture.SetDesiredSize(vector2{X:= DefaultSize.X * BounceRatio, Y:= DefaultSize.Y * BounceRatio})

            Block.SetVisibility(widget_visibility.Visible)
            TextBlock.SetVisibility(widget_visibility.Visible)
            Texture.SetVisibility(widget_visibility.Visible)

    TickMovements()<suspends>:void =
        for (i:= 0..Active.Length-1; Active[i]?; Text := Texts[i]; Texture := Textures[i]; Block := Blocks[i]):
            BlockY : float = Block.GetDesiredSize().Y
            if(BlockY > FadeStart):
                if:
                    OpacityRatio := (BlockY - FadeStart) / (FadeEnd - FadeStart)
                    TargetOpacity := type{_X:float where 0.000000 <= _X, _X <= 1.000000}[Clamp(1.0-OpacityRatio, 0.0, 1.0)]
                then:
                    Text.SetTextOpacity(TargetOpacity)
                
                if(BlockY >= FadeEnd):
                    Block.SetDesiredSize(vector2{X:=Block.GetDesiredSize().X,Y:=0.0})
                    Block.SetVisibility(widget_visibility.Collapsed)
                    Text.SetVisibility(widget_visibility.Collapsed)
                    Texture.SetVisibility(widget_visibility.Collapsed)
                    if (set Active[i] = false):
            else:
                Text.SetTextOpacity(1.0)

            BlockPos := Block.GetDesiredSize()
            Block.SetDesiredSize(BlockPos + vector2{X:=0.0, Y:=8.0})

    TickBounces()<suspends>:void =
        for (i:= 0..Active.Length-1; BounceActive[i]?; Text := Texts[i]; Texture := Textures[i]; Block := Blocks[i]):
            Ratio := Texture.GetDesiredSize().X / DefaultSize.X * DecayRatio
            if (Ratio > 1.0):
                Texture.SetDesiredSize(DefaultSize * Ratio)
            else:
                Texture.SetDesiredSize(DefaultSize)