using { /Fortnite.com/Devices }
using { /Fortnite.com/UI }
using { /UnrealEngine.com/Temporary/SpatialMath }
using { /UnrealEngine.com/Temporary/UI }
using { /Verse.org/Assets}
using { /Verse.org/Colors }
using { /Verse.org/Random}
using { /Verse.org/Simulation }
using { /Verse.org/Colors }

#StringToMessage Type Conversion
STM<localizes>(value:string)<computes>:message="{value}"
GetPlayerName<localizes> (Agent:agent)<computes>: message = "{Agent}"
PlayerNameWithString<localizes> (Agent:agent, String:string)<computes>: message = "{Agent}{String}"

GetAtlasIDFromString(String : string):int=
    case (String):
        "Midas Coins" => 1          #0-1
        "gem" => 2                  #0-2
        "Gems" => 2                 
        "Beach Attack Plans" => 3
        "Plans" => 3
        "Candy1" => 4               #0-4
        "Candy2" => 5               #0-5
        "Candy3" => 6               #0-6
        "Candy4" => 7               #0-7
        "Candy5" => 8               #1-0
        "Candy6" => 9               #1-1
        "GoodFlower" => 10          #1-2 
        "Goodflower" => 10          
        "Vicious Treat" => 11
        "Giantflower" => 12         #1-4
        "BoarHide" => 13            #1-5
        "WolfLiver" => 14           #1-6
        "BoarTusk" => 15            #1-7
        "WolfFur" => 16             #2-0
        "WolfClaw" => 17            #2-1
        "NaturesEssence" => 18      #2-2
        "Coal" => 19                #2-3
        "UmbralOre" => 20           #2-4
        "Diamond" => 21             #2-5
        "TribalHeaddress" => 22     #2-6
        "TribalBeads" => 23         #2-7
        "AncientTotem" => 24        #3-0
        "RangeMedal" => 25          #3-1
        "KikoLuckyStick" => 26      #3-2
        "EssenceHourglass" => 27    #3-3
        "Helmet" => 28              #3-4
        "Body" => 29                #3-5
        "Legs" => 30                #3-6
        "Boots" => 31               #3-7
        "Glove" => 32               #4-0
        "Ring" => 33                #4-1
        
        "Bomb" => 34                #4-2
        "KeyLower" => 35            #4-3
        "KeyUpper" => 36            #4-4
        "Garlic" => 37              #4-5
        "WoodStake" => 38           #4-6
        "SilverCross" => 39         #4-7
        "BlueKey" => 40             #5-0
        "GreenKey" => 41            #5-1
        "RedKey" => 42              #5-2
        "ForgeTongs" => 43          #5-3
        "ForgeApron" => 44          #5-4
        "ForgeHammer" => 45         #5-5
        "VampireOre" => 46          #5-6
        "MithrilOre" => 47          #5-7
        "TungstenOre" => 48         #6-0
        "SilverOre" => 49           #6-1

        "BigBomb" => 50             #6-2
        "PirateMap" => 51           #6-3
        "PiratePlans" => 52         #6-4
        "Cannonball" => 53          #6-5
        "Swimming Manual" => 54     #6-6
        "PirateDecoder" => 55       #6-7

        "Gunpowder" => 56           #7-0
        "KeyCard" => 57             #7-1
        "KeySharkey" => 58          #7-2
        "KeyGenga" => 59            #7-3
        "KeyUrjent" => 60           #7-4
        "KeyGiant" => 61            #7-5
        "KeyPirateKing" => 62       #7-6

        
        _ => 0                      #0-0


GetTextureFromString(String : string):texture=
    case (String):            
        #Abilities
        "LOCKED" => Images.QuestionMark
        "stash" => Images.Chest
        "DailyDamageMult" => Images.QuestionMark
        "Healingwinds" => Images.Healingwinds
        "Focus Fire" => Images.FocusFire
        "Momentum" => Images.Momentum
        "Teamwork" => Images.Teamwork
        "Handcannon" => Images.Handcannon
        "Ambush" => Images.WolfClaw
        "Inventory" => Images.Backpack
        "BossKeyCap" => Images.BossKeyChest
        "Stealth Attack" => Images.Stealth_Attack
        "Sliding Rifle" => Images.Sliding_Rifle
        "SuperHandcannon" => Images.SuperHandcannon

        #NPCS
        "Hawk" => Images.Hawk
        "LonelyGiant" => Images.LonelyGiant
        "Lonely Giant" => Images.LonelyGiant
        "Candy" => Images.CandyNPCIcon
        "CandyNPCIcon" => Images.CandyNPCIcon
        "Lyra" => Images.LyraIcon
        "LyraIcon" => Images.LyraIcon
        "Midas" => Images.MidasProfileIcon
        "MidasIcon" => Images.MidasProfileIcon
        "Mel" => Images.MelIcon 
        "MelIcon" => Images.MelIcon
        "Germit" => Images.GermitIcon
        "GermitIcon" => Images.GermitIcon
        "Gleeb" => Images.GleebProfile
        "Globb" => Images.GlobProfile
        "Starchild" => Images.StarchildIcon
        "StarchildIcon" => Images.StarchildIcon
        "Odysseus" => Images.OdyIcon
        "OdysseusIcon" => Images.OdyIcon
        "Samira" => Images.SamiraIcon
        "SamiraIcon" => Images.SamiraIcon
        "Midas Lieutenant" => Images.MidasLieutenantIcon
        "Midas LieutenantIcon" => Images.MidasLieutenantIcon
        "Daily Harvest" => Images.DailyHarvest
        "Bogstick" => Images.Bogstick
        "BeachInvader" => Images.BeachInvader
        "Funguy" => Images.Funguy
        "Sharkey" => Images.Sharkey
        "Blossom" => Images.BlossomIcon
        "BlossomIcon" => Images.BlossomIcon
        "ForestBanditIcon" => Images.ForestBanditIcon
        "Sarge" => Images.SargeIcon
        "SargeIcon" => Images.SargeIcon
        "Genga" => Images.GengaIcon
        "GengaIcon" => Images.GengaIcon
        "Barbariana" => Images.BarbarianaIcon
        "BarbarianaIcon" => Images.BarbarianaIcon
        "MidasOrb" => Images.MidasOrb
        "VampireHunterG" => Images.VampireHunterGIcon
        "Vampire Hunter G" => Images.VampireHunterGIcon
        "Stinky Pete" => Images.StinkyPete
        "StinkyPeteIcon" => Images.StinkyPete
        "Urjent" => Images.Urjent
        "UrjentIcon" => Images.Urjent
        "Charon" => Images.Charon
        "CharonIcon" => Images.Charon

        #GUNS
        "Remote" => Images.Remote
        "Gun_0" => Images.Remote
        "Gun_0Fixed" => Images.Remote
        "Icon" => Images.Icon
        "Gun_1" => Images.SuppressedPistol
        "Gun_2" => Images.SuppressedRifle
        "Gun_3" => Images.Gun_RocketLauncher
        "Gun_4" => Images.Gun_Shotgun
        "Gun_5" => Images.Gun_Bow
        "Gun_6" => Images.Gun_Railgun
        "Gun_7" => Images.Minigun
        "Gun_8" => Images.Hand_Cannon
        "Gun_9" => Images.Gun_Stake
        "Gun_10" => Images.Gun_Scythe
        
        #LOOT
        "Midas Coins" => Images.MidasCoins
        "gem" => Images.Gem2
        "Gems" => Images.Gem2
        "Candy1" => Images.Candy1
        "Candy2" => Images.Candy2
        "Candy3" => Images.Candy3
        "Candy4" => Images.Candy4
        "Candy5" => Images.Candy5
        "Candy6" => Images.Candy6
        "GoodFlower" => Images.GoodFlower
        "Goodflower" => Images.GoodFlower
        "Giantflower" => Images.Giantflower
        "BoarHide" => Images.BoarHide
        "WolfLiver" => Images.BoarLiver
        "WolfFur" => Images.WolfFur
        "WolfClaw" => Images.WolfClaw
        "RangeMedal" => Images.RangeMedal
        "BossKey" => Images.BossKey
        "Peely" => Images.PeelyDoll
        "UpgradeHammer" => Images.UpgradeHammer
        "Vicious Treat" => Images.SharkeyFood
        "Bomb" => Images.Bomb
        "KeyLower" => Images.KeyLower
        "KeyUpper" => Images.KeyUpper
        "Garlic" => Images.Garlic
        "WoodStake" => Images.WoodStake
        "SilverCross" => Images.SilverCross
        "BlueKey" => Images.BlueKey
        "RedKey" => Images.RedKey
        "GreenKey" => Images.GreenKey
        "ForgeHammer" => Images.ForgeHammer
        "ForgeApron" => Images.ForgeApron
        "ForgeTongs" => Images.ForgeTongs
        "VampireOre" => Images.VampireOre
        "MithrilOre" => Images.MithrilOre
        "TungstenOre" => Images.TungstenOre
        "SilverOre" => Images.SilverOre
        "BigBomb" => Images.BigBomb
        "PirateMap" => Images.PirateMap
        "PiratePlans" => Images.PiratePlans
        "Cannonball" => Images.Canonballs
        "Swimming Manual" => Images.SwimmingManual
        "PirateDecoder" => Images.PirateDecoder
        "Gunpowder" => Images.Gunpowder
        "KeyCard" => Images.KeyCard
        "KeySharkey" => Images.KeySharkey
        "KeyGenga" => Images.KeyGenga
        "KeyUrjent" => Images.KeyUrjent
        "KeyGiant" => Images.KeyGiant
        "KeyPirateKing" => Images.KeyPirateKing
        "TribalHeaddress" => Images.TribalHeaddress
        "PrestigeGem" => Images.PrestigeGem
        "P_PrestigeGem" => Images.PrestigeGem


        #Equipment
        "Helmet" => Images.Helm
        "Body" => Images.Body
        "Legs" => Images.Legs
        "Boots" => Images.Boots
        "Glove" => Images.Glove
        "Ring" => Images.Ring
        
        #Other
        "Plans" => Images.PiratePlans
        "Beach Attack Plans" => Images.PiratePlans
        "TownMeteor" => Images.TownMeteor
        "headshot" => Images.Crit
        "QuestQuestion" => Images.QuestQuestion
        "Inn" => Images.Inn
        "Beach" => Images.Beach
        "Portal" => Images.Portal
        "Bosses" => Images.BossTeleport
        "essence" => Images.Essence        
        "level" => Images.Levelup
        "TickingClock" => Images.TickingClock
        "strength" => Images.StrengthIcon
        "UpgradeIcon" => Images.UpgradeIcon
        "BoarExample" => Images.BushrangerIcon
        "WolfExample" => Images.QuarryWolfExample
        "Boar Area" => Images.BushrangerIcon
        "GiantlandsTeleport" => Images.GiantlandsTeleport
        "BossPortal" => Images.BossPortal
        

        #COUNTDOWN
        "Countdown10" => Images.Countdown10
        "Countdown9" => Images.Countdown9
        "Countdown8" => Images.Countdown8
        "Countdown7" => Images.Countdown7
        "Countdown6" => Images.Countdown6
        "Countdown5" => Images.Countdown5
        "Countdown4" => Images.Countdown4
        "Countdown3" => Images.Countdown3
        "Countdown2" => Images.Countdown2
        "Countdown1" => Images.Countdown1
        "CountdownFinish" => Images.CountdownFinish
        _ => Images.Empty
        
ParseText(Text : string, ?ImgSize : vector2 = vector2{X := 40.0, Y := 40.0}, ?Color : color = NamedColors.White, ?ShadowColor : color = NamedColors.Black) : []stack_box_slot =
    
    var StackBoxSlots : []stack_box_slot = array{}    

    var WorkingString : string = ""
    var Index : int = 0 

    for (Char : Text):
        if("{Char}" = "<"):            
            #CreateTextBlock
            TempSlot := stack_box_slot:
                Widget:= CreateTextBlock(WorkingString, Color, ShadowColor) 
                VerticalAlignment:= vertical_alignment.Top

            set StackBoxSlots = StackBoxSlots + array{TempSlot}

            #Reset the WorkingString
            set WorkingString = ""

        if("{Char}" = ">"):
            if(WorkingString = "br"):
                #Signal Linebreak
                set WorkingString = "<br>"
                
                #CreateTextBlock
                TempSlot := stack_box_slot:
                    Widget:= CreateTextBlock(WorkingString, Color, ShadowColor) 
                    VerticalAlignment:= vertical_alignment.Top

                set StackBoxSlots = StackBoxSlots + array{TempSlot}
            else:
                #GetTextureBlock
                Texture := GetTextureFromString(WorkingString)
                TempTextureBlock := texture_block{ DefaultImage := Texture, DefaultDesiredSize := ImgSize} 

                TempSlot := stack_box_slot:
                    Widget:= TempTextureBlock
                    VerticalAlignment:= vertical_alignment.Top

                set StackBoxSlots = StackBoxSlots + array{TempSlot}

            #Reset the WorkingString
            set WorkingString = ""

        if(not("{Char}" = ">" or "{Char}" = "<")):
            set WorkingString = WorkingString + array{Char}

        set Index += 1

        if(Index = Text.Length):
            #CreateTextBlock
            TempSlot := stack_box_slot:
                Widget:= CreateTextBlock(WorkingString, Color, ShadowColor)
                VerticalAlignment:= vertical_alignment.Top

            set StackBoxSlots = StackBoxSlots + array{TempSlot}
            
            #Reset the WorkingString
            set WorkingString = ""
        
    return StackBoxSlots

CreateTextBlock(String : string, Color : color, ShadowColor : color): text_block=
    TempText := text_block{DefaultTextColor:=Color,
    DefaultShadowColor:=ShadowColor,
    DefaultShadowOffset:=option{vector2{X:=2.0,Y:=2.0}}}
    TempText.SetShadowOpacity(1.0)    
    TempText.SetText(STM(String))
    return TempText

CreateTextBlock(String : string): text_block=
    return CreateTextBlock(String, NamedColors.White, NamedColors.Black)

MakeWidgetGrid(Widgets:[]widget, GridWidth:int, Spacing:vector2):stack_box=
    Container := stack_box{Orientation:=orientation.Vertical}
    var InnerContainer : stack_box = stack_box{Orientation:=orientation.Horizontal}
    var InnerContainerHasItems : logic= false

    for (Index := 0..Widgets.Length-1):
        if (SetWidget := Widgets[Index]):
            InnerContainer.AddWidget(stack_box_slot:
                Widget:= SetWidget
                Padding:= margin{Left:=0.0, Top:=0.0, Bottom:= 0.0, Right:= Spacing.X} 
            )
            set InnerContainerHasItems = true
            
            if (Mod[Index, GridWidth] = GridWidth-1):  
                Container.AddWidget(stack_box_slot:
                    Widget:=InnerContainer
                    Padding:= margin{Left:=0.0, Top:=Spacing.Y, Bottom:= 0.0, Right:= 0.0} 
                )
                set InnerContainer = stack_box{Orientation:=orientation.Horizontal}
                set InnerContainerHasItems = false
    
    if (InnerContainerHasItems?):
        Container.AddWidget(stack_box_slot:
            Widget:=InnerContainer
            Padding:= margin{Left:=Spacing.X, Top:=Spacing.Y, Bottom:= 0.0, Right:= 0.0} 
        )
    
    
    return Container

ui_manager:=class<unique>:
    Player:player
    PlayerManager : player_manager
    
    ImgMidasEssence : texture = Images.Essence
    ImgStash : texture = Images.Chest
    ImgLevelUp : texture = Images.Levelup
    ImgCrit : texture = Images.Crit
    ImgEmpty : texture = Images.Empty
    ImgGem  : texture = Images.Gem2

    var ActiveTarget : string = ""
    var MyCanvas : canvas = canvas{}

    #popup ui
    var Popup : ?widget = false

    #canvas is your main layer
    var MainCanvas : canvas = canvas{}
    
    var LatestBubble : texture_block = texture_block{DefaultImage:= Images.Empty} 
    var LatestIcon : texture_block = texture_block{DefaultImage:= Images.Empty} 
    var LatestText : stack_box = stack_box{Orientation:=orientation.Vertical}
    var LatestVO : audio_player_device = audio_player_device{}

    var GuardWidget : widget = overlay{}
    var GuardHealth : progress_bar = progress_bar{}
    
    var GuardHealthText : text_block = text_block{}
    var GuardName : text_block  = text_block{}
    var GuardArmorImgWidget : texture_block = texture_block{ DefaultImage := Images.Empty, DefaultDesiredSize := vector2{X := 40.0, Y := 40.0}}
    var GuardTeamworkImgWidget : texture_block = texture_block{ DefaultImage := Images.Empty, DefaultDesiredSize := vector2{X := 40.0, Y := 40.0}}
    var GuardKOBonusImgWidget : texture_block = texture_block{ DefaultImage := Images.Empty, DefaultDesiredSize := vector2{X := 40.0, Y := 40.0}}
    var GuardClearBonusImgWidget : texture_block = texture_block{ DefaultImage := Images.Empty, DefaultDesiredSize := vector2{X := 40.0, Y := 40.0}}
    var GuardArmorText : text_block  = text_block{}
    var GuardTeamworkText : text_block  = text_block{}
    var AttackingPlayerCount : int = -1
    var GuardKOBonusText : text_block  = text_block{}
    var GuardClearBonusText : text_block  = text_block{}
    var LastKnownKOBonus : big_number = big_number{}
    var LastKnownClearBonus : big_number = big_number{}
    var GuardVisibilityTimer : float = 0.0

    var DailyBonusWidget : overlay = overlay{}
    DailyBonusBackground : texture_block  = texture_block{DefaultImage:= Images.StoreBox, DefaultDesiredSize := vector2{X:= 410.0, Y:= 180.0}}
    var DailyBonusHeaderText : text_block = text_block{DefaultShadowOpacity := 1.0, DefaultTextColor:= NamedColors.Goldenrod, DefaultShadowColor:=NamedColors.Black, DefaultShadowOffset:=option{vector2{X:=2.0,Y:=2.0}}}
    var DailyBonusWeaponText : text_block = text_block{DefaultShadowOpacity := 1.0, DefaultTextColor:= NamedColors.Goldenrod, DefaultShadowColor:=NamedColors.Black, DefaultShadowOffset:=option{vector2{X:=2.0,Y:=2.0}}}
    DailyBonusWeaponIcon : texture_block = texture_block {DefaultImage:= Images.Empty, DefaultDesiredSize :=  vector2{X := 80.0, Y := 80.0}}
    DailyBonusBossIcon : texture_block = texture_block {DefaultImage:= Images.Empty, DefaultDesiredSize :=  vector2{X := 80.0, Y := 80.0}}

    TargetIcon : texture_block = texture_block {DefaultImage:= Images.Empty, DefaultDesiredSize :=  vector2{X := 80.0, Y := 80.0}}
    TargetIconDefaultSize : vector2 = vector2{X := 80.0, Y := 80.0}

    GuardIcon : texture_block = texture_block {DefaultImage:= Images.Empty, DefaultDesiredSize :=  vector2{X := 40.0, Y := 40.0}}
    GuardIconDefaultSize : vector2 = vector2{X := 40.0, Y := 40.0}
    BuffIconDefaultSize : vector2 = vector2{X := 80.0, Y := 80.0}

    var BossNotificationWidget : widget = overlay{}
    var BossBounceColorWidget : color_block = color_block{}
    var IsBossNotifBouncingLeft : logic = true
    var BossArrowIcon : texture_block = texture_block{ DefaultImage := Images.Empty, DefaultDesiredSize := vector2{X := 150.0, Y := 150.0}}
    var BossRemoteIcon : texture_block = texture_block{ DefaultImage := Images.Empty, DefaultDesiredSize := vector2{X := 80.0, Y := 80.0}}
    var BossNotificationIcon : texture_block = texture_block{ DefaultImage := Images.Empty, DefaultDesiredSize := vector2{X := 80.0, Y := 80.0}}

    # var BossDurationBar : progress_bar = progress_bar{}

    var BossWidget : widget = overlay{}
    var BossHealth : progress_bar = progress_bar{}
    var BossIcon : texture_block = texture_block{ DefaultImage := Images.BossTeleport, DefaultDesiredSize := vector2{X := 80.0, Y := 80.0}}
    
    var BossHealthText : text_block = text_block{}
    var BossName : text_block  = text_block{}
    var BossTeamworkImgWidget : texture_block = texture_block{ DefaultImage := Images.Empty, DefaultDesiredSize := vector2{X := 40.0, Y := 40.0}}
    var BossArmorImgWidget : texture_block = texture_block{ DefaultImage := Images.Empty, DefaultDesiredSize := vector2{X := 40.0, Y := 40.0}}
    var BossKOBonusImgWidget : texture_block = texture_block{ DefaultImage := Images.Empty, DefaultDesiredSize := vector2{X := 40.0, Y := 40.0}}
    var BossClearBonusImgWidget : texture_block = texture_block{ DefaultImage := Images.Empty, DefaultDesiredSize := vector2{X := 40.0, Y := 40.0}}
    var BossArmorText : text_block  = text_block{}
    var BossTeamworkText : text_block  = text_block{}
    var BossKOBonusText : text_block  = text_block{}
    var BossClearBonusText : text_block  = text_block{}

    var ZoneTransitionWidget : color_block = color_block{}
    var ZoneWidget : widget = overlay{}
    var ZoneKOsProgress : progress_bar = progress_bar{}
    var ZoneKOsProgressText : text_block  = text_block{}
    var ZoneClearsProgress : progress_bar = progress_bar{}
    var ZoneClearsProgressText : text_block  = text_block{}
    var ClearsIcon : texture_block = texture_block{DefaultImage:= Images.Empty, DefaultDesiredSize := vector2{X := 28.0, Y := 40.0}}
    var NextGemStackBox : stack_box = stack_box{ Orientation := orientation.Horizontal}
    var NextGemSlotArray : []stack_box_slot = array{} 
    ClearsIconDefaultSize : vector2 = vector2{X := 28.0, Y := 40.0}

    var ZoneName : text_block  = text_block{}
    var ZoneLevelText : text_block  = text_block{}
    var ZoneOverkillText : text_block  = text_block{}
    var ZoneKOBonusText : text_block  = text_block{}
    var ZoneClearBonusText : text_block  = text_block{}

    var QueuedWeaponID : int = -1
    var QueuedWeaponLevel : int = -1
    var QueuedWeaponExperience : int = -1
    var QueuedWeaponExperienceRequirement : int = -1

    var QueuedUpgradeHammerProgress : int = -1
    var QueuedUpgradeHammerRequirement : int = -1

    var QueuedQuestProgress : []big_number = array{}
    var QueuedQuestRequirement : []big_number = array{}
    var QueuedQuestTextureString : []string = array{}

    var QueuedEssenceGain : big_number = big_number{}
    var QueuedKOBonusGain : big_number = big_number{}
    var QueuedZoneBonusGain : big_number = big_number{}
    var IsQueuedHitWeak : logic = false
    var IsQueuedHandcannon : logic = false
    var IsQueuedAmbush : logic = false
    var IsQueuedPoints : logic = false
    var IsQueuedHitBlocked : logic = false
    
    var QueuedGuardKiller : ?agent = false
    var QueuedZoneTag : fetchable_tag = fetchable_tag.none_tag
    var QueuedGuardsAlive : int = -1
    var QueuedMaxGuards : int = 0
    var QueuedPlayerClears : int = -1

    var QueuedGuardID : string = ""
    var QueuedOldHealth : big_number = big_number{}
    var QueuedCurrentHealth : big_number = big_number{}
    var QueuedMaxHealth : big_number = big_number{}
    var QueuedAttackingPlayerCount : int = -1
    var QueuedArmorValue : big_number = big_number{}
    var QueuedKOBonus : big_number = big_number{}
    var QueuedClearBonus : big_number = big_number{}
    var QueuedName : string = ""
    var QueuedIsUpdateFromUIOwner : logic = false
    var QueuedIsBoss : logic = false
    var QueuedIsDamageMastery : logic = false
    var QueuedDamageMasteryMult : float = 0.0

    var QueuedCollectionItem : []string = array{}
    var QueuedCollectionAmount : []big_number = array{}
    
    var StrengthFloatingText : ui_bounce_mover = ui_bounce_mover{}

    var StashPlayerNameHostWidget : overlay = overlay{}
    var StashIcon1: []texture_block  = array{}
    var PlayerName1: []text_block  = array{}
    var StashPlayerColors : []color_block  = array{}
    var CurrentStashPlayer : int = 0
    var ActiveStashPlayers : []int = array{}
    StashPlayerIconDefaultSize : vector2 = vector2{X := 60.0, Y := 60.0}

    var BuffHostWidget : stack_box = stack_box{ Orientation := orientation.Vertical}
    var BuffHostWidget2 : stack_box = stack_box{ Orientation := orientation.Vertical}
    var BuffSlot : [string]stack_box_slot = map{}
    var BuffIcon : [string]texture_block = map{}
    var BuffAmount : [string]text_block = map{}
    var BuffDuration : [string]text_block = map{}

    var Delay : float = 0.0
    var DamageNumbersHostWidget : overlay = overlay{}
    var DamageNumbers1: []text_block  = array{}
    var DamageNumbers2: []text_block  = array{}
    var DamageImgs1: []texture_block  = array{}
    var DamageImgs2: []texture_block  = array{}
    var DamageColors: []color_block  = array{}
    var CurrentDamageCounter : int = 0
    var ActiveNumbers : []int = array{}
    DamageNumbersSize : vector2 = vector2{X := 55.0, Y := 55.0}

    var CollectibleNumbers1: []text_block  = array{}
    var CollectibleImgs1: []texture_block  = array{}
    var CollectibleColors: []color_block  = array{}
    var CurrentCollectibleCounter : int = 0
    var ActiveCollectibleNumbers : []int = array{}
    
    IconBounce : float = 1.25
    GatherBounceIterations : float = 8.0
    var ActiveBounces : []texture_block = array{}
    var ActiveBounceSizes : []vector2 = array{}

    var MidasEssenceIcon1 : texture_block = texture_block{ DefaultImage:= Images.Essence, DefaultDesiredSize := vector2{X := 60.0, Y := 60.0}}
    var MidasEssenceIcon2 : texture_block = texture_block{ DefaultImage:= Images.Essence, DefaultDesiredSize := vector2{X := 60.0, Y := 60.0}}
    MidasEssenceIconDefaultSize : vector2 = vector2{X := 50.0, Y := 50.0}
    
    var StashNotificationWidget : widget = overlay{}
    var StashBounceColorWidget : color_block = color_block{}
    var IsStashNotifBouncingUp : logic = true
    var StashArrowIcon : texture_block = texture_block{ DefaultImage := Images.Empty, DefaultDesiredSize := vector2{X := 150.0, Y := 150.0}}
    var StashRemoteIcon : texture_block = texture_block{ DefaultImage := Images.Empty, DefaultDesiredSize := vector2{X := 80.0, Y := 80.0}}
    var StashNotificationIcon : texture_block = texture_block{ DefaultImage := Images.Empty, DefaultDesiredSize := vector2{X := 80.0, Y := 80.0}}

    var StashTransitionWidget : color_block = color_block{}
    var StashWidget : widget = overlay{}
    var StashIcon : texture_block = texture_block{ DefaultImage := Images.Chest, DefaultDesiredSize := vector2{X := 40.0, Y := 40.0}} 
    StashIconDefaultSize : vector2 = vector2{X := 40.0, Y := 40.0}
    var StashedAmount : big_number = big_number{}
    var StashedText : text_block = text_block{}
    var StashIncomeAmount : big_number = big_number{}
    var StashedIncomeText : text_block = text_block{}
    var StashIncomeTimer : int = 0

    var PeelyWidget : widget = overlay{}
    var PeelyIcon : texture_block = texture_block{ DefaultImage:= Images.PeelyDoll, DefaultDesiredSize := vector2{X := 40.0, Y := 40.0}}
    PeelyIconDefaultSize : vector2 = vector2{X := 40.0, Y := 40.0}
    var PeelyAmount : big_number = big_number{}
    var PeelyText : text_block = text_block{}

    var HammerWidget : widget = overlay{}
    var HammerIcon : texture_block = texture_block{ DefaultImage:= Images.Empty, DefaultDesiredSize := vector2{X := 40.0, Y := 40.0}}
    HammerIconDefaultSize : vector2 = vector2{X := 40.0, Y := 40.0}
    var HammerAmount : big_number = big_number{}
    var HammerText : text_block = text_block{}

    var KeyWidget : widget = overlay{}
    var KeyTransitionWidget : color_block = color_block{}
    var KeyIcon : texture_block = texture_block{ DefaultImage:= Images.BossKey, DefaultDesiredSize := vector2{X := 60.0, Y := 48.0}}
    KeyIconDefaultSize : vector2 = vector2{X := 60.0, Y := 48.0}
    var KeyAmount : big_number = big_number{}
    var KeyText : text_block = text_block{}
    var KeyTimerText : text_block = text_block{}

    var GemWidget : widget = overlay{}
    var GemIcon : texture_block = texture_block{ DefaultImage:= Images.Gem2, DefaultDesiredSize := vector2{X := 60.0, Y := 48.0}}
    GemIconDefaultSize : vector2 = vector2{X := 60.0, Y := 48.0}
    var GemAmount : big_number = big_number{}
    var GemText : text_block = text_block{}

    var CandyWidget : widget = overlay{}
    var CandyTransitionWidget : color_block = color_block{}
    var CandyIcon : texture_block = texture_block{ DefaultImage:= Images.CandyIcon, DefaultDesiredSize := vector2{X := 75.0, Y := 60.0}}
    CandyIconDefaultSize : vector2 = vector2{X := 75.0, Y := 60.0}
    var CandyAmount : big_number = big_number{}
    var CandyText : text_block = text_block{}

    var CoinsWidget : widget = overlay{}
    var CoinsIcon : texture_block = texture_block{ DefaultImage:= Images.MidasCoins, DefaultDesiredSize := vector2{X := 50.0, Y := 50.0}}
    CoinsIconDefaultSize : vector2 = vector2{X := 50.0, Y := 50.0}
    var CoinsAmount : big_number = big_number{}
    var CoinsText : text_block = text_block{}

    var EssenceText : text_block = text_block{}

    var EssenceProgressBar : progress_bar = progress_bar{}
    var ExpText : text_block = text_block{}
    var ProgressTextLeft : text_block = text_block{}
    var ProgressTextRight : text_block = text_block{}

    var GunIcon : texture_block = texture_block{ DefaultImage:= Images.StrengthIcon, DefaultDesiredSize := vector2{}}
    GunIconDefaultSize : vector2 = vector2{X := 60.0, Y := 50.0}
    var GunText : text_block = text_block{}
    var GunText2 : text_block = text_block{}
    var GunTransition : color_block = color_block:
        DefaultColor:= NamedColors.Black
        DefaultOpacity:= 0.0
        DefaultDesiredSize := vector2{X:=0.0, Y:= 175.0}     
        
    var GunDurationTimeLeft : int = 0
    var GunDuration : progress_bar = progress_bar{}
    var GunDurationIcon : texture_block = texture_block{ DefaultImage := Images.Empty, DefaultDesiredSize := vector2{X := 80.0, Y := 80.0}} 
    GunDurationIconDefaultSize : vector2 = vector2{X := 80.0, Y := 80.0}
    var GunDurationText : text_block = text_block{}
    var GunsAvailableStackbox : stack_box = stack_box{ Orientation := orientation.Horizontal}
    var GunsAvailableArray : []stack_box_slot = array{} 
    var GunDurationWidget : stack_box_slot = stack_box_slot{Widget:=overlay{}}

    var UpgradeHammerWidget : stack_box_slot = stack_box_slot{Widget:=overlay{}}
    var UpgradeHammerProgressText : text_block = text_block{}
    var UpgradeHammerProgressBar : progress_bar = progress_bar{}
    var UpgradeHammerIcon : texture_block = texture_block{ DefaultImage := Images.Empty, DefaultDesiredSize := vector2{X := 80.0, Y := 80.0}} 
    UpgradeHammerIconDefaultSize : vector2 = vector2{X := 80.0, Y := 80.0}

    var StrengthIcon : texture_block = texture_block{ DefaultImage:= Images.StrengthIcon, DefaultDesiredSize := vector2{X := 60.0, Y := 50.0}}
    var NextLevelStrengthIcon : texture_block = texture_block{ DefaultImage:= Images.StrengthIcon, DefaultDesiredSize := vector2{X := 40.0, Y := 38.0}}
    var NextLevelGemIcon : texture_block = texture_block{ DefaultImage:= Images.Empty, DefaultDesiredSize := vector2{X := 40.0, Y := 38.0}}
    StrengthIconDefaultSize : vector2 = vector2{X := 60.0, Y := 50.0}
    var NextLevelGemWidget : stack_box = stack_box{ Orientation := orientation.Horizontal} 
    NextLevelStrengthIconDefaultSize : vector2 = vector2{X := 40.0, Y := 38.0}
    var StrengthText : text_block = text_block{}
    var NextLevelGemText : text_block = text_block{}

    var LevelText : text_block = text_block{}
    LevelIconDefaultSize : vector2 = vector2{X := 60.0, Y := 60.0}

    var NextLevelText : text_block = text_block{}
    var NextLevelStrengthText : text_block = text_block{}

    var LevelNumbersHostWidget : overlay = overlay{}
    var LevelNumbers: []text_block  = array{}
    var LevelImgs: []texture_block  = array{}    
    var LevelColors: []color_block  = array{}
    var CurrentLevelCounter : int = 0
    var ActiveLevels : []int = array{}
    
    var QuestHorizontalStackBox1 : stack_box = stack_box{ Orientation := orientation.Horizontal}
    var QuestHorizontalStackBox2 : stack_box = stack_box{ Orientation := orientation.Horizontal}
    var QuestHorizontalStackBox3 : stack_box = stack_box{ Orientation := orientation.Horizontal}
    var IsNewQuestBouncing : logic = false

    var QuestVerticalStackBox : stack_box = stack_box{ Orientation := orientation.Vertical}
    var QuestSlotArray : []stack_box_slot = array{} 
    
    QuestCompleteImg : texture_block = texture_block{DefaultImage:= Images.Empty, DefaultDesiredSize := vector2{X:= 420.0, Y:= 240.0}}
    QuestHeaderImg : texture_block = texture_block{DefaultImage:= Images.BlankPopup, DefaultDesiredSize := vector2{X:= 420.0, Y:= 240.0}}
    QuestHeaderImgDefaultSize : vector2 = vector2{X:= 420.0, Y:= 240.0}
    var QuestHeaderText : text_block = text_block{}
    var QuestProgress : []progress_bar = array{}
    var QuestProgressText : []text_block = array{}
    var QuestIcons : []texture_block = array{}
    MaxQuestProgressBars : int = 3

    var ZoneOpenVerticalStackBox : stack_box = stack_box{ Orientation := orientation.Vertical}
    var ZoneOpenHorizontalStackBox1 : stack_box = stack_box{ Orientation := orientation.Horizontal}
    var ZoneOpenHorizontalStackBox2 : stack_box = stack_box{ Orientation := orientation.Horizontal}
    var ZoneOpenHorizontalStackBox3 : stack_box = stack_box{ Orientation := orientation.Horizontal}
    var ZoneOpenSlotArray : []stack_box_slot = array{} 

    var ZoneMsgVisibilityTimer : float = 0.0

    var DamageMeterWidget : overlay = overlay{}
    var DamageMeterHeader : text_block = text_block{}

    var QueuedDamageDealt : [agent]big_number = map{}

    var TickingClockImg : texture_block = texture_block{DefaultImage:= Images.Empty, DefaultDesiredSize := vector2{X := 80.0, Y := 80.0}}
    TickingClockImgDefaultSize : vector2 = vector2{X := 80.0, Y := 80.0}
    var TickingClockStackbox : stack_box = stack_box{Orientation := orientation.Horizontal}
    var TickingClockSlotArray : []stack_box_slot = array{} 

    var GunTickingClockStackbox : stack_box = stack_box{Orientation := orientation.Horizontal}
    var GunTickingClockSlotArray : []stack_box_slot = array{} 
    
    var HandCannonIcon : texture_block = texture_block{DefaultImage:= Images.Empty, DefaultDesiredSize := vector2{X := 40.0, Y := 40.0}}
    var SuperHandCannonIcon : texture_block = texture_block{DefaultImage:= Images.Empty, DefaultDesiredSize := vector2{X := 40.0, Y := 40.0}}
    var HandCannonProgresses : []progress_bar = array{}
    var SuperHandCannonProgresses : []progress_bar = array{}
    var HandCannonHostWidget : stack_box = stack_box{ Orientation := orientation.Vertical}

    #GunExpWidget
    var GunExpHostWidget : stack_box = stack_box{ Orientation := orientation.Vertical}
    var GunExpBG : texture_block = texture_block{DefaultImage:= Images.Empty, DefaultDesiredSize := vector2{X := 450.0, Y := 600.0}}
    var GunExpTitleImg : texture_block = texture_block{DefaultImage:= Images.Empty, DefaultDesiredSize := vector2{X := 360.0, Y := 180.0}}
    var GunExpBackPlateImg : texture_block = texture_block{DefaultImage:= Images.Empty, DefaultDesiredSize := vector2{X := 250.0, Y := 250.0}}
    var GunExpGunImg : texture_block = texture_block{DefaultImage:= Images.Empty, DefaultDesiredSize := vector2{X := 225.0, Y := 225.0}}
    var GunExpUpdateText : text_block = text_block{}
    var GunExpLevelText : text_block = text_block{}
    var GunExpProgressText : text_block = text_block{}
    var GunExpProgressBar : progress_bar = progress_bar{}
    var GunExpNextLevelText : text_block = text_block{}

    #PermaGunExpWidget
    var PermaGunExpWidget : stack_box_slot = stack_box_slot{ Widget:=overlay{}}
    var PermaGunExpProgressBar : progress_bar = progress_bar{}
    var PermaGunExpLevelText : text_block = text_block{}
    var PermaGunExpProgressText : text_block = text_block{}
    var PermaGunExpGunImg : texture_block = texture_block{DefaultImage:= Images.Empty, DefaultDesiredSize := vector2{X := 80.0, Y := 80.0}}
    
    #Functions Grouped by: (Ctrl+F'able)
        #Level Functions
        #Essence Functions
        #Stash Functions
        #Boss Functions
        #Guard Functions
        #Zone Functions
        #Item Funtions
        #Candy Functions
        #Gem Functions
        #Quest Functions
        #Damage Functions
        #Damage Mastery Functions
        #Buff Functions
        #Setup Functions  
        

    Init():void=   
        Log("Initializing Player UI!")
        SetupTextBlocks()
        SetupDailyBonusWidget()
        SetupDamageNumbers()
        SetupGunExpWidget()
        SetupPermaGunExpWidget()
        SetupCollectibleNumbers()
        SetupStashPlayers()
        SetupLevelNumbers()

        set MainCanvas = CreateUI()
        if(PlayerUI:=GetPlayerUI[Player]):
            PlayerUI.AddWidget(MainCanvas)
        UpdateEssenceUI()
        UpdateLevelUI()
        UpdateUICounts()

        spawn:
            LoopingSecondCheck()
        spawn:
            MoveUINumbers()

        spawn:
            StrengthFloatingText.Activate()

    CreateUI() : canvas =

        DailyBonusHeaderText.SetText(STM("Daily Bonuses:"))
        DailyWeaponID:= PlayerManager.GetDailyWeapon()
        DailyBonusWeaponIcon.SetImage(GetTextureFromString("Gun_{DailyWeaponID}"))
        DailyBossID:= PlayerManager.GetDailyBoss()
        if:
            DailyBossString:= PlayerManager.GameManager.BossTeleportManager.Teleporters[DailyBossID].TeleportInfo.Icon
        then:
            DailyBonusBossIcon.SetImage(GetTextureFromString(DailyBossString))

        set EssenceProgressBar = MakeProgressBar(0.0, vector2{X:= 760.0, Y:= 40.0}, MakeColorFromHex("0072fc"), MakeColorFromHex("000000"), MakeColorFromHex("FFFFFF"))
        MyExpWidget := EssenceProgressBar.GetWidget()
        
        MyQuestWidget := overlay{}

        for (X := 0..MaxQuestProgressBars-1):
            NewQuestProgress := MakeProgressBar(1.0, vector2{X:= 340.0, Y:= 40.0}, NamedColors.Goldenrod, MakeColorFromHex("000000"), MakeColorFromHex("FFFFFF"), ?InMaxColor:= option{NamedColors.Goldenrod})
            
            NewQuestProgressText := text_block{DefaultTextColor:=NamedColors.White,
                DefaultShadowColor:=NamedColors.Black,
                DefaultShadowOffset:=option{vector2{X:=2.0,Y:=2.0}}}

            NewQuestProgressText.SetShadowOpacity(1.0) 

            NewQuestIcon := texture_block{DefaultImage := Images.Empty, DefaultDesiredSize := vector2{X := 40.0, Y := 40.0}}

            set QuestProgress = QuestProgress + array{NewQuestProgress}
            set QuestProgressText = QuestProgressText + array{NewQuestProgressText}
            set QuestIcons = QuestIcons + array{NewQuestIcon}

            MyQuestWidget.AddWidget(overlay_slot:
                Widget:= NewQuestProgress.GetWidget()
                Padding:= margin{Left:= 0.0, Top:= 45.0*X, Bottom:= 0.0, Right:= 0.0}                                
            )
            MyQuestWidget.AddWidget(overlay_slot:
                Widget:= NewQuestProgressText
                Padding:= margin{Left:= 0.0, Top:= 45.0*X, Bottom:= 0.0, Right:= 0.0}                                
            )
            MyQuestWidget.AddWidget(overlay_slot:
                Widget:= NewQuestIcon
                Padding:= margin{Left:= -150.0, Top:= 45.0*X, Bottom:= 0.0, Right:= 0.0}                                
            )

        set NextLevelGemWidget = stack_box:
            Orientation:= orientation.Horizontal
                Slots:= array:
                    stack_box_slot: 
                        Widget:= NextLevelGemIcon
                        HorizontalAlignment:= horizontal_alignment.Right
                    stack_box_slot:
                        Widget:= NextLevelGemText
                        HorizontalAlignment:= horizontal_alignment.Right

        SetupGunDurationWidget()
        SetupUpgradeHammerWidget()
        SetupQuestWidget()
        SetupZoneOpenWidget()
        SetupGuardWidget()
        SetupBossWidget()
        SetupStashNotificationWidget()
        SetupBossNotificationWidget()
        SetupZoneWidget()
        SetupStashWidget()
        SetupGemWidget()
        SetupKeyWidget()
        SetupPeelyWidget()
        SetupHammerWidget()
        SetupCandyWidget()
        SetupCoinsWidget()
        


        set MyCanvas = canvas:
            Slots := array:
                # canvas_slot:
                #     Widget:= stack_box: #Damage Meter
                #         Orientation:= orientation.Vertical
                #             Slots:= array:
                #                 stack_box_slot:
                #                     Widget:= DamageMeterHeader
                #                     HorizontalAlignment:= horizontal_alignment.Center
                #                     Padding:= margin{Left:=0.0, Top:= 0.0, Bottom:= 0.0, Right:= 0.0} 
                #                 stack_box_slot:
                #                     Widget:= DamageMeterWidget
                #     Offsets := margin{Top := 500.0, Left := 0.0, Right := 0.0, Bottom := 0.0}
                #     Anchors := anchors{ Minimum := vector2{X := 0.0, Y := 0.20}, Maximum := vector2{X := 0.0, Y := 0.2}}
                #     Alignment := vector2{X := 0.0, Y := 0.0} 
                canvas_slot:
                    Widget:= stack_box: #Target Health UI
                        Orientation:= orientation.Vertical
                            Slots:= array:
                                stack_box_slot: #Boss & Damage Mastery Widget
                                    Widget:= BossWidget 
                                    Padding:= margin{Left:=0.0, Top:=60.0, Bottom:= 0.0, Right:= 0.0} 
                                    HorizontalAlignment:= horizontal_alignment.Center
                    Anchors := anchors{ Minimum := vector2{X := 0.5, Y := 0.0}, Maximum := vector2{X := 0.5, Y := 0.0}}
                    Alignment := vector2{X := 0.5, Y := 0.0}                
                canvas_slot: 
                    Widget:= stack_box: #Target Health UI
                        Orientation:= orientation.Vertical
                            Slots:= array:
                                stack_box_slot: #Guard Health
                                    Widget:= GuardWidget             
                                    Padding:= margin{Left:=0.0, Top:=10.0, Bottom:= 0.0, Right:= 0.0} 
                                    HorizontalAlignment:= horizontal_alignment.Center
                                stack_box_slot: #ZoneInformation
                                    Widget:= ZoneOpenVerticalStackBox             
                                    Padding:= margin{Left:=0.0, Top:=30.0, Bottom:= 0.0, Right:= 0.0} 
                                    HorizontalAlignment:= horizontal_alignment.Center

                    SizeToContent := true
                    Anchors := anchors{ Minimum := vector2{X := 0.5, Y := 0.0}, Maximum := vector2{X := 0.5, Y := 0.0}}
                    Offsets := margin{Top := 90.0, Left := 0.0, Right := 0.0, Bottom := 0.0}
                    Alignment := vector2{X := 0.5, Y := 0.0}
                # canvas_slot:
                #     Widget:= ZoneName
                #     Anchors := anchors{ Minimum := vector2{X := 1.0, Y := 0.65}, Maximum := vector2{X := 1.0, Y := 0.65}}
                #     Offsets := margin{Top := -620.0, Left := 0.0, Right := 0.0, Bottom := 0.0} ##470
                #     Alignment := vector2{X := 1.0, Y := 1.0}
                canvas_slot: # Daily Bonus
                    Widget:= DailyBonusWidget
                    
                    Anchors := anchors{ Minimum := vector2{X := 1.0, Y := 0.65}, Maximum := vector2{X := 1.0, Y := 0.65}}
                    Offsets := margin{Top := -130.0, Left := 0.0, Right := 0.0, Bottom := 0.0}
                    Alignment := vector2{X := 1.0, Y := 1.0}
                canvas_slot: # Quest Widget
                    Widget:= overlay:
                        Slots:= array:
                            overlay_slot:
                                Widget:= QuestHeaderImg
                                Padding:= margin{Left:= 0.0, Top:=0.0, Bottom:= 0.0, Right:= 0.0}      
                            overlay_slot:
                                Widget:= stack_box:
                                    Orientation:= orientation.Horizontal
                                    Slots:= array:
                                        stack_box_slot: 
                                            Widget:= texture_block{ DefaultImage := Images.QuestQuestion, DefaultDesiredSize := vector2{X := 40.0, Y := 40.0}} 
                                        stack_box_slot: 
                                            Widget:= QuestHeaderText
                                            Padding:= margin{Left:= 5.0, Top:=0.0, Bottom:= 0.0, Right:= 0.0}
                                Padding:= margin{Left:= 2.5, Top:= -65.0, Bottom:= 0.0, Right:= 0.0}                                 
                            overlay_slot:
                                Widget:= MyQuestWidget
                                Padding:= margin{Left:= 2.5, Top:= 65.0, Bottom:= 0.0, Right:= 0.0}                                
                            overlay_slot:
                                Widget:= QuestVerticalStackBox
                                Padding:= margin{Left:= 2.5, Top:= 0.0, Bottom:= 0.0, Right:= 0.0}  
                            overlay_slot:
                                Widget:= QuestCompleteImg
                                Padding:= margin{Left:= 0.0, Top:=0.0, Bottom:= 0.0, Right:= 0.0} 
                    Anchors := anchors{ Minimum := vector2{X := 1.0, Y := 0.65}, Maximum := vector2{X := 1.0, Y := 0.65}}
                    Offsets := margin{Top := -315.0, Left := 5.0, Right := 0.0, Bottom := 0.0}
                    Alignment := vector2{X := 1.0, Y := 1.0}
                canvas_slot: # Zone Widget
                    Widget:= ZoneWidget
                    Anchors := anchors{ Minimum := vector2{X := 1.0, Y := 0.65}, Maximum := vector2{X := 1.0, Y := 0.65}}
                    Offsets := margin{Top := -440.0, Left := 420.0, Right := 0.0, Bottom := 0.0} ##470
                    Alignment := vector2{X := 1.0, Y := 1.0}
                canvas_slot:
                    Widget:= BossNotificationWidget
                    Offsets := margin{Top := -140.0, Left := -300.0, Right := 0.0, Bottom := 0.0} #Top Spot
                    Anchors := anchors{ Minimum := vector2{X := 1.0, Y := 0.65}, Maximum := vector2{X := 1.0, Y := 0.65}}
                    Alignment := vector2{X := 1.0, Y := 0.0}
                canvas_slot: # Key Resources
                    Widget:= KeyWidget
                    Offsets := margin{Top := -105.0, Left := 425.0, Right := 0.0, Bottom := 0.0} #Top Spot
                    Anchors := anchors{ Minimum := vector2{X := 1.0, Y := 0.65}, Maximum := vector2{X := 1.0, Y := 0.65}}
                    Alignment := vector2{X := 1.0, Y := 0.0}
                canvas_slot: #Candy Resources
                    Widget:= CandyWidget
                    Offsets := margin{Top := -105.0, Left := 425.0, Right := 0.0, Bottom := 0.0} #Second from Bottom
                    Anchors := anchors{ Minimum := vector2{X := 1.0, Y := 0.65}, Maximum := vector2{X := 1.0, Y := 0.65}}
                    Alignment := vector2{X := 1.0, Y := 0.0}
                canvas_slot: # Gem Resources
                    Widget:= GemWidget
                    Offsets := margin{Top := -30.0, Left := 5.0, Right := 0.0, Bottom := 0.0} #Second from Top
                    Anchors := anchors{ Minimum := vector2{X := 1.0, Y := 0.65}, Maximum := vector2{X := 1.0, Y := 0.65}}
                    Alignment := vector2{X := 1.0, Y := 0.0}
                canvas_slot: #Coins Resources
                    Widget:= CoinsWidget
                    Offsets := margin{Top := 45.0, Left := 5.0, Right := 0.0, Bottom := 0.0} #Bottom Spot
                    Anchors := anchors{ Minimum := vector2{X := 1.0, Y := 0.65}, Maximum := vector2{X := 1.0, Y := 0.65}}
                    Alignment := vector2{X := 1.0, Y := 0.0}    
                canvas_slot: # Peely Count
                    Widget:= PeelyWidget
                    Offsets := margin{Top := -30.0, Left := 0.0, Right := 0.0, Bottom := 0.0}
                    Anchors := anchors{ Minimum := vector2{X := 0.0, Y := 0.65}, Maximum := vector2{X := 1.0, Y := 0.65}}
                    Alignment := vector2{X := 0.0, Y := 0.0}
                canvas_slot: # Upgrade Hammer Count
                    Widget:= HammerWidget
                    Offsets := margin{Top := -105.0, Left := 0.0, Right := 0.0, Bottom := 0.0}
                    Anchors := anchors{ Minimum := vector2{X := 0.0, Y := 0.65}, Maximum := vector2{X := 1.0, Y := 0.65}}
                    Alignment := vector2{X := 0.0, Y := 0.0}
                canvas_slot: # Candy Eating
                    Widget:= StrengthFloatingText.Setup(20, vector2{X := 55.0, Y := 55.0}, 2.0, 0.95)
                    Anchors := anchors{ Minimum := vector2{X := 0.5, Y := 0.6}, Maximum := vector2{X := 0.5, Y := 0.6}}
                    Offsets := margin{Top := -35.0, Left := 140.0, Right := 0.0, Bottom := 0.0}
                    Alignment := vector2{X := 0.0, Y := 0.5}
                canvas_slot: # Damage Numbers
                    Widget:= DamageNumbersHostWidget
                    Anchors := anchors{ Minimum := vector2{X := 0.5, Y := 0.6}, Maximum := vector2{X := 0.5, Y := 0.6}}
                    Offsets := margin{Top := -35.0, Left := 140.0, Right := 0.0, Bottom := 0.0}
                    Alignment := vector2{X := 0.0, Y := 0.5}
                canvas_slot: # HandCannonIcon
                    Widget:= HandCannonIcon
                    Anchors := anchors{ Minimum := vector2{X := 0.5, Y := 0.6}, Maximum := vector2{X := 0.5, Y := 0.6}}
                    Offsets := margin{Top := 55.0, Left := -150.0, Right := 0.0, Bottom := 0.0}
                    Alignment := vector2{X := 0.0, Y := 0.0}
                    ZOrder := 4
                canvas_slot: # HandCannon
                    Widget:= HandCannonHostWidget
                    Anchors := anchors{ Minimum := vector2{X := 0.5, Y := 0.6}, Maximum := vector2{X := 0.5, Y := 0.6}}
                    Offsets := margin{Top := 55.0, Left := -150.0, Right := 0.0, Bottom := 0.0}
                    Alignment := vector2{X := 0.0, Y := 0.0}
                    ZOrder := 1
                canvas_slot: # GunDuration Widget
                    Widget:= GunTickingClockStackbox
                    Anchors := anchors{ Minimum := vector2{X := 0.5, Y := 0.6}, Maximum := vector2{X := 0.5, Y := 0.6}}
                    Offsets := margin{Top := 135.0, Left := 100.0, Right := 0.0, Bottom := 0.0}
                    Alignment := vector2{X := 0.0, Y := 0.0}
                canvas_slot: #Guns available
                    Widget:= GunsAvailableStackbox
                    Anchors := anchors{ Minimum := vector2{X := 0.5, Y := 0.6}, Maximum := vector2{X := 0.5, Y := 0.6}}
                    Offsets := margin{Top := 105.0, Left := 50.0, Right := 0.0, Bottom := 0.0}
                    Alignment := vector2{X := 1.0, Y := 0.0}
                canvas_slot: # Buffs
                    Widget:= BuffHostWidget
                    Anchors := anchors{ Minimum := vector2{X := 0.5, Y := 0.6}, Maximum := vector2{X := 0.5, Y := 0.6}}
                    Offsets := margin{Top := 0.0, Left := 100.0, Right := 0.0, Bottom := 0.0}
                    Alignment := vector2{X := 0.0, Y := 0.0}
                canvas_slot: # Buffs
                    Widget:= BuffHostWidget2
                    Anchors := anchors{ Minimum := vector2{X := 0.5, Y := 0.6}, Maximum := vector2{X := 0.5, Y := 0.6}}
                    Offsets := margin{Top := 0.0, Left := 350.0, Right := 0.0, Bottom := 0.0}
                    Alignment := vector2{X := 0.0, Y := 0.0}
                canvas_slot: # WeaponDaily Bonus Text
                    Widget:= DailyBonusWeaponText
                    Anchors := anchors{ Minimum := vector2{X := 0.5, Y := 0.6}, Maximum := vector2{X := 0.5, Y := 0.6}}
                    Offsets := margin{Top := -35.0, Left := 375.0, Right := 0.0, Bottom := 0.0}
                    Alignment := vector2{X := 0.0, Y := 0.0}
                canvas_slot: # Level Numbers
                    Widget:= LevelNumbersHostWidget
                    Offsets := margin{Top := -90.0, Left := -110.0, Right := 0.0, Bottom := 0.0}
                    Anchors := anchors{ Minimum := vector2{X := 0.4, Y := 0.945}, Maximum := vector2{X := 0.4, Y := 0.945}}
                    Alignment := vector2{X := 0.5, Y := 0.5}            
                canvas_slot: #Next Level Gain
                    Widget:= stack_box:
                        Orientation:= orientation.Vertical
                        Slots:= array:
                            stack_box_slot: #level
                                Widget:= stack_box:
                                    Orientation:= orientation.Horizontal
                                    Slots:= array:
                                        stack_box_slot:
                                            Widget:= NextLevelText                                                         
                            stack_box_slot: 
                                Widget:= stack_box: 
                                    Orientation:= orientation.Horizontal
                                        Slots:= array:
                                            stack_box_slot: 
                                                Widget:= NextLevelStrengthIcon
                                                HorizontalAlignment:= horizontal_alignment.Right
                                            stack_box_slot:
                                                Widget:= NextLevelStrengthText
                                                HorizontalAlignment:= horizontal_alignment.Right
                    Anchors := anchors{ Minimum := vector2{X := 0.495, Y := 1.0}, Maximum := vector2{X := 0.495, Y := 1.0}}
                    Offsets := margin{Top := -100.0, Left := 420.0, Right := 0.0, Bottom := 0.0}
                    Alignment := vector2{X := 1.0, Y := 1.0}        
                
                canvas_slot: #Level & Strength
                    Widget:=overlay:
                        Slots:=
                            array:
                                overlay_slot:
                                    Widget:= GunTransition
                                overlay_slot:
                                    Widget:= stack_box:
                                        Orientation:= orientation.Vertical
                                            Slots:= array:
                                                stack_box_slot: #level
                                                    Widget:= stack_box:
                                                        Orientation:= orientation.Horizontal
                                                        Slots:= array:
                                                            stack_box_slot:
                                                                Widget:= LevelText                                                         
                                                stack_box_slot: 
                                                    Widget:= stack_box: 
                                                        Orientation:= orientation.Horizontal
                                                            Slots:= array:
                                                                stack_box_slot: 
                                                                    Widget:= StrengthIcon
                                                                stack_box_slot:
                                                                    Widget:= StrengthText
                                                # stack_box_slot:
                                                #     Widget:= stack_box:
                                                #         Orientation := orientation.Horizontal
                                                #             Slots:= array:
                                                #                 stack_box_slot:
                                                #                     Widget:= GunText    
                                                #                 stack_box_slot: 
                                                #                     Widget:= GunIcon
                                                #                 stack_box_slot:
                                                #                     Widget:= GunText2
                                    VerticalAlignment := vertical_alignment.Top                
                    SizeToContent := true
                    Anchors := anchors{ Minimum := vector2{X := 0.150, Y := 1.0}, Maximum := vector2{X := 0.150, Y := 1.0}}
                    Offsets := margin{Top := 80.0, Left := 300.0, Right := 0.0, Bottom := 0.0}
                    Alignment := vector2{X := 1.0, Y := 1.0}    
                canvas_slot:
                    Widget:= StashNotificationWidget
                    Anchors := anchors{ Minimum := vector2{X := 0.460, Y := 0.945}, Maximum := vector2{X := 0.460, Y := 0.945}}
                    Offsets := margin{Top := -105.0, Left := -257.0, Right := 0.0, Bottom := 0.0}
                    Alignment := vector2{X := 0.0, Y := 1.0}      
                canvas_slot: # Stashed Resources
                    Widget:= StashWidget
                    Anchors := anchors{ Minimum := vector2{X := 0.460, Y := 0.945}, Maximum := vector2{X := 0.460, Y := 0.945}}
                    Offsets := margin{Top := 725.0, Left := -277.0, Right := 0.0, Bottom := 0.0}
                    Alignment := vector2{X := 0.0, Y := 1.0}      
                canvas_slot: # Stashed PlayerNameHost
                    Widget:= StashPlayerNameHostWidget
                    Anchors := anchors{ Minimum := vector2{X := 0.460, Y := 0.945}, Maximum := vector2{X := 0.460, Y := 0.945}}
                    Offsets := margin{Top := -110.0, Left := -177.0, Right := 0.0, Bottom := 0.0}
                    Alignment := vector2{X := 0.0, Y := 0.5}     
                canvas_slot: 
                    Widget:= GunExpHostWidget            
                    Anchors := anchors{ Minimum := vector2{X := 0.2, Y := 0.0}, Maximum := vector2{X := 0.2, Y := 0.0}}
                    Offsets := margin{Top := 150.0, Left := 0.0, Right := 0.0, Bottom := 0.0}
                    Alignment := vector2{X := 0.5, Y := 0.0}
                    ZOrder := 10
                canvas_slot: #Bottom UI w/Level * EXP Progression
                    Widget:= stack_box: 
                        Orientation:= orientation.Horizontal
                            Slots:= array:
                                stack_box_slot: #EXP Bar
                                    Widget:= overlay:
                                        Slots:= array:
                                            overlay_slot:
                                                Widget:= MyExpWidget
                                            overlay_slot:
                                                Widget:= stack_box:
                                                    Orientation:= orientation.Horizontal
                                                    Slots:= array:
                                                        stack_box_slot:
                                                            Widget:= ExpText
                                                            HorizontalAlignment:= horizontal_alignment.Left
                                                            Padding:= margin{Left:=2.0, Top:=0.0, Bottom:= 0.0, Right:= 2.0} 
                                                        stack_box_slot:
                                                            Widget:= MidasEssenceIcon1
                                                            HorizontalAlignment:= horizontal_alignment.Left
                                                            Padding:= margin{Left:=2.0, Top:=0.0, Bottom:= 0.0, Right:= 2.0} 
                                                        stack_box_slot:
                                                            Widget:= ProgressTextLeft
                                                        stack_box_slot:
                                                            Widget:= MidasEssenceIcon2
                                                            HorizontalAlignment:= horizontal_alignment.Left
                                                            Padding:= margin{Left:=2.0, Top:=0.0, Bottom:= 0.0, Right:= 2.0} 
                                                        stack_box_slot:
                                                            Widget:= ProgressTextRight
                                                HorizontalAlignment:= horizontal_alignment.Center
                    SizeToContent := true
                    Anchors := anchors{ Minimum := vector2{X := 0.460, Y := 0.945}, Maximum := vector2{X := 0.460, Y := 0.945}}
                    Offsets := margin{Top := 0.0, Left := 103.0, Right := 0.0, Bottom := 0.0}
                    Alignment := vector2{X := 0.5, Y := 0.5}            
                  
        return MyCanvas

    LoopingSecondCheck()<suspends> : void = 
        loop:
            Sleep(1.0)
            #Bounce Available Key Items
            if: 
                CandyAmount.Value > 0.0
                PlayerManager.GetAchievementLevel["CandyUnlocked"] > 0
            then:
                BeginBounceEffect(CandyIcon, CandyIconDefaultSize, IconBounce)

            #Need to check if upgrade is available
            # if:
            #     GemAmount.Value > 0.0
            # then:
            #     BeginBounceEffect(GemIcon, GemIconDefaultSize, IconBounce)

            if:
                KeyAmount.Value > 0.0
            then:
                BeginBounceEffect(KeyIcon, KeyIconDefaultSize, IconBounce)
    
            #Hide UIs
            if(StashIncomeTimer>0):
                set StashIncomeTimer -= 1
                UpdateStashCountdownTimer()
            
            set GuardVisibilityTimer -= 1.0 
            if(GuardVisibilityTimer <= 0.0):
                BossWidget.SetVisibility(widget_visibility.Hidden)
                BossName.SetVisibility(widget_visibility.Hidden)

                GuardWidget.SetVisibility(widget_visibility.Hidden)
                GuardName.SetVisibility(widget_visibility.Hidden)

            set ZoneMsgVisibilityTimer -= 1.0 
            if(ZoneMsgVisibilityTimer <= 0.0):
                UpdateNextGemText(PlayerManager.CurrentZone, false)
                Text:=""
                SetZoneOpenText(Text)

    #Level Functions           
    UpdateLevelUI():void=
        if:
            PlayerManager.GetLevel[] > 1
        then:
            NextLevelGemWidget.SetVisibility(widget_visibility.Visible)
            NextLevelGemIcon.SetImage(Images.Gem2)
            NextLevelGemText.SetText(STM(" + 1"))
        else:
            NextLevelGemWidget.SetVisibility(widget_visibility.Collapsed)
            NextLevelGemIcon.SetImage(Images.Empty)
            NextLevelGemText.SetText(STM(""))
        
        if:
            CurrentLevel:=PlayerManager.GetLevel[]
            CurrentStrength:=PlayerManager.GetDisplayStrength[]
        then:
            StrengthGain := PlayerManager.GetStrengthGainedPerLevel(CurrentLevel + 1) * PlayerManager.GetPlayerStat(player_stat.DamageMult)
            LevelText.SetText(STM("Level {CurrentLevel}:"))
            StrengthText.SetText(STM(" {CurrentStrength}"))
            NextLevelStrengthText.SetText(STM(" + {StrengthGain}"))
            
            Temp:= GunIcon.GetDesiredSize()
            if:
                Temp.X > 0.0
                Mult:= PlayerManager.GetGunStats[PlayerManager.GetHeldGun[]].DamageMult
            then:
                GunText2.SetText(STM(" {Mult * CurrentStrength}!"))
                BeginBounceEffect(GunIcon, GunIconDefaultSize, IconBounce)
            BeginBounceEffect(StrengthIcon, StrengthIconDefaultSize, IconBounce)
            BeginBounceEffect(NextLevelStrengthIcon, NextLevelStrengthIconDefaultSize, IconBounce)

    GainLevel(MultiLevel:int):void=
        ShowLevelNumber(MultiLevel)
    
    ShowLevelNumber(MultiLevel:int):void=
        if:
            Number1:=LevelNumbers[CurrentLevelCounter]
            Color:=LevelColors[CurrentLevelCounter]
            Image:=LevelImgs[CurrentLevelCounter]
        then:            
            Color.SetDesiredSize(vector2{Y:=36.0})
            if(MultiLevel = 1):
                Number1.SetText(STM(" Level up!"))
                Number1.SetTextColor(NamedColors.White)
            if(MultiLevel = 5):
                Number1.SetText(STM(" Level up x5!"))
                Number1.SetTextColor(MakeColorFromHex("36CC42")) #Green
            if(MultiLevel = 25):
                Number1.SetText(STM(" Level up x25!"))
                Number1.SetTextColor(MakeColorFromHex("1985A3")) #Blue
            if(MultiLevel = 100):
                Number1.SetText(STM(" Level up x100!"))
                Number1.SetTextColor(MakeColorFromHex("8A2BE2")) #Purple
            Number1.SetTextOpacity(1.0)
            Number1.SetShadowOpacity(1.0)
            
            EssenceProgressBar.BeginBounceEffect()
            BeginBounceEffect(Image, LevelIconDefaultSize, IconBounce)
            BeginBounceEffect(MidasEssenceIcon1, MidasEssenceIconDefaultSize, IconBounce)
            BeginBounceEffect(MidasEssenceIcon2, MidasEssenceIconDefaultSize, IconBounce)
            set ActiveLevels = ActiveLevels + array{CurrentLevelCounter}
    
        #IncrementCounter without overflowing
        set CurrentLevelCounter += 1
        if(CurrentLevelCounter >= LevelNumbers.Length):
            set CurrentLevelCounter = 0

    LoopActiveBounces():void=
        var IndicesToRemove : []int = array{}
        var ActiveNumIndex : int = 0

        #Efficient Loop through ActiveBounces
        for (ActiveBounce:ActiveBounces):
            if:
                ActiveBounceSize := ActiveBounceSizes[ActiveNumIndex]
            then:
                IterateBounceImageEffect(ActiveBounce, ActiveBounceSize, GatherBounceIterations)
                CurrentX := ActiveBounce.GetDesiredSize().X 
                
                if:
                    CurrentX = ActiveBounceSize.X
                then:
                    set IndicesToRemove = IndicesToRemove + array{ActiveNumIndex}
            #Iterate Index
            set ActiveNumIndex += 1

        #Remove from ActiveBounces
        for(Index:IndicesToRemove):
            if:
                TempValue:=ActiveBounces[Index]
                TempArray:=ActiveBounces.RemoveElement[Index]

                TempValue2:=ActiveBounceSizes[Index]
                TempArray2:=ActiveBounceSizes.RemoveElement[Index]
            then:
                set ActiveBounces = TempArray
                set ActiveBounceSizes = TempArray2

    LoopLevelNumbers():void=
        var IndicesToRemove : []int = array{}
        var ActiveNumIndex : int = 0
        BeginFadePosition:=200.0
        MaxFadePosition:=300.0

        #Efficient Loop through Active Numbers
        for (ActiveLevel:ActiveLevels):
            if:
                Number1:=LevelNumbers[ActiveLevel]
                Color:=LevelColors[ActiveLevel]
                
            then:
                Temp : float = Color.GetDesiredSize().Y
                if(Temp > BeginFadePosition):
                    SetOpacity(Number1, Number1, MaxFadePosition - Temp)
                    if(Temp > MaxFadePosition):
                        #Bump Position Off-Screen
                        Color.SetDesiredSize(vector2{X:=Color.GetDesiredSize().X,Y:=Color.GetDesiredSize().Y + 3000.0})
                        set IndicesToRemove = IndicesToRemove + array{ActiveNumIndex}
                Color.SetDesiredSize(vector2{X:=Color.GetDesiredSize().X,Y:=Color.GetDesiredSize().Y + 8.0})
                
                #Iterate Index
                set ActiveNumIndex += 1
            
        #Remove from ActiveNumbers
        for(Index:IndicesToRemove):
            if:
                TempValue:=ActiveLevels[Index]
                TempArray:=ActiveLevels.RemoveElement[Index]
            then:
                set ActiveLevels = TempArray
    
    #
    QueueGainUpgradeHammerProgress(Amt:int, Req: int):void=
        if:
            QueuedUpgradeHammerProgress = -1
        then:
            spawn:
                StartGainUpgradeHammerProgress()

        set QueuedUpgradeHammerProgress = Amt
        set QueuedUpgradeHammerRequirement = Req

    StartGainUpgradeHammerProgress()<suspends>:void=
        Sleep(0.1)
        GainUpgradeHammerProgress(QueuedUpgradeHammerProgress, QueuedUpgradeHammerRequirement)
        set QueuedUpgradeHammerProgress = -1
        set QueuedUpgradeHammerRequirement = -1

    GainUpgradeHammerProgress(Amt:int, Req:int):void=
        if: 
            PlayerManager.GetAchievementLevel["Sharkey_NextTier"] > 0
        then:
            UpgradeHammerWidget.Widget.SetVisibility(widget_visibility.Visible)

        UpgradeHammerIcon.SetImage(Images.UpgradeHammer)
        Percent := (Amt * 1.0)/(Req * 1.0)
        UpgradeHammerProgressBar.SetProgress(Percent, true)
        UpgradeHammerProgressText.SetText(STM("{Amt} / {Req}"))

    #Essence Functions    
    UpdateEssenceUI():void=
        if:
            Temp : tuple(float, tuple(big_number,big_number)) = PlayerManager.GetLevelProgress[]
            CurrentEssence:=PlayerManager.GetEssence[]
        then:
            EssenceProgressBar.SetProgress(Temp(0), true)
            EssenceText.SetText(STM("{FormatBigNumber(CurrentEssence)}"))
            ProgressTextLeft.SetText(STM("{FormatBigNumber(Temp(1)(0))} /"))
            ProgressTextRight.SetText(STM("{FormatBigNumber(Temp(1)(1))}"))
            
            BeginBounceEffect(MidasEssenceIcon1, MidasEssenceIconDefaultSize, IconBounce)

    UpdateEssenceUIOnLevelUp():void=
        if:
            Temp : tuple(float, tuple(big_number,big_number)) = PlayerManager.GetLevelProgress[]
            CurrentEssence:=PlayerManager.GetEssence[]
        then:
            EssenceProgressBar.SetProgress(Temp(0), true)
            EssenceText.SetText(STM("{FormatBigNumber(CurrentEssence)}"))
            ProgressTextLeft.SetText(STM("{FormatBigNumber(Temp(1)(0))} /"))
            ProgressTextRight.SetText(STM("{FormatBigNumber(Temp(1)(1))}"))            

    QueueDamageNumber(Damage : big_number, KOBonus : big_number, ZoneBonus : big_number, IsWeakHit:logic, IsHandcannon:logic, IsAmbush:logic, IsBlocked:logic, ?IsPoints : logic = false):void=
        if:
            Damage.Value > 0.0 or KOBonus.Value > 0.0 or ZoneBonus.Value > 0.0 or IsBlocked?
            QueuedEssenceGain.Value = 0.0 and QueuedKOBonusGain.Value = 0.0 and QueuedZoneBonusGain.Value = 0.0 and not IsQueuedHitBlocked?
        then:
            spawn:
                StartDamageNumber()
        
        set QueuedEssenceGain = BigNumber_Add(Damage, QueuedEssenceGain)
        set QueuedKOBonusGain =  BigNumber_Add(KOBonus, QueuedKOBonusGain)
        set QueuedZoneBonusGain =  BigNumber_Add(ZoneBonus, QueuedZoneBonusGain)
        
        if:
            not(IsQueuedHitWeak?)
        then:
            set IsQueuedHitWeak = IsWeakHit

        if:
            not(IsQueuedHandcannon?)
        then:
            set IsQueuedHandcannon = IsHandcannon

        if:
            not(IsQueuedAmbush?)
        then:
            set IsQueuedAmbush = IsAmbush
        
        if:
            not(IsQueuedPoints?)
        then:
            set IsQueuedPoints = IsPoints

        if:
            not(IsQueuedHitBlocked?)
        then:
            set IsQueuedHitBlocked = IsBlocked
    
    StartDamageNumber()<suspends>:void=
        Sleep(0.1)
        ShowDamageNumber(QueuedEssenceGain, QueuedKOBonusGain, QueuedZoneBonusGain, IsQueuedHitWeak, IsQueuedHandcannon, IsQueuedAmbush, IsQueuedHitBlocked, ?IsPoints :=  IsQueuedPoints)
        set QueuedEssenceGain = big_number{}
        set QueuedKOBonusGain =  big_number{}
        set QueuedZoneBonusGain =  big_number{}
        set IsQueuedHitWeak = false
        set IsQueuedHandcannon = false
        set IsQueuedAmbush = false
        set IsQueuedPoints = false
        set IsQueuedHitBlocked = false

    QueueCollectionItem(Count : big_number, ItemName : string):void=
        if:
            Count.Value = 0.0
        then:
            return
        
        if:
            QueuedCollectionAmount.Length = 0
        then:
            spawn:
                StartCollectionItem()

        set QueuedCollectionItem = QueuedCollectionItem + array{ItemName}
        set QueuedCollectionAmount = QueuedCollectionAmount + array{Count}
        # Log("Queue Item Length: {QueuedCollectionAmount.Length}")

    FoundPeely(Location : string, LocalPeelysFound : int, LocalMaxPeelys : int, TotalPeelys : big_number):void=
        PlayerManager.GameManager.AudioManager.PeelyLaugh.Play(PlayerManager.Agent)

        var Text : string = ""
        if:
            Amt := 100.0 * MakeBigNumber(1.25, Int[BigNumber_ToFloat(TotalPeelys + 1.0)])
        then:
            set Text = "Found {LocalPeelysFound} out of {LocalMaxPeelys} <Peely> Peelys in {Location} today!<br>Next Peely will grant you {Amt} <Midas Coins>"

        if:
            LocalPeelysFound = LocalMaxPeelys
            Amt := 100.0 * MakeBigNumber(1.25, Int[BigNumber_ToFloat(TotalPeelys)])
        then:
            PlayerManager.GameManager.AudioManager.UnlockZonePlayer.Play(PlayerManager.Agent)
            set Text = Text + "<br>Bonus {(Amt) * (LocalMaxPeelys * 1.0)} <Midas Coins> Reward!"
        
        Count := PlayerManager.GetCollectibleCount("Peely")
        
        SetZoneOpenText(Text)
        set ZoneMsgVisibilityTimer = 10.0

    StartCollectionItem()<suspends>:void=
        loop:
            Sleep(0.1)
            if:
                Name:= QueuedCollectionItem[0]
                Amount:= QueuedCollectionAmount[0]
            then:
                ShowCollectionItem(Amount, Name)
                
                Sleep(0.25)
                if:
                    Temp1 := QueuedCollectionItem.RemoveElement[0]
                    set QueuedCollectionItem = Temp1
                    
                    Temp2 := QueuedCollectionAmount.RemoveElement[0]
                    set QueuedCollectionAmount = Temp2
                # Log("Removed 1 Item. New Queue Item Length: {QueuedCollectionAmount.Length}")
                if:
                    QueuedCollectionAmount.Length = 0
                then:
                    return
            else:
                return

    ShowCollectionItem(Count : big_number, ItemName : string)<suspends>:void=
        if:
            Count.Value = 0.0
        then:
            return
        
        if:
            Number1:=CollectibleNumbers1[CurrentCollectibleCounter]
            Img1:=CollectibleImgs1[CurrentCollectibleCounter]
            Color:=CollectibleColors[CurrentCollectibleCounter]
            Total := PlayerManager.GetCollectibleCount(ItemName)
        then:
            Color.SetDesiredSize(vector2{Y:=32.0})
            if:
                Count.Value > 0.0
            then:
                Number1.SetText(STM(" + {Count} ({Total})"))
            else:
                Number1.SetText(STM(" {Count} ({Total})"))
            Number1.SetTextColor(MakeColorFromHex("FFFFFF"))
            Number1.SetTextOpacity(1.0)
            Number1.SetShadowOpacity(1.0)
            Img1.SetImage(GetTextureFromString(ItemName))
            BeginBounceEffect(Img1, DamageNumbersSize, IconBounce)
             
            #Add to ActiveNumbers
            set ActiveCollectibleNumbers = ActiveCollectibleNumbers + array{CurrentCollectibleCounter}
        
        #IncrementCounter without overflowing
        set CurrentCollectibleCounter += 1
        if(CurrentCollectibleCounter >= CollectibleNumbers1.Length):
            set CurrentCollectibleCounter = 0

    ShowDamageNumber(Damage : big_number, KOBonus : big_number, ZoneBonus : big_number, IsWeakHit:logic, IsHandcannon:logic, IsAmbush:logic, IsBlocked:logic, ?IsPoints : logic = false)<suspends>:void=       
        UpdateEssenceUI()
        if:
            Number1:=DamageNumbers1[CurrentDamageCounter]
            Number2:=DamageNumbers2[CurrentDamageCounter]
            Img1:=DamageImgs1[CurrentDamageCounter]
            Img2:=DamageImgs2[CurrentDamageCounter]
            Color:=DamageColors[CurrentDamageCounter]
        then:
            Color.SetDesiredSize(vector2{Y:=32.0})

            if(IsBlocked?):
                PlayerManager.GameManager.AudioManager.ArmorDeflectionSound.Play(PlayerManager.Agent)

            if(IsHandcannon?):
                PlayerManager.GameManager.AudioManager.PlayHandcannon(PlayerManager.Agent)
            
            if(IsAmbush?):
                PlayerManager.GameManager.AudioManager.Ambush.Play(PlayerManager.Agent)

            if(IsWeakHit? or IsHandcannon? or IsAmbush?):
                Number1.SetText(STM(" {Damage + KOBonus}!"))                
            else:
                Number1.SetText(STM(" {FormatBigNumber(BigNumber_Add(Damage, KOBonus))}"))

            #Color
            if(IsWeakHit? and IsHandcannon? and IsAmbush?):
                Number1.SetTextColor(MakeColorFromHex("A31985")) #Pink
            else:
                var FoundMatch : logic = false
                if(IsWeakHit? and IsHandcannon?):
                    Number1.SetTextColor(MakeColorFromHex("DD8852")) #Orange
                    set FoundMatch = true
                
                if(IsAmbush? and IsHandcannon?):
                    Number1.SetTextColor(MakeColorFromHex("77A6B8")) #Green
                    set FoundMatch = true

                if(IsWeakHit? and IsAmbush?):
                    Number1.SetTextColor(MakeColorFromHex("7854B4")) #Purple
                    set FoundMatch = true

                if(not FoundMatch?):
                    if(IsWeakHit?):
                        Number1.SetTextColor(MakeColorFromHex("CC362F")) #Red
                        set FoundMatch = true
                    if(IsHandcannon?):
                        Number1.SetTextColor(NamedColors.Goldenrod) #Yellow
                        set FoundMatch = true
                    if(IsAmbush?):
                        Number1.SetTextColor(MakeColorFromHex("0072fc")) #Blue
                        set FoundMatch = true
                
                if(not FoundMatch?):
                    Number1.SetTextColor(MakeColorFromHex("FFFFFF")) #White

            if(IsBlocked?):
                Number1.SetText(STM(" BLOCKED!"))                
                Number1.SetTextColor(MakeColorFromHex("FFFFFF")) #White
                Img1.SetImage(Images.Body)
            else:
                #Image Priority
                if(KOBonus.Value > 0.0):
                    # Img1.SetImage(Images.KOBonus)
                else:
                    if(IsAmbush?): #Blue - Only Ambush
                        Img1.SetImage(Images.WolfClaw)
                    else:
                        if(IsHandcannon?):
                            Img1.SetImage(Images.Handcannon)
                        else:
                            if(IsWeakHit?):
                                Img1.SetImage(ImgCrit)
                            else:    
                                if(IsPoints?):
                                    Img1.SetImage(Images.UpgradeIcon)
                                else:
                                    Img1.SetImage(ImgMidasEssence)
                        
            Number2.SetText(STM(" {FormatBigNumber(ZoneBonus)}"))
            Number2.SetTextColor(MakeColorFromHex("FFFFFF"))
            # Img2.SetImage(Images.GemHunting)

            BeginBounceEffect(Img1, DamageNumbersSize, IconBounce)
            BeginBounceEffect(Img2, DamageNumbersSize, IconBounce)

            #If Gun Item is Currently Active, Bounce that shit!
            Temp:= GunIcon.GetDesiredSize()
            if:
                Temp.X > 0.0
            then:
                BeginBounceEffect(GunIcon, GunIconDefaultSize, IconBounce)                
                BeginBounceEffect(StrengthIcon, StrengthIconDefaultSize, IconBounce)                
            
            Number1.SetTextOpacity(1.0)
            Number1.SetShadowOpacity(1.0)

            Number2.SetTextOpacity(1.0)
            Number2.SetShadowOpacity(1.0)
            
            if(ZoneBonus.Value <= 0.0):
                Number2.SetText(STM(""))
                Img2.SetImage(ImgEmpty)
            else:
                Number2.SetTextOpacity(1.0)
                Number2.SetShadowOpacity(1.0)
            
            #Add to ActiveNumbers
            set ActiveNumbers = ActiveNumbers + array{CurrentDamageCounter}
        
        #IncrementCounter without overflowing
        set CurrentDamageCounter += 1
        if(CurrentDamageCounter >= DamageNumbers1.Length):
            set CurrentDamageCounter = 0

    MoveUINumbers()<suspends>:void=
        loop:
            Sleep(0.01)
            LoopDamageNumbers()
            LoopCollectibleNumbers()
            LoopActiveBounces()
            LoopLevelNumbers()
            LoopStashPlayers()
            MoveNotifications()

    LoopDamageNumbers():void=
        var IndicesToRemove : []int = array{}
        var ActiveNumIndex : int = 0

        #Efficient Loop through Active Numbers
        for (ActiveNumber:ActiveNumbers):
            BeginFadePosition:=200.0
            MaxFadePosition:=300.0
            if:
                Number1:=DamageNumbers1[ActiveNumber]
                Number2:=DamageNumbers2[ActiveNumber]
                Color:=DamageColors[ActiveNumber]
                Img1:=DamageImgs1[ActiveNumber]
                Img2:=DamageImgs2[ActiveNumber]
            then:
                Temp : float = Color.GetDesiredSize().Y
                if(Temp > BeginFadePosition):
                    SetOpacity(Number1, Number2, MaxFadePosition - Temp)
                    if(Temp > MaxFadePosition):
                        #Bump Position Off-Screen
                        Color.SetDesiredSize(vector2{X:=Color.GetDesiredSize().X,Y:=Color.GetDesiredSize().Y + 3000.0})
                        set IndicesToRemove = IndicesToRemove + array{ActiveNumIndex}
                Color.SetDesiredSize(vector2{X:=Color.GetDesiredSize().X,Y:=Color.GetDesiredSize().Y + 8.0})
                
                #Iterate Index
                set ActiveNumIndex += 1
            
        #Remove from ActiveNumbers
        for(Index:IndicesToRemove):
            if:
                TempValue:=ActiveNumbers[Index]
                TempArray:=ActiveNumbers.RemoveElement[Index]
            then:
                set ActiveNumbers = TempArray

    LoopCollectibleNumbers():void=
        var IndicesToRemove : []int = array{}
        var ActiveNumIndex : int = 0

        #Efficient Loop through Active Numbers
        for (ActiveNumber:ActiveCollectibleNumbers):
            BeginFadePosition:=200.0
            MaxFadePosition:=300.0
            if:
                Number1:=CollectibleNumbers1[ActiveNumber]
                Color:=CollectibleColors[ActiveNumber]
                Img1:=CollectibleImgs1[ActiveNumber]
            then:
                Temp : float = Color.GetDesiredSize().Y
                if(Temp > BeginFadePosition):
                    SetOpacity(Number1, Number1, MaxFadePosition - Temp)
                    if(Temp > MaxFadePosition):
                        #Bump Position Off-Screen
                        Color.SetDesiredSize(vector2{X:=Color.GetDesiredSize().X,Y:=Color.GetDesiredSize().Y + 3000.0})
                        set IndicesToRemove = IndicesToRemove + array{ActiveNumIndex}
                Color.SetDesiredSize(vector2{X:=Color.GetDesiredSize().X,Y:=Color.GetDesiredSize().Y + 8.0})
                
                #Iterate Index
                set ActiveNumIndex += 1
            
        #Remove from ActiveNumbers
        for(Index:IndicesToRemove):
            if:
                TempValue:=ActiveCollectibleNumbers[Index]
                TempArray:=ActiveCollectibleNumbers.RemoveElement[Index]
            then:
                set ActiveCollectibleNumbers = TempArray

    BeginBounceEffect(Img : texture_block, DefaultSize : vector2, BounceAmount : float):void=
        Img.SetDesiredSize(vector2{X:= DefaultSize.X * BounceAmount, Y:= DefaultSize.Y * BounceAmount})
        set ActiveBounces = ActiveBounces + array{Img}
        set ActiveBounceSizes = ActiveBounceSizes + array{DefaultSize}
        
    #Stash Functions
    UpdateStashUI(Total:big_number, Gain:big_number, CurrentValue : big_number):void=
        var ShowStashUI : logic = false

        if(not(BigNumber_Equals(StashedAmount, Total)?)):
            BeginBounceEffect(StashIcon, StashIconDefaultSize, IconBounce)
        set StashedAmount = Total
        StashedText.SetText(STM(" {FormatBigNumber(StashedAmount)}"))

        if:
            Essence := PlayerManager.GetEssence[]
            EssenceRequired := PlayerManager.GetLevelRequirement(PlayerManager.GetLevel[])
            BigNumber_GreaterThanOrEquals(StashedAmount, big_number{Value:=1.0})?
            BigNumber_GreaterThanOrEquals(StashedAmount, EssenceRequired)?
        then:
            #Show Notification
            StashArrowIcon.SetImage(Images.NotificationArrowDown)
            StashRemoteIcon.SetImage(Images.Remote)
            StashNotificationIcon.SetImage(Images.Chest)
        else:
            #Hide Notification
            StashArrowIcon.SetImage(Images.Empty)
            StashRemoteIcon.SetImage(Images.Empty)
            StashNotificationIcon.SetImage(Images.Empty)

        if(StashedAmount.Value > 0.0 or StashIncomeAmount.Value > 0.0):
            set ShowStashUI = true

        if(ShowStashUI?):
            spawn:
                IterateResourceTransition(StashTransitionWidget, 800.0)
        else:
            spawn:
                IterateResourceTransition(StashTransitionWidget, 0.0)

        
    StashEmptyMessage():void=
        PlayerManager.GameManager.AudioManager.PlayUpgradeFailSFX(PlayerManager.Agent)
        var Text : string ="<stash> STASH is empty!<br>Try getting OTHER PLAYERS' to earn <KO Bonus> KOs or <Clear Bonus> CLEARS nearby."

        if:
            PlayerManager.GameManager.PlayerManagers.Length <= 1
        then:
            set Text = Text + "<br><stash> STASH only grows when playing with OTHER PLAYERS."
        else:
            if:
                ZM := PlayerManager.GameManager.GetZoneManagerFromZoneTag[PlayerManager.CurrentZone]?
                ZM.PlayersInZone <= 1
            then:
                set Text = Text + "<br><stash> STASH only grows when in the same area as OTHER PLAYERS."

        SetZoneOpenText(Text)
        set ZoneMsgVisibilityTimer = 10.0

    MissingUpgradeMessage(Upgrade : string):void=
        PlayerManager.GameManager.AudioManager.PlayUpgradeFailSFX(PlayerManager.Agent)
        var Text : string ="Cannot collect! Missing upgrade <{Upgrade}> {Upgrade}."
        SetZoneOpenText(Text)
        set ZoneMsgVisibilityTimer = 10.0

    UpdateStashPassiveUI():void=
        if:
            PassiveIncome:= PlayerManager.GetPassiveIncomeAmount[]
            StashAmount:= PlayerManager.GetStashedEssence[]
        then:
            set StashIncomeAmount = PassiveIncome
            set StashedAmount = StashAmount

        if(StashIncomeAmount.Value > 0.0):
            StashedIncomeText.SetText(STM(" + {FormatBigNumber(StashIncomeAmount)} in {StashIncomeTimer}s"))
        else:
            StashedIncomeText.SetText(STM(""))

    ResetStashCountdownTimer():void=
        set StashIncomeTimer = PlayerManager.GameManager.PassiveIncomeTimer
        UpdateStashCountdownTimer()

    UpdateStashCountdownTimer():void=
        if(StashIncomeAmount.Value > 0.0):
            StashedIncomeText.SetText(STM(" + {FormatBigNumber(StashIncomeAmount)} in {StashIncomeTimer}s"))
    
    ShowStashPlayer(GuardKiller : agent, GuardsAlive : int):void=       
        if:
            Number1:=PlayerName1[CurrentStashPlayer]
            Img1:=StashIcon1[CurrentStashPlayer]
            Color:=StashPlayerColors[CurrentStashPlayer]
        then:
            Color.SetDesiredSize(vector2{Y:=32.0})
            Number1.SetText(GetPlayerName(GuardKiller))
            Number1.SetTextColor(MakeColorFromHex("FFFFFF"))
            
            if:
                GuardsAlive = 0
            then:
                # Img1.SetImage(Images.GemHunting)
            else:
                # Img1.SetImage(Images.KOBonus)

            BeginBounceEffect(Img1, StashPlayerIconDefaultSize, IconBounce)

            Number1.SetTextOpacity(1.0)
            Number1.SetShadowOpacity(1.0)
            
            #Add to ActiveNumbers
            set ActiveStashPlayers = ActiveStashPlayers + array{CurrentStashPlayer}
        
        #IncrementCounter without overflowing
        set CurrentStashPlayer += 1
        if(CurrentStashPlayer >= PlayerName1.Length):
            set CurrentStashPlayer = 0
    
    MoveNotifications():void=
        if:
            IsStashNotifBouncingUp?
        then:
            StashBounceColorWidget.SetDesiredSize(vector2{X:=StashBounceColorWidget.GetDesiredSize().X,Y:=StashBounceColorWidget.GetDesiredSize().Y + 2.0})
            Temp:=StashBounceColorWidget.GetDesiredSize().Y
            if:
                Temp >= 240.0
            then:
                set IsStashNotifBouncingUp = false
        else:
            StashBounceColorWidget.SetDesiredSize(vector2{X:=StashBounceColorWidget.GetDesiredSize().X,Y:=StashBounceColorWidget.GetDesiredSize().Y - 2.0})
            Temp:=StashBounceColorWidget.GetDesiredSize().Y
            if:
                Temp <= 200.0
            then:
                set IsStashNotifBouncingUp = true
        
        if:
            IsBossNotifBouncingLeft?
        then:
            BossBounceColorWidget.SetDesiredSize(vector2{X:=BossBounceColorWidget.GetDesiredSize().X + 2.0,Y:=BossBounceColorWidget.GetDesiredSize().Y})
            Temp:=BossBounceColorWidget.GetDesiredSize().X
            if:
                Temp >= 40.0
            then:
                set IsBossNotifBouncingLeft = false
        else:
            BossBounceColorWidget.SetDesiredSize(vector2{X:=BossBounceColorWidget.GetDesiredSize().X - 2.0,Y:=BossBounceColorWidget.GetDesiredSize().Y})
            Temp:=BossBounceColorWidget.GetDesiredSize().X
            if:
                Temp <= 0.0
            then:
                set IsBossNotifBouncingLeft = true


    LoopStashPlayers():void=
        var IndicesToRemove : []int = array{}
        var ActiveNumIndex : int = 0
        BeginFadePosition:=100.0
        MaxFadePosition:=200.0

        #Efficient Loop through Active Numbers
        for (ActiveStash:ActiveStashPlayers):
            if:
                Number1:=PlayerName1[ActiveStash]
                Color:=StashPlayerColors[ActiveStash]
            then:
                Temp : float = Color.GetDesiredSize().Y
                if(Temp > BeginFadePosition):
                    SetOpacity(Number1, Number1, MaxFadePosition - Temp)
                    if(Temp > MaxFadePosition):
                        #Bump Position Off-Screen
                        Color.SetDesiredSize(vector2{X:=Color.GetDesiredSize().X,Y:=Color.GetDesiredSize().Y + 3000.0})
                        set IndicesToRemove = IndicesToRemove + array{ActiveNumIndex}
                Color.SetDesiredSize(vector2{X:=Color.GetDesiredSize().X,Y:=Color.GetDesiredSize().Y + 8.0})
                #Iterate Index
                set ActiveNumIndex += 1
            
        #Remove from ActiveNumbers
        for(Index:IndicesToRemove):
            if:
                TempValue:=ActiveStashPlayers[Index]
                TempArray:=ActiveStashPlayers.RemoveElement[Index]
            then:
                set ActiveStashPlayers = TempArray

    #Boss Functions
    UpdateBossKeyTimer(Seconds : int) : void =
        if: 
            PlayerManager.GetAchievementLevel["Sarge_NextTier"] > 0
        then:
            spawn:
                IterateRightsideTransition(KeyTransitionWidget, 420.0)

        if:
            Minutes := Int[(Seconds * 1.0) / 60.0]
            NewSeconds : int = Mod[Seconds, 60]
        then:
            if: 
                NewSeconds < 10
            then: 
                KeyTimerText.SetText(STM("{Minutes}:0{NewSeconds}"))
            else:
                KeyTimerText.SetText(STM("{Minutes}:{NewSeconds}"))

    ShowBossUI(IsDamageMastery : logic) : void = 
        if(IsDamageMastery?):
            BossIcon.SetImage(Images.Empty)
        else:
            BossIcon.SetImage(Images.BossTeleport)

        BossWidget.SetVisibility(widget_visibility.Visible)
        BossWidget.SetVisibility(widget_visibility.Visible)
        BossName.SetVisibility(widget_visibility.Visible)

        GuardWidget.SetVisibility(widget_visibility.Hidden)
        GuardWidget.SetVisibility(widget_visibility.Hidden)
        GuardName.SetVisibility(widget_visibility.Hidden)
        set GuardVisibilityTimer = 5.0

    #Guard Functions
    ShowGuardUI() : void =         
        GuardWidget.SetVisibility(widget_visibility.Visible)
        GuardWidget.SetVisibility(widget_visibility.Visible)
        GuardName.SetVisibility(widget_visibility.Visible)

        BossWidget.SetVisibility(widget_visibility.Hidden)
        BossWidget.SetVisibility(widget_visibility.Hidden)
        BossName.SetVisibility(widget_visibility.Hidden)
        set GuardVisibilityTimer = 5.0
     
    BossCountdown(TimeLeft : int):void=
        var Text :string="<Countdown{TimeLeft}>"

        if:
            TimeLeft > 10
        then:
            set Text = " {TimeLeft}s"

        if:
            TimeLeft <= 0
        then:
            set Text ="<CountdownFinish>"
            PlayerManager.GameManager.AudioManager.DamageMasteryTimerEnd.Play(PlayerManager.Agent)
        else:
            PlayerManager.GameManager.AudioManager.DamageMasteryTimerTick.Play(PlayerManager.Agent)
        
        if:
            TimeLeft <= 3
        then:
            PlayerManager.GameManager.AudioManager.Countdown(PlayerManager.Agent, TimeLeft)

        SetBossTimerText(Text, ?ImgSize:= vector2{X:=120.0, Y:=120.0})

    GunCountdown(TimeLeft : int):void=
        if:
            TimeLeft <=0
        then:
            return

        if:
            not TimeLeft = GunDurationTimeLeft
        then:
            set GunDurationTimeLeft = TimeLeft
        else:
            return

        if:
            TimeLeft <= 3
        then:
            PlayerManager.GameManager.AudioManager.Countdown(PlayerManager.Agent, TimeLeft)
        else:
            PlayerManager.GameManager.AudioManager.DamageMasteryTimerTick.Play(PlayerManager.Agent)


    UpdateGuardPlayerCount(GuardID : string, PlayerCount : float, IsQuestEnemy : logic):void=
        if:
            not(ActiveTarget = GuardID)
        then:
            return

    QueueUpdateGuardUI(current_health :  big_number, old_health : big_number, max_health : big_number , Name: string, AttackingPlayerCountIn : int, KOBonus : big_number, ClearBonus : big_number, GuardID : string, IsUpdateFromUIOwner : logic, IsBoss : logic, IsDamageMastery : logic, ?ArmorValue : ?big_number = false, ?DamageMasteryMult : float = 0.0) : void =
        if:
            not IsUpdateFromUIOwner?
            not(ActiveTarget = GuardID)
        then:
            return

        if:
            QueuedGuardID = ""
        then:
            set QueuedOldHealth = old_health
            spawn:
                StartUpdateGuardUI()
        
        set QueuedGuardID = GuardID
        set QueuedCurrentHealth = current_health
        set QueuedMaxHealth = max_health
        set QueuedName = Name
        set QueuedIsUpdateFromUIOwner = IsUpdateFromUIOwner
        set QueuedAttackingPlayerCount = AttackingPlayerCountIn
        if:
            Armor:=ArmorValue?
        then:
            set QueuedArmorValue = Armor
        set QueuedKOBonus = KOBonus
        set QueuedClearBonus = ClearBonus
        set QueuedIsBoss =  IsBoss
        set QueuedIsDamageMastery = IsDamageMastery
        set QueuedDamageMasteryMult = DamageMasteryMult

    StartUpdateGuardUI()<suspends>:void=
        Sleep(0.1)
        UpdateGuardUI(QueuedCurrentHealth, QueuedOldHealth, QueuedMaxHealth, QueuedName, QueuedAttackingPlayerCount, QueuedKOBonus, QueuedClearBonus, QueuedGuardID, QueuedIsUpdateFromUIOwner, QueuedIsBoss, QueuedIsDamageMastery, ?ArmorValue := option{QueuedArmorValue} ,?DamageMasteryMult := QueuedDamageMasteryMult)
        set QueuedOldHealth = big_number{}
        set QueuedGuardID = ""
        set QueuedCurrentHealth = big_number{}
        set QueuedMaxHealth = big_number{}
        set QueuedName = ""
        set QueuedIsUpdateFromUIOwner = false
        set QueuedAttackingPlayerCount = -1
        set QueuedArmorValue = big_number{}
        set QueuedKOBonus = big_number{}
        set QueuedClearBonus = big_number{}
        set QueuedIsBoss =  false
        set QueuedIsDamageMastery = false
        set QueuedDamageMasteryMult = 0.0

    UpdateGuardUI(current_health :  big_number, old_health : big_number, max_health : big_number , Name: string, AttackingPlayerCountIn : int, KOBonus : big_number, ClearBonus : big_number, GuardID : string, IsUpdateFromUIOwner : logic, IsBoss : logic, IsDamageMastery : logic, ?ArmorValue : ?big_number = false, ?DamageMasteryMult : float = 0.0) : void =
        var IsNewTarget : logic = false
        if(IsUpdateFromUIOwner?):
            if(not(ActiveTarget = GuardID)):
                set ActiveTarget = GuardID
                set IsNewTarget = true
        else:
            if(not(ActiveTarget = GuardID)):
                return
        percent:=BigNumber_Percentage(current_health, max_health)
        old_percent := BigNumber_Percentage(old_health, max_health)

        if:
            Armor := ArmorValue?
            Armor.Value > 0.0
        then:
            if:
                IsBoss?
            then:
                BossArmorImgWidget.SetImage(Images.Body)
                BeginBounceEffect(BossArmorImgWidget, vector2{X := 40.0, Y := 40.0}, IconBounce)
                BossArmorText.SetText(STM(" {Armor}"))
            else:
                GuardArmorImgWidget.SetImage(Images.Body)
                BeginBounceEffect(GuardArmorImgWidget, vector2{X := 40.0, Y := 40.0}, IconBounce)
                GuardArmorText.SetText(STM(" {Armor}"))
        else:
            BossArmorImgWidget.SetImage(Images.Empty)
            GuardArmorImgWidget.SetImage(Images.Empty)
            BossArmorText.SetText(STM(""))
            GuardArmorText.SetText(STM(""))

        if:
            TeamworkMultiplier := PlayerManager.GetUpgradeValueSafe("Teamwork", false) > 0.0
        then:
            if:
                IsBoss?
            then:
                BossTeamworkImgWidget.SetImage(Images.Teamwork)
            else:
                GuardTeamworkImgWidget.SetImage(Images.Teamwork)

            set AttackingPlayerCount = AttackingPlayerCountIn 
            if:
                IsBoss?
            then:
                BeginBounceEffect(BossTeamworkImgWidget, vector2{X := 40.0, Y := 40.0}, IconBounce)
                BossTeamworkText.SetText(STM(" {FloatToTruncatedString(AttackingPlayerCount * TeamworkMultiplier, 0)}%"))
            else:
                BeginBounceEffect(GuardTeamworkImgWidget, vector2{X := 40.0, Y := 40.0}, IconBounce)
                GuardTeamworkText.SetText(STM(" {FloatToTruncatedString(AttackingPlayerCount * TeamworkMultiplier, 0)}%"))
        else:
            BossTeamworkImgWidget.SetImage(Images.Empty)
            GuardTeamworkImgWidget.SetImage(Images.Empty)
                

        if(IsBoss? or IsDamageMastery?):
            ShowBossUI(IsDamageMastery)
            BossName.SetText(STM("{Name}: "))
            if(IsNewTarget? or BigNumber_Equals(old_health, max_health)?):
                BossHealth.PrepareProgressForNextBar(old_percent)

            if(not(percent = BossHealth.GetProgress())):
                if:
                    percent > 0.0
                then:
                    BossHealth.SetProgress(percent, true, ?IsNewTarget := IsNewTarget)
        else:
            ShowGuardUI()
            GuardName.SetText(STM("{Name}: "))
            if(IsNewTarget? or BigNumber_Equals(old_health, max_health)?):
                GuardHealth.PrepareProgressForNextBar(old_percent)

            if(not(percent = GuardHealth.GetProgress())):
                if:
                    percent > 0.0
                then:
                    GuardHealth.SetProgress(percent, true, ?IsNewTarget := IsNewTarget)

        if(not(BigNumber_Equals(KOBonus, LastKnownKOBonus)?)):
            set LastKnownKOBonus = KOBonus
            if(KOBonus.Value > 0.0 and not(IsDamageMastery?)):
                BossKOBonusText.SetText(STM("{FormatBigNumber(KOBonus)}"))
                GuardKOBonusText.SetText(STM("{FormatBigNumber(KOBonus)}"))
                if(IsBoss?):
                    # BossKOBonusImgWidget.SetImage(Images.KOBonus)
                else:
                    # GuardKOBonusImgWidget.SetImage(Images.KOBonus)
            else:
                BossKOBonusText.SetText(STM(""))
                GuardKOBonusText.SetText(STM(""))
                GuardKOBonusImgWidget.SetImage(ImgEmpty)
                BossKOBonusImgWidget.SetImage(ImgEmpty)
                
        if(not(BigNumber_Equals(ClearBonus, LastKnownClearBonus)?)):
            set LastKnownClearBonus = ClearBonus

            if(ClearBonus.Value > 0.0 and not(IsDamageMastery?)):
                BossClearBonusText.SetText(STM(" {FormatBigNumber(ClearBonus)}"))
                GuardClearBonusText.SetText(STM(" {FormatBigNumber(ClearBonus)}"))
                if(IsBoss?):
                    # BossClearBonusImgWidget.SetImage(Images.GemHunting)
                else:
                    # GuardClearBonusImgWidget.SetImage(Images.GemHunting)
            else:
                BossClearBonusText.SetText(STM(""))
                GuardClearBonusText.SetText(STM(""))
                GuardClearBonusImgWidget.SetImage(ImgEmpty)
                BossClearBonusImgWidget.SetImage(ImgEmpty)

        if(IsDamageMastery?):
            BossClearBonusText.SetText(STM("Damage Mastery Multiplier: {FloatToTruncatedString(DamageMasteryMult,0)}x"))

        if(percent <= 0.0):
            
            #Resetting ActiveTarget treats a reincarnated Guard like a new target - even if its the same target
            set ActiveTarget = ""
            
            if(IsBoss?):
                RandomGuardHealthText(BossHealthText)
                BossName.SetText(STM(""))
                BossHealth.BeginBounceEffect()
                BossHealth.SetProgress(percent, true, ?IsNewTarget := IsNewTarget)
            else:
                GuardName.SetText(STM(""))
                RandomGuardHealthText(GuardHealthText)
                GuardHealth.BeginBounceEffect()
                GuardHealth.SetProgress(percent, true, ?IsNewTarget := IsNewTarget)
        else:

            BossHealthText.SetText(STM("  {FormatBigNumber(current_health)} / {FormatBigNumber(max_health)} HP"))
            GuardHealthText.SetText(STM("  {FormatBigNumber(current_health)} / {FormatBigNumber(max_health)} HP"))

    RandomGuardHealthText(Text : text_block):void=
        rand := GetRandomInt(0,22)
        case (rand):
            0 => Text.SetText(STM("Obliterated!"))
            1 => Text.SetText(STM("Wiped Out!"))
            2 => Text.SetText(STM("Decimated!"))
            3 => Text.SetText(STM("Smashed!"))
            5 => Text.SetText(STM("Crushed!"))
            6 => Text.SetText(STM("Destroyed!"))
            7 => Text.SetText(STM("Demolished!"))
            8 => Text.SetText(STM("Destroyed!"))
            9 => Text.SetText(STM("Vaporized!"))
            10 => Text.SetText(STM("Definitely dead!"))
            11 => Text.SetText(STM("Smoked!"))
            13 => Text.SetText(STM("Erased!"))
            14 => Text.SetText(STM("Ravaged!"))
            15 => Text.SetText(STM("Sent to another dimension!"))
            16 => Text.SetText(STM("Sent to the shadow realm!"))
            17 => Text.SetText(STM("Target eliminated!"))
            19 => Text.SetText(STM("Shredded!"))
            20 => Text.SetText(STM("Nuked!"))
            21 => Text.SetText(STM("Knocked out!"))
            22 => Text.SetText(STM("It was super effective!"))
            _ => Text.SetText(STM("Destroyed!"))
    
    
    #Zone Functions
    UpdateAllZoneValues(ZoneTag:fetchable_tag, GuardsAlive:int, MaxGuards:int, PlayerClears:int, ClearsNeeded:int, ZoneLevel:int, ZoneOverkill:float):void=
        if:
            ZoneTag = fetchable_tag.midastag or ZoneTag = fetchable_tag.zone3tag 
        then:
            spawn:
                IterateRightsideTransition(ZoneTransitionWidget, 0.0)
            return

        ZoneName.SetText(STM(ZoneTag.ToZoneName()))

        UpdateZoneKOs(ZoneTag, false, GuardsAlive, MaxGuards, PlayerClears)
        UpdateZoneClears(ZoneTag, PlayerClears, ClearsNeeded, ZoneLevel, true)

    QueueGainWeaponExperience(ID : int, Level : int, EXP: int, Requirement:int):void=
        if:
            QueuedWeaponID = -1
        then:
            spawn:
                StartGainWeaponExperience()
        
        set QueuedWeaponID = ID
        set QueuedWeaponLevel = Level
        set QueuedWeaponExperience = EXP
        set QueuedWeaponExperienceRequirement = Requirement

    StartGainWeaponExperience()<suspends>:void=
        Sleep(0.1)
        GainWeaponExperience(QueuedWeaponID, QueuedWeaponLevel, QueuedWeaponExperience, QueuedWeaponExperienceRequirement)
        set QueuedWeaponID = -1
        set QueuedWeaponLevel = -1
        set QueuedWeaponExperience = -1
        set QueuedWeaponExperienceRequirement = -1

    GainWeaponExperience(ID : int, Level : int, EXP: int, Requirement: int)<suspends>:void=
        # PermaGunExpGunImg.SetImage(GetTextureFromString("Gun_{ID}"))
        EXPShow := MakeBigNumber(EXP * 1.0,1)
        ReqShow := MakeBigNumber(Requirement * 1.0,1)
        if:
            ID > 0
        then:
            PermaGunExpLevelText.SetText(STM("Lv. {Level}:"))
            PermaGunExpProgressText.SetText(STM("{EXPShow} / {ReqShow}"))
            PermaGunExpProgressBar.SetProgress((EXP*1.0)/(Requirement*1.0), true)
        else:
            PermaGunExpLevelText.SetText(STM(""))
            PermaGunExpProgressText.SetText(STM(""))
            PermaGunExpProgressBar.SetInvisible()
     


    
    QueueZoneKOs(ZoneTag:fetchable_tag, GuardKiller : player_manager, GuardsAlive:int, MaxGuards:int, PlayerClears:int):void=
        if:
            QueuedMaxGuards = 0
        then:
            spawn:
                StartZoneKOs()
        
        set QueuedZoneTag = ZoneTag
        set QueuedGuardKiller = option{GuardKiller.Agent}
        set QueuedGuardsAlive = GuardsAlive
        set QueuedMaxGuards = MaxGuards
        set QueuedPlayerClears = PlayerClears
    
    StartZoneKOs()<suspends>:void=
        Sleep(0.1)
        UpdateZoneKOs(QueuedZoneTag, QueuedGuardKiller, QueuedGuardsAlive, QueuedMaxGuards, QueuedPlayerClears)
        set QueuedZoneTag = fetchable_tag.none_tag
        set QueuedGuardsAlive = -1
        set QueuedMaxGuards = 0
        set QueuedPlayerClears = -1
        set QueuedGuardKiller = false

    UpdateZoneKOs(ZoneTag : fetchable_tag, GuardKiller : ?agent, GuardsAlive:int, MaxGuards:int, PlayerClears:int):void=
        if:
            ZoneTag = fetchable_tag.midastag or ZoneTag = fetchable_tag.zone3tag 
        then:
            spawn:
                IterateRightsideTransition(ZoneTransitionWidget, 0.0)
            return

        if:
            SetGuardKiller := GuardKiller?
            # not(PlayerManager = SetGuardKiller)
        then:
            #ShowStashPlayer
            ShowStashPlayer(SetGuardKiller, GuardsAlive)

        KOPercent:= (GuardsAlive * 1.0) / (MaxGuards * 1.0)

        #Get Correct Icon per Zone
        GuardIcon.SetImage(GetTextureFromString(ZoneTag.ToGuardImage()))

        var DoUpdateProgress:logic = true
        #Wonky, but need a carve out for this otherwise the bar will go full.
        if(GuardsAlive < MaxGuards):
            set DoUpdateProgress = true
        if(not(ZoneKOsProgress.GetProgress()=KOPercent)):
            BeginBounceEffect(GuardIcon, GuardIconDefaultSize, IconBounce)
        ZoneKOsProgress.SetProgress(KOPercent, DoUpdateProgress, ?DoColorUpdate := false)
        ZoneKOsProgressText.SetText(STM(" {GuardsAlive} Alive"))

    UpdateNextGemText(ZoneTag : fetchable_tag, IsTimerOverride:logic):void=
        ZoneLevelCap := PlayerManager.GameManager.GetZoneLevelCap(ZoneTag)
        ZoneGemReward := ZoneTag.ToZoneGemGain()

        if: 
            ZoneLevel := PlayerManager.GetZoneLevel[ZoneTag]
            ZoneLevel >= ZoneLevelCap or ZoneGemReward = 0
        then:
            ZoneKOsProgress.FrontColorBlock.SetColor(MakeColorFromHex("CC362F"))

            if:
                ZoneMsgVisibilityTimer <= 0.0 or IsTimerOverride?
            then:
                NextGemText := "No <gem> left here."
                SetNextGemText(NextGemText)
        else:
            ZoneKOsProgress.FrontColorBlock.SetColor(ZoneKOsProgress.FrontColor)
            if:
                ZoneMsgVisibilityTimer <= 0.0 or IsTimerOverride?
            then:
                NextGemText := "Hunt all <{ZoneTag.ToGuardImage()}> for {FloatToTruncatedString(ZoneGemReward,0)} <gem>"
                SetNextGemText(NextGemText)

    UpdateZoneClears(ZoneTag:fetchable_tag, PlayerClears:int, ClearsNeeded:int, ZoneLevel:int, IsTimerOverride : logic):void=                        
        #Resetting ActiveTarget treats a reincarnated Guard like a new target - even if its the same target
        set ActiveTarget = ""

        if (PlayerManager.GetAchievementLevel["Damage Mastery"] > 0):
        else:
            ResetDamageMeter()
        
        ZoneLevelCap := PlayerManager.GameManager.GetZoneLevelCap(ZoneTag)
        UpdateNextGemText(ZoneTag, IsTimerOverride)

        if(ZoneLevel >= ZoneLevelCap):
            ClearsPercent:= 1.0
            if(not(ZoneClearsProgress.GetProgress()=ClearsPercent)):
                BeginBounceEffect(ClearsIcon, ClearsIconDefaultSize, IconBounce)
            ZoneClearsProgress.SetProgress(ClearsPercent, true)
            ZoneClearsProgressText.SetText(STM(" Max!"))
        else:
            ClearsPercent:= (PlayerClears * 1.0) / (ClearsNeeded * 1.0)
            if(not(ZoneClearsProgress.GetProgress()=ClearsPercent)):
                BeginBounceEffect(ClearsIcon, ClearsIconDefaultSize, IconBounce)
            ZoneClearsProgress.SetProgress(ClearsPercent, true)
            if:
                not(PlayerClears = 1)
            then:
                ZoneClearsProgressText.SetText(STM(" {PlayerClears} Clears"))
            else:
                ZoneClearsProgressText.SetText(STM(" {PlayerClears} Clear"))

    ShowZoneClearMessage(ZoneTag : fetchable_tag, DidGainGem : logic) : void =
        return
        #Disabled as of 10/5/2024

        # var Text : string = "{ZoneTag.ToZoneName()} <GemHunting> HUNT COMPLETE!<br><{ZoneTag.ToGuardImage()}> GUARDS respawning!"
        
        # set ZoneMsgVisibilityTimer = 4.0
        # GemsEarned := ZoneTag.ToZoneGemGain()

        # if(DidGainGem?):
        #     NextGemText:="Found {FloatToTruncatedString(GemsEarned,0)} <gem>!"
        #     set Text = Text + " " + NextGemText
        #     SetNextGemText(NextGemText)
        #     set ZoneMsgVisibilityTimer = 8.0
        
        # SetZoneOpenText(Text)
    
    ShowZoneUnlockMessageUI(ZoneTag:fetchable_tag) : void = 
        Text:="You can now access {ZoneTag.ToZoneName()}!"
        SetZoneOpenText(Text)

        set ZoneMsgVisibilityTimer = 10.0
    
    SetNextGemText(Text : string): void = 
        for (Slot : NextGemSlotArray):
            NextGemStackBox.RemoveWidget(Slot.Widget)
        
        TempArray:=ParseText(Text)
        set NextGemSlotArray = TempArray
        
        for (Slot : TempArray):
            var Skip : logic = false
            if:
                TempBlock:=text_block[Slot.Widget]
            then:
                TempText:=TempBlock.GetText()
                TempBlock.SetTextColor(MakeColorFromHex("CC362F"))
                if (TempText = "<br>"):
                    set Skip = true
        
            if(not(Skip?)):
                NextGemStackBox.AddWidget(Slot)  

    SetBossTimerText(Text : string, ?ImgSize : vector2 = vector2{X:=40.0, Y:=40.0}):void=
        for (Slot : TickingClockSlotArray):
            TickingClockStackbox.RemoveWidget(Slot.Widget)
        
        TempArray:=ParseText(Text, ?ImgSize := ImgSize)
        set TickingClockSlotArray = TempArray
        for (Slot : TempArray):
            TickingClockStackbox.AddWidget(Slot)

    SetGunTimerText(Text : string, ?ImgSize : vector2 = vector2{X:=40.0, Y:=40.0}):void=
        for (Slot : GunTickingClockSlotArray):
            GunTickingClockStackbox.RemoveWidget(Slot.Widget)
        
        TempArray:=ParseText(Text, ?ImgSize := ImgSize)
        set GunTickingClockSlotArray = TempArray
        for (Slot : TempArray):
            GunTickingClockStackbox.AddWidget(Slot)
        
    SetZoneOpenText(Text : string, ?ImgSize : vector2 = vector2{X:=40.0, Y:=40.0}):void=
        for (Slot : ZoneOpenSlotArray):
            ZoneOpenHorizontalStackBox1.RemoveWidget(Slot.Widget)
            ZoneOpenHorizontalStackBox2.RemoveWidget(Slot.Widget)
            ZoneOpenHorizontalStackBox3.RemoveWidget(Slot.Widget)
        
        TempArray:=ParseText(Text, ?ImgSize := ImgSize)
        set ZoneOpenSlotArray = TempArray

        var BreakCount : int = 0
        
        for (Slot : TempArray):
            var Skip : logic = false
            if:
                TempBlock:=text_block[Slot.Widget]
            then:
                TempText:=TempBlock.GetText()
                if (TempText = "<br>"):
                    set BreakCount += 1
                    set Skip = true
        
            if(not(Skip?)):
                case(BreakCount):
                    1 => ZoneOpenHorizontalStackBox2.AddWidget(Slot)
                    2 => ZoneOpenHorizontalStackBox3.AddWidget(Slot)
                    _ => ZoneOpenHorizontalStackBox1.AddWidget(Slot)

    ShowZoneLevelUpMessageUI(ZoneTag:fetchable_tag, ClearNeeded : int, CurrentZoneLevel : int) : void = 
        #Zone Level Increase means KOs are 0
        ZoneKOsProgress.PrepareProgressForNextBar(0.0)
        ZoneKOsProgress.SetProgress(0.0, true, ?DoColorUpdate:=false)

    ShowEnterZoneMessageUI(ZoneTag:fetchable_tag, GuardsAlive: int) : void = 
        if:
            ZoneTag = fetchable_tag.midastag
        then:
            block:
        else:
            Text:="Now entering {ZoneTag.ToZoneName()}!"
            SetZoneOpenText(Text)
            set ZoneMsgVisibilityTimer = 10.0

    ShowNewTeleporterUnlocked(Name : string) : void = 
        Text:="You can now use the <Gun_0> REMOTE to Teleport to {Name}!"
        
        SetZoneOpenText(Text)
        set ZoneMsgVisibilityTimer = 10.0
    
    ExitZone(ZoneTag:fetchable_tag):void=
        if:
            PlayerManager.CurrentZone = ZoneTag or PlayerManager.CurrentZone = fetchable_tag.none_tag
        then:
            spawn:
                IterateRightsideTransition(ZoneTransitionWidget, 0.0)

    IterateResourceTransition(ColorBlock:color_block, TargetSize:float)<suspends>:void=
        var CurrentSize : float = ColorBlock.GetDesiredSize().Y
        Delta : float = TargetSize - CurrentSize 
        Adjustment : float = 40.0

        if(Delta = 0.0):
            return

        loop:
            Sleep(0.01)
            set CurrentSize = ColorBlock.GetDesiredSize().Y
            if(Delta > 0.0):
                if(TargetSize - CurrentSize >= Adjustment):
                    ColorBlock.SetDesiredSize(vector2{Y:= CurrentSize + Adjustment})
                else:
                    if(TargetSize - CurrentSize < Adjustment):
                        ColorBlock.SetDesiredSize(vector2{Y:= TargetSize})
                        return
            else:
                if(CurrentSize - TargetSize >= Adjustment):
                    ColorBlock.SetDesiredSize(vector2{Y:= CurrentSize - Adjustment})                    
                else:
                    if(CurrentSize - TargetSize < Adjustment):
                        ColorBlock.SetDesiredSize(vector2{Y:= TargetSize})                        
                        return

    IterateRightsideTransition(ColorBlock:color_block, TargetSize:float)<suspends>:void=
        var CurrentSize : float = ColorBlock.GetDesiredSize().X
        Delta : float = TargetSize - CurrentSize 
        Adjustment : float = 40.0

        if(Delta = 0.0):
            return

        loop:
            Sleep(0.01)
            set CurrentSize = ColorBlock.GetDesiredSize().X
            if(Delta > 0.0):
                if(TargetSize - CurrentSize >= Adjustment):
                    ColorBlock.SetDesiredSize(vector2{X:= CurrentSize + Adjustment})
                else:
                    if(TargetSize - CurrentSize < Adjustment):
                        ColorBlock.SetDesiredSize(vector2{X:= TargetSize})
                        return
            else:
                if(CurrentSize - TargetSize >= Adjustment):
                    ColorBlock.SetDesiredSize(vector2{X:= CurrentSize - Adjustment})                    
                else:
                    if(CurrentSize - TargetSize < Adjustment):
                        ColorBlock.SetDesiredSize(vector2{X:= TargetSize})
                        return

    #Item Functions
    UpdateGunIcon(GunID : int, Dmg : big_number):void=
        if:
            GunID > 0
            CurrentStrength:=PlayerManager.GetDisplayStrength[]
        then:
            Texture := GetTextureFromString("Gun_{GunID}")
            GunIcon.SetImage(Texture)
            GunIcon.SetDesiredSize(GunIconDefaultSize)
            GunText.SetText(STM(""))            
            GunText2.SetText(STM(" x {Dmg * CurrentStrength}!")) 
            GunTransition.SetDesiredSize(vector2{Y:=225.0})
        
    RemoveGunIcon():void=
        GunIcon.SetDesiredSize(vector2{})
        GunText.SetText(STM(""))
        GunText2.SetText(STM(""))
        GunTransition.SetDesiredSize(vector2{Y:=175.0})

    #Resource Functions
    UpdateUICounts():void=
        if:
            PeelyCount := PlayerManager.GetCollectibleCount("Peely")
            CandyCount := PlayerManager.GetCandyCount[]
            CoinCount := PlayerManager.GetCollectibleCount("Midas Coins")
            GemCount := PlayerManager.GetCollectibleCount("Gems")
            KeyCount := PlayerManager.GetCollectibleCount("BossKey")
            KeyCap := PlayerManager.GetUpgradeValue["BossKeyCap", false]
            HammerCount:= PlayerManager.GetCollectibleCount("UpgradeHammer")
        then:
            UpdateHammerUI(HammerCount)
            UpdatePeelyUI(PeelyCount)
            UpdateCandyUI(CandyCount)
            UpdateCoinsUI(CoinCount)
            UpdateGemUI(GemCount, big_number{})
            UpdateKeyUI(KeyCount, big_number{}, KeyCap)

        if:
            CurrentQuests:=PlayerManager.GetCurrentQuests[]
        then:
            for (X := 0..CurrentQuests.Length - 1):
                if:
                    QuestStats := CurrentQuests[X]
                    QuestManager := PlayerManager.QuestManager?
                then:
                    QuestManager.UpdateQuestUI(QuestStats, ?IsNewQuest:=true)
    
    UpdateHammerUI(Total : big_number):void=
        if: 
            PlayerManager.GetAchievementLevel["Sharkey_NextTier"] > 0
        then:
            HammerWidget.SetVisibility(widget_visibility.Visible)
            DailyBonusWidget.SetVisibility(widget_visibility.Visible)

        set HammerAmount = Total
        HammerIcon.SetImage(Images.UpgradeHammer)
        HammerText.SetText(STM("  {HammerAmount}"))
        BeginBounceEffect(HammerIcon, HammerIconDefaultSize, IconBounce)

    UpdatePeelyUI(Total : big_number):void=
        set PeelyAmount = Total
        PeelyText.SetText(STM("  {PeelyAmount}"))
        BeginBounceEffect(PeelyIcon, PeelyIconDefaultSize, IconBounce)

    UpdateCandyUI(Total : big_number):void=
        return
        # if: 
        #     PlayerManager.GetAchievementLevel["CandyUnlocked"] > 0
        # then:
        #     spawn:
        #         IterateRightsideTransition(CandyTransitionWidget, 420.0)
        
        # set CandyAmount = Total
        # CandyText.SetText(STM("  {CandyAmount}"))
        # BeginBounceEffect(CandyIcon, CandyIconDefaultSize, IconBounce)

    CreateStrengthFloatingText(It : collectible_data)<suspends>:void =
        if:
            CandyID := ParseInt[It.Key.Slice[5, 6]]
        then:
            Str := big_number{Value:=1.0, Exponent:=CandyID} * PlayerManager.GetPlayerStat(player_stat.DamageMult)
            StrengthFloatingText.Create(It.GetUseImage(), " + {Str}")

    UpdateCoinsUI(Total : big_number):void=
        set CoinsAmount = Total
        CoinsText.SetText(STM("  {CoinsAmount}"))
        BeginBounceEffect(CoinsIcon, CoinsIconDefaultSize, IconBounce)

    #Gem Functions
    UpdateGemUI(Total:big_number, Gain:big_number, ?DidLevelUp: logic = false):void=
        set GemAmount = Total
        GemText.SetText(STM("  {FormatBigNumber(GemAmount)}"))
        BeginBounceEffect(GemIcon, GemIconDefaultSize, IconBounce)
        if:
            DidLevelUp?
            Level := PlayerManager.GetLevel[]
            Level > 1
        then:

            var Text : string = ""
            set ZoneMsgVisibilityTimer = 0.0

            if:
                Offset := PlayerManager.GetUpgradeLevel["P_ReducedUpgradeLevelRequirement"] or 0
                Mod[Level + Offset, 10] = 5
                Level + Offset > 5
            then:
                PlayerManager.GameManager.AudioManager.PlayUnlockZoneSFX(PlayerManager.Agent)
                set Text = "Reached <level> Level {Level}!!<br>New Upgrades Available in <Gun_0>"
                set ZoneMsgVisibilityTimer = 10.0
            
            SetZoneOpenText(Text)            

    UpdateKeyUI(Total:big_number, Gain:big_number, KeyCap : float):void=
        if: 
            PlayerManager.GetAchievementLevel["Lyra_NextTier"] > 0
        then:
            spawn:
                IterateRightsideTransition(KeyTransitionWidget, 420.0)
        
        if:
            KeyCap > 0.0
        then:
            set KeyAmount = Total
            KeyText.SetText(STM("  {FormatBigNumber(KeyAmount)}"))
            BeginBounceEffect(KeyIcon, KeyIconDefaultSize, IconBounce)
        else:
            KeyText.SetText(STM(""))
            KeyTimerText.SetText(STM(""))
            spawn:
                IterateRightsideTransition(KeyTransitionWidget, 0.0)

        if:
            KeyAmount.Value >= 2.0
            KeyCap > 0.0
        then:
            #Show Notification
            BossArrowIcon.SetImage(Images.NotificationArrowRight)
            BossRemoteIcon.SetImage(Images.Remote)
            BossNotificationIcon.SetImage(Images.BossTeleport)
        else:
            #Hide Notification
            BossArrowIcon.SetImage(Images.Empty)
            BossRemoteIcon.SetImage(Images.Empty)
            BossNotificationIcon.SetImage(Images.Empty)

        if:
            Gain.Value > 0.0
        then:
            var Text : string = "Gained a Boss Key! <BossKey>!<br> Bosses are ready to fight!" 
            PlayerManager.GameManager.AudioManager.PlayUnlockZoneSFX(PlayerManager.Agent)
            SetZoneOpenText(Text)
            set ZoneMsgVisibilityTimer = 10.0

    #Weapon Functions
    OnCooldownsUpdated(CooldownStates : []cooldown_state):void=
        var ShowGunCooldown : logic = true
        var KeepLoop : logic = true
        for (State : CooldownStates):
            if:
                KeepLoop?
                GunStats := PlayerManager.GameManager.GunManager.StatsTable[State.ID]
                Percent := (State.RemainingDuration * 1.0) / (GunStats.GetDuration(PlayerManager) * 1.0)
                ID := PlayerManager.GetHeldGun[]
            then:
                if (State.Slot = -1):
                    
                else:
                    
                    if:
                        State.RemainingDuration > 0
                        ID = State.ID
                    then:
                        set ShowGunCooldown = false
                        GunDurationIcon.SetImage(GetTextureFromString("Gun_{State.ID}"))
                        if:
                            PlayerManager.IsDailyWeapon(State.ID)?
                        then:
                            DailyBonusWeaponText.SetText(STM("Daily Bonus 2x!"))
                        else:
                            DailyBonusWeaponText.SetText(STM(""))

                        GunDuration.SetProgress(Percent, true)

                        if (PlayerManager.ShouldTickCooldown[State.ID]):
                            GunCountdown(State.RemainingDuration)
                            if:
                                State.RemainingDuration <= 60
                            then:
                                GunDurationText.SetText(STM(" {State.RemainingDuration}s"))
                            else:
                                if:
                                    Mins:= Int[State.RemainingDuration * 1.0 / 60.0]
                                    Secs:= Mod[State.RemainingDuration, 60]
                                then:
                                    if:
                                        Secs < 10
                                    then:
                                        GunDurationText.SetText(STM("{Mins}:0{Secs}"))
                                    else:
                                        GunDurationText.SetText(STM("{Mins}:{Secs}"))
                            BeginBounceEffect(GunDurationIcon , GunDurationIconDefaultSize, 1.05)
                        else:
                            GunDurationText.SetText(STM("PAUSED"))
                        
                        GunDurationWidget.Widget.SetVisibility(widget_visibility.Visible)           
                        GunDurationIcon.SetVisibility(widget_visibility.Visible)           
                        set KeepLoop = false
                    else:
                        GunDurationWidget.Widget.SetVisibility(widget_visibility.Hidden)
                        GunDurationIcon.SetVisibility(widget_visibility.Hidden)           
                        GunDurationText.SetText(STM(""))
            
        if:
            ShowGunCooldown?
        then:
            var LowestCD : int = 9999
            var LowestState : cooldown_state = cooldown_state{}
            for (State : CooldownStates):
                if: 
                    State.Slot = -1
                    State.RemainingCooldown > 0
                    State.RemainingCooldown < LowestCD
                then:
                    set LowestCD = State.RemainingCooldown
                    set LowestState = State

            if:
                GunStats := PlayerManager.GameManager.GunManager.StatsTable[LowestState.ID]
                Percent := (LowestState.RemainingCooldown * 1.0) / (GunStats.Cooldown * 1.0 - GunStats.GetDuration(PlayerManager) * 1.0)
            then:
                GunDurationIcon.SetImage(GetTextureFromString("Gun_{LowestState.ID}"))
                GunDuration.FrontColorBlock.SetColor(MakeColorFromHex("CC362F"))
                GunDuration.SetProgress(Percent, true, ?DoColorUpdate := false)

                if:
                    LowestState.RemainingCooldown <= 60
                then:
                    GunDurationText.SetText(STM(" {LowestState.RemainingCooldown}s"))
                else:
                    if:
                        Mins:= Int[LowestState.RemainingCooldown * 1.0 / 60.0]
                        Secs:= Mod[LowestState.RemainingCooldown, 60]
                    then:
                        if:
                            Secs < 10
                        then:
                            GunDurationText.SetText(STM("{Mins}:0{Secs}"))
                        else:
                            GunDurationText.SetText(STM("{Mins}:{Secs}"))

                BeginBounceEffect(GunDurationIcon , GunDurationIconDefaultSize, 1.05)
                GunDurationWidget.Widget.SetVisibility(widget_visibility.Visible)           
                GunDurationIcon.SetVisibility(widget_visibility.Visible)           
            else:
                GunDurationWidget.Widget.SetVisibility(widget_visibility.Hidden)
                GunDurationIcon.SetVisibility(widget_visibility.Hidden)           
                GunDurationText.SetText(STM(""))

        var GunsAvailableText : string = ""

        if:
            PlayerManager.GameManager.CheckSaveDataForPlayer[PlayerManager.Agent]
            SourceData := PlayerProfileDataMap[PlayerManager.Player]
        then:
            for (i := 0..99):
                if:
                    i > 0
                    PlayerManager.GetUpgradeLevel["Gun_{i}"] > 0
                    PlayerManager.HasCompletedTask(0,2)?
                    PlayerManager.GetRemainingCooldown[i] = 0
                then:
                    ##Add to Available List
                    set GunsAvailableText = GunsAvailableText + "<Gun_{i}>"

        for (Slot : GunsAvailableArray):
            GunsAvailableStackbox.RemoveWidget(Slot.Widget)

        if:
            GunsAvailableText = ""
        then:
            block:
        else:
            set GunsAvailableText = GunsAvailableText + " Available"
            TempArray:=ParseText(GunsAvailableText, ?ImgSize := vector2{X:=60.0, Y:=60.0})
            set GunsAvailableArray = TempArray

            for (Slot : GunsAvailableArray):
                GunsAvailableStackbox.AddWidget(Slot)
    

    HideGunDurationWidget():void=
        GunDurationWidget.Widget.SetVisibility(widget_visibility.Hidden)
        GunDurationIcon.SetVisibility(widget_visibility.Hidden)           
        GunDurationText.SetText(STM(""))

    #Quest Functions
    SetQuestText(Text : string, ?DoesSkipCompleteFlair : logic = false):void=
        #Popup
        if:
            not DoesSkipCompleteFlair?
        then:
            spawn:
                QuestPopup()
        else:
            spawn:
                PrepareToBumpNewQuest(10)  

        for (Slot : QuestSlotArray):
            QuestHorizontalStackBox1.RemoveWidget(Slot.Widget)
            QuestHorizontalStackBox2.RemoveWidget(Slot.Widget)
            QuestHorizontalStackBox3.RemoveWidget(Slot.Widget)
        
        TempArray:=ParseText(Text)
        set QuestSlotArray = TempArray

        var BreakCount : int = 0
        
        for (Slot : TempArray):
            var Skip : logic = false
            if:
                TempBlock:=text_block[Slot.Widget]
            then:
                TempText:=TempBlock.GetText()
                if (TempText = "<br>"):
                    set BreakCount += 1
                    set Skip = true
        
            if(not(Skip?)):
                case(BreakCount):
                    1 => QuestHorizontalStackBox2.AddWidget(Slot)
                    2 => QuestHorizontalStackBox3.AddWidget(Slot)
                    _ => QuestHorizontalStackBox1.AddWidget(Slot)
            

    QuestPopup()<suspends>:void=
        PlayerManager.GameManager.AudioManager.TaskCompletePlayer.Play(PlayerManager.Agent)
        QuestCompleteImg.SetImage(Images.QuestPopup)
        Sleep(2.0)
        QuestCompleteImg.SetImage(Images.Empty)
        spawn:
            PrepareToBumpNewQuest(10)  

    QueueQuestProgress(CompoundProgresses : []quest_progress_info) : void=
        # Print("In QueueQuestProgress (1)")
        for (X := 0..MaxQuestProgressBars -1):
            if:
                Progress := CompoundProgresses[X]
            then:
                # Print("Queueing {Progress.Current} / {Progress.Total} for ID {X}, IconText {Progress.IconText}")
                QueueQuestProgress(Progress.Current, Progress.Total, Progress.IconText, ?ID := X)
            else:
                # Print("Queueing EMPTY for ID {X}")
                QueueQuestProgress(big_number{}, big_number{}, "",?ID := X)

    QueueQuestProgress(NewProgress : big_number, RequiredProgress : big_number, TextureString : string, ?ID : int = 0) : void=
        for (X := QueuedQuestProgress.Length .. ID):
            set QueuedQuestProgress = QueuedQuestProgress + array{big_number{}}
            set QueuedQuestRequirement = QueuedQuestRequirement + array{big_number{}}
            set QueuedQuestTextureString = QueuedQuestTextureString + array{""}

        # Print("In QueueQuestProgress (2)")
        if:
            NewProgress.Value > 0.0 or RequiredProgress.Value > 0.0
            QueuedQuestProgress[ID].Value = 0.0 and QueuedQuestRequirement[ID].Value = 0.0
        then:
            spawn:
                StartQuestProgress()
        
        if:
            set QueuedQuestProgress[ID] = NewProgress
            set QueuedQuestRequirement[ID] = RequiredProgress
            set QueuedQuestTextureString[ID] = TextureString
        

    var QuestProgressRunning : logic = false
    StartQuestProgress()<suspends> : void=
        if(QuestProgressRunning?):
            return
        set QuestProgressRunning = true
        Sleep(0.1)
        for (ID := 0..QueuedQuestProgress.Length - 1; Progress := QueuedQuestProgress[ID]; Requirement := QueuedQuestRequirement[ID]; TextureString := QueuedQuestTextureString[ID]):
            UpdateQuestProgress(Progress, Requirement, ID, TextureString)
            if:
                set QueuedQuestProgress[ID] = big_number{}
                set QueuedQuestRequirement[ID] = big_number{}
                set QueuedQuestTextureString[ID] = ""
        set QuestProgressRunning = false

    UpdateQuestProgress(NewProgress : big_number, RequiredProgress : big_number, ID : int, TextureString : string) : void=
        if:
            SelectedQuestProgress := QuestProgress[ID]
            SelectedQuestProgressText := QuestProgressText[ID]
            SelectedQuestIcon := QuestIcons[ID]
        then:
            # Print("======")
            # Print("Texture String: {TextureString}")
            # Print("Bar ID: {ID}")
            # Print("Progress: {NewProgress}")

            if:
                RequiredProgress.Value <= 1.0
                TextureString = ""
            then:
                SelectedQuestProgress.SetInvisible()
                SelectedQuestIcon.SetImage(Images.Empty)
                SelectedQuestProgressText.SetText(STM(""))
                return

            SelectedQuestIcon.SetImage(GetTextureFromString(TextureString))

            SelectedQuestProgress.SetVisible()
            Percent : float = Min(BigNumber_Percentage(NewProgress , RequiredProgress), 1.0)

            SelectedQuestProgress.SetProgress(Percent, true)
            SelectedQuestProgressText.SetText(STM("{NewProgress} / {RequiredProgress}"))

    PrepareToBumpNewQuest(Duration : int)<suspends>:void=
        PlayerManager.GameManager.AudioManager.PlayNewTaskSFX(PlayerManager.Agent)
        for (X:= 0..Duration):
            Sleep(1.0)
            if:
                IsNewQuestBouncing?
            else:
                BumpNewQuest(Duration - X)
                return

    BumpNewQuest(Duration : int)<suspends>:void=
        set IsNewQuestBouncing = true
        for (X:= 0..Duration):
            Sleep(1.0)
            BeginBounceEffect(QuestHeaderImg, QuestHeaderImgDefaultSize, 1.05)
        set IsNewQuestBouncing = false

    #Damage Functions
    ResetDamageMeter():void=
        block:

    QueueDamageDealt(AgentIn : agent, Damage : big_number):void=
        if:
            #Don't do Damage Meters
            true?
            # PlayerManager.GetAchievementLevel["EnteredMidas"] = 0
        then:
            return

        if:
            QueuedDamageDealt.Length = 0
        then:
            spawn:
                StartDamageDealt()

        if:
            Value := QueuedDamageDealt[AgentIn]
            set QueuedDamageDealt[AgentIn] = BigNumber_Add(Value, Damage)
        else:
            if:
                set QueuedDamageDealt[AgentIn] = Damage

    StartDamageDealt()<suspends>:void=
        block:

    UpdateDamageDealt(AgentIn : agent, Damage : big_number):void=
        block:
    
    #Damage Mastery Functions
    UpdateDamageMasteryTimer(TimeLeft : int):void=
        if:
            TimeLeft = 60
        then:
            ResetDamageMeter()
            PlayerManager.GameManager.AudioManager.PlayDamageMasteryTimerStart(PlayerManager.Agent)
        else:
            PlayerManager.GameManager.AudioManager.PlayDamageMasteryTimerTick(PlayerManager.Agent)

        if:
            TimeLeft <= 10
        then:
            TickingClockImg.SetImage(Images.Empty)
            TickingClockImg.SetDesiredSize(vector2{X := 0.0, Y := 0.0})
        else:
            TickingClockImg.SetImage(Images.TickingClock)
            TickingClockImg.SetDesiredSize(vector2{X := 80.0, Y := 80.0})
            BeginBounceEffect(TickingClockImg, TickingClockImgDefaultSize, IconBounce)

    UpdateDamageMasteryTimerBossVersionHack(TimeLeft : int):void=
        PlayerManager.GameManager.AudioManager.PlayDamageMasteryTimerTick(PlayerManager.Agent)

        if:
            TimeLeft <= 10
        then:
            TickingClockImg.SetImage(Images.Empty)
            TickingClockImg.SetDesiredSize(vector2{X := 0.0, Y := 0.0})
        else:
            TickingClockImg.SetImage(Images.TickingClock)
            TickingClockImg.SetDesiredSize(vector2{X := 80.0, Y := 80.0})
            BeginBounceEffect(TickingClockImg, TickingClockImgDefaultSize, IconBounce)
        
    UpdateDamageMasteryText(MasteryLevel : int, MasteryMult : float):void=
        PlayerManager.GameManager.AudioManager.PlayDamageMasteryTimerStart(PlayerManager.Agent)
        BossHealth.BeginBounceEffect()

    ShowDamageMasteryReport(TotalDamage : big_number, EssenceGained : big_number, EssenceStashed : big_number):void=
        PlayerManager.GameManager.AudioManager.PlayDamageMasteryTimerEnd(PlayerManager.Agent)
        var Text : string = "<Damage Mastery> Damage Mastery Complete <Damage Mastery><br>Total Damage: {FormatBigNumber(TotalDamage)}! <essence> Essence Gained: {FormatBigNumber(EssenceGained)}!"

        if: 
            EssenceStashed.Value > 0.0
        then:
            set Text = Text + "<br><stash> Essence Stashed: {FormatBigNumber(EssenceStashed)}!"

        SetZoneOpenText(Text)
        set ZoneMsgVisibilityTimer = 10.0

    HideDamageMasteryTimer():void=
        for (Slot : TickingClockSlotArray):
            TickingClockStackbox.RemoveWidget(Slot.Widget)

        set TickingClockSlotArray = array{}
        TickingClockImg.SetImage(Images.Empty)
        
        BossWidget.SetVisibility(widget_visibility.Hidden)
        BossName.SetVisibility(widget_visibility.Hidden)

    #Gun Exp Functions
    var IsShowingGunEXP : logic = false
    UpdateGunExp(ID : int, Level : int, NewExp : int, Req : int, OldLevel : int, GainAmount : int)<suspends>:void=
        #Wait for previous gun exp popupto close
        loop:
            if (not IsShowingGunEXP?):
                break
            Sleep(1.0)
        
        set IsShowingGunEXP = true
        GunExpProgressBar.SetInvisible()
        GunExpTitleImg.SetImage(Images.ChestDestroyed)
        GunExpBG.SetImage(Images.StoreBox)
        GunExpBackPlateImg.SetImage(Images.StoreBox)
        PlayerManager.GameManager.AudioManager.DamageMasteryTimerStart.Play(PlayerManager.Agent)

        var GunIDs : []int = array{}
        for (i := 0..PlayerManager.GameManager.GunManager.StatsTable.Length-1):
            if:
                i > 0
                GunStat := PlayerManager.GameManager.GunManager.StatsTable[i]
            then:
                set GunIDs = GunIDs + array{i}

        LoopGuns(GunIDs)

        PlayerManager.GameManager.AudioManager.DamageMasteryTimerEnd.Play(PlayerManager.Agent)
        GunExpGunImg.SetImage(GetTextureFromString("Gun_{ID}"))
        var UpdateText : string = "Gained {GainAmount} XP!"
        if:
            Level > OldLevel
        then:
            set UpdateText = UpdateText + " Level up!"
        GunExpUpdateText.SetText(STM(UpdateText))
        GunExpLevelText.SetText(STM("Level {Level}:"))
        GunExpProgressBar.SetVisible()
        GunExpProgressBar.SetProgress((NewExp * 1.0)/(Req * 1.0), true)
        GunExpProgressText.SetText(STM("{NewExp} / {Req}"))
        GunExpNextLevelText.SetText(STM("Next Lv.: +{Level * 10}% Dmg!"))
        HideGunText()
        set IsShowingGunEXP = false

    HideGunText()<suspends>:void=
        Sleep(3.0)
        PlayerManager.ReleaseStasis()
        GunExpBG.SetImage(Images.Empty)
        GunExpTitleImg.SetImage(Images.Empty)
        GunExpBackPlateImg.SetImage(Images.Empty)
        GunExpGunImg.SetImage(Images.Empty)
        GunExpUpdateText.SetText(STM(""))
        GunExpLevelText.SetText(STM(""))
        GunExpProgressBar.SetProgress(0.0, true)
        GunExpProgressBar.SetInvisible()
        GunExpProgressText.SetText(STM(""))
        GunExpNextLevelText.SetText(STM(""))
        
    LoopGuns(GunIDs : []int)<suspends>:void=
        base : float = 1.01
        var iteration : float = 1.0
        race:
            Sleep(GetRandomFloat(3.1,4.3))
            loop:
                for (ID : GunIDs):
                    GunExpGunImg.SetImage(GetTextureFromString("Gun_{ID}"))
                    PlayerManager.GameManager.AudioManager.DamageMasteryTimerTick.Play(PlayerManager.Agent)
                    Sleep((0.15)*(Pow(base, iteration)))
                    set iteration += 1.0                    
        
    #Buff Functions
    UpdateHandcannon(Handcannon : int, SuperHandcannon : int, IsSuperHandcannonUnlocked : logic):void=
        if:
            HandCannonProgresses.Length = 0
        then:
            SetupHandCannon()

        if:
            TempBar1 := HandCannonProgresses[0]
            TempBar2 := HandCannonProgresses[1]
            TempBar3 := HandCannonProgresses[2]
        then:
            case (Handcannon):
                1 =>
                    TempBar1.SetProgress(1.0, true)
                    TempBar2.SetProgress(0.0, true)
                    TempBar3.SetProgress(0.0, true)
                2 =>
                    TempBar1.SetProgress(1.0, true)
                    TempBar2.SetProgress(1.0, true)
                    TempBar3.SetProgress(0.0, true)
                3 =>
                    TempBar1.SetProgress(1.0, true)
                    TempBar2.SetProgress(1.0, true)
                    TempBar3.SetProgress(1.0, true)
                _ =>
                    TempBar1.SetProgress(0.0, true)
                    TempBar2.SetProgress(0.0, true)
                    TempBar3.SetProgress(0.0, true)

        
        if:
            IsSuperHandcannonUnlocked?
            SuperHandCannonProgresses.Length = 0
        then:
            SetupSuperHandCannon()

        var SuperTemp : int = SuperHandcannon   
        if:
            IsSuperHandcannonUnlocked?
            TempBar1 := SuperHandCannonProgresses[0]
            TempBar2 := SuperHandCannonProgresses[1]
            TempBar3 := SuperHandCannonProgresses[2]
        then:
            if:
                Handcannon = 4
            then:
                set SuperTemp += 1

            case (SuperTemp):
                1 =>
                    TempBar1.SetProgress(1.0, true)
                    TempBar2.SetProgress(0.0, true)
                    TempBar3.SetProgress(0.0, true)
                2 =>
                    TempBar1.SetProgress(1.0, true)
                    TempBar2.SetProgress(1.0, true)
                    TempBar3.SetProgress(0.0, true)
                3 =>
                    TempBar1.SetProgress(1.0, true)
                    TempBar2.SetProgress(1.0, true)
                    TempBar3.SetProgress(1.0, true)
                _ =>    
                    TempBar1.SetProgress(0.0, true)
                    TempBar2.SetProgress(0.0, true)
                    TempBar3.SetProgress(0.0, true)


    UpdateBuffDurations(BuffMap : [string]buff, IsBounce : logic):void=
        for (Buff : BuffMap):
            if:
                not(BuffSlot[Buff.Name])
            then:
                if:
                    Buff.IsSecondary?
                then:
                    SetupBuffWidget(BuffHostWidget2, Buff)
                else:    
                    SetupBuffWidget(BuffHostWidget, Buff)

            if:
                TempIcon:= BuffIcon[Buff.Name]
                TempAmount:= BuffAmount[Buff.Name]
                TempDuration:= BuffDuration[Buff.Name]
                Slot:= BuffSlot[Buff.Name]
            then:
                TempAmount.SetText(STM(" {Buff.Amount}{Buff.Suffix}"))

                if:
                    Buff.Duration <= 60.0
                then:
                    TempDuration.SetText(STM(" | {FloatToTruncatedString(Buff.Duration,0)}s"))
                else:
                    if:
                        Mins:= Int[Buff.Duration * 1.0 / 60.0]
                        Secs:= Mod[Int[Buff.Duration], 60]
                    then:
                        if:
                            Secs < 10
                        then:
                            TempDuration.SetText(STM(" | {Mins}:0{Secs}"))
                        else:
                            TempDuration.SetText(STM(" | {Mins}:{Secs}"))

                if:
                    IsBounce?
                then:
                    BeginBounceEffect(TempIcon, BuffIconDefaultSize, 1.05)
    
                if:
                    Buff.Duration <= 0.0
                then:
                    Slot.Widget.SetVisibility(widget_visibility.Hidden)
                else:
                    Slot.Widget.SetVisibility(widget_visibility.Visible)
                    
# 
    SetupDailyBonusWidget():void=
        set DailyBonusWidget = overlay:
            Slots:= array:
                overlay_slot:
                    Widget:= DailyBonusBackground
                overlay_slot:
                    Widget:= stack_box:
                        Orientation:= orientation.Vertical
                        Slots:= array:
                            stack_box_slot:
                                Widget:= DailyBonusHeaderText
                            stack_box_slot:
                                Widget:= stack_box:
                                    Orientation:= orientation.Horizontal
                                    Slots:= array:
                                        stack_box_slot:
                                            Widget:= DailyBonusWeaponIcon
                                            Padding:= margin{Left:= 0.0, Top:=0.0, Bottom:= 0.0, Right:= 0.0} 
                                        stack_box_slot:
                                            Widget:= DailyBonusBossIcon
                                            Padding:= margin{Left:= 20.0, Top:=0.0, Bottom:= 0.0, Right:= 0.0} 
        DailyBonusWidget.SetVisibility(widget_visibility.Hidden)

    SetupBuffWidget(HostWidget : stack_box, Buff : buff):void=

        TempIcon := texture_block{ DefaultImage := GetTextureFromString(Buff.Name), DefaultDesiredSize := vector2{X := 80.0, Y := 80.0}} 
        TempAmount := text_block{DefaultText := STM(" {Buff.Amount}{Buff.Suffix}"),DefaultTextColor:= MakeColorFromHex("FFFFFF"),
            DefaultShadowColor:=NamedColors.Black,
            DefaultShadowOpacity := 1.0,
            DefaultShadowOffset:=option{vector2{X:=2.0,Y:=2.0}}}
        TempDuration := text_block{DefaultText := STM(" | {FloatToTruncatedString(Buff.Duration,0)}s"),DefaultTextColor:= MakeColorFromHex("CC362F"),
            DefaultShadowColor:=NamedColors.Black,
            DefaultShadowOpacity := 1.0,
            DefaultShadowOffset:=option{vector2{X:=2.0,Y:=2.0}}}

        TempStackBoxSlot := stack_box_slot:
            Widget:= overlay:
                Slots:= array:
                    overlay_slot:
                        Widget := color_block:
                            DefaultColor:= NamedColors.Black
                            DefaultOpacity:= 0.5
                            DefaultDesiredSize := vector2{X:= 240.0, Y:= 40.0}        
                    overlay_slot:
                        Widget:= stack_box:
                            Orientation := orientation.Horizontal
                            Slots:= array:
                                stack_box_slot:
                                    Widget:=TempIcon
                                stack_box_slot:
                                    Widget:=TempAmount
                                stack_box_slot:
                                    Widget:=TempDuration
        
        if:
            set BuffSlot[Buff.Name] = TempStackBoxSlot
            set BuffIcon[Buff.Name] = TempIcon
            set BuffAmount[Buff.Name] = TempAmount
            set BuffDuration[Buff.Name] = TempDuration
        
        HostWidget.AddWidget(TempStackBoxSlot)

    #Setup Functions
    SetupHandCannon():void=
        TempBar1 := MakeProgressBar(0.0, vector2{X:= 78.0, Y:= 18.5}, NamedColors.Goldenrod, MakeColorFromHex("000000"), MakeColorFromHex("FFFFFF"), ?InLimit := 3) #Red
        TempBar2 := MakeProgressBar(0.0, vector2{X:= 78.0, Y:= 18.5}, NamedColors.Goldenrod, MakeColorFromHex("000000"), MakeColorFromHex("FFFFFF"), ?InLimit := 3) #Red
        TempBar3 := MakeProgressBar(0.0, vector2{X:= 78.0, Y:= 18.5}, NamedColors.Goldenrod, MakeColorFromHex("000000"), MakeColorFromHex("FFFFFF"), ?InLimit := 3) #Red
        set HandCannonProgresses = array{TempBar1,TempBar2,TempBar3}

        HandCannonIcon.SetImage(Images.Handcannon)

        TempHandCannon := stack_box_slot:
            Widget:= overlay:
                Slots:= array:
                    overlay_slot:
                        Widget:= stack_box:
                            Orientation := orientation.Horizontal
                            Slots:= array:
                                stack_box_slot:
                                    Widget:=TempBar1.GetWidget()
                                    Padding:= margin{Left:=3.0, Top:=0.0, Bottom:= 0.0, Right:= 0.0} 
                                stack_box_slot:
                                    Widget:=TempBar2.GetWidget()
                                    Padding:= margin{Left:=3.0, Top:=0.0, Bottom:= 0.0, Right:= 0.0} 
                                stack_box_slot:
                                    Widget:=TempBar3.GetWidget()
                                    Padding:= margin{Left:=3.0, Top:=0.0, Bottom:= 0.0, Right:= 0.0} 
        
        HandCannonHostWidget.AddWidget(TempHandCannon)  

    SetupSuperHandCannon():void=        
        TempBar4 := MakeProgressBar(0.0, vector2{X:= 78.0, Y:= 18.5}, MakeColorFromHex("A31985"), MakeColorFromHex("000000"), MakeColorFromHex("FFFFFF"), ?InLimit := 3) #Red
        TempBar5 := MakeProgressBar(0.0, vector2{X:= 78.0, Y:= 18.5}, MakeColorFromHex("A31985"), MakeColorFromHex("000000"), MakeColorFromHex("FFFFFF"), ?InLimit := 3) #Red
        TempBar6 := MakeProgressBar(0.0, vector2{X:= 78.0, Y:= 18.5}, MakeColorFromHex("A31985"), MakeColorFromHex("000000"), MakeColorFromHex("FFFFFF"), ?InLimit := 3) #Red
        set SuperHandCannonProgresses = array{TempBar4,TempBar5,TempBar6}
        
        TempBar4.SetInvisible()
        TempBar5.SetInvisible()
        TempBar6.SetInvisible()

        SuperHandCannonIcon.SetImage(Images.SuperHandcannon)
          
        TempSuperHandCannon := stack_box_slot:
            Widget:= overlay:
                Slots:= array:
                    overlay_slot:
                        Widget:= stack_box:
                            Orientation := orientation.Horizontal
                            Slots:= array:
                                stack_box_slot:
                                    Widget:=TempBar4.GetWidget()
                                    Padding:= margin{Left:=3.0, Top:=0.0, Bottom:= 0.0, Right:= 0.0} 
                                stack_box_slot:
                                    Widget:=TempBar5.GetWidget()
                                    Padding:= margin{Left:=3.0, Top:=0.0, Bottom:= 0.0, Right:= 0.0} 
                                stack_box_slot:
                                    Widget:=TempBar6.GetWidget()
                                    Padding:= margin{Left:=3.0, Top:=0.0, Bottom:= 0.0, Right:= 0.0} 
                        Padding:= margin{Left:=0.0, Top:=3.0, Bottom:= 0.0, Right:= 0.0} 
        
        HandCannonHostWidget.AddWidget(TempSuperHandCannon)        
        

    SetupLevelNumbers():void=
        for (X:= 0..12):
            var TempLevelText : text_block = text_block{DefaultTextColor:=NamedColors.White,
            DefaultShadowColor:=NamedColors.Black,
            DefaultShadowOffset:=option{vector2{X:=2.0,Y:=2.0}}}
            TempLevelText.SetTextOpacity(0.0) 
            TempLevelText.SetShadowOpacity(0.0) 

            var TempLevelImg : texture_block = texture_block:
                DefaultImage := ImgLevelUp
                DefaultDesiredSize := LevelIconDefaultSize
            var TempColor : color_block = color_block{}
            TempColor.SetOpacity(0.0)
            TempColor.SetDesiredSize(vector2{Y:=3000.0})

            set LevelNumbers = LevelNumbers + array{TempLevelText}
            set LevelImgs = LevelImgs + array{TempLevelImg}
            set LevelColors = LevelColors + array{TempColor}

            TempOverlay:=overlay_slot:
                Widget:=overlay:
                    Slots:=
                        array:
                            overlay_slot:
                                Widget:=TempColor 
                            overlay_slot:
                                Widget:= stack_box:
                                    Orientation:= orientation.Horizontal
                                    Slots:= array:
                                        stack_box_slot:
                                            Widget:= TempLevelImg
                                            HorizontalAlignment:= horizontal_alignment.Left
                                            Padding:= margin{Left:=2.0, Top:=0.0, Bottom:= 0.0, Right:= 4.0} 
                                        stack_box_slot:
                                            Widget:= TempLevelText
                                            Padding:= margin{Left:=0.0, Top:=0.0, Bottom:= 0.0, Right:= 8.0} 
                                VerticalAlignment:= vertical_alignment.Top
                                HorizontalAlignment:= horizontal_alignment.Left
                HorizontalAlignment:= horizontal_alignment.Left

            LevelNumbersHostWidget.AddWidget(TempOverlay)

    SetupStashPlayers():void=
        for (X:= 0..10):
            var PlayerText : text_block = text_block{DefaultTextColor:=NamedColors.White,
            DefaultShadowColor:=NamedColors.Black,
            DefaultShadowOffset:=option{vector2{X:=2.0,Y:=2.0}}}
            PlayerText.SetTextOpacity(0.0) 
            PlayerText.SetShadowOpacity(0.0) 

            var TempStashImg : texture_block = texture_block:
                DefaultImage := Images.Chest
                DefaultDesiredSize := StashPlayerIconDefaultSize
                        
            var TempColor : color_block = color_block{}
            TempColor.SetOpacity(0.0)
            TempColor.SetDesiredSize(vector2{Y:=3000.0})

            set PlayerName1 = PlayerName1 + array{PlayerText}
                set StashIcon1 = StashIcon1 + array{TempStashImg}
            set StashPlayerColors = StashPlayerColors + array{TempColor}

            TempOverlay:=overlay_slot:
                Widget:=overlay:
                    Slots:=
                        array:
                            overlay_slot:
                                Widget:=TempColor 
                            overlay_slot:
                                Widget:= stack_box:
                                    Orientation:= orientation.Horizontal
                                    Slots:= array:
                                        stack_box_slot:
                                            Widget:= TempStashImg
                                            HorizontalAlignment:= horizontal_alignment.Left
                                            Padding:= margin{Left:=2.0, Top:=0.0, Bottom:= 0.0, Right:= 0.0} 
                                        stack_box_slot:
                                            Widget:= PlayerText
                                            Padding:= margin{Left:=2.0, Top:=0.0, Bottom:= 0.0, Right:= 0.0} 
                                VerticalAlignment:= vertical_alignment.Top
                                HorizontalAlignment:= horizontal_alignment.Left
                HorizontalAlignment:= horizontal_alignment.Left

            StashPlayerNameHostWidget.AddWidget(TempOverlay)

    SetupGunExpWidget():void=
        set GunExpProgressBar = MakeProgressBar(0.0, vector2{X:= 180.0, Y:= 50.0}, MakeColorFromHex("0072fc"), MakeColorFromHex("000000"), NamedColors.Goldenrod) #Exp Blue
        GunExpProgressBar.SetInvisible()

        TempOverlay:=stack_box_slot:
            Widget:= canvas:
                Slots:= array:
                    canvas_slot:
                        Widget:= GunExpBG
                        Offsets:= margin{Left:=-25.0, Top:= 0.0, Bottom:= 0.0, Right:= 0.0} 
                        ZOrder:=1
                    canvas_slot:
                        Widget:= GunExpTitleImg
                        Offsets:= margin{Left:=10.0, Top:= 15.0, Bottom:= 0.0, Right:= 0.0} 
                        ZOrder:=2
                    canvas_slot:
                        Widget:= overlay:
                            Slots:= array:
                                overlay_slot:
                                    Widget:= GunExpBackPlateImg
                                overlay_slot:
                                    Widget:= GunExpGunImg
                        Offsets:= margin{Left:=80.0, Top:= 160.0, Bottom:= 0.0, Right:= 0.0} 
                        ZOrder:=3
                    canvas_slot:
                        Widget:= GunExpUpdateText
                        Offsets:= margin{Left:=60.0, Top:=360.0, Bottom:= 0.0, Right:= 0.0} 
                        ZOrder:=4
                    canvas_slot:
                        Widget:= stack_box:
                            Orientation:= orientation.Horizontal
                            Slots:= array:
                                stack_box_slot:
                                    Widget:= GunExpLevelText
                                stack_box_slot:
                                    Widget:= overlay:
                                        Slots:= array:
                                            overlay_slot:
                                                Widget:= GunExpProgressBar.GetWidget()
                                            overlay_slot:
                                                Widget:= GunExpProgressText
                                    Padding:= margin{Left:=10.0, Top:=00.0, Bottom:= 0.0, Right:= 4.0} 
                        Offsets:= margin{Left:=30.0, Top:=420.0, Bottom:= 0.0, Right:= 0.0} 
                        ZOrder:=5
                    canvas_slot:    
                        Widget:= GunExpNextLevelText
                        Offsets:= margin{Left:=30.0, Top:=490.0, Bottom:= 0.0, Right:= 0.0} 
                        ZOrder:=6

        GunExpHostWidget.AddWidget(TempOverlay)

    SetupPermaGunExpWidget():void=
        set PermaGunExpProgressBar = MakeProgressBar(0.0, vector2{X:= 200.0, Y:= 40.0}, MakeColorFromHex("0072fc"), MakeColorFromHex("000000"), NamedColors.Goldenrod) #Exp Blue
        PermaGunExpProgressBar.SetInvisible()

        set PermaGunExpWidget = stack_box_slot:
            Widget:= stack_box:
                Orientation:= orientation.Horizontal
                Slots:= array:
                    stack_box_slot:
                        Widget:= PermaGunExpLevelText
                    stack_box_slot:
                        Widget:= overlay:
                            Slots:= array:
                                overlay_slot:
                                    Widget:= PermaGunExpProgressBar.GetWidget()
                                overlay_slot:
                                    Widget:= PermaGunExpProgressText
                                overlay_slot:
                                    Widget:= PermaGunExpGunImg
                                    Padding:= margin{Left:=0.0, Top:=0.0, Bottom:= 0.0, Right:= 0.0} 

        PermaGunExpWidget.Widget.SetVisibility(widget_visibility.Visible)
        BuffHostWidget2.AddWidget(PermaGunExpWidget)

    SetupDamageNumbers():void=
        for (X:= 0..20):
            var TempDamageText : text_block = text_block{DefaultTextColor:=NamedColors.White,
            DefaultShadowColor:=NamedColors.Black,
            DefaultShadowOffset:=option{vector2{X:=2.0,Y:=2.0}}}
            TempDamageText.SetTextOpacity(0.0) 
            TempDamageText.SetShadowOpacity(0.0) 
            
            var TempStashText : text_block = text_block{DefaultTextColor:=MakeColorFromHex("36CC42"),
            DefaultShadowColor:=NamedColors.Black,
            DefaultShadowOffset:=option{vector2{X:=2.0,Y:=2.0}}}
            TempStashText.SetTextOpacity(0.0) 
            TempStashText.SetShadowOpacity(0.0) 

            var TempDamageImg : texture_block = texture_block:
                DefaultImage := ImgMidasEssence
                DefaultDesiredSize := DamageNumbersSize
                        
            var TempStashImg : texture_block = texture_block:
                DefaultImage := ImgStash
                DefaultDesiredSize := vector2{X := 60.0, Y := 60.0}

            var TempColor : color_block = color_block{}
            TempColor.SetOpacity(0.0)
            TempColor.SetDesiredSize(vector2{Y:=3000.0})

            set DamageNumbers1 = DamageNumbers1 + array{TempDamageText}
            set DamageImgs1 = DamageImgs1 + array{TempDamageImg}
            
            set DamageNumbers2 = DamageNumbers2 + array{TempStashText}
            set DamageImgs2 = DamageImgs2 + array{TempStashImg}
            
            set DamageColors = DamageColors + array{TempColor}

            TempOverlay:=overlay_slot:
                Widget:=overlay:
                    Slots:=
                        array:
                            overlay_slot:
                                Widget:=TempColor 
                            overlay_slot:
                                Widget:= stack_box:
                                    Orientation:= orientation.Horizontal
                                    Slots:= array:
                                        stack_box_slot:
                                            Widget:= TempDamageImg
                                            HorizontalAlignment:= horizontal_alignment.Left
                                            Padding:= margin{Left:=2.0, Top:=0.0, Bottom:= 0.0, Right:= 4.0} 
                                        stack_box_slot:
                                            Widget:= TempDamageText
                                            Padding:= margin{Left:=0.0, Top:=0.0, Bottom:= 0.0, Right:= 8.0} 
                                        stack_box_slot:
                                            Widget:= TempStashImg
                                            HorizontalAlignment:= horizontal_alignment.Left
                                            Padding:= margin{Left:=2.0, Top:=0.0, Bottom:= 0.0, Right:= 4.0} 
                                        stack_box_slot:
                                            Widget:= TempStashText
                                            Padding:= margin{Left:=0.0, Top:=0.0, Bottom:= 0.0, Right:= 4.0} 
                                        # stack_box_slot:
                                        #     Widget:= TempOverkillText
                                        #     Padding:= margin{Left:=0.0, Top:=0.0, Bottom:= 0.0, Right:= 4.0} 
    
                                VerticalAlignment:= vertical_alignment.Top
                                HorizontalAlignment:= horizontal_alignment.Left
                HorizontalAlignment:= horizontal_alignment.Left

            DamageNumbersHostWidget.AddWidget(TempOverlay)

    SetupCollectibleNumbers():void=
        for (X:= 0..20):
            var TempCollectibleText : text_block = text_block{DefaultTextColor:=NamedColors.White,
            DefaultShadowColor:=NamedColors.Black,
            DefaultShadowOffset:=option{vector2{X:=2.0,Y:=2.0}}}
            TempCollectibleText.SetTextOpacity(0.0) 
            TempCollectibleText.SetShadowOpacity(0.0) 
            
            var TempStashText : text_block = text_block{DefaultTextColor:=MakeColorFromHex("36CC42"),
            DefaultShadowColor:=NamedColors.Black,
            DefaultShadowOffset:=option{vector2{X:=2.0,Y:=2.0}}}
            TempStashText.SetTextOpacity(0.0) 
            TempStashText.SetShadowOpacity(0.0) 

            var TempCollectibleImg : texture_block = texture_block:
                DefaultImage := ImgMidasEssence
                DefaultDesiredSize := DamageNumbersSize
                        
            var TempStashImg : texture_block = texture_block:
                DefaultImage := ImgStash
                DefaultDesiredSize := vector2{X := 60.0, Y := 60.0}

            var TempColor : color_block = color_block{}
            TempColor.SetOpacity(0.0)
            TempColor.SetDesiredSize(vector2{Y:=3000.0})

            set CollectibleNumbers1 = CollectibleNumbers1 + array{TempCollectibleText}
            set CollectibleImgs1 = CollectibleImgs1 + array{TempCollectibleImg}
            
            set CollectibleColors = CollectibleColors + array{TempColor}

            TempOverlay:=overlay_slot:
                Widget:=overlay:
                    Slots:=
                        array:
                            overlay_slot:
                                Widget:=TempColor 
                            overlay_slot:
                                Widget:= stack_box:
                                    Orientation:= orientation.Horizontal
                                    Slots:= array:
                                        stack_box_slot:
                                            Widget:= TempCollectibleImg
                                            HorizontalAlignment:= horizontal_alignment.Left
                                            Padding:= margin{Left:=2.0, Top:=0.0, Bottom:= 0.0, Right:= 4.0} 
                                        stack_box_slot:
                                            Widget:= TempCollectibleText
                                            Padding:= margin{Left:=0.0, Top:=0.0, Bottom:= 0.0, Right:= 8.0} 
                                VerticalAlignment:= vertical_alignment.Top
                                HorizontalAlignment:= horizontal_alignment.Left
                HorizontalAlignment:= horizontal_alignment.Left

            DamageNumbersHostWidget.AddWidget(TempOverlay)
        
    SetupStashNotificationWidget():void=
        set StashBounceColorWidget = color_block:
            DefaultColor:= NamedColors.Black
            DefaultOpacity:= 0.0
            DefaultDesiredSize := vector2{X:=1.0, Y:= 10.0}        

        set StashNotificationWidget = 
            overlay:
                Slots:= array:
                    overlay_slot:
                        Widget:= StashBounceColorWidget    
                    overlay_slot:
                        Widget := stack_box:
                            Orientation:= orientation.Vertical
                            Slots:=array:
                                stack_box_slot:
                                    Widget:=overlay:   
                                        Slots:= array:
                                            overlay_slot:
                                                Widget:= StashArrowIcon
                                            overlay_slot:
                                                Widget:= stack_box:
                                                    Orientation:= orientation.Horizontal
                                                    Slots:= array:
                                                        stack_box_slot: 
                                                            Widget:= StashRemoteIcon
                                                            Padding:= margin{Left:=0.0, Top:=-30.0, Bottom:= 0.0, Right:= 0.0}
                                                        stack_box_slot:
                                                            Widget:= StashNotificationIcon
                                                            Padding:= margin{Left:=-20.0, Top:=-30.0, Bottom:= 0.0, Right:= 0.0}
                                                HorizontalAlignment:= horizontal_alignment.Center
                        VerticalAlignment:= vertical_alignment.Top
                        HorizontalAlignment:= horizontal_alignment.Left

    
    SetupStashWidget():void= 
        set StashTransitionWidget = color_block:
            DefaultColor:= NamedColors.Black
            DefaultOpacity:= 0.0
            DefaultDesiredSize := vector2{X:=0.0, Y:= 0.01}        
        
        set StashWidget = 
            overlay:
                Slots:= array:
                    overlay_slot:
                        Widget:= StashTransitionWidget 
                    overlay_slot:
                        Widget := overlay:
                            Slots:= array:
                                overlay_slot:
                                    Widget:= color_block:
                                        DefaultColor:= NamedColors.Black
                                        DefaultOpacity:= 0.4
                                        DefaultDesiredSize := vector2{X:= 250.0, Y:= 50.0}
                                overlay_slot:
                                    Widget:= stack_box:
                                        Orientation:= orientation.Horizontal
                                        Slots:= array:
                                            stack_box_slot:
                                                Widget:= StashIcon
                                                HorizontalAlignment:= horizontal_alignment.Left
                                                Padding:= margin{Left:=2.0, Top:=0.0, Bottom:= 0.0, Right:= -2.0} 
                                            # stack_box_slot:
                                            #     Widget:= texture_block:
                                            #         DefaultImage:= ImgMidasEssence
                                            #         DefaultDesiredSize := vector2{X := 60.0, Y := 60.0}
                                            #     HorizontalAlignment:= horizontal_alignment.Left
                                            #     Padding:= margin{Left:=2.0, Top:=0.0, Bottom:= 0.0, Right:= 2.0} 
                                            stack_box_slot:
                                                Widget:= StashedText
                                                HorizontalAlignment:= horizontal_alignment.Left
                                                Padding:= margin{Left:=2.0, Top:=0.0, Bottom:= 0.0, Right:= 2.0} 
                                    HorizontalAlignment:= horizontal_alignment.Left
                                    Padding:= margin{Left:=10.0, Top:=0.0, Bottom:= 0.0, Right:= 10.0} 
                                overlay_slot:
                                    Widget:= StashedIncomeText
                                    Padding:= margin{Left:=10.0, Top:=-45.0, Bottom:= 0.0, Right:= 10.0} 
                        Padding:= margin{Left:=0.0, Top:=0.0, Bottom:= 0.0, Right:= 10.0}
                        VerticalAlignment:= vertical_alignment.Top
                        HorizontalAlignment:= horizontal_alignment.Left
        
    SetupPeelyWidget(): void = 
        set PeelyWidget = overlay:
                Slots:= array:
                    overlay_slot:
                        Widget := overlay:
                            Slots:= array:
                                overlay_slot:
                                    Widget:= color_block:
                                        DefaultColor:= NamedColors.Black
                                        DefaultOpacity:= 0.4
                                        DefaultDesiredSize := vector2{X:= 200.0, Y:= 50.0}
                                overlay_slot:
                                    Widget:= stack_box:
                                        Orientation:= orientation.Horizontal
                                        Slots:= array:
                                            stack_box_slot:
                                                Widget:= PeelyIcon                                                    
                                                HorizontalAlignment:= horizontal_alignment.Left
                                                Padding:= margin{Left:=2.0, Top:=0.0, Bottom:= 0.0, Right:= -2.0} 
                                                VerticalAlignment:= vertical_alignment.Center
                                            stack_box_slot:
                                                Widget:= PeelyText
                                                HorizontalAlignment:= horizontal_alignment.Left
                                                Padding:= margin{Left:=2.0, Top:=0.0, Bottom:= 0.0, Right:= 2.0} 
                                    HorizontalAlignment:= horizontal_alignment.Left
                                    Padding:= margin{Left:=10.0, Top:=0.0, Bottom:= 0.0, Right:= 10.0} 
                        Padding:= margin{Left:=0.0, Top:=0.0, Bottom:= 0.0, Right:= 10.0}
                        VerticalAlignment:= vertical_alignment.Top
                        HorizontalAlignment:= horizontal_alignment.Left

    SetupHammerWidget(): void = 
        set HammerWidget = overlay:
                Slots:= array:
                    overlay_slot:
                        Widget := overlay:
                            Slots:= array:
                                overlay_slot:
                                    Widget:= color_block:
                                        DefaultColor:= NamedColors.Black
                                        DefaultOpacity:= 0.4
                                        DefaultDesiredSize := vector2{X:= 200.0, Y:= 50.0}
                                overlay_slot:
                                    Widget:= stack_box:
                                        Orientation:= orientation.Horizontal
                                        Slots:= array:
                                            stack_box_slot:
                                                Widget:= HammerIcon                                                    
                                                HorizontalAlignment:= horizontal_alignment.Left
                                                Padding:= margin{Left:=2.0, Top:=0.0, Bottom:= 0.0, Right:= -2.0} 
                                                VerticalAlignment:= vertical_alignment.Center
                                            stack_box_slot:
                                                Widget:= HammerText
                                                HorizontalAlignment:= horizontal_alignment.Left
                                                Padding:= margin{Left:=2.0, Top:=0.0, Bottom:= 0.0, Right:= 2.0} 
                                    HorizontalAlignment:= horizontal_alignment.Left
                                    Padding:= margin{Left:=10.0, Top:=0.0, Bottom:= 0.0, Right:= 10.0} 
                        Padding:= margin{Left:=0.0, Top:=0.0, Bottom:= 0.0, Right:= 10.0}
                        VerticalAlignment:= vertical_alignment.Top
                        HorizontalAlignment:= horizontal_alignment.Left
        HammerWidget.SetVisibility(widget_visibility.Hidden)

                            
    SetupGemWidget(): void = 
        
        set GemWidget = overlay:
                Slots:= array:
                    overlay_slot:
                        Widget := overlay:
                            Slots:= array:
                                overlay_slot:
                                    Widget:= color_block:
                                        DefaultColor:= NamedColors.Black
                                        DefaultOpacity:= 0.4
                                        DefaultDesiredSize := vector2{X:= 200.0, Y:= 50.0}
                                overlay_slot:
                                    Widget:= stack_box:
                                        Orientation:= orientation.Horizontal
                                        Slots:= array:
                                            stack_box_slot:
                                                Widget:= GemIcon                                                    
                                                HorizontalAlignment:= horizontal_alignment.Left
                                                Padding:= margin{Left:=2.0, Top:=0.0, Bottom:= 0.0, Right:= -2.0} 
                                                VerticalAlignment:= vertical_alignment.Center
                                            stack_box_slot:
                                                Widget:= GemText
                                                HorizontalAlignment:= horizontal_alignment.Left
                                                Padding:= margin{Left:=2.0, Top:=0.0, Bottom:= 0.0, Right:= 2.0} 
                                    HorizontalAlignment:= horizontal_alignment.Left
                                    Padding:= margin{Left:=10.0, Top:=0.0, Bottom:= 0.0, Right:= 10.0} 
                        Padding:= margin{Left:=0.0, Top:=0.0, Bottom:= 0.0, Right:= 10.0}
                        VerticalAlignment:= vertical_alignment.Top
                        HorizontalAlignment:= horizontal_alignment.Left
    

    SetupKeyWidget(): void =
        
        set KeyTransitionWidget = color_block:
            DefaultColor:= NamedColors.Black
            DefaultOpacity:= 0.0
            DefaultDesiredSize := vector2{X:=0.0, Y:= 0.01}        
        
        set KeyWidget = stack_box:
            Orientation:= orientation.Horizontal
            Slots:= array:
                stack_box_slot:
                    Widget := overlay:
                        Slots:= array:
                            overlay_slot:
                                Widget:= color_block:
                                    DefaultColor:= NamedColors.Black
                                    DefaultOpacity:= 0.4
                                    DefaultDesiredSize := vector2{X:= 200.0, Y:= 50.0}
                            overlay_slot:
                                Widget:= KeyTimerText
                                HorizontalAlignment:= horizontal_alignment.Left
                                Padding:= margin{Left:=-90.0, Top:= 0.0, Bottom:= 0.0, Right:= 0.0} 
                            overlay_slot:
                                Widget:= stack_box:
                                    Orientation:= orientation.Horizontal
                                    Slots:= array:
                                        stack_box_slot:
                                            Widget:= KeyIcon                                                    
                                            HorizontalAlignment:= horizontal_alignment.Left
                                            Padding:= margin{Left:=2.0, Top:=0.0, Bottom:= 0.0, Right:= -2.0} 
                                            VerticalAlignment:= vertical_alignment.Center
                                        stack_box_slot:
                                            Widget:= KeyText
                                            HorizontalAlignment:= horizontal_alignment.Left
                                            Padding:= margin{Left:=2.0, Top:=0.0, Bottom:= 0.0, Right:= 2.0} 
                                HorizontalAlignment:= horizontal_alignment.Left
                                Padding:= margin{Left:=10.0, Top:=0.0, Bottom:= 0.0, Right:= 10.0} 
                    Padding:= margin{Left:=0.0, Top:=0.0, Bottom:= 0.0, Right:= 10.0}
                    VerticalAlignment:= vertical_alignment.Top
                    HorizontalAlignment:= horizontal_alignment.Left
                stack_box_slot:
                    Widget:= KeyTransitionWidget
    
    SetupCandyWidget(): void = 
        
        set CandyTransitionWidget = color_block:
            DefaultColor:= NamedColors.Black
            DefaultOpacity:= 0.0
            DefaultDesiredSize := vector2{X:=0.0, Y:= 0.01}        
        
        set CandyWidget = stack_box:
            Orientation:= orientation.Horizontal
            Slots:= array:
                stack_box_slot:
                    Widget := overlay:
                        Slots:= array:
                            overlay_slot:
                                Widget:= color_block:
                                    DefaultColor:= NamedColors.Black
                                    DefaultOpacity:= 0.4
                                    DefaultDesiredSize := vector2{X:= 200.0, Y:= 50.0}
                            overlay_slot:
                                Widget:= stack_box:
                                    Orientation:= orientation.Horizontal
                                    Slots:= array:
                                        stack_box_slot:
                                            Widget:= CandyIcon                                                    
                                            HorizontalAlignment:= horizontal_alignment.Left
                                            Padding:= margin{Left:=2.0, Top:=0.0, Bottom:= 0.0, Right:= -2.0} 
                                            VerticalAlignment:= vertical_alignment.Center
                                        stack_box_slot:
                                            Widget:= CandyText
                                            HorizontalAlignment:= horizontal_alignment.Left
                                            Padding:= margin{Left:=2.0, Top:=0.0, Bottom:= 0.0, Right:= 2.0} 
                                HorizontalAlignment:= horizontal_alignment.Left
                                Padding:= margin{Left:=10.0, Top:=0.0, Bottom:= 0.0, Right:= 10.0} 
                    Padding:= margin{Left:=0.0, Top:=0.0, Bottom:= 0.0, Right:= 10.0}
                    VerticalAlignment:= vertical_alignment.Top
                    HorizontalAlignment:= horizontal_alignment.Left
                stack_box_slot:
                    Widget:= CandyTransitionWidget

    SetupCoinsWidget(): void = 
        set CoinsWidget = overlay:
                Slots:= array:
                    overlay_slot:
                        Widget := overlay:
                            Slots:= array:
                                overlay_slot:
                                    Widget:= color_block:
                                        DefaultColor:= NamedColors.Black
                                        DefaultOpacity:= 0.4
                                        DefaultDesiredSize := vector2{X:= 200.0, Y:= 50.0}
                                overlay_slot:
                                    Widget:= stack_box:
                                        Orientation:= orientation.Horizontal
                                        Slots:= array:
                                            stack_box_slot:
                                                Widget:= CoinsIcon                                                    
                                                HorizontalAlignment:= horizontal_alignment.Left
                                                Padding:= margin{Left:=2.0, Top:=0.0, Bottom:= 0.0, Right:= -2.0} 
                                                VerticalAlignment:= vertical_alignment.Center
                                            stack_box_slot:
                                                Widget:= CoinsText
                                                HorizontalAlignment:= horizontal_alignment.Left
                                                Padding:= margin{Left:=2.0, Top:=0.0, Bottom:= 0.0, Right:= 2.0} 
                                    HorizontalAlignment:= horizontal_alignment.Left
                                    Padding:= margin{Left:=10.0, Top:=0.0, Bottom:= 0.0, Right:= 10.0} 
                        Padding:= margin{Left:=0.0, Top:=0.0, Bottom:= 0.0, Right:= 10.0}
                        VerticalAlignment:= vertical_alignment.Top
                        HorizontalAlignment:= horizontal_alignment.Left
                        
    SetupQuestWidget():void=    
        set QuestVerticalStackBox = stack_box:
            Orientation := orientation.Vertical
            Slots:= array:
                stack_box_slot:
                    Widget:=QuestHorizontalStackBox1
                    HorizontalAlignment:= horizontal_alignment.Center
                stack_box_slot:
                    Widget:=QuestHorizontalStackBox2
                    HorizontalAlignment:= horizontal_alignment.Center
                    Padding:= margin{Left:=0.0, Top:=-5.0, Bottom:= 0.0, Right:= 0.0}
                stack_box_slot:
                    Widget:=QuestHorizontalStackBox3
                    HorizontalAlignment:= horizontal_alignment.Center
                    Padding:= margin{Left:=0.0, Top:=-5.0, Bottom:= 0.0, Right:= 0.0}

    SetupZoneOpenWidget():void=    
        set ZoneOpenVerticalStackBox = stack_box:
            Orientation := orientation.Vertical
            Slots:= array:
                stack_box_slot:
                    Widget:=ZoneOpenHorizontalStackBox1
                    HorizontalAlignment:= horizontal_alignment.Center
                stack_box_slot:
                    Widget:=ZoneOpenHorizontalStackBox2
                    HorizontalAlignment:= horizontal_alignment.Center
                    Padding:= margin{Left:=0.0, Top:=-5.0, Bottom:= 0.0, Right:= 0.0}
                stack_box_slot:
                    Widget:=ZoneOpenHorizontalStackBox3
                    HorizontalAlignment:= horizontal_alignment.Center
                    Padding:= margin{Left:=0.0, Top:=-5.0, Bottom:= 0.0, Right:= 0.0}
                    
    SetupGuardWidget():void=
        set GuardHealth  = MakeProgressBar(0.0, vector2{X:= 650.0, Y:= 50.0}, MakeColorFromHex("CC362F"), MakeColorFromHex("000000"), NamedColors.Goldenrod) #Red
        GuardHealthWidget := GuardHealth.GetWidget()
        
        set GuardWidget = 
            overlay:   
                Slots:= array:
                    overlay_slot:
                        Widget:= GuardHealthWidget
                    overlay_slot:
                        Widget:= stack_box:
                            Orientation:= orientation.Horizontal
                            Slots:= array:
                                stack_box_slot:
                                    Widget:= GuardName                                    
                                stack_box_slot:
                                    Widget:= GuardHealthText
                        HorizontalAlignment:= horizontal_alignment.Center
                    overlay_slot:  #KO Bonus & Clear Bonus
                        Widget:= stack_box:
                            Orientation:= orientation.Horizontal
                            Slots:= array:
                                stack_box_slot:
                                    Widget:= GuardArmorImgWidget
                                stack_box_slot:
                                    Widget:= GuardArmorText
                                stack_box_slot:
                                    Widget:= GuardTeamworkImgWidget
                                    Padding:= margin{Left:=20.0, Top:=0.0, Bottom:= 0.0, Right:= 0.0} 
                                stack_box_slot:
                                    Widget:= GuardTeamworkText
                                stack_box_slot:
                                    Widget:= GuardKOBonusImgWidget
                                    Padding:= margin{Left:=20.0, Top:=0.0, Bottom:= 0.0, Right:= 0.0} 
                                stack_box_slot:
                                    Widget:= GuardKOBonusText
                                stack_box_slot:
                                    Widget:= GuardClearBonusImgWidget
                                    Padding:= margin{Left:=20.0, Top:=0.0, Bottom:= 0.0, Right:= 0.0} 
                                stack_box_slot:
                                    Widget:= GuardClearBonusText
                                stack_box_slot:
                                    Widget:= ZoneOverkillText
                        Padding:= margin{Left:=0.0, Top:=50.0, Bottom:= 0.0, Right:= 0.0} 
                        HorizontalAlignment:= horizontal_alignment.Center
    
        GuardWidget.SetVisibility(widget_visibility.Hidden)

    SetupGunDurationWidget():void=
        #CAKO
        set GunDuration  = MakeProgressBar(0.0, vector2{X:= 240.0, Y:= 40.0}, MakeColorFromHex("6a9d5a"), MakeColorFromHex("000000"), MakeColorFromHex("FFFFFF")) #Red
        
        set GunDurationWidget =
            stack_box_slot:
                Widget:= overlay:   
                    Slots:= array:
                        overlay_slot:
                            Widget:= GunDuration.GetWidget()    
                        overlay_slot:
                            Widget:= stack_box:
                                Orientation:= orientation.Horizontal
                                Slots:= array:
                                    stack_box_slot:
                                        Widget:=  GunDurationIcon
                                    stack_box_slot:
                                        Widget:= GunDurationText
                            HorizontalAlignment:= horizontal_alignment.Center                    
    
        GunDurationWidget.Widget.SetVisibility(widget_visibility.Hidden)
        BuffHostWidget.AddWidget(GunDurationWidget)

    SetupUpgradeHammerWidget():void=
        #CAKO
        set UpgradeHammerProgressBar  = MakeProgressBar(0.0, vector2{X:= 280.0, Y:= 40.0}, NamedColors.Goldenrod, MakeColorFromHex("000000"), MakeColorFromHex("FFFFFF")) #Red
        
        set UpgradeHammerWidget =
            stack_box_slot:
                Widget:= overlay:   
                    Slots:= array:
                        overlay_slot:
                            Widget:= UpgradeHammerProgressBar.GetWidget()    
                        overlay_slot:
                            Widget:= stack_box:
                                Orientation:= orientation.Horizontal
                                Slots:= array:
                                    stack_box_slot:
                                        Widget:=  UpgradeHammerIcon
                                    stack_box_slot:
                                        Widget:= UpgradeHammerProgressText
                            HorizontalAlignment:= horizontal_alignment.Center                    
    
        UpgradeHammerWidget.Widget.SetVisibility(widget_visibility.Hidden)
        BuffHostWidget2.AddWidget(UpgradeHammerWidget)

    SetupBossNotificationWidget():void=
        set BossBounceColorWidget = color_block:
            DefaultColor:= NamedColors.Black
            DefaultOpacity:= 0.0
            DefaultDesiredSize := vector2{X:=0.0, Y:= 0.01}        

        set BossNotificationWidget = 
            stack_box:
                Orientation:= orientation.Horizontal
                Slots:= array:
                    stack_box_slot:
                        Widget:=overlay:   
                            Slots:= array:
                                overlay_slot:
                                    Widget:= BossArrowIcon
                                overlay_slot:
                                    Widget:= stack_box:
                                        Orientation:= orientation.Horizontal
                                        Slots:= array:
                                            stack_box_slot: 
                                                Widget:= BossRemoteIcon
                                                Padding:= margin{Left:=0.0, Top:=-30.0, Bottom:= 0.0, Right:= 0.0}
                                            stack_box_slot:
                                                Widget:= BossNotificationIcon
                                                Padding:= margin{Left:=-20.0, Top:=-30.0, Bottom:= 0.0, Right:= 0.0}
                                    HorizontalAlignment:= horizontal_alignment.Center
                    stack_box_slot:
                        Widget:= BossBounceColorWidget                      
                
    SetupBossWidget():void=
        set BossHealth  = MakeProgressBar(0.0, vector2{X:= 950.0, Y:= 75.0}, MakeColorFromHex("CC362F"), MakeColorFromHex("000000"), MakeColorFromHex("FFFFFF")) #Red
        BossHealthWidget := BossHealth.GetWidget()
        
        # set BossDurationBar = MakeProgressBar(0.0, vector2{X:= 650.0, Y:= 60.0}, MakeColorFromHex("36CC42"), MakeColorFromHex("000000"), MakeColorFromHex("FFFFFF")) #Green

        set BossWidget = 
            overlay:   
                Slots:= array:
                    overlay_slot:
                        Widget:= BossIcon
                        Padding:= margin{Left:=-525.0, Top:=0.0, Bottom:= 0.0, Right:= 0.0}
                        HorizontalAlignment:= horizontal_alignment.Center 
                    overlay_slot:
                        Widget:= BossHealthWidget    
                    overlay_slot:
                        Widget:= stack_box:
                            Orientation:= orientation.Horizontal
                            Slots:= array:
                                stack_box_slot: #Boss name
                                    Widget:= BossName
                                stack_box_slot:
                                    Widget:= BossHealthText
                        HorizontalAlignment:= horizontal_alignment.Center
                    # overlay_slot:
                    #     Widget:= BossDurationBar.GetWidget()
                    #     Padding:= margin{Left:=0.0, Top:=75.0, Bottom:= 0.0, Right:= 0.0} 
                    #     HorizontalAlignment:= horizontal_alignment.Center
                    # overlay_slot:
                    #     Widget:= stack_box: #TickingClock for DamageMastery
                    #         Orientation:= orientation.Horizontal
                    #             Slots:= array:
                    #                 stack_box_slot: 
                    #                     Widget:= TickingClockImg
                    #                     HorizontalAlignment:= horizontal_alignment.Left
                    #                 stack_box_slot: 
                    #                     Widget:= TickingClockStackbox
                    #                     HorizontalAlignment:= horizontal_alignment.Left
                    #     Padding:= margin{Left:=0.0, Top:=75.0, Bottom:= 0.0, Right:= 0.0} 
                    #     HorizontalAlignment:= horizontal_alignment.Center
                    overlay_slot:  #KO Bonus
                        Widget:= stack_box:
                            Orientation:= orientation.Horizontal
                            Slots:= array:
                                stack_box_slot:
                                    Widget:= BossArmorImgWidget
                                stack_box_slot:
                                    Widget:= BossArmorText
                                stack_box_slot:
                                    Widget:= BossTeamworkImgWidget
                                    Padding:= margin{Left:=20.0, Top:=0.0, Bottom:= 0.0, Right:= 0.0} 
                                stack_box_slot:
                                    Widget:= BossTeamworkText
                                stack_box_slot:
                                    Widget:= BossKOBonusImgWidget
                                stack_box_slot:
                                    Widget:= BossKOBonusText
                                stack_box_slot:
                                    Widget:= BossClearBonusImgWidget
                                    Padding:= margin{Left:=20.0, Top:=00.0, Bottom:= 0.0, Right:= 0.0} 
                                stack_box_slot:
                                    Widget:= BossClearBonusText
                                stack_box_slot:
                                    Widget:= ZoneOverkillText
                        Padding:= margin{Left:=0.0, Top:=75.0, Bottom:= 0.0, Right:= 0.0} 
                        HorizontalAlignment:= horizontal_alignment.Center
    
        BossWidget.SetVisibility(widget_visibility.Hidden)

    SetupZoneWidget():void=
        set ZoneKOsProgress  = MakeProgressBar(0.0, vector2{X:= 380.0, Y:= 40.0}, MakeColorFromHex("1985A3"), MakeColorFromHex("000000"), MakeColorFromHex("FFFFFF")) #Blue
        ZoneKOProgressWidget := ZoneKOsProgress.GetWidget()        
        
        # set ZoneClearsProgress  = MakeProgressBar(0.0, vector2{X:= 180.01, Y:= 40.0}, MakeColorFromHex("2CA736"), MakeColorFromHex("000000"), MakeColorFromHex("FFFFFF"), ?InMaxColor := option{MakeColorFromHex("CC362F")}) #Original Green Hex: 36CC42
        # ZoneClearsProgressWidget := ZoneClearsProgress.GetWidget()

        set ZoneTransitionWidget = color_block:
                DefaultColor:= NamedColors.Black
                DefaultOpacity:= 0.0
                DefaultDesiredSize := vector2{X:=0.0, Y:= 0.01}        

        set ZoneWidget = 
            Widget:= stack_box:
                Orientation:= orientation.Horizontal
                Slots:= array:
                    stack_box_slot:
                        Widget := overlay:
                            Slots:= array:
                                overlay_slot:
                                    Widget:= texture_block{DefaultImage:= Images.StoreBox, DefaultDesiredSize := vector2{X:= 420.0, Y:= 190.0}}
                                    VerticalAlignment:= vertical_alignment.Top
                                overlay_slot:
                                    Widget:= stack_box:
                                        Orientation:= orientation.Vertical
                                        Slots:= array:
                                            stack_box_slot:
                                                Widget:= stack_box:
                                                    Orientation:= orientation.Horizontal
                                                    Slots:= array:
                                                        stack_box_slot:
                                                            Widget := texture_block{ DefaultImage := Images.Empty, DefaultDesiredSize := vector2{X := 40.0, Y := 40.0}}
                                                        stack_box_slot:
                                                            Widget:= ZoneName
                                                            # Widget:= text_block{DefaultText := STM("Treasure Hunting"),DefaultTextColor:=NamedColors.Goldenrod,
                                                            # DefaultShadowColor:=NamedColors.Black,
                                                            # DefaultShadowOpacity := 1.0,
                                                            # DefaultShadowOffset:=option{vector2{X:=2.0,Y:=2.0}}}
                                                            # HorizontalAlignment:= horizontal_alignment.Center    
                                                            Padding:= margin{Left:=5.0, Top:=0.0, Bottom:= 0.0, Right:= 0.0}
                                                Padding:= margin{Left:=5.0, Top:=10.0, Bottom:= 0.0, Right:= 0.0}
                                            stack_box_slot:
                                                Widget := stack_box:
                                                    Orientation := orientation.Horizontal
                                                    Slots := array:
                                                        stack_box_slot:
                                                            Widget:= overlay:
                                                                Slots:=array:
                                                                    overlay_slot:
                                                                        Widget:= ZoneKOProgressWidget    
                                                                    overlay_slot:
                                                                        Widget:= stack_box:
                                                                            Orientation:= orientation.Horizontal
                                                                            Slots:= array:
                                                                                stack_box_slot:
                                                                                    Widget:= GuardIcon                                                                                        
                                                                                stack_box_slot:
                                                                                    Widget:= ZoneKOsProgressText
                                                            HorizontalAlignment:= horizontal_alignment.Center                                                        
                                                Padding:= margin{Left:=10.0, Top:=15.0, Bottom:= 0.0, Right:= 0.0}
                                            stack_box_slot:
                                                Widget:= NextGemStackBox
                                                HorizontalAlignment:= horizontal_alignment.Center
                                                Padding:= margin{Left:=10.0, Top:=15.0, Bottom:= 0.0, Right:= 0.0}  
                                    VerticalAlignment:= vertical_alignment.Top
                                    HorizontalAlignment:= horizontal_alignment.Left  
                    stack_box_slot:
                        Widget:= ZoneTransitionWidget                      
                                                                          
    SetupTextBlocks():void=
        if:
            CurrentEssence:=PlayerManager.GetEssence[]
            CurrentProgress:=PlayerManager.GetLevelProgress[]
            CurrentLevel:=PlayerManager.GetLevel[]
            CurrentStrength:=PlayerManager.GetDisplayStrength[]
        then:
            CurrentStrengthGain:= PlayerManager.GetStrengthGainedPerLevel(CurrentLevel + 1)
            set EssenceText = text_block{DefaultTextColor:=NamedColors.White,
            DefaultShadowColor:=NamedColors.Black,
            DefaultShadowOffset:=option{vector2{X:=2.0,Y:=2.0}}}
            EssenceText.SetShadowOpacity(1.0)
            EssenceText.SetText(STM("{CurrentEssence}"))
            
            set StrengthText = text_block{DefaultTextColor:=NamedColors.Goldenrod,
            DefaultShadowColor:=NamedColors.Black,
            DefaultShadowOffset:=option{vector2{X:=2.0,Y:=2.0}}}
            StrengthText.SetShadowOpacity(1.0)
            StrengthText.SetText(STM("{CurrentStrength}"))

            set NextLevelStrengthText = text_block{DefaultTextColor:=NamedColors.Goldenrod,
            DefaultShadowColor:=NamedColors.Black,
            DefaultShadowOffset:=option{vector2{X:=2.0,Y:=2.0}}}
            NextLevelStrengthText.SetShadowOpacity(1.0)
            NextLevelStrengthText.SetText(STM(" + {CurrentStrengthGain}"))

            set NextLevelGemText = text_block{DefaultTextColor:=NamedColors.Goldenrod,
            DefaultShadowColor:=NamedColors.Black,
            DefaultShadowOffset:=option{vector2{X:=2.0,Y:=2.0}}}
            NextLevelGemText.SetShadowOpacity(1.0)
            NextLevelGemText.SetText(STM(""))

            set GunText = text_block{DefaultTextColor:=MakeColorFromHex("CC362F"),
            DefaultShadowColor:=NamedColors.Black,
            DefaultShadowOffset:=option{vector2{X:=2.0,Y:=2.0}}}
            GunText.SetShadowOpacity(1.0)
            GunText.SetText(STM(""))

            set GunText2 = text_block{DefaultTextColor:=MakeColorFromHex("CC362F"),
            DefaultShadowColor:=NamedColors.Black,
            DefaultShadowOffset:=option{vector2{X:=2.0,Y:=2.0}}}
            GunText2.SetShadowOpacity(1.0)
            GunText2.SetText(STM(""))

            set LevelText = text_block{DefaultTextColor:=NamedColors.White,
            DefaultShadowColor:=NamedColors.Black,
            DefaultShadowOffset:=option{vector2{X:=2.0,Y:=2.0}}}
            LevelText.SetShadowOpacity(1.0)
            LevelText.SetText(STM("{CurrentLevel}"))
            
            set NextLevelText = text_block{DefaultTextColor:=NamedColors.Goldenrod,
            DefaultShadowColor:=NamedColors.Black,
            DefaultShadowOffset:=option{vector2{X:=2.0,Y:=2.0}}}
            NextLevelText.SetShadowOpacity(1.0)
            NextLevelText.SetText(STM("Reward:"))

            Temp : tuple(float, tuple(big_number,big_number)) = CurrentProgress
            EssenceProgressBar.SetProgress(Temp(0), true)
            
            set ExpText = text_block{DefaultTextColor:=NamedColors.White,
            DefaultShadowColor:=NamedColors.Black,
            DefaultShadowOffset:=option{vector2{X:=2.0,Y:=2.0}}}
            ExpText.SetShadowOpacity(1.0)
            ExpText.SetText(STM("XP: "))

            set ProgressTextLeft = text_block{DefaultTextColor:=NamedColors.White,
            DefaultShadowColor:=NamedColors.Black,
            DefaultShadowOffset:=option{vector2{X:=2.0,Y:=2.0}}}
            ProgressTextLeft.SetShadowOpacity(1.0)
            ProgressTextLeft.SetText(STM("{FormatBigNumber(Temp(1)(0))} / "))

            set ProgressTextRight = text_block{DefaultTextColor:=NamedColors.White,
            DefaultShadowColor:=NamedColors.Black,
            DefaultShadowOffset:=option{vector2{X:=2.0,Y:=2.0}}}
            ProgressTextRight.SetShadowOpacity(1.0)        
            ProgressTextRight.SetText(STM("{FormatBigNumber(Temp(1)(1))}"))
            
            set BossHealthText = text_block{DefaultTextColor:=NamedColors.White,
            DefaultShadowColor:=NamedColors.Black,
            DefaultShadowOffset:=option{vector2{X:=2.0,Y:=2.0}}}
            BossHealthText.SetShadowOpacity(1.0)      
            BossHealthText.SetText(STM(""))

            set BossName =  text_block{DefaultTextColor:=NamedColors.White,
            DefaultShadowColor:=NamedColors.Black,
            DefaultShadowOffset:=option{vector2{X:=2.0,Y:=2.0}}}
            BossName.SetShadowOpacity(1.0)    
            BossName.SetText(STM(""))

            set BossArmorText =  text_block{DefaultTextColor:=NamedColors.White,
            DefaultShadowColor:=NamedColors.Black,
            DefaultShadowOffset:=option{vector2{X:=2.0,Y:=2.0}}}
            BossArmorText.SetShadowOpacity(1.0)    
            BossArmorText.SetText(STM(""))

            set BossTeamworkText =  text_block{DefaultTextColor:=NamedColors.Goldenrod,
            DefaultShadowColor:=NamedColors.Black,
            DefaultShadowOffset:=option{vector2{X:=2.0,Y:=2.0}}}
            BossTeamworkText.SetShadowOpacity(1.0)    
            BossTeamworkText.SetText(STM(""))

            set BossKOBonusText =  text_block{DefaultTextColor:=NamedColors.Goldenrod,
            DefaultShadowColor:=NamedColors.Black,
            DefaultShadowOffset:=option{vector2{X:=2.0,Y:=2.0}}}
            BossKOBonusText.SetShadowOpacity(1.0)    
            BossKOBonusText.SetText(STM(""))

            set BossClearBonusText =  text_block{DefaultTextColor:=NamedColors.Goldenrod,
            DefaultShadowColor:=NamedColors.Black,
            DefaultShadowOffset:=option{vector2{X:=2.0,Y:=2.0}}}
            BossClearBonusText.SetShadowOpacity(1.0)    
            BossClearBonusText.SetText(STM(""))

            set GuardHealthText = text_block{DefaultTextColor:=NamedColors.White,
            DefaultShadowColor:=NamedColors.Black,
            DefaultShadowOffset:=option{vector2{X:=2.0,Y:=2.0}}}
            GuardHealthText.SetShadowOpacity(1.0)      
            GuardHealthText.SetText(STM(""))

            set GuardName =  text_block{DefaultTextColor:=NamedColors.White,
            DefaultShadowColor:=NamedColors.Black,
            DefaultShadowOffset:=option{vector2{X:=2.0,Y:=2.0}}}
            GuardName.SetShadowOpacity(1.0)    
            GuardName.SetText(STM(""))
            
            set GuardTeamworkText =  text_block{DefaultTextColor:=NamedColors.Goldenrod,
            DefaultShadowColor:=NamedColors.Black,
            DefaultShadowOffset:=option{vector2{X:=2.0,Y:=2.0}}}
            GuardTeamworkText.SetShadowOpacity(1.0)    
            GuardTeamworkText.SetText(STM(""))

            set GuardArmorText =  text_block{DefaultTextColor:=NamedColors.White,
            DefaultShadowColor:=NamedColors.Black,
            DefaultShadowOffset:=option{vector2{X:=2.0,Y:=2.0}}}
            GuardArmorText.SetShadowOpacity(1.0)    
            GuardArmorText.SetText(STM(""))

            set GuardKOBonusText =  text_block{DefaultTextColor:=NamedColors.Goldenrod,
            DefaultShadowColor:=NamedColors.Black,
            DefaultShadowOffset:=option{vector2{X:=2.0,Y:=2.0}}}
            GuardKOBonusText.SetShadowOpacity(1.0)    
            GuardKOBonusText.SetText(STM(""))

            set GuardClearBonusText =  text_block{DefaultTextColor:=NamedColors.Goldenrod,
            DefaultShadowColor:=NamedColors.Black,
            DefaultShadowOffset:=option{vector2{X:=2.0,Y:=2.0}}}
            GuardClearBonusText.SetShadowOpacity(1.0)    
            GuardClearBonusText.SetText(STM(""))

            set ZoneName = text_block{DefaultTextColor:=NamedColors.Goldenrod,
            DefaultShadowColor:=NamedColors.Black,
            DefaultShadowOffset:=option{vector2{X:=2.0,Y:=2.0}}}
            ZoneName.SetShadowOpacity(1.0)    
            ZoneName.SetText(STM("Midas Entrance"))

            set ZoneKOsProgressText = text_block{DefaultTextColor:=NamedColors.White,
            DefaultShadowColor:=NamedColors.Black,
            DefaultShadowOffset:=option{vector2{X:=2.0,Y:=2.0}}}
            ZoneKOsProgressText.SetShadowOpacity(1.0)    
            ZoneKOsProgressText.SetText(STM(" "))

            set ZoneLevelText = text_block{DefaultTextColor:=NamedColors.White,
            DefaultShadowColor:=NamedColors.Black,
            DefaultShadowOffset:=option{vector2{X:=2.0,Y:=2.0}}}
            ZoneLevelText.SetShadowOpacity(1.0)    
            ZoneLevelText.SetText(STM("Level 1"))

            set ZoneOverkillText = text_block{DefaultTextColor:=MakeColorFromHex("FFA500"), #original red F9423A
            DefaultShadowColor:=NamedColors.Black,
            DefaultShadowOffset:=option{vector2{X:=2.0,Y:=2.0}}}
            ZoneOverkillText.SetShadowOpacity(1.0)    
            ZoneOverkillText.SetText(STM(" "))

            set ZoneClearsProgressText = text_block{DefaultTextColor:=NamedColors.White,
            DefaultShadowColor:=NamedColors.Black,
            DefaultShadowOffset:=option{vector2{X:=2.0,Y:=2.0}}}
            ZoneClearsProgressText.SetShadowOpacity(1.0)    
            ZoneClearsProgressText.SetText(STM(" "))

            set ZoneKOBonusText = text_block{DefaultTextColor:=NamedColors.White,
            DefaultShadowColor:=NamedColors.Black,
            DefaultShadowOffset:=option{vector2{X:=2.0,Y:=2.0}}}
            ZoneKOBonusText.SetShadowOpacity(1.0)    
            ZoneKOBonusText.SetText(STM("KO Bonus: 12.5"))

            set ZoneClearBonusText = text_block{DefaultTextColor:=NamedColors.White,
            DefaultShadowColor:=NamedColors.Black,
            DefaultShadowOffset:=option{vector2{X:=2.0,Y:=2.0}}}
            ZoneClearBonusText.SetShadowOpacity(1.0)    
            ZoneClearBonusText.SetText(STM("Clear Bonus: 100"))

            set StashedText = text_block{DefaultTextColor:=NamedColors.White,
            DefaultShadowColor:=NamedColors.Black,
            DefaultShadowOffset:=option{vector2{X:=2.0,Y:=2.0}}}
            StashedText.SetShadowOpacity(1.0)    
            StashedText.SetText(STM(""))

            set GemText = text_block{DefaultTextColor:=MakeColorFromHex("CC362F"),
            DefaultShadowColor:=NamedColors.Black,
            DefaultShadowOffset:=option{vector2{X:=2.0,Y:=2.0}}}
            GemText.SetShadowOpacity(1.0)    
            GemText.SetText(STM(" 0"))

            set KeyText = text_block{DefaultTextColor:=MakeColorFromHex("CC362F"),
            DefaultShadowColor:=NamedColors.Black,
            DefaultShadowOffset:=option{vector2{X:=2.0,Y:=2.0}}}
            KeyText.SetShadowOpacity(1.0)    
            KeyText.SetText(STM(" 0"))

            set KeyTimerText = text_block{DefaultTextColor:=MakeColorFromHex("CC362F"),
            DefaultShadowColor:=NamedColors.Black,
            DefaultShadowOffset:=option{vector2{X:=2.0,Y:=2.0}}}
            KeyTimerText.SetShadowOpacity(1.0)    
            KeyTimerText.SetText(STM(""))

            set PeelyText = text_block{DefaultTextColor:=MakeColorFromHex("CC362F"),
            DefaultShadowColor:=NamedColors.Black,
            DefaultShadowOffset:=option{vector2{X:=2.0,Y:=2.0}}}
            PeelyText.SetShadowOpacity(1.0)    
            PeelyText.SetText(STM(" 0"))

            set HammerText = text_block{DefaultTextColor:=MakeColorFromHex("CC362F"),
            DefaultShadowColor:=NamedColors.Black,
            DefaultShadowOffset:=option{vector2{X:=2.0,Y:=2.0}}}
            HammerText.SetShadowOpacity(1.0)    
            HammerText.SetText(STM(""))

            set CandyText = text_block{DefaultTextColor:=MakeColorFromHex("CC362F"),
            DefaultShadowColor:=NamedColors.Black,
            DefaultShadowOffset:=option{vector2{X:=2.0,Y:=2.0}}}
            CandyText.SetShadowOpacity(1.0)    
            CandyText.SetText(STM(" 0"))

            set CoinsText = text_block{DefaultTextColor:=MakeColorFromHex("CC362F"),
            DefaultShadowColor:=NamedColors.Black,
            DefaultShadowOffset:=option{vector2{X:=2.0,Y:=2.0}}}
            CoinsText.SetShadowOpacity(1.0)    
            CoinsText.SetText(STM(" 0"))

            set StashedIncomeText = text_block{DefaultTextColor:=NamedColors.Goldenrod,
            DefaultShadowColor:=NamedColors.Black,
            DefaultShadowOffset:=option{vector2{X:=2.0,Y:=2.0}}}
            StashedIncomeText.SetShadowOpacity(1.0)    
            StashedIncomeText.SetText(STM(""))

            set GunDurationText = text_block{DefaultTextColor:=NamedColors.White,
            DefaultShadowColor:=NamedColors.Black,
            DefaultShadowOffset:=option{vector2{X:=2.0,Y:=2.0}}}
            GunDurationText.SetShadowOpacity(1.0)    
            GunDurationText.SetText(STM(""))

            set GunDurationText = text_block{DefaultTextColor:=NamedColors.White,
            DefaultShadowColor:=NamedColors.Black,
            DefaultShadowOffset:=option{vector2{X:=2.0,Y:=2.0}}}
            GunDurationText.SetShadowOpacity(1.0)    
            GunDurationText.SetText(STM(""))

            set DamageMeterHeader = text_block{DefaultTextColor:=NamedColors.White,
            DefaultShadowColor:=NamedColors.Black,
            DefaultShadowOffset:=option{vector2{X:=2.0,Y:=2.0}}}
            DamageMeterHeader.SetShadowOpacity(1.0)    
            DamageMeterHeader.SetText(STM(""))

            set QuestHeaderText = text_block{DefaultTextColor:=NamedColors.Goldenrod,
            DefaultShadowColor:=NamedColors.Black,
            DefaultShadowOffset:=option{vector2{X:=2.0,Y:=2.0}}}
            QuestHeaderText.SetShadowOpacity(1.0)    
            QuestHeaderText.SetText(STM("Main Quest"))

            set GunExpUpdateText = text_block{DefaultTextColor:=NamedColors.White,
            DefaultShadowColor:=NamedColors.Black,
            DefaultShadowOffset:=option{vector2{X:=2.0,Y:=2.0}}}
            GunExpUpdateText.SetShadowOpacity(1.0)    
            GunExpUpdateText.SetText(STM(""))

            set GunExpLevelText = text_block{DefaultTextColor:=NamedColors.White,
            DefaultShadowColor:=NamedColors.Black,
            DefaultShadowOffset:=option{vector2{X:=2.0,Y:=2.0}}}
            GunExpLevelText.SetShadowOpacity(1.0)    
            GunExpLevelText.SetText(STM(""))

            set GunExpProgressText = text_block{DefaultTextColor:=NamedColors.White,
            DefaultShadowColor:=NamedColors.Black,
            DefaultShadowOffset:=option{vector2{X:=2.0,Y:=2.0}}}
            GunExpProgressText.SetShadowOpacity(1.0)    
            GunExpProgressText.SetText(STM(""))

            set GunExpNextLevelText = text_block{DefaultTextColor:=NamedColors.White,
            DefaultShadowColor:=NamedColors.Black,
            DefaultShadowOffset:=option{vector2{X:=2.0,Y:=2.0}}}
            GunExpNextLevelText.SetShadowOpacity(1.0)    
            GunExpNextLevelText.SetText(STM(""))

            set PermaGunExpLevelText = text_block{DefaultTextColor:=NamedColors.White,
            DefaultShadowColor:=NamedColors.Black,
            DefaultShadowOffset:=option{vector2{X:=2.0,Y:=2.0}}}
            PermaGunExpLevelText.SetShadowOpacity(1.0)    
            PermaGunExpLevelText.SetText(STM(""))

            set PermaGunExpProgressText = text_block{DefaultTextColor:=NamedColors.White,
            DefaultShadowColor:=NamedColors.Black,
            DefaultShadowOffset:=option{vector2{X:=2.0,Y:=2.0}}}
            PermaGunExpProgressText.SetShadowOpacity(1.0)    
            PermaGunExpProgressText.SetText(STM(""))

            set UpgradeHammerProgressText = text_block{DefaultTextColor:=NamedColors.White,
            DefaultShadowColor:=NamedColors.Black,
            DefaultShadowOffset:=option{vector2{X:=2.0,Y:=2.0}}}
            UpgradeHammerProgressText.SetShadowOpacity(1.0)    
            UpgradeHammerProgressText.SetText(STM(""))
            
            
    #Popup Functions
    ClosePopup():void=
        if:
            MyUI := Popup?
            PlayerUI := GetPlayerUI[Player]
        then:
            MyUI.SetEnabled(false)
            MyUI.SetVisibility(widget_visibility.Hidden)
            PlayerUI.RemoveWidget(MyUI)
            if (set PlayerManager.UIManager?.Popup = false) {}

    ShowNonBlockingDialog(Dialog : dialog_message, Duration : float, ?VODelay : float = 0.0)<suspends>:void=
        spawn {ShowNonBlockingDialog_Internal(Dialog, Duration, ?VODelay := VODelay)}
        Sleep(Duration + VODelay)

    ShowNonBlockingDialog_Internal(Dialog : dialog_message, Duration : float, ?VODelay : float = 0.0)<suspends>:void=
        MainCanvas.RemoveWidget(LatestBubble)
        MainCanvas.RemoveWidget(LatestIcon)
        MainCanvas.RemoveWidget(LatestText)
        LatestVO.Stop(PlayerManager.Player)        
        
        Bubble := texture_block{DefaultImage := Images.SpeechBubble, DefaultDesiredSize := vector2{X := 1498.0, Y := 106.0}}

        var Icon : texture_block = texture_block{DefaultImage := Images.Empty}

        Text := WordWrap(Dialog.Message, 90, ?Color := NamedColors.Black, ?ShadowColor := NamedColors.White) 

        set LatestBubble = Bubble
        set LatestIcon = Icon
        set LatestText = Text
        if:
            VO := Dialog.VO?
            set LatestVO = VO

        MainCanvas.AddWidget(canvas_slot:
                Widget:= Bubble
                SizeToContent := true
                Anchors := anchors{ Minimum := vector2{X := 0.5, Y := 0.025}, Maximum := vector2{X := 0.5, Y := 0.025}}                        
                Offsets := margin{Top := 0.0, Left := 0.0, Right := 0.0, Bottom := 0.0}
                Alignment := vector2{X := 0.5, Y := 0.0}
                ZOrder := 1001
        )
        MainCanvas.AddWidget(canvas_slot:
                Widget:= Text
                SizeToContent := true
                Anchors := anchors{ Minimum := vector2{X := 0.5, Y := 0.025}, Maximum := vector2{X := 0.5, Y := 0.025}}                        
                Offsets := margin{Top := 40.0, Left := 0.0, Right := 0.0, Bottom := 0.0}
                Alignment := vector2{X := 0.5, Y := 0.5}
                ZOrder := 1002
        )

        if (Title := Dialog.Title?):
            set Icon = texture_block{DefaultImage := GetTextureFromString(Title), DefaultDesiredSize := vector2{X :=100.0, Y := 100.0}}
            MainCanvas.AddWidget(canvas_slot:
                Widget:= Icon
                SizeToContent := true
                Anchors := anchors{ Minimum := vector2{X := 0.5, Y := 0.025}, Maximum := vector2{X := 0.5, Y := 0.025}}                        
                Offsets := margin{Top := 100.0, Left := 480.0, Right := 0.0, Bottom := 0.0}
                Alignment := vector2{X := 0.0, Y := 0.0}
                ZOrder := 1000
            )

        if (VODelay > 0.0):
            Sleep(VODelay)
        
        if (VO := Dialog.VO?):
            VO.Play(PlayerManager.Player)

        Sleep(Duration)

        MainCanvas.RemoveWidget(Bubble)
        MainCanvas.RemoveWidget(Icon)
        MainCanvas.RemoveWidget(Text)
