using { /Verse.org/Simulation }
using { /Fortnite.com/Devices }
using { /UnrealEngine.com/Temporary/UI }
using { /Fortnite.com/UI }
using { /UnrealEngine.com/Temporary/SpatialMath }
using { /Verse.org/Colors }

#an individual item that can be purchased
upgrade_item:= class<concrete>:
    @editable
    var Name : string = "ITEM"

    @editable
    var Name_Override : string = ""
    
    @editable
    var Description : string = "ITEM"

    @editable
    var LevelCurrencies : []string = array{"Gems"}

    @editable
    var LevelCosts : []int = array{}

    @editable
    var LevelValues : []float = array{}

    @editable
    var ValueUnit : string = "%"

    @editable
    var ValueDecimals : int = 2

    @editable
    var QuestRequirement : quest_stats = quest_stats{}

    #These Restricted Values are used if the Requirement hasn't been met.
    @editable
    var RestrictedName : string = "ITEM"

    @editable
    var RestrictedName_Override : string = ""
    
    @editable
    var RestrictedDescription : string = "ITEM"

    var GameManager : game_manager = game_manager{}
    var UpgradeShop : ?player_upgrade_shop = false

    Clone() : upgrade_item = 
        Result :=upgrade_item:
        set Result.Name = Name
        set Result.Name_Override = Name_Override
        set Result.Description = Description
        set Result.LevelCurrencies = LevelCurrencies
        set Result.LevelCosts = LevelCosts
        set Result.LevelValues = LevelValues
        set Result.ValueUnit = ValueUnit
        set Result.ValueDecimals = ValueDecimals
        set Result.QuestRequirement = QuestRequirement
        set Result.RestrictedName = RestrictedName
        set Result.RestrictedName_Override = RestrictedName_Override
        set Result.RestrictedDescription = RestrictedDescription
        set Result.GameManager = GameManager
        set Result.UpgradeShop = UpgradeShop

        return Result

    OnBuy(Wdata:widget_message, Index : int):void =
        if:
            Player := Wdata.Player
            PlayerUI := GetPlayerUI[Player]
            PlayerProfile := PlayerProfileDataMap[Player]
            PlayerManager := GameManager.PlayerManagers[Player]
            PlayerLevel := PlayerManager.GetLevel[]
            MyShop := UpgradeShop?
        then:
            var CurrentLevel : int = 0
            if:
                TempCurrentLevel := PlayerManager.GetUpgradeLevel[Name]
                set CurrentLevel = TempCurrentLevel
                
            CurrencyName := GetCurrencyName(CurrentLevel)
            if:
                Cost := PlayerManager.GetUpgradeLevelCost[Name, option{CurrentLevel}]
                CurrentCurrency := PlayerManager.GetCollectibleCount(CurrencyName)
                PlayerManager.HasCompletedTask(QuestRequirement.QuestID, QuestRequirement.TaskID)?
            then:
                if:
                    BigCost := big_number{Value:=1.0*Cost}
                    BigNumber_GreaterThanOrEquals(CurrentCurrency, BigCost)?
                    PlayerLevel >= MyShop.GetLevelRequirement(CurrentLevel)
                then:
                    PlayerManager.GainUpgradeLevel(Name, 1)
                    PlayerManager.GainCollectible(collectible_data{Key:=CurrencyName, Count:=BigCost * -1.0})
                    
                    PlayerManager.UpdateQuestProgressForTask(task_event.CollectUpgrades, big_number{Value := CurrentLevel*1.0 + 1.0}, false, ?Target := Name, ?IsPersistent := true)
                    if:
                        CumulativeUpgrades := PlayerManager.GetCumulativeUpgradeLevels[]
                    then:
                        PlayerManager.UpdateQuestProgressForTask(task_event.CumulativeUpgrades, big_number{Value := CumulativeUpgrades * 1.0}, false, ?Target := Name, ?IsPersistent := true)
                    GameManager.AudioManager.PlayUpgradeSuccessSFX(Player)
                
                    MyShop.UpdateButtonsText(Player)
                    MyShop.BeginBounceEffect(Index)

                else:
                    #Not enough gems
                    GameManager.AudioManager.PlayUpgradeFailSFX(Player)
            else:
                #Level capped
                GameManager.AudioManager.PlayUpgradeFailSFX(Player)
        else:
            Log("Something went wrong buying item {Name}")

    GetCurrencyName(Level : int) : string =
        if:
            Temp :=LevelCurrencies[Min(Level, LevelCurrencies.Length - 1)]
        then:
            return Temp
        else:
            return "Gems"


upgrade_shop:= class(creative_device):

    @editable:
    GameManager : game_manager = game_manager{}

    @editable
    var Title : string = "SHOP"

    @editable
    var Items : []upgrade_item = array{}

    OnBegin<override>()<suspends>:void=
        for (Item : Items):
            if:
                set GameManager.UpgradeValues[Item.Name] = array{}
                set GameManager.UpgradeDatas[Item.Name] = Item
                for (Value : Item.LevelValues):
                    if:
                        NewUpgradeValues := GameManager.UpgradeValues[Item.Name] + array{Value}
                        set GameManager.UpgradeValues[Item.Name] = NewUpgradeValues
                
                set GameManager.UpgradeLevelCosts[Item.Name] = array{}
                for (Value : Item.LevelCosts):
                    if:
                        NewLevelCosts := GameManager.UpgradeLevelCosts[Item.Name] + array{Value}
                        set GameManager.UpgradeLevelCosts[Item.Name] = NewLevelCosts

    TryOpenMenu(Name : string, InPlayer:agent, ?NPCContactMenu : ?npc_contact_device = false):void=
        if:
            PlayerManager:=GameManager.PlayerManagers[InPlayer]
            not (PlayerManager.UIManager?.Popup?)
        then:
            var NewItems : []upgrade_item = array{}
            for (Item : Items):
                set NewItems = NewItems + array{Item.Clone()}

            MakeUpgradeShop(GameManager, Name, Title, NewItems, InPlayer, NPCContactMenu).OpenMenu()
        else:
            block:#Log("DOUBLE MENU")
                    
MakeUpgradeShop<constructor>(GameManagerIn : game_manager, InName : string, InTitle : string, ItemsIn : []upgrade_item, AgentIn : agent, InMenu : ?npc_contact_device) := player_upgrade_shop:
    GameManager :=  GameManagerIn
    Name:= InName
    Title := InTitle
    Items := ItemsIn
    Agent := AgentIn
    Menu := InMenu

player_upgrade_shop:= class:
    GameManager : game_manager
    Name : string 
    Title : string 
    Items : []upgrade_item 
    Agent : agent
    Menu : ?npc_contact_device
    
    BaseLevelRequirement : int = 5
    PerLevelRequirement : int = 10

    var ItemButtons : [player][]text_button_base = map{}
    var ButtonIcons : [player][]texture_block = map{}
    StoreBackground : texture_block = texture_block{ DefaultImage := Images.StoreBox, DefaultDesiredSize := vector2{X := 2000.0, Y := 1000.0}}
    GemPlate : texture_block = texture_block{ DefaultImage := Images.DialogBox, DefaultDesiredSize := vector2{X := 300.0, Y := 60.0}}    
    
    var ItemLevelTexts : []stack_box_slot = array{}
    var Item1LevelText : stack_box_slot = stack_box_slot{ Widget:= overlay{}}
    var Item2LevelText : stack_box_slot = stack_box_slot{ Widget:= overlay{}}
    var Item3LevelText : stack_box_slot = stack_box_slot{ Widget:= overlay{}}

    var ItemValueTexts : []stack_box_slot = array{}
    var Item1ValueText : stack_box_slot = stack_box_slot{ Widget:= overlay{}}
    var Item2ValueText : stack_box_slot = stack_box_slot{ Widget:= overlay{}}
    var Item3ValueText : stack_box_slot = stack_box_slot{ Widget:= overlay{}}

    var ItemLevelImgs : []stack_box_slot = array{}
    var Item1LevelImg : stack_box_slot = stack_box_slot{ Widget:= overlay{}}
    var Item2LevelImg : stack_box_slot = stack_box_slot{ Widget:= overlay{}}
    var Item3LevelImg : stack_box_slot = stack_box_slot{ Widget:= overlay{}}
    
    var CurrencyIcon1 : texture_block = texture_block{ DefaultImage:= Images.Empty, DefaultDesiredSize := vector2{X := 50.0, Y := 50.0}}
    var CurrencyIcon2 : texture_block = texture_block{ DefaultImage:= Images.Empty, DefaultDesiredSize := vector2{X := 50.0, Y := 50.0}}
    var CurrencyIcon3 : texture_block = texture_block{ DefaultImage:= Images.Empty, DefaultDesiredSize := vector2{X := 50.0, Y := 50.0}}

    var CurrencyText1 : text_block = text_block{DefaultTextColor:=NamedColors.White,
            DefaultShadowColor:=NamedColors.Black,
            DefaultShadowOffset:=option{vector2{X:=2.0,Y:=2.0}}}
    var CurrencyText2 : text_block = text_block{DefaultTextColor:=NamedColors.White,
            DefaultShadowColor:=NamedColors.Black,
            DefaultShadowOffset:=option{vector2{X:=2.0,Y:=2.0}}}
    var CurrencyText3 : text_block = text_block{DefaultTextColor:=NamedColors.White,
            DefaultShadowColor:=NamedColors.Black,
            DefaultShadowOffset:=option{vector2{X:=2.0,Y:=2.0}}}

    Size : vector2 = vector2{X := 50.0, Y := 50.0}
    IconBounce : float = 1.25
    GatherBounceIterations : float = 8.0

    GetLevelRequirement(ItemLevel : int)<transacts>:int=
        if:
            ItemLevel > 0
            InPlayer := player[Agent]
            PlayerManager := GameManager.PlayerManagers[InPlayer]
        then: 
            return BaseLevelRequirement + PerLevelRequirement * ItemLevel - (PlayerManager.GetUpgradeLevel["P_ReducedUpgradeLevelRequirement"] or 0)
        
        return 0

    BeginBounceEffect(Index : int):void=
        if:
            Img := texture_block[ItemLevelImgs[Index].Widget]
            Item := Items[Index]
        then:
            
            var ItemLevel : int = 0
            if:
                TempItemLevel := PlayerProfileDataMap[Agent].UpgradeLevels[Item.Name]
                set ItemLevel = TempItemLevel
            CurrencyName := Item.GetCurrencyName(ItemLevel)

            Txt1 := CurrencyText1.GetText()
            Txt2 := CurrencyText2.GetText()
            Txt3 := CurrencyText3.GetText()

            if (Txt1 = CurrencyName):
                CurrencyIcon1.SetDesiredSize(vector2{X:= Size.X * IconBounce, Y:= Size.Y * IconBounce})
                spawn:
                    IterateBounceEffect(CurrencyIcon1)  
            else if (Txt2 = CurrencyName):
                CurrencyIcon2.SetDesiredSize(vector2{X:= Size.X * IconBounce, Y:= Size.Y * IconBounce})
                spawn:
                    IterateBounceEffect(CurrencyIcon2)  
            else if (Txt3 = CurrencyName):
                CurrencyIcon3.SetDesiredSize(vector2{X:= Size.X * IconBounce, Y:= Size.Y * IconBounce})
                spawn:
                    IterateBounceEffect(CurrencyIcon3)  


            Img.SetDesiredSize(vector2{X:= Size.X * IconBounce, Y:= Size.Y * IconBounce})
            spawn:
                IterateBounceEffect(Img)        

    IterateBounceEffect(Img : texture_block)<suspends>:void =
        limit:int=12

        for(X:=1..limit):
            Sleep(0.01)
            CurrentX := Img.GetDesiredSize().X 
            if(CurrentX > Size.X):
                CurrentY := Img.GetDesiredSize().Y
                ShrinkX := Max(CurrentX * ( IconBounce - 1.0 ) / GatherBounceIterations, 0.25)
                ShrinkY := Max(CurrentY * ( IconBounce - 1.0 ) / GatherBounceIterations, 0.25)
                Img.SetDesiredSize(vector2{X := Max(CurrentX - ShrinkX, Size.X), Y:= Max(CurrentY - ShrinkY, Size.Y)})
        

    OpenMenu():void =
        SetupItemArrays()
        
        if:
            InPlayer := player[Agent]
            PlayerUI := GetPlayerUI[InPlayer]
            PlayerManager := GameManager.PlayerManagers[InPlayer]
            UIManager := PlayerManager.UIManager?
            Gems:= PlayerManager.GetGems[]
        then:
            CurrencyText1.SetShadowOpacity(1.0)    
            CurrencyText1.SetText(STM(""))  
            CurrencyText2.SetShadowOpacity(1.0)    
            CurrencyText2.SetText(STM(""))   
            CurrencyText3.SetShadowOpacity(1.0)    
            CurrencyText3.SetText(STM(""))   

            AbilityBGStackBox := stack_box{Orientation:=orientation.Horizontal}

            if:
                not ItemButtons[InPlayer]
                set ItemButtons[InPlayer] = array{}
                set ButtonIcons[InPlayer] = array{}
            then:
                var i : int = 0

                for (Item:Items):
                    ItemButton := button_loud{}

                    set Item.GameManager = GameManager
                    set Item.UpgradeShop = option{Self}
                    
                    # var ItemText : text_block = text_block{
                    #     DefaultTextColor:=NamedColors.White,
                    #     DefaultShadowColor:=NamedColors.Black,
                    #     DefaultShadowOffset:=option{vector2{X:=2.0,Y:=2.0}}}

                    var ItemLevel : int = 0
                    if:
                        TempItemLevel := PlayerProfileDataMap[InPlayer].UpgradeLevels[Item.Name]
                        set ItemLevel = TempItemLevel
        
                    TempEventID := Item.QuestRequirement
                    
                    ItemButton.OnClick().SubscribeWidget(Item.OnBuy, i)

                    if:
                        TempArray := ItemButtons[InPlayer] 
                        set ItemButtons[InPlayer] = TempArray + array{ItemButton}


                    ButtonIcon := texture_block{DefaultImage := Images.Empty, DefaultDesiredSize := vector2{X := 50.0, Y := 50.0}}
                    
                    if:
                        TempArray := ButtonIcons[InPlayer] 
                        set ButtonIcons[InPlayer] = TempArray + array{ButtonIcon}
                    
                    set i += 1
                    
            var Index : int = 0
            for (Btn:ItemButtons[InPlayer]):
                if:
                    Item := Items[Index]
                    ButtonIcon := ButtonIcons[InPlayer][Index]
                then:
                    AbilityStack := stack_box:
                        Orientation:= orientation.Vertical
                    
                    UpgradePlate : texture_block = texture_block{ DefaultImage := Images.DialogBoxVertical, DefaultDesiredSize := vector2{X := 610.0, Y := 720.0}}

                    TempSlot := stack_box_slot:
                        Widget:= overlay:
                            Slots:= array:
                                overlay_slot:
                                    Widget:=UpgradePlate
                                    HorizontalAlignment:= horizontal_alignment.Center
                                overlay_slot:
                                    Widget:=AbilityStack
                                    HorizontalAlignment:= horizontal_alignment.Center
                                    VerticalAlignment:= vertical_alignment.Center
                        Padding:= margin{Top := 0.0, Left := 2.5, Right := 0.0, Bottom := 00.0}
                    AbilityBGStackBox.AddWidget(TempSlot)

                    #Ability Name
                    var DisplayName : string = ""
                    var TextureLookupString : string = ""
                    var TempArray : []stack_box_slot = array{}
                    TempEventID := Item.QuestRequirement
                    
                    #Ability Values
                    var ValueText: message = STM("")
                        
                    #Ability Level
                    var TempItemLevel : int = 0
                    var TempMsg : message = STM("[ {TempItemLevel} ]")
                    
                    if(PlayerManager.HasCompletedTask(TempEventID.QuestID, TempEventID.TaskID)?):
                        PlayerManager.GainAchievementLevel("InteractedWithNPC_{Name}", 1)
                        set DisplayName = Item.Name
                        set TextureLookupString = Item.Name
                        set TempArray = ParseText(Item.Description)
                        if (Item.Name_Override <> "") { set DisplayName = Item.Name_Override }

                        #Ability Level
                        if:
                            Temp := PlayerProfileDataMap[InPlayer].UpgradeLevels[Item.Name]
                            set TempItemLevel = Temp
                        then:
                            set TempMsg = STM("[ {TempItemLevel} ]")
                        if:
                            PlayerManager.GetUpgradeValue[Item.Name, false] < 900.0
                            ValueTemp := "{FloatToTruncatedString(PlayerManager.GetUpgradeValue[Item.Name, false],Item.ValueDecimals)}{Item.ValueUnit}"
                        then:
                            set ValueText = STM("{ValueTemp}")
                        else:
                            if:
                                ValueBigNum := "{big_number{Value:=PlayerManager.GetUpgradeValue[Item.Name, false]}}{Item.ValueUnit}"
                            then:
                                set ValueText = STM("{ValueBigNum}")

                        if:
                            PlayerManager.GetUpgradeValue[Item.Name, false] < 900.0
                            ValueTemp := "{FloatToTruncatedString(PlayerManager.GetUpgradeValue[Item.Name, false],Item.ValueDecimals)}{Item.ValueUnit} -> {FloatToTruncatedString(PlayerManager.GetUpgradeValue[Item.Name, option{TempItemLevel + 1}],Item.ValueDecimals)}{Item.ValueUnit}"
                            #Uncommon Green 32cf13
                        then:
                            set ValueText = STM("{ValueTemp}")
                        else:
                            if:
                                ValueBigNum := "{big_number{Value:=PlayerManager.GetUpgradeValue[Item.Name, false]}}{Item.ValueUnit} -> {big_number{Value:=PlayerManager.GetUpgradeValue[Item.Name, option{TempItemLevel + 1}]}}{Item.ValueUnit}"
                            then:
                                set ValueText = STM("{ValueBigNum}")

                    else:
                        set DisplayName = Item.RestrictedName
                        set TextureLookupString = Item.RestrictedName
                        if (Item.RestrictedName_Override <> "") { set DisplayName = Item.RestrictedName_Override }
                        set TempArray = ParseText(Item.RestrictedDescription)
                        set TempMsg = STM("[ ??? / ??? ]")

                    AbilityStack.AddWidget(
                        stack_box_slot:
                            Widget:= text_block:
                                DefaultTextColor:=NamedColors.White
                                DefaultShadowColor:=NamedColors.Black
                                DefaultShadowOffset:=option{vector2{X:=2.0,Y:=2.0}}
                                DefaultShadowOpacity:= 1.0
                                DefaultText:= STM("{DisplayName}")
                        )

                    if:
                        set ItemLevelTexts[Index] = stack_box_slot:
                                Widget := text_block:
                                    DefaultTextColor:=NamedColors.White
                                    DefaultShadowColor:=NamedColors.Black
                                    DefaultShadowOffset:=option{vector2{X:=2.0,Y:=2.0}}
                                    DefaultShadowOpacity:= 1.0
                                    DefaultText:= TempMsg
                            Padding:= margin{Top := -5.0, Left := 5.0, Right := 0.0, Bottom := 0.0}
                        
                        set ItemLevelImgs[Index] = stack_box_slot:
                                Widget := texture_block{ DefaultImage := Images.Levelup, DefaultDesiredSize := Size} 
                                Padding:= margin{Top := -5.0, Left := 0.0, Right := 0.0, Bottom := 0.0}

                        TempLevelSlot := stack_box_slot:
                            Widget := stack_box:
                                Orientation := orientation.Horizontal
                                Slots := array{ItemLevelImgs[Index], ItemLevelTexts[Index]}
                    then:
                        AbilityStack.AddWidget(TempLevelSlot)
                
                    #Ability Icon
                    AbilityStack.AddWidget(
                        stack_box_slot:
                            Widget:= texture_block:
                                DefaultImage:=  GetTextureFromString(TextureLookupString)
                                DefaultDesiredSize := vector2{X:= 250.0, Y:= 250.0}
                            Padding:= margin{Top := 5.0, Left := 0.0, Right := 0.0, Bottom := 0.0}                                
                        )


                    if:
                        set ItemValueTexts[Index] = stack_box_slot:
                            Widget:= text_block:
                                DefaultTextColor:=NamedColors.White
                                DefaultShadowColor:=NamedColors.Black
                                DefaultShadowOffset:=option{vector2{X:=2.0,Y:=2.0}}
                                DefaultShadowOpacity:= 1.0
                                DefaultText:= ValueText
                            Padding:= margin{Top := 5.0, Left := 0.0, Right := 0.0, Bottom := 0.0}       
                        
                        TempValueSlot := ItemValueTexts[Index]
                    then:
                        AbilityStack.AddWidget(TempValueSlot)

                    #Ability Description
                    DescBox1 := stack_box{Orientation:=orientation.Horizontal}
                    DescBox2 := stack_box{Orientation:=orientation.Horizontal}
                    DescBox3 := stack_box{Orientation:=orientation.Horizontal}
                    VerticalDescBox : stack_box = stack_box:
                        Orientation := orientation.Vertical
                        Slots:= array:
                            stack_box_slot:
                                Widget:=DescBox1
                                HorizontalAlignment:= horizontal_alignment.Left
                                Padding:= margin{Top := 0.0, Left := 10.0, Right := 0.0, Bottom := 0.0}
                            stack_box_slot:
                                Widget:=DescBox2
                                HorizontalAlignment:= horizontal_alignment.Left
                                Padding:= margin{Top := -5.0, Left := 10.0, Right := 0.0, Bottom := 0.0}
                            stack_box_slot:
                                Widget:=DescBox3
                                HorizontalAlignment:= horizontal_alignment.Left
                                Padding:= margin{Top := -5.0, Left := 10.0, Right := 0.0, Bottom := 0.0}
                    
                    AbilityStack.AddWidget(
                        stack_box_slot:
                            Widget:= VerticalDescBox
                            Padding:= margin{Top := 10.0, Left := 0.0, Right := 0.0, Bottom := 0.0}
                        )
                    
                    var BreakCount : int = 0
                    for(Slot : TempArray):
                        var Skip : logic = false
                        if:
                            TempBlock:=text_block[Slot.Widget]
                        then:
                            TempText:=TempBlock.GetText()
                            if (TempText = "<br>"):
                                set BreakCount += 1
                                set Skip = true
                        
                        if(not(Skip?)):
                            case(BreakCount):
                                1 => DescBox2.AddWidget(Slot)
                                2 => DescBox3.AddWidget(Slot)
                                _ => DescBox1.AddWidget(Slot)

                    #Ability Button
                    AbilityStack.AddWidget(
                        stack_box_slot:
                            Widget:= overlay:
                                Slots:= array:
                                    overlay_slot:
                                        Widget:= Btn
                                        Padding:= margin{Top := 00.0, Left := 0.0, Right := 0.0, Bottom := 00.0}
                                    overlay_slot:
                                        Widget:= ButtonIcon
                                        HorizontalAlignment := horizontal_alignment.Right
                                        Padding:= margin{Top := -20.0, Left := -20.0, Right := 10.0, Bottom := -20.0}
                            Padding:= margin{Top := 20.0, Left := 0.0, Right := 0.0, Bottom := 10.0}
                        )
                    
                set Index += 1

            UpdateButtonsText(InPlayer)

            CloseButton := button_quiet:
            CloseButton.SetText(STM("                                    BACK                                    "))            
            CloseButton.OnClick().Subscribe(OnClose)
            
            Canvas := canvas:
                Slots := array:
                    canvas_slot:
                        Widget:= StoreBackground
                        SizeToContent := true
                        Anchors := anchors{ Minimum := vector2{X := 0.005, Y := 0.95}, Maximum := vector2{X := 0.005, Y := 0.95}}
                        Alignment := vector2{X := 0.0, Y := 1.0}
                        ZOrder := 1
                    canvas_slot:
                        Widget := AbilityBGStackBox
                        Anchors := anchors{ Minimum := vector2{X := 0.5, Y := 0.82}, Maximum := vector2{X := 0.5, Y := 0.82}}                        
                        Alignment := vector2{X := 0.5, Y := 1.0}
                        ZOrder := 2
                    canvas_slot:
                        Widget:= text_block:
                            DefaultTextColor:=NamedColors.White
                            DefaultShadowColor:=NamedColors.Black
                            DefaultShadowOffset:=option{vector2{X:=2.0,Y:=2.0}}
                            DefaultShadowOpacity:= 1.0
                            DefaultText:= STM("{Title}")
                        Anchors := anchors{ Minimum := vector2{X := 0.5, Y := 0.95}, Maximum := vector2{X := 0.5, Y := 0.95}}
                        Offsets := margin{Top := -900.0, Left := 0.0, Right := 0.0, Bottom := 0.0}
                        Alignment := vector2{X := 0.5, Y := 1.0}
                        ZOrder := 3
                    canvas_slot:
                        Widget:= GemPlate
                        SizeToContent := true
                        Anchors := anchors{ Minimum := vector2{X := 0.75, Y := 0.95}, Maximum := vector2{X := 0.75, Y := 0.95}}
                        Offsets := margin{Top := -865.0, Left := 0.0, Right := 0.0, Bottom := 0.0}
                        Alignment := vector2{X := 0.0, Y := 1.0}
                        ZOrder := 4
                    canvas_slot:
                        Widget:= stack_box:
                            Orientation:= orientation.Horizontal
                            Slots:= array:
                                stack_box_slot:
                                    Widget:= CurrencyIcon1                                                    
                                    HorizontalAlignment:= horizontal_alignment.Left
                                    Padding:= margin{Left:=2.0, Top:=0.0, Bottom:= 0.0, Right:= -2.0} 
                                    VerticalAlignment:= vertical_alignment.Center
                                stack_box_slot:
                                    Widget:= CurrencyText1
                                    HorizontalAlignment:= horizontal_alignment.Left
                                    Padding:= margin{Left:=2.0, Top:=0.0, Bottom:= 0.0, Right:= 2.0} 
                                stack_box_slot:
                                    Widget:= CurrencyIcon2                                                    
                                    HorizontalAlignment:= horizontal_alignment.Left
                                    Padding:= margin{Left:=12.0, Top:=0.0, Bottom:= 0.0, Right:= -2.0} 
                                    VerticalAlignment:= vertical_alignment.Center
                                stack_box_slot:
                                    Widget:= CurrencyText2
                                    HorizontalAlignment:= horizontal_alignment.Left
                                    Padding:= margin{Left:=2.0, Top:=0.0, Bottom:= 0.0, Right:= 2.0} 
                                stack_box_slot:
                                    Widget:= CurrencyIcon3                                                    
                                    HorizontalAlignment:= horizontal_alignment.Left
                                    Padding:= margin{Left:=12.0, Top:=0.0, Bottom:= 0.0, Right:= -2.0} 
                                    VerticalAlignment:= vertical_alignment.Center
                                stack_box_slot:
                                    Widget:= CurrencyText3
                                    HorizontalAlignment:= horizontal_alignment.Left
                                    Padding:= margin{Left:=2.0, Top:=0.0, Bottom:= 0.0, Right:= 2.0} 
                        Anchors := anchors{ Minimum := vector2{X := 0.75, Y := 0.95}, Maximum := vector2{X := 0.75, Y := 0.95}}
                        Offsets := margin{Top := -870.0, Left := 20.0, Right := 0.0, Bottom := 0.0}
                        Alignment := vector2{X := 0.0, Y := 1.0}
                        ZOrder := 5
                    canvas_slot:
                        Widget:= CloseButton
                        Anchors := anchors{ Minimum := vector2{X := 0.5, Y := 0.89}, Maximum := vector2{X := 0.5, Y := 0.89}}
                        Alignment := vector2{X := 0.5, Y := 1.0}
                        Offsets := margin{Top := 0.0, Left := 0.0, Right := 0.0, Bottom := 0.0}
                        ZOrder := 6

            PlayerUI.AddWidget(Canvas, player_ui_slot{InputMode := ui_input_mode.All})
            if (set GameManager.PlayerManagers[InPlayer].UIManager?.Popup = option{Canvas}) {}
        
    OnClose(Msg:widget_message):void=
        if:
            InPlayer := player[Agent]
            PlayerManager := GameManager.PlayerManagers[InPlayer]
        then:
            PlayerManager.ClosePopup(Msg)
            if:
                SetMenu := Menu?
            then:
                SetMenu.OpenMenu(Agent)

    UpdateButtonsText(InPlayer : player):void =
        if:        
            PlayerManager := GameManager.PlayerManagers[InPlayer]
            PlayerLevel := PlayerManager.GetLevel[]
        then:   
            #Update Currency Displays
            var CurrentCurrencies : []string = array{}
            for (Item : Items):
                var ItemLevel : int = 0
                if:
                    TempItemLevel := PlayerProfileDataMap[InPlayer].UpgradeLevels[Item.Name]
                    set ItemLevel = TempItemLevel
                CurrencyName := Item.GetCurrencyName(ItemLevel)
                if (CurrentCurrencies.Find[CurrencyName]):
                    set CurrentCurrencies += array{"SKIP"}
                else:
                    set CurrentCurrencies += array{CurrencyName}

            var CurrencyID : int = 0
            for (Currency : CurrentCurrencies):
                if (Currency = "SKIP"):
                    case (CurrencyID):
                        0 => CurrencyText1.SetVisibility(widget_visibility.Collapsed)
                        1 => CurrencyText2.SetVisibility(widget_visibility.Collapsed)
                        2 => CurrencyText3.SetVisibility(widget_visibility.Collapsed)
                        _ => Log("ERROR: Invalid Currency ID")
                    case (CurrencyID):
                        0 => CurrencyIcon1.SetVisibility(widget_visibility.Collapsed)
                        1 => CurrencyIcon2.SetVisibility(widget_visibility.Collapsed)
                        2 => CurrencyIcon3.SetVisibility(widget_visibility.Collapsed)
                        _ => Log("ERROR: Invalid Currency ID")
                else:
                    case (CurrencyID):
                        0 => CurrencyText1.SetVisibility(widget_visibility.Visible)
                        1 => CurrencyText2.SetVisibility(widget_visibility.Visible)
                        2 => CurrencyText3.SetVisibility(widget_visibility.Visible)
                        _ => Log("ERROR: Invalid Currency ID")
                    case (CurrencyID):
                        0 => CurrencyIcon1.SetVisibility(widget_visibility.Visible)
                        1 => CurrencyIcon2.SetVisibility(widget_visibility.Visible)
                        2 => CurrencyIcon3.SetVisibility(widget_visibility.Visible)
                        _ => Log("ERROR: Invalid Currency ID")
                    case (CurrencyID):
                        0 => CurrencyText1.SetText(STM("{PlayerManager.GetCollectibleCount(Currency)}"))
                        1 => CurrencyText2.SetText(STM("{PlayerManager.GetCollectibleCount(Currency)}"))
                        2 => CurrencyText3.SetText(STM("{PlayerManager.GetCollectibleCount(Currency)}"))
                        _ => Log("ERROR: Invalid Currency ID")
                    case (CurrencyID):
                        0 => CurrencyIcon1.SetImage(GetTextureFromString(Currency))
                        1 => CurrencyIcon2.SetImage(GetTextureFromString(Currency))
                        2 => CurrencyIcon3.SetImage(GetTextureFromString(Currency))
                        _ => Log("ERROR: Invalid Currency ID")
                set CurrencyID += 1

            #Update buttons
            var i : int = 0
            for (Item:Items):
                if:
                    ItemButton := ItemButtons[InPlayer][i]
                    ItemIcon := ButtonIcons[InPlayer][i]
                    var CostString: string = "MAXED" 
                then:
                    var ItemLevel : int = 0
                    var Cost : int = 0
                    if:
                        TempItemLevel := PlayerProfileDataMap[InPlayer].UpgradeLevels[Item.Name]
                        set ItemLevel = TempItemLevel

                    CurrencyName := Item.GetCurrencyName(ItemLevel)
                    if:
                        CurrencyCount := PlayerManager.GetCollectibleCount(CurrencyName)
                        TempCost := PlayerManager.GetUpgradeLevelCost[Item.Name, option{ItemLevel}]
                        set Cost = TempCost
                        set CostString = "{Cost} {CurrencyName}"
                    
                    TempEventID := Item.QuestRequirement
                    LevelRequirement := GetLevelRequirement(ItemLevel)
                    if(PlayerManager.HasCompletedTask(TempEventID.QuestID, TempEventID.TaskID)?):
                        TempMsg:= STM("[ {ItemLevel} ]")


                        #Update Item Level Texts
                        if:
                            Slot:=ItemLevelTexts[i]
                            TextBlock:= text_block[Slot.Widget]
                        then:
                            TextBlock.SetText(TempMsg)

                        #Update Item Value Texts
                        var ValueText: message = STM("")
                        if:
                            PlayerManager.GetUpgradeValue[Item.Name, false] < 900.0
                            ValueTemp := "{FloatToTruncatedString(PlayerManager.GetUpgradeValue[Item.Name, false],Item.ValueDecimals)}{Item.ValueUnit}"
                        then:
                            set ValueText = STM("{ValueTemp}")
                        else:
                            if:
                                ValueBigNum := "{big_number{Value:=PlayerManager.GetUpgradeValue[Item.Name, false]}}{Item.ValueUnit}"
                            then:
                                set ValueText = STM("{ValueBigNum}")

                        if:
                            PlayerManager.GetUpgradeValue[Item.Name, false] < 900.0
                            ValueTemp := "{FloatToTruncatedString(PlayerManager.GetUpgradeValue[Item.Name, false],Item.ValueDecimals)}{Item.ValueUnit} -> {FloatToTruncatedString(PlayerManager.GetUpgradeValue[Item.Name, option{ItemLevel + 1}],Item.ValueDecimals)}{Item.ValueUnit}"
                            #Uncommon Green 32cf13
                        then:
                            set ValueText = STM("{ValueTemp}")
                        else:
                            if:
                                ValueBigNum := "{big_number{Value:=PlayerManager.GetUpgradeValue[Item.Name, false]}}{Item.ValueUnit} -> {big_number{Value:=PlayerManager.GetUpgradeValue[Item.Name, option{ItemLevel + 1}]}}{Item.ValueUnit}"
                            then:
                                set ValueText = STM("{ValueBigNum}")
                        if:
                            Slot:= ItemValueTexts[i]
                            TextBlock:=text_block[Slot.Widget]
                        then:
                            TextBlock.SetText(ValueText)

                        if:
                            PlayerLevel < LevelRequirement
                        then:
                            ItemButton.SetText(STM("Requires Level {LevelRequirement}"))
                            ItemIcon.SetImage(Images.Empty)
                        else:
                            if (CostString = "MAXED"):
                                ItemButton.SetText(STM("MAX"))
                                ItemIcon.SetImage(Images.Empty)
                            else:
                                ItemButton.SetText(STM("UPGRADE: {Cost}    "))
                                ItemIcon.SetImage(GetTextureFromString(CurrencyName))
        
                    else:
                        if (CostString = "MAXED"):
                            ItemButton.SetText(STM("MAX"))
                        else:
                            ItemButton.SetText(STM("LOCKED"))
                        ItemIcon.SetImage(Images.Empty)
                    set i = i + 1
                        
    SetupItemArrays():void=
        set ItemLevelTexts = array{Item1LevelText, Item2LevelText, Item3LevelText}
        set ItemValueTexts = array{Item1ValueText, Item2ValueText, Item3ValueText}
        set ItemLevelImgs = array{Item1LevelImg, Item2LevelImg, Item3LevelImg}