using { /Verse.org/Simulation }
using { /UnrealEngine.com/Temporary/UI }
using { /UnrealEngine.com/Temporary/SpatialMath }
using { /UnrealEngine.com/Temporary/UI }
using { /Fortnite.com/UI }
using { /Verse.org/Random }
using { /Verse.org/Colors }

(Listenable : listenable(agent)).SubscribeAgent(OutputFunc : tuple(agent, t)->void, Extra2Data : t where t:type) : cancelable =
    Wrapper := wrapper_agent(t){ExtraData := Extra2Data, OutputFunc := OutputFunc}
    Listenable.Subscribe(Wrapper.InputFunc)
 
wrapper_agent(t : type) := class():
    ExtraData : t;
    OutputFunc : tuple(agent, t) -> void
    InputFunc(Agent : agent):void = OutputFunc(Agent, ExtraData)
 
(Listenable : listenable(tuple())).SubscribeEmpty(OutputFunc : t -> void, ExtraData : t where t:type) : cancelable =
    Wrapper := wrapper_empty(t) {ExtraData := ExtraData, OutputFunc := OutputFunc}
    Listenable.Subscribe(Wrapper.InputFunc)
    
(Listenable : listenable(widget_message)).SubscribeWidget(OutputFunc : tuple(widget_message, t)->void, Extra2Data : t where t:type) : cancelable =
    Wrapper := wrapper_widget(t){ExtraData := Extra2Data, OutputFunc := OutputFunc}
    Listenable.Subscribe(Wrapper.InputFunc)
    
wrapper_widget(t : type) := class():
    ExtraData : t;
    OutputFunc : tuple(widget_message, t) -> void
    InputFunc(Msg : widget_message):void = OutputFunc(Msg, ExtraData)
 
wrapper_empty(t : type) := class():
    ExtraData : t;
    OutputFunc : t -> void
    InputFunc():void = OutputFunc(ExtraData)
    
# Removes an element from the given map and returns a new map without that element
RemoveKeyFromMap(ExampleMap:[string]int, ElementToRemove:string):[string]int=
    var NewMap:[string]int = map{}
    # Concatenate Keys from ExampleMap into NewMap, excluding ElementToRemove
    for (Key -> Value : ExampleMap, Key <> ElementToRemove):
        set NewMap = ConcatenateMaps(NewMap, map{Key => Value})
    return NewMap

# Removes an element from the given map and returns a new map without that element
RemoveKeyFromMap(ExampleMap:[int]tuple(int,int), ElementToRemove:int):[int]tuple(int,int)=
    var NewMap:[int]tuple(int,int) = map{}
    # Concatenate Keys from ExampleMap into NewMap, excluding ElementToRemove
    for (Key -> Value : ExampleMap, Key <> ElementToRemove):
        set NewMap = ConcatenateMaps(NewMap, map{Key => Value})
    return NewMap

# Removes an element from the given map and returns a new map without that element
RemoveKeyFromMap(ExampleMap:[player]float, ElementToRemove:player):[player]float= 
    var NewMap:[player]float = map{}
    # Concatenate Keys from ExampleMap into NewMap, excluding ElementToRemove
    for (Key -> Value : ExampleMap, Key <> ElementToRemove):
        set NewMap = ConcatenateMaps(NewMap, map{Key => Value})
    return NewMap

# Removes an element from the given map and returns a new map without that element
RemoveKeyFromMap3(ExampleMap:[player]cancelable, ElementToRemove:player):[player]cancelable= 
    var NewMap:[player]cancelable = map{}
    # Concatenate Keys from ExampleMap into NewMap, excluding ElementToRemove
    for (Key -> Value : ExampleMap, Key <> ElementToRemove):
        set NewMap = ConcatenateMaps(NewMap, map{Key => Value})
    return NewMap

# Removes an element from the given map and returns a new map without that element
RemoveKeyFromMap2(ExampleMap:[player]big_number, ElementToRemove:player):[player]big_number= 
    var NewMap:[player]big_number = map{}
    # Concatenate Keys from ExampleMap into NewMap, excluding ElementToRemove
    for (Key -> Value : ExampleMap, Key <> ElementToRemove):
        set NewMap = ConcatenateMaps(NewMap, map{Key => Value})
    return NewMap

# Removes an element from the given map and returns a new map without that element
RemoveKeyFromMap2(ExampleMap:[string]collectible_data, ElementToRemove:string):[string]collectible_data= 
    var NewMap:[string]collectible_data = map{}
    # Concatenate Keys from ExampleMap into NewMap, excluding ElementToRemove
    for (Key -> Value : ExampleMap, Key <> ElementToRemove):
        set NewMap = ConcatenateMaps(NewMap, map{Key => Value})
    return NewMap

# Removes an element from the given map and returns a new map without that element
# RemoveKeyFromMap(ExampleMap: [comparable]type, ElementToRemove: comparable): [comparable]type =
#     var NewMap : [comparable]type = map{}
    
#     # Concatenate Keys from ExampleMap into NewMap, excluding ElementToRemove
#     for (Key -> Value : ExampleMap, Key <> ElementToRemove):
#         set NewMap = ConcatenateMaps(NewMap, map{Key => Value})
#     return NewMap

#Gets the first N chars of a string
GetCharsFromString(String : string, N : int)<transacts>:[]char=
    var Output : string = ""
    for (X := 0 .. N - 1):
        if:
            Char := String[X]
        then:
            set Output = Output + array{Char}
    return Output

# Returns Number converted to a string, truncated to Decimals places.
FloatToTruncatedString<public>(Number:float, Decimals:int)<transacts>:[]char=
    var Str:[]char = ToString(Number)
    if:
        DotIndex := Str.Find['.']
        StopIndex := if (Decimals > 0) then Min(DotIndex+Decimals+1,Str.Length) else DotIndex
        Temp:= Str.Slice[0, StopIndex] 
        set Str = Temp
    Str

ParseDigit(Char : char)<decides><transacts>:int=
    return case (Char):
        '0' => 0
        '1' => 1
        '2' => 2
        '3' => 3
        '4' => 4
        '5' => 5
        '6' => 6
        '7' => 7
        '8' => 8
        '9' => 9
        _ => block:
            false?
            0

ParseInt(Str : string)<decides><transacts>:int=
    var Output : int = 0
    if:
        Str[0] = '-'
        PositivePart := Str.Slice[1, Str.Length]
    then:
        set Output = -1 * ParseInt[PositivePart] or 0
    else:
        for (X := 0..Str.Length-1):
            if:
                Digit := ParseDigit[Str[X]]
            then:
                set Output = Output * 10 + Digit
            else if:
                Str[X] = ','
            else:
                false?

    return Output

Sign(X:float)<transacts>:int=
    if (X < 0.0):
        return -1
    else:
        return 1

Sign(X:int)<transacts>:int=
    if (X < 0):
        return -1
    else:
        return 1

# Nth Fibonnaci number
Fibonacci(N:int)<transacts>:int=
    if (N <= 1):
        return N
    else:
        return Fibonacci(N-1) + Fibonacci(N-2)

MakeUIPanel(?Width : float = 0.8, ?Height : float = 0.8):canvas =
    MarginH := (1.0-Width)/2.0
    MarginV := (1.0-Height)/2.0
    Canvas := canvas:
        Slots := array:
            canvas_slot:
                Widget:= texture_block{DefaultImage := Images.DialogBox}
                SizeToContent := true
                Anchors := anchors{ Minimum := vector2{X := MarginH, Y := MarginV}, Maximum := vector2{X := 1.0-MarginH, Y := 1.0-MarginV}}
                ZOrder := 0

WordWrap(TextIn : string, CharLimit : int, ?Alignment : horizontal_alignment = horizontal_alignment.Left, ?Color : color = NamedColors.White, ?ShadowColor : color = NamedColors.Black):stack_box=
    Text := TextIn

    var Breaks : []string = array{}

    var IsInBracket : logic = false
    var BracketExceptions : int = 0
    var BracketSpacing : int = 0 #Images rendered from brackets take up ~2 spaces
    var LastSpace : int = 0
    var LastBreak : int = 0
    for (i := 0..Text.Length-1):
        if:
            IsInBracket?
        then:
            set BracketExceptions += 1
            if:
                Temp:=Text[i] 
                Temp = '>'
            then:
                set IsInBracket = false
                set BracketSpacing += 3
        else:
            if (Text[i] = ' ' or i - (BracketExceptions - BracketSpacing) - LastBreak >= CharLimit):
                if (LastSpace = 0 or i - (BracketExceptions - BracketSpacing) - LastBreak < CharLimit):
                    set LastSpace = i
                else if (Pre := Text.Slice[LastBreak, LastSpace]):
                    set Breaks = Breaks + array{Pre}
                    set LastBreak = LastSpace+1
                    set BracketExceptions = 0
                    set BracketSpacing = 0
            else:
                if:
                    Temp:=Text[i] 
                    Temp = '<'
                then:
                    set IsInBracket = true
                    set BracketExceptions += 1

    var FinalText : string = ""
    # 
    if (Pre := Text.Slice[LastBreak, Text.Length]):
        set Breaks = Breaks + array{Pre}

    for (Break : Breaks):
        if:
            FinalText = ""
        then:
            set FinalText = Break
        else:
            set FinalText = FinalText + "<br>" + Break

    return MakeTextBox(FinalText, ?Alignment := Alignment, ?Color := Color, ?ShadowColor := ShadowColor)


MakeTextBox(Text : string, ?Alignment : horizontal_alignment = horizontal_alignment.Left, ?ImgSize : vector2 = vector2{X:=40.0, Y:=40.0}, ?Color : color = NamedColors.White, ?ShadowColor : color = NamedColors.Black):stack_box=
    TempArray:=ParseText(Text, ?ImgSize := ImgSize, ?Color := Color, ?ShadowColor := ShadowColor)

    var InnerBoxes : []stack_box = array{stack_box{Orientation:=orientation.Horizontal}}   
    
    var BreakCount : int = 0
    for (Slot : TempArray):
        var Skip : logic = false
        if:
            TempBlock:= text_block[Slot.Widget]
        then:
            TempText:= TempBlock.GetText()
            if (TempText = "<br>"):
                set BreakCount += 1
                set Skip = true
                set InnerBoxes += array{stack_box{Orientation:=orientation.Horizontal}}
    
        if:
            not(Skip?)
            Box := InnerBoxes[BreakCount]
        then:
            Box.AddWidget(Slot)
    
    OuterBox : stack_box = stack_box:
        Orientation := orientation.Vertical
    
    for (InnerBox : InnerBoxes):
        OuterBox.AddWidget(
            stack_box_slot:
                Widget:=InnerBox
                HorizontalAlignment:= Alignment
                Padding:= margin{Top := 0.0, Left := 0.0, Right := 0.0, Bottom := 0.0}
            )
    
    return OuterBox

SetOpacity(Text: text_block, Text2 : text_block, FloatValue : float):void=
    Text2Opacity:float=Text2.GetTextOpacity()
    if: 
        IntValue:=Int[FloatValue]
    then:
        case(IntValue):
            96=>
                Text.SetTextOpacity(0.90)
                Text.SetShadowOpacity(0.90)
                if(Text2Opacity > 0.0):
                    Text2.SetTextOpacity(0.90)
                    Text2.SetShadowOpacity(0.90)
            88=>
                Text.SetTextOpacity(0.80)
                Text.SetShadowOpacity(0.80)
                if(Text2Opacity > 0.0):
                    Text2.SetTextOpacity(0.80)
                    Text2.SetShadowOpacity(0.80)
            76=>
                Text.SetTextOpacity(0.70)
                Text.SetShadowOpacity(0.70)
                if(Text2Opacity > 0.0):
                    Text2.SetTextOpacity(0.70)
                    Text2.SetShadowOpacity(0.70)
            68=>
                Text.SetTextOpacity(0.60)
                Text.SetShadowOpacity(0.60)
                if(Text2Opacity > 0.0):
                    Text2.SetTextOpacity(0.60)
                    Text2.SetShadowOpacity(0.60)
            56=>
                Text.SetTextOpacity(0.50)
                Text.SetShadowOpacity(0.50)
                if(Text2Opacity > 0.0):
                    Text2.SetTextOpacity(0.50)
                    Text2.SetShadowOpacity(0.50)
            48=>
                Text.SetTextOpacity(0.4)
                Text.SetShadowOpacity(0.4)
                if(Text2Opacity > 0.0):
                    Text2.SetTextOpacity(0.4)
                    Text2.SetShadowOpacity(0.4)
            36=>
                Text.SetTextOpacity(0.3)
                Text.SetShadowOpacity(0.3)
                if(Text2Opacity > 0.0):
                    Text2.SetTextOpacity(0.3)
                    Text2.SetShadowOpacity(0.3)
            28=>
                Text.SetTextOpacity(0.20)
                Text.SetShadowOpacity(0.2)
                if(Text2Opacity > 0.0):
                    Text2.SetTextOpacity(0.20)
                    Text2.SetShadowOpacity(0.2)
            16=>
                Text.SetTextOpacity(0.10)
                Text.SetShadowOpacity(0.1)
                if(Text2Opacity > 0.0):
                    Text2.SetTextOpacity(0.10)
                    Text2.SetShadowOpacity(0.1)
            8=>
                Text.SetTextOpacity(0.00)
                Text.SetShadowOpacity(0.0)
                if(Text2Opacity > 0.0):
                    Text2.SetTextOpacity(0.00)
                    Text2.SetShadowOpacity(0.0)
            _=> 

IterateBounceImageEffect(Img : texture_block, DefaultSize : vector2, GatherBounceIterations : float):void =
    CurrentX := Img.GetDesiredSize().X 
    if(CurrentX > DefaultSize.X):
        CurrentY := Img.GetDesiredSize().Y
        ShrinkX := Max((CurrentX - DefaultSize.X) / GatherBounceIterations, 0.25)
        ShrinkY := Max((CurrentY - DefaultSize.Y) / GatherBounceIterations, 0.25)
        Img.SetDesiredSize(vector2{X := Max(CurrentX - ShrinkX, DefaultSize.X), Y:= Max(CurrentY - ShrinkY, DefaultSize.Y)})

FormatTime(InSeconds : int)<transacts>:string =
    if (InSeconds < 60):
        return "{InSeconds}s"
    else if:
        InSeconds < 3600
        Minutes := Int[1.0*InSeconds / 60.0]
        Seconds := Mod[InSeconds, 60]
    then:
        return "{Minutes}m {Seconds}s"
    else if:
        InSeconds < 86400
        Hours := Int[1.0*InSeconds / 3600.0]
        Minutes := Int[(1.0*InSeconds - Hours * 3600.0) / 60.0]
        Seconds := Mod[InSeconds, 60]
    then:
        return "{Hours}h {Minutes}m {Seconds}s"
    else if:
        Days := Int[1.0*InSeconds / 86400.0]
        Hours := Int[(1.0*InSeconds - Days * 86400.0) / 3600.0]
        Minutes := Int[(1.0*InSeconds - Days * 86400.0 - Hours * 3600.0) / 60.0]
        Seconds := Mod[InSeconds, 60]
    then:
        return "{Days}d {Hours}h {Minutes}m {Seconds}s"
    
    return "{InSeconds}s"
    
RandomVecXY():vector3=
    return RandomUnitVecXY() *GetRandomFloat(0.0, 1.0)
    
RandomUnitVecXY():vector3=
    Ang := GetRandomFloat(0.0, PiFloat*2.0)
    return vector3{X:=Cos(Ang), Y:=Sin(Ang), Z:=0.0}

FormatPercent(Value:float, Decimals:int)<transacts>:string=
    return "{Value*100.0}".Slice[0,4+Decimals] + "%" or "100%"

FormatPercent(Value : float)<transacts>:string =
    return "{Floor[Value*100.0]}%" or "100%"

FormatPercent(Value : int)<transacts>:string =
    return "{Value*100}%"

SampleWeights(Weights : []int)<transacts>:int =
    var Total : int = 0
    for (Weight : Weights):
        set Total = Total + Weight
    
    Chance := GetRandomFloat(0.0, Total*1.0)
    
    var Sum : int = 0
    var Index :int = 0
    for (Weight : Weights):
        set Sum = Sum + Weight
        if (Chance < Sum * 1.0):
            return Index
        set Index = Index + 1

    return Weights.Length - 1

#Toggleable logging
Log(Text : string)<transacts>:void =
    block:
        Print(Text)