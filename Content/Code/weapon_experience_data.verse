using { /Verse.org/Simulation }

weapon_experience_data := struct <persistable><concrete><public>:
    @editable
    Name <public>: string = ""

    @editable
    GainAmount <public>: int = 0

    @editable
    CooldownSeconds <public>: int = 0