using { /Fortnite.com/Characters }
using { /Fortnite.com/Devices }
using { /UnrealEngine.com/Temporary/SpatialMath }
using { /Verse.org/Simulation }

zone_manager_device := class(creative_device):
    @editable 
    ZoneTag : fetchable_tag = fetchable_tag.zone1tag
    
    @editable 
    var ZoneMutuator :mutator_zone_device = mutator_zone_device{}
    
    @editable 
    var Barrier :barrier_device = barrier_device{}
    
    @editable 
    var Stash :button_device = button_device{}
    
    @editable 
    var BarrierLimit : int = 0
    
    # @editable 
    # var BonusChest : creative_prop = creative_prop{}
    
    @editable 
    var BonusChestButton : button_device = button_device{}

    @editable
    var LivingGuardMapIndicatorFilter : string = ""

    @editable
    var LivingGuardMapIndicator : ?map_indicator_device = false

    @editable
    var LivingGuardBeacon : ?beacon_device = false

    var LivingGuardMapIndicatorAssignment : string = ""
    
    @editable
    var CollectibleMapIndicator : ?map_indicator_device = false

    @editable
    var CollectibleBeacon : ?beacon_device = false

    @editable
    var OrbsToFloat : []creative_prop = array{}

    @editable
    var FloatingOrbManager : ?floating_orb_manager = false
    
    @editable
    var GuardSummoningButton : ?button_device = false

    @editable 
    GuardManagers : []guard_manager = array{}

    var GuardManagersOld : []guard_manager_device = array{}
    
    var GameManager : game_manager = game_manager{}
    
    var MaxGuards : int = 0
    var GuardsAlive : int = 0
    var PlayersInZone : int = 0
    var MasteryPlayersInZone : int = 0
    var DamageMasteryDamageDealt : big_number = big_number{}
    DamageMasteryCooldown : int = 10
    var DamageMasteryTimer : int = -99
    var CurrentDamageMasteryLevel : int = 0
    var CollectibleDevices : []collectible_device = array{}

    Init(GM : game_manager):void=
        set GameManager = GM
        InitGuards()
        InitCollectibles()
        ZoneMutuator.AgentEntersEvent.Subscribe(HandlePlayerEntersZone)
        ZoneMutuator.AgentExitsEvent.Subscribe(HandlePlayerExitsZone)
        BonusChestButton.InteractedWithEvent.Subscribe(HandlePlayerOpensChest)
        spawn:
            FloatOrbs()

        if:
            SetGuardButton := GuardSummoningButton?
        then:
            SetGuardButton.InteractedWithEvent.Subscribe(SummonGuards)

        if:
            SetFloatingOrbManager := FloatingOrbManager?
        then:
            set SetFloatingOrbManager.ZoneManager = option{Self}
            spawn:
                SetFloatingOrbManager.FloatOrbs()
    
    SummonGuards(Agent:agent):void=
        block:

    FloatOrbs()<suspends>:void=
        loop:
            if:
                PlayersInZone > 0  
                OrbsToFloat.Length > 0
            then:
                for (Orb : OrbsToFloat):
                    T := Orb.GetTransform()
                    P := T.Translation
                    R := T.Rotation
                    NewR := R.ApplyYaw(3.0)
                    spawn:
                        Orb.MoveTo(P - vector3{Z:= 30.0}, NewR, 10.0)
                Sleep(10.0)
                for (Orb : OrbsToFloat):
                    T := Orb.GetTransform()
                    P := T.Translation
                    R := T.Rotation
                    NewR := R.ApplyYaw(3.0)
                    spawn:
                        Orb.MoveTo(P + vector3{Z:= 30.0}, NewR, 10.0)
            Sleep(10.0)

    InitCollectibles():void=
        Collectibles := FindCreativeObjectsWithTag(ZoneTag.ToTagClass())
        for (Obj : Collectibles):
            if:
                Collectible : []collectible_device = array{collectible_device[Obj]}
                set CollectibleDevices = CollectibleDevices + Collectible
        
        spawn:
            OnCollectibleTaken()

    InitGuards():void=
        Guards := FindCreativeObjectsWithTag(ZoneTag.ToTagClass())
        for (Obj : Guards):
            if:
                Guard : []guard_manager_device = array{guard_manager_device[Obj]}
                set GuardManagersOld = GuardManagersOld + Guard

        for (GuardManager : GuardManagersOld):
            GuardManager.Init(GameManager, Self, ZoneTag, ZoneTag.ToZoneName()+"{MaxGuards}")
            set MaxGuards += 1
            set GuardsAlive += 1

        for (GuardManager : GuardManagers):
            GuardManager.Init(GameManager, Self, ZoneTag, ZoneTag.ToZoneName()+"{MaxGuards}")
            if:
                SetLivingGuardMapIndicator := LivingGuardMapIndicator?
                LivingGuardMapIndicatorAssignment = ""
                LivingGuardMapIndicatorFilter = "" or GuardManager.GuardName = LivingGuardMapIndicatorFilter or GuardManager.CustomName? = LivingGuardMapIndicatorFilter
            then:
                spawn:
                    MoveLivingGuardIndicator(GuardManager, ZoneTag.ToZoneName()+"{MaxGuards}")
            set MaxGuards += 1
            set GuardsAlive += 1

    MoveLivingGuardIndicator(GuardManager:guard_manager, GuardID : string)<suspends>:void=
        Sleep(5.25)
        if:
            SetLivingGuardMapIndicator := LivingGuardMapIndicator?
            set LivingGuardMapIndicatorAssignment = GuardID
            SetLivingGuardMapIndicator.TeleportTo[GuardManager.StartPosition + vector3{Z:=120.0}, GuardManager.StartRotation]    
        
        if:
            SetLivingGuardBeacon := LivingGuardBeacon?
            SetLivingGuardBeacon.TeleportTo[GuardManager.StartPosition + vector3{Z:=120.0}, GuardManager.StartRotation]    
        
    
    OnCollectibleTaken()<suspends>:void=
        if:
            SetCollectibleMapIndicator := CollectibleMapIndicator?
        then:
            for (Collectible : CollectibleDevices):
                if:
                    not Collectible.IsRespawning?
                then: 
                    if:
                        SetCollectibleMapIndicator.TeleportTo[Collectible.Prop.GetTransform()]

                    if:
                        SetCollectibleBeacon := CollectibleBeacon?
                        SetCollectibleBeacon.TeleportTo[Collectible.Prop.GetTransform()]    

    OnGuardDestroyed(GuardID : string, RespawnTimer : float)<suspends>:void=
        if: 
            RespawnTimer = 0.0
        then:
            set GuardsAlive -= 1

        #MoveMapIndicator
        if:
            SetLivingGuardMapIndicator := LivingGuardMapIndicator?
            LivingGuardMapIndicatorAssignment = GuardID
        then:
            for (GuardManager : GuardManagers):
                if:
                    GuardManager.CurrentHealth.Value > 0.0
                    LivingGuardMapIndicatorFilter = "" or GuardManager.GuardName = LivingGuardMapIndicatorFilter or GuardManager.CustomName? = LivingGuardMapIndicatorFilter
                then: 
                    set LivingGuardMapIndicatorAssignment = GuardManager.GuardID
                    if:
                        SetLivingGuardMapIndicator.TeleportTo[GuardManager.StartPosition + vector3{Z:=120.0}, GuardManager.StartRotation]
                    if:
                        SetLivingGuardBeacon := LivingGuardBeacon?
                        SetLivingGuardBeacon.TeleportTo[GuardManager.StartPosition + vector3{Z:=120.0}, GuardManager.StartRotation]    
       
    DidPlayerHitGuardRecently(PlayerIn:player, RecentPlayersHittingGuard : [player]big_number)<transacts>:logic=
        if:
            RecentPlayersHittingGuard[PlayerIn].Value > 0.0
        then:
            return true
        return false
    
    
    GetClearBonus(PlayerManager : player_manager)<transacts>:big_number=
        var TotalBonus : big_number = big_number{}
        for (GuardManager : GuardManagers):
            set TotalBonus = TotalBonus + GuardManager.GetKOBonus(PlayerManager)

        set TotalBonus = TotalBonus * PlayerManager.GetClearBonusMultiplier()

    OnZoneCleared(?BossLevel : int = -1)<suspends>:void=
        for (PlayerManager: GameManager.PlayerManagers):
            if:
                Agent:=agent[PlayerManager.Player]
                ZoneMutuator.IsInVolume[Agent]
            then:                
                var ClearBonus : big_number = GetClearBonus(PlayerManager)
                var TempPlayers : int = PlayersInZone

                if:
                    PlayerManager.GetAchievementLevel["Damage Mastery"] > 0
                    MasteryLevel := GetDamageMasteryLevel()
                    MasteryMax := ZoneTag.ToDamageMasteryArray()[MasteryLevel]
                    MasteryMult := ZoneTag.ToDamageMasteryMultipliers()[MasteryLevel]
                then:
                    set ClearBonus = BigNumber_Multiply(ClearBonus, MasteryMult)
                    set TempPlayers = MasteryPlayersInZone

                # Log("ZoneTag: {ZoneTag.ToZoneName()}")
                if (Lv := PlayerManager.GetZoneLevel[ZoneTag]):
                    Log("PlayerLevel: {Lv}")
                # Log("BossLevel: {BossLevel}")
                if:
                    BossLevel = -1
                then: 
                    PlayerManager.OnZoneCleared(ZoneTag, DamageMasteryTimer, ClearBonus, TempPlayers)
                else:
                    #Check if they clear Boss Level
                    if:
                        PlayerManager.GetZoneLevel[ZoneTag] <= BossLevel
                    then:
                        PlayerManager.OnZoneCleared(ZoneTag, DamageMasteryTimer, ClearBonus, TempPlayers)                        
                        PlayerManager.GrantUniqueBossItem(ZoneTag, 1)


    OnZoneDamageDealt(Damage : big_number):void =
        if:
            DamageMasteryTimer <= (0 - DamageMasteryCooldown)
        then:
            spawn:
                StartDamageMasteryTimer()
        
        if:
            DamageMasteryTimer > 0
        then:
            set DamageMasteryDamageDealt = BigNumber_Add(DamageMasteryDamageDealt, Damage)
            if: 
                GetDamageMasteryLevel() > CurrentDamageMasteryLevel 
                MasteryMult := ZoneTag.ToDamageMasteryMultipliers()[GetDamageMasteryLevel()]
            then: 
                set CurrentDamageMasteryLevel = GetDamageMasteryLevel()
                for (PlayerManager: GameManager.PlayerManagers):
                    if:
                        Agent:=agent[PlayerManager.Player]
                        ZoneMutuator.IsInVolume[Agent]
                        PlayerManager.GetAchievementLevel["Damage Mastery"] > 0
                        SetUIManager := PlayerManager.UIManager?
                    then:                
                        SetUIManager.UpdateDamageMasteryText(CurrentDamageMasteryLevel, MasteryMult)
            

    StartDamageMasteryTimer()<suspends>:void=
        set DamageMasteryTimer = 60
        set CurrentDamageMasteryLevel = 0
        for (X:= 0..59):
            for (PlayerManager: GameManager.PlayerManagers):
                if:
                    Agent:=agent[PlayerManager.Player]
                    ZoneMutuator.IsInVolume[Agent]
                    PlayerManager.GetAchievementLevel["Damage Mastery"] > 0
                    SetUIManager := PlayerManager.UIManager?
                then:                
                    SetUIManager.UpdateDamageMasteryTimer(DamageMasteryTimer)
            Sleep(1.0)
            set DamageMasteryTimer -= 1

        set DamageMasteryTimer = 0
        for (PlayerManager: GameManager.PlayerManagers):
            if:
                Agent:=agent[PlayerManager.Player]
                ZoneMutuator.IsInVolume[Agent]
                PlayerManager.GetAchievementLevel["Damage Mastery"] > 0
                SetUIManager := PlayerManager.UIManager?
            then:                
                SetUIManager.HideDamageMasteryTimer()
                SetUIManager.ShowDamageMasteryReport(DamageMasteryDamageDealt, PlayerManager.DamageMasteryEssenceGained, PlayerManager.DamageMasteryEssenceStashed)

                set PlayerManager.DamageMasteryEssenceGained = big_number{}
                set PlayerManager.DamageMasteryEssenceStashed = big_number{}
        
        set DamageMasteryDamageDealt = big_number{}

        for (X:= 0..DamageMasteryCooldown):
            Sleep(1.0)
            set DamageMasteryTimer -= 1

    GetDamageMasteryLevel()<transacts>:int=
        var MasteryLevel : int = 0
        for (MaxHealth : ZoneTag.ToDamageMasteryArray()):
            if:
                BigNumber_GreaterThanOrEquals(DamageMasteryDamageDealt, MaxHealth)?
            then:
                set MasteryLevel += 1
                            
        MasteryLevel

    CheckBarrierLimit(Agent:agent, PlayerManager:player_manager):void=
        return
        # if:
        #     CurrentLevel := PlayerManager.GetLevel[]
        # then:
        #     if(CurrentLevel < BarrierLimit):
        #         Barrier.RemoveFromIgnoreList(Agent)
        #     if(CurrentLevel = BarrierLimit):
        #         Barrier.AddToIgnoreList(Agent)
        #         PlayerManager.UnlockZone(ZoneTag)
        #     if(CurrentLevel > BarrierLimit):
        #         Barrier.AddToIgnoreList(Agent)

    HandlePlayerEntersZone(Agent:agent):void=
        if:
            Player:=player[Agent]
            PlayerManager:=GameManager.PlayerManagers[Player]
        then:
            PlayerManager.OnEnterZone(ZoneTag, GuardsAlive, MaxGuards)   
            set PlayersInZone += 1
            if:
                PlayerManager.GetAchievementLevel["Damage Mastery"] > 0
            then:
                set MasteryPlayersInZone += 1

    HandlePlayerExitsZone(Agent:agent):void=
        if:
            Player:=player[Agent]
            PlayerManager:=GameManager.PlayerManagers[Player]
        then:
            PlayerManager.OnExitZone(ZoneTag)
            PlayerManager.ExitZone(ZoneTag)
            set PlayersInZone -= 1
            if:
                PlayerManager.GetAchievementLevel["Damage Mastery"] > 0
            then:
                set MasteryPlayersInZone -= 1
 
    BringGuardHome(Agent:agent, GuardManager : guard_manager_device)<suspends>:void=
        loop:
            Sleep(0.5)
            if:
                Print("D")
                FortChar := Agent.GetFortCharacter[]
                Print("E")
                FortChar.TeleportTo[GuardManager.GetTransform().Translation, GuardManager.GetTransform().Rotation]
            then:
                return
        
    HandlePlayerOpensChest(Agent:agent):void=
        if:
            Player:=player[Agent]
            PlayerManager:=GameManager.PlayerManagers[Player]
        then:
            PlayerManager.CollectStash()